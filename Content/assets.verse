Materials<public> := module:
    InfoPanels<public> := module{}
    Minimap<public> := module{}
    UI<public> := module{}

Modules<public> := module:
    Components<public> := module{}
    WeaponAbilities<public> := module{}

NPC<public> := module:
    AnimationSequences<public> := module{}
    Materials<public> := module:
        Forest<public> := module{}
        Wasteland<public> := module{}
        Elysium<public> := module{}

Props<public> := module:
    ItemDrops<public> := module{}

UserInterface<public> := module:
    Gears<public> := module{}
    Icons<public> := module{}
    InfoPanel<public> := module{}
    HUD<public> := module{}
    HUDMessage<public> := module{}
    Minimaps<public> := module{}
    Slimes<public> := module{}
    News<public> := module{}
    VerseWidget<public> := module{}