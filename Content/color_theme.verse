using { /Verse.org/Colors }
using { /Verse.org/Simulation }
  
ThemeColors<public> := module:
    Heading<public>:color = MakeColorFromHex("FFFFFF")
    PrimaryText<public>:color = MakeColorFromHex("FFFFFF")
    SecondaryText<public>:color = MakeColorFromHex("AAAAAA")
    Background<public>:color = MakeColorFromHex("003DAD")
    Highlight<public>:color = MakeColorFromHex("FF8332")

    Loud<public>:color = MakeColorFromSRGBValues(247, 255, 25)
    Regular<public>:color = MakeColorFromHex("0055FE")
    RegularLight<public>:color = MakeColorFromHex("007FFF")

    Success<public>:color = MakeColorFromHex("A7FF90")
    SuccessDark<public>:color = MakeColorFromHex("56B849")
    Failure<public>:color = MakeColorFromHex("FF3C47")
    Inactive<public>:color = MakeColorFromHex("001A4F")
    Disabled<public>:color = MakeColorFromHex("303030")
    Locked<public>:color = MakeColorFromHex("454545")
    GameProgressBarInactive<public>:color = MakeColorFromHex("595959")
    
    RarityCommon<public>:color = MakeColorFromHex("898E93")
    RarityUncommon<public>:color = MakeColorFromHex("499401")
    RarityRare<public>:color = MakeColorFromHex("009FE4")
    RarityEpic<public>:color = MakeColorFromHex("B448E6")
    RarityLegendary<public>:color = MakeColorFromHex("EA842D")
    RarityMythic<public>:color = MakeColorFromHex("B89F49")
    RarityExotic<public>:color = MakeColorFromHex("38BDA2")
    RarityDivine<public>:color = MakeColorFromHex("FFFFFF")
    RarityImmortal<public>:color = MakeColorFromHex("BD3838")

    Common<public>:color = color{R := 0.25, G := 0.27, B := 0.292}
    Uncommon<public>:color = color{R := 0.067, G := 0.296, B := 0.0}
    Rare<public>:color = color{R := 0.0, G := 0.347, B := 0.776}
    Epic<public>:color = color{R := 0.456, G := 0.065, B := 0.791}
    Legendary<public>:color = color{R := 0.823, G := 0.231, B := 0.026}

    WeaponDamage<public>:color = MakeColorFromHex("FF3C47") * 2
    WeaponCooldown<public>:color = MakeColorFromHex("3C89FF") * 2
    WeaponSize<public>:color = MakeColorFromHex("ABFF19") * 2
    WeaponLevel<public>:color = MakeColorFromHex("38BDA2") * 2
    WeaponEvolutionA<public>:color = MakeColorFromHex("38BDA2") * 2
    WeaponEvolutionB<public>:color = MakeColorFromHex("B89F49") * 2
    WeaponEvolutionUnique<public>:color = MakeColorFromHex("B448E6") * 2

    Weapon<public>:color = color{R := 0.0, G := 0.092, B := 1.0}
    WeaponEvolved<public>:color = color{R := 0.25, G := 0.0, B := 1.0}
    WeaponAwakened<public>:color = color{R := 1.0, G := 0.045, B := 0.063}
    CriticalHit<public>:color = color{R := 0.0, G := 0.667, B := 1.0}

    Weapon1<public>:color = color{R := 0.0, G := 0.092, B := 1.0}
    Weapon2<public>:color = color{R := 0.054, G := 0.0, B := 1.0}
    Weapon3<public>:color = color{R := 0.137, G := 0.0, B := 1.0}
    Weapon4<public>:color = color{R := 0.25, G := 0.0, B := 1.0}

    PlayerColor1<public>:color = MakeColorFromHex("FF3F3F")
    PlayerColor2<public>:color = MakeColorFromHex("40A6FF")
    PlayerColor3<public>:color = MakeColorFromHex("FFE03F")
    PlayerColor4<public>:color = MakeColorFromHex("8DFF3F")

GetWeaponColor(Index:int):color =
    case (Index):
        0 => ThemeColors.Weapon1
        1 => ThemeColors.Weapon2
        2 => ThemeColors.Weapon3
        3 => ThemeColors.Weapon4
        _ => ThemeColors.Weapon1