using { /Fortnite.com/UI }
using { /UnrealEngine.com/Temporary/UI }

<# ----------------------------------------------------------------------------------------------------
    pagination_ui
---------------------------------------------------------------------------------------------------- #>
pagination_ui := class(ui_component):
    TotalEntries:int
    EntriesPerPage:int
    
    PageChangedEvent:events(tuple()) = events(tuple()){}

    CurrentPageText:text_block = text_block:
        DefaultTextColor := ThemeColors.PrimaryText
        DefaultTextOpacity := 0.5

    var CurrentPage:int = 1

    <# ----------------------------------------------------------------------------------------------------
        Initialization
    ---------------------------------------------------------------------------------------------------- #>
    CreateUI<override>():widget =
        PreviousButton := button_quiet{DefaultText := MakeMessage("◀")}
        PreviousButton.OnClick().Subscribe(OnClickPrevious)

        NextButton := button_quiet{DefaultText := MakeMessage("▶")}
        NextButton.OnClick().Subscribe(OnClickNext)

        CurrentPageText.SetText(MakeMessage("1/{GetMaxPage()}"))

        stack_box:
            Orientation := orientation.Horizontal
            Slots := array:
                stack_box_slot:
                    HorizontalAlignment := horizontal_alignment.Fill
                    VerticalAlignment := vertical_alignment.Fill
                    Distribution := option{1.0}
                    Widget := PreviousButton
                stack_box_slot:
                    HorizontalAlignment := horizontal_alignment.Center
                    Padding := margin{Left := 20.0, Top := 0.0, Right := 20.0, Bottom := 0.0}
                    Distribution := option{0.5}
                    Widget := CurrentPageText
                stack_box_slot:
                    HorizontalAlignment := horizontal_alignment.Fill
                    VerticalAlignment := vertical_alignment.Fill
                    Distribution := option{1.0}
                    Widget := NextButton

    <# ----------------------------------------------------------------------------------------------------
        Get Functions
    ---------------------------------------------------------------------------------------------------- #>
    GetCurrentPage()<transacts>:int =
        CurrentPage

    GetMaxPage()<transacts>:int =
        Ceil[(TotalEntries * 1.0) / (EntriesPerPage * 1.0)] or 1

    <# ----------------------------------------------------------------------------------------------------
        Update Functions
    ---------------------------------------------------------------------------------------------------- #>
    Update<override>():void =
        CurrentPageText.SetText(MakeMessage("{CurrentPage}/{GetMaxPage()}"))
        PageChangedEvent.Signal()

    SetPage(Page:int):void =
        set CurrentPage = Clamp(Page, 1, GetMaxPage())
        Update()

    <# ----------------------------------------------------------------------------------------------------
        Event Handlers
    ---------------------------------------------------------------------------------------------------- #>
    OnClickPrevious(Message:widget_message):void =
        set CurrentPage = (CurrentPage = 1) and GetMaxPage() or CurrentPage - 1
        Update()
    
    OnClickNext(Message:widget_message):void =
        set CurrentPage = (CurrentPage = GetMaxPage()) and 1 or CurrentPage + 1
        Update()
        