using { /Fortnite.com/UI }
using { /Verse.org/Assets }
using { /Verse.org/Colors }
using { /Verse.org/Simulation }
using { /UnrealEngine.com/Temporary/SpatialMath }
using { /UnrealEngine.com/Temporary/UI }

<#>
    Operator utilities

operator'*'<public>(Margin:margin, Multiply:float):margin =
    margin{Left := Margin.Left * Multiply, Top := Margin.Top * Multiply, Bottom := Margin.Bottom * Multiply, Right := Margin.Right * Multiply}

<#>
    hud_ui

hud_ui := class:
    Player:player
    IsStartVisible:logic = true

    var MaybeUI:?widget = false

    Initialize():void =
        UI := CreateUI()
        set MaybeUI = option{UI}

        if (IsStartVisible?, PlayerUI := GetPlayerUI[Player]):
            PlayerUI.AddWidget(UI, player_ui_slot{InputMode := ui_input_mode.None})

    Show():void =
        if (UI := MaybeUI?):
            UI.SetVisibility(widget_visibility.Visible)

    Hide():void =
        if (UI := MaybeUI?):
            UI.SetVisibility(widget_visibility.Collapsed)

    CreateUI():widget =
        canvas{}

<#>
    ui_component

ui_component := class:
    Player:player

    var UI:widget = canvas{}

    Initialize():void =
        set UI = CreateUI()

    CreateUI():widget =
        canvas{}

    GetUI()<transacts>:widget =
        UI

    Show():void =
        UI.SetVisibility(widget_visibility.Visible)
        OnShow()

    Hide():void =
        UI.SetVisibility(widget_visibility.Collapsed)
        OnHide()

    Update():void = {}
    
    OnShow():void = {}
    OnHide():void = {}

<#>
    Make box UI with a border.

MakeBox(
    ?Width:float = 0.0, ?Height:float = 0.0,
    ?BorderRadius:float = 1.0, ?Color:color = NamedColors.Black, ?Opacity:type{_X:float where 0.000000 <= _X, _X <= 1.000000} = 0.1,
    ?BorderColor:color = NamedColors.White, ?BorderOpacity:type{_X:float where 0.000000 <= _X, _X <= 1.000000} = 0.1
):overlay =
    overlay:
        Slots := array:
            overlay_slot:
                HorizontalAlignment := horizontal_alignment.Fill
                VerticalAlignment := vertical_alignment.Fill
                Padding := margin{Left := BorderRadius, Top := BorderRadius, Right := BorderRadius, Bottom := BorderRadius}
                Widget := color_block:
                    DefaultColor := Color
                    DefaultOpacity := Opacity
                    DefaultDesiredSize := vector2{X := Width, Y := Height}
            overlay_slot:
                HorizontalAlignment := horizontal_alignment.Fill
                VerticalAlignment := vertical_alignment.Top
                Padding := margin{Left := BorderRadius, Top := 0.0, Right := BorderRadius, Bottom := 0.0}
                Widget := color_block:
                    DefaultColor := BorderColor
                    DefaultOpacity := BorderOpacity
                    DefaultDesiredSize := vector2{X := 0.0, Y := BorderRadius}
            overlay_slot:
                HorizontalAlignment := horizontal_alignment.Fill
                VerticalAlignment := vertical_alignment.Bottom
                Padding := margin{Left := BorderRadius, Top := 0.0, Right := BorderRadius, Bottom := 0.0}
                Widget := color_block:
                    DefaultColor := BorderColor
                    DefaultOpacity := BorderOpacity
                    DefaultDesiredSize := vector2{X := 0.0, Y := BorderRadius}
            overlay_slot:
                HorizontalAlignment := horizontal_alignment.Left
                VerticalAlignment := vertical_alignment.Fill
                Widget := color_block:
                    DefaultColor := BorderColor
                    DefaultOpacity := BorderOpacity
                    DefaultDesiredSize := vector2{X := BorderRadius, Y := 0.0}
            overlay_slot:
                HorizontalAlignment := horizontal_alignment.Right
                VerticalAlignment := vertical_alignment.Fill
                Widget := color_block:
                    DefaultColor := BorderColor
                    DefaultOpacity := BorderOpacity
                    DefaultDesiredSize := vector2{X := BorderRadius, Y := 0.0}

<#>
    Utility function for creating stat entry UIs (label: value).

CreateStatUI(Text:text_block, Label:string, ?Icon:?texture = false, ?Padding:float = 100.0):stack_box =
    StatUI := stack_box:
        Orientation := orientation.Horizontal

    if (Image := Icon?):
        Slot := stack_box_slot:
            HorizontalAlignment := horizontal_alignment.Center
            VerticalAlignment := vertical_alignment.Center
            Padding := margin{Left := 0.0, Top := 0.0, Right := 10.0, Bottom := 0.0}
            Widget := texture_block:
                DefaultImage := Image
                DefaultDesiredSize := vector2{X := 40.0, Y := 40.0}

        StatUI.AddWidget(Slot)

    Slots := array:
        stack_box_slot:
            HorizontalAlignment := horizontal_alignment.Left
            Distribution := option{1.0}
            Widget := text_block:
                DefaultText := MakeMessage("{Label}:")
                DefaultTextColor := ThemeColors.PrimaryText
        stack_box_slot:
            HorizontalAlignment := horizontal_alignment.Right
            Widget := Text

    for (Slot : Slots). StatUI.AddWidget(Slot)
    
    StatUI

CreateStatUI(TextA:text_block, TextB:text_block, Label:string, ?Icon:?texture = false, ?Padding:float = 100.0):stack_box =
    StatUI := stack_box:
        Orientation := orientation.Horizontal

    if (Image := Icon?):
        Slot := stack_box_slot:
            HorizontalAlignment := horizontal_alignment.Center
            VerticalAlignment := vertical_alignment.Center
            Padding := margin{Left := 0.0, Top := 0.0, Right := 10.0, Bottom := 0.0}
            Widget := texture_block:
                DefaultImage := Image
                DefaultDesiredSize := vector2{X := 40.0, Y := 40.0}

        StatUI.AddWidget(Slot)

    Slots := array:
        stack_box_slot:
            HorizontalAlignment := horizontal_alignment.Left
            Distribution := option{2.0}
            Widget := text_block:
                DefaultText := MakeMessage("{Label}:")
                DefaultTextColor := ThemeColors.PrimaryText
        stack_box_slot:
            HorizontalAlignment := horizontal_alignment.Right
            Distribution := option{1.0}
            Widget := TextA
        stack_box_slot:
            HorizontalAlignment := horizontal_alignment.Right
            Distribution := option{1.0}
            Widget := TextB

    for (Slot : Slots). StatUI.AddWidget(Slot)
    
    StatUI

<#>
    Set text color.

(TextBlock:text_block).SetColor(Color:color):void =
    TextBlock.SetTextColor(Color)
    TextBlock.SetShadowColor(Color * 0.125)
    TextBlock.SetShadowOpacity(1.0)
    TextBlock.SetShadowOffset(option{vector2{X := 2.0, Y := 2.0}})

<#>
    UMG widgets for tier texts (common, uncommon, rare etc...).

GetTierWidget(Tier:int)<transacts>:widget =
    TierWidget := case (Tier):
        1 => UserInterface.VerseWidget.WBP_TierText_Common{}
        2 => UserInterface.VerseWidget.WBP_TierText_Uncommon{}
        3 => UserInterface.VerseWidget.WBP_TierText_Rare{}
        4 => UserInterface.VerseWidget.WBP_TierText_Epic{}
        5 => UserInterface.VerseWidget.WBP_TierText_Legendary{}
        _ => UserInterface.VerseWidget.WBP_TierText_Common{}

GetTierWidgetSmall(Tier:int)<transacts>:widget =
    TierWidget := case (Tier):
        1 => UserInterface.VerseWidget.WBP_TierText_Common_Small{}
        2 => UserInterface.VerseWidget.WBP_TierText_Uncommon_Small{}
        3 => UserInterface.VerseWidget.WBP_TierText_Rare_Small{}
        4 => UserInterface.VerseWidget.WBP_TierText_Epic_Small{}
        5 => UserInterface.VerseWidget.WBP_TierText_Legendary_Small{}
        _ => UserInterface.VerseWidget.WBP_TierText_Common_Small{}
