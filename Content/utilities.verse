using { /Fortnite.com/Characters }
using { /Fortnite.com/Devices }
using { /Verse.org/Colors }
using { /Verse.org/Random }
using { /Verse.org/SceneGraph }
using { /Verse.org/Simulation }
using { /UnrealEngine.com/Temporary/Diagnostics }
using { /UnrealEngine.com/Temporary/SpatialMath }
using { /UnrealEngine.com/Temporary/UI }

<#>
    Useful constants.

OneTick:float = 1.0 / 30.0
TwoTicks:float = OneTick * 2
ThreeTicks:float = OneTick * 3
FourTicks:float = OneTick * 4
FiveTicks:float = OneTick * 5

CharacterHeightOffset:float = 77.0
PlayerCharacterSize:float = 25.0
UnitVector:vector3 = vector3{X := 1.0, Y := 1.0, Z := 1.0}
TimeDilationFactor:float = 10000.0
MapIndicatorDisabledPosition:vector3 = vector3{X := 10000.0}
MapIndicatorEnabledPosition:vector3 = vector3{Z := 100.0}
MinScale:float = 0.000001

<#>
    Make a message from a string.

MakeMessage<localizes>(Message:string):message = "{Message}"
MakeMessage<localizes>(Agent:agent):message = "{Agent}"
operator'+'<public><localizes>(MessageA:message, MessageB:message)<computes>:message = "{MessageA}{MessageB}"

<#>
    Dispose of a spawned prop after the specified duration.

DisposeProp(MaybeProp:?creative_prop, Duration:float)<suspends>:void =
    if (Prop := MaybeProp?):
        Sleep(Duration)
        Prop.Dispose()

<#>
    Dispose of an entity after the specified duration.

RemoveEntity(Entity:entity, ?Duration:float = 5.0)<suspends>:void =
    Sleep(Duration)
    Entity.RemoveFromParent()

<#>
    Toggle widget visibility Hidden/Visible to force widget to be displayed immediately.

(Widget:widget).ToggleVisibility():void =
    WidgetVisibility := Widget.GetVisibility()
    Widget.SetVisibility(WidgetVisibility = widget_visibility.Visible and widget_visibility.Hidden or widget_visibility.Visible)

<#>
    Returns the average value of the values of the array.

Average(Numbers:[]float)<transacts>:float =
    var TotalValue:float = 0.0
    for (Number : Numbers):
        set TotalValue += Number
    TotalValue / (Numbers.Length * 1.0)

<#>
    A parametric remove key from map function.

RemoveKeyFromMap(Map:[t]u, KeyToRemove:t where t:subtype(comparable), u:type)<transacts>:[t]u =
    var NewMap:[t]u = map{}
    for (Key -> Value : Map, Key <> KeyToRemove):
        set NewMap = ConcatenateMaps(NewMap, map{Key => Value})
    NewMap

RemoveKeysFromMap(Map:[t]u, KeysToRemove:[]t where t:subtype(comparable), u:type)<transacts>:[t]u =
    var NewMap:[t]u = map{}
    for (Key -> Value : Map, not KeysToRemove.Find[Key]):
        set NewMap = ConcatenateMaps(NewMap, map{Key => Value})
    NewMap

<#>
    Global log function.

global_log := class(log_channel) {}

Logger()<transacts>:log =
    log{Channel := global_log}

<#>
    Utility function to get a random position on a circle.

GetRandomPositionOnCircle(Radius:float, ?Origin:vector3 = vector3{})<transacts>:vector3 =
    Angle := GetRandomFloat(0.0, 1.0) * PiFloat * 2.0;
    vector3{X := Origin.X + Cos(Angle) * Radius, Y := Origin.Y + Sin(Angle) * Radius, Z := Origin.Z}

GetRandomPositionInCircle(Radius:float, ?Origin:vector3 = vector3{})<transacts>:vector3 =
    Angle := GetRandomFloat(0.0, 1.0) * PiFloat * 2.0;
    RandomRadius := GetRandomFloat(0.0, Radius)
    vector3{X := Origin.X + Cos(Angle) * RandomRadius, Y := Origin.Y + Sin(Angle) * RandomRadius, Z := Origin.Z}

GetRandomPositionInDonut(InnerRadius:float, OuterRadius:float, ?Origin:vector3 = vector3{})<transacts>:vector3 =
    Angle := GetRandomFloat(0.0, 1.0) * PiFloat * 2.0;
    RandomRadius := GetRandomFloat(InnerRadius, OuterRadius)
    vector3{X := Origin.X + Cos(Angle) * RandomRadius, Y := Origin.Y + Sin(Angle) * RandomRadius, Z := Origin.Z}

<#>
    Utility class to calculate DeltaTime/ElapsedTime for animations.

delta_timer := class:
    var StartTime:float = 0.0
    var PreviousTime:float = 0.0
    
    block:
        if (StartTime = 0.0):
            set StartTime = GetSimulationElapsedTime()

        if (PreviousTime = 0.0):
            set PreviousTime = GetSimulationElapsedTime()

    GetStartTime()<transacts>:float =
        StartTime

    GetDeltaTime()<transacts>:float =
        CurrentTime := GetSimulationElapsedTime()
        DeltaTime := CurrentTime - PreviousTime
        set PreviousTime = CurrentTime
        DeltaTime

    GetElapsedTime()<transacts>:float =
        GetSimulationElapsedTime() - StartTime

    Reset():void =
        set StartTime = GetSimulationElapsedTime()
        set PreviousTime = GetSimulationElapsedTime()

<#>
    Sleep function that scales with Time Dilation (only use when the game is paused).

SleepTimeDilation(Duration:float)<suspends>:void =
    StartTime := GetSecondsSinceEpoch()
    loop:
        Sleep(0.0)
        if (GetSecondsSinceEpoch() - StartTime >= Duration):
            break

<#>
    Function to calculate weighted random, returns a random index from array of weights.

GetWeightedRandom(Weights:[]int)<transacts>:int =
    var RandomIndex:int = -1
    var CumulativeWeights:[]float = array{}
    var MaxCumulativeWeight:float = 0.0

    for (Index -> Weight : Weights):
        PreviousCumulativeWeight := CumulativeWeights[Index - 1] or 0.0
        set CumulativeWeights = CumulativeWeights + array{Weight * 1.0 + PreviousCumulativeWeight}
        set MaxCumulativeWeight += Weight * 1.0

    RandomNumber := MaxCumulativeWeight * GetRandomFloat(0.0, 1.0)
    for:
        Index -> Weight : Weights
        CumulativeWeight := CumulativeWeights[Index]
        CumulativeWeight >= RandomNumber
        RandomIndex = -1
    do:
        set RandomIndex = Index

    RandomIndex

<#>
    Queue Implementation for handling data in a First In First Out (FIFO) order.

queue<public>(t:type) := class:
    Elements<internal>:[]t = array{}

    Enqueue<public>(NewElement:t)<transacts>:queue(t) =
        queue(t){Elements := Elements + array{NewElement}}

    EnqueueFront<public>(NewElement:t)<transacts>:queue(t) =
        queue(t){Elements := array{NewElement} + Elements}
    
    Dequeue<public>()<decides><transacts>:tuple(queue(t), t) =
        FirstElement := Front[]
        (queue(t){Elements := Elements.RemoveElement[0]}, FirstElement)

    Size<public>()<transacts>:int =
        Elements.Length
    
    IsEmpty<public>()<transacts>:logic =
        logic{Size() = 0}
    
    Front<public>()<decides><transacts>:t =
        Elements[0]
    
    Rear<public>()<decides><transacts>:t =
        Elements[Elements.Length - 1]
        
CreateQueue<public><constructor>(InitialElements:[]t where t:type) := queue(t):
    Elements := InitialElements

<#>
    Grant percentage health/shield to player.

(Player:player).HealPercent(Percent:type{_X:float where 0.000000 <= _X, _X <= 1.000000}):float =
    var HealAmount:float = 0.0
    if (Character := Player.GetFortCharacter[]):
        set HealAmount = Character.GetMaxHealth() * Percent
        Character.Heal(HealAmount)

    HealAmount

(Player:player).ShieldPercent(Percent:type{_X:float where 0.000000 <= _X, _X <= 1.000000}):void =
    if (Character := Player.GetFortCharacter[]):
        Character.SetShield(Character.GetShield() + Character.GetMaxShield() * Percent)

(Player:player).IsMaxHealth()<transacts>:logic =
    logic{Player.GetFortCharacter[].GetHealth() = Player.GetFortCharacter[].GetMaxHealth()}

(Player:player).IsMaxShield()<transacts>:logic =
    logic{Player.GetFortCharacter[].GetShield() = Player.GetFortCharacter[].GetMaxShield()}

<#>
    Returns vector3 with the z-axis zeroed out.

(Vector:vector3).Make2D()<transacts>:vector3 =
    vector3{X := Vector.X, Y := Vector.Y, Z := 0.0}

<#>
    Random functions.

GetRandomSign()<transacts>:float =
    if (GetRandomFloat(0.0, 1.0) <= 0.5) then 1.0 else -1.0

GetRandomSign(Factor:float)<transacts>:float =
    Factor * GetRandomSign()

<#>
    Check if position is within the area of a rectangle.

       A __________
        |          |
        |__________|
       B            C
    

rectangle := struct:
    A:vector3
    B:vector3
    C:vector3

MakeRectangle(Position:vector3, Rotation:rotation, Width:float, Length:float)<transacts>:rectangle =
    ForwardDirection := Rotation.GetLocalForward()
    RightDirection := Rotation.GetLocalRight()
    
    A := Position - RightDirection * Width / 2.0
    B := Position + RightDirection * Width / 2.0
    C := B + ForwardDirection * Length
    
    rectangle:
        A := A
        B := B
        C := C

IsInRectangle(Position:vector3, Rectangle:rectangle)<transacts>:logic =
    AB := Rectangle.B - Rectangle.A
    BC := Rectangle.C - Rectangle.B
    DotABAM := DotProduct(AB, Position - Rectangle.A)
    DotBCBM := DotProduct(BC, Position - Rectangle.B)

    logic:
        0.0 <= DotABAM and DotABAM <= DotProduct(AB, AB) and
        0.0 <= DotBCBM and DotBCBM <= DotProduct(BC, BC)

IsInRectangles(Position:vector3, Rectangles:[]rectangle)<transacts>:logic =
    var IsInRectagle:logic = false
    for (Rectangle : Rectangles, not IsInRectagle?, IsInRectangle(Position, Rectangle)?):
        set IsInRectagle = true
    IsInRectagle

<#>
    Function to get the nearest target agents to the specified agent.

GetNearestPosition(Position:vector3, TargetPositions:[]vector3)<transacts>:?int =
    var MaybeNearestIndex:?int = false
    var NearestDistenceSquared:float = 0.0
    
    for:
        Index -> TargetPosition : TargetPositions
        DistanceSquaredFromTarget := DistanceSquaredXY(Position, TargetPosition)
        not MaybeNearestIndex? or DistanceSquaredFromTarget < NearestDistenceSquared
    do:
        set MaybeNearestIndex = option{Index}
        set NearestDistenceSquared = DistanceSquaredFromTarget

    MaybeNearestIndex

GetNearestAgent(Agent:agent, Targets:[]agent)<transacts>:?agent =
    var MaybeNearestTarget:?agent = false
    var NearestDistenceSquared:float = 0.0
    
    for:
        Target : Targets
        DistanceSquaredFromTarget := DistanceSquaredXY(Agent.GetFortCharacter[].GetTransform().Translation, Target.GetFortCharacter[].GetTransform().Translation)
        not MaybeNearestTarget? or DistanceSquaredFromTarget < NearestDistenceSquared
    do:
        set MaybeNearestTarget = option{Target}
        set NearestDistenceSquared = DistanceSquaredFromTarget

    MaybeNearestTarget

<#>
    Calculate the sum of the array of numbers.

Sum(Numbers:[]int)<transacts>:int =
    var Total:int = 0
    for (Number : Numbers):
        set Total += Number
    Total

SumFloat(Numbers:[]float)<transacts>:float =
    var Total:float = 0.0
    for (Number : Numbers):
        set Total += Number
    Total

<#>
    Mod for floating point numbers.

FMod(X:float, Y:float)<decides><transacts>:float =
    if (X.IsFinite[] and not Y.IsFinite[] and Y <> NaN):
        X
    else:
        X - Int[X / Y] * Y

<#>
    Makes rotation that points from an initial position to a target position.

MakeRotation(From:vector3, To:vector3)<transacts>:rotation =
    Direction := (To - From).MakeUnitVector[] or vector3{}
    MakeRotation(vector3{X := 0.0, Y := 0.0, Z := 1.0}, ArcTan(Direction.Y, Direction.X))

<#>
    Less functions used for sorting.

LessPlayerFloat(PlayerFloatA:tuple(player, float), PlayerFloatB:tuple(player, float))<computes><decides>:void =
    PlayerFloatA(1) < PlayerFloatB(1)

MorePlayerFloat(PlayerFloatA:tuple(player, float), PlayerFloatB:tuple(player, float))<computes><decides>:void =
    PlayerFloatA(1) > PlayerFloatB(1)

LessFloat(FloatA:tuple(t, float), FloatB:tuple(t, float) where t:type)<computes><decides>:void =
    FloatA(1) < FloatB(1)

MoreFloat(FloatA:tuple(t, float), FloatB:tuple(t, float) where t:type)<computes><decides>:void =
    FloatA(1) > FloatB(1)

<#>
    ArrayRace will call the Function argument for each element in the Array and wait for only one of the async functions to complete, canceling all the rest.
    A divide-and-conquer concurrency algorithm that divides an array into two and races between the two subarrays.
    This is a recursive implementation, where the function calls itself to race between the divided arrays.
    The base case (the condition to stop the recursion) is the array has only one element.
    This is a generic implementation using parametric types, so you can provide your own type and your own async function as arguments.

ArrayRace(Array:[]t, Function(Element:t)<suspends>:u where t:type, u:type)<suspends>:u=
    # If more than two elements in the array, then split the array in half,
    # and race between the subarrays recursively.
    if:
        Array.Length > 1
        Mid := Floor(Array.Length / 2)
        LeftArray := Array.Slice[0, Mid]
        RightArray := Array.Slice[Mid]
    then:
        race:
            ArrayRace(LeftArray, Function)
            ArrayRace(RightArray, Function)
    else if:
        First := Array[0]
    then:
        # Grab the only element in the array and call the async function with it as the argument.
        Function(First)
    else:
        # A race with no subtasks never completes, so sleep infinitely and call Err() because we shouldn't return a value here.
        Sleep(Inf)
        Err()
 
<#>
    ArraySync will call the Function argument for each element in the Array and wait for all the async functions to complete.
    A divide-and-conquer concurrency algorithm that divides an array into two and syncs between the two subarrays.
    This is a recursive implementation, where the function calls itself to sync between the divided arrays.
    The base case (the condition to stop the recursion) is the array has only one element.
    This is a generic implementation using parametric types, so you can provide your own type and your own async function as arguments.

ArraySync(Array:[]t, Function(Element:t)<suspends>:u where t:type, u:type)<suspends>:[]u=
    # If more than two elements in the array, then split the array in half,
    # and sync between the subarrays recursively.
    if:
        Array.Length > 1
        Mid := Floor(Array.Length / 2)
        LeftArray := Array.Slice[0, Mid]
        RightArray := Array.Slice[Mid]
    then:
        SyncResult := sync:
            ArraySync(LeftArray, Function)
            ArraySync(RightArray, Function)
 
        # Return the results of the sync expressions as an array.
        SyncResult(0) + SyncResult(1)
    else if:
        First := Array[0]
    then:
        # Grab the only element in the array and call the async function with it as the argument and add the result to an array.
        array{Function(First)}
    else:
        # No element in the given Array so return an empty array.
        array{}

<#>
    Filter out volume_device GetAgentsInVolume() to exclude NPC agents.

(VolumeDevice:volume_device).GetPlayersInVolume()<reads>:[]player =
    for (AgentInVolume : VolumeDevice.GetAgentsInVolume(), PlayerInVolume := player[AgentInVolume])
        {PlayerInVolume}

(VolumeDevice:volume_device).IsEmpty()<decides><transacts>:void =
    not VolumeDevice.GetPlayersInVolume()[0]

<#>
    Class for object pooling.

poolable<public> := class:
    var IsAvailable<public>:logic = true

    Activate<public>():void =
        set IsAvailable = false

    Deactivate<public>():void =
        set IsAvailable = true

    ScheduleDeactivate<public>()<suspends>:void = {}

<#>
    Class for object pooling.

OverlapAABB(A:vector3, B:vector3, Size:float)<decides><transacts>:void =
    A.X <= B.X + Size and A.X >= B.X - Size
    A.Y <= B.Y + Size and A.Y >= B.Y - Size

<#>
    Remaps the specified value from one range to another range of values.

RemapRange(Value:float, MinA:float, MaxA:float, MinB:float, MaxB:float)<computes>:float =
    Clamp(MinB + (Value - MinA) * (MaxB - MinB) / (MaxA - MinA), MinB, MaxB)
