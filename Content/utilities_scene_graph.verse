using { /Verse.org/SceneGraph }
using { /Verse.org/Simulation }

(Entity:entity).Initialize():void =
    if (SimulationEntity := Modules.GetSimulationEntity[]):
        SimulationEntity.AddEntities(array{Entity})

(Entity:entity).InitializeParticle()<suspends>:void =
    Entity.Initialize()

    # [Hotfix] Fix for VFX not showing the first time it is played.
    if (ParticleSystem := Entity.GetComponent[particle_system_component]):
        ParticleSystem.Initialize()

(ParticleSystem:particle_system_component).Initialize()<suspends>:void =
    ParticleSystem.Play()
    Sleep(0.0)
    ParticleSystem.Stop()
    Sleep(0.0)