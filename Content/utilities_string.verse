using { /Verse.org/Colors }

<#>
    Constants

DefaultTextSize:float = 32.0

<#>
    Wrap text function.

CharSizeMapping:[char]int = map:
    ' ' => 8, '!' => 11, '"' => 13, '#' => 20, '$' => 20, '%' => 27, '&' => 24, ''' => 6, '(' => 11, ')' => 11, '*' => 13, '+' => 16, ',' => 9, '-' => 12,
    '.' => 9, '/' => 15, '0' => 21, '1' => 14, '2' => 19, '3' => 20, '4' => 20, '5' => 20, '6' => 21, '7' => 18, '8' => 21, '9' => 20, ':' => 9, ';' => 10,
    '<' => 16, '=' => 15, '>' => 16, '?' => 21, '@' => 34, 'A' => 22, 'B' => 20, 'C' => 22, 'D' => 21, 'E' => 18, 'F' => 17, 'G' => 22, 'H' => 20, 'I' => 11,
    'J' => 19, 'K' => 21, 'L' => 17, 'M' => 30, 'N' => 21, 'O' => 24, 'P' => 19, 'Q' => 23, 'R' => 20, 'S' => 20, 'T' => 19, 'U' => 21, 'V' => 21, 'W' => 29,
    'X' => 21, 'Y' => 22, 'Z' => 18, '[' => 12, '\' => 15, ']' => 12, '^' => 17, '_' => 16, '`' => 18, 'a' => 18, 'b' => 19, 'c' => 18, 'd' => 19, 'e' => 18,
    'f' => 14, 'g' => 18, 'h' => 18, 'i' => 9, 'j' => 9, 'k' => 18, 'l' => 9, 'm' => 28, 'n' => 18, 'o' => 19, 'p' => 20, 'q' => 19, 'r' => 15, 's' => 17,
    't' => 14, 'u' => 18, 'v' => 18, 'w' => 26, 'x' => 19, 'y' => 18, 'z' => 15, '{' => 15, '|' => 10, '}' => 15, '~' => 17

(Char:char).GetSize()<transacts>:int =
    CharSizeMapping[Char] or 0

GetStringSize(String:string)<transacts>:float =
    var TotalStringSize:float = 0.0
    for (CharIndex -> Char : String):
        CharSize := Char.GetSize()
        set TotalStringSize += CharSize * 1.0
    TotalStringSize

    
(Text:string).WrapText(AvailableWidth:float, ?TextSize:float = DefaultTextSize):string =
    FinalAvailableWidth := AvailableWidth * DefaultTextSize / TextSize

    var WrappedText:string = ""
    var CurrentTextSize:int = 0
    var TextSizeFromLastSpaceChar:int = 0
    var LastSpaceCharIndex:int = 0

    for (CharIndex -> Char : Text):
        CharSize := Char.GetSize()
        set CurrentTextSize += CharSize
        set TextSizeFromLastSpaceChar += CharSize

        if (CurrentTextSize * 1.0 >= FinalAvailableWidth):
            if (set WrappedText[LastSpaceCharIndex] = '\n') {}
            set CurrentTextSize = TextSizeFromLastSpaceChar
            
        if (Char = ' '):
            set LastSpaceCharIndex = CharIndex
            set TextSizeFromLastSpaceChar = 0

        set WrappedText += ToString(Char)

    WrappedText

<#>
    Convert string to & from lowercase & uppercase character.

(Char:char).ToLower<public>()<computes>:char =
    case(Char):
        'A' => 'a', 'B' => 'b', 'C' => 'c', 'D' => 'd',
        'E' => 'e', 'F' => 'f', 'G' => 'g', 'H' => 'h',
        'I' => 'i', 'J' => 'j', 'K' => 'k', 'L' => 'l',
        'M' => 'm', 'N' => 'n', 'O' => 'o', 'P' => 'p',
        'Q' => 'q', 'R' => 'r', 'S' => 's', 'T' => 't',
        'U' => 'u', 'V' => 'v', 'W' => 'w', 'X' => 'x',
        'Y' => 'y', 'Z' => 'z', _ => Char

(Char:char).ToUpper<public>()<computes>:char =
    case(Char):
        'a' => 'A', 'b' => 'B', 'c' => 'C', 'd' => 'D',
        'e' => 'E', 'f' => 'F', 'g' => 'G', 'h' => 'H',
        'i' => 'I', 'j' => 'J', 'k' => 'K', 'l' => 'L',
        'm' => 'M', 'n' => 'N', 'o' => 'O', 'p' => 'P',
        'q' => 'Q', 'r' => 'R', 's' => 'S', 't' => 'T',
        'u' => 'U', 'v' => 'V', 'w' => 'W', 'x' => 'X',
        'y' => 'Y', 'z' => 'Z', _ => Char

(String:string).ToLower<public>()<computes>:string =
    for (Char : String):
        Char.ToLower()

(String:string).ToUpper<public>()<computes>:string =
    for (Char : String):
        Char.ToUpper()

<#>
    Utility function for replacing identifier in strings e.g. "Hi [PlayerName].".Replace("PlayerName", "John") => "Hi John."

(InputString:string).Replace(IdentifierToReplace:string, ReplacementString:string):string =
    var FinalString:string = InputString
    loop:
        if:
            StartIndex := FinalString.Find['[']
            StopIndex := FinalString.Find[']']
            Identifier := FinalString.Slice[StartIndex + 1, StopIndex]
            Identifier = IdentifierToReplace
            BeforeString := FinalString.Slice[0, StartIndex]
            AfterString := FinalString.Slice[StopIndex + 1, FinalString.Length]
        then:
            set FinalString = BeforeString + ReplacementString + AfterString
        else:
            break

    FinalString

<#>
    Utility function to get tier string from integer values (1 = Common .. 9 = Immortal).

(Tier:int).ToTierString()<transacts>:string =
    case (Tier):
        1 => "Common"
        2 => "Uncommon"
        3 => "Rare"
        4 => "Epic"
        5 => "Legendary"
        6 => "Mythic"
        7 => "Exotic"
        _ => "Common"

<#>
    Utility function to get tier colors from integer values (1 = Common .. 7 = Exotic).

(Tier:int).ToTierColor()<transacts>:color =
    case (Tier):
        1 => ThemeColors.RarityCommon
        2 => ThemeColors.RarityUncommon
        3 => ThemeColors.RarityRare
        4 => ThemeColors.RarityEpic
        5 => ThemeColors.RarityLegendary
        6 => ThemeColors.RarityMythic
        7 => ThemeColors.RarityExotic
        8 => ThemeColors.RarityImmortal
        9 => ThemeColors.RarityImmortal
        10 => ThemeColors.RarityImmortal
        _ => ThemeColors.RarityCommon

<#>
    Utility function to make a string of a decimal number rounded to a specified decimal point
    Reference: https://forums.unrealengine.com/t/how-to-round-float-to-x-decimal-places/821189/5?u=jutpz1517

GetRoundedString(Input:float, DecimalPoints:int)<transacts>:string =
    if:
        Multiplier := Pow(10.0, DecimalPoints * 1.0)
        RoundedValue := float[Round[Input * Multiplier] * 1.0] / Multiplier
        BeforeDecimal := Floor[RoundedValue]
        var AfterDecimal:string = ToString(Abs(Round[(RoundedValue - BeforeDecimal * 1.0) * Multiplier]))

        for (Index := 0..(DecimalPoints - AfterDecimal.Length - 1)):
            set AfterDecimal = array{'0'} + AfterDecimal
    then:       
        if (DecimalPoints = 0):
            "{BeforeDecimal}"
        else:
            "{BeforeDecimal}.{AfterDecimal}"
    else:
        ToString(Input)

<#>
    Utility function to format float into time format string (mm:ss).

(Time:float).FormatAsTime()<transacts>:string =
    var FormattedTime:string = "00:00:00"
    if:
        var Seconds:int = Floor[Time]
        Hours := Floor[Time / 3600.0]
        SecondsFromHours := Mod[Seconds, 3600]
        set Seconds = SecondsFromHours
        Minutes := Floor[Seconds * 1.0 / 60.0]
        SecondsFromMinutes := Mod[Seconds,  60]
        set Seconds = SecondsFromMinutes
    then:
        if (Time >= 3600.0):
            set FormattedTime = "{Hours.FormatLeadingZeros(2)}:{Minutes.FormatLeadingZeros(2)}:{Seconds.FormatLeadingZeros(2)}"
        else:
            set FormattedTime = "{Minutes.FormatLeadingZeros(2)}:{Seconds.FormatLeadingZeros(2)}"

    FormattedTime

<#>
    Formats a number with leading zeros.

(Number:int).FormatLeadingZeros(Digits:int)<transacts>:string =
    var FormattedNumber:string = ToString(Number)
    Difference := Max(0, Digits - FormattedNumber.Length)
    if (Difference > 0):
        for (Index := 1..Difference):
            set FormattedNumber = "0" + FormattedNumber
    FormattedNumber

<#>
    Convert an integer value to Roman Numeral.

(Val:int).ToRoman()<transacts>:string =
    Letters := array{"M", "CM", "D", "CD", "C", "XC", "L", "XL", "X", "IX", "V", "IV", "I"}
    LookupValues := array{1000, 900, 500, 400, 100, 90, 50, 40, 10, 9, 5, 4, 1}
        
    var Result:string = ""
    var Number:int = Val
    
    for (Index := 0..LookupValues.Length - 1, Number > 0):
        loop:
            if:
                Number >= LookupValues[Index]
                Value := LookupValues[Index]
                Letter := Letters[Index]
            then:
                set Number -= Value
                set Result += Letter
            else:
                break

    Result

(Logic:logic).ToString()<transacts>:string =
    Logic? and "true" or "false"


ThemeTextSize<public> := module:
    VerySmall<public>:float = 16.0
    Small<public>:float = 24.0
    Default<public>:float = 32.0