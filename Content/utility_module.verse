using { /Verse.org/Random }
using { /UnrealEngine.com/Temporary/SpatialMath }
using { Modules }

Utility<public> := module:

    # Function for creating a string from a positive number by padding it with a character. 
    PadNumber<public>(PositiveNumber:int, PadToLength:int, Character:char)<transacts>:[]char =
        var Result:[]char = "{PositiveNumber}"

        if (Result.Length < PadToLength):
            PadCount := PadToLength - Result.Length
            for(Count := 1..PadCount):
                set Result = array{Character} + Result

        Result

    PoissonDiscSamplingGeneratePoints<public>(Radius:float, ArenaSize:float, ?MinRadius:float = 800.0, ?SamplesBeforeRejection:int = 30):[]vector2 =
        CellSize := Radius / Sqrt(2.0)
        SampleRegionSize := vector2{X := ArenaSize, Y := ArenaSize} * 2.0
        GridColumns := Ceil[SampleRegionSize.X / CellSize] or 0
        GridRows := Ceil[SampleRegionSize.Y / CellSize] or 0
        ArenaSizeSquared := ArenaSize * ArenaSize

        var Grid:[]int = for (Index := 1.. GridColumns * GridRows) {-1}
        var Points:[]vector2 = array{}
        var PointsLength:int = 0
        var SpawnPoints:[]vector2 = array{vector2{X := ArenaSize - MinRadius, Y := ArenaSize - MinRadius} / 2.0}

        loop:
            if (SpawnPoints.Length <= 0):
                break
            
            SpawnIndex:int = GetRandomInt(0, SpawnPoints.Length - 1)
            SpawnPosition:vector2 = SpawnPoints[SpawnIndex] or vector2{}

            var IsCandidateAccepted:logic = false
            
            for (Index := 0..SamplesBeforeRejection - 1, not IsCandidateAccepted?):
                Angle := GetRandomFloat(0.0, 1.0) * PiFloat * 2
                Direction := vector2{X := Sin(Angle), Y := Cos(Angle)}
                CandidatePoint := SpawnPosition + Direction * GetRandomFloat(Radius, Radius * 2)

                IsValid := PoissonDiscSamplingCheckIsValid(CandidatePoint, SampleRegionSize, CellSize, Radius, Points, Grid, GridColumns, GridRows, ArenaSizeSquared, MinRadius)
                if (IsValid?):
                    set Points += array{CandidatePoint - SampleRegionSize / 2.0}
                    set SpawnPoints += array{CandidatePoint}
                    if (set Grid[Floor[CandidatePoint.X / CellSize] + Floor[CandidatePoint.Y / CellSize] * GridColumns] = PointsLength) {}
                    set PointsLength += 1
                    set IsCandidateAccepted = true

            if:
                not IsCandidateAccepted?
                UpdatedSpawnPoints := SpawnPoints.RemoveElement[SpawnIndex]
                set SpawnPoints = UpdatedSpawnPoints


        Points

    PoissonDiscSamplingCheckIsValid<public>(CandidatePoint:vector2, SampleRegionSize:vector2, CellSize:float, Radius:float, Points:[]vector2, Grid:[]int, GridColumns:int, GridRows:int, ArenaSizeSquared:float, MinRadius:float):logic =
        if (DistanceSquared(CandidatePoint, SampleRegionSize / 2.0) >= MinRadius * MinRadius and DistanceSquared(CandidatePoint, SampleRegionSize / 2.0) <= ArenaSizeSquared):
            CellX := Floor[CandidatePoint.X / CellSize] or 0
            CellY := Floor[CandidatePoint.Y / CellSize] or 0
            SearchStartX := Max(0, CellX - 2)
            SearchEndX := Min(CellX + 2, GridColumns)
            SearchStartY := Max(0, CellY - 2)
            SearchEndY := Min(CellY + 2, GridRows)

            for (X := SearchStartX..SearchEndX):
                for (Y := SearchStartY..SearchEndY):
                    if:
                        PointIndex := Grid[X + Y * GridColumns]
                        PointIndex <> -1
                        PointDistance := ((CandidatePoint - SampleRegionSize / 2.0) - Points[PointIndex]).LengthSquared()
                        PointDistance < Radius * Radius
                    then:
                        return false

            return true
        else:
            return false
        