using { /Verse.org/Simulation }

<# ----------------------------------------------------------------------------------------------------
    Data Definitions
---------------------------------------------------------------------------------------------------- #>
ability<public> := class<abstract>:
    Id<public>:weapon_id
    Player<public>:player
    DeactivatedEvent<public>:event(void) = event(void){}

    var IsActive:logic = false
    var StartTime:float = 0.0
    var CastTime:float = 0.0
    var CooldownTimeRemaining:float = 0.0
    var TotalDamage:float = 0.0
    var ProjectileCount:int = 1

    var IsUniqueEvolution:logic = false
    var AttackCooldownTime:float = 1.0
    var AttackCooldownTimeDamageMultiplier:float = 1.0
    var AttackRadius:float = 100.0
    var AttackDamage:float = 10.0
    var AttackDamageMultiplier:float = 1.0

    <# ----------------------------------------------------------------------------------------------------
        Initialization
    ---------------------------------------------------------------------------------------------------- #>
    Initialize():void =
        set StartTime = GetSimulationElapsedTime()
        set TotalDamage = 0.0
       
    Activate():void =
        UpdateStats()
        spawn{Ability()}
        set IsActive = true

    Ability()<suspends>:void
    Deactivate():void
    Deinitialize():void =
        Deactivate()

    <# ----------------------------------------------------------------------------------------------------
        Get Functions
    ---------------------------------------------------------------------------------------------------- #>
    GetIsActive()<transacts>:logic =
        IsActive

    GetTotalDamage()<transacts>:float =
        TotalDamage

    GetDuration(?EndTime:?float = false)<transacts>:float =
        if (Time := EndTime?):
            Time - StartTime
        else:
            GetSimulationElapsedTime() - StartTime

    GetDPS(?EndTime:?float = false)<transacts>:float =
        GetTotalDamage() / GetDuration(?EndTime := EndTime)

    <# ----------------------------------------------------------------------------------------------------
        Update Functions
    ---------------------------------------------------------------------------------------------------- #>
    Damage(Target:npc_behavior_base, DamageArgs:ability_damage_args, ?IsKnockback:logic = true):void =
        Player.AddGameStat(player_game_stat_id.DamageDealt, DamageArgs.Amount)
        Target.Damage(Player, DamageArgs, ?IsKnockback := IsKnockback)
        set TotalDamage += DamageArgs.Amount

    UpdateStats():void =
        if (WeaponInfo := GetWeaponInfo[Id]):
            set IsUniqueEvolution = Player.IsWeaponEvolutionAbilityActive(Id, ability_id.WeaponEvolution)

            ActualAttackCooldownTime := WeaponInfo.Cooldown / (1.0 + Player.GetWeaponStats(Id, stat_id.Cooldown) + Player.GetStat(stat_id.Cooldown))
            set AttackCooldownTime = Max(WeaponInfo.MinCooldown, ActualAttackCooldownTime)
            set AttackCooldownTimeDamageMultiplier = AttackCooldownTime / ActualAttackCooldownTime

            set AttackRadius = WeaponInfo.Size * (1.0 + Player.GetWeaponStats(Id, stat_id.Size) + Player.GetStat(stat_id.Size))

            set AttackDamage = WeaponInfo.Damage * AttackCooldownTimeDamageMultiplier
            set AttackDamageMultiplier = Player.GetStat(stat_id.Damage) +
                ((Player.HasBoon(boon_id.AllForOne)?) and (Id = Player.GetEquippedWeapon()? and 0.5 or -0.25) or 0.0)