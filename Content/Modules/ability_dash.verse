using { /Fortnite.com/Characters }
using { /Verse.org }
using { /Verse.org/SceneGraph }
using { /Verse.org/Simulation }
using { /UnrealEngine.com/Temporary/SpatialMath }
using { /UnrealEngine.com/Temporary/UI }

<# ----------------------------------------------------------------------------------------------------
    Constants
---------------------------------------------------------------------------------------------------- #>
AbilityDashCooldownTime:float = 5.0
DashKnockbackWidth:float = 300.0
DashKnockbackLength:float = 700.0

<# ----------------------------------------------------------------------------------------------------
    ability_dash
---------------------------------------------------------------------------------------------------- #>
ability_dash := class:
    Player:player

    RechargeCancelledEvent:event(void) = event(void){}
    TrackPositionEndEvent:event(void) = event(void){}

    var MaybeUI:?ability_dash_ui = false
    var Charge:int = 1
    var DashCount:int = 0
    var IsRecharging:logic = false
    var IsTrackPosition:logic = false
    var MaybeDirection:?vector3 = false
    var PreviousPosition:vector3 = vector3{}
    var Subscriptions:[]events_cancelable = array{}

    <# ----------------------------------------------------------------------------------------------------
        Initialization
    ---------------------------------------------------------------------------------------------------- #>
    Initialize()<suspends>:void =
        WaitForUIManager(Player)
        WaitForPlayerManager(Player)
        WaitForInputManager()
        
        AbilityHUDUI := ability_dash_ui{Player := Player}
        AbilityHUDUI.Initialize()
        set MaybeUI = option{AbilityHUDUI}

        Reset()

        if (Player.GetInputMode() = input_mode.Touch):
            spawn{TrackPosition()}

        if (InputManager := GetInputManager[]):
            set Subscriptions += array{InputManager.InputModeChangedEvent.Subscribe(OnInputModeChanged)}

    Uninitialize():void =
        RechargeCancelledEvent.Signal()
        TrackPositionEndEvent.Signal()

        for (Subscription : Subscriptions):
            Subscription.Cancel()

        set Subscriptions = array{}

    <# ----------------------------------------------------------------------------------------------------
        Get Functions
    ---------------------------------------------------------------------------------------------------- #>
    GetCooldownTime()<transacts>:float =
        if (IsGameInProgress()?):
            AbilityDashCooldownTime / (1.0 + Player.GetStat(stat_id.DashCooldown))
        else:
            0.5

    GetMaxCharge()<transacts>:int =
        1 + Floor[Player.GetPowerUpValue(power_up_id.DashCount)] or 0

    <# ----------------------------------------------------------------------------------------------------
        Update Functions
    ---------------------------------------------------------------------------------------------------- #>
    TrackPosition()<suspends>:void =
        set IsTrackPosition = true
        if (Character := Player.GetFortCharacter[]):
            race:
                TrackPositionEndEvent.Await() 
                loop:
                    if (Character.IsActive[]):
                        set MaybeDirection = option{(Character.GetTransform().Translation - PreviousPosition).MakeUnitVector[]}
                        set PreviousPosition = Character.GetTransform().Translation
                        Sleep(TwoTicks)
                    else:
                        break

        set IsTrackPosition = false

    Activate():void =
        if (Charge = 0):
            return

        ActivateDash()
        
        if (Charge < GetMaxCharge(), not IsRecharging?):
            spawn{Recharge()}

    Recharge()<suspends>:void =
        set IsRecharging = true
        defer:
            set IsRecharging = false

        race:
            RechargeCancelledEvent.Await()
            block:
                DeltaTimer := delta_timer{}
                if (UI := MaybeUI?):
                    loop:
                        Sleep(0.0)
                        if (Player.IsActive[]):
                            Progress := DeltaTimer.GetElapsedTime() / GetCooldownTime()
                            UI.UpdateProgress(Progress)
                            if (DeltaTimer.GetElapsedTime() >= GetCooldownTime()):
                                set Charge = Min(GetMaxCharge(), Charge + 1)
                                UI.UpdateCharge(Charge)
                                if (Charge = GetMaxCharge()):
                                    break
                                else:
                                    DeltaTimer.Reset()
                        else:
                            break

    Reset():void =
        RechargeCancelledEvent.Signal()

        if (UI := MaybeUI?):
            set Charge = GetMaxCharge()
            UI.UpdateProgress(1.0)
            UI.UpdateCharge(GetMaxCharge())

    ActivateDash():void =
        if:
            not Player.IsInStasis()?
            Character := Player.GetFortCharacter[]
            Character.IsActive[]
            not Character.IsDownButNotOut[]
        then:
            if (Character.IsOnGround[]):
                set DashCount = 0
            else:
                if (Player.GetPowerUpLevel(power_up_id.AirDash) <= DashCount):
                    return

                set DashCount += 1

            DashRotation := if (Player.GetInputMode() = input_mode.Touch):
                if (Direction := MaybeDirection?):
                    MakeRotation(vector3{X := 0.0, Y := 0.0, Z := 1.0}, ArcTan(Direction.Y, Direction.X))
                else:
                    Character.GetTransform().Rotation

            else:
                YawAmount := if (Player.IsInputHeld[input_id.Forward] and Player.IsInputHeld[input_id.Right]) {45.0}
                    else if (Player.IsInputHeld[input_id.Right] and Player.IsInputHeld[input_id.Backward]) {135.0}
                    else if (Player.IsInputHeld[input_id.Backward] and Player.IsInputHeld[input_id.Left]) {225.0}
                    else if (Player.IsInputHeld[input_id.Left] and Player.IsInputHeld[input_id.Forward]) {315.0}
                    else if (Player.IsInputHeld[input_id.Forward]) {0.0}
                    else if (Player.IsInputHeld[input_id.Right]) {90.0}
                    else if (Player.IsInputHeld[input_id.Backward]) {180.0}
                    else if (Player.IsInputHeld[input_id.Left]) {270.0}
                    else 0.0

                if (Player.GetActiveCamera() = camera_type.Orbit or Player.GetActiveCamera() = camera_type.OrbitFree):
                    Character.GetTransform().Rotation
                else:
                    if (not Player.IsInputHeld[input_id.Forward] and not Player.IsInputHeld[input_id.Backward] and not Player.IsInputHeld[input_id.Left] and not Player.IsInputHeld[input_id.Right]):
                        Character.GetTransform().Rotation
                    else:
                        MakeRotationFromYawPitchRollDegrees(YawAmount, 0.0, 0.0)

            Player.ApplyImpulse(Player.HasBoon(boon_id.DashingWings)? and impulse_id.DashBoon or impulse_id.Dash, ?Rotation := DashRotation)            
            Player.PlayAudio(audio_id.Impulse)
            Player.SetAFK(false)
            KnockbackNPCs(Character, DashRotation)
            set Charge -= 1
            if (UI := MaybeUI?):
                UI.UpdateCharge(Charge)

    KnockbackNPCs(Character:fort_character, DashRotationRotation:rotation):void =
        DashPosition := Character.GetTransform().Translation.Make2D()
        Direction := DashRotationRotation.GetLocalForward()
        TargetPosition := DashPosition + Direction * DashKnockbackLength / 2.0
        Collision := collision_box:
            Extents := SpatialMath.vector3{Left := -DashKnockbackWidth / 2.0, Up := 300.0, Forward := DashKnockbackLength / 2.0}
        
        for:
            OverlapHit : GetSimulationEntity[].FindOverlapHits(FromTransform of transform{Translation := TargetPosition, Rotation := DashRotationRotation}, Collision)
            TargetComponent := OverlapHit.TargetComponent
            TargetEntity := TargetComponent.Entity
            NPCCollisionComponent := TargetEntity.GetComponent[npc_collision_component]
            NPCHit := NPCCollisionComponent.MaybeNPC?
            NPCHit.IsDamageable()?
        do:
            NPCHit.Knockback(knockback_properties{Duration := 1.0, Rotation := DashRotationRotation})

    OnInputModeChanged(InPlayer:player):void =
        if (InPlayer <> Player) then return

        if (Player.GetInputMode() = input_mode.Touch):
            if (IsTrackPosition?) then return
            spawn{TrackPosition()}
        else:
            TrackPositionEndEvent.Signal()