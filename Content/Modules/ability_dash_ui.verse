using { /Fortnite.com/UI }
using { /Verse.org/Colors }
using { /UnrealEngine.com/Temporary/SpatialMath }
using { /UnrealEngine.com/Temporary/UI }

<# ----------------------------------------------------------------------------------------------------
    Constants
---------------------------------------------------------------------------------------------------- #>
AbilityDashUI<public> := module:
    Padding<public>:float = 20.0
    IconSize<public>:float = 40.0
    BorderWidth<public>:float = 15.0
    MaxHeight<public>:float = IconSize + Padding * 2.0 - BorderWidth * 2.0

<# ----------------------------------------------------------------------------------------------------
    ability_dash_ui
---------------------------------------------------------------------------------------------------- #>
ability_dash_ui := class(ui_component):
    AbilityChargeText:text_block = text_block:
        DefaultTextColor := ThemeColors.Highlight
        DefaultShadowOpacity := 0.5
        DefaultShadowOffset := option{vector2{X := 2.0, Y := 2.0}}

    ProgressColorBlock:color_block = color_block:
        DefaultColor := NamedColors.White
        DefaultOpacity := 0.3
        DefaultDesiredSize := vector2{X := 0.0, Y := 0.0}

    <# ----------------------------------------------------------------------------------------------------
        Initialization
    ---------------------------------------------------------------------------------------------------- #>
    CreateUI<override>():widget =
        AbilityChargeText.SetText(MakeMessage("1"))

        StackBox := stack_box:
            Orientation := orientation.Vertical
            Slots := array:
                stack_box_slot:
                    HorizontalAlignment := horizontal_alignment.Center
                    VerticalAlignment := vertical_alignment.Fill
                    Widget := AbilityChargeText
                stack_box_slot:
                    HorizontalAlignment := horizontal_alignment.Fill
                    VerticalAlignment := vertical_alignment.Fill
                    Widget := overlay:
                        Slots := array:
                            overlay_slot:
                                HorizontalAlignment := horizontal_alignment.Fill
                                VerticalAlignment := vertical_alignment.Fill
                                Widget := UserInterface.VerseWidget.WBP_AbilityHUD{}
                            overlay_slot:
                                HorizontalAlignment := horizontal_alignment.Fill
                                VerticalAlignment := vertical_alignment.Fill
                                Padding := margin{Left := 5.0, Top := 5.0, Right := 5.0, Bottom := 5.0}
                                Widget := UserInterface.VerseWidget.WBP_AbilityHUD_Entry{}
                            overlay_slot:
                                HorizontalAlignment := horizontal_alignment.Fill
                                VerticalAlignment := vertical_alignment.Bottom
                                Padding := margin{Left := 5.0, Top := 5.0, Right := 5.0, Bottom := 5.0}
                                Widget := ProgressColorBlock
                            overlay_slot:
                                Padding := margin{Left := AbilityDashUI.BorderWidth, Top := AbilityDashUI.BorderWidth, Right := AbilityDashUI.BorderWidth, Bottom := AbilityDashUI.BorderWidth}
                                Widget := texture_block:
                                    DefaultImage := UserInterface.Icons.DashCooldown_64
                                    DefaultDesiredSize := vector2{X := AbilityDashUI.IconSize, Y := AbilityDashUI.IconSize}

        Player.AddWidgetToHUD of canvas_slot:
            Anchors := anchors{Minimum := vector2{X := 1.0, Y := 1.0}, Maximum := vector2{X := 1.0, Y := 1.0}}
            Offsets := margin{Left := -542.0, Top := -64.0, Right := 0.0, Bottom := 0.0}
            Alignment := vector2{X := 1.0, Y := 1.0}
            Widget := StackBox

        StackBox

    <# ----------------------------------------------------------------------------------------------------
        Update Functions
    ---------------------------------------------------------------------------------------------------- #>
    UpdateProgress(Progress:float):void =
        Height := AbilityDashUI.MaxHeight * (1.0 - Progress)
        ProgressColorBlock.SetDesiredSize(vector2{X := 1.0, Y := Height})

    UpdateCharge(Amount:int):void =
        AbilityChargeText.SetText(MakeMessage("{Amount}"))