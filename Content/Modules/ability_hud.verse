using { /Verse.org/Simulation }
using { /UnrealEngine.com/Temporary/SpatialMath }
using { /UnrealEngine.com/Temporary/UI }

ability_hud := class:
    Player:player

    var MaybeAbilityHUDUI:?ability_hud_ui = false
    var Subscriptions:[]events_cancelable = array{}

    <# ----------------------------------------------------------------------------------------------------
        Initialization
    ---------------------------------------------------------------------------------------------------- #>
    Initialize()<suspends>:void =
        WaitForGameManager()
        WaitForUIManager(Player)
        WaitForPlayerManager(Player)

        AbilityHUDUI := ability_hud_ui{Player := Player}
        AbilityHUDUI.Initialize()
        set MaybeAbilityHUDUI = option{AbilityHUDUI}

        if (GameManager := GetGameManager[], PlayerStat := Player.GetStats[]):
            set Subscriptions = array:
                GameManager.GameStartedEvent.Subscribe(OnGameStarted)
                GameManager.GameEndedEvent.Subscribe(OnGameEnded)
                GameManager.AbilityActivatedEvent.Subscribe(OnAbilityActivated)
                PlayerStat.AbilityChangedEvent.Subscribe(OnAbilityChanged)
                PlayerStat.NewWeaponEquippedEvent.Subscribe(OnNewWeaponEquipped)

    Uninitialize():void =
        for (Subscription : Subscriptions):
            Subscription.Cancel()

        set Subscriptions = array{}

        if (AbilityHUDUI := MaybeAbilityHUDUI?):
            AbilityHUDUI.Uninitialize()
            set MaybeAbilityHUDUI = false

    <# ----------------------------------------------------------------------------------------------------
        Update Functions
    ---------------------------------------------------------------------------------------------------- #>
    Show():void =
        if (AbilityHUDUI := MaybeAbilityHUDUI?):
            AbilityHUDUI.Show()

    Hide():void =
        if (AbilityHUDUI := MaybeAbilityHUDUI?):
            AbilityHUDUI.Hide()

    Update(Index:int, WeaponInfo:weapon_info):void =
        if (AbilityHUDUI := MaybeAbilityHUDUI?):
            AbilityHUDUI.Update(Index, WeaponInfo)

    <# ----------------------------------------------------------------------------------------------------
        Event Handlers
    ---------------------------------------------------------------------------------------------------- #>
    OnGameStarted():void =
        if (AbilityHUDUI := MaybeAbilityHUDUI?):
            AbilityHUDUI.Reset()
            if (EquippedWeapon := Player.GetEquippedWeapon()?, WeaponInfo := GetWeaponInfo[EquippedWeapon]):
                AbilityHUDUI.Update(0, WeaponInfo)

        Show()

    OnGameEnded():void =
        Hide()

    OnAbilityChanged(AbilityId:ability_id):void =
        var Index:int = 0
        for (Id -> WeaponLevel : Player.GetWeaponLevels(), WeaponInfo := GetWeaponInfo[Id]):
            Update(Index, WeaponInfo)
            set Index += 1

    OnAbilityActivated():void =
        if (AbilityHUDUI := MaybeAbilityHUDUI?):
            AbilityHUDUI.StartAnimation()

    OnNewWeaponEquipped(Id:weapon_id):void =
        if (AbilityHUDUI := MaybeAbilityHUDUI?):
            AbilityHUDUI.AddWeapon(Id)