using { /Fortnite.com/UI }
using { /Verse.org/Colors }
using { /Verse.org/Simulation }
using { /UnrealEngine.com/Temporary/SpatialMath }
using { /UnrealEngine.com/Temporary/UI }

ability_hud_ui := class(ui_component):
    CancelAnimationEvent:event(void) = event(void){}

    var Icons:[]texture_block = array{}
    var LevelTexts:[]text_block = array{}
    var Entries:[]widget = array{}
    var ActiveAbilities:[]ability = array{}
    var ColorBlocks:[]color_block = array{}

    <# ----------------------------------------------------------------------------------------------------
        Initialization
    ---------------------------------------------------------------------------------------------------- #>
    CreateUI<override>():widget =
        StackBox := stack_box:
            Orientation := orientation.Vertical
            Slots := array:
                stack_box_slot:
                    HorizontalAlignment := horizontal_alignment.Fill
                    VerticalAlignment := vertical_alignment.Fill
                    Widget := stack_box:
                        Orientation := orientation.Horizontal
                        Slots := for (Index := 1..4):
                            LevelText := text_block:
                                DefaultShadowOffset := option{vector2{X := 2.0, Y := 2.0}}
                                DefaultShadowOpacity := 0.5

                            set LevelTexts += array{LevelText}

                            stack_box_slot:
                                VerticalAlignment := vertical_alignment.Fill
                                Distribution := option{1.0}
                                Widget := LevelText

                stack_box_slot:
                    HorizontalAlignment := horizontal_alignment.Fill
                    VerticalAlignment := vertical_alignment.Fill
                    Widget := overlay:
                        Slots := array:
                            overlay_slot:
                                HorizontalAlignment := horizontal_alignment.Fill
                                VerticalAlignment := vertical_alignment.Fill
                                Widget := UserInterface.VerseWidget.WBP_AbilityHUD{}
                            overlay_slot:
                                HorizontalAlignment := horizontal_alignment.Fill
                                VerticalAlignment := vertical_alignment.Fill
                                Padding := margin{Left := 5.0, Top := 5.0, Right := 5.0, Bottom := 5.0}
                                Widget := stack_box:
                                    Orientation := orientation.Horizontal
                                    Slots := for (Index := 1..4):
                                        Icon := texture_block:
                                            DefaultImage := UserInterface.Icons.Lock_64
                                            DefaultDesiredSize := vector2{X := 40.0, Y := 40.0}
                                        
                                        ColorBlock := color_block:
                                            DefaultColor := NamedColors.White
                                            DefaultOpacity := 0.3
                                            DefaultDesiredSize := vector2{X := 60.0, Y := 0.0}
                                            
                                        Entry := overlay:
                                            Slots := array:
                                                overlay_slot:
                                                    HorizontalAlignment := horizontal_alignment.Fill
                                                    VerticalAlignment := vertical_alignment.Fill
                                                    Widget := UserInterface.VerseWidget.WBP_AbilityHUD_Entry{}
                                                overlay_slot:
                                                    HorizontalAlignment := horizontal_alignment.Fill
                                                    VerticalAlignment := vertical_alignment.Bottom
                                                    Widget := ColorBlock
                                                overlay_slot:
                                                    Padding := margin{Left := 10.0, Top := 10.0, Right := 10.0, Bottom := 10.0}
                                                    Widget := Icon

                                        set Icons += array{Icon}
                                        set ColorBlocks += array{ColorBlock}
                                        set Entries += array{Entry}

                                        stack_box_slot:
                                            HorizontalAlignment := horizontal_alignment.Fill
                                            VerticalAlignment := vertical_alignment.Fill
                                            Padding := margin{Left := (Index = 1) and 0.0 or 5.0, Top := 0.0, Right := 0.0, Bottom := 0.0}
                                            Widget := Entry

        Player.AddWidgetToHUD of canvas_slot:
            Anchors := anchors{Minimum := vector2{X := 0.5, Y := 1.0}, Maximum := vector2{X := 0.5, Y := 1.0}}
            Offsets := margin{Left := 0.0, Top := -64.0, Right := 0.0, Bottom := 0.0}
            Alignment := vector2{X := 0.5, Y := 1.0}
            Widget := StackBox

        StackBox.SetVisibility(widget_visibility.Collapsed)
        StackBox

    Uninitialize():void =
        StopAnimation()

    <# ----------------------------------------------------------------------------------------------------
        Get Functions
    ---------------------------------------------------------------------------------------------------- #>
    GetLevel(Index:int)<transacts>:string =
        Level := if (Index = 1). ""
            else if (Index = 2). "{WeaponUnlockLevel1}"
            else if (Index = 3). "{WeaponUnlockLevel2}"
            else. "{WeaponUnlockLevel3}"

    GetCooldownProgress(Ability:ability)<transacts>:float =
        (GetSimulationElapsedTime() - Ability.CastTime) / Ability.AttackCooldownTime

    <# ----------------------------------------------------------------------------------------------------
        Update Functions
    ---------------------------------------------------------------------------------------------------- #>
    Update(Index:int, WeaponInfo:weapon_info):void =
        if (Icon := Icons[Index], LevelText := LevelTexts[Index]):
            Icon.SetImage(WeaponInfo.IconSmall)
            Icon.SetTint(NamedColors.White)
            LevelText.SetText(MakeMessage("{Player.GetWeaponLevel(WeaponInfo.Id)}"))
            LevelText.SetTextColor(ThemeColors.Highlight)

    AddWeapon(Id:weapon_id):void =
        if (Ability := GetPlayerActiveAbilities()[Player][Id]):
            set ActiveAbilities += array{Ability}

    Reset():void =
        MaxWeapons := GetGameSettings[].Weapons or 4
        
        set ActiveAbilities = array{}
        for (ColorBlock : ColorBlocks):
            ColorBlock.SetDesiredSize(vector2{X := 60.0, Y := 0.0})
            
        for (Index := 1..Icons.Length):
            if (Icon := Icons[Index - 1], LevelText := LevelTexts[Index - 1], Entry := Entries[Index - 1]):
                if (MaxWeapons - Index < 0):
                    LevelText.SetVisibility(widget_visibility.Collapsed)
                    Entry.SetVisibility(widget_visibility.Collapsed)
                else:
                    LevelText.SetVisibility(widget_visibility.Visible)
                    Entry.SetVisibility(widget_visibility.Visible)
                    Icon.SetImage(UserInterface.Icons.Lock_64)
                    Icon.SetTint(ThemeColors.Locked)
                    LevelText.SetText(MakeMessage("{GetLevel(Index)}"))
                    LevelText.SetTextColor(ThemeColors.Failure)

    StartAnimation():void =
        if:
            Ability := GetPlayerActiveAbilities()[Player][Player.GetEquippedWeapon()?]
            set ActiveAbilities = array{Ability}

        spawn{AnimateCooldown()}

    StopAnimation():void =
        CancelAnimationEvent.Signal()

    AnimateCooldown()<suspends>:void =
        race:
            CancelAnimationEvent.Await()

            loop:
                for (Index -> Ability : ActiveAbilities, ColorBlock := ColorBlocks[Index]):
                    Progress := Clamp(1.0 - GetCooldownProgress(Ability), 0.0, 1.0)
                    ColorBlock.SetDesiredSize(vector2{X := 60.0, Y := 60.0 * Progress})

                Sleep(0.0)

    <# ----------------------------------------------------------------------------------------------------
        Event Handler
    ---------------------------------------------------------------------------------------------------- #>
    OnHide<override>():void =
        StopAnimation()