using { /Verse.org/Assets }
using { /Verse.org/Colors }
using { /Verse.org/Random }
using { /Verse.org/Simulation }

<# ----------------------------------------------------------------------------------------------------
    Constants
---------------------------------------------------------------------------------------------------- #>
MaxAbilityLevel:int = 10
DefaultItemPickupDistance:float = 100.0
MaxWeaponLevel:int = 18
WeaponEvolutionInterval:int = 6
BaseMaxHealth:float = 200.0
BaseHealthRegeneration:float = 0.5
BaseAbilityTierWeights:[]int = array{60000, 25000, 10000, 4000, 1000}

<# ----------------------------------------------------------------------------------------------------
    Data Definitions
---------------------------------------------------------------------------------------------------- #>
ability_id := enum:
    Damage
    Cooldown
    Size
    MoveSpeed
    ExperienceMultiplier
    ItemPickupDistance
    MaxHealth
    HealthRegeneration
    DashCooldown
    GatherRate
    CriticalRate
    CriticalDamage
    Luck
    WeaponLevel

    ProjectileCount
    ProjectileSpeed
    ProjectilePiercing
    ProjectileRange
    ProjectileSplit

    FireRate
    Multicast
    SwordAngle
    AuraDuration
    AuraPower
    TorrentFireRate
    ShieldMaxHealth
    ShieldRegenerationRate
    TrailChargedChance
    TrailChargedPower
    WeaponEvolution
    ChainLightningBounce
    ChainLightningResilience
    StaffThrust
    StaffPillar
    StaffThrow
    StaffSwirl

ability_info_id := struct<computes>:
    MaybeWeaponId:?weapon_id = false
    Id:ability_id

ability_info := struct<computes>:
    MaybeWeaponId:?weapon_id = false
    Id:ability_id
    Name:string
    ShortName:string = ""
    Description:string
    Icon:texture
    IconSmall:texture
    IconColor:color = NamedColors.White
    BaseValue:float = 0.0
    TierValue:[]?float = array{}
    MaybeStatId:?stat_id = false
    MaybeMinLevel:?int = false
    MaybeMaxValue:?float = false
    FormatValue:type{_(:float)<transacts>:string}
    MaybeStatIcon:?texture = false
    StatIconColor:color = NamedColors.White

<# ----------------------------------------------------------------------------------------------------
    Ability Info
---------------------------------------------------------------------------------------------------- #>
Abilities:[ability_id]ability_info = map:
    ability_id.Damage => ability_info:
        Id := ability_id.Damage
        Name := "Damage"
        ShortName := "Damage"
        Description := "Increase damage of all weapons."
        Icon := UserInterface.Icons.ProjectileDamage_128
        IconSmall := UserInterface.Icons.ProjectileDamage_64
        IconColor := ThemeColors.WeaponDamage
        TierValue := array{option{0.05}, option{0.1}, option{0.15}, option{0.2}, option{0.3}}
        MaybeStatId := option{stat_id.Damage}
        FormatValue := FormatAsPercent
    ability_id.Cooldown => ability_info:
        Id := ability_id.Cooldown
        Name := "Cooldown"
        ShortName := "Cooldown"
        Description := "Reduces cooldown of all weapons."
        Icon := UserInterface.Icons.AbilityCooldown_128
        IconSmall := UserInterface.Icons.AbilityCooldown_64
        IconColor := ThemeColors.WeaponCooldown
        TierValue := array{option{0.05}, option{0.1}, option{0.15}, option{0.2}, option{0.3}}
        MaybeStatId := option{stat_id.Cooldown}
        FormatValue := FormatAsPercent
    ability_id.Size => ability_info:
        Id := ability_id.Size
        Name := "Size"
        ShortName := "Size"
        Description := "Increase the size of all weapons."
        Icon := UserInterface.Icons.AbilitySize_128
        IconSmall := UserInterface.Icons.AbilitySize_64
        IconColor := ThemeColors.WeaponSize
        TierValue := array{option{0.05}, option{0.1}, option{0.15}, option{0.2}, option{0.3}}
        MaybeStatId := option{stat_id.Size}
        FormatValue := FormatAsPercent
    ability_id.CriticalRate => ability_info:
        Id := ability_id.CriticalRate
        Name := "Critical Rate"
        ShortName := "C.Rate"
        Description := "Increase chance of critical hit."
        Icon := UserInterface.Icons.CriticalRate_128
        IconSmall := UserInterface.Icons.CriticalRate_64
        TierValue := array{option{0.05}, option{0.07}, option{0.09}, option{0.12}, option{0.15}}
        BaseValue := 0.1
        MaybeStatId := option{stat_id.CriticalRate}
        FormatValue := FormatAsPercent
    ability_id.CriticalDamage => ability_info:
        Id := ability_id.CriticalDamage
        Name := "Critical Damage"
        ShortName := "C.Dmg"
        Description := "Increase critical damage multiplier."
        Icon := UserInterface.Icons.CriticalDamage_128
        IconSmall := UserInterface.Icons.CriticalDamage_64
        TierValue := array{option{0.05}, option{0.1}, option{0.15}, option{0.35}, option{0.5}}
        BaseValue := 0.5
        MaybeStatId := option{stat_id.CriticalDamage}
        FormatValue := FormatAsPercent
    ability_id.MaxHealth => ability_info:
        Id := ability_id.MaxHealth
        Name := "Max Health"
        ShortName := "Health"
        Description := "Increase max health."
        Icon := UserInterface.Icons.MaxHealth_128
        IconSmall := UserInterface.Icons.MaxHealth_64
        BaseValue := BaseMaxHealth
        TierValue := array{option{20.0}, option{30.0}, option{40.0}, option{60.0}, option{90.0}}
        MaybeStatId := option{stat_id.MaxHealth}
        FormatValue := FormatAsInt
    ability_id.HealthRegeneration => ability_info:
        Id := ability_id.HealthRegeneration
        Name := "Health Regeneration"
        ShortName := "Regen"
        Description := "Increase health regeneration rate."
        Icon := UserInterface.Icons.HealthRegeneration_128
        IconSmall := UserInterface.Icons.HealthRegeneration_64
        BaseValue := BaseHealthRegeneration
        TierValue := array{option{0.5}, option{1.0}, option{1.5}, option{2.0}, option{3.0}}
        MaybeStatId := option{stat_id.HealthRegeneration}
        FormatValue := FormatAsRate
    ability_id.MoveSpeed => ability_info:
        Id := ability_id.MoveSpeed
        Name := "Movement Speed"
        ShortName := "Speed"
        Description := "Increase movement speed."
        Icon := UserInterface.Icons.MoveSpeed_128
        IconSmall := UserInterface.Icons.MoveSpeed_64
        TierValue := array{option{0.1}, option{0.2}, option{0.3}, option{0.4}, option{0.5}}
        MaybeStatId := option{stat_id.MoveSpeed}
        MaybeMaxValue := option{2.0}
        FormatValue := FormatAsPercent
    ability_id.DashCooldown => ability_info:
        Id := ability_id.DashCooldown
        Name := "Dash Cooldown"
        ShortName := "Dash"
        Description := "Reduce the cooldown of the dash ability."
        Icon := UserInterface.Icons.DashCooldown_128
        IconSmall := UserInterface.Icons.DashCooldown_64
        TierValue := array{false, option{0.5}, option{0.75}, option{1.0}, option{1.5}}
        MaybeStatId := option{stat_id.DashCooldown}
        FormatValue := FormatAsPercent
    ability_id.GatherRate => ability_info:
        Id := ability_id.GatherRate
        Name := "Gather Rate"
        ShortName := "Gather"
        Description := "Increase wood gathering speed & amount (x2 effectiveness for gathering speed)."
        Icon := UserInterface.Icons.GatherRate_128
        IconSmall := UserInterface.Icons.GatherRate_64
        TierValue := array{false, option{0.1}, option{0.15}, option{0.2}, option{0.25}}
        MaybeStatId := option{stat_id.GatherRate}
        FormatValue := FormatAsPercent
    ability_id.Luck => ability_info:
        Id := ability_id.Luck
        Name := "Luck"
        ShortName := "Luck"
        Description := "Increase chance of receiving higher tier abilities."
        Icon := UserInterface.Icons.AbilityLuck_128
        IconSmall := UserInterface.Icons.AbilityLuck_64
        TierValue := array{false, false, option{0.25}, option{0.35}, option{0.5}}
        MaybeStatId := option{stat_id.Luck}
        FormatValue := FormatAsPercent
    ability_id.ItemPickupDistance => ability_info:
        Id := ability_id.ItemPickupDistance
        Name := "Pickup Range"
        ShortName := "Pickup"
        Description := "Increase distance player can pick up EXP orbs."
        Icon := UserInterface.Icons.ItemPickupDistance_128
        IconSmall := UserInterface.Icons.ItemPickupDistance_64
        TierValue := array{false, false, option{0.5}, option{1.0}, option{2.0}}
        MaybeStatId := option{stat_id.ItemPickupDistance}
        FormatValue := FormatAsPercent
    ability_id.ExperienceMultiplier => ability_info:
        Id := ability_id.ExperienceMultiplier
        Name := "Experience Multiplier"
        ShortName := "Exp"
        Description := "Increase experience earned from orbs (average of all players)."
        Icon := UserInterface.Icons.ExperienceMultiplier_128
        IconSmall := UserInterface.Icons.ExperienceMultiplier_64
        TierValue := array{false, option{0.1}, option{0.15}, option{0.2}, option{0.25}}
        MaybeStatId := option{stat_id.ExperienceMultiplier}
        FormatValue := FormatAsPercent

<# ----------------------------------------------------------------------------------------------------
    Global Functions
---------------------------------------------------------------------------------------------------- #>
GetAbilityInfo()<transacts>:[ability_id]ability_info =
    Abilities

GetAbilityInfo(Id:ability_id)<decides><transacts>:ability_info =
    GetAbilityInfo()[Id]

GetAbilityIds()<transacts>:[]ability_id =
    for (Id -> AbilityInfo : GetAbilityInfo()) {Id}

GetTreasureChestRandomAbility(Player:player)<transacts>:?tuple(ability_info, int) =
    IsSelected := logic{GetLevel() > 1 and Player.GetPowerUpLevel(power_up_id.TreasureChestFocus) > 0}
    MinTier := (Floor[Player.GetPowerUpValue(power_up_id.TreasureChestTier)] or 0) + 1
    GetRandomAbility(Player, ?LuckMultiplier := Player.GetPowerUpValue(power_up_id.TreasureChestLuck), ?IsSelected := IsSelected, ?MinTier := MinTier)

GetRandomAbility(Player:player, ?LuckMultiplier:float = 0.0, ?LuckOverride:?float = false, ?IsSelected:logic = false, ?MinTier:int = 1)<transacts>:?tuple(ability_info, int) =
    option{GetRandomAbilities(Player, 1, ?LuckMultiplier := LuckMultiplier, ?LuckOverride := LuckOverride, ?IsSelected := IsSelected, ?MinTier := MinTier)[0]}

GetRandomAbilities(Player:player, Count:int, ?Ignore:[]ability_id = array{}, ?LuckMultiplier:float = 0.0, ?LuckOverride:?float = false, ?IsSelected:logic = false, ?MinTier:int = 1)<transacts>:[]tuple(ability_info, int) =
    PlayerBanishedAbilities := Player.GetBanishedAbilities()
    BanishedWeaponAbilities := for (BanishedAbility : PlayerBanishedAbilities, BanishedAbility.MaybeWeaponId?) {BanishedAbility}
    BanishedAbilityIds := for (BanishedAbility : PlayerBanishedAbilities, not BanishedAbility.MaybeWeaponId?) {BanishedAbility.Id}

    AbilityInfos := for:
        AbilityInfo : GetAbilityInfo()
        not BanishedAbilityIds.Find[AbilityInfo.Id]
        not AbilityInfo.MaybeMaxValue? or (Player.GetStat(AbilityInfo.MaybeStatId?) < AbilityInfo.MaybeMaxValue?)
        not IsSelected? or Player.GetAbilityValue(AbilityInfo.Id) > 0.0
    do:
        AbilityInfo

    WeaponAbilityInfos := for:
        WeaponAbilityInfo : GetAbilityManager[].GetWeaponAbilityInfo(Player)
        not GetIsWeaponAbilityBanished(Player, WeaponAbilityInfo, BanishedWeaponAbilities)?
        not IsSelected? or (Player.GetWeaponStats(WeaponAbilityInfo.MaybeWeaponId?, WeaponAbilityInfo.MaybeStatId?) > 0.0)
    do:            
        WeaponAbilityInfo

    var RandomAbilityIndexes:[]int = array{}
    var RandomWeaponIndexes:[]int = array{}
    var RandomAbilities:[]tuple(ability_info, int) = array{}
    loop:
        if (RandomAbilityIndexes.Length + RandomWeaponIndexes.Length >= Count):
            break

        IsWeapon := logic{GetRandomFloat(0.0, 1.0) <= 0.5}
        RandomAbilityInfos := IsWeapon? and WeaponAbilityInfos or AbilityInfos
        RandomIndex := GetRandomInt(0, RandomAbilityInfos.Length - 1)
        RandomIndexes := IsWeapon? and RandomWeaponIndexes or RandomAbilityIndexes
        Tier := GetRandomAbilityTier(Player, ?LuckMultiplier := LuckMultiplier, ?LuckOverride := LuckOverride, ?MinTier := MinTier)

        if (not RandomIndexes.Find[RandomIndex], AbilityInfo := RandomAbilityInfos[RandomIndex], TierValue := AbilityInfo.TierValue[Tier - 1]?):
            set RandomAbilities += array{(AbilityInfo, Tier)}

            if (IsWeapon?):
                set RandomWeaponIndexes += array{RandomIndex}
            else:
                set RandomAbilityIndexes += array{RandomIndex}

    RandomAbilities

GetIsWeaponAbilityBanished(Player:player, WeaponAbilityInfo:ability_info, BanishedWeaponAbilities:[]ability_info_id)<transacts>:logic =
    var IsBanished:logic = false
    for:
        BanishedWeaponAbility : BanishedWeaponAbilities
        not IsBanished?
        BanishedWeaponId := BanishedWeaponAbility.MaybeWeaponId?
        WeaponId := WeaponAbilityInfo.MaybeWeaponId?
        WeaponId = BanishedWeaponId and BanishedWeaponAbility.Id = WeaponAbilityInfo.Id
    do:
        set IsBanished = true

    IsBanished

(Player:player).GetPlayerAbilityTierWeights(?LuckMultiplier:float = 0.0)<transacts>:[]int =
    CalculateAbilityTierWeights(Player.GetStat(stat_id.Luck) + LuckMultiplier)

CalculateAbilityTierWeights(LuckMultiplier:float)<transacts>:[]int =
    for (Tier -> BaseAbilityTierWeight : BaseAbilityTierWeights):
        Floor[BaseAbilityTierWeight * (1.0 + LuckMultiplier * (Pow((Tier + 1) * 1.0, 1.5) - 1.0) * 0.3)] or BaseAbilityTierWeight

GetRandomAbilityTier(Player:player, ?LuckMultiplier:float = 0.0, ?LuckOverride:?float = false, ?MinTier:int = 1)<transacts>:int =
    TierWeights := if (Luck := LuckOverride?):
        CalculateAbilityTierWeights(Luck)
    else:
        Player.GetPlayerAbilityTierWeights(?LuckMultiplier := LuckMultiplier)

    FilteredTierWeights := TierWeights.Slice[MinTier - 1] or TierWeights
    GetWeightedRandom(FilteredTierWeights) + MinTier

(AbilityInfo:ability_info).GetId()<transacts>:ability_info_id =
    ability_info_id:
        MaybeWeaponId := AbilityInfo.MaybeWeaponId
        Id := AbilityInfo.Id

<# ----------------------------------------------------------------------------------------------------
    Ability Value Formatting
---------------------------------------------------------------------------------------------------- #>
FormatAsFloat(Value:float, ?Decimals:int = 2)<transacts>:string =
    GetRoundedString(Value, Decimals)

FormatAsPercent(Value:float, ?Decimals:int = 0)<transacts>:string =
    "{GetRoundedString((Value * 100.0), Decimals)}%"

FormatAsInt(Value:float)<transacts>:string =
    "{Floor[Value] or 0}"

FormatAsRate(Value:float)<transacts>:string =
    "{GetRoundedString(Value, 2)}/s"

FormatAsTime(Value:float)<transacts>:string =
    "{GetRoundedString(Value, 2)}s"
    
FormatAsUnique(Value:float)<transacts>:string =
    "UNIQUE"

FormatAsDegree(Value:float)<transacts>:string =
    "{GetRoundedString(Value, 0)}°"

FormatAsTier(Value:float)<transacts>:string =
    "{Floor[Value + 1.0].ToTierString()}" or "Common"