using { /Fortnite.com/Characters }
using { /Fortnite.com/Devices }
using { /Verse.org/Concurrency }
using { /Verse.org/Random }
using { /Verse.org/SceneGraph }
using { /Verse.org/Simulation }
using { /UnrealEngine.com/Temporary/SpatialMath }
using { /UnrealEngine.com/Temporary/UI }
using { WeaponAbilities }

<# ----------------------------------------------------------------------------------------------------
    Constants
---------------------------------------------------------------------------------------------------- #>
HealingAuraBaseRange:float = 300.0

<# ----------------------------------------------------------------------------------------------------
    Data Definitions
---------------------------------------------------------------------------------------------------- #>
ability_damage_args := struct:
    Amount:float
    IsCriticalHit:logic = false

dash_mode := enum:
    InputTrigger
    Velocity

<# ----------------------------------------------------------------------------------------------------
    Global Functions
---------------------------------------------------------------------------------------------------- #>
var GlobalAbilityManager:weak_map(session, ability_manager) = map{}
var GlobalSimulationEntity:weak_map(session, entity) = map{}

GetAbilityManager()<decides><transacts>:ability_manager =
    GlobalAbilityManager[GetSession()]

GetSimulationEntity<public>()<decides><transacts>:entity =
    GlobalSimulationEntity[GetSession()]

(Player:player).GrantAbilityUpgrade():void =
    if (AbilityManager := GetAbilityManager[]):
        AbilityManager.GrantUpgrade(Player)

(Player:player).ShowAbilitySelectUI(Level:int, ?IsLegendary:logic = false):void =
    if (AbilityManager := GetAbilityManager[]):
        AbilityManager.ShowAbilitySelectUI(Player, Level, ?IsLegendary := IsLegendary)

(Player:player).ShowWeaponEvolutionUI(Id:weapon_id, EvolutionLevel:int, ?IsItemShop:logic = false, ?IsPreview:logic = false):void =
    if (AbilityManager := GetAbilityManager[]):
        AbilityManager.ShowWeaponEvolutionUI(Player, Id, ?IsItemShop := IsItemShop, ?IsPreview := IsPreview)

(Player:player).ShowBoonSelectUI():void =
    if (AbilityManager := GetAbilityManager[]):
        AbilityManager.ShowBoonSelectUI(Player)

(Player:player).ActivateAbility():void =
    if (AbilityManager := GetAbilityManager[]):
        AbilityManager.ActivateAbility(Player)

(Player:player).ActivateNewAbility(Id:weapon_id):void =
    if (AbilityManager := GetAbilityManager[]):
        AbilityManager.ActivateNewAbility(Player, Id)

(Player:player).DeactivateAbility(?IsDeinitialize:logic = false):void =
    if (AbilityManager := GetAbilityManager[]):
        AbilityManager.DeactivateAbility(Player, ?IsDeinitialize := IsDeinitialize)

(Player:player).DeactivateAbility():void =
    Player.DeactivateAbility(?IsDeinitialize := false)

(Player:player).ReactivateAbility():void =
    if (AbilityManager := GetAbilityManager[]):
        AbilityManager.ReactivateAbility(Player)

(Player:player).ToggleAbility():void =
    if (AbilityManager := GetAbilityManager[]):
        AbilityManager.ToggleAbility(Player)

(Player:player).GetDamageArgs(Target:npc_behavior_base, Id:weapon_id, WeaponDamage:float, ?Multiplier:float = 1.0):ability_damage_args =
    DamageMultplier := 1.0 + Multiplier +
        if (Target.GetIsBoss()?):
            Player.GetStat(stat_id.BossDamage) + (Player.HasBoon(boon_id.BossHunter)? and BoonBossHunterMultiplier or 0.0)
        else:
            0.0
    
    DamageAmount := WeaponDamage * DamageMultplier

    CriticalRate := Player.GetStat(stat_id.CriticalRate)
    IsCriticalHit := logic{CriticalRate >= 1.0 or GetRandomFloat(0.0, 1.0) <= CriticalRate}
    FinalDamageAmount := if (IsCriticalHit?):
        CriticalMultiplier := Player.GetStat(stat_id.CriticalDamage)
        DamageAmount * (1.0 + CriticalMultiplier)
    else:
        DamageAmount

    ability_damage_args:
        Amount := FinalDamageAmount
        IsCriticalHit := IsCriticalHit

operator'*'(DamageArges:ability_damage_args, Multiplier:float)<transacts>:ability_damage_args =
    ability_damage_args:
        Amount := DamageArges.Amount * Multiplier
        IsCriticalHit := DamageArges.IsCriticalHit

GetPlayerActiveAbilities()<transacts>:[player][weapon_id]ability =
    GetAbilityManager[].PlayerActiveAbilities or map{}

(Player:player).GetPlayerActiveAbility(Id:weapon_id)<decides><transacts>:ability =
    GetPlayerActiveAbilities()[Player][Id]

(Player:player).GetWeaponDPS(WeaponId:weapon_id, ?EndTime:?float = false)<transacts>:float =
    GetAbilityManager[].GetWeaponDPS(Player, WeaponId, ?EndTime := EndTime) or 0.0

(Player:player).GetWeaponTotalDamage(WeaponId:weapon_id)<transacts>:float =
    GetAbilityManager[].GetWeaponTotalDamage(Player, WeaponId) or 0.0

(Player:player).GetWeaponDuration(WeaponId:weapon_id, ?EndTime:?float = false)<transacts>:float =
    GetAbilityManager[].GetWeaponDuration(Player, WeaponId, ?EndTime := EndTime) or 0.0

(NPCBehavior:npc_behavior_base).ActivateMovementModulator(MoveSpeed:float):void =   
    if (AbilityManager := GetAbilityManager[], Agent := NPCBehavior.GetAgent[]):
        AbilityManager.UpdateMoveSpeed(Agent, MoveSpeed)

(NPCBehavior:npc_behavior_base).DeactivateMovementModulator():void =
    if (Agent := NPCBehavior.GetAgent[]):
        Agent.DeactivateMovespeed()

(Player:player).UpdateMaxHealth():void =
    if (AbilityManager := GetAbilityManager[]):
        AbilityManager.UpdateMaxHealth(Player)

(Player:player).ActivateHealingAura():void =
    if (AbilityManager := GetAbilityManager[]):
        AbilityManager.ActivateHealingAura(Player)

(Player:player).Heal(Amount:type{_X:float where 0.000000 <= _X, _X <= 1.000000}):void =
    HealedAmount := Player.HealPercent(Amount)
    Player.PlayAudio(audio_id.Heal)
    Player.ActivateVFXPowerup(vfx_powerup_id.Heal, ?Duration := 0.5)

    if (Character := Player.GetFortCharacter[]):
        if (VFXManager := GetVFXManager[]):
            SpawnNumberVFX(VFXManager.PlayerHealNumberParticleSystem, Character.GetTransform().Translation + vector3{Z := 100.0}, HealedAmount)

(Agent:agent).DeactivateMovespeed():void =
    if (AbilityManager := GetAbilityManager[]):
        AbilityManager.MovementModulatorDisable.Activate(Agent)

(Player:player).ReactivateMovespeed():void =
    if (AbilityManager := GetAbilityManager[]):
        AbilityManager.UpdateMoveSpeed(Player)

<# ----------------------------------------------------------------------------------------------------
    ability_manager
---------------------------------------------------------------------------------------------------- #>
ability_manager := class(creative_device):
    @editable
    HealthPowerup<private>:health_powerup_device = health_powerup_device{}
    
    @editable
    HealthPowerupDisable<private>:health_powerup_device = health_powerup_device{}

    @editable
    MovementModulators:[]movement_modulator_device = array{}

    @editable
    MovementModulatorDisable:movement_modulator_device = movement_modulator_device{}

    @editable
    InputTriggerDash:input_trigger_device = input_trigger_device{}
    
    @editable
    PlayerMovementSettings:player_movement_settings_device = player_movement_settings_device{}

    var PlayerAbilityUpgradeUI:[player]ability_upgrade_ui = map{}
    var PlayerAbilitySelectUI:[player]ability_select_ui = map{}
    var PlayerBoonSelectUI:[player]boon_select_ui = map{}
    var PlayerWeaponEvolutionUI:[player]weapon_evolution_ui = map{}
    var PlayerDashAbility:[player]ability_dash = map{}
    var PlayerAbilityHUD:[player]ability_hud = map{}
    var PlayerIsInitialized:[player]logic = map{}
    var PlayerActiveAbilities:[player][weapon_id]ability = map{}
    var PlayerAbilityUpgradeCounter:[player]int = map{}
    var PlayerHealingAuraCancelEvents:[player]event(void) = map{}

    var WeaponAbilityInfo:[]ability_info = array{}

    <# ----------------------------------------------------------------------------------------------------
        Initialization
    ---------------------------------------------------------------------------------------------------- #>
    OnBegin<override>()<suspends>:void =
        if (set GlobalAbilityManager[GetSession()] = Self) {}

        if:
            SimulationEntity := Self.GetSimulationEntity[]
            set GlobalSimulationEntity[GetSession()] = SimulationEntity

        Playspace := GetPlayspace()
        Playspace.PlayerAddedEvent().Subscribe(OnPlayerAdded)
        Playspace.PlayerRemovedEvent().Subscribe(OnPlayerRemoved)

        InputTriggerDash.PressedEvent.Subscribe(OnPressedDash)

        set WeaponAbilityInfo = GetWeaponAbilityInfo()

        for (Player : Playspace.GetPlayers()):
            spawn{Initialize(Player)}

        WaitForGameManager()
        if (GameManager := GetGameManager[]):
            GameManager.GameStartedEvent.Subscribe(OnGameStarted)
            GameManager.GameEndedEvent.Subscribe(OnGameEnded)

        WaitForInputManager()
        if:
            InputOpenTreasureChestPressedEvent := GetInputPressedEvent[input_id.OpenTreasureChest]
            InputAutoAimPressedEvent := GetInputPressedEvent[input_id.AutoAim]
        then:
            InputOpenTreasureChestPressedEvent.Subscribe(OnInputPressedOpenTreasureChest)
            InputAutoAimPressedEvent.Subscribe(OnInputPressedAutoAim)

    Initialize(Player:player)<suspends>:void =
        if (PlayerIsInitialized[Player]?) then return
        if (set PlayerIsInitialized[Player] = true) {}

        WaitForPlayerManager(Player)
        if (PlayerStats := Player.GetStats[]):
            spawn{AwaitAbilityChanged(Player, PlayerStats.AbilityChangedEvent)}
            spawn{AwaitStatsReset(Player, PlayerStats.StatsResetEvent)}

        AbilityDash := ability_dash{Player := Player}
        AbilityDash.Initialize()
        if (set PlayerDashAbility[Player] = AbilityDash) {}

        AbilityHUD := ability_hud{Player := Player}
        AbilityHUD.Initialize()
        if (set PlayerAbilityHUD[Player] = AbilityHUD) {}

        WaitForInputManager()

    <# ----------------------------------------------------------------------------------------------------
        Get Functions
    ---------------------------------------------------------------------------------------------------- #>
    GetAbility(Player:player, Id:weapon_id)<transacts>:ability =
        case (Id):
            weapon_id.Sword => ability_sword_sg{Player := Player}
            weapon_id.Bow => ability_bow_sg{Player := Player}
            weapon_id.Shuriken => ability_shuriken_sg{Player := Player}
            weapon_id.Sentry => ability_sentry_sg{Player := Player}
            weapon_id.Aura => ability_aura{Player := Player}
            weapon_id.Torrent => ability_torrent_sg{Player := Player}
            weapon_id.Lightning => ability_lightning_sg{Player := Player}
            weapon_id.Trail => ability_trail_sg{Player := Player}
            weapon_id.Shield => ability_shield_sg{Player := Player}
            weapon_id.ChainLightning => ability_chain_lightning_sg{Player := Player}
            weapon_id.EnergyBomb => ability_energy_bomb_sg{Player := Player}
            weapon_id.Boomerang => ability_boomerang_sg{Player := Player}
            weapon_id.Kunai => ability_kunai_sg{Player := Player}
            weapon_id.EnergyBeam => ability_energy_beam_sg{Player := Player}
            weapon_id.DancingDagger => ability_dancing_dagger_sg{Player := Player}
            weapon_id.TeslaCoil => ability_tesla_coil_sg{Player := Player}
            weapon_id.Hammer => ability_hammer_sg{Player := Player}
            weapon_id.Staff => ability_staff_sg{Player := Player}

    GetWeaponAbilityInfo(Player:player)<transacts>:[]ability_info =
        for:
            AbilityInfo : WeaponAbilityInfo
            WeaponId := AbilityInfo.MaybeWeaponId?
            Player.HasWeapon(WeaponId)?
            not AbilityInfo.MaybeMinLevel? or GetLevel() >= AbilityInfo.MaybeMinLevel?
            not AbilityInfo.Id = ability_id.WeaponLevel or (AbilityInfo.Id = ability_id.WeaponLevel and Player.GetWeaponLevel(WeaponId) < MaxWeaponLevel - 1)
        do:
            AbilityInfo

    GetWeaponAbilityInfo(WeaponId:weapon_id, AbilityId:ability_id)<transacts>:?ability_info =
        var WeaponAbilityInfoResult:?ability_info = false
        for (AbilityInfo : WeaponAbilityInfo, not WeaponAbilityInfoResult?):
            if (WeaponId = AbilityInfo.MaybeWeaponId?, AbilityId = AbilityInfo.Id):
                set WeaponAbilityInfoResult = option{AbilityInfo}

        WeaponAbilityInfoResult

    GetWeaponAbilityInfo()<transacts>:[]ability_info =
        Concatenate of array:
            for (WeaponInfo : WeaponInfos):
                ability_info:
                    MaybeWeaponId := option{WeaponInfo.Id}
                    Id := ability_id.Damage
                    Name := "{WeaponInfo.Id} Damage"
                    Description := "Increase damage of {WeaponInfo.Id} weapon."
                    Icon := WeaponInfo.Icon
                    IconSmall := WeaponInfo.IconSmall
                    MaybeStatIcon := option{UserInterface.Icons.ProjectileDamage_64}
                    StatIconColor := ThemeColors.WeaponDamage
                    TierValue := array{option{0.1}, option{0.15}, option{0.25}, option{0.35}, option{0.5}}
                    MaybeStatId := option{stat_id.Damage}
                    FormatValue := FormatAsPercent
    
            for (WeaponInfo : WeaponInfos):
                ability_info:
                    MaybeWeaponId := option{WeaponInfo.Id}
                    Id := ability_id.Cooldown
                    Name := (WeaponInfo.Id = weapon_id.Trail) and "{WeaponInfo.Id} Duration" or "{WeaponInfo.Id} Cooldown"
                    Description := (WeaponInfo.Id = weapon_id.Trail) and "Increase duration of {WeaponInfo.Id} weapon." or "Reduce cooldown of {WeaponInfo.Id} weapon."
                    Icon := WeaponInfo.Icon
                    IconSmall := WeaponInfo.IconSmall
                    MaybeStatIcon := option{UserInterface.Icons.AbilityCooldown_64}
                    StatIconColor := ThemeColors.WeaponCooldown
                    TierValue := array{option{0.1}, option{0.15}, option{0.25}, option{0.35}, option{0.5}}
                    MaybeStatId := option{stat_id.Cooldown}
                    FormatValue := FormatAsPercent
    
            for (WeaponInfo : WeaponInfos):
                ability_info:
                    MaybeWeaponId := option{WeaponInfo.Id}
                    Id := ability_id.Size
                    Name := "{WeaponInfo.Id} Size"
                    Description := "Increase size of {WeaponInfo.Id} weapon."
                    Icon := WeaponInfo.Icon
                    IconSmall := WeaponInfo.IconSmall
                    MaybeStatIcon := option{UserInterface.Icons.AbilitySize_64}
                    StatIconColor := ThemeColors.WeaponSize
                    TierValue := array{option{0.1}, option{0.15}, option{0.25}, option{0.35}, option{0.5}}
                    MaybeStatId := option{stat_id.Size}
                    FormatValue := FormatAsPercent

            for (WeaponInfo : WeaponInfos):
                ability_info:
                    MaybeWeaponId := option{WeaponInfo.Id}
                    Id := ability_id.WeaponLevel
                    Name := "{WeaponInfo.Id} Level"
                    Description := "Increase the level of {WeaponInfo.Id} weapon."
                    Icon := WeaponInfo.Icon
                    IconSmall := WeaponInfo.IconSmall
                    MaybeStatIcon := option{UserInterface.Icons.WeaponLevel_64}
                    StatIconColor := ThemeColors.WeaponLevel
                    TierValue := array{false, false, option{1.0}, option{2.0}, false}
                    MaybeMinLevel := option{6}
                    FormatValue := FormatAsInt

    GetWeaponDPS(Player:player, WeaponId:weapon_id, ?EndTime:?float = false)<transacts>:float =
        PlayerActiveAbilities[Player][WeaponId].GetDPS(?EndTime := EndTime) or 0.0

    GetWeaponTotalDamage(Player:player, WeaponId:weapon_id)<transacts>:float =
        PlayerActiveAbilities[Player][WeaponId].GetTotalDamage() or 0.0

    GetWeaponDuration(Player:player, WeaponId:weapon_id, ?EndTime:?float = false)<transacts>:float =
        PlayerActiveAbilities[Player][WeaponId].GetDuration(?EndTime := EndTime) or 0.0

    <# ----------------------------------------------------------------------------------------------------
        Update Functions
    ---------------------------------------------------------------------------------------------------- #>
    ShowAbilitySelectUI(Player:player, Level:int, ?IsItemShop:logic = false, ?IsLegendary:logic = false):void =
        UI := if (AbilitySelectUI := PlayerAbilitySelectUI[Player]):
            AbilitySelectUI
        else:
            AbilitySelectUI := ability_select_ui{Player := Player}
            if (set PlayerAbilitySelectUI[Player] = AbilitySelectUI) {}
            AbilitySelectUI

        UI.Show()
        UI.OnShow(Level, ?InIsLegendary := IsLegendary)

    ShowWeaponEvolutionUI(Player:player, Id:weapon_id, ?IsItemShop:logic = false, ?IsPreview:logic = false):void =
        UI := if (WeaponEvolutionUI := PlayerWeaponEvolutionUI[Player]):
            WeaponEvolutionUI
        else:
            WeaponEvolutionUI := weapon_evolution_ui{Player := Player}
            if (set PlayerWeaponEvolutionUI[Player] = WeaponEvolutionUI) {}
            WeaponEvolutionUI

        UI.Show()
        UI.OnShow(Id, ?IsItemShop := IsItemShop, ?InIsPreview := IsPreview)

    GrantUpgrade(Player:player):void =
        UI := if (AbilityUpgradeUI := PlayerAbilityUpgradeUI[Player]):
            AbilityUpgradeUI
        else:
            AbilityUpgradeUI := ability_upgrade_ui{Player := Player}
            if (set PlayerAbilityUpgradeUI[Player] = AbilityUpgradeUI) {}
            AbilityUpgradeUI
            
        if (RandomAbility := GetTreasureChestRandomAbility(Player)?):
            UI.Show()
            UI.Update(RandomAbility)

    ShowBoonSelectUI(Player:player):void =
        BoonSelectUI := if (UI := PlayerBoonSelectUI[Player]):
            UI
        else:
            UI := boon_select_ui{Player := Player}
            if (set PlayerBoonSelectUI[Player] = UI) {}
            UI

        BoonSelectUI.Show()

    ActivateAbility(Player:player):void =
        if (EquippedWeapon := Player.GetEquippedWeapon()?):
            if (PlayerActiveAbilities[Player][EquippedWeapon]) then return
            Ability := GetAbility(Player, EquippedWeapon)
            Ability.Initialize()
            Ability.Activate()
            if (set PlayerActiveAbilities[Player] = map{EquippedWeapon => Ability}) {}

    ActivateNewAbility(Player:player, Id:weapon_id):void =
        if (PlayerActiveAbilities[Player][Id]) then return
        Ability := GetAbility(Player, Id)
        Ability.Initialize()

        if (Character := Player.GetFortCharacter[], not Character.IsDownButNotOut[], not Player.IsPlayerPaused()?):
            Ability.Activate()
        
        if (AbilityHUD := PlayerAbilityHUD[Player], AbilityIndex := PlayerActiveAbilities[Player].Length, WeaponInfo := GetWeaponInfo[Id]):
            AbilityHUD.Update(AbilityIndex, WeaponInfo)

        if:
            (set PlayerActiveAbilities[Player][Id] = Ability) or
            (set PlayerActiveAbilities[Player] = map{Id => Ability})

    DeactivateAbility(Player:player, ?IsDeinitialize:logic = false):void =
        for (ActiveAbility : PlayerActiveAbilities[Player]):
            if (IsDeinitialize?):
                ActiveAbility.Deinitialize()
            else:
                ActiveAbility.Deactivate()

    ReactivateAbility(Player:player):void =
        if (Character := Player.GetFortCharacter[], Character.IsDownButNotOut[]) then return
        if (Player.IsInStageVolume()?) then return

        for (ActiveAbility : PlayerActiveAbilities[Player], not ActiveAbility.IsActive?):
            ActiveAbility.Activate()

    ToggleAbility(Player:player):void =
        if (ActiveAbilities := PlayerActiveAbilities[Player]):
            for (ActiveAbility : ActiveAbilities):
                if (ActiveAbility.GetIsActive()?):
                    DeactivateAbility(Player)
                else:
                    ReactivateAbility(Player)

        else if (EquippedWeapon := Player.GetEquippedWeapon()?):
            ActivateNewAbility(Player, EquippedWeapon)

    UpdateMaxHealth(Player:player):void =
        if (Character := Player.GetFortCharacter[]):
            MaxHealth := Player.GetStat(stat_id.MaxHealth)
            Character.SetMaxHealth(MaxHealth)
            Character.SetMaxShield(MaxHealth)

    UpdateHealthRegeneration(Player:player):void =
        HealthPowerup.Spawn()
        HealthRegeneration := Player.GetStat(stat_id.HealthRegeneration)
        HealthPowerup.SetMagnitude(HealthRegeneration)
        HealthPowerup.Pickup(Player)

    UpdateMoveSpeed(Player:player):void =
        MoveSpeed := Player.GetStat(stat_id.MoveSpeed)
        UpdateMoveSpeed(Player, MoveSpeed)

    UpdateMoveSpeed(Agent:agent, MoveSpeed:float):void =
        if (MoveSpeed = 0.0):
            MovementModulatorDisable.Activate(Agent)
        else if (Index := Floor[MoveSpeed * 10.0], MovementModulator := MovementModulators[Clamp(Index - 1, 0, MovementModulators.Length - 1)]):
            MovementModulator.Activate(Agent)

    GrantAbilityUpgrade(Player:player)<suspends>:void =
        if (set PlayerAbilityUpgradeCounter[Player] = 1) {}
        Player.ActivateVFXPowerup(vfx_powerup_id.TreasureChestOpen, ?Duration := 1.0)
        
        Sleep(1.0)
        for (Index := 0..PlayerAbilityUpgradeCounter[Player] - 1):
            PauseGame(Player, pause_info{Type := pause_type.TreasureChest})

        if (set PlayerAbilityUpgradeCounter[Player] = 0) {}

    ActivateHealingAura(Player:player):void =
        if (HealingAuraCancelEvent := PlayerHealingAuraCancelEvents[Player]):
            HealingAuraCancelEvent.Signal()
        
        CancelEvent := event(void){}
        if (set PlayerHealingAuraCancelEvents[Player] = CancelEvent) {}
        spawn{HealingAura(Player, CancelEvent)}
            
    HealingAura(Player:player, CancelEvent:event(void))<suspends>:void =
        HealablePlayers := for (ActivePlayer : GetActivePlayers(), ActivePlayer <> Player) {ActivePlayer}

        if (Character := Player.GetFortCharacter[]):
            race:
                CancelEvent.Await()

                Sleep(Player.GetPowerUpValue(power_up_id.HealthPickupDuration))

                loop:
                    Sleep(1.0)
                    if (Character.IsDownButNotOut[]):
                        break

                    Player.ActivateVFXPowerup(vfx_powerup_id.Heal, ?Duration := 0.5)
                    Player.PlayAudio(audio_id.HealAura)
                    Range := HealingAuraBaseRange * (1.0 + Player.GetPowerUpValue(power_up_id.HealthPickupSize))

                    for:
                        HealablePlayer : HealablePlayers
                        HealableCharacter := HealablePlayer.GetFortCharacter[]
                        HealableCharacter.GetHealth() < HealableCharacter.GetMaxHealth()
                        DistanceSquaredXY(HealableCharacter.GetTransform().Translation, Character.GetTransform().Translation) <= Range * Range
                    do:
                        HealablePlayer.Heal(0.1)

        set PlayerHealingAuraCancelEvents = RemoveKeyFromMap(PlayerHealingAuraCancelEvents, Player)
    
    <# ----------------------------------------------------------------------------------------------------
        Event Handlers
    ---------------------------------------------------------------------------------------------------- #>
    OnPlayerAdded(Player:player):void =
        spawn{Initialize(Player)}

    OnPlayerRemoved(Player:player):void =
        if (DashAbility := PlayerDashAbility[Player]):
            DashAbility.Uninitialize()

        if (ActiveAbilities := PlayerActiveAbilities[Player]):
            for (ActiveAbility : ActiveAbilities):
                ActiveAbility.Deactivate()
        
        if (AbilityHUD := PlayerAbilityHUD[Player]):
            AbilityHUD.Uninitialize()

        if (AbilitySelectUI := PlayerAbilitySelectUI[Player]):
            AbilitySelectUI.Uninitialize()

        if (WeaponEvolutionUI := PlayerWeaponEvolutionUI[Player]):
            WeaponEvolutionUI.Uninitialize()

        if (AbilityUpgradeUI := PlayerAbilityUpgradeUI[Player]):
            AbilityUpgradeUI.Uninitialize()

        if (BoonSelectUI := PlayerBoonSelectUI[Player]):
            BoonSelectUI.Uninitialize()

        if (HealingAuraCancelEvent := PlayerHealingAuraCancelEvents[Player]):
            HealingAuraCancelEvent.Signal()

        set PlayerAbilityUpgradeUI = RemoveKeyFromMap(PlayerAbilityUpgradeUI, Player)
        set PlayerAbilitySelectUI = RemoveKeyFromMap(PlayerAbilitySelectUI, Player)
        set PlayerWeaponEvolutionUI = RemoveKeyFromMap(PlayerWeaponEvolutionUI, Player)
        set PlayerBoonSelectUI = RemoveKeyFromMap(PlayerBoonSelectUI, Player)
        set PlayerDashAbility = RemoveKeyFromMap(PlayerDashAbility, Player)
        set PlayerActiveAbilities = RemoveKeyFromMap(PlayerActiveAbilities, Player)
        set PlayerAbilityHUD = RemoveKeyFromMap(PlayerAbilityHUD, Player)
        set PlayerAbilityUpgradeCounter = RemoveKeyFromMap(PlayerAbilityUpgradeCounter, Player)
        set PlayerIsInitialized = RemoveKeyFromMap(PlayerIsInitialized, Player)
        set PlayerHealingAuraCancelEvents = RemoveKeyFromMap(PlayerHealingAuraCancelEvents, Player)

    AwaitAbilityChanged(Player:player, AbilityChangedEvent:events(ability_id))<suspends>:void =
        if (PlayerRemovedEvent := Player.GetPlayerRemovedEvent[]):
            race:
                PlayerRemovedEvent.Await()
                loop:
                    AbilityId := AbilityChangedEvent.Await()

                    case (AbilityId):
                        ability_id.HealthRegeneration => UpdateHealthRegeneration(Player)
                        ability_id.MaxHealth => UpdateMaxHealth(Player)
                        ability_id.MoveSpeed =>
                            if (not (ObjectiveSnowStorm := objective_snow_storm[GetCurrentObjective()?], ObjectiveSnowStorm.GetIsActive()?)):
                                UpdateMoveSpeed(Player)

                        _ => {}

    AwaitStatsReset(Player:player, StatsResetEvent:events(tuple()))<suspends>:void =
        if (PlayerRemovedEvent := Player.GetPlayerRemovedEvent[]):
            race:
                PlayerRemovedEvent.Await()
                loop:
                    StatsResetEvent.Await()
                    if (Character := Player.GetFortCharacter[]):
                        Character.SetMaxHealth(100.0)
                        Character.SetHealth(100.0)
                        Character.SetMaxShield(100.0)
                        Character.SetShield(0.0)
                        HealthPowerupDisable.Pickup(Player)
                        MovementModulatorDisable.Activate(Player)

    OnPressedDash(Agent:agent):void =
        if (Player := player[Agent], DashAbility := PlayerDashAbility[Player]):            
            # [Hotfix]
            if (Player.IsPlayerPaused()?, GetSecondsSinceEpoch() - Player.GetPauseTime() >= 1.0, CurrentPause := Player.GetCurrentPause[]):
                Player.ShowPauseUI(CurrentPause)
                Player.PlayAudio(audio_id.Activate)
            else:
                DashAbility.Activate()

    OnGameStarted():void =
        for (Player : GetActivePlayers(), DashAbility := PlayerDashAbility[Player]):
            DashAbility.Reset()
            UpdateMaxHealth(Player)
            UpdateHealthRegeneration(Player)
            UpdateMoveSpeed(Player)

    OnGameEnded():void =
        for (Player : GetAllPlayers()):
            Player.DeactivateAbility(?IsDeinitialize := true)

        set PlayerActiveAbilities = map{}

    OnInputPressedOpenTreasureChest(Agent:agent):void =
        if (Player := player[Agent]):
            if (Player.IsInStageVolume()?) then return

            if (Player.GetGameCurrency(game_currency_id.TreasureChest) > 0.0):
                if (PlayerAbilityUpgradeCounter[Player] > 0):
                    if (set PlayerAbilityUpgradeCounter[Player] += 1) {}
                else:
                    spawn{GrantAbilityUpgrade(Player)}

                Player.RemoveGameCurrency(game_currency_id.TreasureChest, 1.0)
                Player.PlayAudio(audio_id.OpenTreasureChest)

    OnInputPressedAutoAim(Agent:agent):void =
        if (Player := player[Agent]):
            Player.ToggleAutoAim()
            Player.RefreshControls()
            Player.PlayAudio(audio_id.Activate)
            Player.ShowMessage(message_id.AutoAim, "{Player.GetData[].Settings.AutoAim? and "ON" or "OFF"}", ?DisplayTime := 3.0)