using { /Fortnite.com/UI }
using { /Verse.org/Colors }
using { /UnrealEngine.com/Temporary/SpatialMath }
using { /UnrealEngine.com/Temporary/UI }

<# ----------------------------------------------------------------------------------------------------
    ability_select_tier_rates_ui
---------------------------------------------------------------------------------------------------- #>
ability_select_tier_rates_ui := class(ui_component):
    ItemShopLuckText:text_block = text_block:
        DefaultTextColor := ThemeColors.Success       
        DefaultShadowColor := ThemeColors.Success * 0.125
        DefaultShadowOffset := option{vector2{X := 2.0, Y := 2.0}}
        DefaultShadowOpacity := 1.0

    var ItemShopInfoUI:widget = canvas{}
    var TierTexts:[]text_block = array{}
    var TierPreviewTexts:[]text_block = array{}
        
    <# ----------------------------------------------------------------------------------------------------
        Initialization
    ---------------------------------------------------------------------------------------------------- #>
    CreateUI<override>():widget =
        overlay:
            Slots := array:
                overlay_slot:
                    HorizontalAlignment := horizontal_alignment.Fill
                    VerticalAlignment := vertical_alignment.Fill
                    Widget := UserInterface.VerseWidget.WBP_Panel_Background{}
                overlay_slot:
                    HorizontalAlignment := horizontal_alignment.Fill
                    VerticalAlignment := vertical_alignment.Fill
                    Padding := margin{Left := 40.0, Top := 40.0, Right := 40.0, Bottom := 40.0}
                    Widget := stack_box:
                        Orientation := orientation.Vertical
                        Slots := array:
                            stack_box_slot:
                                HorizontalAlignment := horizontal_alignment.Center
                                VerticalAlignment := vertical_alignment.Fill
                                Padding := margin{Left := 0.0, Top := 0.0, Right := 0.0, Bottom := 10.0}
                                Widget := text_block:
                                    DefaultText := MakeMessage("Tier Rates")
                                    DefaultTextColor := ThemeColors.PrimaryText
                            stack_box_slot:
                                HorizontalAlignment := horizontal_alignment.Fill
                                VerticalAlignment := vertical_alignment.Top
                                Padding := margin{Left := 0.0, Top := 0.0, Right := 0.0, Bottom := 40.0}
                                Widget := color_block:
                                    DefaultColor := NamedColors.White
                                    DefaultOpacity := 0.1
                                    DefaultDesiredSize := vector2{X := 0.0, Y := 1.0}
                            stack_box_slot:
                                HorizontalAlignment := horizontal_alignment.Fill
                                VerticalAlignment := vertical_alignment.Fill
                                Padding := margin{Left := 0.0, Top := 0.0, Right := 0.0, Bottom := 20.0}
                                Widget := set ItemShopInfoUI = CreateItemShopInfoUI()
                            stack_box_slot:
                                HorizontalAlignment := horizontal_alignment.Fill
                                VerticalAlignment := vertical_alignment.Fill
                                Padding := margin{Left := 0.0, Top := 0.0, Right := 0.0, Bottom := 20.0}
                                Widget := overlay:
                                    Slots := array:
                                        overlay_slot:
                                            HorizontalAlignment := horizontal_alignment.Fill
                                            VerticalAlignment := vertical_alignment.Fill
                                            Widget := UserInterface.VerseWidget.WBP_RoundedBox{}
                                        overlay_slot:
                                            HorizontalAlignment := horizontal_alignment.Fill
                                            VerticalAlignment := vertical_alignment.Fill
                                            Padding := margin{Left := 40.0, Top := 40.0, Right := 40.0, Bottom := 40.0}
                                            Widget := stack_box:
                                                Orientation := orientation.Vertical
                                                Slots := for (Index := 1..5):
                                                    TierText := text_block:
                                                        DefaultText := MakeMessage("0.00%")
                                                        DefaultTextColor := ThemeColors.Success
                                                        DefaultShadowColor := ThemeColors.Success * 0.125
                                                        DefaultShadowOffset := option{vector2{X := 2.0, Y := 2.0}}
                                                        DefaultShadowOpacity := 1.0

                                                    TierPreviewText := text_block:
                                                        DefaultText := MakeMessage("▶ 0.00%")
                                                        DefaultTextColor := ThemeColors.Success
                                                        DefaultShadowColor := ThemeColors.Success * 0.125
                                                        DefaultShadowOffset := option{vector2{X := 2.0, Y := 2.0}}
                                                        DefaultShadowOpacity := 1.0

                                                    set TierTexts += array{TierText}
                                                    set TierPreviewTexts += array{TierPreviewText}

                                                    stack_box_slot:
                                                        HorizontalAlignment := horizontal_alignment.Fill
                                                        VerticalAlignment := vertical_alignment.Fill
                                                        Widget := stack_box:
                                                            Orientation := orientation.Horizontal
                                                            Slots := array:
                                                                stack_box_slot:
                                                                    HorizontalAlignment := horizontal_alignment.Left
                                                                    VerticalAlignment := vertical_alignment.Fill
                                                                    Distribution := option{1.0}
                                                                    Widget := text_block:
                                                                        DefaultText := MakeMessage(Index.ToTierString().ToUpper())
                                                                        DefaultTextColor := Index.ToTierColor() * 2.0
                                                                        DefaultShadowColor := Index.ToTierColor() * 0.25
                                                                        DefaultShadowOffset := option{vector2{X := 2.0, Y := 2.0}}
                                                                        DefaultShadowOpacity := 1.0
                                                                stack_box_slot:
                                                                    HorizontalAlignment := horizontal_alignment.Left
                                                                    VerticalAlignment := vertical_alignment.Fill
                                                                    Distribution := option{1.0}
                                                                    Padding := margin{Left := 100.0, Top := 0.0, Right := 0.0, Bottom := 0.0}
                                                                    Widget := TierText
                                                                stack_box_slot:
                                                                    HorizontalAlignment := horizontal_alignment.Left
                                                                    VerticalAlignment := vertical_alignment.Fill
                                                                    Distribution := option{1.0}
                                                                    Widget := stack_box:
                                                                        Orientation := orientation.Horizontal
                                                                        Slots := array:
                                                                            stack_box_slot:
                                                                                HorizontalAlignment := horizontal_alignment.Fill
                                                                                VerticalAlignment := vertical_alignment.Fill
                                                                                Widget := text_block:
                                                                                    DefaultText := MakeMessage("▶")
                                                                                    DefaultTextColor := ThemeColors.PrimaryText
                                                                                    DefaultShadowOffset := option{vector2{X := 2.0, Y := 2.0}}
                                                                                    DefaultShadowOpacity := 0.5
                                                                            stack_box_slot:
                                                                                HorizontalAlignment := horizontal_alignment.Fill
                                                                                VerticalAlignment := vertical_alignment.Fill
                                                                                Padding := margin{Left := 10.0, Top := 0.0, Right := 0.0, Bottom := 0.0}
                                                                                Widget := TierPreviewText

    CreateItemShopInfoUI():widget =
        stack_box:
            Orientation := orientation.Vertical
            Slots := array:
                stack_box_slot:
                    HorizontalAlignment := horizontal_alignment.Fill
                    VerticalAlignment := vertical_alignment.Fill
                    Widget := text_block:
                        DefaultText := MakeMessage("Item Shop ignores player's luck.\nItem Shop luck scales with stages.")
                        DefaultTextColor := ThemeColors.PrimaryText
                        DefaultShadowColor := NamedColors.Black
                        DefaultShadowOffset := option{vector2{X := 2.0, Y := 2.0}}
                        DefaultShadowOpacity := 0.5
                stack_box_slot:
                    HorizontalAlignment := horizontal_alignment.Fill
                    VerticalAlignment := vertical_alignment.Fill
                    Padding := margin{Left := 0.0, Top := 20.0, Right := 0.0, Bottom := 0.0}
                    Widget := overlay:
                        Slots := array:
                            overlay_slot:
                                HorizontalAlignment := horizontal_alignment.Fill
                                VerticalAlignment := vertical_alignment.Fill
                                Widget := UserInterface.VerseWidget.WBP_RoundedBox{}
                            overlay_slot:
                                VerticalAlignment := vertical_alignment.Fill
                                Padding := margin{Left := 0.0, Top := 20.0, Right := 0.0, Bottom := 20.0}
                                Widget := stack_box:
                                    Orientation := orientation.Horizontal
                                    Slots := array:
                                        stack_box_slot:
                                            HorizontalAlignment := horizontal_alignment.Left
                                            Widget := texture_block:
                                                DefaultImage := UserInterface.Icons.AbilityLuck_64
                                                DefaultDesiredSize := vector2{X := 40.0, Y := 40.0}
                                        stack_box_slot:
                                            HorizontalAlignment := horizontal_alignment.Fill
                                            Padding := margin{Left := 10.0, Top := 0.0, Right := 0.0, Bottom := 0.0}
                                            Widget := ItemShopLuckText

    <# ----------------------------------------------------------------------------------------------------
        Update Functions
    ---------------------------------------------------------------------------------------------------- #>
    Update<override>(?LuckMultiplier:float = 0.0, ?MaybeAddLuckAmount:?float = false, ?LuckOverride:?float = false):void =
        ItemShopInfoUI.SetVisibility(LuckOverride? and widget_visibility.Visible or widget_visibility.Collapsed)
        
        PlayerAbilityTierWeights := if (Luck := LuckOverride?):
            ItemShopLuckText.SetText(MakeMessage("{FormatAsPercent(Luck)}"))
            CalculateAbilityTierWeights(Luck)
        else:
            Player.GetPlayerAbilityTierWeights(?LuckMultiplier := LuckMultiplier)

        TotalWeight := Sum(PlayerAbilityTierWeights)

        for (Index -> TierText : TierTexts, PlayerAbilityTierWeight := PlayerAbilityTierWeights[Index]):
            Percent := (PlayerAbilityTierWeight * 1.0) / (TotalWeight * 1.0)
            TierText.SetText(MakeMessage("{FormatAsPercent(Percent, ?Decimals := 2)}"))

        if (AddLuckAmount := MaybeAddLuckAmount?):
            PlayerAbilityTierPreviewWeights := Player.GetPlayerAbilityTierWeights(?LuckMultiplier := LuckMultiplier + AddLuckAmount)
            TotalPreviewWeight := Sum(PlayerAbilityTierPreviewWeights)

            for (Index -> TierPreviewText : TierPreviewTexts, PlayerAbilityTierPreviewWeight := PlayerAbilityTierPreviewWeights[Index], ParentWidget := TierPreviewText.GetParentWidget[]):
                ParentWidget.SetVisibility(widget_visibility.Visible)
                Percent := (PlayerAbilityTierPreviewWeight * 1.0) / (TotalPreviewWeight * 1.0)
                TierPreviewText.SetText(MakeMessage("{FormatAsPercent(Percent, ?Decimals := 2)}"))
        else:
            for (Index -> TierPreviewText : TierPreviewTexts, ParentWidget := TierPreviewText.GetParentWidget[]):
                    ParentWidget.SetVisibility(widget_visibility.Collapsed)