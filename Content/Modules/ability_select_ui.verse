using { /Fortnite.com/UI }
using { /Verse.org/Colors }
using { /Verse.org/Random }
using { /Verse.org/Simulation }
using { /UnrealEngine.com/Temporary/SpatialMath }
using { /UnrealEngine.com/Temporary/UI }

<# ----------------------------------------------------------------------------------------------------
    Constants
---------------------------------------------------------------------------------------------------- #>
AutoSelectCountAFKThreshold:int = 2
LuckyDiceAudioIds:[]audio_id = array:
    audio_id.Reroll
    audio_id.Reroll2
    audio_id.Reroll3
    audio_id.Reroll4
    audio_id.Reroll5

<# ----------------------------------------------------------------------------------------------------
    Data Definitions
---------------------------------------------------------------------------------------------------- #>
ability_select_info_type := enum:
    StatsInfo
    TierRates

<# ----------------------------------------------------------------------------------------------------
    Global Functions
---------------------------------------------------------------------------------------------------- #>
IsWeaponUnlockLevel(Level:int)<transacts>:logic =
    if (GameSettings := GetGameSettings[]):
        case (GameSettings.Weapons):
            2 => logic{Level = WeaponUnlockLevel1}
            3 => logic{Level = WeaponUnlockLevel1 or Level = WeaponUnlockLevel2}
            4 => logic{Level = WeaponUnlockLevel1 or Level = WeaponUnlockLevel2 or Level = WeaponUnlockLevel3}
            _ => false
    else:
        logic{Level = WeaponUnlockLevel1 or Level = WeaponUnlockLevel2 or Level = WeaponUnlockLevel3}

<# ----------------------------------------------------------------------------------------------------
    ability_select_ui
---------------------------------------------------------------------------------------------------- #>
ability_select_ui := class(panel_ui):
    Title<override>:string = "Select Ability"
    HasCloseButton<override>:logic = false
    BackgroundOpacity<override>:type {_X:float where 0.000000 <= _X, _X <= 1.000000} = 0.7

    LuckyDiceText:text_block = text_block:
        DefaultTextColor := ThemeColors.Success
        DefaultShadowColor := ThemeColors.Success * 0.125
        DefaultShadowOffset := option{vector2{X := 2.0, Y := 2.0}}
        DefaultShadowOpacity := 1.0

    RerollButton:button_quiet = button_quiet{}
    BanishButton:button_quiet = button_quiet{}
    BoostButton:button_quiet = button_quiet{}
    
    var InfoUIs:[ability_select_info_type]ui_component = map{}
    var AbilitySelectEntryUIs:[]ability_select_entry_ui = array{}
    var LuckyDiceUI:widget = canvas{}

    var CurrentAbilities:[]tuple(ability_info, int) = array{}
    var AbilitySelectMode:ability_select_mode = ability_select_mode.Default
    var HasBanished:logic = false
    var HasSelected:logic = false
    var HasUsedSpareDice:logic = false
    var IsAutoSelect:logic = false
    var AutoSelectCount:int = 0
    var CurrentLevel:int = 0
    var RerollCount:int = 0
    var ShowTime:float = 0.0
    var MaybeLuckAbility:?tuple(ability_info, int) = false
    var Subscriptions:[]cancelable = array{}
    var EventsSubscriptions:[]events_cancelable = array{}
    var IsLegendary:logic = false

    <# ----------------------------------------------------------------------------------------------------
        Initialization
    ---------------------------------------------------------------------------------------------------- #>
    CreateContent<override>():widget =
        if (GameManager := GetGameManager[]):
            set EventsSubscriptions = array:
                ClosedEvent.Subscribe(OnClosed)
                GameManager.ResumeCountdownStartEvent.Subscribe(OnResumeCountdownStart)
                GameManager.ResumeCountdownCompleteEvent.Subscribe(OnResumeCountdownComplete)
                GameManager.ResumeCountdownEvent.Subscribe(OnResumeCountdown)
                GameManager.GameResumedEvent.Subscribe(OnGameResumed)
        
        TierRatesButton := button_quiet{DefaultText := MakeMessage("Tier Rates")}
        StatsInfoButton := button_quiet{DefaultText := MakeMessage("Stats Info")}
        
        set Subscriptions = array:
            RerollButton.OnClick().Subscribe(OnClickReroll)
            BanishButton.OnClick().Subscribe(OnClickBanish)
            BoostButton.OnClick().Subscribe(OnClickBoost)
            TierRatesButton.OnClick().Subscribe(OnClickTierRates)
            StatsInfoButton.OnClick().Subscribe(OnClickStatsInfo)

        AbilitiesStackBox := stack_box{Orientation := orientation.Horizontal}

        set AbilitySelectEntryUIs = for (Index := 1..3):
            AbilitySelectEntryUI := ability_select_entry_ui{Player := Player}
            AbilitySelectEntryUI.Initialize()
            set EventsSubscriptions += array:
                AbilitySelectEntryUI.SelectEvent.Subscribe(OnSelect)
                AbilitySelectEntryUI.SelectWeaponEvent.Subscribe(OnSelectWeapon)

            AbilitiesStackBox.AddWidget of stack_box_slot:
                HorizontalAlignment := horizontal_alignment.Fill
                VerticalAlignment := vertical_alignment.Fill
                Padding := margin{Left := (Index = 1) and 0.0 or 20.0, Top := 0.0, Right := 0.0, Bottom := 0.0}
                Distribution := option{1.0}
                Widget := AbilitySelectEntryUI.GetUI()

            AbilitySelectEntryUI

        stack_box:
            Orientation := orientation.Vertical
            Slots := array:
                stack_box_slot:
                    HorizontalAlignment := horizontal_alignment.Fill
                    VerticalAlignment := vertical_alignment.Fill
                    Padding := margin{Left := 0.0, Top := 0.0, Right := 0.0, Bottom := 20.0}
                    Widget := stack_box:
                        Orientation := orientation.Horizontal
                        Slots := array:
                            stack_box_slot:
                                HorizontalAlignment := horizontal_alignment.Fill
                                VerticalAlignment := vertical_alignment.Fill
                                Distribution := option{1.0}
                                Padding := margin{Left := 0.0, Top := 0.0, Right := 0.0, Bottom := 0.0}
                                Widget := set LuckyDiceUI = overlay:
                                    Slots := array:
                                        overlay_slot:
                                            HorizontalAlignment := horizontal_alignment.Fill
                                            VerticalAlignment := vertical_alignment.Fill
                                            Widget := MakeBox()
                                        overlay_slot:
                                            HorizontalAlignment := horizontal_alignment.Fill
                                            VerticalAlignment := vertical_alignment.Fill
                                            Padding := margin{Left := 20.0, Top := 10.0, Right := 20.0, Bottom := 10.0}
                                            Widget := stack_box:
                                                Orientation := orientation.Horizontal
                                                Slots := array:
                                                    stack_box_slot:
                                                        HorizontalAlignment := horizontal_alignment.Left
                                                        Padding := margin{Left := 0.0, Top := 0.0, Right := 0.0, Bottom := 0.0}
                                                        Widget := texture_block:
                                                            DefaultImage := UserInterface.Icons.Boon_LuckyDice_64
                                                            DefaultDesiredSize := vector2{X := 40.0, Y := 40.0}
                                                    stack_box_slot:
                                                        HorizontalAlignment := horizontal_alignment.Fill
                                                        Padding := margin{Left := 20.0, Top := 0.0, Right := 0.0, Bottom := 0.0}
                                                        Widget := text_block:
                                                            DefaultText := MakeMessage("Bonus Luck")
                                                            DefaultTextColor := ThemeColors.PrimaryText
                                                            DefaultShadowColor := NamedColors.Black
                                                            DefaultShadowOffset := option{vector2{X := 2.0, Y := 2.0}}
                                                            DefaultShadowOpacity := 0.5
                                                    stack_box_slot:
                                                        HorizontalAlignment := horizontal_alignment.Right
                                                        Distribution := option{1.0}
                                                        Widget := LuckyDiceText
                            stack_box_slot:
                                Distribution := option{1.0}
                                Padding := margin{Left := 20.0, Top := 0.0, Right := 0.0, Bottom := 0.0}
                                Widget := color_block:
                                    DefaultOpacity := 0.0
                            stack_box_slot:
                                HorizontalAlignment := horizontal_alignment.Fill
                                Distribution := option{1.0}
                                Padding := margin{Left := 20.0, Top := 0.0, Right := 0.0, Bottom := 0.0}
                                Widget := stack_box:
                                    Orientation := orientation.Horizontal
                                    Slots := array:
                                        stack_box_slot:
                                            HorizontalAlignment := horizontal_alignment.Fill
                                            VerticalAlignment := vertical_alignment.Fill
                                            Distribution := option{1.0}
                                            Widget := StatsInfoButton
                                        stack_box_slot:
                                            HorizontalAlignment := horizontal_alignment.Fill
                                            VerticalAlignment := vertical_alignment.Fill
                                            Distribution := option{1.0}
                                            Padding := margin{Left := 20.0, Top := 0.0, Right := 0.0, Bottom := 0.0}
                                            Widget := TierRatesButton
                stack_box_slot:
                    HorizontalAlignment := horizontal_alignment.Fill
                    VerticalAlignment := vertical_alignment.Fill
                    Padding := margin{Left := 0.0, Top := 0.0, Right := 0.0, Bottom := 20.0}
                    Widget := AbilitiesStackBox
                stack_box_slot:
                    HorizontalAlignment := horizontal_alignment.Fill
                    VerticalAlignment := vertical_alignment.Bottom
                    Distribution := option{1.0}
                    Widget := color_block:
                        DefaultColor := NamedColors.White
                        DefaultOpacity := 0.1
                        DefaultDesiredSize := vector2{X := 0.0, Y := 1.0}
                stack_box_slot:
                    HorizontalAlignment := horizontal_alignment.Fill
                    VerticalAlignment := vertical_alignment.Fill
                    Padding := margin{Left := 0.0, Top := 20.0, Right := 0.0, Bottom := 0.0}
                    Widget := stack_box:
                        Orientation := orientation.Horizontal
                        Slots := array:
                            stack_box_slot:
                                HorizontalAlignment := horizontal_alignment.Fill
                                VerticalAlignment := vertical_alignment.Fill
                                Distribution := option{1.0}
                                Padding := margin{Left := 0.0, Top := 0.0, Right := 0.0, Bottom := 0.0}
                                Widget := RerollButton
                            stack_box_slot:
                                HorizontalAlignment := horizontal_alignment.Fill
                                VerticalAlignment := vertical_alignment.Fill
                                Distribution := option{1.0}
                                Padding := margin{Left := 20.0, Top := 0.0, Right := 0.0, Bottom := 0.0}
                                Widget := BanishButton
                            stack_box_slot:
                                HorizontalAlignment := horizontal_alignment.Fill
                                VerticalAlignment := vertical_alignment.Fill
                                Distribution := option{1.0}
                                Padding := margin{Left := 20.0, Top := 0.0, Right := 0.0, Bottom := 0.0}
                                Widget := BoostButton

    Uninitialize<override>():void =
        for (AbilitySelectEntryUI : AbilitySelectEntryUIs):
            AbilitySelectEntryUI.Uninitialize()
            
        for (EventsSubscription : EventsSubscriptions):
            EventsSubscription.Cancel()
    
        for (Subscription : Subscriptions):
            Subscription.Cancel()

    <# ----------------------------------------------------------------------------------------------------
        Get Functions
    ---------------------------------------------------------------------------------------------------- #>
    GetCurrentAbilities()<transacts>:[]ability_id =
        for (AbilitySelectEntryUI : AbilitySelectEntryUIs, AbilityInfo := AbilitySelectEntryUI.GetActiveAbilityInfo()?):
            AbilityInfo.Id

    GetCurrentWeapons()<transacts>:[]weapon_id =
        for (AbilitySelectEntryUI : AbilitySelectEntryUIs, WeaponInfo := AbilitySelectEntryUI.GetActiveWeaponInfo()?):
            WeaponInfo.Id

    CanSelect()<transacts>:logic =
        logic{not IsAutoSelect? and not HasSelected? and GetSecondsSinceEpoch() - ShowTime >= Player.GetSelectBufferTime()}

    GetAvailableAbilitySelectEntryUIs()<transacts>:[]ability_select_entry_ui =
        if (IsWeaponUnlockLevel(CurrentLevel)?):
            AbilitySelectEntryUIs
        else:
            for (AbilitySelectEntryUI : AbilitySelectEntryUIs, AbilitySelectEntryUI.MaybeAbilityInfo?) {AbilitySelectEntryUI}

    GetRerollsAvailable()<transacts>:int =
        IsWeaponUnlockLevel(CurrentLevel)? and Player.GetPlayerResource(player_resource_id.WeaponReroll) or Player.GetPlayerResource(player_resource_id.Reroll)

    GetIsFreeReroll()<transacts>:logic =
        logic{Player.HasBoon(boon_id.SpareDice)? and not IsWeaponUnlockLevel(CurrentLevel)? and not HasUsedSpareDice?}

    GetLuckAmount()<transacts>:?float =
        if (LuckAbility := MaybeLuckAbility?):
            option{LuckAbility(0).TierValue[LuckAbility(1) - 1]?}
        else:
            false

    GetLuckDiceMultiplier()<transacts>:float =
        (Player.HasBoon(boon_id.LuckyDice)? and not IsWeaponUnlockLevel(CurrentLevel)?) and (0.25 * RerollCount) or 0.0

    <# ----------------------------------------------------------------------------------------------------
        Update Functions
    ---------------------------------------------------------------------------------------------------- #>
    Update(Level:int, ?IsRoll:logic = true):void =
        if (IsWeaponUnlockLevel(CurrentLevel)?):
            RandomWeapons := GetRandomWeapons(Player, 3, ?Ignore := GetCurrentWeapons())
            for (Index -> AbilitySelectEntryUI : AbilitySelectEntryUIs, WeaponInfo := RandomWeapons[Index]):
                AbilitySelectEntryUI.ShowWeapon(WeaponInfo)

            BanishButton.SetVisibility(widget_visibility.Collapsed)
            BoostButton.SetVisibility(widget_visibility.Collapsed)
        else:
            for (Index -> AbilitySelectEntryUI : AbilitySelectEntryUIs):
                if (AbilityInfo := CurrentAbilities[Index]):
                    AbilitySelectEntryUI.Update(AbilityInfo(0), AbilityInfo(1), IsRoll, ?AbilitySelectMode := AbilitySelectMode)
                    AbilitySelectEntryUI.Show()

            if (HasBanished?):
                BanishButton.SetText(MakeMessage("Banished Used!"))
            else:
                BanishButton.SetText(MakeMessage((AbilitySelectMode = ability_select_mode.Banish) and "Cancel" or "Banish ({Player.GetBanish()} Remaining)"))
                BanishButton.SetVisibility(Player.GetBanish() > 0 and widget_visibility.Visible or widget_visibility.Collapsed)

            BoostButton.SetVisibility(Player.GetPlayerResource(player_resource_id.Boost) > 0 and widget_visibility.Visible or widget_visibility.Collapsed)
            BoostButton.SetText(MakeMessage((AbilitySelectMode = ability_select_mode.Boost) and "Cancel" or "Boost ({Player.GetPlayerResource(player_resource_id.Boost)} Remaining)"))

        RerollButton.SetText(MakeMessage("Reroll ({GetIsFreeReroll()? and "Free" or "{GetRerollsAvailable()} Remaining"})"))
        LuckyDiceText.SetText(MakeMessage("+{FormatAsPercent(GetLuckDiceMultiplier())}"))
        LuckyDiceUI.SetVisibility(GetLuckDiceMultiplier() > 0.0 and widget_visibility.Visible or widget_visibility.Hidden)

    Roll(?IsReroll:logic = false):void =
        #AbilityCount := GetRandomFloat(0.0, 1.0) <= 0.1 and 4 or 3
        AbilityCount := 3
        IgnoreAbilities:[]ability_id = (IsReroll? and GetCurrentAbilities() or array{})
        set CurrentAbilities = GetRandomAbilities(Player, AbilityCount, ?Ignore := IgnoreAbilities, ?LuckMultiplier := GetLuckDiceMultiplier(), ?MinTier := IsLegendary? and 5 or 1)

        LuckAbility := for (CurrentAbility : CurrentAbilities, CurrentAbility(0).Id = ability_id.Luck) {CurrentAbility}
        set MaybeLuckAbility = option{LuckAbility[0]}

        if (IsReroll?):
            set RerollCount += 1
            UpdateRatesUI()

    SetMode(NewMode:ability_select_mode):void =
        if (AbilitySelectMode = NewMode):
            set AbilitySelectMode = ability_select_mode.Default
        else:
            set AbilitySelectMode = NewMode
        
        Update(CurrentLevel, ?IsRoll := false)
        Player.PlayAudio(audio_id.Activate)

    SelectAbility(AbilitySelectEntryUI:ability_select_entry_ui):void =
        if (WeaponInfo := AbilitySelectEntryUI.MaybeWeaponInfo?):
            Player.ActivateNewAbility(WeaponInfo.Id)
            Player.EquipNewWeapon(WeaponInfo.Id)
            Player.PlayAudio(audio_id.Upgrade)

        else if (AbilityInfo := AbilitySelectEntryUI.MaybeAbilityInfo?):
            Tier := AbilitySelectEntryUI.GetTier()
            if (AbilitySelectMode = ability_select_mode.Boost):
                if (Tier <= 4):
                    Player.AddAbility(AbilityInfo, ?Tier := Tier + 1)
                    Player.RemovePlayerResource(player_resource_id.Boost, 1)
                    Player.PlayAudio(audio_id.Boost)
                else:
                    Player.AddAbility(AbilityInfo, ?Tier := Tier)
            else:
                Player.AddAbility(AbilityInfo, ?Tier := Tier)
                Player.PlayAudio(audio_id.Upgrade)

    SelectRandomAbility()<suspends>:void =
        defer:
            set IsAutoSelect = false
            
        set IsAutoSelect = true
        set AutoSelectCount += 1
        if (AutoSelectCount >= AutoSelectCountAFKThreshold):
            Player.SetAFK(true)

        for (Index -> AbilitySelectEntryUI : Shuffle(GetAvailableAbilitySelectEntryUIs())):
            if (Index = 0):
                AbilitySelectEntryUI.SelectButton.SetText(MakeMessage("Selected"))
                SelectAbility(AbilitySelectEntryUI)
            else:
                AbilitySelectEntryUI.InfoOverlay.SetVisibility(widget_visibility.Hidden)
                AbilitySelectEntryUI.SelectButton.SetText(MakeMessage(""))

        SleepTimeDilation(3.0)
        Hide()

    ToggleInfo(AbilitySelectInfoType:ability_select_info_type):void =
        for (Type -> InfoUI : InfoUIs, Type <> AbilitySelectInfoType):
            InfoUI.Hide()

        UI := if (InfoUI := InfoUIs[AbilitySelectInfoType]):
            Visibility := InfoUI.GetUI().GetVisibility()
            if (Visibility = widget_visibility.Visible):
                InfoUI.Hide()
            else:
                InfoUI.Show()
                InfoUI.Update()
            InfoUI
        else:
            InfoUI := case (AbilitySelectInfoType):
                ability_select_info_type.StatsInfo => ability_stat_info_ui{Player := Player, BackgroundType := background_type.Opaque, TextSize := ThemeTextSize.Small, IconSize := 0.75}
                ability_select_info_type.TierRates => ability_select_tier_rates_ui{Player := Player}

            InfoUI.Initialize()
            if (set InfoUIs[AbilitySelectInfoType] = InfoUI) {}

            if (PanelUI := canvas[MaybeUI?]):
                PanelUI.AddWidget of canvas_slot:
                    Anchors := case (AbilitySelectInfoType):
                        ability_select_info_type.StatsInfo => anchors{Minimum := vector2{X := 0.0, Y := 0.0}, Maximum := vector2{X := 0.6, Y := 0.0}}
                        ability_select_info_type.TierRates => anchors{Minimum := vector2{X := 0.0, Y := 0.0}, Maximum := vector2{X := 0.0, Y := 0.0}}
                    Alignment := vector2{X := 0.0, Y := 0.0}
                    ZOrder := 1
                    Widget := InfoUI.GetUI()

                InfoUI.Update()

            InfoUI

        if (AbilitySelectInfoType = ability_select_info_type.TierRates):
            UpdateRatesUI()

    UpdateRatesUI():void =
        if (RatesUI := ability_select_tier_rates_ui[InfoUIs[ability_select_info_type.TierRates]]):
            RatesUI.Update(?LuckMultiplier := GetLuckDiceMultiplier(), ?MaybeAddLuckAmount := GetLuckAmount())
    
    Complete():void =
        if (NextPause := Player.GetNextPause[], Level := pause_level_up[NextPause].Level):
            OnShow(Level, ?IsFirst := false)
            ResumeGame(Player, ?IsShow := false)
        else:
            Hide()

    UpdateTitle()<suspends>:void =
        Sleep(0.0)
        PauseCount := Player.GetPauseCount(pause_type.LevelUp)
        TitleText.SetText(MakeMessage("{IsLegendary? and "Legendary " or ""}{Title} ({PauseCount})"))
        TitleText.SetTextColor(ThemeColors.PrimaryText)
        TitleText.SetShadowOpacity(0.5)
    
    <# ----------------------------------------------------------------------------------------------------
        Event Handlers
    ---------------------------------------------------------------------------------------------------- #>
    OnShow(Level:int, ?IsFirst:logic = true, ?InIsLegendary:logic = false):void =
        set AbilitySelectMode = ability_select_mode.Default
        set HasBanished = false
        set HasSelected = false
        set CurrentLevel = Level
        set HasUsedSpareDice = false
        set RerollCount = 0
        set IsLegendary = InIsLegendary

        if (IsFirst?):
            set ShowTime = GetSecondsSinceEpoch()
        
        spawn{UpdateTitle()}
        for (AbilitySelectEntryUI : AbilitySelectEntryUIs) {AbilitySelectEntryUI.Clear()}
        Roll(?IsReroll := true)
        Update(Level)

        for (InfoUI : InfoUIs):
            InfoUI.Hide()

    OnClosed(InPlayer:player):void =
        for (AbilitySelectEntryUI : AbilitySelectEntryUIs):
            AbilitySelectEntryUI.CancelAnimationEvent.Signal()

        ResumeGame(Player)

    OnSelect(AbilityInfo:ability_info, Tier:int, IsEvolution:logic):void =
        if (not CanSelect()?) then return
        
        set AutoSelectCount = 0
        Player.SetAFK(false)

        if (AbilitySelectMode <> ability_select_mode.Banish):
            set HasSelected = true

        case (AbilitySelectMode):
            ability_select_mode.Default =>
                Player.AddAbility(AbilityInfo, ?Tier := Tier)
                Player.PlayAudio(audio_id.Activate)
                
                if (IsEvolution?, WeaponId := AbilityInfo.MaybeWeaponId?):
                    Hide(?IsSignal := false)
                    PauseGameOverride(Player, pause_weapon_evolution{Type := pause_type.WeaponEvolution, Id := WeaponId, EvolutionLevel := Player.GetWeaponEvolutionLevel(WeaponId), IsItemShop := false})
                    Player.PlayAudio(audio_id.Boost)
                else:
                    Complete()

            ability_select_mode.Banish => 
                Player.BanishAbility(AbilityInfo)
                Player.PlayAudio(audio_id.Banish)
                set HasBanished = true
                SetMode(ability_select_mode.Default)

            ability_select_mode.Boost =>
                if (Tier <= 4):
                    Player.AddAbility(AbilityInfo, ?Tier := Tier + 1)
                    Player.RemovePlayerResource(player_resource_id.Boost, 1)
                    Player.PlayAudio(audio_id.Boost)
                else:
                    Player.AddAbility(AbilityInfo, ?Tier := Tier)

                if (IsEvolution?, WeaponId := AbilityInfo.MaybeWeaponId?):
                    Hide(?IsSignal := false)
                    PauseGameOverride(Player, pause_weapon_evolution{Type := pause_type.WeaponEvolution, Id := WeaponId, EvolutionLevel := Player.GetWeaponEvolutionLevel(WeaponId), IsItemShop := false})
                    Player.PlayAudio(audio_id.Boost)
                else:
                    Complete()

    OnSelectWeapon(WeaponInfo:weapon_info):void =
        if (not CanSelect()?) then return
        set HasSelected = true

        Player.ActivateNewAbility(WeaponInfo.Id)
        Player.EquipNewWeapon(WeaponInfo.Id)
        Complete()

    OnClickReroll(Message:widget_message):void =
        if (not CanSelect()?) then return
        if (CurrentAbilities.Length < 3):
            Player.PlayAudio(audio_id.Error)
            return
            
        IsFreeReroll := GetIsFreeReroll()
        
        if (IsFreeReroll? or GetRerollsAvailable() > 0):
            set AbilitySelectMode = ability_select_mode.Default
            
            if (IsWeaponUnlockLevel(CurrentLevel)?):
                Player.RemovePlayerResource(player_resource_id.WeaponReroll, 1)
            else:
                if (IsFreeReroll?):
                    set HasUsedSpareDice = true
                else:
                    Player.RemovePlayerResource(player_resource_id.Reroll, 1)

            if (GetLuckDiceMultiplier() > 0.0, LuckyDiceAudioId := LuckyDiceAudioIds[Mod[RerollCount - 1, 5]]):
                Player.PlayAudio(LuckyDiceAudioId)
            else:
                Player.PlayAudio(audio_id.Reroll)

            Roll(?IsReroll := true)
            Update(CurrentLevel)
        else:
            Player.PlayAudio(audio_id.Error)

    OnClickBoost(Message:widget_message):void =
        if (not CanSelect()?) then return

        SetMode(ability_select_mode.Boost)

        if (MaybeLuckAbility?):
            UpdateRatesUI()

    OnClickBanish(Message:widget_message):void =
        if (not CanSelect()?) then return
        if (HasBanished?):
            Player.PlayAudio(audio_id.Error)
            return

        SetMode(ability_select_mode.Banish)

    OnResumeCountdownStart():void =
        if (not IsShowing?) then return
        Player.PlayAudio(audio_id.TimerWarning)
        TitleText.SetText(MakeMessage("{GameResumeCountdownTimeDefault}s"))
        TitleText.SetTextColor(ThemeColors.Failure)
        TitleText.SetShadowColor(ThemeColors.Failure * 0.125)
        TitleText.SetShadowOffset(option{vector2{X := 2.0, Y := 2.0}})
        TitleText.SetShadowOpacity(1.0)
        
    OnResumeCountdownComplete():void =
        if (not IsShowing?) then return
        TitleText.SetText(MakeMessage("Auto Selected"))
        spawn{SelectRandomAbility()}

    OnResumeCountdown(TimeRemaining:int):void =
        if (not IsShowing?) then return
        if (Player.GetIsAFK()?):
            if (AbilitySelectEntryUI := Shuffle(GetAvailableAbilitySelectEntryUIs())[0]):
                SelectAbility(AbilitySelectEntryUI)

            Hide()
            return

        TitleText.SetText(MakeMessage("{TimeRemaining}s"))

    OnGameResumed():void =
        if (not IsShowing?) then return
        spawn{UpdateTitle()}

    OnClickTierRates(Message:widget_message):void =
        ToggleInfo(ability_select_info_type.TierRates)

    OnClickStatsInfo(Message:widget_message):void =
        ToggleInfo(ability_select_info_type.StatsInfo)