using { /Fortnite.com/UI }
using { /Verse.org/Colors }
using { /UnrealEngine.com/Temporary/SpatialMath }
using { /UnrealEngine.com/Temporary/UI }

<# ----------------------------------------------------------------------------------------------------
    Data Definitions
---------------------------------------------------------------------------------------------------- #>
background_type := enum:
    None
    Opaque
    Transparent

<# ----------------------------------------------------------------------------------------------------
    ability_stat_info_ui
---------------------------------------------------------------------------------------------------- #>
ability_stat_info_ui := class(ui_component):
    BackgroundType:background_type
    TextSize:float = ThemeTextSize.Default
    IconSize:float = 1.0

    TeamExperienceMultiplierText:text_block = text_block:
        DefaultTextColor := ThemeColors.Success
        DefaultShadowColor := ThemeColors.Success * 0.125
        DefaultShadowOffset := option{vector2{X := 2.0, Y := 2.0}}
        DefaultShadowOpacity := 1.0

    var AbilityTexts:[ability_id]text_block = map{}
    var AbilityTotalTexts:[ability_id]text_block = map{}
    var GameResultWeaponEntryUIs:[]game_result_weapon_entry_ui = array{}
    var BoonIcons:[]texture_block = array{}
    var BoonTexts:[]text_block = array{}

    <# ----------------------------------------------------------------------------------------------------
        Initialization
    ---------------------------------------------------------------------------------------------------- #>
    CreateUI<override>():widget =
        if (BackgroundType = background_type.None):
            CreatePlayerUI()
        else:
            Background := (BackgroundType = background_type.Opaque) and UserInterface.VerseWidget.WBP_Panel_Background{} or UserInterface.VerseWidget.WBP_Panel_Background_Transparent{}

            overlay:
                Slots := array:
                    overlay_slot:
                        HorizontalAlignment := horizontal_alignment.Fill
                        VerticalAlignment := vertical_alignment.Fill
                        Widget := Background
                    overlay_slot:
                        HorizontalAlignment := horizontal_alignment.Fill
                        VerticalAlignment := vertical_alignment.Fill
                        Padding := margin{Left := 40.0, Top := 40.0, Right := 40.0, Bottom := 40.0} * IconSize
                        Widget := CreatePlayerUI()

    CreatePlayerUI():widget =
        stack_box:
            Orientation := orientation.Horizontal
            Slots := array:
                stack_box_slot:
                    HorizontalAlignment := horizontal_alignment.Fill
                    VerticalAlignment := vertical_alignment.Fill
                    Distribution := option{1.0}
                    Widget := CreateAbilitiesUI()
                stack_box_slot:
                    HorizontalAlignment := horizontal_alignment.Fill
                    VerticalAlignment := vertical_alignment.Fill
                    Distribution := option{1.4}
                    Padding := margin{Left := 20.0, Top := 0.0, Right := 0.0, Bottom := 0.0} * IconSize
                    Widget := stack_box:
                        Orientation := orientation.Vertical
                        Slots := array:
                            stack_box_slot:
                                HorizontalAlignment := horizontal_alignment.Fill
                                VerticalAlignment := vertical_alignment.Fill
                                Distribution := option{1.0}
                                Padding := margin{Left := 0.0, Top := 0.0, Right := 0.0, Bottom := 20.0} * IconSize
                                Widget := CreateWeaponsUI()
                            stack_box_slot:
                                HorizontalAlignment := horizontal_alignment.Fill
                                VerticalAlignment := vertical_alignment.Fill
                                Widget := CreateBoonsUI()

    CreateAbilitiesUI():widget =
        AbilitiesStackBox := stack_box:
            Orientation := orientation.Vertical
            Slots := array:
                stack_box_slot:
                    HorizontalAlignment := horizontal_alignment.Fill
                    VerticalAlignment := vertical_alignment.Fill
                    Padding := margin{Left := 0.0, Top := 0.0, Right := 0.0, Bottom := 20.0} * IconSize
                    Widget := text_block:
                        DefaultTextSize := TextSize
                        DefaultText := MakeMessage("Abilities")
                        DefaultTextColor := ThemeColors.PrimaryText
                        DefaultTextOpacity := 0.5
                stack_box_slot:
                    HorizontalAlignment := horizontal_alignment.Fill
                    VerticalAlignment := vertical_alignment.Fill
                    Widget := stack_box:
                        Orientation := orientation.Vertical
                        Slots := 
                            var Index:int = 0
                            for (AbilityInfo : GetAbilityInfo()):
                                AbilityText := text_block:
                                    DefaultTextSize := TextSize
                                    DefaultShadowOffset := option{vector2{X := 2.0, Y := 2.0}}
                                    DefaultShadowColor := ThemeColors.Success * 0.125

                                AbilityTotalText := text_block:
                                    DefaultTextSize := TextSize
                                    DefaultTextColor := ThemeColors.Success
                                    DefaultShadowOpacity := 1.0
                                    DefaultShadowOffset := option{vector2{X := 2.0, Y := 2.0}}
                                    DefaultShadowColor := ThemeColors.Success * 0.125
                    
                                if (set AbilityTexts[AbilityInfo.Id] = AbilityText) {}
                                if (set AbilityTotalTexts[AbilityInfo.Id] = AbilityTotalText) {}
                                set Index += 1
                
                                stack_box_slot:
                                    HorizontalAlignment := horizontal_alignment.Fill
                                    Padding := margin{Left := 0.0, Top := (Index = 1) and 0.0 or 4.0, Right := 0.0, Bottom := 0.0} * IconSize
                                    Widget := stack_box:
                                        Orientation := orientation.Horizontal
                                        Slots := array:
                                            stack_box_slot:
                                                HorizontalAlignment := horizontal_alignment.Left
                                                VerticalAlignment := vertical_alignment.Center
                                                Padding := margin{Left := 0.0, Top := 0.0, Right := 10.0, Bottom := 0.0} * IconSize
                                                Widget := texture_block:
                                                    DefaultImage := AbilityInfo.IconSmall
                                                    DefaultDesiredSize := vector2{X := 40.0, Y := 40.0} * IconSize
                                            stack_box_slot:
                                                HorizontalAlignment := horizontal_alignment.Left
                                                Distribution := option{0.5}
                                                Widget := text_block:
                                                    DefaultTextSize := TextSize
                                                    DefaultText := MakeMessage("{AbilityInfo.ShortName}")
                                                    DefaultTextColor := ThemeColors.PrimaryText
                                            stack_box_slot:
                                                HorizontalAlignment := horizontal_alignment.Right
                                                Distribution := option{1.0}
                                                Widget := AbilityTotalText
                                            stack_box_slot:
                                                HorizontalAlignment := horizontal_alignment.Right
                                                Distribution := option{1.0}
                                                Widget := AbilityText

        TeamExperienceMultiplierText.SetTextSize(TextSize)
        AbilitiesStackBox.AddWidget of stack_box_slot:
            HorizontalAlignment := horizontal_alignment.Fill
            VerticalAlignment := vertical_alignment.Bottom
            Distribution := option{1.0}
            Widget := stack_box:
                Orientation := orientation.Vertical
                Slots := array:
                    stack_box_slot:
                        HorizontalAlignment := horizontal_alignment.Fill
                        VerticalAlignment := vertical_alignment.Bottom
                        Distribution := option{1.0}
                        Padding := margin{Left := 0.0, Top := 20.0, Right := 0.0, Bottom := 20.0} * IconSize
                        Widget := color_block:
                            DefaultColor := NamedColors.White
                            DefaultOpacity := 0.1
                            DefaultDesiredSize := vector2{X := 0.0, Y := 1.0}
                    stack_box_slot:
                        HorizontalAlignment := horizontal_alignment.Fill
                        VerticalAlignment := vertical_alignment.Fill
                        Widget := stack_box:
                            Orientation := orientation.Horizontal
                            Slots := array:
                                stack_box_slot:
                                    HorizontalAlignment := horizontal_alignment.Left
                                    Padding := margin{Left := 0.0, Top := 0.0, Right := 10.0, Bottom := 0.0} * IconSize
                                    Widget := texture_block:
                                        DefaultImage := UserInterface.Icons.ExperienceMultiplier_64
                                        DefaultDesiredSize := vector2{X := 40.0, Y := 40.0} * IconSize
                                stack_box_slot:
                                    HorizontalAlignment := horizontal_alignment.Fill
                                    Widget := text_block:
                                        DefaultTextSize := TextSize
                                        DefaultText := MakeMessage("Team Multiplier")
                                        DefaultTextColor := ThemeColors.PrimaryText
                                stack_box_slot:
                                    HorizontalAlignment := horizontal_alignment.Right
                                    Distribution := option{1.0}
                                    Widget := TeamExperienceMultiplierText

        overlay:
            Slots := array:
                overlay_slot:
                    HorizontalAlignment := horizontal_alignment.Fill
                    VerticalAlignment := vertical_alignment.Fill
                    Widget := UserInterface.VerseWidget.WBP_RoundedBox{}
                overlay_slot:
                    HorizontalAlignment := horizontal_alignment.Fill
                    VerticalAlignment := vertical_alignment.Fill
                    Padding := margin{Left := 40.0, Top := 40.0, Right := 40.0, Bottom := 40.0} * IconSize
                    Widget := AbilitiesStackBox

    CreateWeaponsUI():widget =
        overlay:
            Slots := array:
                overlay_slot:
                    HorizontalAlignment := horizontal_alignment.Fill
                    VerticalAlignment := vertical_alignment.Fill
                    Widget := UserInterface.VerseWidget.WBP_RoundedBox{}
                overlay_slot:
                    HorizontalAlignment := horizontal_alignment.Fill
                    VerticalAlignment := vertical_alignment.Fill
                    Padding := margin{Left := 40.0, Top := 40.0, Right := 40.0, Bottom := 40.0} * IconSize
                    Widget := stack_box:
                        Orientation := orientation.Vertical
                        Slots := array:
                            stack_box_slot:
                                HorizontalAlignment := horizontal_alignment.Fill
                                VerticalAlignment := vertical_alignment.Fill
                                Padding := margin{Left := 0.0, Top := 0.0, Right := 0.0, Bottom := 20.0} * IconSize
                                Widget := stack_box:
                                    Orientation := orientation.Horizontal
                                    Slots := array:
                                        stack_box_slot:
                                            HorizontalAlignment := horizontal_alignment.Left
                                            Distribution := option{2.0}
                                            Widget := text_block:
                                                DefaultTextSize := TextSize
                                                DefaultText := MakeMessage("Weapons")
                                                DefaultTextColor := ThemeColors.PrimaryText
                                                DefaultTextOpacity := 0.5
                                        stack_box_slot:
                                            HorizontalAlignment := horizontal_alignment.Left
                                            Distribution := option{1.0}
                                            Widget := texture_block:
                                                DefaultImage := UserInterface.Icons.ProjectileDamage_64
                                                DefaultTint := ThemeColors.WeaponDamage
                                                DefaultDesiredSize := vector2{X := 40.0, Y := 40.0} * IconSize
                                        stack_box_slot:
                                            HorizontalAlignment := horizontal_alignment.Left
                                            Distribution := option{1.0}
                                            Widget := texture_block:
                                                DefaultImage := UserInterface.Icons.AbilityCooldown_64
                                                DefaultTint := ThemeColors.WeaponCooldown
                                                DefaultDesiredSize := vector2{X := 40.0, Y := 40.0} * IconSize
                                        stack_box_slot:
                                            HorizontalAlignment := horizontal_alignment.Left
                                            Distribution := option{1.0}
                                            Widget := texture_block:
                                                DefaultImage := UserInterface.Icons.AbilitySize_64
                                                DefaultTint := ThemeColors.WeaponSize
                                                DefaultDesiredSize := vector2{X := 40.0, Y := 40.0} * IconSize
                            stack_box_slot:
                                HorizontalAlignment := horizontal_alignment.Fill
                                VerticalAlignment := vertical_alignment.Fill
                                Distribution := option{1.0}
                                Widget := stack_box:
                                    Orientation := orientation.Vertical
                                    Slots := for (Index := 1..4):
                                        GameResultWeaponEntryUI := game_result_weapon_entry_ui{Player := Player, TextSize := TextSize, IconSize := IconSize}
                                        GameResultWeaponEntryUI.Initialize()
                                        set GameResultWeaponEntryUIs += array{GameResultWeaponEntryUI}

                                        stack_box_slot:
                                            HorizontalAlignment := horizontal_alignment.Fill
                                            VerticalAlignment := vertical_alignment.Top
                                            Padding := margin{Left := 0.0, Top := (Index = 1) and 0.0 or 20.0, Right := 0.0, Bottom := 0.0} * IconSize
                                            Distribution := option{1.0}
                                            Widget := GameResultWeaponEntryUI.GetUI()
    CreateBoonsUI():widget =
        overlay:
            Slots := array:
                overlay_slot:
                    HorizontalAlignment := horizontal_alignment.Fill
                    VerticalAlignment := vertical_alignment.Fill
                    Widget := UserInterface.VerseWidget.WBP_RoundedBox{}
                overlay_slot:
                    HorizontalAlignment := horizontal_alignment.Fill
                    VerticalAlignment := vertical_alignment.Fill
                    Padding := margin{Left := 40.0, Top := 20.0, Right := 40.0, Bottom := 20.0} * IconSize
                    Widget := stack_box:
                        Orientation := orientation.Horizontal
                        Slots := array:
                            stack_box_slot:
                                HorizontalAlignment := horizontal_alignment.Fill
                                VerticalAlignment := vertical_alignment.Top
                                Distribution := option{1.5}
                                Widget := text_block:
                                    DefaultTextSize := TextSize
                                    DefaultText := MakeMessage("Boons")
                                    DefaultTextColor := ThemeColors.PrimaryText
                                    DefaultTextOpacity := 0.5
                            stack_box_slot:
                                HorizontalAlignment := horizontal_alignment.Fill
                                VerticalAlignment := vertical_alignment.Fill
                                Distribution := option{5.0}
                                Widget := stack_box:
                                    Orientation := orientation.Vertical
                                    Slots := array:
                                        stack_box_slot:
                                            HorizontalAlignment := horizontal_alignment.Fill
                                            VerticalAlignment := vertical_alignment.Fill
                                            Widget := stack_box:
                                                Orientation := orientation.Horizontal
                                                Slots := for (Index := 1..5):
                                                    stack_box_slot:
                                                        Distribution := option{1.0}
                                                        Widget := CreateBoonEntry()
                                        stack_box_slot:
                                            HorizontalAlignment := horizontal_alignment.Fill
                                            VerticalAlignment := vertical_alignment.Fill
                                            Padding := margin{Left := 0.0, Top := 10.0, Right := 0.0, Bottom := 0.0} * IconSize
                                            Widget := stack_box:
                                                Orientation := orientation.Horizontal
                                                Slots := for (Index := 1..5):
                                                    stack_box_slot:
                                                        Distribution := option{1.0}
                                                        Widget := CreateBoonEntry()

    CreateBoonEntry():widget =
        BoonIcon := texture_block:
            DefaultImage := UserInterface.Icons.Boon_64
            DefaultDesiredSize := vector2{X := 64.0, Y := 64.0} * IconSize

        BoonText := text_block:
            DefaultTextSize := TextSize
            DefaultTextColor := ThemeColors.Success
            DefaultShadowColor := ThemeColors.Success * 0.125
            DefaultShadowOffset := option{vector2{X := 2.0, Y := 2.0}}
            DefaultShadowOpacity := 1.0

        set BoonIcons += array{BoonIcon}
        set BoonTexts += array{BoonText}

        stack_box:
            Orientation := orientation.Vertical
            Slots := array:
                stack_box_slot:
                    Widget := BoonIcon
                stack_box_slot:
                    Widget := BoonText

    <# ----------------------------------------------------------------------------------------------------
        Update Functions
    ---------------------------------------------------------------------------------------------------- #>
    Update<override>(?MaybeGameResult:?tuple(game_result_info, player_stat) = false):void =
        OverridePlayerStats:?player_stat = if (GameResult := MaybeGameResult?):
            option{GameResult(1)}
        else:
            false

        if (PlayerStats := OverridePlayerStats? or Player.GetStats[]):
            BanishedIds := for (BanishedAbility : PlayerStats.GetBanishedAbilities(), not BanishedAbility.MaybeWeaponId?) {BanishedAbility.Id}
            
            for (Id -> AbilityText : AbilityTexts, AbilityTotalText := AbilityTotalTexts[Id], AbilityInfo := GetAbilityInfo[Id]):
                IsBanished := logic{BanishedIds.Find[Id]}
                AbilityValue := PlayerStats.GetAbilityValue(Id)
                TotalAbilityValue := if (StatId := AbilityInfo.MaybeStatId?): # [TODO] Refactor
                    PlayerStats.GetStat(StatId)
                else:
                    PlayerStats.GetAbilityValue(Id)

                AbilityTotalText.SetText(MakeMessage("{AbilityInfo.FormatValue(TotalAbilityValue)}"))

                if (AbilityValue > 0.0):
                    AbilityText.SetText(MakeMessage("+{AbilityInfo.FormatValue(AbilityValue)}"))
                    AbilityText.SetColor(IsBanished? and ThemeColors.Failure or ThemeColors.Success)
                else:
                    AbilityText.SetText(MakeMessage("0"))
                    if (IsBanished?):
                        AbilityText.SetColor(ThemeColors.Failure)
                    else:
                        AbilityText.SetTextColor(ThemeColors.Inactive)
                        AbilityText.SetShadowOpacity(0.0)

            WeaponIds := for (WeaponId -> WeaponLevel : PlayerStats.GetWeaponLevels()) {WeaponId}
            for (Index -> GameResultWeaponEntryUI : GameResultWeaponEntryUIs):
                if (WeaponId := WeaponIds[Index], WeaponStats := PlayerStats.GetWeaponStats[WeaponId]):
                    MaybeResultsInfo:?game_result_info = if (GameResult := MaybeGameResult?) then option{GameResult(0)} else false
                    WeaponLevel := PlayerStats.GetWeaponLevel(WeaponId)
                    GameResultWeaponEntryUI.Update(Player, WeaponId, WeaponLevel, WeaponStats, ?MaybeResultsInfo := MaybeResultsInfo)
                    GameResultWeaponEntryUI.Show()
                else:
                    if (Index < GetGameSettings[].Weapons):
                        GameResultWeaponEntryUI.SetInactive()
                        GameResultWeaponEntryUI.Show()
                    else:
                        GameResultWeaponEntryUI.Hide()
            
            TeamExperienceMultiplier := if (GameResult := MaybeGameResult?) then GameResult(0).TeamExperienceMultiplier else GetTeamExperienceMultiplier()
            TeamExperienceMultiplierText.SetText(MakeMessage("+{FormatAsPercent(TeamExperienceMultiplier)}"))

            PlayerBoons :=  for (BoonId -> BoonValue : (MaybeGameResult? and PlayerStats.BoonValues or PlayerStats.GetBoons())) {BoonId}
            IsEndless := logic{if (GameResult := MaybeGameResult?) then GameResult(0).GameSettings.IsEndless? else GetGameSettings[].IsEndless?}
            StageCount := IsEndless? and MaxBoonCount or GetArenaStageCount()

            for (Index -> BoonIcon : BoonIcons, BoonText := BoonTexts[Index]):
                if (BoonId := PlayerBoons[Index], BoonInfo := GetBoonInfo[BoonId]):
                    if (GetValue := BoonInfo.MaybeGetValue?):
                        Value := MaybeGameResult? and (PlayerStats.BoonValues[BoonId] or 0.0) or GetValue(Player)
                        BoonText.SetText(MakeMessage("{Value > 0.0 and "+" or ""}{BoonInfo.FormatValue(Value)}"))

                        if (Value > 0.0):
                            BoonText.SetColor(ThemeColors.Success)
                        else:                        
                            BoonText.SetTextColor(ThemeColors.Inactive)
                            BoonText.SetShadowOpacity(0.0)
                    else:
                        BoonText.SetText(MakeMessage(""))

                    BoonIcon.SetImage(BoonInfo.IconSmall)
                    BoonIcon.SetTint(BoonInfo.Tier.ToTierColor() * 2)
                else:
                    BoonIcon.SetImage(UserInterface.Icons.Lock_64)
                    BoonIcon.SetTint(ThemeColors.Inactive)
                    BoonIcon.SetVisibility(Index + 1 <= StageCount and widget_visibility.Visible or widget_visibility.Collapsed)
                    BoonText.SetText(MakeMessage(""))