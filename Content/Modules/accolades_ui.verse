using { /Fortnite.com/UI }
using { /Verse.org/Colors }
using { /Verse.org/Simulation }
using { /UnrealEngine.com/Temporary/SpatialMath }
using { /UnrealEngine.com/Temporary/UI }

accolades_ui := class(panel_ui):
    Title<override>:string = "Accolades"
    Anchors<override>:anchors = anchors{Minimum := vector2{X := 0.15, Y := 0.5}, Maximum := vector2{X := 0.85, Y := 0.5}}
    BackgroundOpacity<override>:type {_X:float where 0.000000 <= _X, _X <= 1.000000} = 0.7

    <# ----------------------------------------------------------------------------------------------------
        Initialization
    ---------------------------------------------------------------------------------------------------- #>
    CreateContent<override>():widget =
        AccoladeInfo := GetAccoladeInfo()
        stack_box:
            Orientation := orientation.Vertical
            Slots := array:
                stack_box_slot:
                    VerticalAlignment := vertical_alignment.Fill
                    Widget := stack_box:
                        Orientation := orientation.Horizontal
                        Slots := array:
                            stack_box_slot:
                                HorizontalAlignment := horizontal_alignment.Fill
                                Widget := text_block:
                                    DefaultTextSize := ThemeTextSize.Small
                                    DefaultText := MakeMessage("Battle Pass")
                                    DefaultTextColor := ThemeColors.PrimaryText
                            stack_box_slot:
                                Padding := margin{Left := 10.0, Top := 0.0, Right := 10.0, Bottom := 0.0}
                                Widget := texture_block:
                                    DefaultImage := UserInterface.Icons.BattlepassXP_64
                                    DefaultDesiredSize := vector2{X := 64.0, Y := 64.0}
                            stack_box_slot:
                                HorizontalAlignment := horizontal_alignment.Fill
                                Widget := text_block:
                                    DefaultTextSize := ThemeTextSize.Small
                                    DefaultText := MakeMessage("is awarded when completing the below accolades.")
                                    DefaultTextColor := ThemeColors.PrimaryText
                        
                stack_box_slot:
                    HorizontalAlignment := horizontal_alignment.Fill
                    VerticalAlignment := vertical_alignment.Fill
                    Padding := margin{Left := 0.0, Top := 20.0, Right := 0.0, Bottom := 0.0}
                    Widget := stack_box:
                        Orientation := orientation.Horizontal
                        Slots := array:
                            stack_box_slot:
                                HorizontalAlignment := horizontal_alignment.Fill
                                VerticalAlignment := vertical_alignment.Fill
                                Distribution := option{1.0}
                                Padding := margin{Left := 0.0, Top := 0.0, Right := 0.0, Bottom := 0.0}
                                Widget := overlay:
                                    Slots := array:
                                        overlay_slot:
                                            HorizontalAlignment := horizontal_alignment.Fill
                                            VerticalAlignment := vertical_alignment.Fill
                                            Widget := UserInterface.VerseWidget.WBP_RoundedBox{}
                                        overlay_slot:
                                            HorizontalAlignment := horizontal_alignment.Fill
                                            VerticalAlignment := vertical_alignment.Fill
                                            Padding := margin{Left := 40.0, Top := 40.0, Right := 40.0, Bottom := 40.0}
                                            Widget := CreateAccoladeUI(AccoladeInfo(0))
                            stack_box_slot:
                                HorizontalAlignment := horizontal_alignment.Fill
                                VerticalAlignment := vertical_alignment.Fill
                                Distribution := option{1.0}
                                Padding := margin{Left := 20.0, Top := 0.0, Right := 0.0, Bottom := 0.0}
                                Widget := overlay:
                                    Slots := array:
                                        overlay_slot:
                                            HorizontalAlignment := horizontal_alignment.Fill
                                            VerticalAlignment := vertical_alignment.Fill
                                            Widget := UserInterface.VerseWidget.WBP_RoundedBox{}
                                        overlay_slot:
                                            HorizontalAlignment := horizontal_alignment.Fill
                                            VerticalAlignment := vertical_alignment.Fill
                                            Padding := margin{Left := 40.0, Top := 40.0, Right := 40.0, Bottom := 40.0}
                                            Widget := CreateAccoladeCountUI(AccoladeInfo(1))

    CreateLabel(Label:string, ?Distribution:?float = option{1.0}):stack_box_slot =
        stack_box_slot:
            HorizontalAlignment := horizontal_alignment.Fill
            VerticalAlignment := vertical_alignment.Fill
            Distribution := Distribution
            Widget := text_block:
                DefaultTextSize := ThemeTextSize.Small
                DefaultText := MakeMessage("{Label}")
                DefaultTextColor := ThemeColors.PrimaryText
                DefaultTextOpacity := 0.5

    CreateAccoladeEntry(Accolade:accolade):widget =
        stack_box:
            Orientation := orientation.Horizontal
            Slots := array:
                stack_box_slot:
                    HorizontalAlignment := horizontal_alignment.Fill
                    VerticalAlignment := vertical_alignment.Fill
                    Distribution := option{1.0}
                    Widget := stack_box:
                        Orientation := orientation.Horizontal
                        Slots := array:
                            stack_box_slot:
                                HorizontalAlignment := horizontal_alignment.Left
                                Widget := texture_block:
                                    DefaultImage := Accolade.Icon
                                    DefaultDesiredSize := vector2{X := 32.0, Y := 32.0}
                            stack_box_slot:
                                HorizontalAlignment := horizontal_alignment.Fill
                                VerticalAlignment := vertical_alignment.Fill
                                Padding := margin{Left := 20.0, Top := 0.0, Right := 0.0, Bottom := 0.0}
                                Widget := text_block:
                                    DefaultTextSize := ThemeTextSize.Small
                                    DefaultText := MakeMessage("{Accolade.Id}")
                                    DefaultTextColor := ThemeColors.PrimaryText
                stack_box_slot:
                    HorizontalAlignment := horizontal_alignment.Fill
                    VerticalAlignment := vertical_alignment.Fill
                    Distribution := option{0.5}
                    Widget := text_block:
                        DefaultTextSize := ThemeTextSize.Small
                        DefaultText := MakeMessage("{if (Cooldown := Accolade.MaybeCooldown?) then "{Cooldown.FormatAsTime()}" else ""}")
                        DefaultTextColor := ThemeColors.Success
                        DefaultShadowColor := ThemeColors.Success * 0.125
                        DefaultShadowOffset := option{vector2{X := 2.0, Y := 2.0}}
                        DefaultShadowOpacity := 1.0

    CreateAccoladeUI(Accolades:[]accolade):widget =
        stack_box:
            Orientation := orientation.Vertical
            Slots := array:
                stack_box_slot:
                    HorizontalAlignment := horizontal_alignment.Fill
                    VerticalAlignment := vertical_alignment.Fill
                    Widget := stack_box:
                        Orientation := orientation.Horizontal
                        Slots := array:
                            CreateLabel("Accolade")
                            CreateLabel("Cooldown", ?Distribution := option{0.5})

                stack_box_slot:
                    HorizontalAlignment := horizontal_alignment.Fill
                    VerticalAlignment := vertical_alignment.Fill
                    Padding := margin{Left := 0.0, Top := 20.0, Right := 0.0, Bottom := 0.0}
                    Widget := stack_box:
                        Orientation := orientation.Vertical
                        Slots := for (Accolade : Accolades):
                            stack_box_slot:
                                HorizontalAlignment := horizontal_alignment.Fill
                                VerticalAlignment := vertical_alignment.Fill
                                Widget := CreateAccoladeEntry(Accolade)

    CreateAccoladeCountUI(AccoladeCounts:[]accolade_count):widget =
        stack_box:
            Orientation := orientation.Vertical
            Slots := array:
                stack_box_slot:
                    HorizontalAlignment := horizontal_alignment.Fill
                    VerticalAlignment := vertical_alignment.Fill
                    Widget := stack_box:
                        Orientation := orientation.Horizontal
                        Slots := array:
                            CreateLabel("Accolade")
                            CreateLabel("Target", ?Distribution := option{0.5})

                stack_box_slot:
                    HorizontalAlignment := horizontal_alignment.Fill
                    VerticalAlignment := vertical_alignment.Fill
                    Padding := margin{Left := 0.0, Top := 20.0, Right := 0.0, Bottom := 0.0}
                    Widget := stack_box:
                        Orientation := orientation.Vertical
                        Slots := for (AccoladeCount : AccoladeCounts):
                            stack_box_slot:
                                HorizontalAlignment := horizontal_alignment.Fill
                                VerticalAlignment := vertical_alignment.Fill
                                Widget := stack_box:
                                    Orientation := orientation.Horizontal
                                    Slots := array:
                                        stack_box_slot:
                                            HorizontalAlignment := horizontal_alignment.Fill
                                            VerticalAlignment := vertical_alignment.Fill
                                            Distribution := option{1.0}
                                            Widget := stack_box:
                                                Orientation := orientation.Horizontal
                                                Slots := array:
                                                    stack_box_slot:
                                                        HorizontalAlignment := horizontal_alignment.Left
                                                        Widget := texture_block:
                                                            DefaultImage := AccoladeCount.Icon
                                                            DefaultDesiredSize := vector2{X := 32.0, Y := 32.0}
                                                    stack_box_slot:
                                                        HorizontalAlignment := horizontal_alignment.Fill
                                                        VerticalAlignment := vertical_alignment.Fill
                                                        Padding := margin{Left := 20.0, Top := 0.0, Right := 0.0, Bottom := 0.0}
                                                        Widget := text_block:
                                                            DefaultTextSize := ThemeTextSize.Small
                                                            DefaultText := MakeMessage("{AccoladeCount.Id}")
                                                            DefaultTextColor := ThemeColors.PrimaryText
                                        stack_box_slot:
                                            HorizontalAlignment := horizontal_alignment.Fill
                                            VerticalAlignment := vertical_alignment.Fill
                                            Distribution := option{0.5}
                                            Widget := text_block:
                                                DefaultTextSize := ThemeTextSize.Small
                                                DefaultText := MakeMessage("{AccoladeCount.Count}")
                                                DefaultTextColor := ThemeColors.Success
                                                DefaultShadowColor := ThemeColors.Success * 0.125
                                                DefaultShadowOffset := option{vector2{X := 2.0, Y := 2.0}}
                                                DefaultShadowOpacity := 1.0