using { /Fortnite.com/Devices }
using { /UnrealEngine.com/Temporary/SpatialMath }
using { /Verse.org/Assets }
using { /Verse.org/Simulation }

<# ----------------------------------------------------------------------------------------------------
    Data definitions
---------------------------------------------------------------------------------------------------- #>
accolade_id := enum:
    EquipWeapon
    StageCompleted
    SupplyDropSecured
    ItemShop
    CompleteGame
    Victory
    CompleteMission
    WeaponEvolved
    TreasureChest
    HealthOrb
    SurvivedWave
    PetSlime
    DailyQuest
    GoldenTree
    Magnet

ToString(Id:accolade_id)<transacts>:string =
    case (Id):
        accolade_id.EquipWeapon => "Equip New Weapon"
        accolade_id.StageCompleted => "Complete Stage"
        accolade_id.SupplyDropSecured => "Secure Supply Drop"
        accolade_id.ItemShop => "Item Shop Purchase"
        accolade_id.CompleteGame => "Complete Game"
        accolade_id.Victory => "Achieve Victory"
        accolade_id.CompleteMission => "Complete Mission"
        accolade_id.WeaponEvolved => "Evolve Weapon"
        accolade_id.TreasureChest => "Pickup Treasure Chest"
        accolade_id.HealthOrb => "Pickup Health Orb"
        accolade_id.SurvivedWave => "Survive Wave"
        accolade_id.PetSlime => "Pet The Slime"
        accolade_id.DailyQuest => "Complete Daily Quest"
        accolade_id.GoldenTree => "Gather Golden Tree"
        accolade_id.Magnet => "Pickup Magnet"

accolade_count_id := enum:
    EliminateSlimes
    LevelUp
    GatherTree
    EliminateSlimeMelee
    EliminateSlimeMeleeElite
    EliminateSlimeMeleeGeneral
    EliminateSlimeMiniboss
    EliminateSlimeBoss

ToString(Id:accolade_count_id)<transacts>:string =
    case (Id):
        accolade_count_id.EliminateSlimes => "Eliminations"
        accolade_count_id.LevelUp => "Level Up"
        accolade_count_id.GatherTree => "Gather Tree"
        accolade_count_id.EliminateSlimeMelee => "Eliminate Slime"
        accolade_count_id.EliminateSlimeMeleeElite => "Eliminate Slime Elite"
        accolade_count_id.EliminateSlimeMeleeGeneral => "Eliminate Slime General"
        accolade_count_id.EliminateSlimeMiniboss => "Eliminate Miniboss"
        accolade_count_id.EliminateSlimeBoss => "Eliminate Boss"

accolade := class<concrete>:
    @editable
    Id:accolade_id = accolade_id.Victory
    @editable
    Device:accolades_device = accolades_device{}
    @editable
    Icon:texture = UserInterface.Icons.Elimination_64
    @editable
    MaybeCooldown:?float = false

accolade_count := class<concrete>:
    @editable
    Id:accolade_count_id = accolade_count_id.EliminateSlimes
    @editable
    Device:accolades_device = accolades_device{}
    @editable
    Icon:texture = UserInterface.Icons.Elimination_64
    @editable
    Count:int = 1

<# ----------------------------------------------------------------------------------------------------
    Global functions
---------------------------------------------------------------------------------------------------- #>
var GlobalAccoladeManager:weak_map(session, accolade_manager) = map{}

GetAccoladeManager()<decides><transacts>:accolade_manager =
    GlobalAccoladeManager[GetSession()]

GetAccoladeInfo()<transacts>:tuple([]accolade, []accolade_count) =
    var Accolades:[]accolade = array{}
    var AccoladeCounts:[]accolade_count = array{}
    
    if (AccoladeManager := GetAccoladeManager[]):
        set Accolades = AccoladeManager.AccoladeDevices
        set AccoladeCounts = AccoladeManager.AccoladeCountDevices

    (Accolades, AccoladeCounts)

(Player:player).AwardAccolade(Id:accolade_id):void =
    if (AccoladeManager := GetAccoladeManager[]):
        AccoladeManager.Award(Player, Id)

AwardAccolade(Id:accolade_id, Players:[]player):void =
    if (AccoladeManager := GetAccoladeManager[]):
        AccoladeManager.Award(Id, Players)

(Player:player).AwardAccolade(Id:accolade_count_id):void =
    if (AccoladeManager := GetAccoladeManager[]):
        AccoladeManager.Award(Player, Id)

AwardAccolade(Id:accolade_count_id, Players:[]player):void =
    if (AccoladeManager := GetAccoladeManager[]):
        AccoladeManager.Award(Id, Players)

(Player:player).IsAccoladeReady(Id:accolade_id)<transacts>:logic =
    if (AccoladeManager := GetAccoladeManager[]):
        AccoladeManager.IsAccoladeReady(Player, Id)
    else:
        false

<# ----------------------------------------------------------------------------------------------------
    accolade_manager
---------------------------------------------------------------------------------------------------- #>
accolade_manager := class(creative_device):    
    @editable
    AccoladeDevices:[]accolade = array{}

    @editable
    AccoladeCountDevices:[]accolade_count = array{}

    @editable
    PetSlimeButton:button_device = button_device{}

    @editable
    PetSlimeSwitch:switch_device = switch_device{}

    @editable
    PetSlimeTimer:timer_device = timer_device{}

    @editable
    Button:button_device = button_device{}
    
    
    var MappedAccolades:[accolade_id]accolade = map{}
    var MappedAccoladeCounts:[accolade_count_id][]accolade_count = map{}
    
    var PlayerLastAwardTime:[player][accolade_id]float = map{}
    var PlayerAwardCount:[player][accolade_count_id]int = map{}

    <# ----------------------------------------------------------------------------------------------------
        Initialization
    ---------------------------------------------------------------------------------------------------- #>
    OnBegin<override>()<suspends>:void =
        if (set GlobalAccoladeManager[GetSession()] = Self) {}

        for (AccoladeDevice : AccoladeDevices):
            if (set MappedAccolades[AccoladeDevice.Id] = AccoladeDevice) {}

        for (AccoladeCountDevice : AccoladeCountDevices):
            if:
                not MappedAccoladeCounts[AccoladeCountDevice.Id]
                set MappedAccoladeCounts[AccoladeCountDevice.Id] = array{AccoladeCountDevice}
            else if:
                set MappedAccoladeCounts[AccoladeCountDevice.Id] += array{AccoladeCountDevice}
        
        GetPlayspace().PlayerAddedEvent().Subscribe(OnPlayerAdded)
        GetPlayspace().PlayerRemovedEvent().Subscribe(OnPlayerRemoved)
        PetSlimeButton.InteractedWithEvent.Subscribe(OnInteractedWithPetSlime)
        PetSlimeTimer.SuccessEvent.Subscribe(OnTimerSuccess)
        Button.InteractedWithEvent.Subscribe(OnInteractedWith)

        for (Player : GetPlayspace().GetPlayers()):
            Initialize(Player)

    Initialize(Player:player):void =
        if:
            not PlayerLastAwardTime[Player]
            set PlayerLastAwardTime[Player] = map{}

        if:
            not PlayerAwardCount[Player]
            set PlayerAwardCount[Player] = map{}

        PetSlimeTimer.SetActiveDuration(GetCooldown(accolade_id.PetSlime), Player)

    <# ----------------------------------------------------------------------------------------------------
        Get functions
    ---------------------------------------------------------------------------------------------------- #>
    GetDevice(Id:accolade_id)<decides><transacts>:accolade =
        MappedAccolades[Id]

    GetDevices(Id:accolade_count_id)<decides><transacts>:[]accolade_count =
        MappedAccoladeCounts[Id]
        
    GetCooldown(Id:accolade_id)<transacts>:float =
        GetDevice[accolade_id.PetSlime].MaybeCooldown? or 0.0

    IsAccoladeReady(Player:player, Id:accolade_id)<transacts>:logic =
        if:
            Accolade := GetDevice[Id]
            Cooldown := Accolade.MaybeCooldown?
            Cooldown > 0.0
        then:
            if:
                LastAwardTime := PlayerLastAwardTime[Player][Id]
                ElapsedTime := GetSimulationElapsedTime() - LastAwardTime
                ElapsedTime < Cooldown
            then:
                false
            else:
                true
        else:
            true

    <# ----------------------------------------------------------------------------------------------------
        Update functions
    ---------------------------------------------------------------------------------------------------- #>
    Award(Player:player, Id:accolade_id):void =
        if (Accolade := GetDevice[Id]):
            if (Cooldown := Accolade.MaybeCooldown?, Cooldown > 0.0):
                if (LastAwardTime := PlayerLastAwardTime[Player][Id]):
                    ElapsedTime := GetSimulationElapsedTime() - LastAwardTime

                    if (ElapsedTime >= Cooldown):
                        Accolade.Device.Award(Player)
                        if (set PlayerLastAwardTime[Player][Id] = GetSimulationElapsedTime()) {}
                else:
                    Accolade.Device.Award(Player)
                    if (set PlayerLastAwardTime[Player][Id] = GetSimulationElapsedTime()) {}
            else:
                Accolade.Device.Award(Player)

    Award(Id:accolade_id, Players:[]player):void =
        for (Player : Players):
            Award(Player, Id)

    Award(Player:player, Id:accolade_count_id):void =
        if (set PlayerAwardCount[Player][Id] = PlayerAwardCount[Player][Id] + 1 or 1) {}
        
        Accolades := for (Accolade : GetDevices[Id], Mod[PlayerAwardCount[Player][Id], Accolade.Count] = 0) {Accolade}

        if (Accolade:= Accolades[Accolades.Length - 1]):
            Accolade.Device.Award(Player)

    Award(Id:accolade_count_id, Players:[]player):void =
        for (Player : Players):
            Award(Player, Id)

    <# ----------------------------------------------------------------------------------------------------
        Event handlers
    ---------------------------------------------------------------------------------------------------- #>
    OnPlayerAdded(Player:player):void =
        Initialize(Player)

    OnPlayerRemoved(Player:player):void =
        set PlayerLastAwardTime = RemoveKeyFromMap(PlayerLastAwardTime, Player)
        set PlayerAwardCount = RemoveKeyFromMap(PlayerAwardCount, Player)

    OnInteractedWithPetSlime(Agent:agent):void =
        if (Player := player[Agent]):
            if (IsAccoladeReady(Player, accolade_id.PetSlime)?):
                Player.PlayAudio(audio_id.PetSlime)
                Player.AwardAccolade(accolade_id.PetSlime)
                SpawnVFX(vfx_id.Hearts, PetSlimeSwitch.GetTransform().Translation + vector3{Z := 100.0})
                PetSlimeSwitch.TurnOff(Player)
                PetSlimeTimer.Enable(Player)
                PetSlimeTimer.SetActiveDuration(GetCooldown(accolade_id.PetSlime), Player)
                PetSlimeTimer.Start(Player)

    OnTimerSuccess(MaybeAgent:?agent):void =
        if (Agent := MaybeAgent?):
            PetSlimeSwitch.TurnOn(Agent)
            PetSlimeTimer.Disable(Agent)


    OnInteractedWith(Agent:agent):void =
        if (Player := player[Agent]):
            Player.ShowPanelUI(panel_ui_id.Accolades)
            Player.PlayAudio(audio_id.Activate)
            