using { /Fortnite.com/UI }
using { /Verse.org/Assets }
using { /Verse.org/Colors }
using { /Verse.org/Simulation }
using { /UnrealEngine.com/Temporary/SpatialMath }
using { /UnrealEngine.com/Temporary/UI }

<# ----------------------------------------------------------------------------------------------------
    Global Functions
---------------------------------------------------------------------------------------------------- #>
GetArenaBackgroundImage(Level:int)<transacts>:texture =
    case (Level):
        1 => UserInterface.Minimaps.T_ArenaBackground_01_512
        2 => UserInterface.Minimaps.T_ArenaBackground_02_512
        3 => UserInterface.Minimaps.T_ArenaBackground_03_512
        4 => UserInterface.Minimaps.T_ArenaBackground_04_512
        5 => UserInterface.Minimaps.T_ArenaBackground_05_512
        6 => UserInterface.Minimaps.T_ArenaBackground_06_512
        7 => UserInterface.Minimaps.T_ArenaBackground_07_512
        8 => UserInterface.Minimaps.T_ArenaBackground_08_512
        9 => UserInterface.Minimaps.T_ArenaBackground_09_512
        _ => UserInterface.Minimaps.T_ArenaBackground_01_512

<# ----------------------------------------------------------------------------------------------------
    arena_info_ui
---------------------------------------------------------------------------------------------------- #>
arena_info_ui := class(ui_component):
    Icon:texture_block = texture_block:
        DefaultImage := UserInterface.Minimaps.T_Arena_01_128
        DefaultDesiredSize := vector2{X := 128.0, Y := 128.0}

    PreviewMaterial:material_block = material_block:
        DefaultImage := Materials.UI.M_GradientCircle_UI{}
        DefaultDesiredSize := vector2{X := 512.0, Y := 256.0}

    NameText:text_block = text_block:
        DefaultTextColor := ThemeColors.PrimaryText
        DefaultShadowOpacity := 0.5
        DefaultShadowOffset := option{vector2{X := 2.0, Y := 2.0}}

    LevelText:text_block = text_block:
        DefaultTextColor := ThemeColors.PrimaryText
        DefaultShadowOpacity := 0.5
        DefaultShadowOffset := option{vector2{X := 2.0, Y := 2.0}}

    CancelAnimationEvent:event() = event(){}
    FadeDuration:float = 1.0

    var PreviewMaterialInstance:Materials.UI.MI_ArenaPreview_UI = Materials.UI.MI_ArenaPreview_UI{}

    <# ----------------------------------------------------------------------------------------------------
        Initialization
    ---------------------------------------------------------------------------------------------------- #>
    CreateUI<override>():widget =
        set PreviewMaterialInstance = Materials.UI.MI_ArenaPreview_UI:
            BackgroundImage := UserInterface.Minimaps.T_ArenaBackground_01_512
            Icon := UserInterface.Slimes.Slime_Forest_Boss_128

        PreviewMaterial.SetImage(PreviewMaterialInstance)

        overlay:
            Slots := array:
                overlay_slot:
                    HorizontalAlignment := horizontal_alignment.Fill
                    VerticalAlignment := vertical_alignment.Fill
                    Widget := PreviewMaterial
                overlay_slot:
                    HorizontalAlignment := horizontal_alignment.Left
                    VerticalAlignment := vertical_alignment.Bottom
                    Padding := margin{Left := 20.0, Top := 0.0, Right := 0.0, Bottom := 10.0}
                    Widget := NameText
                overlay_slot:
                    HorizontalAlignment := horizontal_alignment.Left
                    VerticalAlignment := vertical_alignment.Top
                    Padding := margin{Left := 10.0, Top := 10.0, Right := 0.0, Bottom := 0.0}
                    Widget := overlay:
                        Slots := array:
                            overlay_slot:
                                HorizontalAlignment := horizontal_alignment.Fill
                                VerticalAlignment := vertical_alignment.Fill
                                Widget := UserInterface.VerseWidget.WBP_RoundedBox_Dark{}
                            overlay_slot:
                                Padding := margin{Left := 40.0, Top := 5.0, Right := 40.0, Bottom := 5.0}
                                Widget := stack_box:
                                    Orientation := orientation.Horizontal
                                    Slots := array:
                                        stack_box_slot:
                                            HorizontalAlignment := horizontal_alignment.Fill
                                            VerticalAlignment := vertical_alignment.Fill
                                            Widget := text_block:
                                                DefaultText := MakeMessage("Lv. ")
                                                DefaultTextColor := ThemeColors.Highlight
                                                DefaultShadowColor := ThemeColors.Highlight * 0.125
                                                DefaultShadowOffset := option{vector2{X := 2.0, Y := 2.0}}
                                                DefaultShadowOpacity := 1.0
                                        stack_box_slot:
                                            HorizontalAlignment := horizontal_alignment.Fill
                                            VerticalAlignment := vertical_alignment.Fill
                                            Widget := LevelText

    <# ----------------------------------------------------------------------------------------------------
        Update Functions
    ---------------------------------------------------------------------------------------------------- #>
    Update(Arena:arena):void =
        Icon.SetImage(Arena.Icon)
        NameText.SetText(MakeMessage(Arena.Name))
        
        SelectedArenaLevel := CalculateArenaLevel(Arena)
        LevelText.SetText(MakeMessage("{SelectedArenaLevel}"))

        if (NPCInfo := GetNPCInfo[Arena.BossId]):
            set PreviewMaterialInstance.BackgroundImage = GetArenaBackgroundImage(Arena.Level)
            set PreviewMaterialInstance.Icon = NPCInfo.Icon

        CancelAnimationEvent.Signal()
        spawn{FadeAnimation()}

    FadeAnimation()<suspends>:void =
        DeltaTimer := delta_timer{}
        
        defer:
            PreviewMaterial.SetTint(NamedColors.White)

        race:
            block:
                CancelAnimationEvent.Await()

            loop:
                Progress := DeltaTimer.GetElapsedTime() / FadeDuration
                PreviewMaterial.SetTint(NamedColors.White * Progress)

                if (DeltaTimer.GetElapsedTime() >= FadeDuration):
                    break

                Sleep(0.0)