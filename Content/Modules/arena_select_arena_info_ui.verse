using { /Fortnite.com/UI }
using { /Verse.org/Colors }
using { /UnrealEngine.com/Temporary/SpatialMath }
using { /UnrealEngine.com/Temporary/UI }

<# ----------------------------------------------------------------------------------------------------
    arena_select_arena_info_ui
---------------------------------------------------------------------------------------------------- #>
arena_select_arena_info_ui := class(ui_component):
    LabelPadding:float = 200.0
   
    DifficultyText:text_block = text_block:
        DefaultTextColor := ThemeColors.PrimaryText
        DefaultShadowOpacity := 0.5
        DefaultShadowOffset := option{vector2{X := 2.0, Y := 2.0}}

    StageText:text_block = text_block:
        DefaultTextColor := ThemeColors.PrimaryText
        DefaultShadowOpacity := 0.5
        DefaultShadowOffset := option{vector2{X := 2.0, Y := 2.0}}

    GoldMultiplierText:text_block = text_block:
        DefaultTextColor := ThemeColors.Success
        DefaultShadowColor := ThemeColors.Success * 0.125
        DefaultShadowOpacity := 1.0
        DefaultShadowOffset := option{vector2{X := 2.0, Y := 2.0}}

    HealthMultiplierText:text_block = text_block:
        DefaultTextColor := ThemeColors.Failure
        DefaultShadowColor := ThemeColors.Failure * 0.125
        DefaultShadowOpacity := 1.0
        DefaultShadowOffset := option{vector2{X := 2.0, Y := 2.0}}

    DamageMultiplierText:text_block = text_block:
        DefaultTextColor := ThemeColors.Failure
        DefaultShadowColor := ThemeColors.Failure * 0.125
        DefaultShadowOpacity := 1.0
        DefaultShadowOffset := option{vector2{X := 2.0, Y := 2.0}}

    MovespeedMultiplierText:text_block = text_block:
        DefaultTextColor := ThemeColors.Failure
        DefaultShadowColor := ThemeColors.Failure * 0.125
        DefaultShadowOpacity := 1.0
        DefaultShadowOffset := option{vector2{X := 2.0, Y := 2.0}}

    StageMultiplierText:text_block = text_block:
        DefaultTextColor := ThemeColors.Failure
        DefaultShadowColor := ThemeColors.Failure * 0.125
        DefaultShadowOpacity := 1.0
        DefaultShadowOffset := option{vector2{X := 2.0, Y := 2.0}}

    var MaybeArenaInfoUI:?arena_info_ui = false

    <# ----------------------------------------------------------------------------------------------------
        Initialization
    ---------------------------------------------------------------------------------------------------- #>
    CreateUI<override>():widget =
        ArenaInfoUI := arena_info_ui{Player := Player}
        ArenaInfoUI.Initialize()
        set MaybeArenaInfoUI = option{ArenaInfoUI}

        stack_box:
            Orientation := orientation.Vertical
            Slots := array:
                stack_box_slot:
                    HorizontalAlignment := horizontal_alignment.Fill
                    Widget := ArenaInfoUI.GetUI()
                stack_box_slot:
                    HorizontalAlignment := horizontal_alignment.Fill
                    Padding := margin{Left := 0.0, Top := 20.0, Right := 0.0, Bottom := 10.0}
                    Widget := text_block:
                        DefaultText := MakeMessage("Difficulty")
                        DefaultTextColor := ThemeColors.PrimaryText
                        DefaultTextOpacity := 0.5
                stack_box_slot:
                    HorizontalAlignment := horizontal_alignment.Fill
                    Widget := CreateStatUI(HealthMultiplierText, "Health", ?Padding := LabelPadding)
                stack_box_slot:
                    HorizontalAlignment := horizontal_alignment.Fill
                    Widget := CreateStatUI(DamageMultiplierText, "Damage", ?Padding := LabelPadding)
                stack_box_slot:
                    HorizontalAlignment := horizontal_alignment.Fill
                    Widget := CreateStatUI(MovespeedMultiplierText, "Movespeed", ?Padding := LabelPadding)
                stack_box_slot:
                    HorizontalAlignment := horizontal_alignment.Fill
                    Widget := CreateStatUI(StageText, "Stages", ?Padding := LabelPadding)
                stack_box_slot:
                    HorizontalAlignment := horizontal_alignment.Fill
                    Widget := CreateStatUI(StageMultiplierText, "Stage (HP/DMG)", ?Padding := LabelPadding)

                stack_box_slot:
                    HorizontalAlignment := horizontal_alignment.Fill
                    VerticalAlignment := vertical_alignment.Top
                    Padding := margin{Left := 0.0, Top := 40.0, Right := 0.0, Bottom := 40.0}
                    Widget := color_block:
                        DefaultColor := NamedColors.White
                        DefaultOpacity := 0.1
                        DefaultDesiredSize := vector2{X := 0.0, Y := 1.0}
                stack_box_slot:
                    HorizontalAlignment := horizontal_alignment.Fill
                    Padding := margin{Left := 0.0, Top := 0.0, Right := 0.0, Bottom := 10.0}
                    Widget := text_block:
                        DefaultText := MakeMessage("Rewards")
                        DefaultTextColor := ThemeColors.PrimaryText
                        DefaultTextOpacity := 0.5
                stack_box_slot:
                    HorizontalAlignment := horizontal_alignment.Fill
                    Widget := CreateStatUI(GoldMultiplierText, "Gold Multiplier", ?Padding := LabelPadding)

    <# ----------------------------------------------------------------------------------------------------
        Update Functions
    ---------------------------------------------------------------------------------------------------- #>
    Update(Arena:arena):void =
        if (ArenaInfoUI := MaybeArenaInfoUI?):
            ArenaInfoUI.Update(Arena)
        
        DifficultyText.SetText(MakeMessage("{Arena.Difficulty}"))

        StageText.SetText(MakeMessage("{GetArenaStageCount()}"))
        GoldMultiplierText.SetText(MakeMessage("+{FormatAsPercent(CalculateArenaGoldMultiplier(Arena))}"))
        
        ArenaHealthMultiplier := CalculateArenaHealthMultiplier(Arena)
        HealthMultiplierText.SetText(MakeMessage("+{FormatAsPercent(ArenaHealthMultiplier)}"))
        HealthMultiplierText.SetColor(ArenaHealthMultiplier > 0.0 and ThemeColors.Failure or ThemeColors.PrimaryText)
        
        ArenaDamageMultiplier := CalculateArenaDamageMultiplier(Arena)
        DamageMultiplierText.SetText(MakeMessage("+{FormatAsPercent(ArenaDamageMultiplier)}"))
        DamageMultiplierText.SetColor(ArenaDamageMultiplier > 0.0 and ThemeColors.Failure or ThemeColors.PrimaryText)

        ArenaMovespeedMultiplier := CalculateArenaMovespeedMultiplier(Arena)
        MovespeedMultiplierText.SetText(MakeMessage("+{FormatAsPercent(ArenaMovespeedMultiplier)}"))
        MovespeedMultiplierText.SetColor(ArenaMovespeedMultiplier > 0.0 and ThemeColors.Failure or ThemeColors.PrimaryText)
        
        ArenaStageMultiplier := NPCMultiplierPerArenaStage
        StageMultiplierText.SetText(MakeMessage("+{FormatAsPercent(ArenaStageMultiplier)}/Stage"))
        StageMultiplierText.SetColor(ArenaStageMultiplier > 0.0 and ThemeColors.Failure or ThemeColors.PrimaryText)

    Hide<override>():void =
        UI.SetVisibility(widget_visibility.Hidden)