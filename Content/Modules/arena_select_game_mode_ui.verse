using { /Fortnite.com/UI }
using { /Verse.org/Colors }
using { /UnrealEngine.com/Temporary/SpatialMath }
using { /UnrealEngine.com/Temporary/UI }

<# ----------------------------------------------------------------------------------------------------
    Data Definitions
---------------------------------------------------------------------------------------------------- #>
game_mode := enum:
    WeaponCount
    (game_mode:)PowerUps
    EndlessMode
    EnemyCount

ToString(GameMode:game_mode)<transacts>:string =
    case (GameMode):
        game_mode.WeaponCount => "Weapon Count"
        game_mode.PowerUps => "Power Ups"
        game_mode.EndlessMode => "Endless Mode"
        game_mode.EnemyCount => "Enemy Count"

<# ----------------------------------------------------------------------------------------------------
    arena_select_game_mode_ui
---------------------------------------------------------------------------------------------------- #>
arena_select_game_mode_ui := class(ui_component):
    WeaponsModeButton:button_quiet = button_quiet{}
    PowerUpsModeButton:button_quiet = button_quiet{}
    EndlessModeButton:button_quiet = button_quiet{}
    EnemyCountModeButton:button_quiet = button_quiet{}

    EndlessModeChangedEvent:events(logic) = events(logic){}
    
    <# ----------------------------------------------------------------------------------------------------
        Initialization
    ---------------------------------------------------------------------------------------------------- #>
    CreateUI<override>():widget =
        Update()

        stack_box:
            Orientation := orientation.Vertical
            Slots := array:
                stack_box_slot:
                    HorizontalAlignment := horizontal_alignment.Fill
                    VerticalAlignment := vertical_alignment.Fill
                    Widget := CreateEntryUI(game_mode.WeaponCount, "Set the max amount of weapons.\n(max banish -2 per weapon removed)", WeaponsModeButton)
                <#
                stack_box_slot:
                    HorizontalAlignment := horizontal_alignment.Fill
                    VerticalAlignment := vertical_alignment.Fill
                    Padding := margin{Left := 0.0, Top := 20.0, Right := 0.0, Bottom := 0.0}
                    Widget := CreateEntryUI(game_mode.PowerUps, "Individual - player's own power ups.\nShared - average of all players.", PowerUpsModeButton)
                #>
                stack_box_slot:
                    HorizontalAlignment := horizontal_alignment.Fill
                    VerticalAlignment := vertical_alignment.Fill
                    Padding := margin{Left := 0.0, Top := 20.0, Right := 0.0, Bottom := 0.0}
                    Widget := CreateEntryUI(game_mode.EndlessMode, "Infinite stages, enemy health/damage\nincreases exponentially each stage.", EndlessModeButton)
                stack_box_slot:
                    HorizontalAlignment := horizontal_alignment.Fill
                    VerticalAlignment := vertical_alignment.Fill
                    Padding := margin{Left := 0.0, Top := 20.0, Right := 0.0, Bottom := 0.0}
                    Widget := CreateEntryUI(game_mode.EnemyCount, "Set the max enemy count.", EnemyCountModeButton)

    CreateEntryUI(GameMode:game_mode, Description:string, Button:text_button_base):widget =
        ButtonHandler := case (GameMode):
            game_mode.WeaponCount => Button.OnClick().Subscribe(OnClickWeaponsMode)
            game_mode.PowerUps => Button.OnClick().Subscribe(OnClickPowerUpMode)
            game_mode.EndlessMode => Button.OnClick().Subscribe(OnClickEndlessMode)
            game_mode.EnemyCount => Button.OnClick().Subscribe(OnClickEnemyCount)

        stack_box:
            Orientation := orientation.Horizontal
            Slots := array:
                stack_box_slot:
                    HorizontalAlignment := horizontal_alignment.Fill
                    VerticalAlignment := vertical_alignment.Fill
                    Distribution := option{1.0}
                    Widget := stack_box:
                        Orientation := orientation.Vertical
                        Slots := array:
                            stack_box_slot:
                                HorizontalAlignment := horizontal_alignment.Fill
                                VerticalAlignment := vertical_alignment.Fill
                                Widget := text_block:
                                    DefaultText := MakeMessage("{GameMode}")
                                    DefaultTextColor := ThemeColors.PrimaryText
                                    DefaultShadowColor := NamedColors.Black
                                    DefaultShadowOffset := option{vector2{X := 2.0, Y := 2.0}}
                                    DefaultShadowOpacity := 0.5
                            stack_box_slot:
                                HorizontalAlignment := horizontal_alignment.Fill
                                VerticalAlignment := vertical_alignment.Fill
                                Widget := text_block:
                                    DefaultText := MakeMessage("{Description}")
                                    DefaultTextColor := ThemeColors.PrimaryText
                                    DefaultTextOpacity := 0.5
                stack_box_slot:
                    HorizontalAlignment := horizontal_alignment.Fill
                    VerticalAlignment := vertical_alignment.Fill
                    Distribution := option{0.4}
                    Padding := margin{Left := 20.0, Top := 0.0, Right := 0.0, Bottom := 0.0}
                    Widget := Button

    <# ----------------------------------------------------------------------------------------------------
        Update Functions
    ---------------------------------------------------------------------------------------------------- #>
    Update<override>():void =
        if (CurrentGameSettings := GetGameSettings[]):
            WeaponsModeButton.SetText(MakeMessage("{CurrentGameSettings.Weapons}"))
            PowerUpsModeButton.SetText(MakeMessage("{CurrentGameSettings.IsPowerUpsShared? and "Shared" or "Individual"}"))
            EndlessModeButton.SetText(MakeMessage("{CurrentGameSettings.IsEndless? and "ON" or "OFF"}"))
            EnemyCountModeButton.SetText(MakeMessage("{CurrentGameSettings.IsDoubleEnemyCount? and "90" or "45"}"))

    Hide<override>():void =
        UI.SetVisibility(widget_visibility.Hidden)

    <# ----------------------------------------------------------------------------------------------------
        Event Handlers
    ---------------------------------------------------------------------------------------------------- #>
    OnClickWeaponsMode(Message:widget_message):void =
        if (CurrentGameSettings := GetGameSettings[]):
            UpdatedWeapons := (CurrentGameSettings.Weapons = 4) and 1 or CurrentGameSettings.Weapons + 1
            UpdatedGameSettings := game_settings:
                MakeGameSettings<constructor>(CurrentGameSettings)
                Weapons := UpdatedWeapons
            
            SetGameSettings(UpdatedGameSettings)
            Update()

    OnClickPowerUpMode(Message:widget_message):void =
        if (CurrentGameSettings := GetGameSettings[]):
            UpdatedIsPowerUpsShared := logic{not CurrentGameSettings.IsPowerUpsShared?}
            UpdatedGameSettings := game_settings:
                MakeGameSettings<constructor>(CurrentGameSettings)
                IsPowerUpsShared := UpdatedIsPowerUpsShared
            
            SetGameSettings(UpdatedGameSettings)
            Update()

    OnClickEndlessMode(Message:widget_message):void =
        if (CurrentGameSettings := GetGameSettings[]):
            UpdatedIsEndless := logic{not CurrentGameSettings.IsEndless?}
            UpdatedGameSettings := game_settings:
                MakeGameSettings<constructor>(CurrentGameSettings)
                IsEndless := UpdatedIsEndless
            
            SetGameSettings(UpdatedGameSettings)
            EndlessModeChangedEvent.Signal(UpdatedIsEndless)
            Update()

    OnClickEnemyCount(Message:widget_message):void =
        if (CurrentGameSettings := GetGameSettings[]):
            UpdatedIsDoubleEnemyCount := logic{not CurrentGameSettings.IsDoubleEnemyCount?}
            UpdatedGameSettings := game_settings:
                MakeGameSettings<constructor>(CurrentGameSettings)
                IsDoubleEnemyCount := UpdatedIsDoubleEnemyCount
            
            SetGameSettings(UpdatedGameSettings)
            Update()