using { /Fortnite.com/UI }
using { /Verse.org/Colors }
using { /UnrealEngine.com/Temporary/SpatialMath }
using { /UnrealEngine.com/Temporary/UI }

<# ----------------------------------------------------------------------------------------------------
    arena_select_player_info_ui
---------------------------------------------------------------------------------------------------- #>
arena_select_player_info_ui := class(ui_component):
    LabelPadding:float = 200.0
    PlayersReadyText:text_block = text_block{}
    
    SelectedArenaLevelText:text_block = text_block:
        DefaultTextColor := ThemeColors.PrimaryText
        DefaultShadowOpacity := 0.5
        DefaultShadowOffset := option{vector2{X := 2.0, Y := 2.0}}

    var PlayerInfoEntryUIs:[]arena_select_player_info_entry_ui = array{}

    <# ----------------------------------------------------------------------------------------------------
        Initialization
    ---------------------------------------------------------------------------------------------------- #>
    CreateUI<override>():widget =
        stack_box:
            Orientation := orientation.Vertical
            Slots := array:
                stack_box_slot:
                    HorizontalAlignment := horizontal_alignment.Fill
                    Widget := CreateStatUI(PlayersReadyText, "Players Ready", ?Padding := LabelPadding)
                stack_box_slot:
                    HorizontalAlignment := horizontal_alignment.Fill
                    Widget := stack_box:
                        Orientation := orientation.Horizontal
                        Slots := array:
                             stack_box_slot:
                                HorizontalAlignment := horizontal_alignment.Left
                                Distribution := option{1.0}
                                Padding := margin{Left := 0.0, Top := 0.0, Right := LabelPadding, Bottom := 0.0}
                                Widget := text_block:
                                    DefaultText := MakeMessage("Selected Arena Level:")
                                    DefaultTextColor := ThemeColors.PrimaryText
                            stack_box_slot:
                                HorizontalAlignment := horizontal_alignment.Right
                                Widget := stack_box:
                                    Orientation := orientation.Horizontal
                                    Slots := array:
                                        stack_box_slot:
                                            HorizontalAlignment := horizontal_alignment.Fill
                                            VerticalAlignment := vertical_alignment.Fill
                                            Widget := text_block:
                                                DefaultText := MakeMessage("Lv. ")
                                                DefaultTextColor := ThemeColors.Highlight
                                                DefaultShadowColor := ThemeColors.Highlight * 0.125
                                                DefaultShadowOffset := option{vector2{X := 2.0, Y := 2.0}}
                                                DefaultShadowOpacity := 1.0
                                        stack_box_slot:
                                            HorizontalAlignment := horizontal_alignment.Fill
                                            VerticalAlignment := vertical_alignment.Fill
                                            Widget := SelectedArenaLevelText
                stack_box_slot:
                    HorizontalAlignment := horizontal_alignment.Fill
                    Padding := margin{Left := 0.0, Top := 40.0, Right := 0.0, Bottom := 40.0}
                    Widget := color_block:
                        DefaultColor := NamedColors.White
                        DefaultOpacity := 0.1
                        DefaultDesiredSize := vector2{X := 0.0, Y := 1.0}

                stack_box_slot:
                    HorizontalAlignment := horizontal_alignment.Fill
                    VerticalAlignment := vertical_alignment.Fill
                    Widget := stack_box:
                        Orientation := orientation.Vertical
                        Slots := for (Index := 1..MaxPlayers):
                            PlayerInfoEntryUI := arena_select_player_info_entry_ui{Player := Player}
                            PlayerInfoEntryUI.Initialize()
                            set PlayerInfoEntryUIs += array{PlayerInfoEntryUI}

                            stack_box_slot:
                                HorizontalAlignment := horizontal_alignment.Fill
                                VerticalAlignment := vertical_alignment.Fill
                                Padding := margin{Left := 0.0, Top := (Index = 1) and 0.0 or 10.0, Right := 0.0, Bottom := 0.0}
                                Widget := PlayerInfoEntryUI.GetUI()

    <# ----------------------------------------------------------------------------------------------------
        Update Functions
    ---------------------------------------------------------------------------------------------------- #>
    Update(Arena:arena):void =
        PlayersReadyText
        SelectedArenaLevelText

    UpdatePlayerInfo(CurrentArenaIndex:int):void =
        AllPlayers := GetAllPlayers()
        TeamArenaLevel := GetTeamArenaLevel()

        for (Index -> PlayerInfoEntryUI : PlayerInfoEntryUIs):
            if (AllPlayer := AllPlayers[Index]):
                PlayerInfoEntryUI.Show()
                PlayerInfoEntryUI.Update(AllPlayer, CurrentArenaIndex, logic{Index = TeamArenaLevel(0)})
            else:
                PlayerInfoEntryUI.Hide()

    Hide<override>():void =
        UI.SetVisibility(widget_visibility.Hidden)