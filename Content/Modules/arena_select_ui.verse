using { /Fortnite.com/Devices }
using { /Fortnite.com/UI }
using { /Verse.org/Colors }
using { /Verse.org/Concurrency }
using { /Verse.org/Simulation }
using { /UnrealEngine.com/Temporary/SpatialMath }
using { /UnrealEngine.com/Temporary/UI }

<# ----------------------------------------------------------------------------------------------------
    Global Functions
---------------------------------------------------------------------------------------------------- #>
GetTeamArenaLevel()<transacts>:tuple(int, int) =
    var MaxArenaProgressIndex:int = 0
    var MaxArenaProgress:int = 0

    for (Index -> AllPlayer : GetAllPlayers(), PlayerArenaProgress := AllPlayer.GetArenaProgress(), Index = 0 or PlayerArenaProgress > MaxArenaProgress):
        set MaxArenaProgressIndex = Index
        set MaxArenaProgress = PlayerArenaProgress

    (MaxArenaProgressIndex, MaxArenaProgress)

<# ----------------------------------------------------------------------------------------------------
    arena_select_ui
---------------------------------------------------------------------------------------------------- #>
arena_select_ui := class(panel_ui):
    Title<override>:string = "Arena Select"
    Anchors<override>:anchors = anchors{Minimum := vector2{X := 0.0, Y := 0.0}, Maximum := vector2{X := 1.0, Y := 1.0}}
    BackgroundOpacity<override>:type {_X:float where 0.000000 <= _X, _X <= 1.000000} = 0.0
    HasCloseButton<override>:logic = false
    LabelPadding:float = 200.0

    LoadArenaCancelledEvent:events(tuple()) = events(tuple()){}
    TimeoutCancelEvent:events(tuple()) = events(tuple()){}
    TimeoutDuration:float = 15.0
    TimeoutDeltaTimer:delta_timer = delta_timer{}

    StartButton:button_regular = button_regular{}
    InfoButton:button_quiet = button_quiet{}
    MenuButton:button_quiet = button_quiet{}

    var MaybeLoadingTask:?task(void) = false

    var CurrentArenaIndex:int = 0
    var CurrentDifficulty:int = 1
    var CurrentGameSpeed:int = 1
    
    var IsPlayerInfo:logic = false
    var IsSelectGameModes:logic = false
    var MaybeArenaInfoUI:?arena_select_arena_info_ui = false
    var MaybePlayerInfoUI:?arena_select_player_info_ui = false
    var MaybeGameModesUI:?arena_select_game_mode_ui = false
    var ArenaSettingsUI:widget = canvas{}
    var PlayerInfoEntryUIs:[]arena_select_player_info_entry_ui = array{}

    var ArenaColorBlocks:[]color_block = array{}
    var ArenaColorButtons:[]text_button_base = array{}
    var DifficultyColorBlocks:[]color_block = array{}
    var DifficultyColorButtons:[]text_button_base = array{}
    var GameSpeedColorBlocks:[]color_block = array{}
    var GameSpeedColorButtons:[]text_button_base = array{}
    var Subscriptions:[]cancelable = array{}
    var EventsSubscriptions:[]events_cancelable = array{}

    <# ----------------------------------------------------------------------------------------------------
        Initialization
    ---------------------------------------------------------------------------------------------------- #>
    CreateContent<override>():widget =
        if (GameManager := GetGameManager[], GameLobby := GetGameLobby[]):
            set Subscriptions = array:
                GameManager.GetPlayspace().PlayerAddedEvent().Subscribe(OnPlayerAdded)
                GameManager.GetPlayspace().PlayerRemovedEvent().Subscribe(OnPlayerRemoved)

            set EventsSubscriptions = array:
                ClosedEvent.Subscribe(OnClosed)
                GameManager.GameStartedEvent.Subscribe(OnGameStarted)
                GameLobby.PlayerReadiedEvent.Subscribe(UpdatePlayerReady)
                GameLobby.PlayerUnreadiedEvent.Subscribe(UpdatePlayerReady)

        ArenaInfoUI := arena_select_arena_info_ui{Player := Player}
        ArenaInfoUI.Initialize()
        set MaybeArenaInfoUI = option{ArenaInfoUI}

        PlayerInfoUI := arena_select_player_info_ui{Player := Player}
        PlayerInfoUI.Initialize()
        set MaybePlayerInfoUI = option{PlayerInfoUI}

        stack_box:
            Orientation := orientation.Horizontal
            Slots := array:               
                stack_box_slot:
                    HorizontalAlignment := horizontal_alignment.Fill
                    VerticalAlignment := vertical_alignment.Fill
                    Padding := margin{Left := 0.0, Top := 0.0, Right := 20.0, Bottom := 0.0}
                    Distribution := option{1.0}
                    Widget := CreateMenuUI()
                stack_box_slot:
                    HorizontalAlignment := horizontal_alignment.Fill
                    VerticalAlignment := vertical_alignment.Fill
                    Distribution := option{0.7}
                    Widget := overlay:
                        Slots := array:
                            overlay_slot:
                                HorizontalAlignment := horizontal_alignment.Fill
                                VerticalAlignment := vertical_alignment.Fill
                                Widget := UserInterface.VerseWidget.WBP_RoundedBox{}
                            overlay_slot:
                                HorizontalAlignment := horizontal_alignment.Fill
                                VerticalAlignment := vertical_alignment.Fill
                                Padding := margin{Left := 40.0, Top := 40.0, Right := 40.0, Bottom := 40.0}
                                Widget := stack_box:
                                    Orientation := orientation.Vertical
                                    Slots := array:
                                        stack_box_slot:
                                            HorizontalAlignment := horizontal_alignment.Fill
                                            VerticalAlignment := vertical_alignment.Fill
                                            Distribution := option{1.0}
                                            Widget := overlay:
                                                Slots := array:
                                                    overlay_slot:
                                                        HorizontalAlignment := horizontal_alignment.Fill
                                                        VerticalAlignment := vertical_alignment.Fill
                                                        Widget := ArenaInfoUI.GetUI()
                                                    overlay_slot:
                                                        HorizontalAlignment := horizontal_alignment.Fill
                                                        VerticalAlignment := vertical_alignment.Fill
                                                        Widget := PlayerInfoUI.GetUI()
                                        stack_box_slot:
                                            HorizontalAlignment := horizontal_alignment.Fill
                                            VerticalAlignment := vertical_alignment.Bottom
                                            Distribution := option{1.0}
                                            Padding := margin{Left := 0.0, Top := 20.0, Right := 0.0, Bottom := 20.0}
                                            Widget := color_block:
                                                DefaultColor := NamedColors.White
                                                DefaultOpacity := 0.1
                                                DefaultDesiredSize := vector2{X := 0.0, Y := 1.0}
                                        stack_box_slot:
                                            HorizontalAlignment := horizontal_alignment.Fill
                                            VerticalAlignment := vertical_alignment.Fill
                                            Widget := InfoButton

    Uninitialize<override>():void =
        for (EventsSubscription : EventsSubscriptions):
            EventsSubscription.Cancel()
    
        for (Subscription : Subscriptions):
            Subscription.Cancel()

    <# ----------------------------------------------------------------------------------------------------
        Menu UI
    ---------------------------------------------------------------------------------------------------- #>
    CreateMenuUI():widget =
        StartButton.SetText(MakeMessage("Start"))
        InfoButton.SetText(MakeMessage("Arena Info"))
        MenuButton.SetText(MakeMessage("Game Modes"))
        CloseButton := button_loud{DefaultText := MakeMessage("Close")}

        GameModesUI := arena_select_game_mode_ui{Player := Player}
        GameModesUI.Initialize()
        GameModesUI.Hide()
        set EventsSubscriptions += array{GameModesUI.EndlessModeChangedEvent.Subscribe(OnEndlessModeChanged)}
        set MaybeGameModesUI = option{GameModesUI}
        
        set Subscriptions += array:
            StartButton.OnClick().Subscribe(OnClickStart)
            InfoButton.OnClick().Subscribe(OnClickInfo)
            MenuButton.OnClick().Subscribe(OnClickMenu)
            CloseButton.OnClick().Subscribe(OnClickClose)

        overlay:
            Slots := array:
                overlay_slot:
                    HorizontalAlignment := horizontal_alignment.Fill
                    VerticalAlignment := vertical_alignment.Fill
                    Widget := UserInterface.VerseWidget.WBP_RoundedBox{}
                overlay_slot:
                    HorizontalAlignment := horizontal_alignment.Fill
                    VerticalAlignment := vertical_alignment.Fill
                    Padding := margin{Left := 40.0, Top := 40.0, Right := 40.0, Bottom := 40.0}
                    Widget := stack_box:
                        Orientation := orientation.Vertical
                        Slots := array:
                            stack_box_slot:
                                HorizontalAlignment := horizontal_alignment.Fill
                                VerticalAlignment := vertical_alignment.Fill
                                Widget := overlay:
                                    Slots := array:
                                        overlay_slot:
                                            HorizontalAlignment := horizontal_alignment.Fill
                                            VerticalAlignment := vertical_alignment.Fill
                                            Widget := set ArenaSettingsUI = CreateArenaSettingsUI()
                                        overlay_slot:
                                            HorizontalAlignment := horizontal_alignment.Fill
                                            VerticalAlignment := vertical_alignment.Fill
                                            Widget := GameModesUI.GetUI()
                            stack_box_slot:
                                HorizontalAlignment := horizontal_alignment.Fill
                                VerticalAlignment := vertical_alignment.Bottom
                                Distribution := option{1.0}
                                Padding := margin{Left := 0.0, Top := 40.0, Right := 0.0, Bottom := 0.0}
                                Widget := MenuButton
                            stack_box_slot:
                                HorizontalAlignment := horizontal_alignment.Fill
                                Padding := margin{Left := 0.0, Top := 20.0, Right := 0.0, Bottom := 20.0}
                                Widget := color_block:
                                    DefaultColor := NamedColors.White
                                    DefaultOpacity := 0.1
                                    DefaultDesiredSize := vector2{X := 0.0, Y := 1.0}                            
                            stack_box_slot:
                                HorizontalAlignment := horizontal_alignment.Fill
                                VerticalAlignment := vertical_alignment.Fill
                                Padding := margin{Left := 0.0, Top := 0.0, Right := 0.0, Bottom := 10.0}
                                Widget := StartButton
                            stack_box_slot:
                                HorizontalAlignment := horizontal_alignment.Fill
                                VerticalAlignment := vertical_alignment.Fill
                                Widget := CloseButton
    
    CreateArenaSettingsUI():widget =
        stack_box:
            Orientation := orientation.Vertical
            Slots := array:
                stack_box_slot:
                    HorizontalAlignment := horizontal_alignment.Fill
                    VerticalAlignment := vertical_alignment.Fill
                    Padding := margin{Left := 0.0, Top := 0.0, Right := 0.0, Bottom := 10.0}
                    Widget := text_block:
                        DefaultText := MakeMessage("Arena")
                        DefaultTextColor := ThemeColors.PrimaryText
                        DefaultShadowOffset := option{vector2{X := 2.0, Y := 2.0}}
                        DefaultShadowOpacity := 0.5
                stack_box_slot:
                    HorizontalAlignment := horizontal_alignment.Fill
                    Widget := CreateArenaSelectUI()
                stack_box_slot:
                    HorizontalAlignment := horizontal_alignment.Fill
                    VerticalAlignment := vertical_alignment.Fill
                    Padding := margin{Left := 0.0, Top := 40.0, Right := 0.0, Bottom := 10.0}
                    Widget := text_block:
                        DefaultText := MakeMessage("Difficulty")
                        DefaultTextColor := ThemeColors.PrimaryText
                        DefaultShadowOffset := option{vector2{X := 2.0, Y := 2.0}}
                        DefaultShadowOpacity := 0.5
                stack_box_slot:
                    HorizontalAlignment := horizontal_alignment.Fill
                    Widget := CreateDifficultySelectUI()
                stack_box_slot:
                    HorizontalAlignment := horizontal_alignment.Fill
                    VerticalAlignment := vertical_alignment.Fill
                    Padding := margin{Left := 0.0, Top := 40.0, Right := 0.0, Bottom := 10.0}
                    Widget := text_block:
                        DefaultText := MakeMessage("Game Speed")
                        DefaultTextColor := ThemeColors.PrimaryText
                        DefaultShadowOffset := option{vector2{X := 2.0, Y := 2.0}}
                        DefaultShadowOpacity := 0.5
                stack_box_slot:
                    HorizontalAlignment := horizontal_alignment.Fill
                    Widget := CreateGameSpeedSelectUI()

    CreateArenaSelectUI():widget =
        stack_box:
            Orientation := orientation.Horizontal
            Slots := for (Index := 1..GetArenaCount()):
                ColorBlock := color_block{DefaultColor := ThemeColors.Loud, DefaultDesiredSize := vector2{Y := 5.0}}
                ColorBlock.SetVisibility(widget_visibility.Hidden)
                Button := button_quiet{DefaultText := MakeMessage("{Index}")}
                spawn{AwaitArenaClicked(Button.OnClick(), Index - 1)}

                set ArenaColorBlocks += array{ColorBlock}
                set ArenaColorButtons += array{Button}

                stack_box_slot:
                    HorizontalAlignment := horizontal_alignment.Fill
                    VerticalAlignment := vertical_alignment.Fill
                    Distribution := option{1.0}
                    Padding := margin{Left := (Index = 1) and 0.0 or 10.0, Top := 0.0, Right := 0.0, Bottom := 0.0}
                    Widget := overlay:
                        Slots := array:
                            overlay_slot:
                                HorizontalAlignment := horizontal_alignment.Fill
                                VerticalAlignment := vertical_alignment.Bottom
                                Padding := margin{Left := 10.0, Top := 0.0, Right := 10.0, Bottom := -10.0}
                                Widget := ColorBlock
                            overlay_slot:
                                HorizontalAlignment := horizontal_alignment.Fill
                                VerticalAlignment := vertical_alignment.Fill
                                Widget := Button

    CreateDifficultySelectUI():widget =
        stack_box:
            Orientation := orientation.Horizontal
            Slots := for (Index := 1..MaxDifficultyLevel):
                ColorBlock := color_block{DefaultColor := ThemeColors.Loud, DefaultDesiredSize := vector2{Y := 5.0}}
                ColorBlock.SetVisibility(widget_visibility.Hidden)
                Button := button_quiet{DefaultText := MakeMessage("{Index}")}
                spawn{AwaitDifficultyClicked(Button.OnClick(), Index)}

                set DifficultyColorBlocks += array{ColorBlock}
                set DifficultyColorButtons += array{Button}

                stack_box_slot:
                    HorizontalAlignment := horizontal_alignment.Fill
                    VerticalAlignment := vertical_alignment.Fill
                    Distribution := option{1.0}
                    Padding := margin{Left := (Index = 1) and 0.0 or 10.0, Top := 0.0, Right := 0.0, Bottom := 0.0}
                    Widget := overlay:
                        Slots := array:
                            overlay_slot:
                                HorizontalAlignment := horizontal_alignment.Fill
                                VerticalAlignment := vertical_alignment.Bottom
                                Padding := margin{Left := 10.0, Top := 0.0, Right := 10.0, Bottom := -10.0}
                                Widget := ColorBlock
                            overlay_slot:
                                HorizontalAlignment := horizontal_alignment.Fill
                                VerticalAlignment := vertical_alignment.Fill
                                Widget := Button

    CreateGameSpeedSelectUI():widget =
        stack_box:
            Orientation := orientation.Horizontal
            Slots := for (Index := 1..MaxGameSpeed):
                ColorBlock := color_block{DefaultColor := ThemeColors.Loud, DefaultDesiredSize := vector2{Y := 5.0}}
                ColorBlock.SetVisibility(widget_visibility.Hidden)
                Button := button_quiet{DefaultText := MakeMessage("x{Index}")}
                spawn{AwaitGameSpeedClicked(Button.OnClick(), Index)}

                set GameSpeedColorBlocks += array{ColorBlock}
                set GameSpeedColorButtons += array{Button}

                stack_box_slot:
                    HorizontalAlignment := horizontal_alignment.Fill
                    VerticalAlignment := vertical_alignment.Fill
                    Distribution := option{1.0}
                    Padding := margin{Left := (Index = 1) and 0.0 or 10.0, Top := 0.0, Right := 0.0, Bottom := 0.0}
                    Widget := overlay:
                        Slots := array:
                            overlay_slot:
                                HorizontalAlignment := horizontal_alignment.Fill
                                VerticalAlignment := vertical_alignment.Bottom
                                Padding := margin{Left := 10.0, Top := 0.0, Right := 10.0, Bottom := -10.0}
                                Widget := ColorBlock
                            overlay_slot:
                                HorizontalAlignment := horizontal_alignment.Fill
                                VerticalAlignment := vertical_alignment.Fill
                                Widget := Button    

    <# ----------------------------------------------------------------------------------------------------
        Get Functions
    ---------------------------------------------------------------------------------------------------- #>
    IsArenaUnlocked(ArenaLevel:int, Difficulty:int)<transacts>:logic =
        for (AllPlayer : GetAllPlayers(), AllPlayer.GetArenaProgress() + 1 >= CalculateArenaLevel(ArenaLevel, Difficulty)):
            return true

        return false

    IsDifficultyUnlocked(Difficulty:int)<transacts>:logic =
        if (GetGameSettings[].IsEndless? and Difficulty > 1):
            return false

        for (AllPlayer : GetAllPlayers(), AllPlayer.GetArenaProgress() >= GetArenaCount() * (Difficulty - 1)):
            return true

        return false

    IsGameSpeedUnlocked(GameSpeed:int)<transacts>:logic =
        for (AllPlayer : GetAllPlayers(), (GameSpeed = 1) or (AllPlayer.GetArenaProgress() >= GetGameSpeedUnlockLevel(GameSpeed))):
            return true

        return false

    
    GetGameSpeedUnlockLevel(GameSpeed:int)<transacts>:int =
        case (GameSpeed):
            1 => 0
            2 => 9
            3 => 18
            4 => 36
            5 => 54
            6 => 72
            _ => 36

    <# ----------------------------------------------------------------------------------------------------
        Update Functions
    ---------------------------------------------------------------------------------------------------- #>
    Update()<suspends>:void =
        if (LoadingTask := MaybeLoadingTask?):
            LoadingTask.Await()
            return

        if (GameSettings := GetGameSettings[]):
            set CurrentArenaIndex = GameSettings.Arena.Level - 1

            if (ArenaInfoUI := MaybeArenaInfoUI?, PlayerInfoUI := MaybePlayerInfoUI?):
                if (IsPlayerInfo?):
                    ArenaInfoUI.Hide()
                    PlayerInfoUI.Show()
                    PlayerInfoUI.UpdatePlayerInfo(CurrentArenaIndex)
                else:
                    ArenaInfoUI.Show()
                    PlayerInfoUI.Hide()

            UpdateInfo(GameSettings.Arena)
            UpdateDifficulty(GameSettings.Difficulty)
            UpdateGameSpeed(GameSettings.GameSpeed)
            UpdatePlayerReady()
        else:
            LoadingTask := spawn{LoadArena()}
            set MaybeLoadingTask = option{LoadingTask}

    UpdateInfo(Arena:arena):void =
        SelectedArenaLevel := CalculateArenaLevel(Arena)
        
        if (PlayerInfoUI := MaybePlayerInfoUI?):
            PlayerInfoUI.SelectedArenaLevelText.SetText(MakeMessage("{SelectedArenaLevel}"))

        if (ArenaInfoUI := MaybeArenaInfoUI?):
            ArenaInfoUI.Update(Arena)

        GameDifficulty := GetGameDifficulty()
        for (Index -> ArenaColorButton : ArenaColorButtons):
            ArenaLevel := Index + 1
            if (IsArenaUnlocked(ArenaLevel, GameDifficulty)?):
                ArenaColorButton.SetText(MakeMessage("{ArenaLevel}"))
            else:
                ArenaColorButton.SetText(MakeMessage("🔒"))
        
        for (Index -> DifficultyColorButton : DifficultyColorButtons):
            Difficulty := Index + 1
            if (IsDifficultyUnlocked(Difficulty)?):
                DifficultyColorButton.SetText(MakeMessage("{Difficulty}"))
            else:
                DifficultyColorButton.SetText(MakeMessage("🔒"))

        for (Index -> GameSpeedColorButton : GameSpeedColorButtons):
            GameSpeed := Index + 1
            if (IsGameSpeedUnlocked(GameSpeed)?):
                GameSpeedColorButton.SetText(MakeMessage("x{GameSpeed}"))
            else:
                GameSpeedColorButton.SetText(MakeMessage("🔒 Lv. {GetGameSpeedUnlockLevel(GameSpeed)}"))

        for (Index -> ColorBlock : ArenaColorBlocks):
            ColorBlock.SetVisibility(Index = Arena.Level - 1 and widget_visibility.Visible or widget_visibility.Hidden)

        if (PlayerInfoUI := MaybePlayerInfoUI?):
            PlayerInfoUI.UpdatePlayerInfo(CurrentArenaIndex)

    UpdateMenu():void =
        if (GameModesUI := MaybeGameModesUI?):
            if (IsSelectGameModes?):
                ArenaSettingsUI.SetVisibility(widget_visibility.Hidden)
                GameModesUI.Show()
            else:
                ArenaSettingsUI.SetVisibility(widget_visibility.Visible)
                GameModesUI.Hide()

    LoadArena()<suspends>:void =
        race:           
            LoadArenaCancelledEvent.Await()
            block:
                if (Arena := GetArenas()[CurrentArenaIndex], ArenaManager := GetArenaManager[]):
                    if:
                        UpdatedGameSettings := game_settings:
                            MakeGameSettings<constructor>(GetGameSettings[])
                                Arena := Arena
                    then:
                        SetGameSettings(UpdatedGameSettings)
                        
                    UpdateInfo(Arena)
                    StartButton.SetText(MakeMessage("Loading ..."))
                    spawn{ArenaManager.LoadArena(CurrentArenaIndex)}
                    Sleep(ArenaManager.GetLoadTimeRemaining())

        set MaybeLoadingTask = false
        UpdatePlayerReady()

    UpdatePlayerReady():void =
        if (MaybeLoadingTask?) then return

        if (PlayerInfoUI := MaybePlayerInfoUI?):
            PlayerInfoUI.PlayersReadyText.SetText(MakeMessage("{GetReadyPlayerCount()}/{GetAllPlayers().Length}"))
            PlayerInfoUI.PlayersReadyText.SetColor(IsAllPlayerReady()? and ThemeColors.Success or ThemeColors.Failure)

        if (IsGameCountdownInProgress()?):
            StartButton.SetText(MakeMessage("Cancel Countdown"))
        else:
            StartButton.SetText(MakeMessage("Start Countdown"))

    UpdateArena(UpdatedIndex:int):void =
        if (IsGameInProgress()?):
            return

        LoadArenaCancelledEvent.Signal()
        set CurrentArenaIndex = UpdatedIndex

        LoadingTask := spawn{LoadArena()}
        set MaybeLoadingTask = option{LoadingTask}

    UpdateDifficulty(Difficulty:int):void =
        set CurrentDifficulty = Difficulty

        if:
            UpdatedGameSettings := game_settings:
                MakeGameSettings<constructor>(GetGameSettings[])
                    Difficulty := Difficulty
        then:
            SetGameSettings(UpdatedGameSettings)

        for (Index -> ColorBlock : DifficultyColorBlocks):
            ColorBlock.SetVisibility(Index = Difficulty - 1 and widget_visibility.Visible or widget_visibility.Hidden)

        if (Arena := GetArenas()[CurrentArenaIndex]):
            UpdateInfo(Arena)

        if (not IsArenaUnlocked(CurrentArenaIndex + 1, Difficulty)?):
            UpdateArena(0)

    UpdateGameSpeed(GameSpeed:int):void =
        set CurrentGameSpeed = GameSpeed

        if:
            UpdatedGameSettings := game_settings:
                MakeGameSettings<constructor>(GetGameSettings[])
                    GameSpeed := GameSpeed
        then:
            SetGameSettings(UpdatedGameSettings)

        for (Index -> ColorBlock : GameSpeedColorBlocks):
            ColorBlock.SetVisibility(Index = GameSpeed - 1 and widget_visibility.Visible or widget_visibility.Hidden)

        if (Arena := GetArenas()[CurrentArenaIndex]):
            UpdateInfo(Arena)

    Timeout()<suspends>:void =
        TimeoutDeltaTimer.Reset()

        race:
            TimeoutCancelEvent.Await()
            loop:
                Sleep(1.0)
                if (TimeoutDeltaTimer.GetElapsedTime() >= TimeoutDuration):
                    Hide()
                    break

    <# ----------------------------------------------------------------------------------------------------
        Event Handlers
    ---------------------------------------------------------------------------------------------------- #>
    AwaitArenaClicked(Event:listenable(widget_message), ArenaIndex:int)<suspends>:void =
        if (PlayerRemovedEvent := Player.GetPlayerRemovedEvent[]):
            race:
                PlayerRemovedEvent.Await()
                loop:
                    Event.Await()
                    TimeoutDeltaTimer.Reset()

                    if (IsGameCountdownInProgress()? or not IsArenaUnlocked(ArenaIndex + 1, GetGameDifficulty())?):
                        Player.PlayAudio(audio_id.Error)
                    else if (ArenaIndex <> CurrentArenaIndex):
                        UpdateArena(ArenaIndex)

    AwaitDifficultyClicked(Event:listenable(widget_message), Difficulty:int)<suspends>:void =
        if (PlayerRemovedEvent := Player.GetPlayerRemovedEvent[]):
            race:
                PlayerRemovedEvent.Await()
                loop:
                    Event.Await()
                    TimeoutDeltaTimer.Reset()

                    if (IsGameCountdownInProgress()? or not IsDifficultyUnlocked(Difficulty)?):
                        Player.PlayAudio(audio_id.Error)
                    else if (Difficulty <> CurrentDifficulty):
                        UpdateDifficulty(Difficulty)

    AwaitGameSpeedClicked(Event:listenable(widget_message), GameSpeed:int)<suspends>:void =
        if (PlayerRemovedEvent := Player.GetPlayerRemovedEvent[]):
            race:
                PlayerRemovedEvent.Await()
                loop:
                    Event.Await()
                    TimeoutDeltaTimer.Reset()

                    if (IsGameCountdownInProgress()? or not IsGameSpeedUnlocked(GameSpeed)?):
                        Player.PlayAudio(audio_id.Error)
                    else if (GameSpeed <> CurrentGameSpeed):
                        UpdateGameSpeed(GameSpeed)

    OnClickStart(Message:widget_message):void =
        TimeoutDeltaTimer.Reset()

        if (IsGameInProgress()? or MaybeLoadingTask?):
            Player.PlayAudio(audio_id.Error)
            return

        if (IsGameCountdownInProgress()?):
            StopGameCountdown()
            spawn{Update()}
            return

        TimeoutCancelEvent.Signal()
        Hide()
        StartGameCountdown(Player, CurrentArenaIndex)

    OnClickInfo(Message:widget_message):void =
        set IsPlayerInfo = logic{not IsPlayerInfo?}
        InfoButton.SetText(MakeMessage("{IsPlayerInfo? and "Arena" or "Player"} Info"))

        if (ArenaInfoUI := MaybeArenaInfoUI?, PlayerInfoUI := MaybePlayerInfoUI?):
            if (IsPlayerInfo?):
                ArenaInfoUI.Hide()
                PlayerInfoUI.Show()
            else:
                ArenaInfoUI.Show()
                PlayerInfoUI.Hide()

    OnClickMenu(Message:widget_message):void =
        set IsSelectGameModes = logic{not IsSelectGameModes?}
        MenuButton.SetText(MakeMessage("{IsSelectGameModes? and "Arena Settings" or "Game Modes"}"))
        UpdateMenu()

    OnShow<override>():void =
        set IsSelectGameModes = false
        UpdateMenu()
        if (GetAllPlayers().Length > 1):
            spawn{Timeout()}

    OnClosed(InPlayer:player):void =
        TimeoutCancelEvent.Signal()

    OnGameStarted():void =
        TimeoutCancelEvent.Signal()

    OnPlayerAdded(InPlayer:player):void =
        if (not IsShowing?) then return
        spawn{HandlePlayerAdded(InPlayer)}

    HandlePlayerAdded(InPlayer:player)<suspends>:void =
        TimeoutCancelEvent.Signal()

        WaitForPlayerDataManager(InPlayer)

        if (PlayerInfoUI := MaybePlayerInfoUI?):
            PlayerInfoUI.UpdatePlayerInfo(CurrentArenaIndex)

        UpdatePlayerReady()
        Timeout()

    OnPlayerRemoved(InPlayer:player):void =
        if (not IsShowing?) then return

        if (PlayerInfoUI := MaybePlayerInfoUI?):
            PlayerInfoUI.UpdatePlayerInfo(CurrentArenaIndex)

        UpdatePlayerReady()

        var IsUpdated:logic = false
        if (not IsArenaUnlocked(CurrentArenaIndex + 1, GetGameDifficulty())?):
            UpdateArena(0)
            set IsUpdated = true

        if (not IsDifficultyUnlocked(CurrentDifficulty)?):
            UpdateDifficulty(1)
            set IsUpdated = true

        if (not IsGameSpeedUnlocked(CurrentGameSpeed)?):
            UpdateGameSpeed(1)
            set IsUpdated = true

        if (not IsUpdated?):
            UpdateGameSpeed(CurrentGameSpeed)

    OnEndlessModeChanged(IsEndless:logic):void =
        if (IsEndless?):
            UpdateDifficulty(1)
        else if (Arena := GetArenas()[CurrentArenaIndex]):
            UpdateInfo(Arena)