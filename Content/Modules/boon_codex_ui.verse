using { /Fortnite.com/UI }
using { /Verse.org/Assets }
using { /Verse.org/Colors }
using { /UnrealEngine.com/Temporary/SpatialMath }
using { /UnrealEngine.com/Temporary/UI }

boon_codex_ui := class(panel_ui):
    Title<override>:string = "Boons Codex"
    BackgroundOpacity<override>:type {_X:float where 0.000000 <= _X, _X <= 1.000000} = 0.7
    Anchors<override>:anchors = anchors{Minimum := vector2{X := 0.05, Y := 0.0}, Maximum := vector2{X := 0.95, Y := 1.0}}

    NameText:text_block = text_block:
        DefaultTextColor := ThemeColors.PrimaryText
        DefaultShadowOffset := option{vector2{X := 2.0, Y := 2.0}}
        DefaultShadowOpacity := 0.5

    Icon:texture_block = texture_block:
        DefaultImage := UserInterface.Icons.WeaponBow_128
        DefaultDesiredSize := vector2{X := 128.0, Y := 128.0}

    TierText:text_block = text_block{}

    DescriptionText:text_block = text_block:
        DefaultTextColor := ThemeColors.PrimaryText
        DefaultTextOpacity := 0.5

    TypeText:text_block = text_block:
        DefaultTextColor := ThemeColors.Success
        DefaultShadowOffset := option{vector2{X := 2.0, Y := 2.0}}
        DefaultShadowColor := ThemeColors.Success * 0.125
        DefaultShadowOpacity := 0.5

    GridRow:int = 3
    GridColumn:int = 5

    var SelectedIndex:int = 0
    var BoonCodexEntries:[]texture_block = array{}

    <# ----------------------------------------------------------------------------------------------------
        Initialization
    ---------------------------------------------------------------------------------------------------- #>
    CreateContent<override>():widget =
        stack_box:
            Orientation := orientation.Horizontal
            Slots := array:
                stack_box_slot:
                    HorizontalAlignment := horizontal_alignment.Fill
                    VerticalAlignment := vertical_alignment.Fill
                    Distribution := option{1.0}
                    Widget := CreateGrid()
                stack_box_slot:
                    HorizontalAlignment := horizontal_alignment.Fill
                    VerticalAlignment := vertical_alignment.Fill
                    Distribution := option{1.0}
                    Padding := margin{Left := 20.0, Top := 0.0, Right := 0.0, Bottom := 0.0}
                    Widget := CreateInfoUI()

    CreateGrid():widget =
        overlay:
            Slots := array:
                overlay_slot:
                    HorizontalAlignment := horizontal_alignment.Fill
                    VerticalAlignment := vertical_alignment.Fill
                    Widget := UserInterface.VerseWidget.WBP_RoundedBox{}
                overlay_slot:
                    HorizontalAlignment := horizontal_alignment.Fill
                    VerticalAlignment := vertical_alignment.Fill
                    Padding := margin{Left := 40.0, Top := 40.0, Right := 40.0, Bottom := 40.0}
                    Widget := stack_box:
                        Orientation := orientation.Vertical
                        Slots := for (RowIndex := 0..GridRow - 1):
                            stack_box_slot:
                                HorizontalAlignment := horizontal_alignment.Fill
                                VerticalAlignment := vertical_alignment.Fill
                                Distribution := option{1.0}
                                Padding := margin{Left := 0.0, Top := (RowIndex = 0) and 0.0 or 20.0, Right := 0.0, Bottom := 0.0}
                                Widget := stack_box:
                                    Orientation := orientation.Horizontal
                                    Slots := for (ColumnIndex := 0..GridColumn - 1):
                                        EntryIndex := GridColumn * RowIndex + ColumnIndex
                                        
                                        EntryIcon:texture_block = texture_block:
                                            DefaultImage := UserInterface.Icons.Boon_64
                                            DefaultDesiredSize := vector2{X := 64.0, Y := 64.0}

                                        Button := button_quiet{}

                                        spawn{AwaitButtonClicked(Button.OnClick(), EntryIndex)}
                                        set BoonCodexEntries += array{EntryIcon}

                                        stack_box_slot:
                                            HorizontalAlignment := horizontal_alignment.Fill
                                            VerticalAlignment := vertical_alignment.Fill
                                            Distribution := option{1.0}
                                            Padding := margin{Left := (ColumnIndex = 0) and 0.0 or 20.0, Top := 0.0, Right := 0.0, Bottom := 0.0}
                                            Widget := overlay:
                                                Slots := array:
                                                    overlay_slot:
                                                        Padding := margin{Left := 20.0, Top := 40.0, Right := 20.0, Bottom := 40.0}
                                                        Widget := EntryIcon
                                                    overlay_slot:
                                                        HorizontalAlignment := horizontal_alignment.Fill
                                                        VerticalAlignment := vertical_alignment.Fill
                                                        Widget := Button

    CreateInfoUI():widget =
        overlay:
            Slots := array:
                overlay_slot:
                    HorizontalAlignment := horizontal_alignment.Fill
                    VerticalAlignment := vertical_alignment.Fill
                    Widget := UserInterface.VerseWidget.WBP_RoundedBox{}
                overlay_slot:
                    HorizontalAlignment := horizontal_alignment.Fill
                    VerticalAlignment := vertical_alignment.Fill
                    Padding := margin{Left := 40.0, Top := 40.0, Right := 40.0, Bottom := 40.0}
                    Widget := stack_box:
                        Orientation := orientation.Vertical
                        Slots := array:
                            stack_box_slot:
                                HorizontalAlignment := horizontal_alignment.Center
                                VerticalAlignment := vertical_alignment.Fill
                                Widget := NameText
                            stack_box_slot:
                                VerticalAlignment := vertical_alignment.Top
                                Padding := margin{Left := 0.0, Top := 20.0, Right := 0.0, Bottom := 20.0}
                                Widget := Icon
                            stack_box_slot:
                                HorizontalAlignment := horizontal_alignment.Center
                                VerticalAlignment := vertical_alignment.Fill
                                Widget := TierText
                            stack_box_slot:
                                HorizontalAlignment := horizontal_alignment.Fill
                                VerticalAlignment := vertical_alignment.Top
                                Padding := margin{Left := 0.0, Top := 20.0, Right := 0.0, Bottom := 40.0}
                                Widget := color_block:
                                    DefaultColor := NamedColors.White
                                    DefaultOpacity := 0.1
                                    DefaultDesiredSize := vector2{X := 0.0, Y := 1.0}
                            stack_box_slot:
                                HorizontalAlignment := horizontal_alignment.Left
                                VerticalAlignment := vertical_alignment.Top
                                Widget := DescriptionText
                            stack_box_slot:
                                HorizontalAlignment := horizontal_alignment.Center
                                VerticalAlignment := vertical_alignment.Bottom
                                Distribution := option{1.0}
                                Widget := TypeText

    <# ----------------------------------------------------------------------------------------------------
        Update Functions
    ---------------------------------------------------------------------------------------------------- #>
    Update():void =
        BoonIds := GetBoonIds()
        for (Index -> BoonCodexEntry : BoonCodexEntries, BoonInfo := GetBoonInfo[BoonIds[Index]]):
            BoonCodexEntry.SetImage(BoonInfo.IconSmall)

            if (Index = SelectedIndex):
                BoonCodexEntry.SetTint(ThemeColors.Highlight * 2.0)
                UpdateInfo(BoonInfo)
            else:
                BoonCodexEntry.SetTint(NamedColors.White)

    UpdateInfo(BoonInfo:boon_info):void =
        NameText.SetText(MakeMessage("{BoonInfo.Name}"))
        Icon.SetImage(BoonInfo.Icon)
        TierText.SetText(MakeMessage("{BoonInfo.Tier.ToTierString().ToUpper()}"))
        TierText.SetColor(BoonInfo.Tier.ToTierColor() * 2)
        DescriptionText.SetText(MakeMessage("{BoonInfo.Description.WrapText(700.0)}"))
        TypeText.SetText(MakeMessage("{BoonInfo.Type}"))

    <# ----------------------------------------------------------------------------------------------------
        Event Handlers
    ---------------------------------------------------------------------------------------------------- #>
    OnShow<override>():void =
        Update()
        if (BoonInfo := GetBoonInfo[GetBoonIds()[0]]):
            UpdateInfo(BoonInfo)

    AwaitButtonClicked(Listenable:listenable(widget_message), Index:int)<suspends>:void =
        if (PlayerRemovedEvent := Player.GetPlayerRemovedEvent[]):
            race:
                PlayerRemovedEvent.Await()
                loop:
                    Listenable.Await()
                    set SelectedIndex = Index
                    Update()
                    Player.PlayAudio(audio_id.Activate)