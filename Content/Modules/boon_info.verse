using { /Fortnite.com/Characters }
using { /Verse.org/Assets }
using { /Verse.org/Colors }
using { /Verse.org/Random }
using { /Verse.org/Simulation }
using { /UnrealEngine.com/Temporary/UI }

<# ----------------------------------------------------------------------------------------------------
    Constants
---------------------------------------------------------------------------------------------------- #>
BoonBossHunterMultiplier:float = 0.5
BoonBossSiphonAmount:float = 50.0
BoonImmortalHeartAmount:float = 25.0
BoonTreasureMapAmount:int = 3
MaxBoonCount:int = 10

<# ----------------------------------------------------------------------------------------------------
    Data Definitions
---------------------------------------------------------------------------------------------------- #>
boon_id := enum:
    Chainsaw
    DiscountCoupon
    ImmortalHeart
    BossHunter
    BossSiphon
    AllForOne
    MultiTool
    SpareDice
    LuckyDice
    TreasureMap
    DashingWings
    WoodenClub
    WoodenArrow
    CriticalOverdrive
    NinjaRun

boon_type := enum:
    Combat
    Survivability
    Util

ToString(Type:boon_type)<transacts>:string =
    case(Type):
        boon_type.Combat => "Combat"
        boon_type.Survivability => "Survivability"
        boon_type.Util => "Utility"

boon_info := struct<computes>:
    Id:boon_id
    Name:string
    Description:string
    Tier:int
    Type:boon_type
    Icon:texture
    IconSmall:texture
    IconColor:color = NamedColors.White
    MaybeOnGrant:?type{_(:player):void} = false
    MaybeGetValue:?type{_(:player)<transacts>:float} = false
    FormatValue:type{_(:float)<transacts>:string}

<# ----------------------------------------------------------------------------------------------------
    Ability Info
---------------------------------------------------------------------------------------------------- #>
BoonInfos:[boon_id]boon_info = map:
    boon_id.Chainsaw => boon_info:
        Id := boon_id.Chainsaw
        Name := "Chainsaw"
        Description := "Doubles your tree gathering speed."
        Tier := 3
        Type := boon_type.Util
        Icon := UserInterface.Icons.Boon_Chainsaw_128
        IconSmall := UserInterface.Icons.Boon_Chainsaw_64
        FormatValue := FormatAsInt
        MaybeOnGrant := option{BoonOnGrantChainsaw}
    boon_id.DiscountCoupon => boon_info:
        Id := boon_id.DiscountCoupon
        Name := "Discount Coupon"
        Description := "20% off cost in the Item Shop."
        Tier := 4
        Type := boon_type.Util
        Icon := UserInterface.Icons.Boon_Discount_128
        IconSmall := UserInterface.Icons.Boon_Discount_64
        FormatValue := FormatAsInt
    boon_id.ImmortalHeart => boon_info:
        Id := boon_id.ImmortalHeart
        Name := "Immortal Heart"
        Description := "Grants 1 revive. +25 Max Health each time an auto revive is used."
        Tier := 3
        Type := boon_type.Survivability
        Icon := UserInterface.Icons.Boon_ImmortalHeart_128
        IconSmall := UserInterface.Icons.Boon_ImmortalHeart_64
        MaybeOnGrant := option{BoonOnGrantImmortalHeart}
        MaybeGetValue := option{GetBoonValueImmortalHeart}
        FormatValue := FormatAsCurrency
    boon_id.BossHunter => boon_info:
        Id := boon_id.BossHunter
        Name := "Boss Hunter"
        Description := "Deal bonus +50% damage to bosses & minibosses."
        Tier := 4
        Type := boon_type.Combat
        Icon := UserInterface.Icons.Boon_BossHunter_128
        IconSmall := UserInterface.Icons.Boon_BossHunter_64
        FormatValue := FormatAsInt
    boon_id.BossSiphon => boon_info:
        Id := boon_id.BossSiphon
        Name := "Boss Siphon"
        Description := "Eliminating minibosses heals & grants +50 max health."
        Tier := 3
        Type := boon_type.Survivability
        Icon := UserInterface.Icons.Boon_BossSiphon_128
        IconSmall := UserInterface.Icons.Boon_BossSiphon_64
        MaybeGetValue := option{GetBoonValueBossSiphon}
        FormatValue := FormatAsCurrency
    boon_id.AllForOne => boon_info:
        Id := boon_id.AllForOne
        Name := "All For One"
        Description := "Your starting weapon gains +50% damage. All other weapons receive -25% damage."
        Tier := 4
        Type := boon_type.Combat
        Icon := UserInterface.Icons.Boon_AllForOne_128
        IconSmall := UserInterface.Icons.Boon_AllForOne_64
        FormatValue := FormatAsInt
    boon_id.MultiTool => boon_info:
        Id := boon_id.MultiTool
        Name := "Multi Tool"
        Description := "+1% damage for every weapon level (max +12% per weapon)."
        Tier := 3
        Type := boon_type.Combat
        Icon := UserInterface.Icons.Boon_Multitool_128
        IconSmall := UserInterface.Icons.Boon_Multitool_64
        MaybeOnGrant := option{BoonOnGrantMultiTool}
        MaybeGetValue := option{GetBoonValueMultiTool}
        FormatValue := FormatAsPercent
    boon_id.SpareDice => boon_info:
        Id := boon_id.SpareDice
        Name := "Spare Dice"
        Description := "The first ability reroll is free on level ups."
        Tier := 4
        Type := boon_type.Util
        Icon := UserInterface.Icons.Boon_SpareDice_128
        IconSmall := UserInterface.Icons.Boon_SpareDice_64
        FormatValue := FormatAsInt
    boon_id.LuckyDice => boon_info:
        Id := boon_id.LuckyDice
        Name := "Lucky Dice"
        Description := "+25% Luck every time you reroll. Bonus luck resets once you select an ability."
        Tier := 3
        Type := boon_type.Util
        Icon := UserInterface.Icons.Boon_LuckyDice_128
        IconSmall := UserInterface.Icons.Boon_LuckyDice_64
        FormatValue := FormatAsInt
    boon_id.TreasureMap => boon_info:
        Id := boon_id.TreasureMap
        Name := "Treasure Map"
        Description := "Grants 3 Treasure Chests."
        Tier := 4
        Type := boon_type.Util
        Icon := UserInterface.Icons.Boon_TreasureMap_128
        IconSmall := UserInterface.Icons.Boon_TreasureMap_64
        MaybeOnGrant := option{BoonOnGrantTreasureMap}
        FormatValue := FormatAsInt
    boon_id.DashingWings => boon_info:
        Id := boon_id.DashingWings
        Name := "Dashing Wings"
        Description := "Increase Dash height and distance (+100% upwards impulse)."
        Tier := 3
        Type := boon_type.Util
        Icon := UserInterface.Icons.Boon_DashingWings_128
        IconSmall := UserInterface.Icons.Boon_DashingWings_64
        FormatValue := FormatAsInt
    boon_id.WoodenClub => boon_info:
        Id := boon_id.WoodenClub
        Name := "Wooden Club"
        Description := "+1% Damage for every 100 Wood you are holding (max +50% at 5000 Wood)."
        Tier := 4
        Type := boon_type.Combat
        Icon := UserInterface.Icons.Boon_WoodenClub_128
        IconSmall := UserInterface.Icons.Boon_WoodenClub_64
        MaybeOnGrant := option{BoonOnGrantWoodenClub}
        MaybeGetValue := option{GetBoomValueWoodenClub}
        FormatValue := FormatAsPercent
    boon_id.WoodenArrow => boon_info:
        Id := boon_id.WoodenArrow
        Name := "Wooden Arrow"
        Description := "+1% Critical Rate for every 100 Wood you are holding (max +50% at 5000 Wood)."
        Tier := 3
        Type := boon_type.Combat
        Icon := UserInterface.Icons.Boon_WoodenArrow_128
        IconSmall := UserInterface.Icons.Boon_WoodenArrow_64
        MaybeOnGrant := option{BoonOnGrantWoodenArrow}
        MaybeGetValue := option{GetBoonValueWoodenArrow}
        FormatValue := FormatAsPercent
    boon_id.CriticalOverdrive => boon_info:
        Id := boon_id.CriticalOverdrive
        Name := "Critical Overdrive"
        Description := "Critical Rate above 100% is converted to +2% Critical Damage (e.g. 125% Critical Rate converts to +50% Critical Damage)."
        Tier := 4
        Type := boon_type.Combat
        Icon := UserInterface.Icons.Boon_CriticalOverdrive_128
        IconSmall := UserInterface.Icons.Boon_CriticalOverdrive_64
        MaybeOnGrant := option{BoonOnGrantCriticalOverdrive}
        MaybeGetValue := option{GetBoonValueCriticalOverdrive}
        FormatValue := FormatAsPercent
    boon_id.NinjaRun => boon_info:
        Id := boon_id.NinjaRun
        Name := "Ninja Run"
        Description := "+1% Dodge Chance for every 10% Movement Speed you have (can still be stunned on dodge)."
        Tier := 3
        Type := boon_type.Survivability
        Icon := UserInterface.Icons.Boon_NinjaRun_128
        IconSmall := UserInterface.Icons.Boon_NinjaRun_64
        MaybeGetValue := option{GetBoonValueNinjaRun}
        FormatValue := FormatAsPercent

<# ----------------------------------------------------------------------------------------------------
    Global Functions
---------------------------------------------------------------------------------------------------- #>
GetBoonInfos()<transacts>:[boon_id]boon_info =
    BoonInfos

GetBoonInfo(Id:boon_id)<decides><transacts>:boon_info =
    GetBoonInfos()[Id]

GetBoonIds()<transacts>:[]boon_id =
    for (Id -> BoonInfo : GetBoonInfos()) {Id}

GetRandomBoons(Player:player, Count:int, ?Ignore:[]boon_id = array{})<transacts>:[]boon_info =
    FilteredBoonInfos := for:
        BoonInfo : GetBoonInfos()
        not Player.HasBoon(BoonInfo.Id)?
        not Ignore.Find[BoonInfo.Id]
    do:
        BoonInfo

    ShuffledBoonInfos := Shuffle(FilteredBoonInfos)
    ShuffledBoonInfos.Slice[0, Min(FilteredBoonInfos.Length, Count)] or array{}

GetRandomBoonInfo(Player:player)<decides><transacts>:boon_info =
    GetRandomBoons(Player, 1)[0]

<# ----------------------------------------------------------------------------------------------------
    GetValue Functions
---------------------------------------------------------------------------------------------------- #>
GetBoonValueImmortalHeart(Player:player)<transacts>:float =
    Player.GetGameStat(player_game_stat_id.AutoRevived) * BoonImmortalHeartAmount

GetBoonValueBossSiphon(Player:player)<transacts>:float =
    Player.GetGameStat(player_game_stat_id.EliminatedMiniboss) * BoonBossSiphonAmount

GetBoonValueMultiTool(Player:player)<transacts>:float =
    (Sum of for (WeaponId -> WeaponLevel : Player.GetWeaponLevels()) {Min(12, WeaponLevel)}) * 0.01

GetBoomValueWoodenClub(Player:player)<transacts>:float =
    CalculateBoomValueWoodenClub(Player.GetGameCurrency(game_currency_id.Wood))

CalculateBoomValueWoodenClub(WoodAmount:float)<transacts>:float =
    Min(0.5, WoodAmount / 100.0 * 0.01)

GetBoonValueWoodenArrow(Player:player)<transacts>:float =
    CalculateBoonValueWoodenArrow(Player.GetGameCurrency(game_currency_id.Wood))

CalculateBoonValueWoodenArrow(WoodAmount:float)<transacts>:float =
    Min(0.5, WoodAmount / 100.0 * 0.01)

GetBoonValueCriticalOverdrive(Player:player)<transacts>:float =
    CalculateBoonValueCriticalOverdrive(Player.GetStat(stat_id.CriticalRate))

CalculateBoonValueCriticalOverdrive(CriticalRate:float)<transacts>:float =
    Max(0.0, (CriticalRate - 1.0) * 2)

GetBoonValueNinjaRun(Player:player)<transacts>:float =
    Player.GetStat(stat_id.MoveSpeed) * 0.1
    
<# ----------------------------------------------------------------------------------------------------
    OnGrant Functions
---------------------------------------------------------------------------------------------------- #>
BoonOnGrantImmortalHeart(Player:player):void =
    Player.AddPlayerResource(player_resource_id.Revives, 1)
    if (Character := Player.GetFortCharacter[], not Character.IsDownButNotOut[]):
        SpawnVFX(vfx_id.Revive, Character.GetTransform().Translation.Make2D(), ?Duration := 3.0)

BoonOnGrantTreasureMap(Player:player):void =
    Player.AddGameCurrency(game_currency_id.TreasureChest, BoonTreasureMapAmount * 1.0)
    Player.ShowMessage(message_id.TreasureChestCollect, "+{BoonTreasureMapAmount}", ?DisplayTime := 1.8)
    Player.ActivateVFXPowerup(vfx_powerup_id.TreasureChestCollect, ?Duration := 1.0)
    Player.PlayAudio(audio_id.Purchase)

BoonOnGrantChainsaw(Player:player):void =
    if (PlayerStats := Player.GetStats[]):
        PlayerStats.AbilityChangedEvent.Signal(ability_id.GatherRate)

BoonOnGrantMultiTool(Player:player):void =
    Player.SetStat(stat_id.Damage, stat_source.Boon2, GetBoonValueMultiTool(Player))

BoonOnGrantWoodenClub(Player:player):void =
    Player.SetStat(stat_id.Damage, stat_source.Boon, GetBoomValueWoodenClub(Player))

BoonOnGrantWoodenArrow(Player:player):void =
    Player.SetStat(stat_id.CriticalRate, stat_source.Boon, GetBoonValueWoodenArrow(Player))

BoonOnGrantCriticalOverdrive(Player:player):void =
    Player.SetStat(stat_id.CriticalDamage, stat_source.Boon, GetBoonValueCriticalOverdrive(Player))