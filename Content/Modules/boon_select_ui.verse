using { /Fortnite.com/UI }
using { /Verse.org/Colors }
using { /Verse.org/Random }
using { /Verse.org/Simulation }
using { /UnrealEngine.com/Temporary/SpatialMath }
using { /UnrealEngine.com/Temporary/UI }

<# ----------------------------------------------------------------------------------------------------
    boon_select_ui
---------------------------------------------------------------------------------------------------- #>
boon_select_ui := class(panel_ui):
    Title<override>:string = "Select Boon"
    Anchors<override>:anchors = anchors{Minimum := vector2{X := 0.0, Y := 0.0}, Maximum := vector2{X := 1.0, Y := 1.0}}
    HasCloseButton<override>:logic = false
    BackgroundOpacity<override>:type {_X:float where 0.000000 <= _X, _X <= 1.000000} = 0.7
    
    RerollButton:button_quiet = button_quiet{}
    
    var InfoUIs:[ability_select_info_type]ui_component = map{}
    var BoonSelectEntryUIs:[]boon_select_entry_ui = array{}
    var Subscriptions:[]cancelable = array{}
    var EventsSubscriptions:[]events_cancelable = array{}
    
    var CurrentBoons:[]boon_info = array{}
    var ShowTime:float = 0.0
    var HasSelected:logic = false

    <# ----------------------------------------------------------------------------------------------------
        Initialization
    ---------------------------------------------------------------------------------------------------- #>
    CreateContent<override>():widget =
        if (GameManager := GetGameManager[]):
            set EventsSubscriptions = array:
                ClosedEvent.Subscribe(OnClosed)
                GameManager.ResumeCountdownStartEvent.Subscribe(OnResumeCountdownStart)
                GameManager.ResumeCountdownCompleteEvent.Subscribe(OnResumeCountdownComplete)
                GameManager.ResumeCountdownEvent.Subscribe(OnResumeCountdown)
                GameManager.GameResumedEvent.Subscribe(OnGameResumed)

        TierRatesButton := button_quiet{DefaultText := MakeMessage("Tier Rates")}
        StatsInfoButton := button_quiet{DefaultText := MakeMessage("Stats Info")}
        
        set Subscriptions = array:
            RerollButton.OnClick().Subscribe(OnClickReroll)
            TierRatesButton.OnClick().Subscribe(OnClickTierRates)
            StatsInfoButton.OnClick().Subscribe(OnClickStatsInfo)

        BoonSelectStackBox := stack_box{Orientation := orientation.Horizontal}

        set BoonSelectEntryUIs = for (Index := 1..3):
            BoonSelectEntryUI := boon_select_entry_ui{Player := Player}
            BoonSelectEntryUI.Initialize()
            set EventsSubscriptions += array{BoonSelectEntryUI.SelectEvent.Subscribe(OnSelect)}

            BoonSelectStackBox.AddWidget of stack_box_slot:
                HorizontalAlignment := horizontal_alignment.Fill
                VerticalAlignment := vertical_alignment.Fill
                Padding := margin{Left := (Index = 1) and 0.0 or 20.0, Top := 0.0, Right := 0.0, Bottom := 0.0}
                Distribution := option{1.0}
                Widget := BoonSelectEntryUI.GetUI()

            BoonSelectEntryUI

        stack_box:
            Orientation := orientation.Vertical
            Slots := array:
                stack_box_slot:
                    HorizontalAlignment := horizontal_alignment.Fill
                    VerticalAlignment := vertical_alignment.Fill
                    Padding := margin{Left := 0.0, Top := 0.0, Right := 0.0, Bottom := 20.0}
                    Widget := stack_box:
                        Orientation := orientation.Horizontal
                        Slots := array:
                            stack_box_slot:
                                Distribution := option{2.0}
                                Widget := color_block:
                                    DefaultOpacity := 0.0
                            stack_box_slot:
                                HorizontalAlignment := horizontal_alignment.Fill
                                Distribution := option{1.0}
                                Widget := stack_box:
                                    Orientation := orientation.Horizontal
                                    Slots := array:
                                        stack_box_slot:
                                            HorizontalAlignment := horizontal_alignment.Fill
                                            VerticalAlignment := vertical_alignment.Fill
                                            Distribution := option{1.0}
                                            Widget := StatsInfoButton
                                        stack_box_slot:
                                            HorizontalAlignment := horizontal_alignment.Fill
                                            VerticalAlignment := vertical_alignment.Fill
                                            Distribution := option{1.0}
                                            Padding := margin{Left := 20.0, Top := 0.0, Right := 0.0, Bottom := 0.0}
                                            Widget := TierRatesButton
                stack_box_slot:
                    HorizontalAlignment := horizontal_alignment.Fill
                    VerticalAlignment := vertical_alignment.Fill
                    Distribution := option{1.0}
                    Widget := BoonSelectStackBox
                stack_box_slot:
                    HorizontalAlignment := horizontal_alignment.Fill
                    Padding := margin{Left := 0.0, Top := 20.0, Right := 0.0, Bottom := 0.0}
                    Widget := color_block:
                        DefaultColor := NamedColors.White
                        DefaultOpacity := 0.1
                        DefaultDesiredSize := vector2{X := 0.0, Y := 1.0}
                stack_box_slot:
                    HorizontalAlignment := horizontal_alignment.Fill
                    VerticalAlignment := vertical_alignment.Fill
                    Padding := margin{Left := 0.0, Top := 20.0, Right := 0.0, Bottom := 0.0}
                    Widget := RerollButton

    Uninitialize<override>():void =
        for (BoonSelectEntryUI : BoonSelectEntryUIs):
            BoonSelectEntryUI.Uninitialize()
            
        for (EventsSubscription : EventsSubscriptions):
            EventsSubscription.Cancel()

        for (Subscription : Subscriptions):
            Subscription.Cancel()

    <# ----------------------------------------------------------------------------------------------------
        Get Functions
    ---------------------------------------------------------------------------------------------------- #>
    GetCurrentBoons()<transacts>:[]boon_id =
        for (BoonSelectEntryUI : BoonSelectEntryUIs, BoonInfo := BoonSelectEntryUI.GetActiveBoonInfo()?):
            BoonInfo.Id

    GetRerollsAvailable()<transacts>:int =        
        Player.GetPlayerResource(player_resource_id.BoonReroll)

    CanSelect()<transacts>:logic =
        logic{not HasSelected?, GetSecondsSinceEpoch() - ShowTime >= Player.GetSelectBufferTime()}

    <# ----------------------------------------------------------------------------------------------------
        Update Functions
    ---------------------------------------------------------------------------------------------------- #>
    Update(?IsRoll:logic = true):void =
        for (Index -> BoonSelectEntryUI : BoonSelectEntryUIs):
            if (BoonInfo := CurrentBoons[Index]):
                BoonSelectEntryUI.Update(BoonInfo)
                BoonSelectEntryUI.Show()
            else:
                BoonSelectEntryUI.Hide()

        RerollButton.SetText(MakeMessage("Reroll ({GetRerollsAvailable()} Remaining)"))

    Roll(?IsReroll:logic = false):void =
        IgnoreBoons:[]boon_id = (IsReroll? and GetCurrentBoons() or array{})
        set CurrentBoons = GetRandomBoons(Player, 3, ?Ignore := IgnoreBoons)

    SelectBoon(BoonSelectEntryUI:boon_select_entry_ui):void =
        if (BoonInfo := BoonSelectEntryUI.MaybeBoonInfo?):
            Player.AddBoon(BoonInfo)
            Player.PlayAudio(audio_id.Upgrade)

    ToggleInfo(AbilitySelectInfoType:ability_select_info_type):void =
        for (Type -> InfoUI : InfoUIs, Type <> AbilitySelectInfoType):
            InfoUI.Hide()

        if (InfoUI := InfoUIs[AbilitySelectInfoType]):
            Visibility := InfoUI.GetUI().GetVisibility()
            if (Visibility = widget_visibility.Visible):
                InfoUI.Hide()
            else:
                InfoUI.Show()
                InfoUI.Update()
        else:
            InfoUI := case (AbilitySelectInfoType):
                ability_select_info_type.StatsInfo => ability_stat_info_ui{Player := Player, BackgroundType := background_type.Opaque, TextSize := ThemeTextSize.Small, IconSize := 0.75}
                ability_select_info_type.TierRates => ability_select_tier_rates_ui{Player := Player}

            InfoUI.Initialize()
            if (set InfoUIs[AbilitySelectInfoType] = InfoUI) {}

            if (PanelUI := canvas[MaybeUI?]):
                PanelUI.AddWidget of canvas_slot:
                    Anchors := case (AbilitySelectInfoType):
                        ability_select_info_type.StatsInfo => anchors{Minimum := vector2{X := 0.0, Y := 0.0}, Maximum := vector2{X := 0.6, Y := 0.0}}
                        ability_select_info_type.TierRates => anchors{Minimum := vector2{X := 0.0, Y := 0.0}, Maximum := vector2{X := 0.0, Y := 0.0}}
                    Alignment := vector2{X := 0.0, Y := 0.0}
                    ZOrder := 1
                    Widget := InfoUI.GetUI()

                InfoUI.Update()

    UpdateTitle()<suspends>:void =
        Sleep(0.0)
        TitleText.SetText(MakeMessage("{Title}"))
        TitleText.SetTextColor(ThemeColors.PrimaryText)
        TitleText.SetShadowOpacity(0.5)

    SelectRandomBoon()<suspends>:void =
      for (Index -> BoonSelectEntryUI : Shuffle(BoonSelectEntryUIs)):
            if (Index = 0, BoonInfo := BoonSelectEntryUI.MaybeBoonInfo?):
                BoonSelectEntryUI.SelectButton.SetText(MakeMessage("Selected"))
                Player.AddBoon(BoonInfo)
                Player.PlayAudio(audio_id.Purchase)
            else:
                BoonSelectEntryUI.InfoOverlay.SetVisibility(widget_visibility.Hidden)
                BoonSelectEntryUI.SelectButton.SetText(MakeMessage(""))

        SleepTimeDilation(3.0)
        Hide()
    
    <# ----------------------------------------------------------------------------------------------------
        Event Handlers
    ---------------------------------------------------------------------------------------------------- #>
    OnShow<override>():void =
        set ShowTime = GetSecondsSinceEpoch()
        set HasSelected = false
        spawn{UpdateTitle()}
        for (BoonSelectEntryUI : BoonSelectEntryUIs) {BoonSelectEntryUI.Clear()}
        Roll(?IsReroll := true)
        Update()

        for (InfoUI : InfoUIs):
            InfoUI.Hide()

    OnClosed(InPlayer:player):void =
        ResumeGame(Player)

    OnSelect(BoonInfo:boon_info):void =
        if (not CanSelect()?) then return
        set HasSelected = true
        Player.AddBoon(BoonInfo)
        Player.PlayAudio(audio_id.Purchase)
        Hide()

    OnClickReroll(Message:widget_message):void =            
        if (GetRerollsAvailable() > 0):
            Player.RemovePlayerResource(player_resource_id.BoonReroll, 1)
            Roll(?IsReroll := true)
            Update()
            Player.PlayAudio(audio_id.Reroll)
        else:
            Player.PlayAudio(audio_id.Error)

    OnResumeCountdownStart():void =
        if (not IsShowing?) then return
        Player.PlayAudio(audio_id.TimerWarning)
        TitleText.SetText(MakeMessage("{GameResumeCountdownTimeDefault}s"))
        TitleText.SetTextColor(ThemeColors.Failure)
        TitleText.SetShadowColor(ThemeColors.Failure * 0.125)
        TitleText.SetShadowOffset(option{vector2{X := 2.0, Y := 2.0}})
        TitleText.SetShadowOpacity(1.0)
        
    OnResumeCountdownComplete():void =
        if (not IsShowing?) then return
        TitleText.SetText(MakeMessage("Auto Selected"))
        spawn{SelectRandomBoon()}

    OnResumeCountdown(TimeRemaining:int):void =
        if (not IsShowing?) then return
        TitleText.SetText(MakeMessage("{TimeRemaining}s"))

    OnGameResumed():void =
        if (not IsShowing?) then return
        spawn{UpdateTitle()}

    OnClickTierRates(Message:widget_message):void =
        ToggleInfo(ability_select_info_type.TierRates)

    OnClickStatsInfo(Message:widget_message):void =
        ToggleInfo(ability_select_info_type.StatsInfo)