using { /Fortnite.com/Characters }
using { /Fortnite.com/Devices }
using { /Verse.org/Simulation }
using { /UnrealEngine.com/Temporary/SpatialMath }

<# ----------------------------------------------------------------------------------------------------
    Constants
---------------------------------------------------------------------------------------------------- #>
MaxCameraDistance:int = 3

<# ----------------------------------------------------------------------------------------------------
    Data Definitions
---------------------------------------------------------------------------------------------------- #>
camera_type := enum:
    TwinStick
    Debug
    Orbit
    OrbitFree
    Orthographic

(Type:camera_type).Next()<transacts>:camera_type =
    case (Type):
        camera_type.TwinStick => camera_type.Debug
        camera_type.Debug => camera_type.Orbit
        camera_type.Orbit => camera_type.OrbitFree
        camera_type.OrbitFree => camera_type.Orthographic
        camera_type.Orthographic => camera_type.TwinStick

ToString(Type:camera_type)<transacts>:string =
    case (Type):
        camera_type.TwinStick => "Twin Stick"
        camera_type.Debug => "Debug"
        camera_type.Orbit => "Orbit"
        camera_type.OrbitFree => "Orbit Free"
        camera_type.Orthographic => "Orthographic"

(Type:camera_type).ToInt()<transacts>:int =
    case (Type):
        camera_type.TwinStick => 0
        camera_type.Debug => 1
        camera_type.Orbit => 2
        camera_type.OrbitFree => 3
        camera_type.Orthographic => 4

(Id:int).ToCameraType()<transacts>:camera_type =
    case (Id):
         0 => camera_type.TwinStick
         1 => camera_type.Debug
         2 => camera_type.Orbit
         3 => camera_type.OrbitFree
         4 => camera_type.Orthographic
         _ => camera_type.TwinStick

<# ----------------------------------------------------------------------------------------------------
    Global Functions
---------------------------------------------------------------------------------------------------- #>
var GlobalCameraManager:weak_map(session, camera_manager) = map{}

GetCameraManager()<decides><transacts>:camera_manager =
    GlobalCameraManager[GetSession()]

WaitForPlayerInitialized(Player:player)<suspends>:void =
    loop:
        Sleep(0.3)
        if (Player.IsActive[] and Player.GetFortCharacter[].IsActive[] and GetCameraManager[].PlayerInitialized[Player]):
            break

(Player:player).GetActiveCamera()<transacts>:camera_type =
    GetCameraManager[].GetActiveCamera(Player) or camera_type.TwinStick

(Player:player).RefreshControls():void =
    if (CameraManager := GetCameraManager[]):
        spawn{CameraManager.RefreshControls(Player)}

(Agent:agent).ToggleCamera():void =
    if (CameraManager := GetCameraManager[]):
        CameraManager.ToggleCamera(Agent)

(Player:player).ToggleCameraDistance():void =
    if:
        CurrentPlayerData := Player.GetData[]
        UpdatedCameraDistance := CurrentPlayerData.Settings.CameraDistance + 1

        set PlayerDataMap[Player] = player_data:
            MakePlayerData<constructor>(CurrentPlayerData)
                Settings := player_settings:
                    MakePlayerSettings<constructor>(CurrentPlayerData.Settings)
                        CameraDistance := UpdatedCameraDistance > MaxCameraDistance and 1 or UpdatedCameraDistance

ActivateCameraShake(Position:vector3):void =
    if (CameraManager := GetCameraManager[]):
        CameraManager.ActivateCameraShake(Position)

<# ----------------------------------------------------------------------------------------------------
    camera_manager
---------------------------------------------------------------------------------------------------- #>
camera_manager := class(creative_device):
    @editable
    GameplayCameraFixedAngleDefault1:gameplay_camera_fixed_angle_device = gameplay_camera_fixed_angle_device{}
    
    @editable
    GameplayCameraFixedAngleDefault2:gameplay_camera_fixed_angle_device = gameplay_camera_fixed_angle_device{}
    
    @editable
    GameplayCameraFixedAngleDefault3:gameplay_camera_fixed_angle_device = gameplay_camera_fixed_angle_device{}

    @editable
    GameplayCameraFixedAngleDebug:gameplay_camera_fixed_angle_device = gameplay_camera_fixed_angle_device{}

    @editable
    GameplayCameraFixedAngleOrthographic1:gameplay_camera_fixed_angle_device = gameplay_camera_fixed_angle_device{}

    @editable
    GameplayCameraFixedAngleOrthographic2:gameplay_camera_fixed_angle_device = gameplay_camera_fixed_angle_device{}

    @editable
    GameplayCameraFixedAngleOrthographic3:gameplay_camera_fixed_angle_device = gameplay_camera_fixed_angle_device{}

    @editable
    GameplayCameraOrbit1:gameplay_camera_orbit_device = gameplay_camera_orbit_device{}
    
    @editable
    GameplayCameraOrbit2:gameplay_camera_orbit_device = gameplay_camera_orbit_device{}
    
    @editable
    GameplayCameraOrbit3:gameplay_camera_orbit_device = gameplay_camera_orbit_device{}

    @editable
    GameplayCameraOrbitFree1:gameplay_camera_orbit_device = gameplay_camera_orbit_device{}
    
    @editable
    GameplayCameraOrbitFree2:gameplay_camera_orbit_device = gameplay_camera_orbit_device{}
    
    @editable
    GameplayCameraOrbitFree3:gameplay_camera_orbit_device = gameplay_camera_orbit_device{}

    @editable
    GameplayControlsThirdPersonTwinStick:gameplay_controls_third_person_device = gameplay_controls_third_person_device{}

    @editable
    GameplayControlsThirdPerson:gameplay_controls_third_person_device = gameplay_controls_third_person_device{}

    @editable
    CameraShakeProp:creative_prop = creative_prop{}

    @editable
    CameraShakeCinematicSequence:cinematic_sequence_device = cinematic_sequence_device{}
    
    @editable
    PlayerSpawners:[]player_spawner_device = array{}

    @editable
    PlayerCounter:player_counter_device = player_counter_device{}

    var PlayerInitialized:[player]logic = map{}

    <# ----------------------------------------------------------------------------------------------------
        Initialization
    ---------------------------------------------------------------------------------------------------- #>
    OnBegin<override>()<suspends>:void =
        if (set GlobalCameraManager[GetSession()] = Self) {}

        for (PlayerSpawner : PlayerSpawners):
            PlayerSpawner.SpawnedEvent.Subscribe(OnSpawned)

        PlayerCounter.CountedEvent.Subscribe(OnCounted)
        GetPlayspace().PlayerRemovedEvent().Subscribe(OnPlayerRemoved)

    Initialize(Player:player)<suspends>:void =
        if (PlayerInitialized[Player]?) then return

        if (set PlayerInitialized[Player] = true) {}

        WaitForVFXManager()
        if (PlayerData := Player.GetData[]):
            UpdateCamera(Player, PlayerData.Settings.Camera.ToCameraType())
        else:
            UpdateCamera(Player, camera_type.TwinStick)

        Player.ActivateVFXPowerup of Player.GetRankVFXPowerUp()

        WaitForGameManager()
        if (GameManager := GetGameManager[]):
            GameManager.AddToGame(Player)

    <# ----------------------------------------------------------------------------------------------------
        Get Functions
    ---------------------------------------------------------------------------------------------------- #>
    GetActiveCamera(Player:player)<transacts>:camera_type =
        Player.GetData[].Settings.Camera.ToCameraType() or camera_type.TwinStick

    <# ----------------------------------------------------------------------------------------------------
        Update Functions
    ---------------------------------------------------------------------------------------------------- #>
    RefreshControls(Player:player)<suspends>:void =
        RemoveCameras(Player)
        Sleep(0.0)
        UpdateCamera(Player, GetActiveCamera(Player))

    (camera_manager:)ActivateCameraShake(Position:vector3):void =
        if (CameraShakeProp.TeleportTo[Position, rotation{}]):
            CameraShakeCinematicSequence.Play()

    ToggleCamera(Agent:agent):void =
        if:
            Player := player[Agent]
            CurrentPlayerData := Player.GetData[]
            NextCamera := CurrentPlayerData.Settings.Camera.ToCameraType().Next()
            UpdatedActiveCamera := if (NextCamera = camera_type.Debug):
                GetTestTool[] and NextCamera or NextCamera.Next()
            else:
                NextCamera

            UpdatedPlayerSettings := player_settings:
                MakePlayerSettings<constructor>(CurrentPlayerData.Settings)
                Camera := UpdatedActiveCamera.ToInt()
           
            set PlayerDataMap[Player] = player_data:
                MakePlayerData<constructor>(CurrentPlayerData)
                Settings := UpdatedPlayerSettings
        then:
            UpdateCamera(Player, UpdatedActiveCamera)

    RemoveCameras(Player:player):void =
        GameplayControlsThirdPersonTwinStick.RemoveFrom(Player)
        GameplayControlsThirdPerson.RemoveFrom(Player)
        GameplayCameraFixedAngleDefault1.RemoveFrom(Player)
        GameplayCameraFixedAngleDefault2.RemoveFrom(Player)
        GameplayCameraFixedAngleDefault3.RemoveFrom(Player)
        GameplayCameraFixedAngleDebug.RemoveFrom(Player)
        GameplayCameraFixedAngleOrthographic1.RemoveFrom(Player)
        GameplayCameraFixedAngleOrthographic2.RemoveFrom(Player)
        GameplayCameraFixedAngleOrthographic3.RemoveFrom(Player)
        GameplayCameraOrbit1.RemoveFrom(Player)
        GameplayCameraOrbit2.RemoveFrom(Player)
        GameplayCameraOrbit3.RemoveFrom(Player)
        GameplayCameraOrbitFree1.RemoveFrom(Player)
        GameplayCameraOrbitFree2.RemoveFrom(Player)
        GameplayCameraOrbitFree3.RemoveFrom(Player)

    UpdateCamera(Player:player, CameraType:camera_type):void =
        RemoveCameras(Player)
        IsAutoAim := logic{Player.GetData[].Settings.AutoAim?}
        GameplayControls := IsAutoAim? and GameplayControlsThirdPerson or GameplayControlsThirdPersonTwinStick
        CameraDistance := Player.GetData[].Settings.CameraDistance or 0

        GameplayCamera := case (CameraType):
            camera_type.TwinStick =>
                GameplayControls.AddTo(Player)
                GameplayCameraFixedAngle := case (CameraDistance):
                    1 => GameplayCameraFixedAngleDefault1
                    2 => GameplayCameraFixedAngleDefault2
                    3 => GameplayCameraFixedAngleDefault3
                    _ => GameplayCameraFixedAngleDefault2

                Player.ActivateVFXPowerup(IsAutoAim? and vfx_powerup_id.Base or vfx_powerup_id.Arrow)
                GameplayCameraFixedAngle.AddTo(Player)
                GameplayCameraFixedAngle

            camera_type.Debug =>
                GameplayControls.AddTo(Player)
                GameplayCameraFixedAngleDebug.AddTo(Player)
                Player.ActivateVFXPowerup(IsAutoAim? and vfx_powerup_id.Base or vfx_powerup_id.Arrow)
                GameplayCameraFixedAngleDebug

            camera_type.Orbit =>
                GameplayControlsThirdPerson.AddTo(Player)
                GameplayCameraOrbit := case (CameraDistance):
                    1 => GameplayCameraOrbit1
                    2 => GameplayCameraOrbit2
                    3 => GameplayCameraOrbit3
                    _ => GameplayCameraOrbit2
                GameplayCameraOrbit.AddTo(Player)
                Player.ActivateVFXPowerup(IsAutoAim? and vfx_powerup_id.Base or vfx_powerup_id.ArrowOrbit)
                GameplayCameraOrbit

            camera_type.OrbitFree =>
                GameplayControlsThirdPerson.AddTo(Player)
                GameplayCameraOrbitFree := case (CameraDistance):
                    1 => GameplayCameraOrbitFree1
                    2 => GameplayCameraOrbitFree2
                    3 => GameplayCameraOrbitFree3
                    _ => GameplayCameraOrbitFree2
                GameplayCameraOrbitFree.AddTo(Player)
                Player.ActivateVFXPowerup(IsAutoAim? and vfx_powerup_id.Base or vfx_powerup_id.ArrowOrbit)
                GameplayCameraOrbitFree

            camera_type.Orthographic =>
                GameplayControls.AddTo(Player)
                GameplayCameraFixedAngleOrthographic := case (CameraDistance):
                    1 => GameplayCameraFixedAngleOrthographic1
                    2 => GameplayCameraFixedAngleOrthographic2
                    3 => GameplayCameraFixedAngleOrthographic3
                    _ => GameplayCameraFixedAngleOrthographic2

                Player.ActivateVFXPowerup(IsAutoAim? and vfx_powerup_id.Base or vfx_powerup_id.Arrow)
                GameplayCameraFixedAngleOrthographic.AddTo(Player)
                GameplayCameraFixedAngleOrthographic

    <# ----------------------------------------------------------------------------------------------------
        Event handlers
    ---------------------------------------------------------------------------------------------------- #>
    OnPlayerRemoved(Player:player):void =
        set PlayerInitialized = RemoveKeyFromMap(PlayerInitialized, Player)

    OnSpawned(Agent:agent):void =
        if (Player := player[Agent]):
            spawn{Initialize(Player)}

    OnCounted(Agent:agent):void =
        if (Player := player[Agent]):
            spawn{Initialize(Player)}