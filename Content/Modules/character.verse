using { /Fortnite.com/Devices }
using { /Verse.org/Simulation }

<# ----------------------------------------------------------------------------------------------------
    character
---------------------------------------------------------------------------------------------------- #>
character := class(creative_device):
    @editable
    Character:character_device = character_device{}

    @editable
    Button:button_device = button_device{}

    @editable
    PanelId:?panel_ui_id = false

    @editable
    AudioPlayer:audio_player_device = audio_player_device{}

    var ShowingPlayer:[player]logic = map{}

    <# ----------------------------------------------------------------------------------------------------
        Initialization
    ---------------------------------------------------------------------------------------------------- #>
    OnBegin<override>()<suspends>:void =
        Button.InteractedWithEvent.Subscribe(OnInteractedWith)
    
    <# ----------------------------------------------------------------------------------------------------
        Update Functions
    ---------------------------------------------------------------------------------------------------- #>
    ShowUI(Player:player)<suspends>:void =
        if (set ShowingPlayer[Player] = true) {}
        Character.PlayEmote()
        AudioPlayer.Play(Player)
        Sleep(0.5)
    
        if (Id := PanelId?):
            Player.ShowPanelUI(Id)

        Sleep(0.5)
        set ShowingPlayer = RemoveKeyFromMap(ShowingPlayer, Player)

    <# ----------------------------------------------------------------------------------------------------
        Event handlers
    ---------------------------------------------------------------------------------------------------- #>
    OnInteractedWith(Agent:agent):void =
        if (Player := player[Agent]):
            if (ShowingPlayer[Player]?) then return
            spawn{ShowUI(Player)}