using { /Verse.org/Simulation }

<# ----------------------------------------------------------------------------------------------------
    Global Functions
---------------------------------------------------------------------------------------------------- #>
FormatAsCurrency(Value:float)<transacts>:string =
    var FormattedValue:float = Value
    var SuffixIndex:int = 0

    loop:
        if (FormattedValue < 1000.0):
            break

        set FormattedValue = FormattedValue / 1000.0
        set SuffixIndex += 1

    if (SuffixIndex = 0):
        GetRoundedString(FormattedValue, 0)
    else:
        Decimals := if (FormattedValue >= 100.0):
            0
        else if (FormattedValue >= 10.0):
            1
        else:
            2

        GetRoundedString(FormattedValue, Decimals) + NumberSuffixes[SuffixIndex] or ""

NumberSuffixes:[]string = array:
    "-", "K", "M", "B", "T", "Qa", "Qt", "Sx", "Sp", "Oc", "No", "Dc",
    "UDc", "DDc", "TDc", "QaDc", "QtDc", "SxDc", "SpDc", "ODc", "NDc", "Vg",
    "UVg", "DVg", "TVg", "QaVg", "QtVg", "SxVg", "SpVg", "OVg", "NVg", "Tg",
    "UTg", "DTg", "TTg", "QaTg", "QtTg", "SxTg", "SpTg", "OTg", "NTg", "Qd",
    "UQd", "DQd", "TQd", "QaQd", "QtQd", "SxQd", "SpQd", "OQd", "NQd", "Qi",
    "UQi", "DQi", "TQi", "QaQi", "QtQi", "SxQi", "SpQi", "OQi", "NQi", "Se",
    "USe", "DSe", "TSe", "QaSe", "QtSe", "SxSe", "SpSe", "OSe", "NSe", "St",
    "USt", "DSt", "TSt", "QaSt", "QtSt", "SxSt", "SpSt", "OSt", "NSt", "Og",
    "UOg", "DOg", "TOg", "QaOg", "QtOg", "SxOg", "SpOg", "OOg", "NOg", "Nn",
    "UNn", "DNn", "TNn", "QaNn", "QtNn", "SxNn", "SpNn", "ONn", "NNn", "Ce"

<# ----------------------------------------------------------------------------------------------------
    currency_hud
---------------------------------------------------------------------------------------------------- #>
currency_hud := class:
    Player:player

    var Subscriptions:[]events_cancelable = array{}

    <# ----------------------------------------------------------------------------------------------------
        Initialization
    ---------------------------------------------------------------------------------------------------- #>
    Initialize(PlayerRemovedEvent:listenable(player)):void =
        if (GameManager := GetGameManager[], PlayerCurrencyChangedEvent := GetCurrencyManager[].PlayerCurrencyChangedEvent[Player]):
            set Subscriptions = array:
                GameManager.GameStartedEvent.Subscribe(OnGameStarted)
                GameManager.GameEndedEvent.Subscribe(OnGameEnded)
                GameManager.GameCurrencyChangedEvent.Subscribe(OnGameCurrencyChanged)
                PlayerCurrencyChangedEvent.Subscribe(OnPlayerCurrencyChanged)

        UpdatePlayerCurrency(currency_id.Gold)
        UpdatePlayerCurrency(currency_id.TreasureChest)
        spawn{InitializeHUDMessage()}

    # [Hotfix] If HUD Message is shown for the first time Hide() fails if called within a short time frame.
    InitializeHUDMessage()<suspends>:void =
        Player.ShowMessage(message_id.GameCurrencyTreasureChest, "{FormatAsCurrency(0.0)}", ?DisplayTime := 1.1)
        Sleep(1.1)
        Player.HideMessage(message_id.GameCurrencyTreasureChest)

    Uninitialize():void =
        for (Subscription : Subscriptions):
            Subscription.Cancel()

        set Subscriptions = array{}

    <# ----------------------------------------------------------------------------------------------------
        Update Functions
    ---------------------------------------------------------------------------------------------------- #>
    Show():void =
        UpdateGameCurrency(game_currency_id.Wood)

    Hide():void =
        Player.HideMessage(message_id.CurrencyWood)
        Player.HideMessage(message_id.CurrencyElimination)
        Player.HideMessage(message_id.GameCurrencyTreasureChest)

    UpdateGameCurrency(Id:game_currency_id):void =
        case (Id):
            game_currency_id.Wood =>
                Player.ShowMessage(message_id.CurrencyWood, "{FormatAsCurrency(Player.GetGameCurrency(game_currency_id.Wood))}", ?DisplayTime := 0.0)

            game_currency_id.TreasureChest =>
                Amount := Player.GetGameCurrency(game_currency_id.TreasureChest)
                if (Amount > 0.0):
                    Player.ShowMessage(message_id.GameCurrencyTreasureChest, "{FormatAsCurrency(Amount)}", ?DisplayTime := 0.0)
                else:
                    Player.HideMessage(message_id.GameCurrencyTreasureChest)

    UpdatePlayerCurrency(Id:currency_id):void =
        case (Id):
            currency_id.Gold =>
                Player.ShowMessage(message_id.CurrencyGold, "{FormatAsCurrency(Player.GetGoldAvailable(currency_id.Gold))}", ?DisplayTime := 0.0)
                Player.UpdatePowerUpIndicator()

            currency_id.TreasureChest =>
                Player.ShowMessage(message_id.CurrencyTreasureChest, "{FormatAsCurrency(Player.GetCurrency(currency_id.TreasureChest))}", ?DisplayTime := 0.0)

            _ => {}

    <# ----------------------------------------------------------------------------------------------------
        Event Handlers
    ---------------------------------------------------------------------------------------------------- #>
    OnGameStarted():void =
        Show()
        Player.HideMessage(message_id.CurrencyGold)
        Player.HideMessage(message_id.CurrencyTreasureChest)

    OnGameEnded():void =
        Hide()

    OnGameCurrencyChanged(Id:game_currency_id):void =
        UpdateGameCurrency(Id)

    OnPlayerCurrencyChanged(Id:currency_id):void =
        UpdatePlayerCurrency(Id)
