using { /Fortnite.com/Devices }
using { /Verse.org/Assets }
using { /Verse.org/Simulation }

<# ----------------------------------------------------------------------------------------------------
    Data definitions
---------------------------------------------------------------------------------------------------- #>
currency_id := enum:
    Gold
    TreasureChest
    GearRecipe

ToString(Id:currency_id)<transacts>:string =
    case (Id):
        currency_id.Gold => "Gold"
        currency_id.TreasureChest => "Treasure Chest"
        currency_id.GearRecipe => "Gear Recipe"

(Id:currency_id).ToInt()<transacts>:int =
    case (Id):
        currency_id.Gold => 0
        currency_id.TreasureChest => 1
        currency_id.GearRecipe => 2
        
(Id:int).ToCurrencyId()<transacts>:currency_id =
    case (Id):
        0 => currency_id.Gold
        1 => currency_id.TreasureChest
        2 => currency_id.GearRecipe
        _ => currency_id.Gold

(Id:currency_id).ToIcon()<transacts>:texture =
    case (Id):
        currency_id.Gold => UserInterface.Icons.CurrencyGold_64
        currency_id.TreasureChest => UserInterface.Icons.TreasureChest_64
        currency_id.GearRecipe => UserInterface.Icons.GearRecipe_64

<# ----------------------------------------------------------------------------------------------------
    Global Functions
---------------------------------------------------------------------------------------------------- #>
var GlobalCurrencyManager:weak_map(session, currency_manager) = map{}

GetCurrencyManager()<decides><transacts>:currency_manager =
    GlobalCurrencyManager[GetSession()]

(Player:player).GetCurrency(Id:currency_id)<transacts>:float =
    if (Id = currency_id.Gold):
        Player.GetGoldAvailable(Id)
    else:
        GetCurrencyManager[].GetCurrency(Player, Id) or 0.0

(Player:player).AddCurrency(Id:currency_id, Amount:float, ?IsSignal:logic = true):void =
    if (CurrencyManager := GetCurrencyManager[]):
        CurrencyManager.AddCurrency(Player, Id, Amount, ?IsSignal := IsSignal)

(Player:player).RemoveCurrency(Id:currency_id, Amount:float):void =
    if (CurrencyManager := GetCurrencyManager[]):
        CurrencyManager.RemoveCurrency(Player, Id, Amount)

(Player:player).RemoveCurrency(UpgradeCosts:[]upgrade_cost):void =
    if (CurrencyManager := GetCurrencyManager[]):
        for (UpgradeCost : UpgradeCosts):
            CurrencyManager.RemoveCurrency(Player, UpgradeCost.Id, UpgradeCost.Amount)
    
(Player:player).GetCurrencyChangedEvent()<decides><transacts>:events(currency_id) =
    GetCurrencyManager[].GetCurrencyChangedEvent[Player]

(Player:player).UpdateGoldCurrencyHUD():void =
    if (CurrencyManager := GetCurrencyManager[], CurrencyHUD := CurrencyManager.PlayerCurrencyHUD[Player]):
        CurrencyHUD.UpdatePlayerCurrency(currency_id.Gold)

(Player:player).CanAfford(Id:currency_id, Amount:float)<transacts>:logic =
    logic{Player.GetCurrency(Id) >= Amount}

(Player:player).CanAfford(UpgradeCost:upgrade_cost)<transacts>:logic =
    Player.CanAfford(UpgradeCost.Id, UpgradeCost.Amount)

(Player:player).CanAfford(UpgradeCosts:[]upgrade_cost)<transacts>:logic =
    var AffordCount:int = 0
    for (UpgradeCost : UpgradeCosts, Player.CanAfford(UpgradeCost)?):
        set AffordCount += 1
    logic{AffordCount >= UpgradeCosts.Length}

(Player:player).GetPowerUpCosts()<transacts>:float =
    GetCurrencyManager[].GetPowerUpCost[Player] or 0.0

(Player:player).UpdatePowerUpCosts()<transacts>:void =
    if (CurrencyManager := GetCurrencyManager[]):
        CurrencyManager.UpdatePowerUpCosts(Player)

<# ----------------------------------------------------------------------------------------------------
    currency_manager
---------------------------------------------------------------------------------------------------- #>
currency_manager := class(creative_device):

    var PlayerCurrencyChangedEvent:[player]events(currency_id) = map{}
    var PlayerCurrencyHUD:[player]currency_hud = map{}
    var PlayerPowerUpCost:[player]float = map{}

    <# ----------------------------------------------------------------------------------------------------
        Initialization
    ---------------------------------------------------------------------------------------------------- #>
    OnBegin<override>()<suspends>:void =
        if (set GlobalCurrencyManager[GetSession()] = Self) {}

        for (Player : GetPlayspace().GetPlayers()):
            spawn{Initialize(Player)}

        GetPlayspace().PlayerAddedEvent().Subscribe(OnPlayerAdded)
        GetPlayspace().PlayerRemovedEvent().Subscribe(OnPlayerRemoved)

    Initialize(Player:player)<suspends>:void =
        if (set PlayerCurrencyChangedEvent[Player] = events(currency_id){}) {}
        
        Sleep(0.0)
        UpdatePowerUpCosts(Player)

        WaitForGameManager()
        WaitForUIManager(Player)

        CurrencyHUD := currency_hud{Player := Player}
        CurrencyHUD.Initialize(GetPlayspace().PlayerRemovedEvent())
        if (set PlayerCurrencyHUD[Player] = CurrencyHUD) {}

    <# ----------------------------------------------------------------------------------------------------
        Get Functions
    ---------------------------------------------------------------------------------------------------- #>
    GetCurrency(Player:player, Id:currency_id)<transacts>:float =
        if (CurrentPlayerData := Player.GetData[], CurrencyAmount := CurrentPlayerData.Currencies[Id.ToInt()]):
            CurrencyAmount
        else:
            0.0

    CanAfford(Player:player, Id:currency_id, Amount:float)<transacts>:logic =
        logic{Player.GetData[].Currencies[Id.ToInt()] >= Amount}

    GetCurrencyChangedEvent(Player:player)<decides><transacts>:events(currency_id) =
        PlayerCurrencyChangedEvent[Player]

    GetPowerUpCost(Player:player)<decides><transacts>:float =
        PlayerPowerUpCost[Player]        

    <# ----------------------------------------------------------------------------------------------------
        Update Functions
    ---------------------------------------------------------------------------------------------------- #>
    AddCurrency(Player:player, Id:currency_id, Amount:float, ?IsSignal:logic = true):void =
        if:
            CurrentPlayerData := Player.GetData[]
            var UpdatedCurrencies:[int]float = CurrentPlayerData.Currencies
            set UpdatedCurrencies[Id.ToInt()] = UpdatedCurrencies[Id.ToInt()] + Amount or Amount

            set PlayerDataMap[Player] = player_data:
                MakePlayerData<constructor>(CurrentPlayerData)
                Currencies := UpdatedCurrencies

        if (IsSignal?, CurrencyChangedEvent := GetCurrencyChangedEvent[Player]):
            CurrencyChangedEvent.Signal(Id)

    RemoveCurrency(Player:player, Id:currency_id, Amount:float):void =
        if:
            CurrentPlayerData := Player.GetData[]
            var UpdatedCurrencies:[int]float = CurrentPlayerData.Currencies
            set UpdatedCurrencies[Id.ToInt()] -= Amount

            set PlayerDataMap[Player] = player_data:
                MakePlayerData<constructor>(CurrentPlayerData)
                Currencies := UpdatedCurrencies
        
        if (CurrencyChangedEvent := GetCurrencyChangedEvent[Player]):
            CurrencyChangedEvent.Signal(Id)

    UpdatePowerUpCosts(Player:player)<transacts>:void =
        if (set PlayerPowerUpCost[Player] = Player.GetTotalPowerUpCost()) {}

    <# ----------------------------------------------------------------------------------------------------
        Event Handlers
    ---------------------------------------------------------------------------------------------------- #>
    OnPlayerAdded(Player:player):void =
        spawn{Initialize(Player)}

    OnPlayerRemoved(Player:player):void =
        if (CurrencyHUD := PlayerCurrencyHUD[Player]):
            CurrencyHUD.Uninitialize()

        set PlayerCurrencyHUD = RemoveKeyFromMap(PlayerCurrencyHUD, Player)
        set PlayerCurrencyChangedEvent = RemoveKeyFromMap(PlayerCurrencyChangedEvent, Player)
        set PlayerPowerUpCost = RemoveKeyFromMap(PlayerPowerUpCost, Player)