using { /Fortnite.com/UI }
using { /Verse.org/Colors }
using { /UnrealEngine.com/Temporary/SpatialMath }
using { /UnrealEngine.com/Temporary/UI }
using { RealTime }

daily_quests_menu_entry_ui := class(ui_component):
    Icon:texture_block = texture_block:
        DefaultImage := UserInterface.Icons.WeaponBow_64
        DefaultDesiredSize := vector2{X := 64.0, Y := 64.0}

    DescriptionText:text_block = text_block:
        DefaultTextColor := ThemeColors.PrimaryText
        DefaultJustification := text_justification.Center

    ProgressText:text_block = text_block:
        DefaultTextColor := ThemeColors.Success
        DefaultShadowColor := ThemeColors.Success * 0.125
        DefaultShadowOffset := option{vector2{X := 2.0, Y := 2.0}}
        DefaultShadowOpacity := 1.0

    CheckmarkIcon:texture_block = texture_block:
        DefaultImage := UserInterface.Icons.Checkmark_64
        DefaultDesiredSize := vector2{X := 40.0, Y := 40.0}
        DefaultTint := ThemeColors.Success

    TierWidgetOverlay:overlay = overlay{}

    var MaybeActiveTierWidget:?widget = false

    <# ----------------------------------------------------------------------------------------------------
        Initialization
    ---------------------------------------------------------------------------------------------------- #>
    CreateUI<override>():widget =
        overlay:
            Slots := array:
                overlay_slot:
                    HorizontalAlignment := horizontal_alignment.Fill
                    VerticalAlignment := vertical_alignment.Fill
                    Padding := margin{Left := 0.0, Top := 0.0, Right := 0.0, Bottom := 0.0}
                    Widget := UserInterface.VerseWidget.WBP_RoundedBox_Light{}
                overlay_slot:
                    HorizontalAlignment := horizontal_alignment.Fill
                    VerticalAlignment := vertical_alignment.Fill
                    Padding := margin{Left := 40.0, Top := 20.0, Right := 40.0, Bottom := 20.0}
                    Widget := stack_box:
                        Orientation := orientation.Horizontal
                        Slots := array:
                            stack_box_slot:
                                HorizontalAlignment := horizontal_alignment.Left
                                Widget := Icon
                            stack_box_slot:
                                HorizontalAlignment := horizontal_alignment.Left
                                VerticalAlignment := vertical_alignment.Fill
                                Padding := margin{Left := 20.0, Top := 0.0, Right := 0.0, Bottom := 0.0}
                                Widget := stack_box:
                                    Orientation := orientation.Vertical
                                    Slots := array:
                                        stack_box_slot:
                                            HorizontalAlignment := horizontal_alignment.Left
                                            VerticalAlignment := vertical_alignment.Fill
                                            Padding := margin{Left := 0.0, Top := 0.0, Right := 0.0, Bottom := -4.0}
                                            Widget := TierWidgetOverlay
                                        stack_box_slot:
                                            HorizontalAlignment := horizontal_alignment.Left
                                            VerticalAlignment := vertical_alignment.Fill
                                            Padding := margin{Left := 0.0, Top := 0.0, Right := 0.0, Bottom := 0.0}
                                            Widget := DescriptionText
                            stack_box_slot:
                                HorizontalAlignment := horizontal_alignment.Right
                                Distribution := option{1.0}
                                Padding := margin{Left := 100.0, Top := 0.0, Right := 0.0, Bottom := 0.0}
                                Widget := overlay:
                                    Slots := array:
                                        overlay_slot:
                                            HorizontalAlignment := horizontal_alignment.Right
                                            VerticalAlignment := vertical_alignment.Fill
                                            Widget := ProgressText
                                        overlay_slot:
                                            HorizontalAlignment := horizontal_alignment.Right
                                            Widget := CheckmarkIcon

    <# ----------------------------------------------------------------------------------------------------
        Update Functions
    ---------------------------------------------------------------------------------------------------- #>
    Update(DailyQuest:daily_quest):void =
        Id := DailyQuest.Id.ToDailyQuestId()
        if (QuestsInfo := GetDailyQuestsInfo[Id]):
            if (ActiveTierWidget := MaybeActiveTierWidget?):
                TierWidgetOverlay.RemoveWidget(ActiveTierWidget)

            TierWidget := GetTierWidgetSmall(DailyQuest.Tier + 1)
            TierWidgetOverlay.AddWidget of overlay_slot{Widget := TierWidget}
            set MaybeActiveTierWidget = option{TierWidget}           

            DescriptionText.SetText(MakeMessage("{QuestsInfo.ShortDescription}"))

            if (WeaponId := DailyQuest.MaybeWeaponId?.ToWeaponId(), WeaponIcon := GetWeaponInfo[WeaponId].Icon):
                Icon.SetImage(WeaponIcon)
            else:
                Icon.SetImage(QuestsInfo.Icon)

            CurrentProgress := Floor[DailyQuest.Progress] or 0
            TargetProgress := Floor[QuestsInfo.ProgressTargets[DailyQuest.Tier]] or 0
            
            if (CurrentProgress >= TargetProgress):
                CheckmarkIcon.SetVisibility(widget_visibility.Visible)
                ProgressText.SetVisibility(widget_visibility.Hidden)
            else:
                CheckmarkIcon.SetVisibility(widget_visibility.Hidden)
                ProgressText.SetVisibility(widget_visibility.Visible)
                ProgressText.SetText(MakeMessage("{CurrentProgress}/{TargetProgress}"))