using { /Fortnite.com/UI }
using { /Verse.org/Colors }
using { /UnrealEngine.com/Temporary/SpatialMath }
using { /UnrealEngine.com/Temporary/UI }
using { RealTime }

daily_quests_menu_ui := class(ui_component):
    GoldText:text_block = text_block:
        DefaultTextColor := ThemeColors.Success
        DefaultShadowColor := ThemeColors.Success * 0.125
        DefaultShadowOffset := option{vector2{X := 2.0, Y := 2.0}}
        DefaultShadowOpacity := 1.0

    var DailyQuestsMenuEntryUIs:[]daily_quests_menu_entry_ui = array{}

    <# ----------------------------------------------------------------------------------------------------
        Initialization
    ---------------------------------------------------------------------------------------------------- #>
    CreateUI<override>():widget =
        StackBox := stack_box:
            Orientation := orientation.Vertical
            Slots := array:
                stack_box_slot:
                    HorizontalAlignment := horizontal_alignment.Center
                    VerticalAlignment := vertical_alignment.Fill
                    Padding := margin{Left := 0.0, Top := 0.0, Right := 0.0, Bottom := 20.0}
                    Widget := text_block:
                        DefaultText := MakeMessage("Daily Quests")
                        DefaultTextColor := ThemeColors.PrimaryText
                        DefaultTextOpacity := 0.5

        for (Index := 1..3):
            DailyQuestsMenuEntryUI := daily_quests_menu_entry_ui{Player := Player}
            DailyQuestsMenuEntryUI.Initialize()
            set DailyQuestsMenuEntryUIs += array{DailyQuestsMenuEntryUI}

            StackBox.AddWidget of stack_box_slot:
                HorizontalAlignment := horizontal_alignment.Fill
                VerticalAlignment := vertical_alignment.Fill
                Padding := margin{Left := 0.0, Top := (Index = 1) and 0.0 or 10.0, Right := 0.0, Bottom := 0.0}
                Widget := DailyQuestsMenuEntryUI.GetUI()

        StackBox.AddWidget of stack_box_slot:
            HorizontalAlignment := horizontal_alignment.Fill
            VerticalAlignment := vertical_alignment.Bottom
            Distribution := option{1.0}
            Padding := margin{Left := 0.0, Top := 40.0, Right := 0.0, Bottom := 20.0}
            Widget := color_block:
                DefaultColor := NamedColors.White
                DefaultOpacity := 0.1
                DefaultDesiredSize := vector2{X := 0.0, Y := 1.0}

        StackBox.AddWidget of stack_box_slot:
            HorizontalAlignment := horizontal_alignment.Fill
            VerticalAlignment := vertical_alignment.Fill
            Widget := overlay:
                Slots := array:
                    overlay_slot:
                        HorizontalAlignment := horizontal_alignment.Fill
                        VerticalAlignment := vertical_alignment.Fill
                        Widget := UserInterface.VerseWidget.WBP_RoundedBox_Light{}
                    overlay_slot:
                        HorizontalAlignment := horizontal_alignment.Fill
                        VerticalAlignment := vertical_alignment.Fill
                        Padding := margin{Left := 40.0, Top := 20.0, Right := 40.0, Bottom := 20.0}
                        Widget := stack_box:
                            Orientation := orientation.Horizontal
                            Slots := array:
                                stack_box_slot:
                                    HorizontalAlignment := horizontal_alignment.Left
                                    VerticalAlignment := vertical_alignment.Fill
                                    Distribution := option{1.0}
                                    Widget := text_block:
                                        DefaultText := MakeMessage("Unclaimed Gold:")
                                        DefaultTextColor := ThemeColors.PrimaryText
                                        DefaultTextOpacity := 0.5
                                stack_box_slot:
                                    HorizontalAlignment := horizontal_alignment.Left
                                    Widget := texture_block:
                                        DefaultImage := currency_id.Gold.ToIcon()
                                        DefaultDesiredSize := vector2{X := 40.0, Y := 40.0}
                                stack_box_slot:
                                    HorizontalAlignment := horizontal_alignment.Fill
                                    VerticalAlignment := vertical_alignment.Fill
                                    Padding := margin{Left := 10.0, Top := 0.0, Right := 0.0, Bottom := 0.0}
                                    Widget := GoldText

        overlay:
            Slots := array:
                overlay_slot:
                    HorizontalAlignment := horizontal_alignment.Fill
                    VerticalAlignment := vertical_alignment.Fill
                    Widget := UserInterface.VerseWidget.WBP_RoundedBox{}
                overlay_slot:
                    HorizontalAlignment := horizontal_alignment.Fill
                    VerticalAlignment := vertical_alignment.Fill
                    Padding := margin{Left := 40.0, Top := 40.0, Right := 40.0, Bottom := 40.0}
                    Widget := StackBox

    <# ----------------------------------------------------------------------------------------------------
        Get Functions
    ---------------------------------------------------------------------------------------------------- #>
    GetUnclaimedGold()<transacts>:float =
        var UnclaimedGold:float = 0.0

        if (PlayerActiveQuests := Player.GetData[].DailyQuests.ActiveQuests):
            for (ActiveQuest : PlayerActiveQuests):
                Id := ActiveQuest.Id.ToDailyQuestId()
                if (QuestsInfo := GetDailyQuestsInfo[Id]):
                    CurrentProgress := Floor[ActiveQuest.Progress] or 0
                    TargetProgress := Floor[QuestsInfo.ProgressTargets[ActiveQuest.Tier]] or 0
                    
                    if (CurrentProgress >= TargetProgress):
                        set UnclaimedGold += (QuestsInfo.RewardAmounts[ActiveQuest.Tier] or 0.0) * GetDailyQuestsRankMultiplier(Player)

        UnclaimedGold

    <# ----------------------------------------------------------------------------------------------------
        Update Functions
    ---------------------------------------------------------------------------------------------------- #>
    Update<override>():void =
        GoldText.SetText(MakeMessage("{FormatAsCurrency(GetUnclaimedGold())}"))

        if (PlayerActiveQuests := Player.GetData[].DailyQuests.ActiveQuests):
            for (Index -> DailyQuestsMenuEntryUI : DailyQuestsMenuEntryUIs):
                if (ActiveQuest := PlayerActiveQuests[Index]):
                    DailyQuestsMenuEntryUI.Update(ActiveQuest)
                    DailyQuestsMenuEntryUI.Show()
                else:
                    DailyQuestsMenuEntryUI.Hide()
                    