using { /Fortnite.com/UI }
using { /Verse.org/Colors }
using { /UnrealEngine.com/Temporary/SpatialMath }
using { /UnrealEngine.com/Temporary/UI }

<# ----------------------------------------------------------------------------------------------------
    daily_quests_tier_rates_ui
---------------------------------------------------------------------------------------------------- #>
daily_quests_tier_rates_ui := class(ui_component):
    var TierTexts:[]text_block = array{}

    <# ----------------------------------------------------------------------------------------------------
        Initialization
    ---------------------------------------------------------------------------------------------------- #>
    CreateUI<override>():widget =
        TotalWeight := Sum(DailyQuestsTierWeights)

        overlay:
            Slots := array:
                overlay_slot:
                    HorizontalAlignment := horizontal_alignment.Fill
                    VerticalAlignment := vertical_alignment.Fill
                    Widget := UserInterface.VerseWidget.WBP_Panel_Background{}
                overlay_slot:
                    HorizontalAlignment := horizontal_alignment.Fill
                    VerticalAlignment := vertical_alignment.Fill
                    Padding := margin{Left := 40.0, Top := 40.0, Right := 40.0, Bottom := 40.0}
                    Widget := stack_box:
                        Orientation := orientation.Vertical
                        Slots := array:
                            stack_box_slot:
                                HorizontalAlignment := horizontal_alignment.Center
                                VerticalAlignment := vertical_alignment.Fill
                                Padding := margin{Left := 0.0, Top := 0.0, Right := 0.0, Bottom := 10.0}
                                Widget := text_block:
                                    DefaultText := MakeMessage("Tier Rates")
                                    DefaultTextColor := ThemeColors.PrimaryText
                            stack_box_slot:
                                HorizontalAlignment := horizontal_alignment.Fill
                                VerticalAlignment := vertical_alignment.Top
                                Padding := margin{Left := 0.0, Top := 0.0, Right := 0.0, Bottom := 40.0}
                                Widget := color_block:
                                    DefaultColor := NamedColors.White
                                    DefaultOpacity := 0.1
                                    DefaultDesiredSize := vector2{X := 0.0, Y := 1.0}
                            stack_box_slot:
                                HorizontalAlignment := horizontal_alignment.Fill
                                VerticalAlignment := vertical_alignment.Fill
                                Padding := margin{Left := 0.0, Top := 0.0, Right := 0.0, Bottom := 20.0}
                                Widget := overlay:
                                    Slots := array:
                                        overlay_slot:
                                            HorizontalAlignment := horizontal_alignment.Fill
                                            VerticalAlignment := vertical_alignment.Fill
                                            Widget := UserInterface.VerseWidget.WBP_RoundedBox{}
                                        overlay_slot:
                                            HorizontalAlignment := horizontal_alignment.Fill
                                            VerticalAlignment := vertical_alignment.Fill
                                            Padding := margin{Left := 40.0, Top := 40.0, Right := 40.0, Bottom := 40.0}
                                            Widget := stack_box:
                                                Orientation := orientation.Vertical
                                                Slots := for (Index := 1..5):
                                                    Percent := ((DailyQuestsTierWeights[Index - 1] * 1.0) / (TotalWeight * 1.0)) or 0.0
                                                    TierText := text_block:
                                                        DefaultText := MakeMessage("{FormatAsPercent(Percent, ?Decimals := 2)}")
                                                        DefaultTextColor := ThemeColors.Success
                                                        DefaultShadowColor := ThemeColors.Success * 0.125
                                                        DefaultShadowOffset := option{vector2{X := 2.0, Y := 2.0}}
                                                        DefaultShadowOpacity := 1.0

                                                    set TierTexts += array{TierText}

                                                    stack_box_slot:
                                                        HorizontalAlignment := horizontal_alignment.Fill
                                                        VerticalAlignment := vertical_alignment.Fill
                                                        Widget := stack_box:
                                                            Orientation := orientation.Horizontal
                                                            Slots := array:
                                                                stack_box_slot:
                                                                    HorizontalAlignment := horizontal_alignment.Left
                                                                    VerticalAlignment := vertical_alignment.Fill
                                                                    Distribution := option{1.0}
                                                                    Widget := text_block:
                                                                        DefaultText := MakeMessage(Index.ToTierString().ToUpper())
                                                                        DefaultTextColor := Index.ToTierColor() * 2.0
                                                                        DefaultShadowColor := Index.ToTierColor() * 0.25
                                                                        DefaultShadowOffset := option{vector2{X := 2.0, Y := 2.0}}
                                                                        DefaultShadowOpacity := 1.0
                                                                stack_box_slot:
                                                                    HorizontalAlignment := horizontal_alignment.Left
                                                                    VerticalAlignment := vertical_alignment.Fill
                                                                    Distribution := option{1.0}
                                                                    Padding := margin{Left := 100.0, Top := 0.0, Right := 0.0, Bottom := 0.0}
                                                                    Widget := TierText
