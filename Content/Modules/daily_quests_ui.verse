using { /Fortnite.com/UI }
using { /Verse.org/Colors }
using { /Verse.org/Simulation }
using { /UnrealEngine.com/Temporary/SpatialMath }
using { /UnrealEngine.com/Temporary/UI }
using { RealTime }

daily_quests_ui := class(panel_ui):
    Title<override>:string = "Daily Quests"
    Anchors<override>:anchors = anchors{Minimum := vector2{X := 0.15, Y := 0.5}, Maximum := vector2{X := 0.85, Y := 0.5}}
    BackgroundOpacity<override>:type {_X:float where 0.000000 <= _X, _X <= 1.000000} = 0.7
    
    RankIcon:texture_block = texture_block:
        DefaultImage := UserInterface.Icons.RankIcon_01_64
        DefaultDesiredSize := vector2{X := 40.0, Y := 40.0}

    RankMultiplier:text_block = text_block:
        DefaultTextColor := ThemeColors.Success
        DefaultShadowColor := ThemeColors.Success * 0.125
        DefaultShadowOffset := option{vector2{X := 2.0, Y := 2.0}}
        DefaultShadowOpacity := 1.0

    RerollTimerText:text_block = text_block:
        DefaultTextColor := ThemeColors.PrimaryText
        DefaultShadowOffset := option{vector2{X := 2.0, Y := 2.0}}
        DefaultShadowOpacity := 0.5
    
    var DailyQuestsEntryUIs:[]daily_quests_entry_ui = array{}
    var EventsSubscriptions:[]events_cancelable = array{}
    var DailyQuestsClaimed:int = 0
    var MaybeTierRatesUI:?daily_quests_tier_rates_ui = false
    var IsShowingTierRatesUI:logic = false

    <# ----------------------------------------------------------------------------------------------------
        Initialization
    ---------------------------------------------------------------------------------------------------- #>
    CreateContent<override>():widget =
        set EventsSubscriptions = array{ClosedEvent.Subscribe(OnClosed)}

        TierRatesButton:button_quiet = button_quiet{DefaultText := MakeMessage("Tier Rates")}        
        TierRatesButton.OnClick().Subscribe(OnClickTierRates)

        stack_box:
            Orientation := orientation.Vertical
            Slots := array:
                stack_box_slot:
                    HorizontalAlignment := horizontal_alignment.Fill
                    VerticalAlignment := vertical_alignment.Fill
                    Widget := stack_box:
                        Orientation := orientation.Horizontal
                        Slots := array:
                            stack_box_slot:
                                HorizontalAlignment := horizontal_alignment.Fill
                                VerticalAlignment := vertical_alignment.Fill
                                Distribution := option{1.0}
                                Widget := overlay:
                                    Slots := array:
                                        overlay_slot:
                                            HorizontalAlignment := horizontal_alignment.Fill
                                            VerticalAlignment := vertical_alignment.Fill
                                            Widget := UserInterface.VerseWidget.WBP_RoundedBox{}
                                        overlay_slot:
                                            Padding := margin{Left := 40.0, Top := 10.0, Right := 40.0, Bottom := 10.0}
                                            Widget := stack_box:
                                                Orientation := orientation.Horizontal
                                                Slots := array:
                                                    stack_box_slot:
                                                        HorizontalAlignment := horizontal_alignment.Left
                                                        Padding := margin{Left := 0.0, Top := 0.0, Right := 20.0, Bottom := 0.0}
                                                        Widget := RankIcon
                                                    stack_box_slot:
                                                        HorizontalAlignment := horizontal_alignment.Fill
                                                        VerticalAlignment := vertical_alignment.Fill
                                                        Widget := RankMultiplier
                            
                            stack_box_slot:
                                HorizontalAlignment := horizontal_alignment.Fill
                                VerticalAlignment := vertical_alignment.Fill
                                Distribution := option{1.0}
                                Padding := margin{Left := 20.0, Top := 0.0, Right := 0.0, Bottom := 0.0}
                                Widget := overlay:
                                    Slots := array:
                                        overlay_slot:
                                            HorizontalAlignment := horizontal_alignment.Fill
                                            VerticalAlignment := vertical_alignment.Fill
                                            Widget := UserInterface.VerseWidget.WBP_RoundedBox{}
                                        overlay_slot:
                                            Padding := margin{Left := 40.0, Top := 10.0, Right := 40.0, Bottom := 10.0}
                                            Widget := stack_box:
                                                Orientation := orientation.Horizontal
                                                Slots := array:
                                                    stack_box_slot:
                                                        HorizontalAlignment := horizontal_alignment.Left
                                                        Padding := margin{Left := 0.0, Top := 0.0, Right := 20.0, Bottom := 0.0}
                                                        Widget := texture_block:
                                                            DefaultImage := UserInterface.Icons.AbilityReroll_64
                                                            DefaultDesiredSize := vector2{X := 40.0, Y := 40.0}
                                                    stack_box_slot:
                                                        HorizontalAlignment := horizontal_alignment.Fill
                                                        VerticalAlignment := vertical_alignment.Fill
                                                        Widget := RerollTimerText
                            stack_box_slot:
                                HorizontalAlignment := horizontal_alignment.Fill
                                VerticalAlignment := vertical_alignment.Fill
                                Distribution := option{1.0}
                                Padding := margin{Left := 20.0, Top := 0.0, Right := 0.0, Bottom := 0.0}
                                Widget := TierRatesButton
                stack_box_slot:
                    HorizontalAlignment := horizontal_alignment.Fill
                    VerticalAlignment := vertical_alignment.Fill
                    Padding := margin{Left := 0.0, Top := 20.0, Right := 0.0, Bottom := 0.0}
                    Distribution := option{1.0}
                    Widget := stack_box:
                        Orientation := orientation.Horizontal
                        Slots := for (Index := 1..MaxDailyQuests):
                            DailyQuestsEntryUI := daily_quests_entry_ui{Player := Player, Index := Index - 1}
                            DailyQuestsEntryUI.Initialize()
                            set EventsSubscriptions += array:
                                DailyQuestsEntryUI.RerollEvent.Subscribe(OnReroll)
                                DailyQuestsEntryUI.ClaimEvent.Subscribe(OnClaim)
                            
                            set DailyQuestsEntryUIs += array{DailyQuestsEntryUI}
                            
                            stack_box_slot:
                                HorizontalAlignment := horizontal_alignment.Fill
                                VerticalAlignment := vertical_alignment.Fill
                                Padding := margin{Left := (Index = 1) and 0.0 or 20.0, Top := 0.0, Right := 0.0, Bottom := 0.0}
                                Distribution := option{1.0}
                                Widget := DailyQuestsEntryUI.GetUI()

    Uninitialize<override>():void =
        for (DailyQuestsEntryUI : DailyQuestsEntryUIs):
            DailyQuestsEntryUI.Uninitialize()
            
        for (EventsSubscription : EventsSubscriptions):
            EventsSubscription.Cancel()

    <# ----------------------------------------------------------------------------------------------------
        Update Functions
    ---------------------------------------------------------------------------------------------------- #>
    Update():void =
        RankIcon.SetImage(Player.GetRank().ToRankIcon())
        RankMultiplier.SetText(MakeMessage("+{FormatAsPercent(GetDailyQuestsRankMultiplier(Player))}"))

        if (Rerolls := Player.GetDailyQuestRerolls(), Rerolls > 0):
            RerollTimerText.SetText(MakeMessage("{Rerolls}"))
        else:
            if (RealTimeDevice := GetDailyQuestsManager[].RealTimeDevice):
                CurrentDateTime := RealTimeDevice.GetDateAndTime()
                TimeInterval := time_interval{Days := 1}
                NextDayDateTime := CurrentDateTime.Add(TimeInterval)
                NextQuestAvailableDateTime := MakeDateAndTime(NextDayDateTime.Years, NextDayDateTime.Months, NextDayDateTime.Days, 0, 0, 0.0)
                TimeIntervalRemaining := NextQuestAvailableDateTime.Subtract(CurrentDateTime)
                RerollTimerText.SetText(MakeMessage("{TimeIntervalRemaining.ToTruncatedString()}"))

        if (PlayerActiveQuests := Player.GetData[].DailyQuests.ActiveQuests):
            var HasSetPending:logic = false
            for (Index -> DailyQuestsEntryUI : DailyQuestsEntryUIs):
                if (PlayerActiveQuest := PlayerActiveQuests[Index]):
                    DailyQuestsEntryUI.SetActive(PlayerActiveQuest)
                else if (not HasSetPending?):
                    DailyQuestsEntryUI.SetPending()
                    set HasSetPending = true
                else:
                    DailyQuestsEntryUI.SetInactive()

    AwardAccolade():void =
        if (DailyQuestsClaimed <= 0) then return
        for (Index := 1..DailyQuestsClaimed) {Player.AwardAccolade(accolade_id.DailyQuest)}
        set DailyQuestsClaimed = 0

    <# ----------------------------------------------------------------------------------------------------
        Event Handlers
    ---------------------------------------------------------------------------------------------------- #>
    OnShow<override>():void =
        if (DailyQuestsManager := GetDailyQuestsManager[]):
            DailyQuestsManager.AssignNewQuests(Player)
            DailyQuestsManager.GrantDailyRerolls(Player)

        Update()

    OnClosed(InPlayer:player):void =
        if (DailyQuestsManager := GetDailyQuestsManager[]):
            DailyQuestsManager.UpdateBeacon(Player)

        AwardAccolade()

    OnReroll():void =
        Update()

    OnClaim():void =
        if (DailyQuestsManager := GetDailyQuestsManager[]):
            set DailyQuestsClaimed += 1
            DailyQuestsManager.AssignNewQuests(Player)
            DailyQuestsManager.UpdateBeacon(Player)
            Update()

    OnClickTierRates(Message:widget_message):void =
        UI := if (TierRatesUI := MaybeTierRatesUI?):
            TierRatesUI
        else:
            TierRatesUI := daily_quests_tier_rates_ui{Player := Player}
            TierRatesUI.Initialize()
            set MaybeTierRatesUI = option{TierRatesUI}

            if (PanelUI := canvas[MaybeUI?]):
                PanelUI.AddWidget of canvas_slot:
                    Anchors := anchors{Minimum := vector2{X := 0.0, Y := 0.0}, Maximum := vector2{X := 0.0, Y := 0.0}}
                    Alignment := vector2{X := 0.0, Y := 0.0}
                    ZOrder := 1
                    Widget := TierRatesUI.GetUI()

            TierRatesUI

        if (IsShowingTierRatesUI?):
            UI.Hide()
        else:
            UI.Show()

        set IsShowingTierRatesUI = logic{not IsShowingTierRatesUI?}