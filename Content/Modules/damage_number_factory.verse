using { /Fortnite.com/Characters }
using { /Verse.org/SceneGraph }
using { /Verse.org/Simulation }
using { /UnrealEngine.com/Temporary/SpatialMath }
using { /Verse.org/Concurrency }
using { Prefabs }

<# ----------------------------------------------------------------------------------------------------
    poolable_entity
---------------------------------------------------------------------------------------------------- #>
poolable_entity := class(poolable):
    Entity:entity

<# ----------------------------------------------------------------------------------------------------
    damage_number_factory
---------------------------------------------------------------------------------------------------- #>
damage_number_factory := class:
    var DamageNumbers:[]poolable_entity = array{}

    Initialize(SimulationEntity:entity, Count:int):void =
        TargetCount := Min(Count + 10, 90) - DamageNumbers.Length
        if (TargetCount = 0) then return

        DamageNumberEntities := for (Index := 1..TargetCount):
            DamageNumberEntity := prefab_damage_number{}
            DamageNumber := poolable_entity{Entity := DamageNumberEntity}
            set DamageNumbers += array{DamageNumber}
            DamageNumberEntity

        SimulationEntity.AddEntities(DamageNumberEntities)

    GetAvailable()<transacts>:?poolable_entity =
        var MaybeDamageNumber:?poolable_entity = false
        for (DamageNumber : DamageNumbers, not MaybeDamageNumber?, DamageNumber.IsAvailable?):
            set MaybeDamageNumber = option{DamageNumber}
        MaybeDamageNumber