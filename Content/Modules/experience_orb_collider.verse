using { /Fortnite.com/Characters }
using { /Verse.org }
using { /Verse.org/SceneGraph }
using { /Verse.org/Simulation }
using { /UnrealEngine.com/Temporary/SpatialMath }
using { /Verse.org/Concurrency }
using { VisualEffects }

<# ----------------------------------------------------------------------------------------------------
    Constants
---------------------------------------------------------------------------------------------------- #>
BaseExperienceOrbPickupDistance:float = 150.0

<# ----------------------------------------------------------------------------------------------------
    experience_orb_collision
---------------------------------------------------------------------------------------------------- #>
experience_orb_collision := class:
    Player:player

    Collision:collision_sphere = collision_sphere{}
    StoppedEvent:event() = event(){}

    var IsActive:logic = false

    <# ----------------------------------------------------------------------------------------------------
        Initialization
    ---------------------------------------------------------------------------------------------------- #>
    Start():void =
        if (IsActive?) then return
        UpdateCollisionRadius()
        spawn{AttachToPlayer()}

    Stop():void =
        StoppedEvent.Signal()

    <# ----------------------------------------------------------------------------------------------------
        Update Functions
    ---------------------------------------------------------------------------------------------------- #>
    AttachToPlayer()<suspends>:void =
        set IsActive = true
        PlayerIndex := GetPlayerIndex(Player)

        if (Character := Player.GetFortCharacter[], Entity := GetSimulationEntity[]):
            race:
                StoppedEvent.Await()
                sync:
                    if (ParticleSystem := NS_ExperienceOrbs_v2[GetExperienceOrbManager[].MaybeExperienceOrbsEntity?.GetComponent[particle_system_component]]):
                        loop:
                            if (set ParticleSystem.TargetPositions[PlayerIndex] = FromVector3 of Character.GetTransform().Translation.Make2D()) {}
                            Sleep(OneTick)

                    loop:
                        Position := FromVector3 of Character.GetTransform().Translation.Make2D()
                        Transform := SpatialMath.transform{Translation := Position}

                        for:
                            OverlapHit : Entity.FindOverlapHits(Transform, Collision)
                            TargetComponent := OverlapHit.TargetComponent
                            TargetEntity := TargetComponent.Entity
                            ExperienceComponent := TargetEntity.GetComponent[experience_orb_component]
                            ExperienceComponent.CanPickup[]
                        do:
                            spawn{ExperienceComponent.Pickup(Player, PlayerIndex, Position)}

                        Sleep(FiveTicks)

        set IsActive = false

    UpdateCollisionRadius():void =
        set Collision.Radius = BaseExperienceOrbPickupDistance * (1.0 + Player.GetStat(stat_id.ItemPickupDistance))