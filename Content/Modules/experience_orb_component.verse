using { /Verse.org/SceneGraph }
using { /Verse.org/Simulation }
using { /Verse.org/SpatialMath }
using { VisualEffects }

<# ----------------------------------------------------------------------------------------------------
    Data Definitions
---------------------------------------------------------------------------------------------------- #>
experience_orb_state := enum:
    Inactive
    Active
    Pickup

<# ----------------------------------------------------------------------------------------------------
    experience_orb
---------------------------------------------------------------------------------------------------- #>
experience_orb_component := class<final_super>(component):
    PickedUpEvent:event(tuple(int, int)) = event(tuple(int, int)){}

    var Index:int = -1
    var Amount:int = 0
    var State:experience_orb_state = experience_orb_state.Inactive
    var SpawnTime:float = 0.0

    <# ----------------------------------------------------------------------------------------------------
        Get Functions
    ---------------------------------------------------------------------------------------------------- #>
    IsAvailable()<decides><transacts>:void =
        State = experience_orb_state.Inactive

    CanPickup()<decides><transacts>:void =
        State = experience_orb_state.Active
        GetSimulationElapsedTime() - SpawnTime > ExperienceOrbSpawnDuration

    <# ----------------------------------------------------------------------------------------------------
        Update Functions
    ---------------------------------------------------------------------------------------------------- #>
    Spawn(Transform:transform, InAmount:int)<suspends>:void =
        set State = experience_orb_state.Active
        set SpawnTime = GetSimulationElapsedTime()
        set Amount = InAmount

        Sleep(ExperienceOrbSpawnDuration)
        Entity.SetGlobalTransform(Transform)

        if (MeshComponent := Entity.GetComponent[mesh_component]):
            MeshComponent.Enable()

    Despawn():void =
        set State = experience_orb_state.Inactive
        set Amount = 0

        if (MeshComponent := Entity.GetComponent[mesh_component]):
            MeshComponent.Disable()

    Pickup(Player:player, PlayerIndex:int, Position:vector3)<suspends>:void =
        set State = experience_orb_state.Pickup
        
        if (MeshComponent := Entity.GetComponent[mesh_component]):
            MeshComponent.Disable()

        PickedUpEvent.Signal(Index, PlayerIndex)
        Sleep(ExperienceOrbPickupDuration + ExperienceOrbAttractDuration)
        Player.PlayAudio(audio_id.Experience)
        AddXP(Amount)
        set State = experience_orb_state.Inactive

    Update(InAmount:int):void =
        set Amount = InAmount