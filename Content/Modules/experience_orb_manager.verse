using { /Fortnite.com/Characters }
using { /Fortnite.com/Devices }
using { /Verse.org/Random }
using { /Verse.org/SceneGraph }
using { /Verse.org/SceneGraph/KeyframedMovement }
using { /Verse.org/Simulation }
using { /Verse.org }
using { /UnrealEngine.com/Temporary/SpatialMath }
using { Prefabs }
using { VisualEffects }

<# ----------------------------------------------------------------------------------------------------
    Constants
---------------------------------------------------------------------------------------------------- #>
MaxExperienceOrbs:int = 100
ReservedExperienceOrbs:int = 20
TotalExperienceOrbs:int = MaxExperienceOrbs + ReservedExperienceOrbs
ExperienceOrbSpawnDuration:float = 0.4
ExperienceOrbPickupDuration:float = 0.2
ExperienceOrbAttractDuration:float = 0.3

<# ----------------------------------------------------------------------------------------------------
    Global Functions
---------------------------------------------------------------------------------------------------- #>
var GlobalExperienceOrbManager:weak_map(session, experience_orb_manager) = map{}

GetExperienceOrbManager()<decides><transacts>:experience_orb_manager =
    GlobalExperienceOrbManager[GetSession()]
<#
(Agent:agent).SpawnExperience(Amount:int):void =
    if (ExperienceOrbManager := GetExperienceOrbManager[], Position := Agent.GetFortCharacter[].GetTransform().Translation.Make2D()):
        ExperienceOrbManager.Spawn(Position, Amount)
#>
SpawnExperience(Position:vector3, Amount:int, ?IsUseReserve:logic = false):void =
    if (ExperienceOrbManager := GetExperienceOrbManager[]):
        ExperienceOrbManager.Spawn(Position, Amount, ?IsUseReserve := IsUseReserve)

(Player:player).UpdateExperienceOrbCollisionRadius():void =
    if (Collision := GetExperienceOrbManager[].PlayerCollision[Player]):
        Collision.UpdateCollisionRadius()

(Player:player).PickupAllExperience():void =
    if (ExperienceOrbManager := GetExperienceOrbManager[]):
        spawn{ExperienceOrbManager.PickupAllExperience(Player)}

<# ----------------------------------------------------------------------------------------------------
    experience_orb_manager
---------------------------------------------------------------------------------------------------- #>
experience_orb_manager := class(creative_device):

    var MaybeExperienceOrbsEntity:?entity = false
    var MaybeMagnetEntity:?prefab_magnet = false
    var ExperienceOrbs:[]prefab_experience_orb = array{}
    var PlayerCollision:[player]experience_orb_collision = map{}
    
    <# ----------------------------------------------------------------------------------------------------
        Initialization
    ---------------------------------------------------------------------------------------------------- #>
    OnBegin<override>()<suspends>:void =
        if (set GlobalExperienceOrbManager[GetSession()] = Self) {}

        set ExperienceOrbs = for (Index := 0..TotalExperienceOrbs - 1):
            ExperienceOrb := prefab_experience_orb{}

            if (ExperienceOrbComponent := ExperienceOrb.GetComponent[experience_orb_component]):
                set ExperienceOrbComponent.Index = Index
                spawn{AwaitPickedUpEvent(ExperienceOrbComponent.PickedUpEvent)}

            ExperienceOrb

        if (SimulationEntity := Self.GetSimulationEntity[]):
            SimulationEntity.AddEntities(ExperienceOrbs)
            
            ExperienceOrbsEntity := prefab_experience_orbs{}    
            SimulationEntity.AddEntities(array{ExperienceOrbsEntity})
            set MaybeExperienceOrbsEntity = option{ExperienceOrbsEntity}

            if:
                ParticleSystem := NS_ExperienceOrbs_v2[ExperienceOrbsEntity.GetComponent[particle_system_component]]
            then:
                set ParticleSystem.SpawnDuration = ExperienceOrbSpawnDuration
                set ParticleSystem.AttractDuration = ExperienceOrbAttractDuration

        for (Player : GetPlayspace().GetPlayers()):
            Initialize(Player)

        GetPlayspace().PlayerAddedEvent().Subscribe(OnPlayerAdded)
        GetPlayspace().PlayerRemovedEvent().Subscribe(OnPlayerRemoved)

        WaitForGameManager()
        if (GameManager := GetGameManager[]):
            GameManager.GameStartedEvent.Subscribe(OnGameStarted)
            GameManager.GameEndedEvent.Subscribe(OnGameEnded)

    Initialize(Player:player):void =
        if:
            not PlayerCollision[Player]
            set PlayerCollision[Player] = experience_orb_collision{Player := Player}

    <# ----------------------------------------------------------------------------------------------------
        Get Functions
    ---------------------------------------------------------------------------------------------------- #>
    GetAvailableExperienceOrb(?IsUseReserve:logic = false)<transacts>:?prefab_experience_orb =
        var MaybeExperienceOrb:?prefab_experience_orb = false
        for:
            Index -> ExperienceOrb : ExperienceOrbs
            not MaybeExperienceOrb?
            Index < MaxExperienceOrbs - 1 or IsUseReserve?
            Component := ExperienceOrb.GetComponent[experience_orb_component]
            Component.IsAvailable[]
        do:
            set MaybeExperienceOrb = option{ExperienceOrb}
        MaybeExperienceOrb

    GetSize(Amount:int)<transacts>:float =
        if (Amount >= 50):
            (20.0 + (Amount - 50) * 0.05)
        else if (Amount >= 25):
            17.5
        else:
            15.0

    <# ----------------------------------------------------------------------------------------------------
        Update Functions
    ---------------------------------------------------------------------------------------------------- #>
    Spawn(Position:vector3, Amount:int, ?IsUseReserve:logic = false):void =
        if:
            ExperienceOrbsEntity := MaybeExperienceOrbsEntity?
            ParticleSystem := NS_ExperienceOrbs_v2[ExperienceOrbsEntity.GetComponent[particle_system_component]]
        then:
            if:
                ExperienceOrb := GetAvailableExperienceOrb(?IsUseReserve := IsUseReserve)?
                ExperienceOrbComponent := ExperienceOrb.GetComponent[experience_orb_component]
                set ParticleSystem.Positions[ExperienceOrbComponent.Index] = FromVector3 of Position
                set ParticleSystem.Sizes[ExperienceOrbComponent.Index] = GetSize(Amount)
                set ParticleSystem.States[ExperienceOrbComponent.Index] = 1
            then:
                spawn:
                    ExperienceOrbComponent.Spawn of:
                        FromTransform of transform{Translation := Position, Scale := UnitVector * 0.35},
                        Amount
            else if:
                ExperienceOrb := ExperienceOrbs[GetRandomInt(0, 10)]
                ExperienceOrbComponent := ExperienceOrb.GetComponent[experience_orb_component]
                UpdatedAmount := ExperienceOrbComponent.Amount + Amount
                set ParticleSystem.Sizes[ExperienceOrbComponent.Index] = GetSize(UpdatedAmount)
            then:
                ExperienceOrbComponent.Update(UpdatedAmount)

    DespawnAll():void =
        for:
            ExperienceOrb : ExperienceOrbs
            ExperienceOrbComponent := ExperienceOrb.GetComponent[experience_orb_component]
        do:
            ExperienceOrbComponent.Despawn()

        if:
            ExperienceOrbsEntity := MaybeExperienceOrbsEntity?
            ParticleSystem := NS_ExperienceOrbs_v2[ExperienceOrbsEntity.GetComponent[particle_system_component]]
            set ParticleSystem.Positions = for (Index := 1..TotalExperienceOrbs) {SpatialMath.vector3{}}
            set ParticleSystem.TargetPositions = for (Index := 1..4) {SpatialMath.vector3{}}
            set ParticleSystem.TargetPositionIndexes = for (Index := 1..TotalExperienceOrbs) {0}
            set ParticleSystem.Sizes = for (Index := 1..TotalExperienceOrbs) {0.0}
            set ParticleSystem.States = for (Index := 1..TotalExperienceOrbs) {0}

    PickupAllExperience(Player:player)<suspends>:void =
        if (not MaybeMagnetEntity?):
            Prefab := prefab_magnet{}
            Prefab.InitializeParticle()
            set MaybeMagnetEntity = option{Prefab}

        if:
            Character := Player.GetFortCharacter[]
            ExperienceOrbsEntity := MaybeExperienceOrbsEntity?
            ExperienceOrbsParticleSystem := NS_ExperienceOrbs_v2[ExperienceOrbsEntity.GetComponent[particle_system_component]]
            MagnetEntity := MaybeMagnetEntity?
            MagnetParticleSystem := NS_ExperienceOrbMagnet[MagnetEntity.GetComponent[particle_system_component]]
            KeyframedMovementComponent := MagnetEntity.GetComponent[keyframed_movement_component]
        then:
            DeltaTimer := delta_timer{}
            Duration := 3.0

            var TotalXPAmount:int = 0
            var StartTime:float = GetSimulationElapsedTime()

            MagnetEntity.SetGlobalTransform of FromTransform of transform:
                Translation := Character.GetTransform().Translation

            Sleep(ThreeTicks)

            ExperienceOrbComponents := for:
                ExperienceOrb : ExperienceOrbs
                ExperienceOrbComponent := ExperienceOrb.GetComponent[experience_orb_component]
                ExperienceOrbComponent.CanPickup[]
            do:
                set TotalXPAmount += ExperienceOrbComponent.Amount
                ExperienceOrbComponent

            set MagnetParticleSystem.Positions = for (ExperienceOrbComponent : ExperienceOrbComponents) {ExperienceOrbComponent.Entity.GetGlobalTransform().Translation}
            set MagnetParticleSystem.Sizes = for (ExperienceOrbComponent : ExperienceOrbComponents) {GetSize(ExperienceOrbComponent.Amount)}

            DespawnAll()
            MagnetParticleSystem.Play()

            race:
                Sleep(5.0)

                block:
                    Sleep(1.0)
                    loop:
                        XPAmount := Ceil[DeltaTimer.GetDeltaTime() / Duration * TotalXPAmount] or 1
                        AddXP(XPAmount)

                        if (DeltaTimer.GetElapsedTime() >= Duration):
                            Sleep(Inf)
                            break

                        Sleep(0.1)

                loop:
                    RelativeKeyframes := array:
                        keyframed_movement_delta:
                            Transform := FromTransform of transform:
                                Translation :=  Character.GetTransform().Translation - FromVector3(MagnetEntity.GetGlobalTransform().Translation)
                                Scale := vector3{}
                            Duration := ThreeTicks
                            Easing := linear_easing_function{}

                    KeyframedMovementComponent.SetKeyframes(RelativeKeyframes, oneshot_keyframed_movement_playback_mode{})
                    KeyframedMovementComponent.Play()
                    Sleep(ThreeTicks)

    <# ----------------------------------------------------------------------------------------------------
        Event Handlers
    ---------------------------------------------------------------------------------------------------- #>
    OnPlayerAdded(Player:player):void =
        Initialize(Player)

    OnPlayerRemoved(Player:player):void =
        if (Collision := PlayerCollision[Player]):
            Collision.Stop()

        set PlayerCollision = RemoveKeyFromMap(PlayerCollision, Player)

    OnGameStarted():void =
        if:
            ExperienceOrbsEntity := MaybeExperienceOrbsEntity?
            ParticleSystem := NS_ExperienceOrbs_v2[ExperienceOrbsEntity.GetComponent[particle_system_component]]
        then:
            set ParticleSystem.Positions = for (Index := 1..TotalExperienceOrbs) {SpatialMath.vector3{}}
            set ParticleSystem.TargetPositions = for (Index := 1..4) {SpatialMath.vector3{}}
            set ParticleSystem.TargetPositionIndexes = for (Index := 1..TotalExperienceOrbs) {0}
            set ParticleSystem.Sizes = for (Index := 1..TotalExperienceOrbs) {0.0}
            set ParticleSystem.States = for (Index := 1..TotalExperienceOrbs) {0}
            ParticleSystem.Enable()

        for (Player : GetActivePlayers(), Collision := PlayerCollision[Player]):
            Collision.Start()

    OnGameEnded():void =
        for (Player : GetActivePlayers(), Collision := PlayerCollision[Player]):
            Collision.Stop()

    AwaitPickedUpEvent(PickedUpEvent:event(tuple(int, int)))<suspends>:void =
        loop:
            PickedUpResult := PickedUpEvent.Await()

            if:
                ExperienceOrbsEntity := MaybeExperienceOrbsEntity?
                ParticleSystem := NS_ExperienceOrbs_v2[ExperienceOrbsEntity.GetComponent[particle_system_component]]
                set ParticleSystem.TargetPositionIndexes[PickedUpResult(0)] = PickedUpResult(1)
                set ParticleSystem.States[PickedUpResult(0)] = 4