using { /Fortnite.com/Devices }
using { /Verse.org/Assets }
using { /Verse.org/Simulation }

<# ----------------------------------------------------------------------------------------------------
    Constants
---------------------------------------------------------------------------------------------------- #>
GameStartPlayerLoadingTime:float = 30.0
GameStartCountdownTime:float = 30.0

<# ----------------------------------------------------------------------------------------------------
    Global Functions
---------------------------------------------------------------------------------------------------- #>
var GlobalGameLobby:weak_map(session, game_lobby) = map{}

GetGameLobby()<decides><transacts>:game_lobby =
    GlobalGameLobby[GetSession()]

GetReadyPlayerCount()<transacts>:int =
    GetGameLobby[].GetReadyPlayerCount() or 0

IsAllPlayerReady()<transacts>:logic =
    GetGameLobby[].IsAllPlayerReady() or false

StartGameCountdown(Player:player, ArenaIndex:int):void =
    if (GameLobby := GetGameLobby[], not GameLobby.MaybeCountdownStartTime?):
        spawn{GameLobby.StartCountdown(Player, ArenaIndex)}

StopGameCountdown():void =
    if (GameLobby := GetGameLobby[]):
        GameLobby.StopCountdown()

IsGameCountdownInProgress()<transacts>:logic =
    logic{GetGameLobby[].IsGameCountdownInProgress()?}

<# ----------------------------------------------------------------------------------------------------
    game_lobby
---------------------------------------------------------------------------------------------------- #>
game_lobby := class(creative_device):
    @editable
    PlayerReadySwitch:switch_device = switch_device{}

    @editable
    ChannelChangingBoothEnable:channel_device = channel_device{}

    @editable
    ChannelChangingBoothDisable:channel_device = channel_device{}

    @editable
    RoundSettings:round_settings_device = round_settings_device{}

    @editable
    MatchmakingStatus:creative_prop = creative_prop{}
    
    @editable
    Billboard:billboard_device = billboard_device{}

    PlayerReadiedEvent<public>:events(tuple()) = events(tuple()){}
    PlayerUnreadiedEvent<public>:events(tuple()) = events(tuple()){}
    CountdownStartedEvent:events(tuple()) = events(tuple()){}
    CancelCountdownEvent:events(tuple()) = events(tuple()){}

    var MaybeCountdownStartTime:?float = false
    var MaybeMatchmakingDisabledTime:?float = false
    var IsMatchmakingDisabled:logic = false
    var IsPlayerLoadingCountdownComplete:logic = false
    var MatchmakingInfoPanelMaterial:Materials.InfoPanels.MI_InfoPanel_Matchmaking = Materials.InfoPanels.MI_InfoPanel_Matchmaking{}
    
    <# ----------------------------------------------------------------------------------------------------
        Initialization
    ---------------------------------------------------------------------------------------------------- #>
    OnBegin<override>()<suspends>:void =
        if (set GlobalGameLobby[GetSession()] = Self) {}

        set MatchmakingInfoPanelMaterial = Materials.InfoPanels.MI_InfoPanel_Matchmaking:
            Image := UserInterface.InfoPanel.InfoPanel_Matchmaking_Enabled
            
        UpdatePlayerCounter()
        MatchmakingStatus.SetMaterial(MatchmakingInfoPanelMaterial)
        
        GetPlayspace().PlayerAddedEvent().Subscribe(OnPlayerAdded)
        GetPlayspace().PlayerRemovedEvent().Subscribe(OnPlayerRemoved)
        PlayerReadySwitch.TurnedOnEvent.Subscribe(OnTurnedOn)
        PlayerReadySwitch.TurnedOffEvent.Subscribe(OnTurnedOff)

    <# ----------------------------------------------------------------------------------------------------
        Get Functions
    ---------------------------------------------------------------------------------------------------- #>
    (game_lobby:)GetReadyPlayerCount()<transacts>:int =
        var PlayersReady:int = 0
        for (Player : GetPlayspace().GetPlayers(), PlayerReadySwitch.GetCurrentState[Player]):
            set PlayersReady += 1
        PlayersReady

    (game_lobby:)IsAllPlayerReady()<transacts>:logic =
        logic{Self.GetReadyPlayerCount() = GetPlayspace().GetPlayers().Length}

    (game_lobby:)IsGameCountdownInProgress()<transacts>:logic =
        logic{MaybeCountdownStartTime?}

    GetCountdownTimeRemaining(Duration:float)<transacts>:float =
        Duration - GetElapsedTime()

    GetElapsedTime()<transacts>:float =
        (GetSimulationElapsedTime() - MaybeCountdownStartTime?) or 0.0

    CanStart()<transacts>:logic =
        logic:
            Self.IsGameCountdownInProgress()?
            (Self.IsAllPlayerReady()? and IsPlayerLoadingCountdownComplete?) or Self.GetReadyPlayerCount() = MaxPlayers

    <# ----------------------------------------------------------------------------------------------------
        Update Functions
    ---------------------------------------------------------------------------------------------------- #>
    UpdatePlayerCounter():void =
        Message := MakeMessage("{Self.GetReadyPlayerCount()}/{GetPlayspace().GetPlayers().Length}")
        Billboard.SetText(Message)
        
        if (Self.IsGameCountdownInProgress()?) then return
        ShowMessage(message_id.PlayerCounter, Message, ?DisplayTime := 0.0)

    StartCountdown(Player:player, ArenaIndex:int)<suspends>:void =
        race:
            CancelCountdownEvent.Await()
            block:
                CountdownStartedEvent.Signal()
                PlayerReadySwitch.TurnOn(Player)
                ChannelChangingBoothDisable.Transmit(false)
                set IsPlayerLoadingCountdownComplete = false

                if (not IsMatchmakingDisabled?, Self.GetReadyPlayerCount() < MaxPlayers):
                    if (MatchmakingDisabledTime := MaybeMatchmakingDisabledTime?):
                        MatchmakingDisabledElapsedTime := GetSimulationElapsedTime() - MatchmakingDisabledTime
                        if (MatchmakingDisabledElapsedTime < GameStartPlayerLoadingTime):
                            set MaybeCountdownStartTime = option{GetSimulationElapsedTime() - MatchmakingDisabledElapsedTime}
                            race:
                                Sleep(GameStartPlayerLoadingTime - MatchmakingDisabledElapsedTime)
                                loop:
                                    Sleep(1.0)
                                    ShowMessage(message_id.PlayerLoading, "{GetRoundedString(GetCountdownTimeRemaining(GameStartPlayerLoadingTime), 0)}", ?DisplayTime := 1.0)
                    else:
                        ToggleMatchmaking(false)
                        set MaybeCountdownStartTime = option{GetSimulationElapsedTime()}
                        race:
                            Sleep(GameStartPlayerLoadingTime)
                            loop:
                                Sleep(1.0)
                                ShowMessage(message_id.PlayerLoading, "{GetRoundedString(GetCountdownTimeRemaining(GameStartPlayerLoadingTime), 0)}", ?DisplayTime := 1.0)

                CountdownTime := Self.IsAllPlayerReady()? and 5.0 or GameStartCountdownTime
                set MaybeCountdownStartTime = option{GetSimulationElapsedTime()}
                set IsPlayerLoadingCountdownComplete = true
                ShowMessage(message_id.Countdown, "{GetRoundedString(CountdownTime, 0)}", ?DisplayTime := 1.0)
                race:
                    loop:
                        Sleep(1.0)
                        ShowMessage(message_id.Countdown, "{GetRoundedString(GetCountdownTimeRemaining(CountdownTime), 0)}", ?DisplayTime := 1.0)
                        
                        for (AllPlayer : GetPlayspace().GetPlayers()):
                            AllPlayer.PlayAudio(audio_id.TimerCountdown)
                    block:
                        Sleep(CountdownTime)
                        StartGame()

        HideMessage(message_id.Countdown)
        HideMessage(message_id.PlayerLoading)
        ChannelChangingBoothEnable.Transmit(false)
        set MaybeCountdownStartTime = false

        if (not IsMatchmakingDisabled?):
            spawn{EnableMatchmaking()}

    StopCountdown():void =
        CancelCountdownEvent.Signal()

    StartGame(?IsSkipIntro:logic = false):void =
        set IsMatchmakingDisabled = true
        set MatchmakingInfoPanelMaterial.Image = UserInterface.InfoPanel.InfoPanel_Matchmaking_Disabled_Permanent
        StopCountdown()
        HideMessage(message_id.PlayerCounter)
        StartGame(GetActiveArenaIndex()? or 0, ?IsSkipIntro := IsSkipIntro)
        
        for (Player : GetPlayspace().GetPlayers()):
            PlayerReadySwitch.SetState(Player, false)

    ToggleMatchmaking(IsEnable:logic):void =
        if (IsEnable?):
            RoundSettings.EnableMatchmaking()
            ShowMessage(message_id.MessageSmall, "Matchmaking Enabled", ?DisplayTime := 3.0)
            set MaybeMatchmakingDisabledTime = false
            set MatchmakingInfoPanelMaterial.Image = UserInterface.InfoPanel.InfoPanel_Matchmaking_Enabled
        else:
            RoundSettings.DisableMatchmaking()
            ShowMessage(message_id.MessageSmall, "Matchmaking Disabled", ?DisplayTime := 3.0)
            set MaybeMatchmakingDisabledTime = option{GetSimulationElapsedTime()}
            set MatchmakingInfoPanelMaterial.Image = UserInterface.InfoPanel.InfoPanel_Matchmaking_Disabled

    EnableMatchmaking()<suspends>:void =
        race:
            block:
                CountdownStartedEvent.Await()
            block:
                Sleep(15.0)
                ToggleMatchmaking(true)

    <# ----------------------------------------------------------------------------------------------------
        Event Handlers
    ---------------------------------------------------------------------------------------------------- #>
    OnPlayerAdded(Player:player):void =
        UpdatePlayerCounter()

    OnPlayerRemoved(Player:player):void =
        if (CanStart()?):
            StartGame()
            return

        UpdatePlayerCounter()

    OnTurnedOn(Agent:agent):void =
        if (Player := player[Agent]):
            Player.SetAFK(false)

        for (Player : GetPlayspace().GetPlayers()):
            Player.PlayAudio(audio_id.Ready)

        if (CanStart()?):
            StartGame()
            return

        UpdatePlayerCounter()
        PlayerReadiedEvent.Signal()

    OnTurnedOff(Agent:agent):void =
        if (Self.IsGameCountdownInProgress()?):
            if (Self.GetReadyPlayerCount() = 0):
                StopCountdown()

        UpdatePlayerCounter()
        PlayerUnreadiedEvent.Signal()