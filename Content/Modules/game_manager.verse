using { /Fortnite.com/AI }
using { /Fortnite.com/Characters }
using { /Fortnite.com/Devices }
using { /Verse.org/Random }
using { /Verse.org/SceneGraph }
using { /Verse.org/Simulation }
using { /UnrealEngine.com/Temporary }
using { /UnrealEngine.com/Temporary/SpatialMath }
using { /UnrealEngine.com/Temporary/UI }

<# ----------------------------------------------------------------------------------------------------
    Constants
---------------------------------------------------------------------------------------------------- #>
TeleportGroupSpacing:float = 100.0
TeleportToArenaPosition:vector3 = vector3{X := 0.0, Y := 0.0, Z := 1300.0}
GameResumeCountdownTimeDefault:int = 15
GameResumeCountdownTimeItemShop:int = 60
VictoryBonusGoldMultplier:float = 0.5
PlayerReviveDuration:float = 2.0
ArenaStageDuration:float = 360.0
EndlessModeBossStageInterval:int = 3

<# ----------------------------------------------------------------------------------------------------
    Data Definitions
---------------------------------------------------------------------------------------------------- #>
pause_type := enum:
    LevelUp
    TreasureChest
    ItemShop
    Boon
    PlayerMenu
    WeaponEvolution

ToString(Type:pause_type)<transacts>:string =
    case (Type):
        pause_type.LevelUp => "LevelUp"
        pause_type.TreasureChest => "TreasureChest"
        pause_type.ItemShop => "ItemShop"
        pause_type.Boon => "Boon"
        pause_type.PlayerMenu => "PlayerMenu"
        pause_type.WeaponEvolution => "WeaponEvolution"

pause_info := class:
    Type:pause_type

pause_level_up := class(pause_info):
    Level:int
    IsLegendary:logic = false

pause_weapon_evolution := class(pause_info):
    Id:weapon_id
    EvolutionLevel:int
    IsItemShop:logic = false

game_currency_id := enum:
    Wood
    TreasureChest

ToString(Id:game_currency_id)<transacts>:string =
    case (Id):
        game_currency_id.Wood => "Wood"
        game_currency_id.TreasureChest => "TreasureChest"

game_currency := class:
    Id:game_currency_id
    Amount:float = 0.0
    TotalAmount:float = 0.0

MakeGameCurrency<constructor>(UpdatedGameCurrency:game_currency)<transacts> := game_currency:
    Id := UpdatedGameCurrency.Id
    Amount := UpdatedGameCurrency.Amount
    TotalAmount := UpdatedGameCurrency.TotalAmount

<# ----------------------------------------------------------------------------------------------------
    Game Settings
---------------------------------------------------------------------------------------------------- #>
game_settings := class<concrete>:
    Arena:arena = arena{}
    Difficulty:int = 1
    GameSpeed:int = 1
    Weapons:int = 4
    IsPowerUpsShared:logic = false
    IsEndless:logic = false
    IsDoubleEnemyCount:logic = false

MakeGameSettings<constructor>(UpdatedGameSettings:game_settings)<transacts> := game_settings:
    Arena := UpdatedGameSettings.Arena
    Difficulty := UpdatedGameSettings.Difficulty
    GameSpeed := UpdatedGameSettings.GameSpeed
    Weapons := UpdatedGameSettings.Weapons
    IsPowerUpsShared := UpdatedGameSettings.IsPowerUpsShared
    IsEndless := UpdatedGameSettings.IsEndless
    IsDoubleEnemyCount := UpdatedGameSettings.IsDoubleEnemyCount

GetGameSettings()<decides><transacts>:game_settings =
    GetGameManager[].GameSettings

GetGameSpeed()<transacts>:float =
    (GetGameSettings[].GameSpeed * 1.0) or 1.0

GetGameDifficulty()<transacts>:int =
    GetGameSettings[].Difficulty or 1

SetGameSettings(GameSettings:game_settings):void =
    if (GameManager := GetGameManager[]):
        set GameManager.GameSettings = GameSettings
        
        GameManager.BillboardArena.SetText of:
            MakeMessage of:
                "{GameSettings.Arena.Name}"

        GameManager.BillboardArenaLevel.SetText of:
            MakeMessage("{CalculateArenaLevel(GameSettings.Arena)}")

        GameManager.BillboardGameSettings.SetText of:
            MakeMessage of:
                "x{Floor[GetGameSpeed()] or 1}" +
                "\n{GameSettings.Weapons}" +
                "\n{GameSettings.IsEndless? and "ON" or "OFF"}" +
                "\n{GameSettings.IsDoubleEnemyCount? and "90" or "45"}"

<# ----------------------------------------------------------------------------------------------------
    Game Stats
---------------------------------------------------------------------------------------------------- #>
game_stat_id := enum:
    EliminatedMiniboss
    GoldenTree

<# ----------------------------------------------------------------------------------------------------
    Global Functions
---------------------------------------------------------------------------------------------------- #>
var GlobalGameManager:weak_map(session, game_manager) = map{}

GetGameManager()<decides><transacts>:game_manager =
    GlobalGameManager[GetSession()]

WaitForGameManager()<suspends>:void =
    loop:
        if (GetGameManager[].GetIsInitialized()) {break}
        Sleep(0.3)

StartGame(ArenaIndex:int, ?IsSkipIntro:logic = false):void =
    if (GameManager := GetGameManager[]):
        spawn{GameManager.Start(ArenaIndex, ?IsSkipIntro := IsSkipIntro)}

PauseGame():void =
    if (GameManager := GetGameManager[]):
        GameManager.Pause()

PauseGame(PauseInfo:pause_info):void =
    if (GameManager := GetGameManager[]):
        GameManager.Pause(PauseInfo)

PauseGame(Player:player, PauseInfo:pause_info):void =
    if (GameManager := GetGameManager[]):
        GameManager.Pause(Player, PauseInfo)

PauseGameFront(Player:player, PauseInfo:pause_info):void =
    if (GameManager := GetGameManager[], UpdatedPauseQueue := GameManager.PlayerPauseQueue[Player].EnqueueFront(PauseInfo)):
        GameManager.SetPlayerPauseQueue(Player, UpdatedPauseQueue)
        GameManager.ShowPauseUI(Player, PauseInfo)

PauseGameFront(Player:player, PauseInfos:[]pause_info):void =
    if (GameManager := GetGameManager[]):
        for (Index -> PauseInfo : PauseInfos, UpdatedPauseQueue := GameManager.PlayerPauseQueue[Player].EnqueueFront(PauseInfo)):
            GameManager.SetPlayerPauseQueue(Player, UpdatedPauseQueue)
            if (Index = PauseInfos.Length - 1):
                GameManager.ShowPauseUI(Player, PauseInfo)

PauseGameOverride(Player:player, PauseInfo:pause_info):void =
    if:
        GameManager := GetGameManager[]
        Dequeued := GameManager.PlayerPauseQueue[Player].Dequeue[]
        UpdatedPauseQueue := Dequeued(0).EnqueueFront(PauseInfo)
    then:
        GameManager.SetPlayerPauseQueue(Player, UpdatedPauseQueue)
        GameManager.ShowPauseUI(Player, PauseInfo)

ResumeGame(?IsInstant:logic = false):void =
    if (GameManager := GetGameManager[]):
        GameManager.Resume(?IsInstant := IsInstant)

ResumeGame(Player:player, ?IsInstant:logic = false, ?IsShow:logic = true):void =
    if (GameManager := GetGameManager[]):
        GameManager.Resume(Player, ?IsInstant := IsInstant, ?IsShow := IsShow)

(Player:player).ShowPauseUI(PauseInfo:pause_info):void =
    if (GameManager := GetGameManager[]):
        GameManager.ShowPauseUI(Player, PauseInfo)

(Player:player).GetCurrentPause()<decides><transacts>:pause_info =
    GetGameManager[].PlayerPauseQueue[Player].Front[]

(Player:player).GetNextPause()<decides><transacts>:pause_info =
    GetGameManager[].PlayerPauseQueue[Player].Elements[1]

GetTotalPauseCount(Type:pause_type)<transacts>:int =
    var Count:int = 0
    for (Player : GetActivePlayers()):
        for (Element : GetGameManager[].PlayerPauseQueue[Player].Elements, Element.Type = Type):
            set Count += 1
    Count

(Player:player).GetPauseCount(Type:pause_type)<transacts>:int =
    var Count:int = 0
    for (Element : GetGameManager[].PlayerPauseQueue[Player].Elements):
        if (Element.Type = Type):
            set Count += 1
        else:
            return Count

    return Count

(Player:player).IsPlayerPaused()<transacts>:logic =
    GetGameManager[].IsPlayerPaused(Player) or false
    
(Player:player).GetPauseTime()<transacts>:float =
    GetGameManager[].PlayerPauseTime[Player] or 0.0

IsGamePaused()<transacts>:logic =
    GetGameManager[].IsPaused() or false

IsGameInProgress()<transacts>:logic =
    GetGameManager[].IsGameInProgress() or false

GetActivePlayers()<transacts>:[]player =
    GetGameManager[].GetActivePlayers() or array{}

GetActivePlayersNotAFK()<transacts>:[]player =
    GetGameManager[].GetActivePlayersNotAFK() or array{}

GetAlivePlayers()<transacts>:[]player =
    GetGameManager[].GetAlivePlayers() or array{}

GetVulnerablePlayers()<transacts>:[]player =
    GetGameManager[].GetVulnerablePlayers() or array{}

GetRandomAlivePlayer()<transacts>:?player =
    GetGameManager[].GetRandomAlivePlayer() or false

GetDownButNotOut()<decides><transacts>:down_but_not_out_device =
    GetGameManager[].GetDownButNotOut()

EndGame(IsVictory:logic):void =
    if (GameManager := GetGameManager[]):
        GameManager.Stop(IsVictory)

(Player:player).GetGameCurrency(Id:game_currency_id)<transacts>:float =
    GetGameManager[].GetGameCurrency(Player, Id) or 0.0

(Player:player).CanAfford(Id:game_currency_id, Amount:float)<transacts>:logic =
    logic{Player.GetGameCurrency(Id) >= Amount or IsAlmostEqual[Player.GetGameCurrency(Id), Amount, 0.5]}

AddGameCurrency(Id:game_currency_id, Amount:float):void =
    if (GameManager := GetGameManager[]):
        GameManager.AddGameCurrency(Id, Amount)

(Player:player).AddGameCurrency(Id:game_currency_id, Amount:float):void =
    if (GameManager := GetGameManager[]):
        GameManager.AddGameCurrency(Player, Id, Amount)

(Player:player).RemoveGameCurrency(Id:game_currency_id, Amount:float):void =
    if (GameManager := GetGameManager[]):
        GameManager.RemoveGameCurrency(Player, Id, Amount)

AddEnemiesEliminated(Player:player, Id:npc_id):void =
    if (GameManager := GetGameManager[]):
        GameManager.AddEnemiesEliminated(Player, Id)

AddObjectivesComplete():void =
    if (GameManager := GetGameManager[]):
        GameManager.AddObjectivesComplete()

GetEnemiesEliminated(Player:player)<transacts>:int =
    GetGameManager[].GetEnemiesEliminated(Player) or 0

SetBossHealthBarProgress(Progress:float):void =
    if (GameManager := GetGameManager[]):
        GameManager.CinematicSequenceBossHealthBar.SetPlaybackTime(Progress)

(Player:player).AutoRevive()<suspends>:void =
    if (GameManager := GetGameManager[]):
        GameManager.AutoRevive(Player)

GetCurrentStage()<transacts>:int =
    GetGameManager[].CurrentStage or 1

GetCurrentMinibossEvent()<transacts>:int =
    GetGameManager[].CurrentMinibossEvent or 0

(Player:player).IsInStageVolume()<transacts>:logic =
    GetGameManager[].IsPlayerInStageVolume(Player) or false

(Player:player).ShowGameResultsUI():void =
    if (GameManager := GetGameManager[], GameResult := GameManager.MaybeGameResult?):
        GameResult.ShowGameResultsUI(Player)

AddGameStat(Id:game_stat_id, Amount:float):void =
    if (GameManager := GetGameManager[]):
        GameManager.AddGameStat(Id, Amount)

GetGameStat(Id:game_stat_id)<transacts>:float =
    GetGameManager[].GetGameStat(Id) or 0.0

CalculateEndlessModeGoldMultiplier(Arena:arena, Stage:int)<transacts>:float =
    Arena.Level * 0.1 * Pow((Floor[(Stage * 1.0) / 3.0] or 0) * 1.0, 1.1)

<# ----------------------------------------------------------------------------------------------------
    game_manager
---------------------------------------------------------------------------------------------------- #>
game_manager := class(creative_device):
    @editable
    CinematicSequenceTimeDilation:cinematic_sequence_device = cinematic_sequence_device{}

    @editable
    CinematicSequenceGameEnd:cinematic_sequence_device = cinematic_sequence_device{}

    @editable
    CinematicSequenceBossHealthBar:cinematic_sequence_device = cinematic_sequence_device{}

    @editable
    GameplayControlsThirdPerson:gameplay_controls_third_person_device = gameplay_controls_third_person_device{}

    @editable
    GameplayCameraFixedAngle:gameplay_camera_fixed_angle_device = gameplay_camera_fixed_angle_device{}

    @editable
    GameplayCameraFixedAngleIntro:gameplay_camera_fixed_angle_device = gameplay_camera_fixed_angle_device{}

    @editable
    GameplayCameraFixedPointBossFight:gameplay_camera_fixed_point_device = gameplay_camera_fixed_point_device{}  
    
    @editable
    CinematicSequenceIntro:cinematic_sequence_device = cinematic_sequence_device{}

    @editable
    CinematicSequenceStageEntry:cinematic_sequence_device = cinematic_sequence_device{}

    @editable
    CinematicSequenceStageExit:cinematic_sequence_device = cinematic_sequence_device{}

    @editable
    CinematicSequenceBossFight:cinematic_sequence_device = cinematic_sequence_device{}

    @editable
    PostGameCinematicSequences:[]cinematic_sequence_device = array{}

    @editable
    MapIndicator:map_indicator_device = map_indicator_device{}

    @editable
    DownButNotOut:down_but_not_out_device = down_but_not_out_device{}
    
    @editable
    GameEndPositionLocator:creative_prop = creative_prop{}

    @editable
    VolumeStage:volume_device = volume_device{}
    
    @editable
    BarrierStage:barrier_device = barrier_device{}

    @editable
    BarrierStageEntry:barrier_device = barrier_device{}

    @editable
    BasicStormController:basic_storm_controller_device = basic_storm_controller_device{}
    
    @editable
    AdvancedStormController:advanced_storm_controller_device = advanced_storm_controller_device{}

    @editable
    BillboardArena:billboard_device = billboard_device{}

    @editable
    BillboardArenaLevel:billboard_device = billboard_device{}

    @editable
    BillboardGameSettings:billboard_device = billboard_device{}

    GameStartedEvent<public>:events(tuple()) = events(tuple()){}
    GameEndedEvent<public>:events(tuple()) = events(tuple()){}
    VictoryEvent<public>:events(tuple()) = events(tuple()){}
    StageEndedEvent<public>:events(tuple()) = events(tuple()){}
    StopTimerEvent<public>:events(tuple()) = events(tuple()){}
    GameResumedEvent<public>:events(tuple()) = events(tuple()){}
    PlayerResumedEvent<public>:events(logic) = events(logic){}
    GameCurrencyChangedEvent<public>:events(game_currency_id) = events(game_currency_id){}
    BossFightStartedEvent<public>:events(tuple()) = events(tuple()){}
    EliminationEvent<public>:events(tuple()) = events(tuple()){}
    CancelResumeCountdownEvent:events(tuple()) = events(tuple()){}
    ResumeCountdownStartEvent:events(tuple()) = events(tuple()){}
    ResumeCountdownCompleteEvent:events(tuple()) = events(tuple()){}
    ResumeCountdownEvent:events(int) = events(int){}
    MinibossEliminatedEvent<public>:events(tuple()) = events(tuple()){}
    AbilityActivatedEvent<public>:events(tuple()) = events(tuple()){}

    Timer:timer = timer{}

    var PlayerEliminationEvents<public>:[player]events(tuple()) = map{}
    var PlayerPauseQueue:[player]queue(pause_info) = map{}
    var PlayerPauseTime:[player]float = map{}

    var IsInitialized:logic = false
    var PlayerGameProgressBarUI:[player]game_progress_bar_ui = map{}
    var ActivePlayers:[]player = array{}

    var GameSettings:game_settings = game_settings{}
    var GameStartTime:float = 0.0
    var GameTime:float = 0.0
    var GameTotalTime:float = 0.0
    var EnemiesEliminated:int = 0
    var NPCsEliminated:[npc_id]float = map{}
    var ObjectivesCompleted:int = 0
    var PlayerGameCurrencies:[player][game_currency_id]game_currency = map{}
    var PlayerEnemiesEliminated:[player]int = map{}
    var MaybeResumeCountdownStartTime:?float = false
    var ResumeCountdownTimeRemaining:int = 0
    var CurrentStage:int = 1
    var CurrentMinibossEvent:int = 1
    var IsStart:logic = false
    var IsShipActive:logic = false
    var GameStats:[game_stat_id]float = map{}
    var PlayerGoldGranted:[player]float = map{}
    var PlayerTreasureChestGranted:[player]float = map{}
    
    var MaybeGameResult:?game_result = false

    <# ----------------------------------------------------------------------------------------------------
        Initialization
    ---------------------------------------------------------------------------------------------------- #>
    OnBegin<override>()<suspends>:void =
        if (set GlobalGameManager[GetSession()] = Self) {}

        for (Player : GetPlayspace().GetPlayers()):
            Initialize(Player)

        if (DefaultGameSettings := game_settings{Arena := GetArenas()[0]}):
            SetGameSettings(DefaultGameSettings)

        CinematicSequenceTimeDilation.SetPlayRate(0.0)
        CinematicSequenceTimeDilation.Play()
        CinematicSequenceTimeDilation.Pause()

        if (MapIndicator.TeleportTo[MapIndicatorDisabledPosition, IdentityRotation()]) {}
        GetPlayspace().PlayerAddedEvent().Subscribe(OnPlayerAdded)
        GetPlayspace().PlayerRemovedEvent().Subscribe(OnPlayerRemoved)
        DownButNotOut.AgentDownedEvent.Subscribe(OnAgentDowned)
        DownButNotOut.AgentRevivedEvent.Subscribe(OnAgentRevived)
        VolumeStage.AgentEntersEvent.Subscribe(OnAgentEntersStageVolume)
        VolumeStage.AgentExitsEvent.Subscribe(OnAgentExitsStageVolume)
        BasicStormController.DestroyStorm()

        set MaybeGameResult = option{game_result{GameManager := Self}}
        set IsInitialized = true

    OnEnd<override>():void =
        CinematicSequenceTimeDilation.GoToEndAndStop()
        CinematicSequenceGameEnd.GoToEndAndStop()

    Initialize(Player:player):void =
        if (set PlayerEliminationEvents[Player] = events(tuple()){}) {}

    AddToGame(Player:player)<suspends>:void =
        if ((game_manager:)IsGameInProgress()?, Character := Player.GetFortCharacter[]):
            set ActivePlayers += array{Player}
            Player.ResetStats()
            Player.CloseActivePanel(?IsSignal := true)
            TeleportGroup(Player, 0, TeleportToArenaPosition, MakeRotation(vector3{Z := 1.0}, PiFloat))
            Player.PlayMusic(GetArenaMusic())
            Player.PlayAudio(audio_id.Start)
            Player.RefreshControls()
            Player.ActivateAbility()
            Player.SetInputTriggers(InputTriggerCollections.Game)
            Player.AssignIndex()
            Character.SetVulnerability(false)
            Sleep(5.0)
            Character.SetVulnerability(true)
            if (set PlayerGameCurrencies[Player] = map{}) {}
    
    <# ----------------------------------------------------------------------------------------------------
        Get Functions
    ---------------------------------------------------------------------------------------------------- #>
    GetIsInitialized()<transacts>:logic =
        IsInitialized

    GetTotalAlivePlayersPaused()<transacts>:int =
        Sum of for (Player -> Queue : PlayerPauseQueue, IsAlive[Player]):
            Queue.IsEmpty()? and 0 or 1

    GetTotalPlayersPaused()<transacts>:int =
        Sum of for (Player -> Queue : PlayerPauseQueue):
            Queue.IsEmpty()? and 0 or 1

    IsPaused()<transacts>:logic =       
        PauseCount := Sum of for:
            Player -> Queue : PlayerPauseQueue
            not Queue.IsEmpty()? and IsAlive[Player]
        do:
            1

        logic{PauseCount = Self.GetAlivePlayers().Length}

    (game_manager:)IsPlayerPaused(Player:player)<transacts>:logic =
        logic{PlayerPauseQueue[Player] and not PlayerPauseQueue[Player].IsEmpty()?}

    (game_manager:)IsGameInProgress()<transacts>:logic =
        logic{ActivePlayers.Length > 0}

    (game_manager:)GetActivePlayers()<transacts>:[]player =
        ActivePlayers

    (game_manager:)GetActivePlayersNotAFK()<transacts>:[]player =
        for (ActivePlayer : ActivePlayers, not ActivePlayer.GetIsAFK()?) {ActivePlayer}

    IsAlive(Player:player)<decides><transacts>:void =
        Character := Player.GetFortCharacter[]
        Character.IsActive[]
        not Character.IsDownButNotOut[] or Player.GetPlayerResource(player_resource_id.Revives) > 0

    (game_manager:)GetAlivePlayers()<transacts>:[]player =
        for (ActivePlayer : ActivePlayers, IsAlive[ActivePlayer]):
            ActivePlayer

    (game_manager:)GetVulnerablePlayers()<transacts>:[]player =
        for (ActivePlayer : ActivePlayers, Character := ActivePlayer.GetFortCharacter[], Character.IsActive[], not Character.IsDownButNotOut[], Character.IsVulnerable[]):
            ActivePlayer

    (game_manager:)GetRandomAlivePlayer()<transacts>:?player =
        AlivePlayers := Self.GetAlivePlayers()
        if (RandomAlivePlayer := AlivePlayers[GetRandomInt(0, AlivePlayers.Length - 1)]):
            option{RandomAlivePlayer}
        else:
            false

    (game_manager:)GetDownButNotOut()<transacts>:down_but_not_out_device =
        DownButNotOut

    (game_manager:)GetGameCurrency(Player:player, Id:game_currency_id)<transacts>:float =
        PlayerGameCurrencies[Player][Id].Amount or 0.0

    GetTimeSurvived()<transacts>:float =
        Min(GameTotalTime, GameTime)

    CalculateGoldEarned(Player:player, IsVictory:logic)<transacts>:float =
        ActiveArena := GetActiveArena[] or arena{}
        
        ArenaGoldMultiplier := GetArenaGoldMultiplier()
        TotalWoodCollected := PlayerGameCurrencies[Player][game_currency_id.Wood].TotalAmount or 0.0
        GoldFromTimeSurvived := (GetTimeSurvived() * 3) * (1.0 + ArenaGoldMultiplier)
        GoldFromWoodCollected := TotalWoodCollected * 0.025 * (1.0 + ArenaGoldMultiplier)
        GoldFromObjectivesCompleted := ObjectivesCompleted * ObjectiveBaseRewardAmount * (1.0 + ArenaGoldMultiplier)
        GoldFromStagesCompleted := CurrentStage * 100.0 * (1.0 + ArenaGoldMultiplier)
        GoldFromGoldenTree := 5000.0 * Self.GetGameStat(game_stat_id.GoldenTree) * (1.0 + ArenaGoldMultiplier)
        
        GoldEarned := GoldFromTimeSurvived + GoldFromWoodCollected + GoldFromObjectivesCompleted + GoldFromStagesCompleted + GoldFromGoldenTree
        GoldFromVictory := if (GameSettings.IsEndless?):
            GoldEarned * CalculateEndlessModeGoldMultiplier(ActiveArena, CurrentStage)
        else:
            GoldEarned * (IsVictory? and VictoryBonusGoldMultplier or 0.0)
            
        GoldMultiplier := Player.GetStat(stat_id.GoldMultiplier) + Player.WeaponProgressGoldMultiplier()
        GoldFromMultiplier := (GoldEarned + GoldFromVictory) * GoldMultiplier
        TotalGoldEarned := (Floor[GoldEarned + GoldFromVictory + GoldFromMultiplier] or 0) * 1.0
    
    CalculateTreasureChestEarned(Player:player, IsVictory:logic)<transacts>:float =
        TreasureChestMultiplier := (IsVictory? and VictoryBonusGoldMultplier or 0.0) + Player.GetPowerUpValue(power_up_id.TreasureChestMultiplier) + Player.GetTreasureChestRankMultiplier()
        TreasureChestEarned := Self.GetGameStat(game_stat_id.EliminatedMiniboss) * (1.0 + TreasureChestMultiplier)
        (Ceil[TreasureChestEarned] * 1.0) or TreasureChestEarned

    GetEnemiesEliminated()<transacts>:int =
        EnemiesEliminated

    (game_manager:)GetEnemiesEliminated(Player:player)<transacts>:int =
        PlayerEnemiesEliminated[Player] or 0
        
    IsPlayerInStageVolume(Player:player)<transacts>:logic =
        logic{VolumeStage.IsInVolume[Player]}

    (game_manager:)GetGameStat(Id:game_stat_id)<transacts>:float =
        GameStats[Id] or 0.0

    <# ----------------------------------------------------------------------------------------------------
        Update Functions
    ---------------------------------------------------------------------------------------------------- #>
    Pause():void =
        if (CinematicSequenceTimeDilation.GetPlaybackTime() > 0.0 and CinematicSequenceTimeDilation.GetPlaybackTime() < 0.5) then return
        CinematicSequenceTimeDilation.SetPlaybackTime(TwoTicks - 0.000001)

    Pause(PauseInfo:pause_info):void =
        if (not Self.IsGameInProgress()?) then return
            
        for (Player : Self.GetActivePlayers()):
            Pause(Player, PauseInfo)

    Pause(Player:player, PauseInfo:pause_info):void =
        if (not Self.IsGameInProgress()?) then return

        if:
            not PlayerPauseQueue[Player]
            set PlayerPauseQueue[Player] = queue(pause_info){}
        
        if (PlayerPauseQueue[Player].IsEmpty()?):
            ShowPauseUI(Player, PauseInfo)
            Player.DeactivateAbility()
            Player.SetVulnerability(false)
            Player.ShowMessage(message_id.ReopenPauseMenu, "", ?DisplayTime := 0.0)
        
        if:
            UpdatedPauseQueue := PlayerPauseQueue[Player].Enqueue(PauseInfo)
            set PlayerPauseTime[Player] = GetSecondsSinceEpoch()
        then:
            SetPlayerPauseQueue(Player, UpdatedPauseQueue)

        if (GetTotalAlivePlayersPaused() = Self.GetAlivePlayers().Length):
            Pause()

    Resume(?IsInstant:logic = false):void =
        CinematicSequenceTimeDilation.SetPlaybackTime(0.6)
        GameResumedEvent.Signal()

    Resume(Player:player, ?IsInstant:logic = false, ?IsShow:logic = true):void =
        var IsGameResumed:logic = logic{IsPaused()? and (not Player.GetFortCharacter[].IsDownButNotOut[] or Player.GetPlayerResource(player_resource_id.Revives) > 0)}

        if (Dequeue := PlayerPauseQueue[Player].Dequeue[], UpdatedQueue := Dequeue(0)):
            if (QueueFront := UpdatedQueue.Front[]):
                if (IsShow?):
                    ShowPauseUI(Player, QueueFront)

                SetPlayerPauseQueue(Player, UpdatedQueue)
                return
            else:
                if (Dequeue(1).Type = pause_type.ItemShop):
                    if (GetTotalPlayersPaused() > 1):
                        set IsGameResumed = false
                    else:
                        set IsGameResumed = true

                if (IsGameResumed?):
                    Resume()
                    CancelResumeCountdownEvent.Signal()
                else if (IsPaused()?):
                    if (Dequeue(1).Type <> pause_type.ItemShop):
                        Player.ShowGamePausedUI()

                    if (not MaybeResumeCountdownStartTime?):
                        spawn{ResumeCountdown(Dequeue(1).Type)}

                SetPlayerPauseQueue(Player, UpdatedQueue)
        
        Player.SetVulnerability(true, ?BufferTime := IsInstant? and 0.0 or DefaultVulnerabilityBufferTime)
        Player.ReactivateAbility()
        Player.UpdateExperienceOrbCollisionRadius()
        Player.HideMessage(message_id.ReopenPauseMenu)

        if (not IsInstant?):
            KnockbackNPC(Player)

        PlayerResumedEvent.Signal(IsGameResumed)

    ShowPauseUI(Player:player, PauseInfo:pause_info):void =
        Player.CloseActivePanel()
        case (PauseInfo.Type):
            pause_type.LevelUp => if (PauseLevelUp := pause_level_up[PauseInfo]):
                Player.ShowAbilitySelectUI(PauseLevelUp.Level, ?IsLegendary := PauseLevelUp.IsLegendary)
                
            pause_type.WeaponEvolution => if (WeaponEvolutionInfo := pause_weapon_evolution[PauseInfo]):
                Player.ShowWeaponEvolutionUI(WeaponEvolutionInfo.Id, WeaponEvolutionInfo.EvolutionLevel, ?IsItemShop := WeaponEvolutionInfo.IsItemShop)

            pause_type.TreasureChest => Player.GrantAbilityUpgrade()
            pause_type.PlayerMenu => Player.ShowPlayerMenuUI()
            pause_type.Boon => Player.ShowBoonSelectUI()
            pause_type.ItemShop => Player.ShowItemShopUI()

    SetPlayerPauseQueue(Player:player, UpdatedQueue:queue(pause_info)):void =
        if (set PlayerPauseQueue[Player] = UpdatedQueue) {}

    ResumeCountdown(PauseType:pause_type)<suspends>:void =
        defer:
            set MaybeResumeCountdownStartTime = false

        GameResumeCountdownTimeDuration := PauseType = pause_type.ItemShop and GameResumeCountdownTimeItemShop or GameResumeCountdownTimeDefault
        set MaybeResumeCountdownStartTime = option{GetSecondsSinceEpoch()}
        ResumeCountdownStartEvent.Signal()

        race:
            CancelResumeCountdownEvent.Await()
            loop:
                SleepTimeDilation(1.0)
                set ResumeCountdownTimeRemaining = GameResumeCountdownTimeDuration - Floor[GetSecondsSinceEpoch() - MaybeResumeCountdownStartTime?] or 0
                ResumeCountdownEvent.Signal(ResumeCountdownTimeRemaining)

                if (ResumeCountdownTimeRemaining <= 0):
                    AutoCompletePauseQueue()
                    ResumeCountdownCompleteEvent.Signal()
                    break

    AutoCompletePauseQueue():void =
        Resume()
        for (Player : Self.GetAlivePlayers(), Character := Player.GetFortCharacter[], not Character.IsDownButNotOut[], IsPlayerPaused(Player)?):
            var AutoSelectCount:int = 0
            loop:
                if (Dequeue := PlayerPauseQueue[Player].Dequeue[]):
                    SetPlayerPauseQueue(Player, Dequeue(0))
                    PauseInfo := Dequeue(1)
                    case (PauseInfo.Type):
                        pause_type.LevelUp => if (Level := pause_level_up[PauseInfo].Level):
                            set AutoSelectCount += 1
                            if (IsWeaponUnlockLevel(Level)?, WeaponInfo := GetRandomWeapons(Player, 1)[0]):
                                Player.ActivateNewAbility(WeaponInfo.Id)
                                Player.EquipNewWeapon(WeaponInfo.Id)

                            else if (AbilityInfo := GetRandomAbility(Player)?):
                                Player.AddAbility(AbilityInfo(0), ?Tier := AbilityInfo(1))

                        pause_type.TreasureChest => if (RandomAbility := GetTreasureChestRandomAbility(Player)?):
                            set AutoSelectCount += 1
                            Player.AddAbility(RandomAbility(0), ?Tier := RandomAbility(1))

                        pause_type.Boon => if (RandomBoonInfo := GetRandomBoonInfo[Player]):
                            Player.AddBoon(RandomBoonInfo)
                        
                        pause_type.ItemShop => {}
                        pause_type.PlayerMenu => {}
                        pause_type.WeaponEvolution => {}

                else:
                    WeaponEvolutionInfos := Player.GetWeaponEvolutions()
                    for (WeaponId -> WeaponLevel : Player.GetWeaponLevels(), WeaponInfo := GetWeaponInfo[WeaponId]):
                        Current := WeaponEvolutionInfos[WeaponId].Length or 0
                        Target := Player.GetWeaponEvolutionLevel(WeaponId)
                        Delta := Target - Current
        
                        for:
                            Index := 1..Delta
                            EvolutionInfos := WeaponInfo.Evolution[Current + (Index - 1)]
                            EvolutionIndex := GetRandomInt(0, EvolutionInfos.Length - 1)
                            WeaponEvolutionInfo := EvolutionInfos[EvolutionIndex]
                        do:
                            Player.AddWeaponEvolution(WeaponEvolutionInfo, EvolutionIndex)

                    Player.CloseActivePanel()
                    Player.ShowMessage(message_id.AutoSelectedAbilities, "{AutoSelectCount}")
                    Resume(Player)
                    break

        PlayerResumedEvent.Signal(true)

    Start(ArenaIndex:int, ?IsSkipIntro:logic = false)<suspends>:void =
        if (IsStart?) then return
        set IsStart = true
        set GameTime = 0.0
        set GameStartTime = GetSecondsSinceEpoch()
        set CurrentStage = 1
        set ActivePlayers = GetAllPlayers()
        set PlayerGoldGranted = map{}
        set PlayerTreasureChestGranted = map{}

        ResetStats()
        ResetGameStats()

        ArenaLevel := CalculateArenaLevel(GameSettings.Arena.Level, GameSettings.Difficulty)

        if (ArenaManager := GetArenaManager[], Arena := ArenaManager.GetActiveArena[], NPCManager := GetNPCManager[]):
            StageCount := GameSettings.IsEndless? and 9999 or GetArenaStageCount()
            set GameTotalTime = StageCount * ArenaStageDuration
            ArenaManager.LoadGatherables(ArenaIndex)
            GameStartedEvent.Signal()

            if (SimulationEntity := Self.GetSimulationEntity[]):
                NPCManager.DamageNumberEntityFactory.Initialize(SimulationEntity, NPCManager.GetMaxSpawnCount())
                NPCManager.CollisionEntityFactory.Initialize(SimulationEntity, NPCManager.GetMaxSpawnCount())

            ShowMessage(message_id.CurrencyElimination, "{FormatAsCurrency(0.0)}", ?DisplayTime := 0.0)
            for (Index -> Player : GetPlayspace().GetPlayers(), Character := Player.GetFortCharacter[]):
                Player.CloseActivePanel(?IsSignal := true)
                Player.SetVulnerability(true)
                Player.ClearMusic()
                Player.PutInStasis(?IsAllowFalling := false)
                Player.SetInputTriggers(InputTriggerCollections.Game)
                TeleportGroup(Player, Index, TeleportToArenaPosition, MakeRotation(vector3{Z := 1.0}, PiFloat))

                if (not IsSkipIntro?, Player.GetActiveCamera() <> camera_type.Orthographic):
                    GameplayCameraFixedAngleIntro.AddTo(Player)

                if (Player.GetIsAFK()?):
                    Player.ActivateVFXPowerup(vfx_powerup_id.AFK)
                else:
                    Player.DeactivateVFXPowerup of Player.GetRankVFXPowerUp()

                if (set PlayerGameCurrencies[Player] = map{}) {}

            Sleep(0.0)
            for (Player : GetPlayspace().GetPlayers()):
                Player.ReleaseFromStasis()

            if (not IsSkipIntro?):
                CinematicSequenceIntro.Play()
                Sleep(4.5)
                GameplayCameraFixedAngleIntro.RemoveFromAll()
                ActivateCameraShake(DefaultArenaOrigin)

                # [Hotfix] Re-hides the ship if it reshows on sequence completion.
                spawn{ClearCinematicSequence(?IsAwaitIntro := true)}

            ArenaMusic := GetArenaMusic()
            for (Index -> Player : GetPlayspace().GetPlayers()):
                Player.RefreshControls()
                Player.ActivateAbility()
                Player.PlayAudio(audio_id.Start)
                Player.PlayMusic(ArenaMusic)

            AbilityActivatedEvent.Signal()
            
            race:
                StopTimerEvent.Await()
                block:
                    for (StageIndex := 1..StageCount):
                        ShowMessage(message_id.StageIndicator, "{StageIndex}/{GameSettings.IsEndless? and "∞" or "{StageCount}"}", ?DisplayTime := 0.0)
                        ShowMessage(message_id.Message, (StageIndex = StageCount) and "Final Stage!" or "Stage {StageIndex}")
                        
                        for (Player : Self.GetActivePlayers()):
                            Player.PlayAudio(audio_id.Start)

                            if (StageIndex = StageCount):
                                Player.PlayMusic(music_id.CombatHigh)

                        spawn{NPCManager.StartWaves(ArenaIndex)}
                        StartTimer(ArenaStageDuration, logic{StageIndex = StageCount})
                        
                        if (not IsMinibossEventCompleted()?):
                            ShowMessage(message_id.MinibossEliminatedCounterIntro, "{GetMinibossEliminationProgress()}/{GetMinibossEliminationTarget()}", ?DisplayTime := 0.0)
                            for (Gatherable : GetGatherables()):
                                Gatherable.DisableRespawn()

                            loop:
                                MinibossEliminatedEvent.Await()
                                HideMessage(message_id.MinibossEliminatedCounterIntro)
                                ShowMessage(message_id.MinibossEliminatedCounter, "{GetMinibossEliminationProgress()}/{GetMinibossEliminationTarget()}", ?DisplayTime := 0.0)

                                if (IsMinibossEventCompleted()?):
                                    Sleep(3.0)
                                    break

                        if (CurrentObjective := GetCurrentObjective()?):
                            ShowMessage(message_id.Timer, "Finish Mission", ?DisplayTime := 0.0)
                            CurrentObjective.CompleteEvent.Await()
                            Sleep(6.0)

                        if (GameSettings.IsEndless?, Mod[StageIndex, EndlessModeBossStageInterval] = 0):
                            StartBossFight()

                        if (StageIndex < StageCount):
                            # Generate storm & station entry cinematic sequence
                            set IsShipActive = true
                            ShowMessage(message_id.PlayerCounterShip, "0/{Self.GetActivePlayers().Length}", ?DisplayTime := 0.0)
                            ToggleItemShop(true)
                            CinematicSequenceStageEntry.Play()


                            # Apply upwards impulse to players in the ship's spawn area to prevent players getting stuck
                            for (Player : Self.GetActivePlayers()):
                                if:
                                    PlayerCharacter := Player.GetFortCharacter[]
                                    PlayerCharacter.IsActive[]
                                    PlayerPosition := PlayerCharacter.GetTransform().Translation
                                    DistanceSquaredXY(PlayerPosition, vector3{}) <= 640000.0
                                then:
                                    Player.ApplyImpulse(impulse_id.Upward)

                            Sleep(0.1)
                            BarrierStageEntry.Enable()
                            Sleep(0.9)
                            BarrierStageEntry.Disable()

                            # Await all players to enter ship
                            Sleep(4.0)
                            AdvancedStormController.GenerateStorm()
                            BasicStormController.GenerateStorm()
                            AdvancedStormController.DestroyStorm()

                            var ShipTimeRemaining:float = 30.0
                            race:
                                block:
                                    Sleep(15.0)
                                    for (Player : Self.GetActivePlayers()) {Player.PlayAudio(audio_id.TimerWarning)}
                                    for (Index := 1..15):
                                        Sleep(1.0)
                                        for (Player : Self.GetActivePlayers()) {Player.PlayAudio(audio_id.TimerCountdown)}
                                    Sleep(Inf)

                                block:
                                    loop:
                                        if (ShipTimeRemaining <= 0.0):
                                            ShowMessage(message_id.TimerShip, "00:00", ?DisplayTime := 0.0)
                                            break
                                        
                                        ShowMessage(message_id.TimerShip, "{ShipTimeRemaining.FormatAsTime()}", ?DisplayTime := 0.0)
                                        Sleep(1.0)
                                        set ShipTimeRemaining -= 1.0
                                        
                                    Sleep(1.0)
                                    for (AlivePlayer : Self.GetAlivePlayers()):
                                        if (not VolumeStage.IsInVolume[AlivePlayer]):
                                            AlivePlayer.GiveUp()

                                loop:
                                    if (VolumeStage.GetPlayersInVolume().Length >= Self.GetActivePlayers().Length):
                                        break

                                    Listenable:listenable(agent) = VolumeStage.AgentEntersEvent
                                    Listenable.Await()

                            Sleep(0.0)
                            set IsShipActive = false
                            BarrierStage.Enable()

                            Sleep(1.0)
                            AwardAccolade(accolade_id.StageCompleted, Self.GetActivePlayers())
                            HideMessage(message_id.TimerShip)
                            StageEndedEvent.Signal()
                            set CurrentStage = StageIndex + 1
                            GrantRewards(false)

                            for (Player : Self.GetActivePlayers(), GameProgressBarUI := PlayerGameProgressBarUI[Player]):
                                GameProgressBarUI.StageEnded()

                            for (NPCIndex -> NPCBehavior : GetAllNPCs()):
                                NPCBehavior.InterruptState(NPCState.npc_state_type.Dance)

                            # Ship exit cinematic sequence
                            CinematicSequenceStageEntry.Stop()
                            CinematicSequenceStageExit.Play()

                            Sleep(6.5)
                            BarrierStage.Disable()

                            # Item shop
                            NPCManager.StopSpawners()
                            BasicStormController.DestroyStorm()
                            ToggleItemShop(false)
                            ClearItemDrops()
                            ArenaManager.LoadGatherables(ArenaIndex)
                            PauseGame(pause_info{Type := pause_type.ItemShop})
                            TeleportGroup(Self.GetActivePlayers(), TeleportToArenaPosition, MakeRotation(vector3{Z := 1.0}, PiFloat))
                            for (Player : Self.GetActivePlayers()):
                                Player.PutInStasis(?IsAllowFalling := false)

                            # Next stage entry cinematic sequence
                            TeleportGroup(Self.GetActivePlayers(), TeleportToArenaPosition, MakeRotation(vector3{Z := 1.0}, PiFloat))
                            for (Player : Self.GetActivePlayers()):
                                Player.ReleaseFromStasis()

                            CinematicSequenceIntro.Play()
                            spawn{ShowSaving()}
                            Sleep(4.5)
                            ActivateCameraShake(DefaultArenaOrigin)
                            CinematicSequenceIntro.StoppedEvent.Await()
                            
                            # [Hotfix] Re-hides the ship if it reshows on sequence completion.
                            spawn{ClearCinematicSequence()}
                            
                            for (Player : Self.GetActivePlayers()):
                                Player.SetVulnerability(true)
                                Player.ReactivateAbility()
                        else:
                            set CurrentStage = StageCount
                            StageEndedEvent.Signal()

                    spawn{StartBossFight()}

    ShowSaving()<suspends>:void =
        Sleep(1.0)
        ShowMessage(message_id.Saving, "", ?DisplayTime := 3.0)

    ClearCinematicSequence(?IsAwaitIntro:logic = false)<suspends>:void =
        if (IsAwaitIntro?):
            CinematicSequenceIntro.StoppedEvent.Await()

        Sleep(2.0)
        CinematicSequenceIntro.SetPlaybackFrame(200)
        CinematicSequenceIntro.Play()

    GrantRewards(IsVictory:logic):void =
        for (Player : Self.GetActivePlayers()):
            GrantRewards(Player, IsVictory)

    GrantRewards(Player:player, IsVictory:logic, ?MaybeResultInfo:?game_result_info = false):void =
        var GoldEarned:float = MaybeResultInfo?.GoldEarned or CalculateGoldEarned(Player, IsVictory)
        var TreasureChestEarned:float = MaybeResultInfo?.TreasureChestEarned or CalculateTreasureChestEarned(Player, IsVictory)

        if (GoldGranted := PlayerGoldGranted[Player]):
            set GoldEarned -= GoldGranted

        if (TreasureChestGranted := PlayerTreasureChestGranted[Player]):
            set TreasureChestEarned -= TreasureChestGranted
        
        Player.AddCurrency(currency_id.Gold, GoldEarned, ?IsSignal := logic{MaybeResultInfo?})
        Player.AddCurrency(currency_id.TreasureChest, TreasureChestEarned, ?IsSignal := logic{MaybeResultInfo?})
        
        if:
            set PlayerGoldGranted[Player] = PlayerGoldGranted[Player] + GoldEarned or GoldEarned
            set PlayerTreasureChestGranted[Player] = PlayerTreasureChestGranted[Player] + TreasureChestEarned or TreasureChestEarned

        if (GameSettings.IsEndless?, ActiveArenaIndex := GetActiveArenaIndex()?):
            Player.SetEndlessModeProgress(ActiveArenaIndex, CurrentStage)

    StartTimer(Duration:float, IsFinalStage:logic)<suspends>:void =
        var StageTime:float = 0.0
        set CurrentMinibossEvent = 0

        GameEvents := if (IsFinalStage? or (GetGameSettings[].IsEndless? and Mod[CurrentStage, EndlessModeBossStageInterval] = 0)):
            GameEventSequenceFinalStage
        else:
            GameEventSequence

        GameEventCount := GameEvents.Length
        GameSpeed := GetGameSpeed()
        
        defer:
            for (Player : Self.GetActivePlayers(), GameProgressBarUI := PlayerGameProgressBarUI[Player]):
                GameProgressBarUI.UpdateEventState(GameEventCount - 1, game_event_state.Active)

        for (Player : Self.GetActivePlayers()):
            ShowGameProgressBarUI(Player)
            if (GameProgressBarUI := PlayerGameProgressBarUI[Player]):
                GameProgressBarUI.InitializeEvents(GameEvents)
        
        race:
            StopTimerEvent.Await()
            block:
                var EventIndex:int = 0
                loop:
                    Sleep(Duration / (GameEventCount * 1.0) / GameSpeed)
                    if (GameEvent := GameEvents[EventIndex]):
                        case (GameEvent):
                            game_events.Miniboss =>
                                set CurrentMinibossEvent += 1
                                SpawnRandomMiniboss()
                                ShowMessage(message_id.Message, "Miniboss Spawned!")

                            game_events.Mission => StartNextObjective()
                            game_events.SupplyDrop => SpawnSupplyDrop()

                            _ => {}

                    for (Player : Self.GetActivePlayers(), GameProgressBarUI := PlayerGameProgressBarUI[Player]):
                        GameProgressBarUI.UpdateEventState(EventIndex, game_event_state.Active)

                    if (EventIndex >= GameEventCount - 1):
                        break

                    set EventIndex += 1

                Sleep(Inf)

            loop:
                TimeRemaining := Duration - StageTime
                Progress := 1.0 - TimeRemaining / Duration
                
                UpdateMessage(message_id.Timer, TimeRemaining.FormatAsTime())
                for (Player : Self.GetActivePlayers(), GameProgressBarUI := PlayerGameProgressBarUI[Player]):
                    GameProgressBarUI.UpdateProgress(Progress)

                if (StageTime >= Duration):
                    break
                else:
                    Sleep(1.0)
                    set StageTime += GameSpeed
                    set GameTime += GameSpeed

    ShowGameProgressBarUI(Player:player):void =
        ProgressBarUI := if (GameProgressBarUI := PlayerGameProgressBarUI[Player]):
            GameProgressBarUI
        else:
            GameProgressBarUI := game_progress_bar_ui{Player := Player}
            GameProgressBarUI.Initialize()
            if (set PlayerGameProgressBarUI[Player] = GameProgressBarUI) {}
            
            Player.AddWidgetToHUD of canvas_slot:
                Anchors := anchors{Minimum := vector2{X := 0.5, Y := 0.0}, Maximum := vector2{X := 0.5, Y := 0.0}}
                Offsets := margin{Left := 0.0, Top := 20.0, Right := 0.0, Bottom := 0.0}
                Alignment := vector2{X := 0.5, Y := 0.0}
                Widget := GameProgressBarUI.GetUI()

            GameProgressBarUI

        ProgressBarUI.Show()
                    
    TeleportGroup(Players:[]player, Position:vector3, Rotation:rotation):void =
        for (Index -> Player : Players):
            TeleportGroup(Player, Index, Position, Rotation)

    TeleportGroup(Player:player, Index:int, Position:vector3, Rotation:rotation):void =
        Offset := Index * TeleportGroupSpacing - (GetAllPlayers().Length - 1) * 0.5 * TeleportGroupSpacing
        TargetPosition := Position - Rotation.GetLocalRight() * Offset

        if :
            Character := Player.GetFortCharacter[]
            Character.TeleportTo[TargetPosition, Rotation]

    StartBossFight()<suspends>:void =
        BossFightStartedEvent.Signal()

        MaybeNPCSpawner:?npc_spawner_device = if (TargetPlayer := Self.GetRandomAlivePlayer()?, Arena := GetActiveArena[]):
            SpawnNPC(Arena.BossId, TargetPlayer, ?MaybePosition := option{vector3{Z := 1400.0}})
        else:
            false

        MaybeAgent:?agent = if (NPCSpawner := MaybeNPCSpawner?):
            Agent := NPCSpawner.SpawnedEvent.Await()
            option{Agent}
        else:
            false

        CinematicSequenceBossFight.Play()
        CinematicSequenceBossHealthBar.Play()

        CinematicSequenceBossHealthBar.Pause()
        Sleep(0.0)
        CinematicSequenceBossHealthBar.SetPlaybackTime(1.0)

        HideMessage(message_id.PlayerCounterShip)
        HideMessage(message_id.TimerShip)
        HideMessage(message_id.MinibossEliminatedCounterIntro)
        HideMessage(message_id.MinibossEliminatedCounter)
        
        ShowMessage(message_id.BossWarning, "", ?DisplayTime := 3.0)
        for (Index -> Player : Self.GetActivePlayers()):
            Player.ToggleHUD(false)
            Player.HideMessage(message_id.Timer)
            Player.HideMessage(message_id.StageIndicator)
            Player.PlayAudio(audio_id.Warning)

            if (GameProgressBarUI := PlayerGameProgressBarUI[Player]):
                GameProgressBarUI.Hide()

            if (not IsPlayerPaused(Player)?):
                Player.SetVulnerability(false)
                Player.SetVulnerability(true, ?BufferTime := 6.0)

        for (Player : Self.GetActivePlayers()):
            GameplayCameraFixedPointBossFight.AddTo(Player)
            Player.DeactivateAbility()
            Player.PlayMusic(music_id.BossFight)
            Player.PutInStasis(?IsAllowFalling := false)

        for (NPCBehavior : GetAllNPCs()):
            NPCBehavior.InterruptState(NPCState.npc_state_type.Pause)

        branch:
            Sleep(3.0)
            for (Index -> Player : Self.GetActivePlayers()):
                GameplayCameraFixedPointBossFight.RemoveFrom(Player)
                Player.ReactivateAbility()
                Player.ReleaseFromStasis()
                Player.ToggleHUD(true)

            Sleep(1.0)
            for (NPCBehavior : GetAllNPCs()):
                NPCBehavior.InterruptState(NPCState.npc_state_type.Chase)
          
        if (Agent := MaybeAgent?, NPCBehaviorBase := npc_behavior_base[Agent.GetNPCBehavior[]]):
            spawn{NPCBehaviorBase.Start()}

        if:
            Agent := MaybeAgent?
            Character := Agent.GetFortCharacter[]
        then:
            race:
                StopTimerEvent.Await()
                loop:
                    MapIndicator.MoveTo(Character.GetTransform(), ThreeTicks)
                block:
                    Character.EliminatedEvent().Await()

                    if (MapIndicator.TeleportTo[MapIndicatorDisabledPosition, IdentityRotation()]) {}                    
                    for (Player : GetAllPlayers()):
                        Player.PlayAudio(audio_id.BossEliminated)

                        if (GameSettings.IsEndless?):
                            ShowMessage(message_id.StageIndicator, "{CurrentStage}/∞", ?DisplayTime := 0.0)
                            Player.PlayMusic(GetArenaMusic())

                            if (ActiveArena := GetActiveArena[]):
                                CurrentGoldMultiplier := CalculateEndlessModeGoldMultiplier(ActiveArena, CurrentStage)
                                PreviousGoldMultiplier := CalculateEndlessModeGoldMultiplier(ActiveArena, CurrentStage - EndlessModeBossStageInterval)
                                BossEliminatedMessage := (CurrentStage > EndlessModeBossStageInterval) and "+{FormatAsPercent(PreviousGoldMultiplier)} ▶ +{FormatAsPercent(CurrentGoldMultiplier)}" or "+{FormatAsPercent(CurrentGoldMultiplier)}"
                                Player.ShowMessage(message_id.EndlessModeBossEliminated, BossEliminatedMessage)

                            if (GameProgressBarUI := PlayerGameProgressBarUI[Player]):
                                GameProgressBarUI.UpdateEventState(GameEventBossIndex, game_event_state.Success)
                                GameProgressBarUI.Show()
                        else:
                            Player.SetVulnerability(false)

                    CinematicSequenceGameEnd.Play()

                    CinematicSequenceGameEnd.StoppedEvent.Await()

                    if (not GameSettings.IsEndless?):
                        DespawnNPC()
                        ClearItemDrops()
                        Sleep(3.0)
                        Stop(true)

    Stop(IsVictory:logic):void =
        set PlayerPauseQueue = map{}
        
        if (GameResult := MaybeGameResult?):
            spawn{GameResult.ShowResults(ActivePlayers, IsVictory)}

        StopTimerEvent.Signal()
        GameEndedEvent.Signal()

        if (MapIndicator.TeleportTo[MapIndicatorDisabledPosition, IdentityRotation()]) {}

        HideMessage(message_id.PlayerCounterShip)
        HideMessage(message_id.TimerShip)
        HideMessage(message_id.MinibossEliminatedCounter)
        HideMessage(message_id.MinibossEliminatedCounterIntro)

        for (Index -> Player : Self.GetActivePlayers(), Character := Player.GetFortCharacter[]):
            Player.CloseActivePanel()
            Player.HideMessage(message_id.Message)
            Player.HideMessage(message_id.ExperienceBar)
            Player.HideMessage(message_id.ExperienceBarLevel)
            Player.HideMessage(message_id.Timer)
            Player.HideMessage(message_id.StageIndicator)
            Player.HideMessage(message_id.ReopenPauseMenu)

            if (GameProgressBarUI := PlayerGameProgressBarUI[Player]):
                GameProgressBarUI.Hide()
            
            if (IsVictory?):
                if (ActiveArenaIndex := GetActiveArenaIndex()?, WeaponId := Player.GetEquippedWeapon()?):
                    ArenaLevel := CalculateArenaLevel(GameSettings.Arena.Level, GameSettings.Difficulty)

                    if (ArenaLevel > Player.GetArenaProgress()):
                        Player.AddArenaProgress()

                    if:
                        CurrentPlayerData := Player.GetData[]
                        WeaponProgress := CurrentPlayerData.WeaponProgress[WeaponId.ToInt()] or 0
                        ArenaLevel > WeaponProgress

                        var UpdatedWeaponProgress:[int]int = CurrentPlayerData.WeaponProgress
                        set UpdatedWeaponProgress[WeaponId.ToInt()] = Min(ArenaLevel, Player.GetArenaProgress())

                        set PlayerDataMap[Player] = player_data:
                            MakePlayerData<constructor>(CurrentPlayerData)
                            WeaponProgress := UpdatedWeaponProgress

            if (GameSettings.IsEndless?, ActiveArenaIndex := GetActiveArenaIndex()?):
                Player.SetEndlessModeProgress(ActiveArenaIndex, CurrentStage)

            for (NPCId -> NPCEliminated : NPCsEliminated, NPCInfo := GetNPCInfo[NPCId]):
                Player.AddAchievement(NPCInfo.AchievementId, NPCEliminated)

        Resume(?IsInstant := true)
        set PlayerGameCurrencies = map{}
        set EnemiesEliminated = 0
        set NPCsEliminated = map{}
        set ObjectivesCompleted = 0
        set PlayerEnemiesEliminated = map{}
        set ActivePlayers = array{}
        set IsStart = false
        
        if (IsVictory?):
            VictoryEvent.Signal()

    (game_manager:)AddGameCurrency(Id:game_currency_id, Amount:float):void =
        for (ActivePlayer : Self.GetActivePlayers()):
            Self.AddGameCurrency(ActivePlayer, Id, Amount, ?IsSignal := false)

        GameCurrencyChangedEvent.Signal(Id)

    (game_manager:)AddGameCurrency(Player:player, Id:game_currency_id, Amount:float, ?IsSignal:logic = true):void =
        if:
            not PlayerGameCurrencies[Player][Id]
            set PlayerGameCurrencies[Player][Id] = game_currency{Id := Id}

        if:
            CurrentAmount := PlayerGameCurrencies[Player][Id]
            UpdatedAmount := CurrentAmount.Amount + Amount
            set PlayerGameCurrencies[Player][Id] = game_currency:
                MakeGameCurrency<constructor>(CurrentAmount)
                Amount := UpdatedAmount
                TotalAmount := CurrentAmount.TotalAmount + Amount
        then:
            if (Id = game_currency_id.Wood):
                if (Player.HasBoon(boon_id.WoodenArrow)?):
                    BoonOnGrantWoodenArrow(Player)

                if (Player.HasBoon(boon_id.CriticalOverdrive)?):
                    BoonOnGrantCriticalOverdrive(Player)

                if (Player.HasBoon(boon_id.WoodenClub)?):
                    BoonOnGrantWoodenClub(Player)

        if (IsSignal?):
            GameCurrencyChangedEvent.Signal(Id)

    (game_manager:)RemoveGameCurrency(Player:player, Id:game_currency_id, Amount:float):void =
        if:
            CurrentAmount := PlayerGameCurrencies[Player][Id]
            UpdatedAmount := CurrentAmount.Amount - Amount
            set PlayerGameCurrencies[Player][Id] = game_currency:
                MakeGameCurrency<constructor>(CurrentAmount)
                Amount := UpdatedAmount
        then:
            if (Id = game_currency_id.Wood):
                if (Player.HasBoon(boon_id.WoodenArrow)?):
                    BoonOnGrantWoodenArrow(Player)

                if (Player.HasBoon(boon_id.CriticalOverdrive)?):
                    BoonOnGrantCriticalOverdrive(Player)

                if (Player.HasBoon(boon_id.WoodenClub)?):
                    BoonOnGrantWoodenClub(Player)

        GameCurrencyChangedEvent.Signal(Id)

    (game_manager:)AddEnemiesEliminated(Player:player, Id:npc_id):void =
        set EnemiesEliminated += 1
        if (set PlayerEnemiesEliminated[Player] = PlayerEnemiesEliminated[Player] + 1 or 1) {}
        if (set NPCsEliminated[Id] = (NPCsEliminated[Id] + 1.0) or 1.0) {}

        UpdateMessage(message_id.CurrencyElimination, "{FormatAsCurrency(EnemiesEliminated * 1.0)}")
        
        EliminationEvent.Signal()

        if (PlayerEliminationEvent := PlayerEliminationEvents[Player]):
            PlayerEliminationEvent.Signal()

    (game_manager:)AddObjectivesComplete():void =
        set ObjectivesCompleted += 1

    (game_manager:)AddGameStat(Id:game_stat_id, Amount:float):void =
        if (not GameStats[Id], set GameStats[Id] = Amount):
            return

        if (set GameStats[Id] += Amount) {}

    (game_manager:)ResetGameStats():void =
        set GameStats = map{}

    <# ----------------------------------------------------------------------------------------------------
        Event Handlers
    ---------------------------------------------------------------------------------------------------- #>
    OnPlayerAdded(Player:player):void =
        Initialize(Player)
        
    OnPlayerRemoved(Player:player):void =
        set ActivePlayers = ActivePlayers.RemoveAllElements(Player)
        set PlayerGameCurrencies = RemoveKeyFromMap(PlayerGameCurrencies, Player)
        set PlayerEnemiesEliminated = RemoveKeyFromMap(PlayerEnemiesEliminated, Player)
        set PlayerEliminationEvents = RemoveKeyFromMap(PlayerEliminationEvents, Player)
        set PlayerPauseQueue = RemoveKeyFromMap(PlayerPauseQueue, Player)
        set PlayerPauseTime = RemoveKeyFromMap(PlayerPauseTime, Player)
        set PlayerGameProgressBarUI = RemoveKeyFromMap(PlayerGameProgressBarUI, Player)

        if (Self.IsGameInProgress()?):
            if (Self.GetAlivePlayers().Length = 0):
                Stop(false)

            else if (GetTotalAlivePlayersPaused() = Self.GetAlivePlayers().Length):
                Pause()

    AutoRevive(Player:player)<suspends>:void =
        if (PlayerStat := Player.GetStats[], Character := Player.GetFortCharacter[], ReviveVolumeActivator := Player.GetReviveVolumeActivator[]):
            if (PlayerStat.IsAutoReviving?) then return
            set PlayerStat.IsAutoReviving = true
            
            ReviveVolumeActivator.Disable()
            Player.SetVulnerability(false)
            Player.DeactivateAbility()
            Sleep(PlayerReviveDuration)
            Player.RemovePlayerResource(player_resource_id.Revives, 1)

            MaybePanel := option{Player.GetActivePanel[]}
            if (Panel := MaybePanel?):
                Panel.Hide(?IsSignal := false)

            DownButNotOut.Revive(Player)
            if (Character.TeleportTo[Character.GetTransform().Translation + vector3{Z := 500.0}, Character.GetTransform().Rotation]) {}
            Player.AddGameStat(player_game_stat_id.AutoRevived, 1.0)
            if (Player.HasBoon(boon_id.ImmortalHeart)?):
                Player.SetStat(stat_id.MaxHealth, stat_source.Boon, GetBoonValueImmortalHeart(Player))

            Player.UpdateMaxHealth()

            if (Panel := MaybePanel?):
                Panel.Show(?IsSignal := false)
            
            if (ReviveBlastValue := Player.GetPowerUpValue(power_up_id.ReviveBlast), ReviveBlastValue > 0.0):
                ReviveBlastSize := 250.0 * (1.0 + ReviveBlastValue)
                ReviveBlastDamage := GetNPCMaxHealth(NPCBaseMaxHealth) * (0.5 + ReviveBlastValue)
                RevivePosition := Character.GetTransform().Translation.Make2D()
                SpawnVFX(vfx_id.ReviveBlast, RevivePosition, ?Scale := vector3{X := ReviveBlastSize * 0.1, Y := ReviveBlastSize * 0.1, Z := 1.0}, ?Duration := 3.0)
                Sleep(1.0)

                for:
                    OverlapHit : Self.GetSimulationEntity[].FindOverlapHits(FromTransform of transform{Translation := RevivePosition}, collision_sphere{Radius := ReviveBlastSize})
                    TargetComponent := OverlapHit.TargetComponent
                    TargetEntity := TargetComponent.Entity
                    NPCCollisionComponent := TargetEntity.GetComponent[npc_collision_component]
                    NPCHit := NPCCollisionComponent.MaybeNPC?
                    NPCHit.IsDamageable()?
                do:
                    DamageArgs := ability_damage_args:
                        Amount := ReviveBlastDamage
                        IsCriticalHit := false

                    NPCHit.Damage(Player, DamageArgs, ?IsKnockback := false)
            else:
                SpawnVFX(vfx_id.Revive, Character.GetTransform().Translation.Make2D(), ?Duration := 3.0)

            KnockbackNPC(Player, ?KnockbackDistanceSquared := DefaultNPCKnockbackDistanceSquaredRevive, ?Duration := 3.0)

            set PlayerStat.IsAutoReviving = false

    OnAgentDowned(Agent:agent):void =
        if (Player := player[Agent]):
            if (Player.GetPlayerResource(player_resource_id.Revives) > 0):
                spawn{AutoRevive(Player)}
                return

            if (Self.GetAlivePlayers().Length = 0):
                Stop(false)

            else if (Character := Player.GetFortCharacter[]):
                Player.SetVulnerability(false)
                Player.DeactivateAbility()

                if (GetTotalAlivePlayersPaused() = Self.GetAlivePlayers().Length):
                    Pause()
                    for (ActivePlayer : Self.GetActivePlayers(), not IsPlayerPaused(ActivePlayer)? and ActivePlayer.GetFortCharacter[].IsDownButNotOut[]):
                        ActivePlayer.ShowGamePausedUI()

                    if (not MaybeResumeCountdownStartTime?):
                        spawn{ResumeCountdown(pause_type.LevelUp)}

    OnAgentRevived(Agent:agent):void =
        if (Player := player[Agent], Self.IsGameInProgress()?):
            if (Character := Player.GetFortCharacter[]):
                Character.SetHealth(Character.GetMaxHealth() * 0.5)

                if (IsPlayerPaused(Player)?):
                    Player.SetVulnerability(false)
                else:
                    Player.SetVulnerability(false)
                    Player.SetVulnerability(true, ?BufferTime := 3.0)
                    Player.ReactivateAbility()

        spawn{ReregisterInput(Agent)}

    ReregisterInput(Agent:agent)<suspends>:void =
        Agent.UnregisterInput(input_id.Dash)
        Sleep(TwoTicks)
        Agent.RegisterInput(input_id.Dash)

    OnAgentEntersStageVolume(Agent:agent):void =
        if (not IsShipActive?) then return
        if (Player := player[Agent]):
            ShowMessage(message_id.PlayerCounterShip, "{VolumeStage.GetPlayersInVolume().Length}/{Self.GetActivePlayers().Length}", ?DisplayTime := 0.0)
            Player.PlayAudio(audio_id.Ready)
            Player.SetVulnerability(false)
            Player.DeactivateAbility()

    OnAgentExitsStageVolume(Agent:agent):void =
        if (not IsShipActive?) then return
        if (Player := player[Agent]):
            ShowMessage(message_id.PlayerCounterShip, "{VolumeStage.GetPlayersInVolume().Length}/{Self.GetActivePlayers().Length}", ?DisplayTime := 0.0)
            Player.SetVulnerability(true)
            Player.ReactivateAbility()