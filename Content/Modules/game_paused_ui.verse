using { /Fortnite.com/Characters }
using { /Fortnite.com/UI }
using { /Verse.org/Colors }
using { /UnrealEngine.com/Temporary/SpatialMath }
using { /UnrealEngine.com/Temporary/UI }

game_paused_ui := class(panel_ui):
    Title<override>:string = "Game Paused"
    HasCloseButton<override>:logic = false
    BackgroundOpacity<override>:type {_X:float where 0.000000 <= _X, _X <= 1.000000} = 0.7

    CountdownText:text_block = text_block:
        DefaultTextColor := ThemeColors.Failure
        DefaultShadowColor := ThemeColors.Failure * 0.125
        DefaultShadowOffset := option{vector2{X := 2.0, Y := 2.0}}
        DefaultShadowOpacity := 1.0
    
    var GamePausedEntryUIs:[]game_paused_entry_ui = array{}
    var EventsSubscriptions:[]events_cancelable = array{}

    <# ----------------------------------------------------------------------------------------------------
        Initialization
    ---------------------------------------------------------------------------------------------------- #>
    CreateContent<override>():widget =
        if (GameManager := GetGameManager[]):
            set EventsSubscriptions = array:
                GameManager.PlayerResumedEvent.Subscribe(OnPlayerResumed)
                GameManager.ResumeCountdownEvent.Subscribe(OnResumeCountdown)

        stack_box:
            Orientation := orientation.Vertical
            Slots := array:
                stack_box_slot:
                    HorizontalAlignment := horizontal_alignment.Fill
                    VerticalAlignment := vertical_alignment.Fill
                    Padding := margin{Left := 0.0, Top := 0.0, Right := 0.0, Bottom := 20.0}
                    Widget := overlay:
                        Slots := array:
                            overlay_slot:
                                HorizontalAlignment := horizontal_alignment.Fill
                                VerticalAlignment := vertical_alignment.Fill
                                Widget := MakeBox()
                            overlay_slot:
                                Padding := margin{Left := 40.0, Top := 20.0, Right := 40.0, Bottom := 20.0}
                                Widget := stack_box:
                                    Orientation := orientation.Horizontal
                                    Slots := array:
                                        stack_box_slot:
                                            Widget := text_block:
                                                DefaultText := MakeMessage("Waiting for players ... ")
                                                DefaultTextColor := ThemeColors.PrimaryText
                                                DefaultTextOpacity := 0.5
                                        stack_box_slot:
                                            Widget := CountdownText
                stack_box_slot:
                    HorizontalAlignment := horizontal_alignment.Fill
                    VerticalAlignment := vertical_alignment.Fill
                    Widget := overlay:
                        Slots := array:
                            overlay_slot:
                                HorizontalAlignment := horizontal_alignment.Fill
                                VerticalAlignment := vertical_alignment.Fill
                                Widget := MakeBox()
                            overlay_slot:
                                HorizontalAlignment := horizontal_alignment.Fill
                                VerticalAlignment := vertical_alignment.Fill
                                Padding := margin{Left := 40.0, Top := 40.0, Right := 40.0, Bottom := 40.0}
                                Widget := stack_box:
                                    Orientation := orientation.Vertical
                                    Slots := for (Index := 1..MaxPlayers):
                                        GamePausedEntryUI := game_paused_entry_ui{Player := Player}
                                        GamePausedEntryUI.Initialize()
                                        set GamePausedEntryUIs += array{GamePausedEntryUI}

                                        stack_box_slot:
                                            HorizontalAlignment := horizontal_alignment.Fill
                                            Padding := margin{Left := 0.0, Top := (Index = 1) and 0.0 or 40.0, Right := 0.0, Bottom := 0.0}
                                            Widget := GamePausedEntryUI.GetUI()

    Uninitialize<override>():void =
        for (EventsSubscription : EventsSubscriptions):
            EventsSubscription.Cancel()

        set EventsSubscriptions = array{}

    <# ----------------------------------------------------------------------------------------------------
        Update Functions
    ---------------------------------------------------------------------------------------------------- #>
    Update():void =
        for (Index -> GamePausedEntryUI : GamePausedEntryUIs):
            if (ActivePlayer := GetActivePlayers()[Index]):
                GamePausedEntryUI.Update(ActivePlayer)
                GamePausedEntryUI.Show()
            else:
                GamePausedEntryUI.Hide()

    <# ----------------------------------------------------------------------------------------------------
        Event Handlers
    ---------------------------------------------------------------------------------------------------- #>
    OnShow<override>():void =
        CountdownText.SetText(MakeMessage("{GetGameManager[].ResumeCountdownTimeRemaining or GameResumeCountdownTimeDefault}s"))
        Update()

    OnPlayerResumed(IsSuccess:logic):void =
        if (IsSuccess?):
            Hide()
        else:
            Update()

    OnResumeCountdown(TimeRemaining:int):void =
        CountdownText.SetText(MakeMessage("{TimeRemaining}s"))