using { /Fortnite.com/UI }
using { /Verse.org/Colors }
using { /UnrealEngine.com/Temporary/SpatialMath }
using { /UnrealEngine.com/Temporary/UI }

using { /Verse.org/Assets }

<# ----------------------------------------------------------------------------------------------------
    Constants
---------------------------------------------------------------------------------------------------- #>
GameEventSupplyDropIndex:int = 1
GameEventMissionIndex:int = 3
GameEventBossIndex:int = 5
MaxGameEvents:int = 6

GameEventSequence:[]game_events = array:
    game_events.Miniboss
    game_events.SupplyDrop
    game_events.Miniboss
    game_events.Mission
    game_events.Miniboss
    game_events.ItemShop

GameEventSequenceFinalStage:[]game_events = array:
    game_events.Miniboss
    game_events.SupplyDrop
    game_events.Miniboss
    game_events.Mission
    game_events.Miniboss
    game_events.Boss

<# ----------------------------------------------------------------------------------------------------
    Data Definitions
---------------------------------------------------------------------------------------------------- #>
game_events := enum:
    Miniboss
    Mission
    ItemShop
    Boss
    SupplyDrop

game_event_state := enum:
    Inactive
    Active
    Success
    Failed

<# ----------------------------------------------------------------------------------------------------
    game_progress_bar_ui
---------------------------------------------------------------------------------------------------- #>
game_progress_bar_ui := class(ui_component):
    Width:float = 600.0
    Height:float = 8.0
    IconSize:float = 32.0
    IconPadding:float = 8.0
    BorderThickness:float = 1.0

    ProgressBar:color_block = color_block:
        DefaultOpacity := 0.0

    BossIcon:material_block = material_block:
        DefaultImage := Materials.UI.M_BossIcon_MPC{}
        DefaultDesiredSize := vector2{X := 48.0, Y := 48.0}
        DefaultTint := ThemeColors.GameProgressBarInactive
        
    var EventEntryUIs:[]tuple(texture_block, texture_block) = array{}

    <# ----------------------------------------------------------------------------------------------------
        Initialization
    ---------------------------------------------------------------------------------------------------- #>
    CreateUI<override>():widget =
        Offset := (IconSize + IconPadding * 2.0) / 2.0

        overlay:
            Slots := array:
                overlay_slot:
                    HorizontalAlignment := horizontal_alignment.Fill
                    VerticalAlignment := vertical_alignment.Center
                    Widget := overlay:
                        Slots := array:
                            overlay_slot:
                                HorizontalAlignment := horizontal_alignment.Fill
                                VerticalAlignment := vertical_alignment.Fill
                                Widget := overlay:
                                    Slots := array:
                                        overlay_slot:
                                            HorizontalAlignment := horizontal_alignment.Fill
                                            VerticalAlignment := vertical_alignment.Fill
                                            Widget := color_block:
                                                DefaultOpacity := 0.0
                                                DefaultDesiredSize := vector2{X := Width, Y := Height}
                                        overlay_slot:
                                            HorizontalAlignment := horizontal_alignment.Fill
                                            VerticalAlignment := vertical_alignment.Fill
                                            Widget := UserInterface.VerseWidget.WBP_ProgressBarBackground{}
                            overlay_slot:
                                HorizontalAlignment := horizontal_alignment.Left
                                VerticalAlignment := vertical_alignment.Fill
                                Padding := margin{Left := BorderThickness, Top := BorderThickness, Right := BorderThickness, Bottom := BorderThickness}
                                Widget := overlay:
                                    Slots := array:
                                        overlay_slot:
                                            HorizontalAlignment := horizontal_alignment.Fill
                                            VerticalAlignment := vertical_alignment.Fill
                                            Widget := ProgressBar
                                        overlay_slot:
                                            HorizontalAlignment := horizontal_alignment.Fill
                                            VerticalAlignment := vertical_alignment.Fill
                                            Widget := UserInterface.VerseWidget.WBP_ProgressBar{}
                overlay_slot:
                    HorizontalAlignment := horizontal_alignment.Fill
                    VerticalAlignment := vertical_alignment.Fill
                    Padding := margin{Left := 0.0, Top := 0.0, Right := -Offset, Bottom := 0.0}
                    Widget := stack_box:
                        Orientation := orientation.Horizontal
                        Slots := 
                            EventIconSize := vector2{X := IconSize, Y := IconSize}
                            EventProgressIconSize := EventIconSize / 2.0

                            for (Index := 1..MaxGameEvents):
                                EventIcon := texture_block:
                                    DefaultImage := UserInterface.Icons.MiniBoss_64
                                    DefaultDesiredSize := EventIconSize
                                
                                EventProgressIcon := texture_block:
                                    DefaultImage := UserInterface.Icons.Checkmark_64
                                    DefaultDesiredSize := EventProgressIconSize

                                EventOverlay := overlay:
                                    Slots := array:
                                        overlay_slot:
                                            HorizontalAlignment := horizontal_alignment.Center
                                            VerticalAlignment := vertical_alignment.Center
                                            Padding := margin{Left := IconPadding, Top := IconPadding, Right := IconPadding, Bottom := IconPadding}
                                            Widget := EventIcon
                                        overlay_slot:
                                            HorizontalAlignment := horizontal_alignment.Right
                                            VerticalAlignment := vertical_alignment.Bottom
                                            Widget := EventProgressIcon  

                                EventEntryUI := (EventIcon, EventProgressIcon)

                                if (Index = MaxGameEvents):
                                    EventOverlay.AddWidget of overlay_slot:
                                        HorizontalAlignment := horizontal_alignment.Center
                                        VerticalAlignment := vertical_alignment.Center
                                        Padding := margin{Left := IconPadding, Top := IconPadding, Right := IconPadding, Bottom := IconPadding}
                                        Widget := BossIcon

                                set EventEntryUIs += array{EventEntryUI}
                                
                                stack_box_slot:
                                    HorizontalAlignment := horizontal_alignment.Right
                                    VerticalAlignment := vertical_alignment.Center
                                    Distribution := option{1.0}
                                    Padding := margin{Left := 0.0, Top := 0.0, Right := (Index = MaxGameEvents) and 0.0 or -Offset, Bottom := 0.0}
                                    Widget := EventOverlay                                      

    <# ----------------------------------------------------------------------------------------------------
        Get Functions
    ---------------------------------------------------------------------------------------------------- #>
    

    <# ----------------------------------------------------------------------------------------------------
        Update Functions
    ---------------------------------------------------------------------------------------------------- #>
    InitializeEvents(Events:[]game_events):void =
        for (Index -> EventEntryUI : EventEntryUIs, Event := Events[Index]):
            EventEntryUI(0).SetTint(ThemeColors.GameProgressBarInactive)
            EventEntryUI(1).SetVisibility(widget_visibility.Hidden)

            case (Event):
                game_events.Miniboss =>
                    EventEntryUI(0).SetImage(UserInterface.Icons.MiniBoss_64)
                game_events.Mission =>
                    if (NextObjective := GetNextObjective[]) {EventEntryUI(0).SetImage(NextObjective.Id.ToIcon())}
                game_events.ItemShop =>
                    EventEntryUI(0).SetImage(UserInterface.Icons.ItemShop_64)
                    EventEntryUI(0).SetVisibility(widget_visibility.Visible)
                    BossIcon.SetVisibility(widget_visibility.Hidden)
                game_events.Boss =>
                    EventEntryUI(0).SetImage(UserInterface.Icons.MiniBoss_64)
                    EventEntryUI(0).SetVisibility(widget_visibility.Hidden)
                    BossIcon.SetVisibility(widget_visibility.Visible)
                game_events.SupplyDrop =>
                    EventEntryUI(0).SetImage(UserInterface.Icons.SupplyDrop_64)

    UpdateProgress(Progress:float):void =
        ProgressBar.SetDesiredSize(vector2{X := Progress * Width, Y := 0.0})

    UpdateEventState(EventIndex:int, EventState:game_event_state):void =
        if (EventEntryUI := EventEntryUIs[EventIndex]):
            case (EventState):
                game_event_state.Inactive =>
                    EventEntryUI(0).SetTint(ThemeColors.GameProgressBarInactive)
                    EventEntryUI(1).SetVisibility(widget_visibility.Hidden)

                game_event_state.Active =>
                    EventEntryUI(0).SetTint(NamedColors.White)
                    EventEntryUI(1).SetVisibility(widget_visibility.Hidden)

                game_event_state.Success =>
                    EventEntryUI(0).SetTint(NamedColors.White)
                    EventEntryUI(1).SetVisibility(widget_visibility.Visible)
                    EventEntryUI(1).SetImage(UserInterface.Icons.GameProgressBar_Checkmark_64)

                game_event_state.Failed =>
                    EventEntryUI(0).SetTint(NamedColors.White)
                    EventEntryUI(1).SetVisibility(widget_visibility.Visible)
                    EventEntryUI(1).SetImage(UserInterface.Icons.GameProgressBar_Cross_64)

    StageEnded():void =
        for (Index -> EventEntryUI : EventEntryUIs):
            WidgetVisibility := EventEntryUI(1).GetVisibility()
            if (WidgetVisibility <> widget_visibility.Visible):
                EventEntryUI(1).SetVisibility(widget_visibility.Visible)
                if (Index = EventEntryUIs.Length - 1):
                    EventEntryUI(1).SetImage(UserInterface.Icons.GameProgressBar_Checkmark_64)
                else:
                    EventEntryUI(1).SetImage(UserInterface.Icons.GameProgressBar_Cross_64)