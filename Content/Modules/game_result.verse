using { /Fortnite.com/Characters }
using { /Fortnite.com/Devices }
using { /Verse.org/Simulation }
using { /UnrealEngine.com/Temporary/UI }

<# ----------------------------------------------------------------------------------------------------
    Data Definitions
---------------------------------------------------------------------------------------------------- #>
game_result_info := struct:
    IsVictory:logic = false
    IsRankUp:logic = false
    Arena:arena = arena{}
    GameSettings:game_settings = game_settings{}
    EndTime:float = 0.0
    GameDuration:float = 0.0
    TimeSurvived:float = 0.0
    EnemiesEliminated:int = 0
    LevelReached:int = 1
    WoodCollected:float = 0.0
    GoldFromTimeSurvived:float = 0.0
    GoldFromWoodCollected:float = 0.0
    GoldFromObjectivesCompleted:float = 0.0
    GoldFromStagesCompleted:float = 0.0
    GoldFromGoldenTree:float = 0.0
    GoldFromVictory:float = 0.0
    GoldEarned:float = 0.0
    GoldMultiplier:float = 0.0
    GoldFromMultiplier:float = 0.0
    TreasureChestEarned:float = 0.0
    ObjectivesCompleted:int = 0
    StagesCompleted:int = 0
    GoldenTree:int = 0
    ActivePlayers:[]player = array{}
    PlayerActiveAbilities:[player][weapon_id]ability = map{}
    PlayerEnemiesEliminated:[player]int = map{}
    PlayerStats:[player]player_stat = map{}
    TeamExperienceMultiplier:float = 0.0

game_result_title_id := enum:
    PlayersRevived
    DamageDealt
    WoodCollected
    TimesDowned

game_result_title := class<concrete>:
    @editable
    Id:game_result_title_id = game_result_title_id.PlayersRevived
    @editable
    Billboard:billboard_device = billboard_device{}
    @editable
    PlayerReference:player_reference_device = player_reference_device{}
    
<# ----------------------------------------------------------------------------------------------------
    Data Definitions
---------------------------------------------------------------------------------------------------- #>
game_result := class:
    GameManager:game_manager

    var PlayerGameResultUI:[player]game_result_ui = map{}
    var PlayerPostGameUI:[player]game_results_post_game_ui = map{}

    <# ----------------------------------------------------------------------------------------------------
        Initialization
    ---------------------------------------------------------------------------------------------------- #>
    Uninitialize(Player:player):void =
        if (GameResultUI := PlayerGameResultUI[Player]):
            GameResultUI.Uninitialize()

        set PlayerGameResultUI = RemoveKeyFromMap(PlayerGameResultUI, Player)
        set PlayerPostGameUI = RemoveKeyFromMap(PlayerPostGameUI, Player)

    <# ----------------------------------------------------------------------------------------------------
        Update Functions
    ---------------------------------------------------------------------------------------------------- #>
    GetResultInfo(Player:player, IsVictory:logic):game_result_info =
        TimeSurvived := GameManager.GetTimeSurvived()
        ActiveArena := GetActiveArena[] or arena{}
        TotalWoodCollected := GameManager.PlayerGameCurrencies[Player][game_currency_id.Wood].TotalAmount or 0.0
        
        var PlayerStats:[player]player_stat = map{}
        for (StatPlayer -> PlayerStat : GetPlayerStats()):
            for (BoonId -> Boon : PlayerStat.GetBoons(), BoonInfo := GetBoonInfo[BoonId]):
                if:
                    GetValue := BoonInfo.MaybeGetValue?
                    set PlayerStat.BoonValues[BoonId] = GetValue(Player)

                else if:
                    set PlayerStat.BoonValues[BoonId] = 0.0
                
            if (set PlayerStats[StatPlayer] = MakePlayerStats(PlayerStat)) {}
            
        ArenaGoldMultiplier := GetArenaGoldMultiplier()
        GoldFromTimeSurvived := (TimeSurvived * 3) * (1.0 + ArenaGoldMultiplier)
        GoldFromWoodCollected := TotalWoodCollected * 0.025 * (1.0 + ArenaGoldMultiplier)
        GoldFromObjectivesCompleted := GameManager.ObjectivesCompleted * ObjectiveBaseRewardAmount * (1.0 + ArenaGoldMultiplier)
        GoldFromStagesCompleted := GameManager.CurrentStage * 100.0 * (1.0 + ArenaGoldMultiplier)
        GoldFromGoldenTree := 5000.0 * GameManager.GetGameStat(game_stat_id.GoldenTree) * (1.0 + ArenaGoldMultiplier)
        
        GoldEarned := GoldFromTimeSurvived + GoldFromWoodCollected + GoldFromObjectivesCompleted + GoldFromStagesCompleted + GoldFromGoldenTree
        GoldFromVictory := if (GameManager.GameSettings.IsEndless?):
            GoldEarned * CalculateEndlessModeGoldMultiplier(ActiveArena, GameManager.CurrentStage)
        else:
            GoldEarned * (IsVictory? and VictoryBonusGoldMultplier or 0.0)
            
        GoldMultiplier := Player.GetStat(stat_id.GoldMultiplier) + Player.WeaponProgressGoldMultiplier()
        GoldFromMultiplier := (GoldEarned + GoldFromVictory) * GoldMultiplier
        TotalGoldEarned := (Floor[GoldEarned + GoldFromVictory + GoldFromMultiplier] or 0) * 1.0

        ArenaLevel := CalculateArenaLevel(GameManager.GameSettings.Arena.Level, GameManager.GameSettings.Difficulty)
        ArenaProgress := Player.GetArenaProgress()
        IsRankUp := logic{IsVictory?, ArenaLevel > ArenaProgress, Mod[ArenaProgress + 1, GetArenaCount()] = 0}

        game_result_info:
            IsVictory := IsVictory
            IsRankUp := IsRankUp
            Arena := ActiveArena
            GameSettings := GameManager.GameSettings
            EndTime := GetSimulationElapsedTime()
            GameDuration := GetSecondsSinceEpoch() - GameManager.GameStartTime
            TimeSurvived := TimeSurvived
            EnemiesEliminated := GameManager.EnemiesEliminated
            LevelReached := GetLevel()
            WoodCollected := TotalWoodCollected
            GoldFromTimeSurvived := GoldFromTimeSurvived
            GoldFromWoodCollected := GoldFromWoodCollected
            GoldFromObjectivesCompleted := GoldFromObjectivesCompleted
            GoldFromStagesCompleted := GoldFromStagesCompleted
            GoldFromGoldenTree := GoldFromGoldenTree
            GoldFromVictory := GoldFromVictory
            GoldEarned := TotalGoldEarned
            GoldMultiplier := GoldMultiplier
            GoldFromMultiplier := GoldFromMultiplier
            TreasureChestEarned := GameManager.CalculateTreasureChestEarned(Player, IsVictory)
            ObjectivesCompleted := GameManager.ObjectivesCompleted
            StagesCompleted := IsVictory? and GameManager.CurrentStage or GameManager.CurrentStage - 1
            GoldenTree := Floor[GameManager.GetGameStat(game_stat_id.GoldenTree)] or 0
            ActivePlayers := GameManager.GetActivePlayers()
            PlayerStats := PlayerStats
            PlayerActiveAbilities := GetPlayerActiveAbilities()
            PlayerEnemiesEliminated := GameManager.PlayerEnemiesEliminated
            TeamExperienceMultiplier := GetTeamExperienceMultiplier()

    <# ----------------------------------------------------------------------------------------------------
        Update Functions
    ---------------------------------------------------------------------------------------------------- #>
    ShowGameResultsUI(Player:player, ?MaybeResultInfo:?game_result_info := false):void =
        ResultUI := if (GameResultUI := PlayerGameResultUI[Player]):
            GameResultUI
        else:
            GameResultUI := game_result_ui{Player := Player}
            if (set PlayerGameResultUI[Player] = GameResultUI) {}
            GameResultUI

        ResultUI.Show()

        if (ResultInfo := MaybeResultInfo?):
            ResultUI.Update(ResultInfo)

    ShowPostGameUI(Player:player, ResultInfo:game_result_info):void =
        ResultUI := if (PostGameUI := PlayerPostGameUI[Player]):
            PostGameUI
        else:
            PostGameUI := game_results_post_game_ui{Player := Player}
            PostGameUI.Initialize()
            if (set PlayerPostGameUI[Player] = PostGameUI) {}
            PostGameUI

        if (PlayerUI := GetPlayerUI[Player]):
            PlayerUI.AddWidget(ResultUI.GetUI(), Slot := player_ui_slot{InputMode := ui_input_mode.None})
            ResultUI.Show(ResultInfo)

    HidePostGameUI(Player:player):void =
        if (PlayerUI := GetPlayerUI[Player], PostGameUI := PlayerPostGameUI[Player]):
            PlayerUI.RemoveWidget(PostGameUI.GetUI())

    ShowResults(Players:[]player, IsVictory:logic)<suspends>:void =
        var PlayerResultInfo:[player]game_result_info = map{}
        for (Player : Players):
            ResultInfo := GetResultInfo(Player, IsVictory)
            if (set PlayerResultInfo[Player] = ResultInfo) {}

        if (not IsVictory?):
            GameManager.CinematicSequenceGameEnd.Play()
            GameManager.CinematicSequenceGameEnd.StoppedEvent.Await()

        DespawnNPC()
        ClearItemDrops()
        GameManager.BasicStormController.DestroyStorm()
        
        for (Index -> Player : Players, Character := Player.GetFortCharacter[]):
            if (Character.IsDownButNotOut[]):
                GameManager.DownButNotOut.Revive(Player)

            Player.ClearMusic()
            Player.PlayAudio(IsVictory? and audio_id.Victory or audio_id.Defeat)
            Player.ShowMessage(IsVictory? and message_id.Victory or message_id.Defeat)
            Player.DeactivateVFXPowerup(vfx_powerup_id.Invulnerability)
            Player.ActivateVFXPowerup of Player.GetRankVFXPowerUp()
            Player.PutInStasis(?IsAllowFalling := false)
            Player.SetInputTriggers(InputTriggerCollections.PostGame)

        Sleep(0.0)
        GameManager.TeleportGroup(Players, GameManager.GameEndPositionLocator.GetTransform().Translation, GameManager.GameEndPositionLocator.GetTransform().Rotation)

        ResetStats()

        for (Index -> Player : Players):
            Player.ToggleHUD(false)

        Sleep(3.0)
        
        for (Index -> Player : Players, ResultInfo := PlayerResultInfo[Player], PostGameCinematicSequence := GameManager.PostGameCinematicSequences[Index]):
            spawn{ShowPostGame(Player, PostGameCinematicSequence, ResultInfo, IsVictory)}

    ShowPostGame(Player:player, PostGameCinematicSequence:cinematic_sequence_device, ResultInfo:game_result_info, IsVictory:logic)<suspends>:void =
        StartTime:float = GetSimulationElapsedTime()

        ShowPostGameUI(Player, ResultInfo)
        PostGameCinematicSequence.Play(Player)
        CheerAudioPlayer := Player.PlayAudio(IsVictory? and audio_id.CheerPositive or audio_id.CheerNegative)

        race:
            block:
                Sleep(14.0) # [Note] Needs to be synced with CinematicSequencePostGame duration (1 second less).
                HidePostGameUI(Player)

            if (InputPressedEvent := GetInputPressedEvent[input_id.Skip]):
                InputPressedEvent.Await()
                
                ElapsedTime := GetSimulationElapsedTime() - StartTime
                if (ElapsedTime < 12.5):
                    PostGameCinematicSequence.SetPlaybackTime(12.5)
                    CheerAudioPlayer.Stop(Player)
                    Player.PlayAudio(audio_id.Skip)
                    HidePostGameUI(Player)
                    Sleep(1.5)
                else:
                    PostGameCinematicSequence.StoppedEvent.Await()

        Player.PlayMusic(music_id.Default)
        Player.ToggleHUD(true)
        Player.ReleaseFromStasis()
        Player.SetInputTriggers(InputTriggerCollections.Lobby)
        ShowGameResultsUI(Player, ?MaybeResultInfo := option{ResultInfo})
        GameManager.GrantRewards(Player, IsVictory, ?MaybeResultInfo := option{ResultInfo})