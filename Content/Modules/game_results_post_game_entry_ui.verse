using { /Fortnite.com/UI }
using { /Verse.org/Assets }
using { /Verse.org/Colors }
using { /Verse.org/Simulation }
using { /UnrealEngine.com/Temporary }
using { /UnrealEngine.com/Temporary/SpatialMath }
using { /UnrealEngine.com/Temporary/UI }

<# ----------------------------------------------------------------------------------------------------
    game_results_post_game_entry_ui
---------------------------------------------------------------------------------------------------- #>
game_results_post_game_entry_ui := class(ui_component):

    MVPIcon:widget = UserInterface.VerseWidget.WBP_GameMVP_Badge{}

    NameText:text_block = text_block:
        DefaultTextSize := ThemeTextSize.Small
        DefaultTextColor := ThemeColors.PrimaryText

    EliminationText:text_block = text_block:
        DefaultTextSize := ThemeTextSize.VerySmall
        DefaultTextColor := ThemeColors.PrimaryText

    WoodText:text_block = text_block:
        DefaultTextSize := ThemeTextSize.VerySmall
        DefaultTextColor := ThemeColors.PrimaryText

    RevivesText:text_block = text_block:
        DefaultTextSize := ThemeTextSize.VerySmall
        DefaultTextColor := ThemeColors.PrimaryText

    DownedText:text_block = text_block:
        DefaultTextSize := ThemeTextSize.VerySmall
        DefaultTextColor := ThemeColors.PrimaryText
        
    DamageText:text_block = text_block:
        DefaultTextSize := ThemeTextSize.Small
        DefaultTextColor := ThemeColors.PrimaryText
    
    <# ----------------------------------------------------------------------------------------------------
        Initialization
    ---------------------------------------------------------------------------------------------------- #>
    CreateUI<override>():widget =
        overlay:
            Slots := array:
                overlay_slot:
                    HorizontalAlignment := horizontal_alignment.Fill
                    VerticalAlignment := vertical_alignment.Fill
                    Widget := UserInterface.VerseWidget.WBP_RoundedBox_Darker{}
                overlay_slot:
                    HorizontalAlignment := horizontal_alignment.Fill
                    VerticalAlignment := vertical_alignment.Fill
                    Widget := color_block:
                        DefaultOpacity := 0.0
                        DefaultDesiredSize := vector2{X := 300.0, Y := 0.0}
                overlay_slot:
                    HorizontalAlignment := horizontal_alignment.Fill
                    VerticalAlignment := vertical_alignment.Fill
                    Padding := margin{Left := 20.0, Top := 20.0, Right := 20.0, Bottom := 20.0}
                    Widget := CreateInfoUI()

    CreateInfoUI():widget =
        stack_box:
            Orientation := orientation.Vertical
            Slots := array:
                stack_box_slot:
                    VerticalAlignment := vertical_alignment.Top
                    Widget := MVPIcon
                stack_box_slot:
                    VerticalAlignment := vertical_alignment.Fill
                    Widget := NameText
                stack_box_slot:
                    HorizontalAlignment := horizontal_alignment.Fill
                    VerticalAlignment := vertical_alignment.Fill
                    Padding := margin{Left := 0.0, Top := 20.0, Right := 0.0, Bottom := 0.0}
                    Widget := CreateStatsUI()
                stack_box_slot:
                    VerticalAlignment := vertical_alignment.Fill
                    Padding := margin{Left := 0.0, Top := 20.0, Right := 0.0, Bottom := 0.0}
                    Widget := CreateDamageStatUI()

    CreateStatsUI():widget =
        stack_box:
            Orientation := orientation.Horizontal
            Slots := array:
                CreateStatSlot(UserInterface.Icons.Elimination_Transparent_64, EliminationText)
                CreateDividerUI()
                CreateStatSlot(UserInterface.Icons.CurrencyWood_Transparent_64, WoodText)
                CreateDividerUI()
                CreateStatSlot(UserInterface.Icons.Revive_Transparent_64, RevivesText)
                CreateDividerUI()
                CreateStatSlot(UserInterface.Icons.Downed_Transparent_64, DownedText)

    CreateDividerUI():stack_box_slot =
        stack_box_slot:
            Widget := color_block:
                DefaultColor := NamedColors.White
                DefaultOpacity := 0.5
                DefaultDesiredSize := vector2{X := 1.0, Y := 32.0}

    CreateStatSlot(Icon:texture, TextBlock:text_block):stack_box_slot =
        stack_box_slot:
            HorizontalAlignment := horizontal_alignment.Fill
            VerticalAlignment := vertical_alignment.Fill
            Distribution := option{1.0}
            Widget := stack_box:
                Orientation := orientation.Vertical
                Slots := array:
                    stack_box_slot:                               
                        VerticalAlignment := vertical_alignment.Top
                        Widget := texture_block:
                            DefaultImage := Icon
                            DefaultDesiredSize := vector2{X := 24.0, Y := 24.0}
                    stack_box_slot:                               
                        VerticalAlignment := vertical_alignment.Fill
                        Widget := TextBlock

    CreateDamageStatUI():widget =
        stack_box:
            Orientation := orientation.Vertical
            Slots := array:
                stack_box_slot:
                    VerticalAlignment := vertical_alignment.Fill
                    Widget := text_block:
                        DefaultTextSize := ThemeTextSize.VerySmall
                        DefaultText := MakeMessage("DAMAGE")
                        DefaultTextColor := ThemeColors.PrimaryText
                        DefaultTextOpacity := 0.5
                stack_box_slot:
                    VerticalAlignment := vertical_alignment.Fill
                    Widget := DamageText

    <# ----------------------------------------------------------------------------------------------------
        Get Functions
    ---------------------------------------------------------------------------------------------------- #>
  

    <# ----------------------------------------------------------------------------------------------------
        Update Functions
    ---------------------------------------------------------------------------------------------------- #>
    Show(InPlayer:player, GameResultInfo:game_result_info, IsMVP:logic):void =
        MVPIcon.SetVisibility(IsMVP? and widget_visibility.Visible or widget_visibility.Hidden)
        NameText.SetText(MakeMessage(InPlayer))

        EnemiesEliminated := GameResultInfo.PlayerEnemiesEliminated[InPlayer] or 0
        UpdateText(EliminationText, EnemiesEliminated * 1.0)

        if (PlayerStats := GameResultInfo.PlayerStats[InPlayer]):
            UpdateText(WoodText, PlayerStats.GetGameStat(player_game_stat_id.WoodCollected))
            UpdateText(RevivesText, PlayerStats.GetGameStat(player_game_stat_id.PlayersRevived))
            UpdateText(DownedText, PlayerStats.GetGameStat(player_game_stat_id.TimesDowned))
            UpdateText(DamageText, PlayerStats.GetGameStat(player_game_stat_id.DamageDealt))

        Show()

    UpdateText(TextBlock:text_block, Value:float):void =
        TextBlock.SetText(MakeMessage("{FormatAsCurrency(Value)}"))

    <# ----------------------------------------------------------------------------------------------------
        Event Handlers
    ---------------------------------------------------------------------------------------------------- #>
