using { /Fortnite.com/UI }
using { /Verse.org/Colors }
using { /Verse.org/Simulation }
using { /UnrealEngine.com/Temporary }
using { /UnrealEngine.com/Temporary/SpatialMath }
using { /UnrealEngine.com/Temporary/UI }

<# ----------------------------------------------------------------------------------------------------
    game_results_post_game_ui
---------------------------------------------------------------------------------------------------- #>
game_results_post_game_ui := class(ui_component):

    VictoryText:widget = UserInterface.VerseWidget.WBP_VictoryText{}
    DefeatText:widget = UserInterface.VerseWidget.WBP_DefeatText{}
    
    var EntryUIs:[]game_results_post_game_entry_ui = array{}

    <# ----------------------------------------------------------------------------------------------------
        Initialization
    ---------------------------------------------------------------------------------------------------- #>
    CreateUI<override>():widget =
        canvas:
            Slots := array:
                canvas_slot:
                    Anchors := anchors{Minimum := vector2{X := 0.5, Y := 0.0}, Maximum := vector2{X := 0.5, Y := 0.0}}
                    Offsets := margin{Left := 0.0, Top := 100.0, Right := 0.0, Bottom := 0.0}
                    Alignment := vector2{X := 0.5, Y := 0.0}
                    Widget := CreateHeadingUI()
                canvas_slot:
                    Anchors := anchors{Minimum := vector2{X := 0.5, Y := 1.0}, Maximum := vector2{X := 0.5, Y := 1.0}}
                    Offsets := margin{Left := 0.0, Top := -100.0, Right := 0.0, Bottom := 0.0}
                    Alignment := vector2{X := 0.5, Y := 1.0}
                    Widget := CreateEntriesUI()

    CreateHeadingUI():widget =
        overlay:
            Slots := array:
                overlay_slot:
                    HorizontalAlignment := horizontal_alignment.Fill
                    VerticalAlignment := vertical_alignment.Fill
                    Widget := UserInterface.VerseWidget.WBP_RoundedBox_Dark{}
                overlay_slot:
                    Padding := margin{Left := 40.0, Top := 10.0, Right := 40.0, Bottom := 10.0}
                    Widget := VictoryText
                overlay_slot:
                    Padding := margin{Left := 40.0, Top := 10.0, Right := 40.0, Bottom := 10.0}
                    Widget := DefeatText

    CreateEntriesUI():widget = 
        stack_box:
            Orientation := orientation.Horizontal
            Slots := for (Index := 1..MaxPlayers):
                EntryUI := game_results_post_game_entry_ui{Player := Player}
                EntryUI.Initialize()
                set EntryUIs += array{EntryUI}

                stack_box_slot:
                    HorizontalAlignment := horizontal_alignment.Fill
                    VerticalAlignment := vertical_alignment.Fill
                    Padding := margin{Left := (Index = 1) and 0.0 or 120.0, Top := 0.0, Right := 0.0, Bottom := 0.0}
                    Widget := EntryUI.GetUI()

    <# ----------------------------------------------------------------------------------------------------
        Update Functions
    ---------------------------------------------------------------------------------------------------- #>
    Show(GameResultInfo:game_result_info):void =
        if (GameResultInfo.IsVictory?):
            VictoryText.SetVisibility(widget_visibility.Visible)
            DefeatText.SetVisibility(widget_visibility.Collapsed)
        else:
            VictoryText.SetVisibility(widget_visibility.Collapsed)
            DefeatText.SetVisibility(widget_visibility.Visible)
        
        var HighestDamageDealt:float = 0.0
        var HighestDamagePlayer:player = Player

        for (ActivePlayer : GameResultInfo.ActivePlayers, PlayerStats := GameResultInfo.PlayerStats[ActivePlayer]):
            DamageDealt := PlayerStats.GetGameStat(player_game_stat_id.DamageDealt)
            if (DamageDealt > HighestDamageDealt):
                set HighestDamageDealt = DamageDealt
                set HighestDamagePlayer = ActivePlayer

        for (Index -> EntryUI : EntryUIs):
            if (ActivePlayer := GameResultInfo.ActivePlayers[Index]):
                EntryUI.Show(ActivePlayer, GameResultInfo, logic{ActivePlayer = HighestDamagePlayer})
            else:
                EntryUI.Hide()

        Show()
