using { /Fortnite.com/UI }
using { /Verse.org/Colors }
using { /Verse.org/Simulation }
using { /UnrealEngine.com/Temporary/SpatialMath }
using { /UnrealEngine.com/Temporary/UI }

game_result_damage_entry_ui := class(ui_component):
    RankIcon:texture_block = texture_block:
        DefaultImage := UserInterface.Icons.RankIcon_01_64
        DefaultDesiredSize := vector2{X := 64.0, Y := 64.0}
        
    NameText:text_block = text_block:
        DefaultTextColor := ThemeColors.PrimaryText
        DefaultShadowOffset := option{vector2{X := 2.0, Y := 2.0}}
        DefaultShadowOpacity := 0.5  

    DamageText:text_block = text_block:
        DefaultTextColor := ThemeColors.PrimaryText
        DefaultShadowOffset := option{vector2{X := 2.0, Y := 2.0}}
        DefaultShadowOpacity := 0.5

    DamageMinibossText:text_block = text_block:
        DefaultTextColor := ThemeColors.PrimaryText
        DefaultShadowOffset := option{vector2{X := 2.0, Y := 2.0}}
        DefaultShadowOpacity := 0.5

    DamageBossText:text_block = text_block:
        DefaultTextColor := ThemeColors.PrimaryText
        DefaultShadowOffset := option{vector2{X := 2.0, Y := 2.0}}
        DefaultShadowOpacity := 0.5

    <# ----------------------------------------------------------------------------------------------------
        Initialization
    ---------------------------------------------------------------------------------------------------- #>
    CreateUI<override>():widget =
        stack_box:
            Orientation := orientation.Horizontal
            Slots := array:
                stack_box_slot:
                    HorizontalAlignment := horizontal_alignment.Fill
                    VerticalAlignment := vertical_alignment.Fill
                    Distribution := option{1.0}
                    Widget := stack_box:
                        Orientation := orientation.Horizontal
                        Slots := array:
                            stack_box_slot:
                                HorizontalAlignment := horizontal_alignment.Left
                                Widget :=  RankIcon
                            stack_box_slot:
                                HorizontalAlignment := horizontal_alignment.Left
                                Padding := margin{Left := 20.0, Top := 0.0, Right := 0.0, Bottom := 0.0}
                                Widget := NameText
                stack_box_slot:
                    HorizontalAlignment := horizontal_alignment.Center
                    Distribution := option{0.5}
                    Widget := DamageText
                stack_box_slot:
                    HorizontalAlignment := horizontal_alignment.Center
                    Distribution := option{0.5}
                    Widget := DamageMinibossText
                stack_box_slot:
                    HorizontalAlignment := horizontal_alignment.Center
                    Distribution := option{0.5}
                    Widget := DamageBossText

    <# ----------------------------------------------------------------------------------------------------
        Update Functions
    ---------------------------------------------------------------------------------------------------- #>
    Update(InPlayer:player, DamageDealt:float, DamageMinibossDealt:float, DamageBossDealt:float):void =
        RankIcon.SetImage(InPlayer.GetRank().ToRankIcon())
        NameText.SetText(MakeMessage(InPlayer))
        DamageText.SetText(MakeMessage("{FormatAsCurrency(DamageDealt)}"))
        DamageMinibossText.SetText(MakeMessage("{FormatAsCurrency(DamageMinibossDealt)}"))
        DamageBossText.SetText(MakeMessage("{FormatAsCurrency(DamageBossDealt)}"))