using { /Fortnite.com/UI }
using { /Verse.org/Colors }
using { /Verse.org/Simulation }
using { /UnrealEngine.com/Temporary/SpatialMath }
using { /UnrealEngine.com/Temporary/UI }

<# ----------------------------------------------------------------------------------------------------
    game_result_team_entry_ui
---------------------------------------------------------------------------------------------------- #>
game_result_team_entry_ui := class(ui_component):
    RankIcon:texture_block = texture_block:
        DefaultImage := UserInterface.Icons.RankIcon_01_64
        DefaultDesiredSize := vector2{X := 64.0, Y := 64.0}
        
    NameText:text_block = text_block:
        DefaultTextColor := ThemeColors.PrimaryText
        DefaultShadowOffset := option{vector2{X := 2.0, Y := 2.0}}
        DefaultShadowOpacity := 0.5
        DefaultOverflowPolicy := text_overflow_policy.Clip

    EliminationText:text_block = text_block:
        DefaultTextColor := ThemeColors.PrimaryText
        DefaultShadowOffset := option{vector2{X := 2.0, Y := 2.0}}
        DefaultShadowOpacity := 0.5

    DamageText:text_block = text_block:
        DefaultTextColor := ThemeColors.PrimaryText
        DefaultShadowOffset := option{vector2{X := 2.0, Y := 2.0}}
        DefaultShadowOpacity := 0.5

    WoodCollectedText:text_block = text_block:
        DefaultTextColor := ThemeColors.PrimaryText
        DefaultShadowOffset := option{vector2{X := 2.0, Y := 2.0}}
        DefaultShadowOpacity := 0.5

    PlayersRevivedText:text_block = text_block:
        DefaultTextColor := ThemeColors.PrimaryText
        DefaultShadowOffset := option{vector2{X := 2.0, Y := 2.0}}
        DefaultShadowOpacity := 0.5

    TimesDownedText:text_block = text_block:
        DefaultTextColor := ThemeColors.PrimaryText
        DefaultShadowOffset := option{vector2{X := 2.0, Y := 2.0}}
        DefaultShadowOpacity := 0.5

    var WeaponWidgets:[]tuple(texture_block, text_block) = array{}

    <# ----------------------------------------------------------------------------------------------------
        Initialization
    ---------------------------------------------------------------------------------------------------- #>
    CreateUI<override>():widget =
        stack_box:
            Orientation := orientation.Vertical
            Slots := array:
                stack_box_slot:
                    HorizontalAlignment := horizontal_alignment.Fill
                    VerticalAlignment := vertical_alignment.Fill
                    Distribution := option{0.6}
                    Widget := stack_box:
                        Orientation := orientation.Vertical
                        Slots := array:
                            stack_box_slot:
                                VerticalAlignment := vertical_alignment.Top
                                Padding := margin{Left := 0.0, Top := 0.0, Right := 0.0, Bottom := 10.0}
                                Widget :=  RankIcon
                            stack_box_slot:
                                Widget := NameText
                stack_box_slot:
                    HorizontalAlignment := horizontal_alignment.Center
                    VerticalAlignment := vertical_alignment.Fill
                    Distribution := option{1.0}
                    Widget := stack_box:
                        Orientation := orientation.Vertical
                        Slots := for (Index := 1..4):
                            WeaponIcon :=  texture_block:
                                DefaultImage := UserInterface.Icons.Lock_64
                                DefaultDesiredSize := vector2{X := 40.0, Y := 40.0}
                                
                            WeaponLevelText := text_block:
                                DefaultTextColor := ThemeColors.Highlight * 2.0
                                DefaultShadowOffset := option{vector2{X := 2.0, Y := 2.0}}
                                DefaultShadowOpacity := 0.5

                            set WeaponWidgets += array{(WeaponIcon, WeaponLevelText)}

                            stack_box_slot:
                                HorizontalAlignment := horizontal_alignment.Fill
                                VerticalAlignment := vertical_alignment.Fill
                                Padding := margin{Left := 0.0, Top := (Index = 1) and 0.0 or 5.0, Right := 0.0, Bottom := 0.0}
                                Widget := stack_box:
                                    Orientation := orientation.Horizontal
                                    Slots := array:
                                        stack_box_slot:
                                            HorizontalAlignment := horizontal_alignment.Left
                                            Padding := margin{Left := 0.0, Top := 0.0, Right := 10.0, Bottom := 0.0}
                                            Widget := WeaponIcon
                                        stack_box_slot:
                                            HorizontalAlignment := horizontal_alignment.Fill
                                            Widget := WeaponLevelText

                stack_box_slot:
                    HorizontalAlignment := horizontal_alignment.Center
                    Widget := stack_box:
                        Orientation := orientation.Horizontal
                        Slots := array:
                            stack_box_slot:
                                HorizontalAlignment := horizontal_alignment.Left
                                Padding := margin{Left := 0.0, Top := 0.0, Right := 10.0, Bottom := 0.0}
                                Widget := texture_block:
                                    DefaultImage := UserInterface.Icons.Elimination_Flat_64
                                    DefaultDesiredSize := vector2{X := 40.0, Y := 40.0}
                            stack_box_slot:
                                HorizontalAlignment := horizontal_alignment.Fill
                                Widget := EliminationText
                stack_box_slot:
                    HorizontalAlignment := horizontal_alignment.Center
                    Widget := DamageText
                stack_box_slot:
                    HorizontalAlignment := horizontal_alignment.Center
                    Widget := WoodCollectedText
                stack_box_slot:
                    HorizontalAlignment := horizontal_alignment.Center
                    Widget := PlayersRevivedText
                stack_box_slot:
                    HorizontalAlignment := horizontal_alignment.Center
                    Widget := TimesDownedText

    <# ----------------------------------------------------------------------------------------------------
        Update Functions
    ---------------------------------------------------------------------------------------------------- #>
    Hide<override>():void =
        UI.SetVisibility(widget_visibility.Hidden)

    Update(InPlayer:player, EnemiesEliminated:int, DamageDealt:float, WeaponLevels:[weapon_id]int, WoodCollected:float, PlayersRevived:float, TimesDowned:float):void =
        RankIcon.SetImage(InPlayer.GetRank().ToRankIcon())
        NameText.SetText(MakeMessage(InPlayer))
        EliminationText.SetText(MakeMessage("{FormatAsCurrency(EnemiesEliminated * 1.0)}"))
        DamageText.SetText(MakeMessage("{FormatAsCurrency(DamageDealt)}"))
        WoodCollectedText.SetText(MakeMessage("{FormatAsCurrency(WoodCollected)}"))
        PlayersRevivedText.SetText(MakeMessage("{FormatAsCurrency(PlayersRevived)}"))
        TimesDownedText.SetText(MakeMessage("{FormatAsCurrency(TimesDowned)}"))

        WeaponIds := for (WeaponId -> WeaponLevel : WeaponLevels) {WeaponId}
        for (Index -> WeaponWidget : WeaponWidgets):
            if (WeaponId := WeaponIds[Index], WeaponInfo := GetWeaponInfo[WeaponId], WeaponLevel := WeaponLevels[WeaponId]):
                WeaponWidget(0).SetImage(WeaponInfo.Icon)
                WeaponWidget(0).SetTint(NamedColors.White)
                WeaponWidget(1).SetText(MakeMessage("{WeaponLevel}"))
            else:
                WeaponWidget(0).SetImage(UserInterface.Icons.Lock_64)
                WeaponWidget(0).SetTint(ThemeColors.Inactive)
                WeaponWidget(1).SetText(MakeMessage(""))