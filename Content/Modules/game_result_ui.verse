using { /Fortnite.com/UI }
using { /Verse.org/Colors }
using { /Verse.org/Simulation }
using { /UnrealEngine.com/Temporary }
using { /UnrealEngine.com/Temporary/SpatialMath }
using { /UnrealEngine.com/Temporary/UI }

<# ----------------------------------------------------------------------------------------------------
    Data Definitions
---------------------------------------------------------------------------------------------------- #>
game_result_type := enum:
    game_result_type.Game
    game_result_type.Team
    game_result_type.Damage
    game_result_type.Player

<# ----------------------------------------------------------------------------------------------------
    game_result_ui
---------------------------------------------------------------------------------------------------- #>
game_result_ui := class(panel_ui):
    Title<override>:string = ""
    BackgroundOpacity<override>:type {_X:float where 0.000000 <= _X, _X <= 1.000000} = 0.7
    Anchors<override>:anchors = anchors{Minimum := vector2{X := 0.0, Y := 0.0}, Maximum := vector2{X := 1.0, Y := 1.0}}
    
    CancelAnimationEvent:events(tuple()) = events(tuple()){}

    ArenaText:text_block = text_block:
        DefaultTextColor := ThemeColors.PrimaryText

    var GameUI:widget = canvas{}
    var TeamUI:widget = canvas{}
    var DamageUI:widget = canvas{}
    var PlayerUI:widget = canvas{}
    var GameResultEntryUIs:[]game_result_entry_ui = array{}

    var MaybeAbilityStatInfoUI:?ability_stat_info_ui = false

    var ActiveGameResultType:game_result_type = game_result_type.Game
    var ActiveResultsInfo:game_result_info = game_result_info{}
    var IsClaimed:logic = true
    var IsRankUp:logic = false

    var Subscriptions:[]cancelable = array{}
    var EventsSubscriptions:[]events_cancelable = array{}
    
    <# ----------------------------------------------------------------------------------------------------
        Initialization
    ---------------------------------------------------------------------------------------------------- #>
    Uninitialize<override>():void =
        for (GameResultEntryUI : GameResultEntryUIs):
            GameResultEntryUI.Uninitialize()
            
        for (EventsSubscription : EventsSubscriptions):
            EventsSubscription.Cancel()

        for (Subscription : Subscriptions):
            Subscription.Cancel()

    CreateContent<override>():widget =
        set EventsSubscriptions = array{ClosedEvent.Subscribe(OnClosed)}
        
        AbilityStatInfoUI := ability_stat_info_ui{Player := Player, BackgroundType := background_type.None}
        AbilityStatInfoUI.Initialize()
        set MaybeAbilityStatInfoUI = option{AbilityStatInfoUI}

        StackBox := stack_box:
            Orientation := orientation.Vertical
            Slots := array:
                
                stack_box_slot:
                    HorizontalAlignment := horizontal_alignment.Fill
                    VerticalAlignment := vertical_alignment.Fill
                    Distribution := option{1.0}
                    Widget := stack_box:
                        Orientation := orientation.Horizontal
                        Slots := array:
                            stack_box_slot:
                                HorizontalAlignment := horizontal_alignment.Fill
                                VerticalAlignment := vertical_alignment.Fill
                                Distribution := option{0.25}
                                Widget := stack_box:
                                    Orientation := orientation.Vertical
                                    Slots := array:
                                        stack_box_slot:
                                            HorizontalAlignment := horizontal_alignment.Fill
                                            VerticalAlignment := vertical_alignment.Fill
                                            Padding := margin{Left := 0.0, Top := 0.0, Right := 0.0, Bottom := 20.0}
                                            Widget := overlay:
                                                Slots := array:
                                                    overlay_slot:
                                                        HorizontalAlignment := horizontal_alignment.Fill
                                                        VerticalAlignment := vertical_alignment.Fill
                                                        Widget := UserInterface.VerseWidget.WBP_RoundedBox{}
                                                    overlay_slot:
                                                        Padding := margin{Left := 20.0, Top := 20.0, Right := 20.0, Bottom := 20.0}
                                                        Widget := ArenaText
                                        stack_box_slot:
                                            HorizontalAlignment := horizontal_alignment.Fill
                                            VerticalAlignment := vertical_alignment.Fill
                                            Distribution := option{1.0}
                                            Widget := CreateMenuUI()
                            stack_box_slot:
                                HorizontalAlignment := horizontal_alignment.Fill
                                VerticalAlignment := vertical_alignment.Fill
                                Distribution := option{1.0}
                                Padding := margin{Left := 20.0, Top := 0.0, Right := 0.0, Bottom := 0.0}
                                Widget := overlay:
                                    Slots := array:
                                        overlay_slot:
                                            HorizontalAlignment := horizontal_alignment.Fill
                                            VerticalAlignment := vertical_alignment.Fill
                                            Widget := set GameUI = CreateGameUI()
                                        overlay_slot:
                                            HorizontalAlignment := horizontal_alignment.Fill
                                            VerticalAlignment := vertical_alignment.Fill
                                            Widget := set TeamUI = CreateTeamUI()
                                        overlay_slot:
                                            HorizontalAlignment := horizontal_alignment.Fill
                                            VerticalAlignment := vertical_alignment.Fill
                                            Widget := set DamageUI = CreateDamageUI()
                                        overlay_slot:
                                            HorizontalAlignment := horizontal_alignment.Fill
                                            VerticalAlignment := vertical_alignment.Fill
                                            Widget := set PlayerUI = AbilityStatInfoUI.GetUI()
        
        TeamUI.SetVisibility(widget_visibility.Collapsed)
        PlayerUI.SetVisibility(widget_visibility.Collapsed)
        DamageUI.SetVisibility(widget_visibility.Collapsed)

        return StackBox
               
    <# ----------------------------------------------------------------------------------------------------
        Menu UI
    ---------------------------------------------------------------------------------------------------- #>
    GameIcon:texture_block = texture_block:
        DefaultImage := UserInterface.Icons.AbilityCooldown_64
        DefaultDesiredSize := vector2{X := 40.0, Y := 40.0}

    TeamIcon:texture_block = texture_block:
        DefaultImage := UserInterface.Icons.Team_64
        DefaultDesiredSize := vector2{X := 40.0, Y := 40.0}

    DamageIcon:texture_block = texture_block:
        DefaultImage := UserInterface.Icons.ProjectileDamage_64
        DefaultDesiredSize := vector2{X := 40.0, Y := 40.0}

    CreateMenuUI():widget =
        GameButton := button_quiet{DefaultText := MakeMessage("Game")}
        TeamButton := button_quiet{DefaultText := MakeMessage("Team")}
        DamageButton := button_quiet{DefaultText := MakeMessage("Damage")}

        set Subscriptions = array:
            GameButton.OnClick().Subscribe(OnClickGame)
            TeamButton.OnClick().Subscribe(OnClickTeam)
            DamageButton.OnClick().Subscribe(OnClickDamage)
        
        overlay:
            Slots := array:
                overlay_slot:
                    HorizontalAlignment := horizontal_alignment.Fill
                    VerticalAlignment := vertical_alignment.Fill
                    Widget := UserInterface.VerseWidget.WBP_RoundedBox{}
                overlay_slot:
                    HorizontalAlignment := horizontal_alignment.Fill
                    VerticalAlignment := vertical_alignment.Fill
                    Padding := margin{Left := 40.0, Top := 40.0, Right := 40.0, Bottom := 40.0}
                    Widget := stack_box:
                        Orientation := orientation.Vertical
                        Slots := array:
                            stack_box_slot:
                                HorizontalAlignment := horizontal_alignment.Fill
                                VerticalAlignment := vertical_alignment.Fill
                                Widget := stack_box:
                                    Orientation := orientation.Horizontal
                                    Slots := array:
                                        stack_box_slot:
                                            HorizontalAlignment := horizontal_alignment.Left
                                            Padding := margin{Left := 0.0, Top := 0.0, Right := 10.0, Bottom := 0.0}
                                            Widget := GameIcon
                                        stack_box_slot:
                                            HorizontalAlignment := horizontal_alignment.Fill
                                            Distribution := option{1.0}
                                            Widget := GameButton
                            stack_box_slot:
                                HorizontalAlignment := horizontal_alignment.Fill
                                VerticalAlignment := vertical_alignment.Fill
                                Padding := margin{Left := 0.0, Top := 10.0, Right := 0.0, Bottom := 0.0}
                                Widget := stack_box:
                                    Orientation := orientation.Horizontal
                                    Slots := array:
                                        stack_box_slot:
                                            HorizontalAlignment := horizontal_alignment.Left
                                            Padding := margin{Left := 0.0, Top := 0.0, Right := 10.0, Bottom := 0.0}
                                            Widget := TeamIcon
                                        stack_box_slot:
                                            HorizontalAlignment := horizontal_alignment.Fill
                                            Distribution := option{1.0}
                                            Widget := TeamButton
                            stack_box_slot:
                                HorizontalAlignment := horizontal_alignment.Fill
                                VerticalAlignment := vertical_alignment.Fill
                                Padding := margin{Left := 0.0, Top := 10.0, Right := 0.0, Bottom := 0.0}
                                Widget := stack_box:
                                    Orientation := orientation.Horizontal
                                    Slots := array:
                                        stack_box_slot:
                                            HorizontalAlignment := horizontal_alignment.Left
                                            Padding := margin{Left := 0.0, Top := 0.0, Right := 10.0, Bottom := 0.0}
                                            Widget := DamageIcon
                                        stack_box_slot:
                                            HorizontalAlignment := horizontal_alignment.Fill
                                            Distribution := option{1.0}
                                            Widget := DamageButton
                            stack_box_slot:
                                HorizontalAlignment := horizontal_alignment.Fill
                                VerticalAlignment := vertical_alignment.Fill
                                Padding := margin{Left := 0.0, Top := 10.0, Right := 0.0, Bottom := 0.0}
                                Widget := stack_box:
                                    Orientation := orientation.Vertical
                                    Slots := for (Index := 0..MaxPlayers - 1):
                                        GameResultEntryUI := game_result_entry_ui{Index := Index, Player := Player}
                                        GameResultEntryUI.Initialize()
                                        set EventsSubscriptions += array{GameResultEntryUI.PlayerClickedEvent.Subscribe(OnPlayerClicked)}
                                        set GameResultEntryUIs += array{GameResultEntryUI}
                        
                                        stack_box_slot:
                                            HorizontalAlignment := horizontal_alignment.Fill
                                            VerticalAlignment := vertical_alignment.Fill
                                            Padding := margin{Left := 0.0, Top := (Index = 0) and 0.0 or 10.0, Right := 0.0, Bottom := 0.0}
                                            Widget := GameResultEntryUI.GetUI()

    <# ----------------------------------------------------------------------------------------------------
        Game UI
    ---------------------------------------------------------------------------------------------------- #>
    GameDurationText:text_block = text_block:
        DefaultTextColor := ThemeColors.Success
        DefaultShadowOffset := option{vector2{X := 2.0, Y := 2.0}}
        DefaultShadowOpacity := 1.0
        DefaultShadowColor := ThemeColors.Success * 0.125

    TimeSurvivedText:text_block = text_block:
        DefaultTextColor := ThemeColors.Success
        DefaultShadowOffset := option{vector2{X := 2.0, Y := 2.0}}
        DefaultShadowOpacity := 1.0
        DefaultShadowColor := ThemeColors.Success * 0.125
        
    EnemiesEliminatedText:text_block = text_block:
        DefaultTextColor := ThemeColors.Success
        DefaultShadowOffset := option{vector2{X := 2.0, Y := 2.0}}
        DefaultShadowOpacity := 1.0
        DefaultShadowColor := ThemeColors.Success * 0.125
        
    LevelReachedText:text_block = text_block:
        DefaultTextColor := ThemeColors.Success
        DefaultShadowOffset := option{vector2{X := 2.0, Y := 2.0}}
        DefaultShadowOpacity := 1.0
        DefaultShadowColor := ThemeColors.Success * 0.125
        
    WoodCollectedText:text_block = text_block:
        DefaultTextColor := ThemeColors.Success
        DefaultShadowOffset := option{vector2{X := 2.0, Y := 2.0}}
        DefaultShadowOpacity := 1.0
        DefaultShadowColor := ThemeColors.Success * 0.125
       
    GoldFromWoodCollectedText:text_block = text_block:
        DefaultTextColor := ThemeColors.Success
        DefaultShadowOffset := option{vector2{X := 2.0, Y := 2.0}}
        DefaultShadowOpacity := 1.0
        DefaultShadowColor := ThemeColors.Success * 0.125

    GoldFromTimeSurvivedText:text_block = text_block:
        DefaultTextColor := ThemeColors.Success
        DefaultShadowOffset := option{vector2{X := 2.0, Y := 2.0}}
        DefaultShadowOpacity := 1.0
        DefaultShadowColor := ThemeColors.Success * 0.125

    GoldFromVictoryLabelText:text_block = text_block:
        DefaultTextColor := ThemeColors.PrimaryText

    BossEliminatedText:text_block = text_block:
        DefaultTextColor := ThemeColors.Success
        DefaultShadowOffset := option{vector2{X := 2.0, Y := 2.0}}
        DefaultShadowOpacity := 1.0
        DefaultShadowColor := ThemeColors.Success * 0.125

    GoldFromVictoryText:text_block = text_block:
        DefaultTextColor := ThemeColors.Success
        DefaultShadowOffset := option{vector2{X := 2.0, Y := 2.0}}
        DefaultShadowOpacity := 1.0
        DefaultShadowColor := ThemeColors.Success * 0.125

    GoldFromObjectivesCompletedText:text_block = text_block:
        DefaultTextColor := ThemeColors.Success
        DefaultShadowOffset := option{vector2{X := 2.0, Y := 2.0}}
        DefaultShadowOpacity := 1.0
        DefaultShadowColor := ThemeColors.Success * 0.125

    GoldFromStagesCompletedText:text_block = text_block:
        DefaultTextColor := ThemeColors.Success
        DefaultShadowOffset := option{vector2{X := 2.0, Y := 2.0}}
        DefaultShadowOpacity := 1.0
        DefaultShadowColor := ThemeColors.Success * 0.125

    GoldFromGoldenTreeText:text_block = text_block:
        DefaultTextColor := ThemeColors.Success
        DefaultShadowOffset := option{vector2{X := 2.0, Y := 2.0}}
        DefaultShadowOpacity := 1.0
        DefaultShadowColor := ThemeColors.Success * 0.125

    GoldMultiplierText:text_block = text_block:
        DefaultTextColor := ThemeColors.Success
        DefaultShadowOffset := option{vector2{X := 2.0, Y := 2.0}}
        DefaultShadowOpacity := 1.0
        DefaultShadowColor := ThemeColors.Success * 0.125
    
    GoldEarnedText:text_block = text_block:
        DefaultTextColor := ThemeColors.Success
        DefaultShadowOffset := option{vector2{X := 2.0, Y := 2.0}}
        DefaultShadowOpacity := 1.0
        DefaultShadowColor := ThemeColors.Success * 0.125

    TreasureChestEarnedText:text_block = text_block:
        DefaultTextColor := ThemeColors.Success
        DefaultShadowOffset := option{vector2{X := 2.0, Y := 2.0}}
        DefaultShadowOpacity := 1.0
        DefaultShadowColor := ThemeColors.Success * 0.125

    ObjectivesCompletedText:text_block = text_block:
        DefaultTextColor := ThemeColors.Success
        DefaultShadowOffset := option{vector2{X := 2.0, Y := 2.0}}
        DefaultShadowOpacity := 1.0
        DefaultShadowColor := ThemeColors.Success * 0.125

    StagesCompletedText:text_block = text_block:
        DefaultTextColor := ThemeColors.Success
        DefaultShadowOffset := option{vector2{X := 2.0, Y := 2.0}}
        DefaultShadowOpacity := 1.0
        DefaultShadowColor := ThemeColors.Success * 0.125

    GoldenTreeText:text_block = text_block:
        DefaultTextColor := ThemeColors.Success
        DefaultShadowOffset := option{vector2{X := 2.0, Y := 2.0}}
        DefaultShadowOpacity := 1.0
        DefaultShadowColor := ThemeColors.Success * 0.125

    CreateGameUI():widget =
        overlay:
            Slots := array:
                overlay_slot:
                    HorizontalAlignment := horizontal_alignment.Fill
                    VerticalAlignment := vertical_alignment.Fill
                    Widget := UserInterface.VerseWidget.WBP_RoundedBox{}
                overlay_slot:
                    HorizontalAlignment := horizontal_alignment.Fill
                    VerticalAlignment := vertical_alignment.Fill
                    Padding := margin{Left := 40.0, Top := 40.0, Right := 40.0, Bottom := 40.0}
                    Widget := stack_box:
                        Orientation := orientation.Vertical
                        Slots := array:
                            stack_box_slot:
                                HorizontalAlignment := horizontal_alignment.Fill
                                VerticalAlignment := vertical_alignment.Fill
                                Padding := margin{Left := 20.0, Top := 0.0, Right := 20.0, Bottom := 20.0}
                                Widget := stack_box:
                                    Orientation := orientation.Horizontal
                                    Slots := array:
                                        stack_box_slot:
                                            HorizontalAlignment := horizontal_alignment.Left
                                            Distribution := option{2.0}
                                            Padding := margin{Left := 0.0, Top := 0.0, Right := 100.0, Bottom := 0.0}
                                            Widget := text_block:
                                                DefaultText := MakeMessage("Results")
                                                DefaultTextColor := ThemeColors.PrimaryText
                                                DefaultTextOpacity := 0.5
                                        stack_box_slot:
                                            HorizontalAlignment := horizontal_alignment.Right
                                            Distribution := option{1.0}
                                            Widget := text_block:
                                                DefaultText := MakeMessage("Score")
                                                DefaultTextColor := ThemeColors.PrimaryText
                                                DefaultTextOpacity := 0.5
                                        stack_box_slot:
                                            HorizontalAlignment := horizontal_alignment.Right
                                            Distribution := option{1.0}
                                            Widget := text_block:
                                                DefaultText := MakeMessage("Gold")
                                                DefaultTextColor := ThemeColors.PrimaryText
                                                DefaultTextOpacity := 0.5
                            stack_box_slot{HorizontalAlignment := horizontal_alignment.Fill, Padding := margin{Left := 20.0, Right := 20.0}, Widget := CreateStatUI(GameDurationText, text_block{DefaultText := MakeMessage("")}, "Game Duration")}
                            stack_box_slot{HorizontalAlignment := horizontal_alignment.Fill, Padding := margin{Left := 20.0, Right := 20.0}, Widget := CreateStatUI(TimeSurvivedText, GoldFromTimeSurvivedText, "Time Survived")}
                            stack_box_slot{HorizontalAlignment := horizontal_alignment.Fill, Padding := margin{Left := 20.0, Right := 20.0}, Widget := CreateStatUI(EnemiesEliminatedText, text_block{DefaultText := MakeMessage("")}, "Enemies Eliminated")}
                            stack_box_slot{HorizontalAlignment := horizontal_alignment.Fill, Padding := margin{Left := 20.0, Right := 20.0}, Widget := CreateStatUI(LevelReachedText, text_block{DefaultText := MakeMessage("")}, "Level Reached")}
                            stack_box_slot{HorizontalAlignment := horizontal_alignment.Fill, Padding := margin{Left := 20.0, Right := 20.0}, Widget := CreateStatUI(WoodCollectedText, GoldFromWoodCollectedText, "Wood Collected")}
                            stack_box_slot{HorizontalAlignment := horizontal_alignment.Fill, Padding := margin{Left := 20.0, Right := 20.0}, Widget := CreateStatUI(ObjectivesCompletedText, GoldFromObjectivesCompletedText, "Missions Completed")}
                            stack_box_slot{HorizontalAlignment := horizontal_alignment.Fill, Padding := margin{Left := 20.0, Right := 20.0}, Widget := CreateStatUI(StagesCompletedText, GoldFromStagesCompletedText, "Stages Completed")}
                            stack_box_slot{HorizontalAlignment := horizontal_alignment.Fill, Padding := margin{Left := 20.0, Right := 20.0}, Widget := CreateStatUI(GoldenTreeText, GoldFromGoldenTreeText, "Golden Tree Collected")}
                            

                            stack_box_slot:
                                HorizontalAlignment := horizontal_alignment.Fill
                                VerticalAlignment := vertical_alignment.Bottom
                                Distribution := option{1.0}
                                Padding := margin{Left := 20.0, Right := 20.0}
                                Widget := stack_box:
                                    Orientation := orientation.Horizontal
                                    Slots := array:
                                        stack_box_slot:
                                            HorizontalAlignment := horizontal_alignment.Left
                                            Distribution := option{2.0}
                                            Widget := GoldFromVictoryLabelText
                                        stack_box_slot:
                                            HorizontalAlignment := horizontal_alignment.Right
                                            Distribution := option{1.0}
                                            Widget := BossEliminatedText
                                        stack_box_slot:
                                            HorizontalAlignment := horizontal_alignment.Right
                                            Distribution := option{1.0}
                                            Widget := GoldFromVictoryText
                            stack_box_slot{HorizontalAlignment := horizontal_alignment.Fill, Padding := margin{Left := 20.0, Right := 20.0}, Widget := CreateStatUI(GoldMultiplierText, "Gold Multipliers")}
                            stack_box_slot:
                                HorizontalAlignment := horizontal_alignment.Fill
                                Padding := margin{Top := 20.0}
                                Widget := overlay:
                                    Slots := array:
                                        overlay_slot:
                                            HorizontalAlignment := horizontal_alignment.Fill
                                            VerticalAlignment := vertical_alignment.Fill
                                            Widget := UserInterface.VerseWidget.WBP_RoundedBox_Light{}
                                        overlay_slot:
                                            HorizontalAlignment := horizontal_alignment.Fill
                                            VerticalAlignment := vertical_alignment.Fill
                                            Padding := margin{Left := 20.0, Top := 20.0, Right := 20.0, Bottom := 20.0}
                                            Widget := stack_box:
                                                Orientation := orientation.Vertical
                                                Slots := array:
                                                    stack_box_slot:
                                                        HorizontalAlignment := horizontal_alignment.Fill
                                                        VerticalAlignment := vertical_alignment.Fill
                                                        Distribution := option{1.0}
                                                        Padding := margin{Left := 0.0, Top := 0.0, Right := 0.0, Bottom := 0.0}
                                                        Widget := CreateStatUI(GoldEarnedText, "Gold Earned", ?Icon := option{currency_id.Gold.ToIcon()})
                                                    stack_box_slot:
                                                        HorizontalAlignment := horizontal_alignment.Fill
                                                        VerticalAlignment := vertical_alignment.Fill
                                                        Distribution := option{1.0}
                                                        Padding := margin{Left := 0.0, Top := 0.0, Right := 0.0, Bottom := 0.0}
                                                        Widget := CreateStatUI(TreasureChestEarnedText, "Treasure Chest Earned", ?Icon := option{currency_id.TreasureChest.ToIcon()})

    <# ----------------------------------------------------------------------------------------------------
        Team UI
    ---------------------------------------------------------------------------------------------------- #>
    var GameResultTeamEntryUIs:[]game_result_team_entry_ui = array{}

    CreateTeamUI():widget =
        overlay:
            Slots := array:
                overlay_slot:
                    HorizontalAlignment := horizontal_alignment.Fill
                    VerticalAlignment := vertical_alignment.Fill
                    Widget := UserInterface.VerseWidget.WBP_RoundedBox{}
                overlay_slot:
                    HorizontalAlignment := horizontal_alignment.Fill
                    VerticalAlignment := vertical_alignment.Fill
                    Padding := margin{Left := 40.0, Top := 20.0, Right := 40.0, Bottom := 20.0}
                    Widget := stack_box:
                        Orientation := orientation.Horizontal
                        Slots := array:
                            stack_box_slot:
                                HorizontalAlignment := horizontal_alignment.Fill
                                VerticalAlignment := vertical_alignment.Fill
                                Padding := margin{Left := 0.0, Top := 40.0, Right := 20.0, Bottom := 40.0}
                                Widget := stack_box:
                                    Orientation := orientation.Vertical
                                    Slots := array:
                                        stack_box_slot:
                                            HorizontalAlignment := horizontal_alignment.Left
                                            VerticalAlignment := vertical_alignment.Fill
                                            Distribution := option{0.6}
                                            Widget := text_block:
                                                DefaultText := MakeMessage("Player")
                                                DefaultTextColor := ThemeColors.PrimaryText
                                                DefaultTextOpacity := 0.5
                                        stack_box_slot:
                                            HorizontalAlignment := horizontal_alignment.Left
                                            VerticalAlignment := vertical_alignment.Fill
                                            Distribution := option{1.0}
                                            Widget := text_block:
                                                DefaultText := MakeMessage("Weapons")
                                                DefaultTextColor := ThemeColors.PrimaryText
                                                DefaultTextOpacity := 0.5
                                        stack_box_slot:
                                            HorizontalAlignment := horizontal_alignment.Left
                                            VerticalAlignment := vertical_alignment.Fill
                                            Widget := text_block:
                                                DefaultText := MakeMessage("Eliminate")
                                                DefaultTextColor := ThemeColors.PrimaryText
                                                DefaultTextOpacity := 0.5
                                        stack_box_slot:
                                            HorizontalAlignment := horizontal_alignment.Left
                                            VerticalAlignment := vertical_alignment.Fill
                                            Widget := text_block:
                                                DefaultText := MakeMessage("Damage")
                                                DefaultTextColor := ThemeColors.PrimaryText
                                                DefaultTextOpacity := 0.5
                                        stack_box_slot:
                                            HorizontalAlignment := horizontal_alignment.Left
                                            VerticalAlignment := vertical_alignment.Fill
                                            Widget := text_block:
                                                DefaultText := MakeMessage("Wood")
                                                DefaultTextColor := ThemeColors.PrimaryText
                                                DefaultTextOpacity := 0.5
                                        stack_box_slot:
                                            HorizontalAlignment := horizontal_alignment.Left
                                            VerticalAlignment := vertical_alignment.Fill
                                            Widget := text_block:
                                                DefaultText := MakeMessage("Revived")
                                                DefaultTextColor := ThemeColors.PrimaryText
                                                DefaultTextOpacity := 0.5
                                        stack_box_slot:
                                            HorizontalAlignment := horizontal_alignment.Left
                                            VerticalAlignment := vertical_alignment.Fill
                                            Widget := text_block:
                                                DefaultText := MakeMessage("Downed")
                                                DefaultTextColor := ThemeColors.PrimaryText
                                                DefaultTextOpacity := 0.5
                                                
                            stack_box_slot:
                                HorizontalAlignment := horizontal_alignment.Fill
                                VerticalAlignment := vertical_alignment.Fill
                                Distribution := option{1.0}
                                Padding := margin{Left := 0.0, Top := 0.0, Right := 0.0, Bottom := 0.0}
                                Widget := stack_box:
                                    Orientation := orientation.Horizontal
                                    Slots := for (Index := 0..MaxPlayers - 1):
                                        GameResultTeamEntryUI := game_result_team_entry_ui{Player := Player}
                                        GameResultTeamEntryUI.Initialize()
                                        set GameResultTeamEntryUIs += array{GameResultTeamEntryUI}

                                        stack_box_slot:
                                            HorizontalAlignment := horizontal_alignment.Fill
                                            VerticalAlignment := vertical_alignment.Fill
                                            Distribution := option{1.0}
                                            Padding := margin{Left := (Index = 0) and 0.0 or 20.0, Top := 0.0, Right := 0.0, Bottom := 0.0}
                                            Widget := overlay:
                                                Slots := array:
                                                    overlay_slot:
                                                        HorizontalAlignment := horizontal_alignment.Fill
                                                        VerticalAlignment := vertical_alignment.Fill
                                                        Widget := UserInterface.VerseWidget.WBP_RoundedBox{}
                                                    overlay_slot:
                                                        HorizontalAlignment := horizontal_alignment.Fill
                                                        VerticalAlignment := vertical_alignment.Fill
                                                        Padding := margin{Left := 20.0, Top := 40.0, Right := 20.0, Bottom := 40.0}
                                                        Widget := GameResultTeamEntryUI.GetUI()

    <# ----------------------------------------------------------------------------------------------------
        Damage UI
    ---------------------------------------------------------------------------------------------------- #>
    var GameResultDamageEntryUIs:[]game_result_damage_entry_ui = array{}

    CreateDamageUI():widget =
        overlay:
            Slots := array:
                overlay_slot:
                    HorizontalAlignment := horizontal_alignment.Fill
                    VerticalAlignment := vertical_alignment.Fill
                    Widget := UserInterface.VerseWidget.WBP_RoundedBox{}
                overlay_slot:
                    HorizontalAlignment := horizontal_alignment.Fill
                    VerticalAlignment := vertical_alignment.Fill
                    Padding := margin{Left := 40.0, Top := 40.0, Right := 40.0, Bottom := 40.0}
                    Widget := stack_box:
                        Orientation := orientation.Vertical
                        Slots := array:
                            stack_box_slot:
                                HorizontalAlignment := horizontal_alignment.Fill
                                VerticalAlignment := vertical_alignment.Fill
                                Padding := margin{Left := 0.0, Top := 0.0, Right := 0.0, Bottom := 20.0}
                                Widget := stack_box:
                                    Orientation := orientation.Horizontal
                                    Slots := array:
                                        stack_box_slot:
                                            HorizontalAlignment := horizontal_alignment.Left
                                            VerticalAlignment := vertical_alignment.Fill
                                            Distribution := option{1.0}
                                            Widget := text_block:
                                                DefaultText := MakeMessage("Player")
                                                DefaultTextColor := ThemeColors.PrimaryText
                                                DefaultTextOpacity := 0.5
                                        stack_box_slot:
                                            HorizontalAlignment := horizontal_alignment.Center
                                            VerticalAlignment := vertical_alignment.Fill
                                            Distribution := option{0.5}
                                            Widget := text_block:
                                                DefaultText := MakeMessage("Total Damage")
                                                DefaultTextColor := ThemeColors.PrimaryText
                                                DefaultTextOpacity := 0.5
                                        stack_box_slot:
                                            HorizontalAlignment := horizontal_alignment.Center
                                            VerticalAlignment := vertical_alignment.Fill
                                            Distribution := option{0.5}
                                            Widget := text_block:
                                                DefaultText := MakeMessage("Miniboss")
                                                DefaultTextColor := ThemeColors.PrimaryText
                                                DefaultTextOpacity := 0.5
                                        stack_box_slot:
                                            HorizontalAlignment := horizontal_alignment.Center
                                            VerticalAlignment := vertical_alignment.Fill
                                            Distribution := option{0.5}
                                            Widget := text_block:
                                                DefaultText := MakeMessage("Boss")
                                                DefaultTextColor := ThemeColors.PrimaryText
                                                DefaultTextOpacity := 0.5
                                                
                            stack_box_slot:
                                HorizontalAlignment := horizontal_alignment.Fill
                                VerticalAlignment := vertical_alignment.Fill
                                Distribution := option{1.0}
                                Padding := margin{Left := 0.0, Top := 0.0, Right := 0.0, Bottom := 0.0}
                                Widget := stack_box:
                                    Orientation := orientation.Vertical
                                    Slots := for (Index := 0..MaxPlayers - 1):
                                        GameResultDamageEntryUI := game_result_damage_entry_ui{Player := Player}
                                        GameResultDamageEntryUI.Initialize()
                                        set GameResultDamageEntryUIs += array{GameResultDamageEntryUI}

                                        stack_box_slot:
                                            HorizontalAlignment := horizontal_alignment.Fill
                                            VerticalAlignment := vertical_alignment.Fill
                                            Padding := if (Index = 0):
                                                margin{Left := 0.0, Top := 0.0, Right := 0.0, Bottom := 0.0}
                                            else:
                                                margin{Left := 20.0, Top := 20.0, Right := 20.0, Bottom := 0.0}

                                            Widget := if (Index = 0):
                                                overlay:
                                                    Slots := array:
                                                        overlay_slot:
                                                            HorizontalAlignment := horizontal_alignment.Fill
                                                            VerticalAlignment := vertical_alignment.Fill
                                                            Widget := UserInterface.VerseWidget.WBP_RoundedBox_Light{}
                                                        overlay_slot:
                                                            HorizontalAlignment := horizontal_alignment.Fill
                                                            VerticalAlignment := vertical_alignment.Fill
                                                            Padding := margin{Left := 20.0, Top := 20.0, Right := 20.0, Bottom := 20.0}
                                                            Widget := GameResultDamageEntryUI.GetUI()
                                            else:
                                                GameResultDamageEntryUI.GetUI()

    <# ----------------------------------------------------------------------------------------------------
        Get Functions
    ---------------------------------------------------------------------------------------------------- #>
    GetUI(GameResultType:game_result_type):widget =
        case (GameResultType):
            game_result_type.Game => GameUI
            game_result_type.Team => TeamUI
            game_result_type.Damage => DamageUI
            game_result_type.Player => PlayerUI

    <# ----------------------------------------------------------------------------------------------------
        Update Functions
    ---------------------------------------------------------------------------------------------------- #>
    AnimateReward(Text:text_block, Value:float, Duration:float, AudioId:audio_id, ?OptionalText:string = "")<suspends>:void =
        if (Value <= 0.0):
            Text.SetText(MakeMessage("{FormatAsCurrency(Value)}{OptionalText}"))
            Sleep(Duration)
            return

        DeltaTimer := delta_timer{}

        loop:
            ElapsedTime := DeltaTimer.GetElapsedTime()
            Progress := ElapsedTime / Duration
            CurrentAmount := Lerp(0.0, Value, Progress)
            Text.SetText(MakeMessage("{FormatAsCurrency(CurrentAmount)}{OptionalText}"))
            Sleep(0.0)
            if (Progress >= 1.0):
                Text.SetText(MakeMessage("{FormatAsCurrency(Value)}{OptionalText}"))
                Player.PlayAudio(AudioId)
                break

    AnimateRewards(ResultsInfo:game_result_info)<suspends>:void =
        defer:
            GoldFromTimeSurvivedText.SetText(MakeMessage("{FormatAsCurrency(ResultsInfo.GoldFromTimeSurvived)}"))
            GoldFromWoodCollectedText.SetText(MakeMessage("{FormatAsCurrency(ResultsInfo.GoldFromWoodCollected)}"))
            GoldFromObjectivesCompletedText.SetText(MakeMessage("{FormatAsCurrency(ResultsInfo.GoldFromObjectivesCompleted)}"))
            GoldFromStagesCompletedText.SetText(MakeMessage("{FormatAsCurrency(ResultsInfo.GoldFromStagesCompleted)}"))
            GoldFromGoldenTreeText.SetText(MakeMessage("{FormatAsCurrency(ResultsInfo.GoldFromGoldenTree)}"))
            GoldFromVictoryText.SetText(MakeMessage("{FormatAsCurrency(ResultsInfo.GoldFromVictory)} (+{FormatAsPercent(ResultsInfo.GameSettings.IsEndless? and CalculateEndlessModeGoldMultiplier(ResultsInfo.Arena, ResultsInfo.StagesCompleted) or VictoryBonusGoldMultplier)})"))
            GoldMultiplierText.SetText(MakeMessage("{FormatAsCurrency(ResultsInfo.GoldFromMultiplier)} (+{FormatAsPercent(ResultsInfo.GoldMultiplier)})"))
            GoldEarnedText.SetText(MakeMessage("{FormatAsCurrency(ResultsInfo.GoldEarned)}"))
            TreasureChestEarnedText.SetText(MakeMessage("{FormatAsCurrency(ResultsInfo.TreasureChestEarned)}"))

        GoldFromTimeSurvivedText.SetText(MakeMessage(""))
        GoldFromWoodCollectedText.SetText(MakeMessage(""))
        GoldFromObjectivesCompletedText.SetText(MakeMessage(""))
        GoldFromStagesCompletedText.SetText(MakeMessage(""))
        GoldFromGoldenTreeText.SetText(MakeMessage(""))
        GoldFromVictoryText.SetText(MakeMessage(""))
        GoldMultiplierText.SetText(MakeMessage(""))
        GoldEarnedText.SetText(MakeMessage(""))
        TreasureChestEarnedText.SetText(MakeMessage(""))

        race:
            CancelAnimationEvent.Await()
            block:
                AnimateReward(GoldFromTimeSurvivedText, ResultsInfo.GoldFromTimeSurvived, 0.5, audio_id.Purchase)
                AnimateReward(GoldFromWoodCollectedText, ResultsInfo.GoldFromWoodCollected, 0.5, audio_id.Purchase)
                AnimateReward(GoldFromObjectivesCompletedText, ResultsInfo.GoldFromObjectivesCompleted, 0.5, audio_id.Purchase)
                AnimateReward(GoldFromStagesCompletedText, ResultsInfo.GoldFromStagesCompleted, 0.5, audio_id.Purchase)
                AnimateReward(GoldFromGoldenTreeText, ResultsInfo.GoldFromGoldenTree, 0.5, audio_id.Purchase)
                AnimateReward(GoldFromVictoryText, ResultsInfo.GoldFromVictory, 0.5, audio_id.Purchase, ?OptionalText := " (+{FormatAsPercent(ResultsInfo.GameSettings.IsEndless? and CalculateEndlessModeGoldMultiplier(ResultsInfo.Arena, ResultsInfo.StagesCompleted) or VictoryBonusGoldMultplier)})")
                AnimateReward(GoldMultiplierText, ResultsInfo.GoldFromMultiplier, 0.5, audio_id.Purchase, ?OptionalText := " (+{FormatAsPercent(ResultsInfo.GoldMultiplier)})")
                AnimateReward(GoldEarnedText, ResultsInfo.GoldEarned, 1.0, audio_id.UpgradeRare)
                AnimateReward(TreasureChestEarnedText, ResultsInfo.TreasureChestEarned, 1.0, audio_id.OpenTreasureChest)

    Update(ResultsInfo:game_result_info):void =
        set ActiveResultsInfo = ResultsInfo
        set IsClaimed = false
        set IsRankUp = ResultsInfo.IsRankUp
        SetType(game_result_type.Game)
        
        ArenaLevel := CalculateArenaLevel(ResultsInfo.GameSettings.Arena.Level, ResultsInfo.GameSettings.Difficulty)
        SetTitle("{ResultsInfo.Arena.Name} (Lv. {ArenaLevel}) (x{ResultsInfo.GameSettings.GameSpeed})")
        ArenaText.SetText(MakeMessage("{ResultsInfo.IsVictory? and "VICTORY" or "DEFEAT" + "!"}"))
        ArenaText.SetColor(ResultsInfo.IsVictory? and ThemeColors.Success or ThemeColors.Failure)

        # Menu UI
        for (Index -> GameResultEntryUI : GameResultEntryUIs):
            if (ActivePlayer := ResultsInfo.ActivePlayers[Index]):
                GameResultEntryUI.Update(ActivePlayer)
                GameResultEntryUI.Show()
            else:
                GameResultEntryUI.Hide()

        # Game UI
        GameDurationText.SetText(MakeMessage(ResultsInfo.GameDuration.FormatAsTime()))
        TimeSurvivedText.SetText(MakeMessage(ResultsInfo.TimeSurvived.FormatAsTime()))
        EnemiesEliminatedText.SetText(MakeMessage("{ResultsInfo.EnemiesEliminated}"))
        LevelReachedText.SetText(MakeMessage("{ResultsInfo.LevelReached}"))
        WoodCollectedText.SetText(MakeMessage("{FormatAsCurrency(ResultsInfo.WoodCollected)}"))
        
        StageCount := ResultsInfo.GameSettings.IsEndless? and "∞" or "{CalculateArenaStageCount(ResultsInfo.GameSettings.Difficulty)}"
        ObjectivesCompletedText.SetText(MakeMessage("{ResultsInfo.ObjectivesCompleted}/{StageCount}"))
        StagesCompletedText.SetText(MakeMessage("{ResultsInfo.StagesCompleted}/{StageCount}"))
        GoldenTreeText.SetText(MakeMessage("{ResultsInfo.GoldenTree}"))
        GoldFromVictoryLabelText.SetText(MakeMessage(ResultsInfo.GameSettings.IsEndless? and "Bosses Eliminated:" or "Victory Bonus:"))
        BossEliminatedText.SetText(MakeMessage(ResultsInfo.GameSettings.IsEndless? and "{Floor(ResultsInfo.StagesCompleted / EndlessModeBossStageInterval)}" or ""))

        spawn{AnimateRewards(ResultsInfo)}

        # Team UI
        PlayerDamageDealt := for (ActivePlayer : ResultsInfo.ActivePlayers, DamageDealt := ResultsInfo.PlayerStats[ActivePlayer].GetGameStat(player_game_stat_id.DamageDealt)) {(ActivePlayer, DamageDealt)}
        SortedPlayerDamageDealt := SortBy(PlayerDamageDealt, LessPlayerFloat)
        for (Index -> GameResultTeamEntryUI : GameResultTeamEntryUIs):
            if (ActivePlayer := SortedPlayerDamageDealt[Index], PlayerStats := ActiveResultsInfo.PlayerStats[ActivePlayer(0)]):
                EnemiesEliminated := ActiveResultsInfo.PlayerEnemiesEliminated[ActivePlayer(0)] or 0
                DamageDealt := PlayerStats.GetGameStat(player_game_stat_id.DamageDealt)
                WeaponLevels := PlayerStats.GetWeaponLevels()
                WoodCollected := PlayerStats.GetGameStat(player_game_stat_id.WoodCollected)
                PlayersRevived := PlayerStats.GetGameStat(player_game_stat_id.PlayersRevived)
                TimesDowned := PlayerStats.GetGameStat(player_game_stat_id.TimesDowned)
                GameResultTeamEntryUI.Update(ActivePlayer(0), EnemiesEliminated, DamageDealt, WeaponLevels, WoodCollected, PlayersRevived, TimesDowned)
                GameResultTeamEntryUI.Show()
            else:
                GameResultTeamEntryUI.Hide()

        # Damage UI
        for (Index -> GameResultDamageEntryUI : GameResultDamageEntryUIs):
            if (ActivePlayer := SortedPlayerDamageDealt[Index], PlayerStats := ActiveResultsInfo.PlayerStats[ActivePlayer(0)]):
                DamageDealt := PlayerStats.GetGameStat(player_game_stat_id.DamageDealt)
                DamageMinibossDealt := PlayerStats.GetGameStat(player_game_stat_id.DamageDealtMiniboss)
                DamageBossDealt := PlayerStats.GetGameStat(player_game_stat_id.DamageDealtBoss)
                GameResultDamageEntryUI.Update(ActivePlayer(0), DamageDealt, DamageMinibossDealt, DamageBossDealt)
                GameResultDamageEntryUI.Show()
            else:
                GameResultDamageEntryUI.Hide()

    UpdatePlayerUI(InPlayer:player):void =
        if (AbilityStatInfoUI := MaybeAbilityStatInfoUI?, PlayerStats := ActiveResultsInfo.PlayerStats[InPlayer]):
            AbilityStatInfoUI.Update(?MaybeGameResult := option{ActiveResultsInfo, PlayerStats})

    SetType(GameResultType:game_result_type, ?MaybePlayerIndex:?int = false):void =
        CancelAnimationEvent.Signal()
        GameIcon.SetTint(NamedColors.White)
        TeamIcon.SetTint(NamedColors.White)
        DamageIcon.SetTint(NamedColors.White)
        for (Index -> GameResultEntryUI : GameResultEntryUIs):
            GameResultEntryUI.SetInactive()

        case (GameResultType):
            game_result_type.Game => GameIcon.SetTint(ThemeColors.Highlight * 2)
            game_result_type.Team => TeamIcon.SetTint(ThemeColors.Highlight * 2)
            game_result_type.Damage => DamageIcon.SetTint(ThemeColors.Highlight * 2)
            game_result_type.Player => if (PlayerIndex := MaybePlayerIndex?):
                for (Index -> GameResultEntryUI : GameResultEntryUIs):
                    if (Index = PlayerIndex):
                        GameResultEntryUI.SetActive()
                    else:
                        GameResultEntryUI.SetInactive()

        GetUI(ActiveGameResultType).SetVisibility(widget_visibility.Collapsed)
        GetUI(GameResultType).SetVisibility(widget_visibility.Visible)

        set ActiveGameResultType = GameResultType

    <# ----------------------------------------------------------------------------------------------------
        Event Handlers
    ---------------------------------------------------------------------------------------------------- #>
    OnClosed(InPlayer:player):void =
        if (not IsClaimed?):
            InPlayer.AwardAccolade(ActiveResultsInfo.IsVictory? and accolade_id.Victory or accolade_id.CompleteGame)
            set IsClaimed = true

            if (IsRankUp?):
                InPlayer.ShowMessage(InPlayer.GetRankUpMessage())
                Player.ActivateVFXPowerup of Player.GetRankVFXPowerUp()
            
        CancelAnimationEvent.Signal()

    OnPlayerClicked(InPlayer:player, PlayerIndex:int):void =
        SetType(game_result_type.Player, ?MaybePlayerIndex := option{PlayerIndex})
        UpdatePlayerUI(InPlayer)
        
    OnClickGame(Message:widget_message):void =
        SetType(game_result_type.Game)       

    OnClickTeam(Message:widget_message):void =
        SetType(game_result_type.Team)

    OnClickDamage(Message:widget_message):void =
        SetType(game_result_type.Damage)