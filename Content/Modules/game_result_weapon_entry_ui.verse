using { /Fortnite.com/UI }
using { /Verse.org/Colors }
using { /Verse.org/Simulation }
using { /UnrealEngine.com/Temporary/SpatialMath }
using { /UnrealEngine.com/Temporary/UI }

game_result_weapon_entry_ui := class(ui_component):
    TextSize:float = ThemeTextSize.Default
    IconSize:float = 1.0

    PlayerWeaponIcon:texture_block = texture_block:
        DefaultImage := UserInterface.Icons.WeaponBow_64

    LevelText:text_block = text_block:
        DefaultTextColor := ThemeColors.PrimaryText
        DefaultShadowOpacity := 0.5
        DefaultShadowOffset := option{vector2{X := 2.0, Y := 2.0}}

    DamageText:text_block = text_block:
        DefaultTextColor := ThemeColors.Inactive
        DefaultShadowOffset := option{vector2{X := 2.0, Y := 2.0}}
        DefaultShadowColor := ThemeColors.Success * 0.125

    CooldownText:text_block = text_block:
        DefaultTextColor := ThemeColors.Inactive
        DefaultShadowOffset := option{vector2{X := 2.0, Y := 2.0}}
        DefaultShadowColor := ThemeColors.Success * 0.125

    SizeText:text_block = text_block:
        DefaultTextColor := ThemeColors.Inactive
        DefaultShadowOffset := option{vector2{X := 2.0, Y := 2.0}}
        DefaultShadowColor := ThemeColors.Success * 0.125

    DPSText:text_block = text_block:
        DefaultTextColor := ThemeColors.Success
        DefaultShadowOpacity := 1.0
        DefaultShadowOffset := option{vector2{X := 2.0, Y := 2.0}}
        DefaultShadowColor := ThemeColors.Success * 0.125

    DPSDamageText:text_block = text_block:
        DefaultTextColor := ThemeColors.Success
        DefaultShadowOpacity := 1.0
        DefaultShadowOffset := option{vector2{X := 2.0, Y := 2.0}}
        DefaultShadowColor := ThemeColors.Success * 0.125

    DPSDurationText:text_block = text_block:
        DefaultTextColor := ThemeColors.Success
        DefaultShadowOpacity := 1.0
        DefaultShadowOffset := option{vector2{X := 2.0, Y := 2.0}}
        DefaultShadowColor := ThemeColors.Success * 0.125

    var EvolutionIcons:[]texture_block = array{}
    
    <# ----------------------------------------------------------------------------------------------------
        Initialization
    ---------------------------------------------------------------------------------------------------- #>
    CreateUI<override>():widget =
        PlayerWeaponIcon.SetDesiredSize(vector2{X := 50.0, Y := 50.0} * IconSize)
        LevelText.SetTextSize(TextSize)
        DamageText.SetTextSize(TextSize)
        CooldownText.SetTextSize(TextSize)
        SizeText.SetTextSize(TextSize)
        DPSText.SetTextSize(TextSize)
        DPSDamageText.SetTextSize(TextSize)
        DPSDurationText.SetTextSize(TextSize)

        stack_box:
            Orientation := orientation.Vertical
            Slots := array:
                stack_box_slot:
                    HorizontalAlignment := horizontal_alignment.Fill
                    VerticalAlignment := vertical_alignment.Fill
                    Widget := stack_box:
                        Orientation := orientation.Horizontal
                        Slots := array:
                            stack_box_slot:
                                HorizontalAlignment := horizontal_alignment.Left
                                Padding := margin{Left := 28.0, Top := 0.0, Right := 0.0, Bottom := 0.0} * IconSize
                                Distribution := option{1.0}
                                Widget := PlayerWeaponIcon
                            stack_box_slot:
                                HorizontalAlignment := horizontal_alignment.Left
                                VerticalAlignment := vertical_alignment.Bottom
                                Distribution := option{1.0}
                                Padding := margin{Left := -28.0, Top := 0.0, Right := 0.0, Bottom := 0.0} * IconSize
                                Widget := stack_box:
                                    Orientation := orientation.Horizontal
                                    Slots := array:
                                        stack_box_slot:
                                            HorizontalAlignment := horizontal_alignment.Fill
                                            VerticalAlignment := vertical_alignment.Fill
                                            Padding := margin{Left := 0.0, Top := 0.0, Right := 10.0, Bottom := 0.0} * IconSize
                                            Widget := text_block:
                                                DefaultText := MakeMessage("Lv.")
                                                DefaultTextSize := TextSize
                                                DefaultTextColor := ThemeColors.Highlight * 2
                                                DefaultShadowColor := ThemeColors.Highlight * 0.25
                                                DefaultShadowOffset := option{vector2{X := 2.0, Y := 2.0}}
                                                DefaultShadowOpacity := 1.0
                                        stack_box_slot:
                                            HorizontalAlignment := horizontal_alignment.Fill
                                            VerticalAlignment := vertical_alignment.Fill
                                            Widget := LevelText
                            stack_box_slot:
                                HorizontalAlignment := horizontal_alignment.Left
                                VerticalAlignment := vertical_alignment.Bottom
                                Distribution := option{1.0}
                                Widget := DamageText
                            stack_box_slot:
                                HorizontalAlignment := horizontal_alignment.Left
                                VerticalAlignment := vertical_alignment.Bottom
                                Distribution := option{1.0}
                                Widget := CooldownText
                            stack_box_slot:
                                HorizontalAlignment := horizontal_alignment.Left
                                VerticalAlignment := vertical_alignment.Bottom
                                Distribution := option{1.0}
                                Widget := SizeText
                stack_box_slot:
                    HorizontalAlignment := horizontal_alignment.Fill
                    VerticalAlignment := vertical_alignment.Fill
                    Widget := stack_box:
                        Orientation := orientation.Horizontal
                        Slots := array:
                            stack_box_slot:
                                HorizontalAlignment := horizontal_alignment.Left
                                Distribution := option{1.0}
                                Widget := stack_box:
                                    Orientation := orientation.Horizontal
                                    Slots := for (Index := 1..3):
                                        EvolutionIcon:= texture_block:
                                            DefaultImage := UserInterface.Icons.Lock_64
                                            DefaultDesiredSize := vector2{X := 32.0, Y := 32.0} * IconSize

                                        set EvolutionIcons += array{EvolutionIcon}

                                        stack_box_slot:
                                            Padding := margin{Left := (Index = 1) and 0.0 or 5.0, Top := 0.0, Right := 0.0, Bottom := 0.0} * IconSize
                                            Widget := EvolutionIcon
                                                
                            stack_box_slot:
                                HorizontalAlignment := horizontal_alignment.Left
                                Distribution := option{1.0}
                                Widget := text_block:
                                    DefaultText := MakeMessage("DPS")
                                    DefaultTextSize := TextSize
                                    DefaultTextColor := ThemeColors.PrimaryText
                                    DefaultTextOpacity := 0.5
                            stack_box_slot:
                                HorizontalAlignment := horizontal_alignment.Left
                                Distribution := option{1.0}
                                Widget := DPSText
                            stack_box_slot:
                                HorizontalAlignment := horizontal_alignment.Left
                                Distribution := option{1.0}
                                Widget := DPSDamageText
                            stack_box_slot:
                                HorizontalAlignment := horizontal_alignment.Left
                                Distribution := option{1.0}
                                Widget := DPSDurationText

    <# ----------------------------------------------------------------------------------------------------
        Update Functions
    ---------------------------------------------------------------------------------------------------- #>
    Update(InPlayer:player, WeaponId:weapon_id, WeaponLevel:int, WeaponStats:[stat_id]float, ?MaybeResultsInfo:?game_result_info = false):void =
        if (WeaponInfo := GetWeaponInfo[WeaponId]):
            PlayerWeaponIcon.SetImage(WeaponInfo.IconSmall)
            PlayerWeaponIcon.SetTint(NamedColors.White)
            LevelText.SetText(MakeMessage("{WeaponLevel}"))

            if (ParentWidget := LevelText.GetParentWidget[]):
                ParentWidget.SetVisibility(widget_visibility.Visible)

            if (PlayerStats := InPlayer.GetStats[]):
                Damage := (WeaponStats[stat_id.Damage] or 0.0) + ((Player.HasBoon(boon_id.AllForOne)?) and (WeaponId = Player.GetEquippedWeapon()? and 0.5 or -0.25) or 0.0)
                Cooldown := WeaponStats[stat_id.Cooldown] or 0.0
                Size := WeaponStats[stat_id.Size] or 0.0

                BanishedAbilities := MaybeResultsInfo?.PlayerStats[InPlayer].GetBanishedAbilities() or InPlayer.GetBanishedAbilities()
                BanishedIds := for (BanishedAbility : BanishedAbilities, BanishedAbility.MaybeWeaponId? = WeaponId) {BanishedAbility.Id}
                
                UpdateStatText(DamageText, Damage, ability_id.Damage, BanishedIds)
                UpdateStatText(CooldownText, Cooldown, ability_id.Cooldown, BanishedIds)
                UpdateStatText(SizeText, Size, ability_id.Size, BanishedIds)

                WeaponEvolutionInfos := MaybeResultsInfo?.PlayerStats[InPlayer].GetWeaponEvolution[WeaponId] or InPlayer.GetWeaponEvolution(WeaponId)
                for (Index -> EvolutionIcon : EvolutionIcons):
                    if (EvolutionInfo := WeaponEvolutionInfos[Index]):
                        EvolutionIcon.SetImage(EvolutionInfo.IconSmall)
                        EvolutionIcon.SetTint(EvolutionInfo.IconColor)
                    else:
                        EvolutionIcon.SetImage(UserInterface.Icons.Lock_64)
                        EvolutionIcon.SetTint(ThemeColors.Inactive)
                        
                    EvolutionIcon.SetVisibility(widget_visibility.Visible)

            if (ResultsInfo := MaybeResultsInfo?, Ability := ResultsInfo.PlayerActiveAbilities[InPlayer][WeaponId]):
                DPSText.SetText(MakeMessage("{FormatAsCurrency(Ability.GetDPS(?EndTime := option{ResultsInfo.EndTime}))}/s"))
                DPSDamageText.SetText(MakeMessage("{FormatAsCurrency(Ability.GetTotalDamage())}"))
                DPSDurationText.SetText(MakeMessage("/ {Ability.GetDuration(?EndTime := option{ResultsInfo.EndTime}).FormatAsTime()}"))
            else:
                DPSText.SetText(MakeMessage("{FormatAsCurrency(InPlayer.GetWeaponDPS(WeaponId))}/s"))
                DPSDamageText.SetText(MakeMessage("{FormatAsCurrency(InPlayer.GetWeaponTotalDamage(WeaponId))}"))
                DPSDurationText.SetText(MakeMessage("/ {InPlayer.GetWeaponDuration(WeaponId).FormatAsTime()}"))

            if (ParentWidget := DPSText.GetParentWidget[]):
                ParentWidget.SetVisibility(widget_visibility.Visible)

    UpdateStatText(TextBlock:text_block, Value:float, AbilityId:ability_id, BanishedIds:[]ability_id):void =
        TextBlock.SetText(MakeMessage("{Value > 0.0 and "+" or ""}{FormatAsPercent(Value)}"))
        if (BanishedIds.Find[AbilityId]):
            TextBlock.SetColor(ThemeColors.Failure)
        else:
            if (Value > 0.0):
                TextBlock.SetColor(ThemeColors.Success)
            else:
                TextBlock.SetTextColor(ThemeColors.Inactive)
                TextBlock.SetShadowOpacity(0.0)

    SetInactive():void =
        PlayerWeaponIcon.SetImage(UserInterface.Icons.Lock_64)
        PlayerWeaponIcon.SetTint(ThemeColors.Inactive)
        DamageText.SetText(MakeMessage(""))
        CooldownText.SetText(MakeMessage(""))
        SizeText.SetText(MakeMessage(""))

        if (ParentWidget := LevelText.GetParentWidget[]):
            ParentWidget.SetVisibility(widget_visibility.Hidden)

        for (EvolutionIcon : EvolutionIcons):
            EvolutionIcon.SetVisibility(widget_visibility.Hidden)

        if (ParentWidget := DPSText.GetParentWidget[]):
            ParentWidget.SetVisibility(widget_visibility.Hidden)