using { /Fortnite.com/Characters }
using { /Fortnite.com/Devices }
using { /Fortnite.com/Devices/CreativeAnimation }
using { /Verse.org/Colors }
using { /Verse.org/Random }
using { /Verse.org/SceneGraph }
using { /Verse.org/Simulation }
using { /UnrealEngine.com/Temporary/SpatialMath }
using { VisualEffects }

<# ----------------------------------------------------------------------------------------------------
    Constants
---------------------------------------------------------------------------------------------------- #>
DefaultGatherDuration:float = 5.0
DefaultGatherableRespawnDuration:float = 90.0
DefaultGatherableItemDropRate:float = 0.1

<# ----------------------------------------------------------------------------------------------------
    Data Definitions
---------------------------------------------------------------------------------------------------- #>
gatherable_state := enum:
    Enabled
    Disabled
    Available
    Gathering
    Respawning
    
<# ----------------------------------------------------------------------------------------------------
    gatherable
---------------------------------------------------------------------------------------------------- #>
gatherable := class<concrete>:
    @editable
    Prop:creative_prop = creative_prop{}

    @editable
    VolumeDevice:volume_device = volume_device{}
    
    @editable
    MapIndicator:map_indicator_device = map_indicator_device{}

    GatheredEvent:events(player) = events(player){}
    GatherStoppedEvent:events(player) = events(player){}
    GatherCancelledEvent:events(tuple()) = events(tuple()){}
    RespawnCancelledEvent:events(tuple()) = events(tuple()){}
    
    var State:gatherable_state = gatherable_state.Available
    var InitialTransform:transform = transform{}
    var InitialVolumeTransform:transform = transform{}
    var LoadedTransform:transform = transform{}
    var MaybeSpawnedItemDrop:?item_drop = false
    var Progress:float = 0.0
    var IsRespawn:logic = true
    var GatheringParticleSystem:NS_Gathering_V2 = NS_Gathering_V2{Entity := entity{}}
  
    OnBegin()<suspends>:void =
        set InitialTransform = Prop.GetTransform()
        set InitialVolumeTransform = VolumeDevice.GetTransform()

        Disable()

        VolumeDevice.AgentEntersEvent.Subscribe(OnAgentEnters)
        VolumeDevice.AgentExitsEvent.Subscribe(OnAgentExits)

        GatheringEntity := entity{}
        NewGatheringParticleSystem := NS_Gathering_V2{Entity := GatheringEntity, Enabled := false}
        GatheringEntity.AddComponents(array{NewGatheringParticleSystem})
        GatheringEntity.Initialize()
        set GatheringParticleSystem = NewGatheringParticleSystem

        if (GameManager := GetGameManager[]):
            GameManager.GameEndedEvent.Subscribe(OnGameEndedEvent)
            GameManager.BossFightStartedEvent.Subscribe(OnBossFightStartedEvent)

    <# ----------------------------------------------------------------------------------------------------
        Get Functions
    ---------------------------------------------------------------------------------------------------- #>
    GetBaseGatherDuration()<transacts>:float =
        DefaultGatherDuration

    <# ----------------------------------------------------------------------------------------------------
        Update Functions
    ---------------------------------------------------------------------------------------------------- #>
    Load(Transform:transform):void =
        if:
            Prop.TeleportTo[Transform.Translation, Transform.Rotation]
            VolumeDevice.TeleportTo[Transform.Translation, Transform.Rotation]
        then:
            set LoadedTransform = Prop.GetTransform()
            Enable()

    StartGather(Player:player, ?Delay:float = -1.0)<suspends>:void =
        Sleep(Delay)

        set State = gatherable_state.Gathering

        GatherStartTime := GetSimulationElapsedTime()

        var HasChainsaw:logic = false
        var GatherRateMultiplier:float = 0.0
        for (PlayerGathering : VolumeDevice.GetPlayersInVolume()):
            set GatherRateMultiplier += PlayerGathering.GetStat(stat_id.GatherRate) * 2.0
            if (PlayerGathering.HasBoon(boon_id.Chainsaw)?):
                set HasChainsaw = true

        InitialProgress := Progress
        GatherRateBoonMultiplier := HasChainsaw? and 2.0 or 1.0
        BaseGatherDuration := GetBaseGatherDuration() / GetGameSpeed()
        GatherDurationFactor := BaseGatherDuration / GetBaseGatherDuration()
        GatherDuration := BaseGatherDuration / (1.0 + GatherRateMultiplier) / GatherRateBoonMultiplier / (VolumeDevice.GetPlayersInVolume().Length * 1.0)
        GatherFactor := GatherDuration / BaseGatherDuration
        GatherDurationRemaining := GatherDuration * (1.0 - Progress)
        ProgressPerHit := 0.2 * (1.0 / GatherDurationFactor) * (DefaultGatherDuration / GetBaseGatherDuration())
        
        TierColor := if (GatherDuration <= GatherDurationFactor * 1.0 / 5.0 * GetBaseGatherDuration()) then ThemeColors.Legendary
        else if (GatherDuration <= GatherDurationFactor * 2.0 / 5.0 * GetBaseGatherDuration()) then ThemeColors.Epic
        else if (GatherDuration <= GatherDurationFactor * 3.0 / 5.0 * GetBaseGatherDuration()) then ThemeColors.Rare
        else if (GatherDuration <= GatherDurationFactor * 4.0 / 5.0 * GetBaseGatherDuration()) then ThemeColors.Uncommon
        else ThemeColors.Common

        set GatheringParticleSystem.Position = FromVector3 of Prop.GetTransform().Translation
        set GatheringParticleSystem.Color = TierColor * 2.0
        set GatheringParticleSystem.GatherFactor = GatherFactor
        set GatheringParticleSystem.IsChainsaw = HasChainsaw

        GatheringParticleSystem.Enable()

        race:
            block:
                Sleep(GatherDurationRemaining)
                OnGathered(Player)
                
            block:
                GatherCancelledEvent.Await()
                set Progress = InitialProgress + (GetSimulationElapsedTime() - GatherStartTime) / GatherDuration
                set State = gatherable_state.Available
                if (Prop.TeleportTo[transform{Translation := Prop.GetTransform().Translation, Rotation := LoadedTransform.Rotation, Scale := Prop.GetTransform().Scale}]) {}

            if (AnimationController := Prop.GetAnimationController[], Character := Player.GetFortCharacter[]):
                loop:
                    if (Character.IsActive[]):
                        Direction := Prop.GetTransform().Translation.Make2D() - Character.GetTransform().Translation.Make2D()
                        RotationVector := MakeRotationFromYawPitchRollDegrees(90.0, 0.0, 0.0).RotateVector(Direction)
                        CurrentScale := Prop.GetTransform().Scale
                        TargetScale:vector3 = InitialTransform.Scale - (InitialTransform.Scale * 0.6) * Clamp(Progress + ProgressPerHit, 0.0, 1.0)
                        DeltaRotation := MakeRotation(RotationVector, DegreesToRadians(15.0))
                        DeltaScale01 := UnitVector - (CurrentScale - TargetScale) * 0.25 / CurrentScale
                        DeltaScale02 := UnitVector - (CurrentScale - TargetScale) * 0.75 / (CurrentScale * DeltaScale01)
                        
                        KeyFrames := array:
                            keyframe_delta:
                                DeltaLocation := vector3{X := 0.0, Y := 0.0, Z := 0.0}
                                DeltaRotation := DeltaRotation
                                DeltaScale := DeltaScale01
                                Interpolation := InterpolationTypes.EaseOut
                                Time := 0.075 * GatherFactor
                            keyframe_delta:
                                DeltaLocation := vector3{X := 0.0, Y := 0.0, Z := 0.0}
                                DeltaRotation := DeltaRotation.Invert()
                                DeltaScale := DeltaScale02
                                Interpolation := InterpolationTypes.EaseOut
                                Time := 0.225 * GatherFactor
            
                        AnimationController.SetAnimation(KeyFrames, ?Mode := animation_mode.OneShot)
                        AnimationController.Play()
                        AnimationController.MovementCompleteEvent.Await()
                        Sleep(0.7 * GatherFactor)
                        set Progress += ProgressPerHit
                    else:
                        GatherCancelledEvent.Signal()
                        Sleep(Inf)

        GatheringParticleSystem.Disable()

    RefreshGather(Player:player):void =
        GatherCancelledEvent.Signal()
        spawn{StartGather(Player, ?Delay := 0.0)}

    Respawn()<suspends>:void =
        defer:
            if (IsRespawn?):
                Reset()
            else:
                Disable()

        race:
            RespawnCancelledEvent.Await()
            block:
                set State = gatherable_state.Respawning
                Prop.Hide()

                if:
                    VolumeDevice.TeleportTo[InitialVolumeTransform]
                    MapIndicator.TeleportTo[MapIndicatorDisabledPosition, IdentityRotation()]
                
                Sleep(DefaultGatherableRespawnDuration / (Max(1, GetActivePlayers().Length) * 1.0) / GetGameSpeed())
                if (Prop.TeleportTo[transform{Translation := Prop.GetTransform().Translation, Rotation := Prop.GetTransform().Rotation, Scale := InitialTransform.Scale * 0.6}]) {}
                Prop.Show()

                if (SpawnedItemDrop := MaybeSpawnedItemDrop?):
                    SpawnedItemDrop.Reposition()
                    set MaybeSpawnedItemDrop = false

                if (PlayAnimationController := Prop.GetAnimationController[]):
                    DeltaScale := InitialTransform.Scale / Prop.GetTransform().Scale
                    KeyFrames := array:
                        keyframe_delta:
                            DeltaLocation := vector3{}
                            DeltaRotation := rotation{}
                            DeltaScale := DeltaScale
                            Time := 1.0
                            Interpolation := InterpolationTypes.EaseIn

                    PlayAnimationController.SetAnimation(KeyFrames, ?Mode := animation_mode.OneShot)
                    PlayAnimationController.Play()
                    PlayAnimationController.MovementCompleteEvent.Await()
  
    Reset():void =
        if (State = gatherable_state.Disabled) then return
        RespawnCancelledEvent.Signal()

        set State = gatherable_state.Available
        set Progress = 0.0

        TargetTransform := transform{Translation := LoadedTransform.Translation, Rotation := LoadedTransform.Rotation, Scale := InitialTransform.Scale}
        if:
            Prop.TeleportTo[TargetTransform]
            MapIndicator.TeleportTo[TargetTransform]
            VolumeDevice.TeleportTo[TargetTransform.Translation, TargetTransform.Rotation]
        then:
            Prop.Show()

        if (Player := VolumeDevice.GetPlayersInVolume()[0]):
            spawn{StartGather(Player)}

    Enable():void =
        set State = gatherable_state.Enabled
        set IsRespawn = true
        Reset()

    Disable():void =
        GatherCancelledEvent.Signal()
        RespawnCancelledEvent.Signal()

        set State = gatherable_state.Disabled

        if:
            VolumeDevice.TeleportTo[InitialVolumeTransform]
            MapIndicator.TeleportTo[MapIndicatorDisabledPosition, IdentityRotation()]

        Prop.Hide()
        
    DisableRespawn():void =
        set IsRespawn = false
        RespawnCancelledEvent.Signal()

    <# ----------------------------------------------------------------------------------------------------
        Event Handlers
    ---------------------------------------------------------------------------------------------------- #>
    OnAgentEnters(Agent:agent):void =
        if (Player := player[Agent]):
            if (Player.GetIsAFK()?) then return
            if (Character := Player.GetFortCharacter[], Character.IsDownButNotOut[]) then return

            if (State = gatherable_state.Available):
                spawn{StartGather(Player)}
                spawn{AwaitAbilityChangedEvent(Player)}
            else if (State = gatherable_state.Gathering):
                RefreshGather(Player)
                spawn{AwaitAbilityChangedEvent(Player)}
                
    OnAgentExits(Agent:agent):void =
        if (Player := player[Agent]):
            GatherStoppedEvent.Signal(Player)
            if (VolumeDevice.IsEmpty[]):
                GatherCancelledEvent.Signal()
            else if (State = gatherable_state.Gathering):
                RefreshGather(Player)
            
    OnGathered(Player:player):void =
        var GatherMultiplier:float = 0.0
        for (PlayerGathering : VolumeDevice.GetPlayersInVolume()):
            PlayerGatherMultiplier := PlayerGathering.GetStat(stat_id.GatherRate)
            if (PlayerGatherMultiplier > GatherMultiplier):
                set GatherMultiplier = PlayerGatherMultiplier

            GatherStoppedEvent.Signal(PlayerGathering)

        if (GatherAmount := Floor[GetRandomFloat(30.0, 60.0) * (1.0 + GatherMultiplier) / (Max(1, GetActivePlayers().Length) * 1.0)] * 1.0):
            AddGameCurrency(game_currency_id.Wood, GatherAmount)
            
            if (VFXManager := GetVFXManager[]):
                SpawnNumberVFX(VFXManager.GatherNumberParticleSystem, Prop.GetTransform().Translation, GatherAmount)

            if (PlayerGathering := VolumeDevice.GetPlayersInVolume()[0]):
                PlayerGathering.AddGameStat(player_game_stat_id.WoodCollected, GatherAmount)

        IsDrop := GetPsuedoRandomResult(pseudo_random_id.HealthDrop)
        if (IsDrop?):
            set MaybeSpawnedItemDrop = SpawnItemDrop(item_drop_id.Health, Prop.GetTransform().Translation)
            SpawnExperience(GetRandomPositionOnCircle(150.0, ?Origin := Prop.GetTransform().Translation), 20)

            for (PlayerGathering : VolumeDevice.GetPlayersInVolume()):
                PlayerGathering.PlayAudio(audio_id.Surprise)
        else:
            SpawnExperience(Prop.GetTransform().Translation, 20)

        AwardAccolade(accolade_count_id.GatherTree, GetActivePlayers())

        if (IsRespawn?):
            spawn{Respawn()}
        else:
            Disable()

        GatheredEvent.Signal(Player)

    OnGameEndedEvent():void =
        Disable()

    OnBossFightStartedEvent():void =
        Disable()

    AwaitAbilityChangedEvent(Player:player)<suspends>:void =
        if (AbilityChangedEvent := Player.GetStats[].AbilityChangedEvent):
            race:
                RespawnCancelledEvent.Await()

                loop:
                    PlayerStopped := GatherStoppedEvent.Await()
                    if (PlayerStopped = Player):
                        break

                loop:
                    AbilityId := AbilityChangedEvent.Await()
                    if (AbilityId = ability_id.GatherRate):
                        RefreshGather(Player)