using { /Verse.org/Random }
using { /Verse.org/Simulation }
using { /UnrealEngine.com/Temporary/SpatialMath }

<# ----------------------------------------------------------------------------------------------------
    Constants
---------------------------------------------------------------------------------------------------- #>
DefaultGoldenTreeGatherDuration:float = 50.0
GoldenTreeGoldReward:float = 5000.0

<# ----------------------------------------------------------------------------------------------------
    gatherable_golden_tree
---------------------------------------------------------------------------------------------------- #>
gatherable_golden_tree := class<concrete>(gatherable):

    <# ----------------------------------------------------------------------------------------------------
        Get Functions
    ---------------------------------------------------------------------------------------------------- #>
    GetBaseGatherDuration<override>()<transacts>:float =
        DefaultGoldenTreeGatherDuration

    <# ----------------------------------------------------------------------------------------------------
        Update Functions
    ---------------------------------------------------------------------------------------------------- #>
    Load<override>(Transform:transform):void =        
        if (GetCurrentStage() <= 1) then return
        if (GetRandomFloat(0.0, 1.0) > Average of for (Player : GetActivePlayers()) {Player.GetPowerUpValue(power_up_id.GoldenTreeRate)}) then return

        (super:)Load(Transform)

    <# ----------------------------------------------------------------------------------------------------
        Event Handlers
    ---------------------------------------------------------------------------------------------------- #>
    OnGathered<override>(Player:player):void =
        var GatherMultiplier:float = 0.0
        for (PlayerGathering : VolumeDevice.GetPlayersInVolume()):
            PlayerGatherMultiplier := Player.GetStat(stat_id.GatherRate)
            if (PlayerGatherMultiplier > GatherMultiplier):
                set GatherMultiplier = PlayerGatherMultiplier

            GatherStoppedEvent.Signal(PlayerGathering)

        if (GatherAmount := Floor[GetRandomFloat(30.0, 60.0) * (1.0 + GatherMultiplier) / (Max(1, GetActivePlayers().Length) * 1.0)] * 1.0 * 15.0):
            AddGameCurrency(game_currency_id.Wood, GatherAmount)
            SpawnVFX(vfx_id.Gather, Prop.GetTransform().Translation, ?Scale := GetDamageNumberScale(GatherAmount), ?Duration := 2.0)

            if (PlayerGathering := VolumeDevice.GetPlayersInVolume()[0]):
                PlayerGathering.AddGameStat(player_game_stat_id.WoodCollected, GatherAmount)

        AddGameStat(game_stat_id.GoldenTree, 1.0)
        AwardAccolade(accolade_id.GoldenTree, GetActivePlayers())

        Disable()