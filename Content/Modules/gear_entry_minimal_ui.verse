using { /Fortnite.com/UI }
using { /Verse.org/Colors }
using { /Verse.org/Simulation }
using { /UnrealEngine.com/Temporary/SpatialMath }
using { /UnrealEngine.com/Temporary/UI }

<# ----------------------------------------------------------------------------------------------------
    gear_entry_minimal_ui
---------------------------------------------------------------------------------------------------- #>
gear_entry_minimal_ui := class(ui_component):
    EntryIndex:int

    Icon:texture_block = texture_block:
        DefaultImage := UserInterface.Gears.Gear_01_Head_64
        DefaultDesiredSize := vector2{X := 64.0, Y := 64.0}

    var GearOverlay:overlay = overlay{}
    var MaybeTierWidget:?widget = false
    var TierStarIcons:[]texture_block = array{}

    <# ----------------------------------------------------------------------------------------------------
        Initialization
    ---------------------------------------------------------------------------------------------------- #>
    CreateUI<override>():widget =
        set GearOverlay = overlay:
            Slots := array:
                overlay_slot:
                    HorizontalAlignment := horizontal_alignment.Center
                    VerticalAlignment := vertical_alignment.Center
                    Padding := margin{Left := 20.0, Top := 20.0, Right := 20.0, Bottom := 20.0}
                    Widget := Icon
                overlay_slot:
                    VerticalAlignment := vertical_alignment.Bottom
                    Widget := stack_box:
                        Orientation := orientation.Horizontal
                        Slots := for (Index := 1..5):
                            TierStarIcon := texture_block:
                                DefaultImage := UserInterface.Icons.Star_64
                                DefaultTint := ThemeColors.Highlight
                                DefaultDesiredSize := vector2{X := 16.0, Y := 16.0}

                            set TierStarIcons += array{TierStarIcon}

                            stack_box_slot:
                                HorizontalAlignment := horizontal_alignment.Fill
                                VerticalAlignment := vertical_alignment.Fill
                                Widget := TierStarIcon

    <# ----------------------------------------------------------------------------------------------------
        Update Functions
    ---------------------------------------------------------------------------------------------------- #>
    Hide<override>():void =
        UI.SetVisibility(widget_visibility.Hidden)

    Update(InPlayer:player):void =
        if (GearData := InPlayer.GetGear[EntryIndex]):
            if (GearInfo := GetGearInfos()[GearData.Evolution][EntryIndex.ToGearSlot()]):
                Icon.SetImage(GearInfo.IconSmall)
                Icon.SetTint(NamedColors.White)

            SetTier(GearData.Tier, GearData.Stars)
            Show()
        else:
            Hide()

    SetTier(Tier:int, Stars:int):void =
        if (TierWidget := MaybeTierWidget?):
            GearOverlay.RemoveWidget of TierWidget

        UpdatedTierWidget := Tier.ToGearTierWidget()
        set MaybeTierWidget = option{UpdatedTierWidget}

        GearOverlay.AddWidget of overlay_slot:
            HorizontalAlignment := horizontal_alignment.Left
            VerticalAlignment := vertical_alignment.Top
            Widget := UpdatedTierWidget

        for (Index -> TierStarIcon : TierStarIcons):
            TierStarIcon.SetVisibility(Stars >= (Index + 1) and widget_visibility.Visible or widget_visibility.Collapsed)