using { /Fortnite.com/UI }
using { /Verse.org/Colors }
using { /Verse.org/Simulation }
using { /UnrealEngine.com/Temporary/SpatialMath }
using { /UnrealEngine.com/Temporary/UI }

<# ----------------------------------------------------------------------------------------------------
    gear_entry_ui
---------------------------------------------------------------------------------------------------- #>
gear_entry_ui := class(ui_component):
    EntryIndex:int

    SelectedEvent:events(int) = events(int){}

    Icon:texture_block = texture_block:
        DefaultImage := UserInterface.Gears.Gear_01_Head_128
        DefaultDesiredSize := vector2{X := 128.0, Y := 128.0}

    LevelText:text_block = text_block:       
        DefaultTextColor := ThemeColors.PrimaryText
        DefaultShadowColor := NamedColors.Black
        DefaultShadowOffset := option{vector2{X := 2.0, Y := 2.0}}
        DefaultShadowOpacity := 0.5

    QuantityText:text_block = text_block:
        DefaultTextColor := ThemeColors.PrimaryText
        DefaultShadowColor := NamedColors.Black
        DefaultShadowOffset := option{vector2{X := 2.0, Y := 2.0}}
        DefaultShadowOpacity := 0.5

    SelectButton:button_quiet = button_quiet{}

    var GearOverlay:overlay = overlay{}
    var MaybeTierWidget:?widget = false
    var TierStarIcons:[]texture_block = array{}
    var TierUI:widget = canvas{}
    var InfoUI:widget = canvas{}

    <# ----------------------------------------------------------------------------------------------------
        Initialization
    ---------------------------------------------------------------------------------------------------- #>
    CreateUI<override>():widget =
        SelectButton.OnClick().Subscribe(OnClickSelect)

        stack_box:
            Orientation := orientation.Vertical
            Slots := array:
                stack_box_slot:
                    HorizontalAlignment := horizontal_alignment.Fill
                    VerticalAlignment := vertical_alignment.Fill
                    Distribution := option{1.0}
                    Widget := overlay:
                        Slots := array:
                            overlay_slot:
                                HorizontalAlignment := horizontal_alignment.Fill
                                VerticalAlignment := vertical_alignment.Fill
                                Widget := UserInterface.VerseWidget.WBP_RoundedBox_Light{}
                            overlay_slot:
                                HorizontalAlignment := horizontal_alignment.Fill
                                VerticalAlignment := vertical_alignment.Fill
                                Padding := margin{Left := 40.0, Top := 20.0, Right := 40.0, Bottom := 20.0}
                                Widget := CreateContent()
                stack_box_slot:
                    HorizontalAlignment := horizontal_alignment.Fill
                    VerticalAlignment := vertical_alignment.Fill
                    Widget := SelectButton

    CreateContent():widget =
        stack_box:
            Orientation := orientation.Horizontal
            Slots := array:
                stack_box_slot:
                    HorizontalAlignment := horizontal_alignment.Left
                    VerticalAlignment := vertical_alignment.Fill
                    Distribution := option{1.0}
                    Widget := set GearOverlay = overlay:
                        Slots := array:
                            overlay_slot:
                                HorizontalAlignment := horizontal_alignment.Center
                                VerticalAlignment := vertical_alignment.Center
                                Widget := Icon
                            overlay_slot:
                                VerticalAlignment := vertical_alignment.Bottom
                                Widget := set TierUI = stack_box:
                                    Orientation := orientation.Horizontal
                                    Slots := for (Index := 1..5):
                                        TierStarIcon := texture_block:
                                            DefaultImage := UserInterface.Icons.Star_64
                                            DefaultTint := ThemeColors.Highlight
                                            DefaultDesiredSize := vector2{X := 24.0, Y := 24.0}

                                        set TierStarIcons += array{TierStarIcon}

                                        stack_box_slot:
                                            HorizontalAlignment := horizontal_alignment.Fill
                                            VerticalAlignment := vertical_alignment.Fill
                                            Widget := TierStarIcon
                stack_box_slot:
                    HorizontalAlignment := horizontal_alignment.Right
                    VerticalAlignment := vertical_alignment.Fill
                    Distribution := option{1.0}
                    Widget := set InfoUI = stack_box:
                        Orientation := orientation.Vertical
                        Slots := array:
                            stack_box_slot:
                                HorizontalAlignment := horizontal_alignment.Right
                                VerticalAlignment := vertical_alignment.Top
                                Distribution := option{1.0}
                                Padding := margin{Left := 0.0, Top := 0.0, Right := 0.0, Bottom := 0.0}
                                Widget := stack_box:
                                    Orientation := orientation.Horizontal
                                    Slots := array:
                                        stack_box_slot:
                                            HorizontalAlignment := horizontal_alignment.Fill
                                            VerticalAlignment := vertical_alignment.Fill
                                            Widget := text_block:
                                                DefaultText := MakeMessage("Lv.")
                                                DefaultTextColor := ThemeColors.Highlight
                                                DefaultShadowColor := ThemeColors.Highlight * 0.125
                                                DefaultShadowOffset := option{vector2{X := 2.0, Y := 2.0}}
                                                DefaultShadowOpacity := 1.0
                                        stack_box_slot:
                                            HorizontalAlignment := horizontal_alignment.Fill
                                            VerticalAlignment := vertical_alignment.Fill
                                            Padding := margin{Left := 10.0, Top := 0.0, Right := 0.0, Bottom := 0.0}
                                            Widget := LevelText
                            stack_box_slot:
                                HorizontalAlignment := horizontal_alignment.Right
                                VerticalAlignment := vertical_alignment.Bottom
                                Widget := QuantityText

    <# ----------------------------------------------------------------------------------------------------
        Update Functions
    ---------------------------------------------------------------------------------------------------- #>
    Update<override>():void =
        if (GearData := Player.GetGear[EntryIndex]):
            if (GearInfo := GetGearInfos()[GearData.Evolution][EntryIndex.ToGearSlot()]):
                Icon.SetImage(GearInfo.Icon)
                Icon.SetTint(NamedColors.White)

            LevelText.SetText(MakeMessage("{GearData.Level}"))
            QuantityText.SetText(MakeMessage("{FormatAsCurrency(GearData.Quantity * 1.0)}"))
            SetTier(GearData.Tier, GearData.Stars)
            TierUI.SetVisibility(widget_visibility.Visible)
            InfoUI.SetVisibility(widget_visibility.Visible)
        else:
            TierUI.SetVisibility(widget_visibility.Hidden)
            InfoUI.SetVisibility(widget_visibility.Hidden)
            if (GearInfo := GetGearInfos()[0][EntryIndex.ToGearSlot()]):
                Icon.SetImage(GearInfo.Icon)
                Icon.SetTint(ThemeColors.Inactive)

    SetTier(Tier:int, Stars:int):void =
        if (TierWidget := MaybeTierWidget?):
            GearOverlay.RemoveWidget of TierWidget

        UpdatedTierWidget := Tier.ToGearTierWidget()
        set MaybeTierWidget = option{UpdatedTierWidget}

        GearOverlay.AddWidget of overlay_slot:
            HorizontalAlignment := horizontal_alignment.Left
            VerticalAlignment := vertical_alignment.Top
            Widget := UpdatedTierWidget

        for (Index -> TierStarIcon : TierStarIcons):
            TierStarIcon.SetVisibility(Stars >= (Index + 1) and widget_visibility.Visible or widget_visibility.Collapsed)

    SetActive(IsActive:logic):void =
        SelectButton.SetText(MakeMessage(IsActive? and "Selected" or ""))
        
    <# ----------------------------------------------------------------------------------------------------
        Event Handlers
    ---------------------------------------------------------------------------------------------------- #>
    OnClickSelect(Message:widget_message):void =
        SelectedEvent.Signal(EntryIndex)