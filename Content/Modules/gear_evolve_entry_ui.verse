using { /Fortnite.com/UI }
using { /Verse.org/Colors }
using { /UnrealEngine.com/Temporary/SpatialMath }
using { /UnrealEngine.com/Temporary/UI }

<# ----------------------------------------------------------------------------------------------------
    gear_evolve_entry_ui
---------------------------------------------------------------------------------------------------- #>
gear_evolve_entry_ui := class(ui_component):
    Icon:texture_block = texture_block:
        DefaultImage := UserInterface.Gears.Gear_01_Head_128
        DefaultDesiredSize := vector2{X := 128.0, Y := 128.0}

    NameText:text_block = text_block:
        DefaultTextColor := ThemeColors.PrimaryText
        DefaultShadowColor := NamedColors.Black
        DefaultShadowOffset := option{vector2{X := 2.0, Y := 2.0}}
        DefaultShadowOpacity := 0.5

    LevelText:text_block = text_block:
        DefaultTextColor := ThemeColors.PrimaryText
        DefaultShadowColor := NamedColors.Black
        DefaultShadowOffset := option{vector2{X := 2.0, Y := 2.0}}
        DefaultShadowOpacity := 0.5

    StatIcon:texture_block = texture_block:
        DefaultImage := UserInterface.Icons.ProjectileDamage_64
        DefaultDesiredSize := vector2{X := 40.0, Y := 40.0}

    StatText:text_block = text_block:
        DefaultTextColor := ThemeColors.Success
        DefaultShadowColor := ThemeColors.Success * 0.125
        DefaultShadowOffset := option{vector2{X := 2.0, Y := 2.0}}
        DefaultShadowOpacity := 1.0

    TierStatIcon:texture_block = texture_block:
        DefaultImage := UserInterface.Icons.ProjectileDamage_64
        DefaultDesiredSize := vector2{X := 40.0, Y := 40.0}

    TierStatText:text_block = text_block:
        DefaultTextColor := ThemeColors.Success
        DefaultShadowColor := ThemeColors.Success * 0.125
        DefaultShadowOffset := option{vector2{X := 2.0, Y := 2.0}}
        DefaultShadowOpacity := 1.0

    var MaybeGearTierUI:?gear_tier_ui = false

    <# ----------------------------------------------------------------------------------------------------
        Initialization
    ---------------------------------------------------------------------------------------------------- #>
    CreateUI<override>():widget =
        GearTierUI := gear_tier_ui{Player := Player}
        GearTierUI.Initialize()
        set MaybeGearTierUI = option{GearTierUI}

        overlay:
            Slots := array:
                overlay_slot:
                    HorizontalAlignment := horizontal_alignment.Fill
                    VerticalAlignment := vertical_alignment.Fill
                    Widget := UserInterface.VerseWidget.WBP_RoundedBox_Light{}
                overlay_slot:
                    HorizontalAlignment := horizontal_alignment.Fill
                    VerticalAlignment := vertical_alignment.Fill
                    Padding := margin{Left := 20.0, Top := 20.0, Right := 20.0, Bottom := 20.0}
                    Widget := stack_box:
                        Orientation := orientation.Vertical
                        Slots := array:
                            stack_box_slot:
                                VerticalAlignment := vertical_alignment.Top
                                Widget := Icon
                            stack_box_slot:
                                HorizontalAlignment := horizontal_alignment.Fill
                                VerticalAlignment := vertical_alignment.Fill
                                Widget := NameText
                            stack_box_slot:
                                HorizontalAlignment := horizontal_alignment.Fill
                                VerticalAlignment := vertical_alignment.Fill
                                Widget := stack_box:
                                    Orientation := orientation.Horizontal
                                    Slots := array:
                                        stack_box_slot:
                                            HorizontalAlignment := horizontal_alignment.Fill
                                            VerticalAlignment := vertical_alignment.Fill
                                            Widget := text_block:
                                                DefaultText := MakeMessage("Lv.")
                                                DefaultTextColor := ThemeColors.Highlight
                                                DefaultShadowColor := ThemeColors.Highlight * 0.125
                                                DefaultShadowOffset := option{vector2{X := 2.0, Y := 2.0}}
                                                DefaultShadowOpacity := 1.0
                                        stack_box_slot:
                                            HorizontalAlignment := horizontal_alignment.Fill
                                            VerticalAlignment := vertical_alignment.Fill
                                            Padding := margin{Left := 10.0, Top := 0.0, Right := 0.0, Bottom := 0.0}
                                            Widget := LevelText
                            stack_box_slot:
                                HorizontalAlignment := horizontal_alignment.Fill
                                VerticalAlignment := vertical_alignment.Fill
                                Widget := GearTierUI.GetUI()
                            stack_box_slot:
                                HorizontalAlignment := horizontal_alignment.Fill
                                VerticalAlignment := vertical_alignment.Fill
                                Widget := stack_box:
                                    Orientation := orientation.Horizontal
                                    Slots := array:
                                        stack_box_slot:
                                            HorizontalAlignment := horizontal_alignment.Fill
                                            VerticalAlignment := vertical_alignment.Fill
                                            Widget := StatIcon
                                        stack_box_slot:
                                            HorizontalAlignment := horizontal_alignment.Fill
                                            VerticalAlignment := vertical_alignment.Fill
                                            Padding := margin{Left := 10.0, Top := 0.0, Right := 0.0, Bottom := 0.0}
                                            Widget := StatText
                            stack_box_slot:
                                HorizontalAlignment := horizontal_alignment.Fill
                                VerticalAlignment := vertical_alignment.Fill
                                Widget := stack_box:
                                    Orientation := orientation.Horizontal
                                    Slots := array:
                                        stack_box_slot:
                                            HorizontalAlignment := horizontal_alignment.Fill
                                            VerticalAlignment := vertical_alignment.Fill
                                            Widget := TierStatIcon
                                        stack_box_slot:
                                            HorizontalAlignment := horizontal_alignment.Fill
                                            VerticalAlignment := vertical_alignment.Fill
                                            Padding := margin{Left := 10.0, Top := 0.0, Right := 0.0, Bottom := 0.0}
                                            Widget := TierStatText
        
    <# ----------------------------------------------------------------------------------------------------
        Update Functions
    ---------------------------------------------------------------------------------------------------- #>
    Update(GearData:player_gear, Slot:gear_slot):void =
        if (GearInfo := GetGearInfo[GearData.Evolution, Slot], GearTierUI := MaybeGearTierUI?):
            Icon.SetImage(GearInfo.Icon)
            NameText.SetText(MakeMessage("{GearInfo.Name.WrapText(250.0)}"))
            LevelText.SetText(MakeMessage("{GearData.Level}/{GearData.GetMaxLevel()}"))
            GearTierUI.SetTier(GearData.Tier, GearData.Stars)
            SlotInfo := GetGearSlotInfo(Slot)
            StatIcon.SetImage(SlotInfo.Stat.ToIcon())
            StatText.SetText(MakeMessage("+{GearData.GetStat(Slot, GearData.Level).Format(SlotInfo.Stat)}"))
            TierStatIcon.SetImage(SlotInfo.TierStat.ToIcon())
            TierStatText.SetText(MakeMessage("+{GearData.GetTierStats(Slot, GearData.Tier).Format(SlotInfo.TierStat)}"))