using { /Verse.org/Assets }
using { /Verse.org/Random }
using { /Verse.org/Simulation }
using { /UnrealEngine.com/Temporary/UI }

<# ----------------------------------------------------------------------------------------------------
    Constants
---------------------------------------------------------------------------------------------------- #>
GearTreasureChestWeights:[]int = array{60, 25, 10, 4, 1}
TreasureChestRecipeRewards:[]int = array{3, 6, 9, 15, 30}
TreasureChestGearRewards:[]int = array{1, 2, 3, 5, 10}
MaxGearStars:int = 5
GearSetBonusStat:stat_id = stat_id.Luck

GearBaseUpgradeGoldCostExponent:float = 0.0275
GearBaseUpgradeGoldCosts:[]int = array{150, 975, 4034, 14340, 47238, 148643, 453649, 1354581, 3528631}
GearBaseUpgradeRecipeCostExponent:float = 0.0175
GearBaseUpgradeRecipeCosts:[]int = array{3, 9, 22, 49, 104, 215, 437, 877, 1613}

GearBaseTierGoldCostExponent:float = 0.185
GearBaseTierGoldCosts:[]int = array{3000, 13822, 47694, 148675, 440037, 1262040, 3543892, 9801691, 23906730}
GearBaseTierRecipeCostExponent:float = 0.135
GearBaseTierRecipeCosts:[]int = array{10, 31, 77, 177, 392, 848, 1806, 3804, 7305}
GearBaseTierGearCostExponent:float = 0.13
GearBaseTierGearCosts:[]int = array{1, 3, 7, 15, 32, 67, 138, 282, 528}

GearBaseEvolveGoldCosts:[]int = array{41466, 143082, 446025, 1320111, 3786120, 10631676, 29405073, 71720190, 158588184}
GearBaseEvolveRecipeCosts:[]int = array{93, 231, 531, 1176, 2544, 5418, 11412, 21915, 39165}

<# ----------------------------------------------------------------------------------------------------
    Data Definitions
---------------------------------------------------------------------------------------------------- #>
gear_info<public> := struct<computes>:
    Name:string
    Icon:texture
    IconSmall:texture

gear_slot<public> := enum:
    Head
    Chest
    Pants
    Shoulder
    Gloves
    Boots

(Slot:gear_slot).ToInt()<transacts>:int =
    case (Slot):
        gear_slot.Head => 0
        gear_slot.Chest => 1
        gear_slot.Pants => 2
        gear_slot.Shoulder => 3
        gear_slot.Gloves => 4
        gear_slot.Boots => 5

(Slot:int).ToGearSlot()<transacts>:gear_slot =
    case (Slot):
        0 => gear_slot.Head
        1 => gear_slot.Chest
        2 => gear_slot.Pants
        3 => gear_slot.Shoulder
        4 => gear_slot.Gloves
        5 => gear_slot.Boots
        _ => gear_slot.Head

upgrade_cost := struct:
    Id:currency_id
    Amount:float

gear_upgrade_cost := struct:
    UpgradeCosts:[]upgrade_cost
    GearCost:int

gear_slot_info := struct:
    Stat:stat_id
    StatFactor:float := 1.0
    TierStat:stat_id
    TierStatFactor:float := 1.0

<# ----------------------------------------------------------------------------------------------------
    Global Functions
---------------------------------------------------------------------------------------------------- #>
(Tier:int).ToGearTierWidget()<transacts>:widget =
    case(Tier):
        1 => UserInterface.VerseWidget.WBP_TierText_G{}
        2 => UserInterface.VerseWidget.WBP_TierText_F{}
        3 => UserInterface.VerseWidget.WBP_TierText_E{}
        4 => UserInterface.VerseWidget.WBP_TierText_D{}
        5 => UserInterface.VerseWidget.WBP_TierText_C{}
        6 => UserInterface.VerseWidget.WBP_TierText_B{}
        7 => UserInterface.VerseWidget.WBP_TierText_A{}
        8 => UserInterface.VerseWidget.WBP_TierText_S{}
        _ => UserInterface.VerseWidget.WBP_TierText_G{}

GearInfos:[][gear_slot]gear_info = array:
    map:
        gear_slot.Head => gear_info:
            Name := "Ancient Circlet"
            Icon := UserInterface.Gears.Gear_01_Head_128
            IconSmall := UserInterface.Gears.Gear_01_Head_64
        gear_slot.Chest => gear_info:
            Name := "Ancient Chest"
            Icon := UserInterface.Gears.Gear_01_Chest_128
            IconSmall := UserInterface.Gears.Gear_01_Chest_64
        gear_slot.Pants => gear_info:
            Name := "Ancient Pants"
            Icon := UserInterface.Gears.Gear_01_Pants_128
            IconSmall := UserInterface.Gears.Gear_01_Pants_64
        gear_slot.Shoulder => gear_info:
            Name := "Ancient Shoulders"
            Icon := UserInterface.Gears.Gear_01_Shoulder_128
            IconSmall := UserInterface.Gears.Gear_01_Shoulder_64
        gear_slot.Gloves => gear_info:
            Name := "Ancient Gloves"
            Icon := UserInterface.Gears.Gear_01_Gloves_128
            IconSmall := UserInterface.Gears.Gear_01_Gloves_64
        gear_slot.Boots => gear_info:
            Name := "Ancient Boots"
            Icon := UserInterface.Gears.Gear_01_Boots_128
            IconSmall := UserInterface.Gears.Gear_01_Boots_64
    map:
        gear_slot.Head => gear_info:
            Name := "Scorching Cloak"
            Icon := UserInterface.Gears.Gear_02_Head_128
            IconSmall := UserInterface.Gears.Gear_02_Head_64
        gear_slot.Chest => gear_info:
            Name := "Scorching Chest"
            Icon := UserInterface.Gears.Gear_02_Chest_128
            IconSmall := UserInterface.Gears.Gear_02_Chest_64
        gear_slot.Pants => gear_info:
            Name := "Scorching Pants"
            Icon := UserInterface.Gears.Gear_02_Pants_128
            IconSmall := UserInterface.Gears.Gear_02_Pants_64
        gear_slot.Shoulder => gear_info:
            Name := "Scorching Shoulders"
            Icon := UserInterface.Gears.Gear_02_Shoulder_128
            IconSmall := UserInterface.Gears.Gear_02_Shoulder_64
        gear_slot.Gloves => gear_info:
            Name := "Scorching Gloves"
            Icon := UserInterface.Gears.Gear_02_Gloves_128
            IconSmall := UserInterface.Gears.Gear_02_Gloves_64
        gear_slot.Boots => gear_info:
            Name := "Scorching Boots"
            Icon := UserInterface.Gears.Gear_02_Boots_128
            IconSmall := UserInterface.Gears.Gear_02_Boots_64
    map:
        gear_slot.Head => gear_info:
            Name := "Nether Helm"
            Icon := UserInterface.Gears.Gear_03_Head_128
            IconSmall := UserInterface.Gears.Gear_03_Head_64
        gear_slot.Chest => gear_info:
            Name := "Nether Chest"
            Icon := UserInterface.Gears.Gear_03_Chest_128
            IconSmall := UserInterface.Gears.Gear_03_Chest_64
        gear_slot.Pants => gear_info:
            Name := "Nether Pants"
            Icon := UserInterface.Gears.Gear_03_Pants_128
            IconSmall := UserInterface.Gears.Gear_03_Pants_64
        gear_slot.Shoulder => gear_info:
            Name := "Nether Shoulders"
            Icon := UserInterface.Gears.Gear_03_Shoulder_128
            IconSmall := UserInterface.Gears.Gear_03_Shoulder_64
        gear_slot.Gloves => gear_info:
            Name := "Nether Gloves"
            Icon := UserInterface.Gears.Gear_03_Gloves_128
            IconSmall := UserInterface.Gears.Gear_03_Gloves_64
        gear_slot.Boots => gear_info:
            Name := "Nether Boots"
            Icon := UserInterface.Gears.Gear_03_Boots_128
            IconSmall := UserInterface.Gears.Gear_03_Boots_64
    map:
        gear_slot.Head => gear_info:
            Name := "Frost Hood"
            Icon := UserInterface.Gears.Gear_04_Head_128
            IconSmall := UserInterface.Gears.Gear_04_Head_64
        gear_slot.Chest => gear_info:
            Name := "Frost Chest"
            Icon := UserInterface.Gears.Gear_04_Chest_128
            IconSmall := UserInterface.Gears.Gear_04_Chest_64
        gear_slot.Pants => gear_info:
            Name := "Frost Pants"
            Icon := UserInterface.Gears.Gear_04_Pants_128
            IconSmall := UserInterface.Gears.Gear_04_Pants_64
        gear_slot.Shoulder => gear_info:
            Name := "Frost Shoulders"
            Icon := UserInterface.Gears.Gear_04_Shoulder_128
            IconSmall := UserInterface.Gears.Gear_04_Shoulder_64
        gear_slot.Gloves => gear_info:
            Name := "Frost Gloves"
            Icon := UserInterface.Gears.Gear_04_Gloves_128
            IconSmall := UserInterface.Gears.Gear_04_Gloves_64
        gear_slot.Boots => gear_info:
            Name := "Frost Boots"
            Icon := UserInterface.Gears.Gear_04_Boots_128
            IconSmall := UserInterface.Gears.Gear_04_Boots_64
    map:
        gear_slot.Head => gear_info:
            Name := "Arcane Circlet"
            Icon := UserInterface.Gears.Gear_05_Head_128
            IconSmall := UserInterface.Gears.Gear_05_Head_64
        gear_slot.Chest => gear_info:
            Name := "Arcane Chest"
            Icon := UserInterface.Gears.Gear_05_Chest_128
            IconSmall := UserInterface.Gears.Gear_05_Chest_64
        gear_slot.Pants => gear_info:
            Name := "Arcane Pants"
            Icon := UserInterface.Gears.Gear_05_Pants_128
            IconSmall := UserInterface.Gears.Gear_05_Pants_64
        gear_slot.Shoulder => gear_info:
            Name := "Arcane Shoulders"
            Icon := UserInterface.Gears.Gear_05_Shoulder_128
            IconSmall := UserInterface.Gears.Gear_05_Shoulder_64
        gear_slot.Gloves => gear_info:
            Name := "Arcane Gloves"
            Icon := UserInterface.Gears.Gear_05_Gloves_128
            IconSmall := UserInterface.Gears.Gear_05_Gloves_64
        gear_slot.Boots => gear_info:
            Name := "Arcane Boots"
            Icon := UserInterface.Gears.Gear_05_Boots_128
            IconSmall := UserInterface.Gears.Gear_05_Boots_64
    map:
        gear_slot.Head => gear_info:
            Name := "Celestial Hood"
            Icon := UserInterface.Gears.Gear_06_Head_128
            IconSmall := UserInterface.Gears.Gear_06_Head_64
        gear_slot.Chest => gear_info:
            Name := "Celestial Chest"
            Icon := UserInterface.Gears.Gear_06_Chest_128
            IconSmall := UserInterface.Gears.Gear_06_Chest_64
        gear_slot.Pants => gear_info:
            Name := "Celestial Pants"
            Icon := UserInterface.Gears.Gear_06_Pants_128
            IconSmall := UserInterface.Gears.Gear_06_Pants_64
        gear_slot.Shoulder => gear_info:
            Name := "Celestial Shoulders"
            Icon := UserInterface.Gears.Gear_06_Shoulder_128
            IconSmall := UserInterface.Gears.Gear_06_Shoulder_64
        gear_slot.Gloves => gear_info:
            Name := "Celestial Gloves"
            Icon := UserInterface.Gears.Gear_06_Gloves_128
            IconSmall := UserInterface.Gears.Gear_06_Gloves_64
        gear_slot.Boots => gear_info:
            Name := "Celestial Boots"
            Icon := UserInterface.Gears.Gear_06_Boots_128
            IconSmall := UserInterface.Gears.Gear_06_Boots_64
    map:
        gear_slot.Head => gear_info:
            Name := "Shadows Hood"
            Icon := UserInterface.Gears.Gear_07_Head_128
            IconSmall := UserInterface.Gears.Gear_07_Head_64
        gear_slot.Chest => gear_info:
            Name := "Shadows Chest"
            Icon := UserInterface.Gears.Gear_07_Chest_128
            IconSmall := UserInterface.Gears.Gear_07_Chest_64
        gear_slot.Pants => gear_info:
            Name := "Shadows Pants"
            Icon := UserInterface.Gears.Gear_07_Pants_128
            IconSmall := UserInterface.Gears.Gear_07_Pants_64
        gear_slot.Shoulder => gear_info:
            Name := "Shadows Shoulders"
            Icon := UserInterface.Gears.Gear_07_Shoulder_128
            IconSmall := UserInterface.Gears.Gear_07_Shoulder_64
        gear_slot.Gloves => gear_info:
            Name := "Shadows Gloves"
            Icon := UserInterface.Gears.Gear_07_Gloves_128
            IconSmall := UserInterface.Gears.Gear_07_Gloves_64
        gear_slot.Boots => gear_info:
            Name := "Shadows Boots"
            Icon := UserInterface.Gears.Gear_07_Boots_128
            IconSmall := UserInterface.Gears.Gear_07_Boots_64
    map:
        gear_slot.Head => gear_info:
            Name := "Enigmatic Hood"
            Icon := UserInterface.Gears.Gear_08_Head_128
            IconSmall := UserInterface.Gears.Gear_08_Head_64
        gear_slot.Chest => gear_info:
            Name := "Enigmatic Chest"
            Icon := UserInterface.Gears.Gear_08_Chest_128
            IconSmall := UserInterface.Gears.Gear_08_Chest_64
        gear_slot.Pants => gear_info:
            Name := "Enigmatic Pants"
            Icon := UserInterface.Gears.Gear_08_Pants_128
            IconSmall := UserInterface.Gears.Gear_08_Pants_64
        gear_slot.Shoulder => gear_info:
            Name := "Enigmatic Shoulders"
            Icon := UserInterface.Gears.Gear_08_Shoulder_128
            IconSmall := UserInterface.Gears.Gear_08_Shoulder_64
        gear_slot.Gloves => gear_info:
            Name := "Enigmatic Gloves"
            Icon := UserInterface.Gears.Gear_08_Gloves_128
            IconSmall := UserInterface.Gears.Gear_08_Gloves_64
        gear_slot.Boots => gear_info:
            Name := "Enigmatic Boots"
            Icon := UserInterface.Gears.Gear_08_Boots_128
            IconSmall := UserInterface.Gears.Gear_08_Boots_64
    map:
        gear_slot.Head => gear_info:
            Name := "Immortal Hood"
            Icon := UserInterface.Gears.Gear_09_Head_128
            IconSmall := UserInterface.Gears.Gear_09_Head_64
        gear_slot.Chest => gear_info:
            Name := "Immortal Chest"
            Icon := UserInterface.Gears.Gear_09_Chest_128
            IconSmall := UserInterface.Gears.Gear_09_Chest_64
        gear_slot.Pants => gear_info:
            Name := "Immortal Pants"
            Icon := UserInterface.Gears.Gear_09_Pants_128
            IconSmall := UserInterface.Gears.Gear_09_Pants_64
        gear_slot.Shoulder => gear_info:
            Name := "Immortal Shoulders"
            Icon := UserInterface.Gears.Gear_09_Shoulder_128
            IconSmall := UserInterface.Gears.Gear_09_Shoulder_64
        gear_slot.Gloves => gear_info:
            Name := "Immortal Gloves"
            Icon := UserInterface.Gears.Gear_09_Gloves_128
            IconSmall := UserInterface.Gears.Gear_09_Gloves_64
        gear_slot.Boots => gear_info:
            Name := "Immortal Boots"
            Icon := UserInterface.Gears.Gear_09_Boots_128
            IconSmall := UserInterface.Gears.Gear_09_Boots_64

GetGearInfos()<transacts>:[][gear_slot]gear_info =
    GearInfos

GetGearInfo(Evolution:int)<decides><transacts>:[gear_slot]gear_info =
    GearInfos[Evolution]

GetGearInfo(Evolution:int, Slot:gear_slot)<decides><transacts>:gear_info =
    GearInfos[Evolution][Slot]

GetGearSlotInfo(Slot:gear_slot)<transacts>:gear_slot_info =
    case (Slot):
        gear_slot.Head => gear_slot_info:
            Stat := stat_id.MaxHealth
            StatFactor := 0.4
            TierStat := stat_id.BossDamage
            TierStatFactor := 0.1
        gear_slot.Chest => gear_slot_info:
            Stat := stat_id.MaxHealth
            StatFactor := 0.4
            TierStat := stat_id.MaxHealth
            TierStatFactor := 20.0
        gear_slot.Pants => gear_slot_info:
            Stat := stat_id.MaxHealth
            StatFactor := 0.4
            TierStat := stat_id.ExperienceMultiplier
            TierStatFactor := 0.05
        gear_slot.Shoulder => gear_slot_info:
            Stat := stat_id.Damage
            StatFactor := 0.004
            TierStat := stat_id.CriticalDamage
            TierStatFactor := 0.1
        gear_slot.Gloves => gear_slot_info:
            Stat := stat_id.Damage
            StatFactor := 0.004
            TierStat := stat_id.GoldMultiplier
            TierStatFactor := 0.2
        gear_slot.Boots => gear_slot_info:
            Stat := stat_id.Damage
            StatFactor := 0.004
            TierStat := stat_id.HealthRegeneration
            TierStatFactor := 0.25

<# ----------------------------------------------------------------------------------------------------
    Get Functions
---------------------------------------------------------------------------------------------------- #>
GetGearTreasureChestRewards(Count:int)<transacts>:tuple(int, [int]int) =
    var RecipeRewards:int = 0
    var GearRewards:[int]int = map{}

    for (Index := 1..Count):
        RandomIndex := GetWeightedRandom(GearTreasureChestWeights)
        if (RecipeReward := TreasureChestRecipeRewards[RandomIndex]):
            set RecipeRewards += RecipeReward
        
        if:
            RewardAmount := TreasureChestGearRewards[RandomIndex]
            RandomGearSlot := GetRandomInt(0, 5)
            set GearRewards[RandomGearSlot] = (GearRewards[RandomGearSlot] + RewardAmount) or RewardAmount

    (RecipeRewards, GearRewards)

(Player:player).CanAfford(GearUpgradeCost:gear_upgrade_cost, Slot:gear_slot)<transacts>:logic =
    var AffordCount:int = 0
    for (UpgradeCost : GearUpgradeCost.UpgradeCosts, Player.CanAfford(UpgradeCost)?):
        set AffordCount += 1
    logic{AffordCount >= GearUpgradeCost.UpgradeCosts.Length and Player.GetGearQuantity(Slot) >= GearUpgradeCost.GearCost}

(PlayerGear:player_gear).GetMaxLevel()<transacts>:int =
    CalculateGearMaxLevel(PlayerGear.Evolution, PlayerGear.Tier, PlayerGear.Stars)

CalculateGearMaxLevel(Evolution:int, Tier:int, Stars:int)<transacts>:int =
    (Tier * 10) + ((Tier - 1) * MaxGearStars * 5) + (Stars * 5)

CalculateGearTotalLevel(Evolution:int)<transacts>:int =
    Sum of for (EvolutionIndex := 1..Evolution):
        CalculateGearMaxLevel(EvolutionIndex - 1, CalculateGearMaxTier(EvolutionIndex - 1), MaxGearStars)

(PlayerGear:player_gear).IsLevelMaxed()<transacts>:logic =
    logic{PlayerGear.Level >= PlayerGear.GetMaxLevel()}

(PlayerGear:player_gear).GetStat(Slot:gear_slot, Level:int)<transacts>:float =
    TotalLevel := CalculateGearTotalLevel(PlayerGear.Evolution) + Level
    TotalLevel * GetGearSlotInfo(Slot).StatFactor

(PlayerGear:player_gear).GetTierStats(Slot:gear_slot)<transacts>:[]float =
    TierStatsIncrement := GetGearSlotInfo(Slot).TierStatFactor
    BaseTierStat := CalculateGearTotalTier(PlayerGear.Evolution) * TierStatsIncrement

    for (Tier := 1..PlayerGear.GetMaxTier()):
        BaseTierStat + (TierStatsIncrement * Tier)

(PlayerGear:player_gear).GetTierStats(Slot:gear_slot, Tier:int)<transacts>:float =
    TierStatsIncrement := GetGearSlotInfo(Slot).TierStatFactor
    BaseTierStat := CalculateGearTotalTier(PlayerGear.Evolution) * TierStatsIncrement
    BaseTierStat + (TierStatsIncrement * Tier)

(PlayerGear:player_gear).GetMaxTier()<transacts>:int =
    CalculateGearMaxTier(PlayerGear.Evolution)

CalculateGearMaxTier(Evolution:int)<transacts>:int =
    Min(Evolution + 2, 8)

CalculateGearTotalTier(Evolution:int)<transacts>:int =
    Sum of for (EvolutionIndex := 1..Evolution):
        CalculateGearMaxTier(EvolutionIndex - 1)

(PlayerGear:player_gear).IsTierMaxed()<transacts>:logic =
    logic{PlayerGear.Tier >= PlayerGear.GetMaxTier() and PlayerGear.Stars >= MaxGearStars}

(PlayerGear:player_gear).CanEvolve()<transacts>:logic =
    logic{PlayerGear.IsLevelMaxed()? and PlayerGear.IsTierMaxed()?}

(Player:player).GetGearMinEvolution()<transacts>:?int =
    var MaybeMinEvolution:?int = false
    
    for (SlotIndex := 0..5, GearData := Player.GetGear[SlotIndex]):
        if (MinEvolution := MaybeMinEvolution?):
            set MaybeMinEvolution = option{Min(MinEvolution, GearData.Evolution)}
        else:
            set MaybeMinEvolution = option{GearData.Evolution}

    MaybeMinEvolution

(Player:player).GetGearStatSummary()<transacts>:[]tuple(stat_id, float) =
    var StatSummary:[stat_id]float = map{}

    for (SlotIndex := 0..5, GearData := Player.GetGear[SlotIndex]):
        SlotInfo := GetGearSlotInfo(SlotIndex.ToGearSlot())
        StatId := SlotInfo.Stat
        StatValue := GearData.GetStat(SlotIndex.ToGearSlot(), GearData.Level)
        if (set StatSummary[StatId] = (StatSummary[StatId] + StatValue) or StatValue) {}

        TierStatId := SlotInfo.TierStat
        TierStatValue := GearData.GetTierStats(SlotIndex.ToGearSlot(), GearData.Tier)
        if (set StatSummary[TierStatId] = (StatSummary[TierStatId] + TierStatValue) or TierStatValue) {}

    if:
        Player.GetGear().Length = 6
        MinEvolution := Player.GetGearMinEvolution()?
        set StatSummary[GearSetBonusStat] = (MinEvolution + 1) * 0.25

    for (StatId -> StatValue : StatSummary):
        (StatId, StatValue)

(PlayerGear:player_gear).GetUpgradeCosts(Count:int)<transacts>:[]upgrade_cost =
    var TotalGoldAmount:float = 0.0
    var TotalGearRecipeAmount:float = 0.0

    GoldCostExponent := 1.0 + GearBaseUpgradeGoldCostExponent / ((PlayerGear.Evolution + 1) * 1.0)
    RecipeCostExponent := 1.0 + GearBaseUpgradeRecipeCostExponent / ((PlayerGear.Evolution + 1) * 1.0)
    for (Index := 0..Count - 1):
        set TotalGoldAmount += (Floor[GearBaseUpgradeGoldCosts[PlayerGear.Evolution] * Pow(GoldCostExponent, (PlayerGear.Level + Index) * 1.0)] or 0) * 1.0
        set TotalGearRecipeAmount +=  (Floor[GearBaseUpgradeRecipeCosts[PlayerGear.Evolution] * Pow(RecipeCostExponent, (PlayerGear.Level + Index) * 1.0)] or 0) * 1.0
        
    TotalUpgradeCost := array:
        upgrade_cost{Id := currency_id.Gold, Amount := TotalGoldAmount}
        upgrade_cost{Id := currency_id.GearRecipe, Amount := TotalGearRecipeAmount}

(Player:player).GetAffordableUpgradeCount(PlayerGear:player_gear, Count:int)<transacts>:int =
    MaxCount := PlayerGear.GetMaxLevel() - PlayerGear.Level
    GoldCostExponent := 1.0 + GearBaseUpgradeGoldCostExponent / ((PlayerGear.Evolution + 1) * 1.0)
    RecipeCostExponent := 1.0 + GearBaseUpgradeRecipeCostExponent / ((PlayerGear.Evolution + 1) * 1.0)
    UpgradeCosts:[]tuple(float, float) = for (Index := 0..MaxCount - 1):
        GoldAmount := (Floor[GearBaseUpgradeGoldCosts[PlayerGear.Evolution] * Pow(GoldCostExponent, (PlayerGear.Level + Index) * 1.0)] or 0) * 1.0
        GearRecipeAmount := (Floor[GearBaseUpgradeRecipeCosts[PlayerGear.Evolution] * Pow(RecipeCostExponent, (PlayerGear.Level + Index) * 1.0)] or 0) * 1.0
        (GoldAmount, GearRecipeAmount)
    
    var AffordCount:int = 1
    var Lower:int = 1
    var Upper:int = Min(MaxCount, Count)

    if (not Player.CanAfford(UpgradeCosts, 1)?):
        return 1

    loop:
        if (Lower >= Upper):
            if (Player.CanAfford(UpgradeCosts, Upper)?):
                set AffordCount = Upper

            break

        if (Middle := Lower + Floor[((Upper - Lower) * 1.0) / 2.0]):
            if (Player.CanAfford(UpgradeCosts, Middle)?):
                set AffordCount = Middle
                set Lower = Middle + 1
            else:
                set Upper = Middle - 1
        else:
            break
    
    return AffordCount

(Player:player).CanAfford(UpgradeCosts:[]tuple(float, float), AffordCount:int)<transacts>:logic =
    var TotalGoldAmount:float = 0.0
    var TotalGearRecipeAmount:float = 0.0

    for (Index := 0..AffordCount - 1, UpgradeCost := UpgradeCosts[Index]):
        set TotalGoldAmount += UpgradeCost(0)
        set TotalGearRecipeAmount += UpgradeCost(1)

    TotalUpgradeCost := array:
        upgrade_cost{Id := currency_id.Gold, Amount := TotalGoldAmount}
        upgrade_cost{Id := currency_id.GearRecipe, Amount := TotalGearRecipeAmount}

    Player.CanAfford(TotalUpgradeCost)

(PlayerGear:player_gear).GetFuseCosts()<transacts>:gear_upgrade_cost =
    gear_upgrade_cost:
        UpgradeCosts := array:
            upgrade_cost:
                Id := currency_id.Gold
                Amount := (Floor[GearBaseTierGoldCosts[PlayerGear.Evolution] * Pow(1.0 + GearBaseTierGoldCostExponent / ((PlayerGear.Evolution + 1) * 1.0), PlayerGear.Tier * 1.0)] or 0) * 1.0
            upgrade_cost:
                Id := currency_id.GearRecipe
                Amount := (Floor[GearBaseTierRecipeCosts[PlayerGear.Evolution] * Pow(1.0 + GearBaseTierRecipeCostExponent / ((PlayerGear.Evolution + 1) * 1.0), PlayerGear.Tier * 1.0)] or 0) * 1.0

        GearCost := Floor[GearBaseTierGearCosts[PlayerGear.Evolution] * Pow(1.0 + GearBaseTierGearCostExponent / ((PlayerGear.Evolution + 1) * 1.0), PlayerGear.Tier * 1.0)] or 0

(PlayerGear:player_gear).GetEvolveCosts()<transacts>:[]upgrade_cost =
    array:
        upgrade_cost:
            Id := currency_id.Gold
            Amount := (GearBaseEvolveGoldCosts[PlayerGear.Evolution] or 0) * 1.0

        upgrade_cost:
            Id := currency_id.GearRecipe
            Amount := (GearBaseEvolveRecipeCosts[PlayerGear.Evolution] or 0) * 1.0