using { /Fortnite.com/UI }
using { /Verse.org/Colors }
using { /UnrealEngine.com/Temporary/SpatialMath }
using { /UnrealEngine.com/Temporary/UI }

<# ----------------------------------------------------------------------------------------------------
    gear_info_ui
---------------------------------------------------------------------------------------------------- #>
gear_info_ui := class(ui_component):
    Icon:texture_block = texture_block:
        DefaultImage := UserInterface.Gears.Gear_01_Head_128
        DefaultDesiredSize := vector2{X := 128.0, Y := 128.0}

    NameText:text_block = text_block:
        DefaultTextColor := ThemeColors.PrimaryText
        DefaultShadowColor := NamedColors.Black
        DefaultShadowOffset := option{vector2{X := 2.0, Y := 2.0}}
        DefaultShadowOpacity := 0.5

    LevelText:text_block = text_block:
        DefaultTextColor := ThemeColors.PrimaryText
        DefaultShadowColor := NamedColors.Black
        DefaultShadowOffset := option{vector2{X := 2.0, Y := 2.0}}
        DefaultShadowOpacity := 0.5

    QuantityText:text_block = text_block:
        DefaultTextColor := ThemeColors.PrimaryText
        DefaultShadowColor := NamedColors.Black
        DefaultShadowOffset := option{vector2{X := 2.0, Y := 2.0}}
        DefaultShadowOpacity := 0.5

    var MaybeGearTierUI:?gear_tier_ui = false

    <# ----------------------------------------------------------------------------------------------------
        Initialization
    ---------------------------------------------------------------------------------------------------- #>
    CreateUI<override>():widget =
        stack_box:
            Orientation := orientation.Vertical
            Slots := array:
                stack_box_slot:
                    HorizontalAlignment := horizontal_alignment.Fill
                    VerticalAlignment := vertical_alignment.Fill
                    Widget := CreateGearInfoUI()

    CreateGearInfoUI():widget =
        GearTierUI := gear_tier_ui{Player := Player}
        GearTierUI.Initialize()
        set MaybeGearTierUI = option{GearTierUI}

        stack_box:
            Orientation := orientation.Horizontal
            Slots := array:
                stack_box_slot:
                    HorizontalAlignment := horizontal_alignment.Left
                    VerticalAlignment := vertical_alignment.Top
                    Widget := Icon
                stack_box_slot:
                    HorizontalAlignment := horizontal_alignment.Fill
                    VerticalAlignment := vertical_alignment.Fill
                    Distribution := option{1.0}
                    Padding := margin{Left := 20.0, Top := 0.0, Right := 0.0, Bottom := 0.0}
                    Widget := stack_box:
                        Orientation := orientation.Vertical
                        Slots := array:
                            stack_box_slot:
                                HorizontalAlignment := horizontal_alignment.Fill
                                VerticalAlignment := vertical_alignment.Fill
                                Widget := NameText
                            stack_box_slot:
                                HorizontalAlignment := horizontal_alignment.Fill
                                VerticalAlignment := vertical_alignment.Fill
                                Widget := stack_box:
                                    Orientation := orientation.Horizontal
                                    Slots := array:
                                        stack_box_slot:
                                            HorizontalAlignment := horizontal_alignment.Fill
                                            VerticalAlignment := vertical_alignment.Fill
                                            Widget := text_block:
                                                DefaultText := MakeMessage("Lv.")
                                                DefaultTextColor := ThemeColors.Highlight
                                                DefaultShadowColor := ThemeColors.Highlight * 0.125
                                                DefaultShadowOffset := option{vector2{X := 2.0, Y := 2.0}}
                                                DefaultShadowOpacity := 1.0
                                        stack_box_slot:
                                            HorizontalAlignment := horizontal_alignment.Fill
                                            VerticalAlignment := vertical_alignment.Fill
                                            Padding := margin{Left := 10.0, Top := 0.0, Right := 0.0, Bottom := 0.0}
                                            Widget := LevelText
                            stack_box_slot:
                                HorizontalAlignment := horizontal_alignment.Fill
                                VerticalAlignment := vertical_alignment.Fill
                                Widget := stack_box:
                                    Orientation := orientation.Horizontal
                                    Slots := array:
                                        stack_box_slot:
                                            HorizontalAlignment := horizontal_alignment.Fill
                                            VerticalAlignment := vertical_alignment.Fill
                                            Widget := GearTierUI.GetUI()
                                        stack_box_slot:
                                            HorizontalAlignment := horizontal_alignment.Right
                                            VerticalAlignment := vertical_alignment.Fill
                                            Distribution := option{1.0}
                                            Widget := stack_box:
                                                Orientation := orientation.Horizontal
                                                Slots := array:
                                                    stack_box_slot:
                                                        HorizontalAlignment := horizontal_alignment.Fill
                                                        VerticalAlignment := vertical_alignment.Fill
                                                        Widget := text_block:
                                                            DefaultText := MakeMessage("Qty.")
                                                            DefaultTextColor := ThemeColors.PrimaryText
                                                            DefaultTextOpacity := 0.5
                                                    stack_box_slot:
                                                        HorizontalAlignment := horizontal_alignment.Fill
                                                        VerticalAlignment := vertical_alignment.Fill
                                                        Padding := margin{Left := 10.0, Top := 0.0, Right := 0.0, Bottom := 0.0}
                                                        Widget := overlay:
                                                            Slots := array:
                                                                overlay_slot:
                                                                    HorizontalAlignment := horizontal_alignment.Fill
                                                                    VerticalAlignment := vertical_alignment.Fill
                                                                    Widget := UserInterface.VerseWidget.WBP_RoundedBox_Light{}
                                                                overlay_slot:
                                                                    HorizontalAlignment := horizontal_alignment.Fill
                                                                    VerticalAlignment := vertical_alignment.Fill
                                                                    Padding := margin{Left := 40.0, Top := 0.0, Right := 40.0, Bottom := 0.0}
                                                                    Widget := QuantityText

    <# ----------------------------------------------------------------------------------------------------
        Update Functions
    ---------------------------------------------------------------------------------------------------- #>
    Update(Slot:gear_slot):void =
        if (GearData := Player.GetGear[Slot]):
            if (GearInfo := GetGearInfo[GearData.Evolution, Slot]):
                Icon.SetImage(GearInfo.Icon)
                NameText.SetText(MakeMessage("{GearInfo.Name}"))

            LevelText.SetText(MakeMessage("{GearData.Level}/{GearData.GetMaxLevel()}"))
            QuantityText.SetText(MakeMessage("{FormatAsCurrency(GearData.Quantity * 1.0)}"))

            if (GearTierUI := MaybeGearTierUI?):
                GearTierUI.SetTier(GearData.Tier, GearData.Stars)