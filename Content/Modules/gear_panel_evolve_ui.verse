using { /Fortnite.com/UI }
using { /Verse.org/Colors }
using { /Verse.org/Simulation }
using { /UnrealEngine.com/Temporary/SpatialMath }
using { /UnrealEngine.com/Temporary/UI }

gear_panel_evolve_ui := class(ui_component):
    EvolvedEvent:events(tuple()) = events(tuple()){}
    
    GoldCostText:text_block = text_block{}
    RecipeCostText:text_block = text_block{}

    InfoText:text_block = text_block:
        DefaultTextColor := ThemeColors.Failure
        DefaultShadowColor := ThemeColors.Failure * 0.125
        DefaultShadowOffset := option{vector2{X := 2.0, Y := 2.0}}
        DefaultShadowOpacity := 1.0

    EvolveButton:button_regular = button_regular{}

    var MaybeGearEvolveEntryUI:?gear_evolve_entry_ui = false
    var MaybeUpdatedGearEvolveEntryUI:?gear_evolve_entry_ui = false

    var ActiveSlot:gear_slot = gear_slot.Head

    <# ----------------------------------------------------------------------------------------------------
        Initialization
    ---------------------------------------------------------------------------------------------------- #>
    CreateUI<override>():widget =
        GearEvolveEntryUI := gear_evolve_entry_ui{Player := Player}
        GearEvolveEntryUI.Initialize()
        set MaybeGearEvolveEntryUI = option{GearEvolveEntryUI}

        GearUpdatedEvolveEntryUI := gear_evolve_entry_ui{Player := Player}
        GearUpdatedEvolveEntryUI.Initialize()
        set MaybeUpdatedGearEvolveEntryUI = option{GearUpdatedEvolveEntryUI}

        EvolveButton.SetText(MakeMessage("Evolve"))
        EvolveButton.OnClick().Subscribe(OnClickEvolve)

        InfoText.SetText(MakeMessage("Requires Max Level & Tier!"))

        stack_box:
            Orientation := orientation.Vertical
            Slots := array:
                stack_box_slot:
                    HorizontalAlignment := horizontal_alignment.Fill
                    VerticalAlignment := vertical_alignment.Fill
                    Distribution := option{1.0}
                    Widget := stack_box:
                        Orientation := orientation.Horizontal
                        Slots := array:
                            stack_box_slot:
                                HorizontalAlignment := horizontal_alignment.Fill
                                VerticalAlignment := vertical_alignment.Fill
                                Distribution := option{1.0}
                                Widget := GearEvolveEntryUI.GetUI()
                            stack_box_slot:
                                HorizontalAlignment := horizontal_alignment.Fill
                                Padding := margin{Left := 20.0, Top := 0.0, Right := 20.0, Bottom := 0.0}
                                Widget := text_block:
                                    DefaultText := MakeMessage("▶")
                                    DefaultTextColor := ThemeColors.PrimaryText
                                    DefaultShadowColor := NamedColors.Black
                                    DefaultShadowOffset := option{vector2{X := 2.0, Y := 2.0}}
                                    DefaultShadowOpacity := 0.5
                            stack_box_slot:
                                HorizontalAlignment := horizontal_alignment.Fill
                                VerticalAlignment := vertical_alignment.Fill
                                Distribution := option{1.0}
                                Widget := GearUpdatedEvolveEntryUI.GetUI()
                stack_box_slot:
                    HorizontalAlignment := horizontal_alignment.Fill
                    VerticalAlignment := vertical_alignment.Fill
                    Padding := margin{Left := 0.0, Top := 40.0, Right := 0.0, Bottom := 20.0}
                    Widget := color_block:
                        DefaultColor := NamedColors.White
                        DefaultOpacity := 0.1
                        DefaultDesiredSize := vector2{X := 0.0, Y := 1.0}
                stack_box_slot:
                    HorizontalAlignment := horizontal_alignment.Fill
                    VerticalAlignment := vertical_alignment.Fill
                    Widget := CreateCostUI()
                stack_box_slot:
                    HorizontalAlignment := horizontal_alignment.Fill
                    VerticalAlignment := vertical_alignment.Fill
                    Padding := margin{Left := 0.0, Top := 20.0, Right := 0.0, Bottom := 0.0}
                    Widget := EvolveButton

    CreateCostUI():widget =
        stack_box:
            Orientation := orientation.Horizontal
            Slots := array:
                stack_box_slot:
                    HorizontalAlignment := horizontal_alignment.Fill
                    VerticalAlignment := vertical_alignment.Fill
                    Widget := text_block:
                        DefaultText := MakeMessage("Cost")
                        DefaultTextColor := ThemeColors.PrimaryText
                        DefaultTextOpacity := 0.5
                stack_box_slot:
                    HorizontalAlignment := horizontal_alignment.Right
                    VerticalAlignment := vertical_alignment.Fill
                    Distribution := option{1.0}
                    Widget := stack_box:
                        Orientation := orientation.Vertical
                        Slots := array:
                            stack_box_slot:
                                HorizontalAlignment := horizontal_alignment.Right
                                VerticalAlignment := vertical_alignment.Fill
                                Widget := CreateCostEntry(currency_id.Gold, GoldCostText)
                            stack_box_slot:
                                HorizontalAlignment := horizontal_alignment.Right
                                VerticalAlignment := vertical_alignment.Fill
                                Widget := CreateCostEntry(currency_id.GearRecipe, RecipeCostText)
                            stack_box_slot:
                                HorizontalAlignment := horizontal_alignment.Right
                                VerticalAlignment := vertical_alignment.Fill
                                Widget := InfoText

    CreateCostEntry(Id:currency_id, TextBlock:text_block):widget =
        stack_box:
            Orientation := orientation.Horizontal
            Slots := array:
                stack_box_slot:
                    HorizontalAlignment := horizontal_alignment.Fill
                    VerticalAlignment := vertical_alignment.Fill
                    Widget := texture_block:
                        DefaultImage := Id.ToIcon()
                        DefaultDesiredSize := vector2{X := 40.0, Y := 40.0}
                stack_box_slot:
                    HorizontalAlignment := horizontal_alignment.Fill
                    VerticalAlignment := vertical_alignment.Fill
                    Padding := margin{Left := 10.0, Top := 0.0, Right := 0.0, Bottom := 0.0}
                    Widget := TextBlock

    <# ----------------------------------------------------------------------------------------------------
        Update Functions
    ---------------------------------------------------------------------------------------------------- #>
    Update(Slot:gear_slot):void =
        if:
            GearData := Player.GetGear[Slot]
            GearEvolveEntryUI := MaybeGearEvolveEntryUI?
            UpdatedGearEvolveEntryUI := MaybeUpdatedGearEvolveEntryUI?
        then:
            set ActiveSlot = Slot
            GearEvolveEntryUI.Update(GearData, Slot)
            UpdatedGearEvolveEntryUI.Update(player_gear{Evolution := GearData.Evolution + 1}, Slot)

            UpgradeCosts := GearData.GetEvolveCosts()
            for (UpgradeCost : UpgradeCosts):
                if (UpgradeCost.Id = currency_id.Gold):
                    GoldCostText.SetText(MakeMessage("{FormatAsCurrency(UpgradeCost.Amount)}"))
                    GoldCostText.SetColor(Player.CanAfford(UpgradeCost)? and ThemeColors.Success or ThemeColors.Failure)
                else if (UpgradeCost.Id = currency_id.GearRecipe):
                    RecipeCostText.SetText(MakeMessage("{FormatAsCurrency(UpgradeCost.Amount)}"))
                    RecipeCostText.SetColor(Player.CanAfford(UpgradeCost)? and ThemeColors.Success or ThemeColors.Failure)

            InfoText.SetVisibility(GearData.CanEvolve()? and widget_visibility.Collapsed or widget_visibility.Visible)

    <# ----------------------------------------------------------------------------------------------------
        Event Handlers
    ---------------------------------------------------------------------------------------------------- #>
    OnClickEvolve(Message:widget_message):void =
        if:
            GearData := Player.GetGear[ActiveSlot]
            GearData.CanEvolve()?
            UpgradeCosts := GearData.GetEvolveCosts()
            Player.CanAfford(GearData.GetEvolveCosts())?
        then:
            Player.RemoveCurrency(UpgradeCosts)
            Player.EvolveGear(ActiveSlot)
            Update(ActiveSlot)
            EvolvedEvent.Signal()
            Player.PlayAudio(audio_id.WeaponEvolution)
        else:
            Player.PlayAudio(audio_id.Error)