using { /Fortnite.com/UI }
using { /Verse.org/Colors }
using { /Verse.org/Simulation }
using { /UnrealEngine.com/Temporary/SpatialMath }
using { /UnrealEngine.com/Temporary/UI }

using { /Verse.org/Random }

gear_panel_info_ui := class(ui_component):
    
    StatIcon:texture_block = texture_block:
        DefaultImage := UserInterface.Icons.ProjectileDamage_64
        DefaultDesiredSize := vector2{X := 40.0, Y := 40.0}

    StatLabelText:text_block = text_block:
        DefaultTextColor := ThemeColors.PrimaryText

    StatValueText:text_block = text_block:
        DefaultTextColor := ThemeColors.Success
        DefaultShadowColor := ThemeColors.Success * 0.125
        DefaultShadowOffset := option{vector2{X := 2.0, Y := 2.0}}
        DefaultShadowOpacity := 1.0

    var MaybeGearInfoUI:?gear_info_ui = false
    var TierBonusEntries:[]tuple(widget, text_block, text_block) = array{}

    <# ----------------------------------------------------------------------------------------------------
        Initialization
    ---------------------------------------------------------------------------------------------------- #>
    CreateUI<override>():widget =
        GearInfoUI := gear_info_ui{Player := Player}
        GearInfoUI.Initialize()
        set MaybeGearInfoUI = option{GearInfoUI}

        stack_box:
            Orientation := orientation.Vertical
            Slots := array:
                stack_box_slot:
                    HorizontalAlignment := horizontal_alignment.Fill
                    VerticalAlignment := vertical_alignment.Fill
                    Widget := GearInfoUI.GetUI()
                stack_box_slot:
                    HorizontalAlignment := horizontal_alignment.Fill
                    VerticalAlignment := vertical_alignment.Fill
                    Padding := margin{Left := 0.0, Top := 20.0, Right := 0.0, Bottom := 20.0}
                    Widget := color_block:
                        DefaultColor := NamedColors.White
                        DefaultOpacity := 0.1
                        DefaultDesiredSize := vector2{X := 0.0, Y := 1.0}
                stack_box_slot:
                    HorizontalAlignment := horizontal_alignment.Fill
                    VerticalAlignment := vertical_alignment.Fill
                    Widget := CreateStatsUI()
                stack_box_slot:
                    HorizontalAlignment := horizontal_alignment.Fill
                    VerticalAlignment := vertical_alignment.Bottom
                    Distribution := option{1.0}
                    Padding := margin{Left := 0.0, Top := 20.0, Right := 0.0, Bottom := 20.0}
                    Widget := color_block:
                        DefaultColor := NamedColors.White
                        DefaultOpacity := 0.1
                        DefaultDesiredSize := vector2{X := 0.0, Y := 1.0}
                stack_box_slot:
                    HorizontalAlignment := horizontal_alignment.Fill
                    VerticalAlignment := vertical_alignment.Fill
                    Widget := CreateTierBonusUI()

    CreateStatsUI():widget =
        stack_box:
            Orientation := orientation.Vertical
            Slots := array:
                stack_box_slot:
                    HorizontalAlignment := horizontal_alignment.Left
                    VerticalAlignment := vertical_alignment.Fill
                    Padding := margin{Left := 0.0, Top := 0.0, Right := 0.0, Bottom := 10.0}
                    Widget := text_block:
                        DefaultText := MakeMessage("Stats")
                        DefaultTextColor := ThemeColors.PrimaryText
                        DefaultTextOpacity := 0.5
                stack_box_slot:
                    HorizontalAlignment := horizontal_alignment.Fill
                    VerticalAlignment := vertical_alignment.Fill
                    Widget := stack_box:
                        Orientation := orientation.Horizontal
                        Slots := array:
                            stack_box_slot:
                                HorizontalAlignment := horizontal_alignment.Left
                                VerticalAlignment := vertical_alignment.Fill
                                Widget := StatIcon
                            stack_box_slot:
                                HorizontalAlignment := horizontal_alignment.Fill
                                VerticalAlignment := vertical_alignment.Fill
                                Padding := margin{Left := 10.0, Top := 0.0, Right := 0.0, Bottom := 0.0}
                                Widget := StatLabelText
                            stack_box_slot:
                                HorizontalAlignment := horizontal_alignment.Right
                                VerticalAlignment := vertical_alignment.Fill
                                Distribution := option{1.0}
                                Widget := StatValueText
        
    CreateTierBonusUI():widget =
        stack_box:
            Orientation := orientation.Vertical
            Slots := array:
                stack_box_slot:
                    HorizontalAlignment := horizontal_alignment.Left
                    VerticalAlignment := vertical_alignment.Fill
                    Padding := margin{Left := 0.0, Top := 0.0, Right := 0.0, Bottom := 10.0}
                    Widget := text_block:
                        DefaultText := MakeMessage("Tier Bonus")
                        DefaultTextColor := ThemeColors.PrimaryText
                        DefaultTextOpacity := 0.5
                stack_box_slot:
                    HorizontalAlignment := horizontal_alignment.Fill
                    VerticalAlignment := vertical_alignment.Fill
                    Widget := stack_box:
                        Orientation := orientation.Vertical
                        Slots := for (Index := 1..8):
                            stack_box_slot:
                                HorizontalAlignment := horizontal_alignment.Fill
                                VerticalAlignment := vertical_alignment.Fill
                                Widget := CreateTierBonusEntryUI(Index)

    CreateTierBonusEntryUI(Tier:int):widget =
        TierBonusWidget := Tier.ToGearTierWidget()
        TierBonusLabelText:text_block = text_block{}
        TierBonusValueText:text_block = text_block{}

        TierBonusStackBox := stack_box:
            Orientation := orientation.Horizontal
            Slots := array:
                stack_box_slot:
                    HorizontalAlignment := horizontal_alignment.Left
                    VerticalAlignment := vertical_alignment.Fill
                    Widget := TierBonusWidget
                stack_box_slot:
                    HorizontalAlignment := horizontal_alignment.Fill
                    VerticalAlignment := vertical_alignment.Fill
                    Padding := margin{Left := 10.0, Top := 0.0, Right := 0.0, Bottom := 0.0}
                    Widget := TierBonusLabelText
                stack_box_slot:
                    HorizontalAlignment := horizontal_alignment.Right
                    VerticalAlignment := vertical_alignment.Fill
                    Distribution := option{1.0}
                    Widget := TierBonusValueText

        set TierBonusEntries += array{(TierBonusStackBox, TierBonusLabelText, TierBonusValueText)}
        TierBonusStackBox

    <# ----------------------------------------------------------------------------------------------------
        Update Functions
    ---------------------------------------------------------------------------------------------------- #>
    Update(Slot:gear_slot):void =
        if (GearData := Player.GetGear[Slot]):
            if (GearInfoUI := MaybeGearInfoUI?):
                GearInfoUI.Update(Slot)

            StatId := GetGearSlotInfo(Slot).Stat
            StatIcon.SetImage(StatId.ToIcon())
            StatLabelText.SetText(MakeMessage("{StatId}"))
            StatValueText.SetText(MakeMessage("+{GearData.GetStat(Slot, GearData.Level).Format(StatId)}"))
            
            TierStats := GearData.GetTierStats(Slot)
            for (Index -> TierBonusEntry : TierBonusEntries):
                if (TierStat := TierStats[Index]):
                    TierStatId := GetGearSlotInfo(Slot).TierStat
                    TierBonusEntry(0).SetVisibility(widget_visibility.Visible)
                    TierBonusEntry(1).SetText(MakeMessage("{TierStatId}"))
                    TierBonusEntry(2).SetText(MakeMessage("+{TierStat.Format(TierStatId)}"))

                    if (GearData.Tier = Index + 1):
                        TierBonusEntry(1).SetTextColor(ThemeColors.PrimaryText)
                        TierBonusEntry(2).SetColor(ThemeColors.Success)
                    else:
                        TierBonusEntry(1).SetTextColor(ThemeColors.Inactive)
                        TierBonusEntry(2).SetTextColor(ThemeColors.Inactive)
                        TierBonusEntry(2).SetShadowOffset(false)
                else:
                    TierBonusEntry(0).SetVisibility(widget_visibility.Hidden)
        
    <# ----------------------------------------------------------------------------------------------------
        Event Handlers
    ---------------------------------------------------------------------------------------------------- #>
    