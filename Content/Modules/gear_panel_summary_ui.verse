using { /Fortnite.com/UI }
using { /Verse.org/Colors }
using { /Verse.org/Simulation }
using { /UnrealEngine.com/Temporary/SpatialMath }
using { /UnrealEngine.com/Temporary/UI }

gear_panel_summary_ui := class(ui_component):

    var StatEntries:[]tuple(widget, texture_block, text_block, text_block) = array{}
    var MaybeSetBonusEntry:?tuple(widget, texture_block, text_block, text_block) = false

    <# ----------------------------------------------------------------------------------------------------
        Initialization
    ---------------------------------------------------------------------------------------------------- #>
    CreateUI<override>():widget =
        StatUI := CreateStatUI()
        StatUI(1).SetImage(GearSetBonusStat.ToIcon())
        StatUI(2).SetText(MakeMessage("{GearSetBonusStat}"))
        StatUI(3).SetText(MakeMessage("+{(0.0).Format(GearSetBonusStat)}"))
        set MaybeSetBonusEntry = option{StatUI}

        stack_box:
            Orientation := orientation.Vertical
            Slots := array:
                stack_box_slot:
                    HorizontalAlignment := horizontal_alignment.Fill
                    VerticalAlignment := vertical_alignment.Fill
                    Widget := text_block:
                        DefaultText := MakeMessage("Stats Summary")
                        DefaultTextColor := ThemeColors.PrimaryText
                        DefaultShadowColor := NamedColors.Black
                        DefaultShadowOffset := option{vector2{X := 2.0, Y := 2.0}}
                        DefaultShadowOpacity := 0.5
                stack_box_slot:
                    HorizontalAlignment := horizontal_alignment.Fill
                    VerticalAlignment := vertical_alignment.Fill
                    Padding := margin{Left := 0.0, Top := 20.0, Right := 0.0, Bottom := 20.0}
                    Widget := color_block:
                        DefaultColor := NamedColors.White
                        DefaultOpacity := 0.1
                        DefaultDesiredSize := vector2{X := 0.0, Y := 1.0}
                stack_box_slot:
                    HorizontalAlignment := horizontal_alignment.Fill
                    VerticalAlignment := vertical_alignment.Fill
                    Widget := CreateStatsUI()

                stack_box_slot:
                    HorizontalAlignment := horizontal_alignment.Fill
                    VerticalAlignment := vertical_alignment.Bottom
                    Padding := margin{Left := 0.0, Top := 20.0, Right := 0.0, Bottom := 0.0}
                    Distribution := option{1.0}
                    Widget := text_block:
                        DefaultText := MakeMessage("Set Bonus")
                        DefaultTextColor := ThemeColors.PrimaryText
                        DefaultShadowColor := NamedColors.Black
                        DefaultShadowOffset := option{vector2{X := 2.0, Y := 2.0}}
                        DefaultShadowOpacity := 0.5
                stack_box_slot:
                    HorizontalAlignment := horizontal_alignment.Fill
                    VerticalAlignment := vertical_alignment.Fill
                    Padding := margin{Left := 0.0, Top := 20.0, Right := 0.0, Bottom := 20.0}
                    Widget := color_block:
                        DefaultColor := NamedColors.White
                        DefaultOpacity := 0.1
                        DefaultDesiredSize := vector2{X := 0.0, Y := 1.0}
                stack_box_slot:
                    HorizontalAlignment := horizontal_alignment.Fill
                    VerticalAlignment := vertical_alignment.Fill
                    Widget := StatUI(0)

    CreateStatsUI():widget =
        stack_box:
            Orientation := orientation.Vertical
            Slots := for (Index := 1..8):
                StatUI := CreateStatUI()
                set StatEntries += array{StatUI}

                stack_box_slot:
                    HorizontalAlignment := horizontal_alignment.Fill
                    VerticalAlignment := vertical_alignment.Fill
                    Widget := StatUI(0)

    CreateStatUI():tuple(widget, texture_block, text_block, text_block) =
        StatIcon := texture_block:
            DefaultImage := UserInterface.Icons.ProjectileDamage_64
            DefaultDesiredSize := vector2{X := 40.0, Y := 40.0}

        StatLabelText:text_block = text_block:
            DefaultTextColor := ThemeColors.PrimaryText
            
        StatValueText:text_block = text_block:
            DefaultTextColor := ThemeColors.Success
            DefaultShadowColor := ThemeColors.Success * 0.125
            DefaultShadowOffset := option{vector2{X := 2.0, Y := 2.0}}
            DefaultShadowOpacity := 1.0

        StatStackBox := stack_box:
            Orientation := orientation.Horizontal
            Slots := array:
                stack_box_slot:
                    HorizontalAlignment := horizontal_alignment.Left
                    VerticalAlignment := vertical_alignment.Fill
                    Widget := StatIcon
                stack_box_slot:
                    HorizontalAlignment := horizontal_alignment.Fill
                    VerticalAlignment := vertical_alignment.Fill
                    Padding := margin{Left := 10.0, Top := 0.0, Right := 0.0, Bottom := 0.0}
                    Widget := StatLabelText
                stack_box_slot:
                    HorizontalAlignment := horizontal_alignment.Right
                    VerticalAlignment := vertical_alignment.Fill
                    Distribution := option{1.0}
                    Widget := StatValueText

        (StatStackBox, StatIcon, StatLabelText, StatValueText)

    <# ----------------------------------------------------------------------------------------------------
        Update Functions
    ---------------------------------------------------------------------------------------------------- #>
    Update<override>():void =
        GearStats := Player.GetGearStatSummary()
        for (Index -> StatEntry : StatEntries):
            if (GearStat := GearStats[Index]):
                StatId := GearStat(0)
                if (StatId = GearSetBonusStat, SetBonusEntry := MaybeSetBonusEntry?):
                    SetBonusEntry(3).SetText(MakeMessage("+{GearStat(1).Format(StatId)}"))
                    StatEntry(0).SetVisibility(widget_visibility.Collapsed)
                else:
                    StatEntry(0).SetVisibility(widget_visibility.Visible)
                    StatEntry(1).SetImage(StatId.ToIcon())
                    StatEntry(2).SetText(MakeMessage("{StatId}"))
                    StatEntry(3).SetText(MakeMessage("+{GearStat(1).Format(StatId)}"))
            else:
                StatEntry(0).SetVisibility(widget_visibility.Collapsed)
        
    <# ----------------------------------------------------------------------------------------------------
        Event Handlers
    ---------------------------------------------------------------------------------------------------- #>
