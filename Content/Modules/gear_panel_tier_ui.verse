using { /Fortnite.com/UI }
using { /Verse.org/Colors }
using { /Verse.org/Simulation }
using { /UnrealEngine.com/Temporary/SpatialMath }
using { /UnrealEngine.com/Temporary/UI }

gear_panel_tier_ui := class(ui_component):
    FusedEvent:events(tuple()) = events(tuple()){}

    MaxLevelText:text_block = text_block:
        DefaultTextColor := ThemeColors.Success
        DefaultShadowColor := ThemeColors.Success * 0.125
        DefaultShadowOffset := option{vector2{X := 2.0, Y := 2.0}}
        DefaultShadowOpacity := 1.0

    UpdatedMaxLevelText:text_block = text_block:
        DefaultTextColor := ThemeColors.Success
        DefaultShadowColor := ThemeColors.Success * 0.125
        DefaultShadowOffset := option{vector2{X := 2.0, Y := 2.0}}
        DefaultShadowOpacity := 1.0

    GoldCostText:text_block = text_block{}
    RecipeCostText:text_block = text_block{}
    GearCostText:text_block = text_block{}

    GearCostIcon:texture_block = texture_block:
        DefaultImage := UserInterface.Gears.Gear_01_Head_64
        DefaultDesiredSize := vector2{X := 40.0, Y := 40.0}

    FuseButton:button_regular = button_regular{}

    var MaybeGearInfoUI:?gear_info_ui = false
    var MaybeGearTierUI:?gear_tier_ui = false
    var MaybeUpdatedGearTierUI:?gear_tier_ui = false
    var MaybeTierStatEntryUI:?gear_stat_upgrade_entry_ui = false
    
    var ActiveSlot:gear_slot = gear_slot.Head
    var UpgradeUI:widget = canvas{}
    var UpgradeMaxedUI:widget = canvas{}

    <# ----------------------------------------------------------------------------------------------------
        Initialization
    ---------------------------------------------------------------------------------------------------- #>
    CreateUI<override>():widget =
        GearInfoUI := gear_info_ui{Player := Player}
        GearInfoUI.Initialize()
        set MaybeGearInfoUI = option{GearInfoUI}

        FuseButton.SetText(MakeMessage("Fuse"))
        FuseButton.OnClick().Subscribe(OnClickFuse)

        stack_box:
            Orientation := orientation.Vertical
            Slots := array:
                stack_box_slot:
                    HorizontalAlignment := horizontal_alignment.Fill
                    VerticalAlignment := vertical_alignment.Fill
                    Widget := GearInfoUI.GetUI()
                stack_box_slot:
                    HorizontalAlignment := horizontal_alignment.Fill
                    VerticalAlignment := vertical_alignment.Fill
                    Padding := margin{Left := 0.0, Top := 20.0, Right := 0.0, Bottom := 20.0}
                    Widget := color_block:
                        DefaultColor := NamedColors.White
                        DefaultOpacity := 0.1
                        DefaultDesiredSize := vector2{X := 0.0, Y := 1.0}
                stack_box_slot:
                    HorizontalAlignment := horizontal_alignment.Fill
                    VerticalAlignment := vertical_alignment.Fill
                    Distribution := option{1.0}
                    Widget := overlay:
                        Slots := array:
                            overlay_slot:
                                HorizontalAlignment := horizontal_alignment.Fill
                                VerticalAlignment := vertical_alignment.Fill
                                Widget := set UpgradeUI = stack_box:
                                    Orientation := orientation.Vertical
                                    Slots := array:
                                        stack_box_slot:
                                            HorizontalAlignment := horizontal_alignment.Fill
                                            VerticalAlignment := vertical_alignment.Fill
                                            Widget := CreateStatsUI()
                                        stack_box_slot:
                                            HorizontalAlignment := horizontal_alignment.Fill
                                            VerticalAlignment := vertical_alignment.Bottom
                                            Padding := margin{Left := 0.0, Top := 20.0, Right := 0.0, Bottom := 20.0}
                                            Distribution := option{1.0}
                                            Widget := color_block: 
                                                DefaultColor := NamedColors.White
                                                DefaultOpacity := 0.1
                                                DefaultDesiredSize := vector2{X := 0.0, Y := 1.0}
                                        stack_box_slot:
                                            HorizontalAlignment := horizontal_alignment.Fill
                                            VerticalAlignment := vertical_alignment.Fill
                                            Widget := CreateCostUI()
                            overlay_slot:
                                HorizontalAlignment := horizontal_alignment.Fill
                                VerticalAlignment := vertical_alignment.Fill
                                Widget := set UpgradeMaxedUI = CreatMaxedUI()
                stack_box_slot:
                    HorizontalAlignment := horizontal_alignment.Fill
                    VerticalAlignment := vertical_alignment.Fill
                    Padding := margin{Left := 0.0, Top := 20.0, Right := 0.0, Bottom := 0.0}
                    Widget := FuseButton

    CreateStatsUI():widget =        
        TierStatEntryUI := gear_stat_upgrade_entry_ui{Player := Player}
        TierStatEntryUI.Initialize()
        set MaybeTierStatEntryUI = option{TierStatEntryUI}

        GearTierUI := gear_tier_ui{Player := Player}
        GearTierUI.Initialize()
        set MaybeGearTierUI = option{GearTierUI}

        UpdatedGearTierUI := gear_tier_ui{Player := Player}
        UpdatedGearTierUI.Initialize()
        set MaybeUpdatedGearTierUI = option{UpdatedGearTierUI}

        stack_box:
            Orientation := orientation.Vertical
            Slots := array:
                stack_box_slot:
                    HorizontalAlignment := horizontal_alignment.Fill
                    VerticalAlignment := vertical_alignment.Fill
                    Padding := margin{Left := 0.0, Top := 0.0, Right := 0.0, Bottom := 10.0}
                    Widget := text_block:
                        DefaultText := MakeMessage("Stats")
                        DefaultTextColor := ThemeColors.PrimaryText
                        DefaultTextOpacity := 0.5
                stack_box_slot:
                    HorizontalAlignment := horizontal_alignment.Fill
                    VerticalAlignment := vertical_alignment.Fill
                    Widget := stack_box:
                        Orientation := orientation.Horizontal
                        Slots := array:
                            stack_box_slot:
                                HorizontalAlignment := horizontal_alignment.Left
                                VerticalAlignment := vertical_alignment.Fill
                                Distribution := option{1.0}
                                Widget := text_block:
                                    DefaultText := MakeMessage("Tier")
                                    DefaultTextColor := ThemeColors.PrimaryText
                            stack_box_slot:
                                HorizontalAlignment := horizontal_alignment.Fill
                                VerticalAlignment := vertical_alignment.Fill
                                Widget := GearTierUI.GetUI()
                            stack_box_slot:
                                HorizontalAlignment := horizontal_alignment.Fill
                                VerticalAlignment := vertical_alignment.Fill
                                Padding := margin{Left := 10.0, Top := 0.0, Right := 10.0, Bottom := 0.0}
                                Widget := text_block:
                                    DefaultText := MakeMessage("▶")
                                    DefaultTextColor := ThemeColors.PrimaryText
                                    DefaultShadowColor := NamedColors.Black
                                    DefaultShadowOffset := option{vector2{X := 2.0, Y := 2.0}}
                                    DefaultShadowOpacity := 0.5
                            stack_box_slot:
                                HorizontalAlignment := horizontal_alignment.Fill
                                VerticalAlignment := vertical_alignment.Fill
                                Widget := UpdatedGearTierUI.GetUI()
                stack_box_slot:
                    HorizontalAlignment := horizontal_alignment.Fill
                    VerticalAlignment := vertical_alignment.Fill
                    Widget := stack_box:
                        Orientation := orientation.Horizontal
                        Slots := array:
                            stack_box_slot:
                                HorizontalAlignment := horizontal_alignment.Left
                                VerticalAlignment := vertical_alignment.Fill
                                Distribution := option{1.0}
                                Widget := text_block:
                                    DefaultText := MakeMessage("Max Lv.")
                                    DefaultTextColor := ThemeColors.Highlight
                                    DefaultShadowColor := ThemeColors.Highlight * 0.125
                                    DefaultShadowOffset := option{vector2{X := 2.0, Y := 2.0}}
                                    DefaultShadowOpacity := 1.0
                            stack_box_slot:
                                HorizontalAlignment := horizontal_alignment.Fill
                                VerticalAlignment := vertical_alignment.Fill
                                Widget := MaxLevelText
                            stack_box_slot:
                                HorizontalAlignment := horizontal_alignment.Fill
                                VerticalAlignment := vertical_alignment.Fill
                                Padding := margin{Left := 10.0, Top := 0.0, Right := 10.0, Bottom := 0.0}
                                Widget := text_block:
                                    DefaultText := MakeMessage("▶")
                                    DefaultTextColor := ThemeColors.PrimaryText
                                    DefaultShadowColor := NamedColors.Black
                                    DefaultShadowOffset := option{vector2{X := 2.0, Y := 2.0}}
                                    DefaultShadowOpacity := 0.5
                            stack_box_slot:
                                HorizontalAlignment := horizontal_alignment.Fill
                                VerticalAlignment := vertical_alignment.Fill
                                Widget := UpdatedMaxLevelText
                stack_box_slot:
                    HorizontalAlignment := horizontal_alignment.Fill
                    VerticalAlignment := vertical_alignment.Fill
                    Widget := TierStatEntryUI.GetUI()
    
    CreateCostUI():widget =
        stack_box:
            Orientation := orientation.Horizontal
            Slots := array:
                stack_box_slot:
                    HorizontalAlignment := horizontal_alignment.Fill
                    VerticalAlignment := vertical_alignment.Fill
                    Widget := text_block:
                        DefaultText := MakeMessage("Cost")
                        DefaultTextColor := ThemeColors.PrimaryText
                        DefaultTextOpacity := 0.5
                stack_box_slot:
                    HorizontalAlignment := horizontal_alignment.Right
                    VerticalAlignment := vertical_alignment.Fill
                    Distribution := option{1.0}
                    Widget := stack_box:
                        Orientation := orientation.Vertical
                        Slots := array:
                            stack_box_slot:
                                HorizontalAlignment := horizontal_alignment.Right
                                VerticalAlignment := vertical_alignment.Fill
                                Widget := CreateCostEntry(currency_id.Gold, GoldCostText)
                            stack_box_slot:
                                HorizontalAlignment := horizontal_alignment.Right
                                VerticalAlignment := vertical_alignment.Fill
                                Widget := CreateCostEntry(currency_id.GearRecipe, RecipeCostText)
                            stack_box_slot:
                                HorizontalAlignment := horizontal_alignment.Right
                                VerticalAlignment := vertical_alignment.Fill
                                Widget := stack_box:
                                    Orientation := orientation.Horizontal
                                    Slots := array:
                                        stack_box_slot:
                                            HorizontalAlignment := horizontal_alignment.Fill
                                            VerticalAlignment := vertical_alignment.Fill
                                            Widget := GearCostIcon
                                        stack_box_slot:
                                            HorizontalAlignment := horizontal_alignment.Fill
                                            VerticalAlignment := vertical_alignment.Fill
                                            Padding := margin{Left := 10.0, Top := 0.0, Right := 0.0, Bottom := 0.0}
                                            Widget := GearCostText
        
    CreateCostEntry(Id:currency_id, TextBlock:text_block):widget =
        stack_box:
            Orientation := orientation.Horizontal
            Slots := array:
                stack_box_slot:
                    HorizontalAlignment := horizontal_alignment.Fill
                    VerticalAlignment := vertical_alignment.Fill
                    Widget := texture_block:
                        DefaultImage := Id.ToIcon()
                        DefaultDesiredSize := vector2{X := 40.0, Y := 40.0}
                stack_box_slot:
                    HorizontalAlignment := horizontal_alignment.Fill
                    VerticalAlignment := vertical_alignment.Fill
                    Padding := margin{Left := 10.0, Top := 0.0, Right := 0.0, Bottom := 0.0}
                    Widget := TextBlock

    CreatMaxedUI():widget =
        overlay:
            Slots := array:
                overlay_slot:
                    HorizontalAlignment := horizontal_alignment.Fill
                    VerticalAlignment := vertical_alignment.Fill
                    Widget := UserInterface.VerseWidget.WBP_RoundedBox_Light{}
                overlay_slot:
                    Widget := text_block:
                        DefaultText := MakeMessage("Tier Maxed!")
                        DefaultTextColor := ThemeColors.Success
                        DefaultShadowColor := ThemeColors.Success * 0.125
                        DefaultShadowOffset := option{vector2{X := 2.0, Y := 2.0}}
                        DefaultShadowOpacity := 1.0

    <# ----------------------------------------------------------------------------------------------------
        Update Functions
    ---------------------------------------------------------------------------------------------------- #>
    Update(Slot:gear_slot):void =
        if:
            GearData := Player.GetGear[Slot]
            GearInfoUI := MaybeGearInfoUI?
            GearTierUI := MaybeGearTierUI?
            UpdatedGearTierUI := MaybeUpdatedGearTierUI?
            TierStatEntryUI := MaybeTierStatEntryUI?
        then:
            set ActiveSlot = Slot
            GearInfoUI.Update(Slot)
            
            IsMaxed := GearData.IsTierMaxed()
            UpgradeUI.SetVisibility(IsMaxed? and widget_visibility.Hidden or widget_visibility.Visible)
            UpgradeMaxedUI.SetVisibility(IsMaxed? and widget_visibility.Visible or widget_visibility.Hidden)
            if (IsMaxed?) then return

            GearTierUI.SetTier(GearData.Tier, GearData.Stars)
            
            UpdatedTier := (GearData.Stars < MaxGearStars) and GearData.Tier or GearData.Tier + 1
            UpdatedStars := (GearData.Stars < MaxGearStars) and GearData.Stars + 1 or 0
            UpdatedGearTierUI.SetTier(UpdatedTier, UpdatedStars)

            MaxLevelText.SetText(MakeMessage("{CalculateGearMaxLevel(GearData.Evolution, GearData.Tier, GearData.Stars)}"))
            UpdatedMaxLevelText.SetText(MakeMessage("{CalculateGearMaxLevel(GearData.Evolution, UpdatedTier, UpdatedStars)}"))
            
            if (UpdatedTier > GearData.Tier):
                TierStat := GearData.GetTierStats(Slot, GearData.Tier)
                UpdatedTierStat := GearData.GetTierStats(Slot, GearData.Tier + 1)
                TierStatEntryUI.Update(GetGearSlotInfo(Slot).TierStat, TierStat, UpdatedTierStat)
                TierStatEntryUI.Show()
            else:
                TierStatEntryUI.Hide()

            GearUpgradeCosts := GearData.GetFuseCosts()
            for (UpgradeCost : GearUpgradeCosts.UpgradeCosts):
                if (UpgradeCost.Id = currency_id.Gold):
                    GoldCostText.SetText(MakeMessage("{FormatAsCurrency(UpgradeCost.Amount)}"))
                    GoldCostText.SetColor(Player.CanAfford(currency_id.Gold, UpgradeCost.Amount)? and ThemeColors.Success or ThemeColors.Failure)
                else if (UpgradeCost.Id = currency_id.GearRecipe):
                    RecipeCostText.SetText(MakeMessage("{FormatAsCurrency(UpgradeCost.Amount)}"))
                    RecipeCostText.SetColor(Player.CanAfford(currency_id.GearRecipe, UpgradeCost.Amount)? and ThemeColors.Success or ThemeColors.Failure)

            if (GearInfo := GetGearInfo[GearData.Evolution, Slot]):
                GearCost := GearUpgradeCosts.GearCost
                GearCostIcon.SetImage(GearInfo.IconSmall)
                GearCostText.SetText(MakeMessage("{GearCost}"))
                GearCostText.SetColor(GearData.Quantity >= GearCost and ThemeColors.Success or ThemeColors.Failure)
        
    <# ----------------------------------------------------------------------------------------------------
        Event Handlers
    ---------------------------------------------------------------------------------------------------- #>
    OnClickFuse(Message:widget_message):void =
        if:
            GearData := Player.GetGear[ActiveSlot]
            not GearData.IsTierMaxed()?
            GearUpgradeCosts := GearData.GetFuseCosts()
            Player.CanAfford(GearUpgradeCosts, ActiveSlot)?
        then:
            Player.RemoveCurrency(GearUpgradeCosts.UpgradeCosts)
            Player.RemoveGearQuantity(ActiveSlot, GearUpgradeCosts.GearCost)
            Player.FuseGear(ActiveSlot)
            Update(ActiveSlot)
            FusedEvent.Signal()
            Player.PlayAudio(GearData.Stars >= MaxGearStars and audio_id.LevelUp or audio_id.UpgradeSuccess)
        else:
            Player.PlayAudio(audio_id.Error)
