using { /Fortnite.com/UI }
using { /Verse.org/Colors }
using { /Verse.org/Simulation }
using { /UnrealEngine.com/Temporary/SpatialMath }
using { /UnrealEngine.com/Temporary/UI }

gear_panel_treasure_ui := class(ui_component):
    TreasureChestOpenedEvent:events(tuple()) = events(tuple()){}
    TreasureChestCompleteEvent:events(tuple()) = events(tuple()){}

    InfoText:text_block = text_block:
        DefaultTextColor := ThemeColors.PrimaryText
        DefaultShadowColor := NamedColors.Black
        DefaultShadowOffset := option{vector2{X := 2.0, Y := 2.0}}
        DefaultShadowOpacity := 0.5

    TreasureChestIcon:texture_block = texture_block:
        DefaultImage := UserInterface.Icons.TreasureChest_128
        DefaultDesiredSize := vector2{X := 128.0, Y := 128.0}
        
    CostText:text_block = text_block{}
    OpenButton:button_regular = button_regular{}
    CountButton:button_regular = button_regular{}

    var RewardEntries:[]tuple(widget, texture_block, text_block) = array{}
    var InfoUI:widget = canvas{}
    var RewardsUI:widget = canvas{}
    var Count:int = 1
    var IsOpening:logic = false

    <# ----------------------------------------------------------------------------------------------------
        Initialization
    ---------------------------------------------------------------------------------------------------- #>
    CreateUI<override>():widget =
        OpenButton.SetText(MakeMessage("Open"))
        OpenButton.OnClick().Subscribe(OnClickOpen)
        CountButton.SetText(MakeMessage("x1"))
        CountButton.OnClick().Subscribe(OnClickCount)

        stack_box:
            Orientation := orientation.Vertical
            Slots := array:
                  stack_box_slot:
                    HorizontalAlignment := horizontal_alignment.Fill
                    VerticalAlignment := vertical_alignment.Fill
                    Distribution := option{1.0}
                    Widget := CreateRewardsUI()
                stack_box_slot:
                    HorizontalAlignment := horizontal_alignment.Fill
                    Padding := margin{Left := 0.0, Top := 20.0, Right := 0.0, Bottom := 20.0}
                    Widget := color_block:
                        DefaultColor := NamedColors.White
                        DefaultOpacity := 0.1
                        DefaultDesiredSize := vector2{X := 0.0, Y := 1.0}
                stack_box_slot:
                    HorizontalAlignment := horizontal_alignment.Fill
                    VerticalAlignment := vertical_alignment.Fill
                    Widget := CreateCostUI()
                stack_box_slot:
                    HorizontalAlignment := horizontal_alignment.Fill
                    VerticalAlignment := vertical_alignment.Fill
                    Padding := margin{Left := 0.0, Top := 20.0, Right := 0.0, Bottom := 0.0}
                    Widget := stack_box:
                        Orientation := orientation.Horizontal
                        Slots := array:
                            stack_box_slot:
                                HorizontalAlignment := horizontal_alignment.Fill
                                VerticalAlignment := vertical_alignment.Fill
                                Distribution := option{0.7}
                                Widget := OpenButton
                            stack_box_slot:
                                HorizontalAlignment := horizontal_alignment.Fill
                                VerticalAlignment := vertical_alignment.Fill
                                Padding := margin{Left := 20.0, Top := 0.0, Right := 0.0, Bottom := 0.0}
                                Distribution := option{0.3}
                                Widget := CountButton

    CreateRewardsUI():widget =
        stack_box:
            Orientation := orientation.Vertical
            Slots := array:
                stack_box_slot:
                    HorizontalAlignment := horizontal_alignment.Fill
                    VerticalAlignment := vertical_alignment.Fill
                    Padding := margin{Left := 0.0, Top := 0.0, Right := 0.0, Bottom := 20.0}
                    Widget := text_block:
                        DefaultText := MakeMessage("Rewards")
                        DefaultTextColor := ThemeColors.PrimaryText
                        DefaultTextOpacity := 0.5
                stack_box_slot:
                    HorizontalAlignment := horizontal_alignment.Fill
                    VerticalAlignment := vertical_alignment.Fill
                    Distribution := option{1.0}
                    Widget := overlay:
                        Slots := array:
                            overlay_slot:
                                HorizontalAlignment := horizontal_alignment.Fill
                                VerticalAlignment := vertical_alignment.Fill
                                Widget := UserInterface.VerseWidget.WBP_RoundedBox_Light{}
                            overlay_slot:
                                HorizontalAlignment := horizontal_alignment.Fill
                                Padding := margin{Left := 40.0, Top := 20.0, Right := 40.0, Bottom := 20.0}
                                Widget := set RewardsUI = stack_box:
                                    Orientation := orientation.Horizontal
                                    Slots := for (Index := 1..7):
                                        stack_box_slot:
                                            HorizontalAlignment := (Index = 1) and horizontal_alignment.Left or horizontal_alignment.Center
                                            VerticalAlignment := vertical_alignment.Fill
                                            Distribution := (Index = 1) and option{1.0} or false
                                            Padding := margin{Left := (Index = 1) and 0.0 or 10.0, Top := 0.0, Right := 0.0, Bottom := 0.0}
                                            Widget := CreateRewardEntryUI()
                            overlay_slot:
                                Padding := margin{Left := 40.0, Top := 20.0, Right := 40.0, Bottom := 20.0}
                                Widget := set InfoUI = CreateInfoUI()

    CreateInfoUI():widget =
        stack_box:
            Orientation := orientation.Vertical
            Slots := array:
                stack_box_slot:
                    VerticalAlignment := vertical_alignment.Top
                    Widget := TreasureChestIcon
                stack_box_slot:
                    Widget := InfoText

    CreateRewardEntryUI():widget =
        RewardIcon := texture_block:
            DefaultImage := UserInterface.Icons.GearRecipe_64
            DefaultDesiredSize := vector2{X := 64.0, Y := 64.0}

        RewardText := text_block:
            DefaultTextColor := ThemeColors.PrimaryText
            DefaultShadowColor := NamedColors.Black
            DefaultShadowOffset := option{vector2{X := 2.0, Y := 2.0}}
            DefaultShadowOpacity := 0.5

        RewardStackBox := stack_box:
            Orientation := orientation.Vertical
            Slots := array:
                stack_box_slot:
                    Widget := RewardIcon
                stack_box_slot:
                    VerticalAlignment := vertical_alignment.Fill
                    Widget := RewardText

        set RewardEntries += array{(RewardStackBox, RewardIcon, RewardText)}
        RewardStackBox        

    CreateCostUI():widget =
        stack_box:
            Orientation := orientation.Horizontal
            Slots := array:
                stack_box_slot:
                    HorizontalAlignment := horizontal_alignment.Fill
                    VerticalAlignment := vertical_alignment.Fill
                    Widget := text_block:
                        DefaultText := MakeMessage("Cost")
                        DefaultTextColor := ThemeColors.PrimaryText
                        DefaultTextOpacity := 0.5
                stack_box_slot:
                    HorizontalAlignment := horizontal_alignment.Right
                    VerticalAlignment := vertical_alignment.Fill
                    Distribution := option{1.0}
                    Widget := stack_box:
                        Orientation := orientation.Vertical
                        Slots := array:
                            stack_box_slot:
                                HorizontalAlignment := horizontal_alignment.Right
                                VerticalAlignment := vertical_alignment.Fill
                                Widget := stack_box:
                                    Orientation := orientation.Horizontal
                                    Slots := array:
                                        stack_box_slot:
                                            HorizontalAlignment := horizontal_alignment.Fill
                                            VerticalAlignment := vertical_alignment.Fill
                                            Widget := texture_block:
                                                DefaultImage := UserInterface.Icons.TreasureChest_64
                                                DefaultDesiredSize := vector2{X := 40.0, Y := 40.0}
                                        stack_box_slot:
                                            HorizontalAlignment := horizontal_alignment.Fill
                                            VerticalAlignment := vertical_alignment.Fill
                                            Padding := margin{Left := 10.0, Top := 0.0, Right := 0.0, Bottom := 0.0}
                                            Widget := CostText
        
    <# ----------------------------------------------------------------------------------------------------
        Get Functions
    ---------------------------------------------------------------------------------------------------- #>
    GetCost()<transacts>:int =
        Min(Count, Floor[Player.GetCurrency(currency_id.TreasureChest)] or 0)

    <# ----------------------------------------------------------------------------------------------------
        Update Functions
    ---------------------------------------------------------------------------------------------------- #>
    Update<override>():void =
        UpdateCost()
        InfoText.SetText(MakeMessage("Treasure Chest"))
        TreasureChestIcon.SetImage(UserInterface.Icons.TreasureChest_128)
        RewardsUI.SetVisibility(widget_visibility.Hidden)
        InfoUI.SetVisibility(widget_visibility.Visible)

    UpdateCost():void =
        Cost := GetCost()
        CostText.SetText(MakeMessage("{Cost = 0 and Count or Cost}"))
        CostText.SetColor(Cost > 0 and Player.CanAfford(currency_id.TreasureChest, Cost * 1.0)? and ThemeColors.Success or ThemeColors.Failure)
        
    OpenAnimation(Rewards:tuple(int, [int]int))<suspends>:void =
        set IsOpening = true
        TreasureChestIcon.SetImage(UserInterface.Icons.TreasureChest_128)
        OpenButton.SetText(MakeMessage("Opening ..."))
        RewardsUI.SetVisibility(widget_visibility.Hidden)
        InfoUI.SetVisibility(widget_visibility.Visible)
        for (Index -> RewardEntry : RewardEntries):
            RewardEntry(0).SetVisibility(widget_visibility.Collapsed)

        Sleep(0.5)
        TreasureChestIcon.SetImage(UserInterface.Icons.TreasureChestOpen_128)
        Player.PlayAudio(audio_id.OpenTreasureChest)
        Sleep(0.5)

        RewardsUI.SetVisibility(widget_visibility.Visible)
        InfoUI.SetVisibility(widget_visibility.Hidden)
        
        for (Index -> RewardEntry : RewardEntries):
            if (Index = 0):
                RewardEntry(0).SetVisibility(widget_visibility.Visible)
                RewardEntry(1).SetImage(currency_id.GearRecipe.ToIcon())
                AnimateText(Rewards(0) * 1.0, RewardEntry(2))
            else:
                GearSlotIndex := Index - 1
                GearEvolution := Player.GetGear[GearSlotIndex].Evolution or 0
                if (GearReward := Rewards(1)[GearSlotIndex], GearInfo := GetGearInfo[GearEvolution, (GearSlotIndex).ToGearSlot()]):
                    RewardEntry(0).SetVisibility(widget_visibility.Visible)
                    RewardEntry(1).SetImage(GearInfo.IconSmall)
                    AnimateText(GearReward * 1.0, RewardEntry(2))
                else:
                    RewardEntry(0).SetVisibility(widget_visibility.Collapsed)

            Player.PlayAudio(audio_id.Purchase)

        OpenButton.SetText(MakeMessage("Open"))
        TreasureChestCompleteEvent.Signal()
        set IsOpening = false

    AnimateText(Value:float, TextBlock:text_block)<suspends>:void =
        DeltaTimer := delta_timer{}
        loop:
            ElapsedTime := DeltaTimer.GetElapsedTime()
            Progress := ElapsedTime / 0.5
            CurrentAmount := Lerp(0.0, Value, Progress)
            TextBlock.SetText(MakeMessage("{FormatAsCurrency(CurrentAmount)}"))
            Sleep(0.0)
            if (Progress >= 1.0):
                TextBlock.SetText(MakeMessage("{FormatAsCurrency(Value)}"))
                break

    <# ----------------------------------------------------------------------------------------------------
        Event Handlers
    ---------------------------------------------------------------------------------------------------- #>
    OnClickOpen(Message:widget_message):void =
        if (IsOpening?) then return
        Cost := GetCost()
        if (Cost = 0 or not Player.CanAfford(currency_id.TreasureChest, Cost * 1.0)?):
            Player.PlayAudio(audio_id.Error)
            return

        Player.RemoveCurrency(currency_id.TreasureChest, Cost * 1.0)
        UpdateCost()
        TreasureChestOpenedEvent.Signal()

        GearTreasureChestRewards := GetGearTreasureChestRewards(Cost)
        Player.AddCurrency(currency_id.GearRecipe, GearTreasureChestRewards(0) * 1.0)
        Player.GrantGearRewards(GearTreasureChestRewards(1))        
        spawn{OpenAnimation(GearTreasureChestRewards)}

    OnClickCount(Message:widget_message):void =
        if (IsOpening?) then return
        set Count = case (Count):
            1 => 10
            10 => 100
            100 => 1000
            _ => 1
        
        CountButton.SetText(MakeMessage("x{Count}"))
        UpdateCost()