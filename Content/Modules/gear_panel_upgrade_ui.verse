using { /Fortnite.com/UI }
using { /Verse.org/Colors }
using { /Verse.org/Simulation }
using { /UnrealEngine.com/Temporary/SpatialMath }
using { /UnrealEngine.com/Temporary/UI }

<# ----------------------------------------------------------------------------------------------------
    gear_panel_upgrade_ui
---------------------------------------------------------------------------------------------------- #>
gear_panel_upgrade_ui := class(ui_component):
    UpgradedEvent:events(tuple()) = events(tuple()){}
    
    LevelText:text_block = text_block:
        DefaultTextColor := ThemeColors.Success
        DefaultShadowColor := ThemeColors.Success * 0.125
        DefaultShadowOffset := option{vector2{X := 2.0, Y := 2.0}}
        DefaultShadowOpacity := 1.0

    UpdatedLevelText:text_block = text_block:
        DefaultTextColor := ThemeColors.Success
        DefaultShadowColor := ThemeColors.Success * 0.125
        DefaultShadowOffset := option{vector2{X := 2.0, Y := 2.0}}
        DefaultShadowOpacity := 1.0

    MaxedText:text_block = text_block:
        DefaultTextColor := ThemeColors.Success
        DefaultShadowColor := ThemeColors.Success * 0.125
        DefaultShadowOffset := option{vector2{X := 2.0, Y := 2.0}}
        DefaultShadowOpacity := 1.0

    GoldCostText:text_block = text_block{}
    RecipeCostText:text_block = text_block{}
    CountText:text_block = text_block:
        DefaultTextColor := ThemeColors.PrimaryText
        DefaultShadowOffset := option{vector2{X := 2.0, Y := 2.0}}
        DefaultShadowOpacity := 0.5

    UpgradeButton:button_regular = button_regular{}
    CountButton:button_regular = button_regular{}

    var MaybeGearInfoUI:?gear_info_ui = false
    var MaybeStatEntryUI:?gear_stat_upgrade_entry_ui = false
    
    var ActiveSlot:gear_slot = gear_slot.Head
    var UpgradeUI:widget = canvas{}
    var UpgradeMaxedUI:widget = canvas{}
    var Count:int = 1
    var AffordableCount:int = 0
    
    <# ----------------------------------------------------------------------------------------------------
        Initialization
    ---------------------------------------------------------------------------------------------------- #>
    CreateUI<override>():widget =
        GearInfoUI := gear_info_ui{Player := Player}
        GearInfoUI.Initialize()
        set MaybeGearInfoUI = option{GearInfoUI}

        UpgradeButton.SetText(MakeMessage("Upgrade"))
        UpgradeButton.OnClick().Subscribe(OnClickUpgrade)

        CountButton.SetText(MakeMessage("x1"))
        CountButton.OnClick().Subscribe(OnClickCount)

        stack_box:
            Orientation := orientation.Vertical
            Slots := array:
                stack_box_slot:
                    HorizontalAlignment := horizontal_alignment.Fill
                    VerticalAlignment := vertical_alignment.Fill
                    Widget := GearInfoUI.GetUI()
                stack_box_slot:
                    HorizontalAlignment := horizontal_alignment.Fill
                    VerticalAlignment := vertical_alignment.Fill
                    Padding := margin{Left := 0.0, Top := 20.0, Right := 0.0, Bottom := 20.0}
                    Widget := color_block:
                        DefaultColor := NamedColors.White
                        DefaultOpacity := 0.1
                        DefaultDesiredSize := vector2{X := 0.0, Y := 1.0}
                stack_box_slot:
                    HorizontalAlignment := horizontal_alignment.Fill
                    VerticalAlignment := vertical_alignment.Fill
                    Distribution := option{1.0}
                    Widget := overlay:
                        Slots := array:
                            overlay_slot:
                                HorizontalAlignment := horizontal_alignment.Fill
                                VerticalAlignment := vertical_alignment.Fill
                                Widget := set UpgradeUI = stack_box:
                                    Orientation := orientation.Vertical
                                    Slots := array:
                                        stack_box_slot:
                                            HorizontalAlignment := horizontal_alignment.Fill
                                            VerticalAlignment := vertical_alignment.Fill
                                            Widget := CreateStatsUI()
                                        stack_box_slot:
                                            HorizontalAlignment := horizontal_alignment.Fill
                                            VerticalAlignment := vertical_alignment.Bottom
                                            Padding := margin{Left := 0.0, Top := 20.0, Right := 0.0, Bottom := 20.0}
                                            Distribution := option{1.0}
                                            Widget := color_block:
                                                DefaultColor := NamedColors.White
                                                DefaultOpacity := 0.1
                                                DefaultDesiredSize := vector2{X := 0.0, Y := 1.0}
                                        stack_box_slot:
                                            HorizontalAlignment := horizontal_alignment.Fill
                                            VerticalAlignment := vertical_alignment.Fill
                                            Widget := CreateCostUI()
                            overlay_slot:
                                HorizontalAlignment := horizontal_alignment.Fill
                                VerticalAlignment := vertical_alignment.Fill
                                Widget := set UpgradeMaxedUI = CreatMaxedUI()
                stack_box_slot:
                    HorizontalAlignment := horizontal_alignment.Fill
                    VerticalAlignment := vertical_alignment.Fill
                    Padding := margin{Left := 0.0, Top := 20.0, Right := 0.0, Bottom := 0.0}
                    Widget := stack_box:
                        Orientation := orientation.Horizontal
                        Slots := array:
                            stack_box_slot:
                                HorizontalAlignment := horizontal_alignment.Fill
                                VerticalAlignment := vertical_alignment.Fill
                                Distribution := option{0.7}
                                Widget := UpgradeButton
                            stack_box_slot:
                                HorizontalAlignment := horizontal_alignment.Fill
                                VerticalAlignment := vertical_alignment.Fill
                                Padding := margin{Left := 20.0, Top := 0.0, Right := 0.0, Bottom := 0.0}
                                Distribution := option{0.3}
                                Widget := CountButton

    CreateStatsUI():widget =
        StatEntryUI := gear_stat_upgrade_entry_ui{Player := Player}
        StatEntryUI.Initialize()
        set MaybeStatEntryUI = option{StatEntryUI}

        stack_box:
            Orientation := orientation.Vertical
            Slots := array:
                stack_box_slot:
                    HorizontalAlignment := horizontal_alignment.Fill
                    VerticalAlignment := vertical_alignment.Fill
                    Padding := margin{Left := 0.0, Top := 0.0, Right := 0.0, Bottom := 10.0}
                    Widget := text_block:
                        DefaultText := MakeMessage("Stats")
                        DefaultTextColor := ThemeColors.PrimaryText
                        DefaultTextOpacity := 0.5
                stack_box_slot:
                    HorizontalAlignment := horizontal_alignment.Fill
                    VerticalAlignment := vertical_alignment.Fill
                    Widget := stack_box:
                        Orientation := orientation.Horizontal
                        Slots := array:
                            stack_box_slot:
                                HorizontalAlignment := horizontal_alignment.Left
                                VerticalAlignment := vertical_alignment.Fill
                                Distribution := option{1.0}
                                Widget := text_block:
                                    DefaultText := MakeMessage("Lv.")
                                    DefaultTextColor := ThemeColors.Highlight
                                    DefaultShadowColor := ThemeColors.Highlight * 0.125
                                    DefaultShadowOffset := option{vector2{X := 2.0, Y := 2.0}}
                                    DefaultShadowOpacity := 1.0
                            stack_box_slot:
                                HorizontalAlignment := horizontal_alignment.Fill
                                VerticalAlignment := vertical_alignment.Fill
                                Widget := LevelText
                            stack_box_slot:
                                HorizontalAlignment := horizontal_alignment.Fill
                                VerticalAlignment := vertical_alignment.Fill
                                Padding := margin{Left := 10.0, Top := 0.0, Right := 10.0, Bottom := 0.0}
                                Widget := text_block:
                                    DefaultText := MakeMessage("▶")
                                    DefaultTextColor := ThemeColors.PrimaryText
                                    DefaultShadowColor := NamedColors.Black
                                    DefaultShadowOffset := option{vector2{X := 2.0, Y := 2.0}}
                                    DefaultShadowOpacity := 0.5
                            stack_box_slot:
                                HorizontalAlignment := horizontal_alignment.Fill
                                VerticalAlignment := vertical_alignment.Fill
                                Widget := UpdatedLevelText
                stack_box_slot:
                    HorizontalAlignment := horizontal_alignment.Fill
                    VerticalAlignment := vertical_alignment.Fill
                    Widget := StatEntryUI.GetUI()
    
    CreateCostUI():widget =
        stack_box:
            Orientation := orientation.Horizontal
            Slots := array:
                stack_box_slot:
                    HorizontalAlignment := horizontal_alignment.Fill
                    VerticalAlignment := vertical_alignment.Fill
                    Widget := text_block:
                        DefaultText := MakeMessage("Cost")
                        DefaultTextColor := ThemeColors.PrimaryText
                        DefaultTextOpacity := 0.5
                stack_box_slot:
                    HorizontalAlignment := horizontal_alignment.Right
                    VerticalAlignment := vertical_alignment.Fill
                    Distribution := option{1.0}
                    Widget := stack_box:
                        Orientation := orientation.Vertical
                        Slots := array:
                            stack_box_slot:
                                HorizontalAlignment := horizontal_alignment.Right
                                VerticalAlignment := vertical_alignment.Fill
                                Widget := CreateCostEntry(currency_id.Gold, GoldCostText)
                            stack_box_slot:
                                HorizontalAlignment := horizontal_alignment.Right
                                VerticalAlignment := vertical_alignment.Fill
                                Widget := CreateCostEntry(currency_id.GearRecipe, RecipeCostText)
                            stack_box_slot:
                                HorizontalAlignment := horizontal_alignment.Right
                                VerticalAlignment := vertical_alignment.Fill
                                Widget := CountText

    CreateCostEntry(Id:currency_id, TextBlock:text_block):widget =
        stack_box:
            Orientation := orientation.Horizontal
            Slots := array:
                stack_box_slot:
                    HorizontalAlignment := horizontal_alignment.Fill
                    VerticalAlignment := vertical_alignment.Fill
                    Widget := texture_block:
                        DefaultImage := Id.ToIcon()
                        DefaultDesiredSize := vector2{X := 40.0, Y := 40.0}
                stack_box_slot:
                    HorizontalAlignment := horizontal_alignment.Fill
                    VerticalAlignment := vertical_alignment.Fill
                    Padding := margin{Left := 10.0, Top := 0.0, Right := 0.0, Bottom := 0.0}
                    Widget := TextBlock

    CreatMaxedUI():widget =
        overlay:
            Slots := array:
                overlay_slot:
                    HorizontalAlignment := horizontal_alignment.Fill
                    VerticalAlignment := vertical_alignment.Fill
                    Widget := UserInterface.VerseWidget.WBP_RoundedBox_Light{}
                overlay_slot:
                    Widget := MaxedText

    <# ----------------------------------------------------------------------------------------------------
        Update Functions
    ---------------------------------------------------------------------------------------------------- #>
    Update(Slot:gear_slot):void =
        if:
            GearData := Player.GetGear[Slot]
            GearInfoUI := MaybeGearInfoUI?
            StatEntryUI := MaybeStatEntryUI?
        then:
            set ActiveSlot = Slot
            set AffordableCount = Player.GetAffordableUpgradeCount(GearData, Count)

            GearInfoUI.Update(Slot)

            IsMaxed := GearData.IsLevelMaxed()
            UpgradeUI.SetVisibility(IsMaxed? and widget_visibility.Hidden or widget_visibility.Visible)
            UpgradeMaxedUI.SetVisibility(IsMaxed? and widget_visibility.Visible or widget_visibility.Hidden)
            if (IsMaxed?):
                MaxedText.SetText(MakeMessage("Level Maxed!" + (GearData.IsTierMaxed()? and "" or "\nTier up to increase max level.")))
                return

            LevelText.SetText(MakeMessage("{GearData.Level}"))
            UpdatedLevelText.SetText(MakeMessage("{Min(GearData.GetMaxLevel(), GearData.Level + AffordableCount)}"))

            Value := GearData.GetStat(Slot, GearData.Level)
            UpdatedValue := GearData.GetStat(Slot, GearData.Level + AffordableCount)
            StatEntryUI.Update(GetGearSlotInfo(Slot).Stat, Value, UpdatedValue)

            CountText.SetText(MakeMessage("x{AffordableCount}"))
            
            for (UpgradeCosts : GearData.GetUpgradeCosts(AffordableCount)):
                if (UpgradeCosts.Id = currency_id.Gold):
                    GoldCostText.SetText(MakeMessage("{FormatAsCurrency(UpgradeCosts.Amount)}"))
                    GoldCostText.SetColor(Player.CanAfford(currency_id.Gold, UpgradeCosts.Amount)? and ThemeColors.Success or ThemeColors.Failure)
                else if (UpgradeCosts.Id = currency_id.GearRecipe):
                    RecipeCostText.SetText(MakeMessage("{FormatAsCurrency(UpgradeCosts.Amount)}"))
                    RecipeCostText.SetColor(Player.CanAfford(currency_id.GearRecipe, UpgradeCosts.Amount)? and ThemeColors.Success or ThemeColors.Failure)
        
    <# ----------------------------------------------------------------------------------------------------
        Event Handlers
    ---------------------------------------------------------------------------------------------------- #>
    OnClickUpgrade(Message:widget_message):void =
        if:
            GearData := Player.GetGear[ActiveSlot]
            not GearData.IsLevelMaxed()?
            AffordableCount > 0
            UpgradeCosts := GearData.GetUpgradeCosts(AffordableCount)
            MaxLevel := GearData.GetMaxLevel()
            GearData.Level < MaxLevel
            Player.CanAfford(UpgradeCosts)?
        then:
            Player.RemoveCurrency(UpgradeCosts)
            Player.UpgradeGear(ActiveSlot, AffordableCount)
            Update(ActiveSlot)
            UpgradedEvent.Signal()

            IsMaxed := logic{GearData.Level + 1 = MaxLevel}
            Player.PlayAudio(IsMaxed? and audio_id.UpgradeComplete or audio_id.UpgradeSuccess)
        else:
            Player.PlayAudio(audio_id.Error)

    OnClickCount(Message:widget_message):void =
        set Count = case (Count):
            1 => 10
            10 => 100
            _ => 1
        
        CountButton.SetText(MakeMessage("x{Count}"))
        Update(ActiveSlot)