using { /Fortnite.com/UI }
using { /Verse.org/Colors }
using { /Verse.org/Simulation }
using { /UnrealEngine.com/Temporary/SpatialMath }
using { /UnrealEngine.com/Temporary/UI }

<# ----------------------------------------------------------------------------------------------------
    gear_stat_upgrade_entry_ui
---------------------------------------------------------------------------------------------------- #>
gear_stat_upgrade_entry_ui := class(ui_component):
    StatIcon:texture_block = texture_block:
        DefaultImage := UserInterface.Icons.ProjectileDamage_64
        DefaultDesiredSize := vector2{X := 40.0, Y := 40.0}

    StatLabelText:text_block = text_block:
        DefaultTextColor := ThemeColors.PrimaryText

    StatText:text_block = text_block:
        DefaultTextColor := ThemeColors.Success
        DefaultShadowColor := ThemeColors.Success * 0.125
        DefaultShadowOffset := option{vector2{X := 2.0, Y := 2.0}}
        DefaultShadowOpacity := 1.0

    UpdatedStatText:text_block = text_block:
        DefaultTextColor := ThemeColors.Success
        DefaultShadowColor := ThemeColors.Success * 0.125
        DefaultShadowOffset := option{vector2{X := 2.0, Y := 2.0}}
        DefaultShadowOpacity := 1.0

    <# ----------------------------------------------------------------------------------------------------
        Initialization
    ---------------------------------------------------------------------------------------------------- #>
    CreateUI<override>():widget =
        stack_box:
            Orientation := orientation.Horizontal
            Slots := array:
                stack_box_slot:
                    HorizontalAlignment := horizontal_alignment.Fill
                    VerticalAlignment := vertical_alignment.Fill
                    Widget := StatIcon
                stack_box_slot:
                    HorizontalAlignment := horizontal_alignment.Left
                    VerticalAlignment := vertical_alignment.Fill
                    Distribution := option{1.0}
                    Padding := margin{Left := 10.0, Top := 0.0, Right := 0.0, Bottom := 0.0}
                    Widget := StatLabelText
                stack_box_slot:
                    HorizontalAlignment := horizontal_alignment.Fill
                    VerticalAlignment := vertical_alignment.Fill
                    Widget := StatText
                stack_box_slot:
                    HorizontalAlignment := horizontal_alignment.Fill
                    VerticalAlignment := vertical_alignment.Fill
                    Padding := margin{Left := 10.0, Top := 0.0, Right := 10.0, Bottom := 0.0}
                    Widget := text_block:
                        DefaultText := MakeMessage("▶")
                        DefaultTextColor := ThemeColors.PrimaryText
                        DefaultShadowColor := NamedColors.Black
                        DefaultShadowOffset := option{vector2{X := 2.0, Y := 2.0}}
                        DefaultShadowOpacity := 0.5
                stack_box_slot:
                    HorizontalAlignment := horizontal_alignment.Fill
                    VerticalAlignment := vertical_alignment.Fill
                    Widget := UpdatedStatText

    <# ----------------------------------------------------------------------------------------------------
        Update Functions
    ---------------------------------------------------------------------------------------------------- #>
    Update(Id:stat_id, Value:float, UpdatedValue:float):void =
        StatIcon.SetImage(Id.ToIcon())
        StatLabelText.SetText(MakeMessage("{Id}"))
        StatText.SetText(MakeMessage("+{Value.Format(Id)}"))
        UpdatedStatText.SetText(MakeMessage("+{UpdatedValue.Format(Id)}"))