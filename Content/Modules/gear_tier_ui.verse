using { /Fortnite.com/UI }
using { /Verse.org/Colors }
using { /UnrealEngine.com/Temporary/SpatialMath }
using { /UnrealEngine.com/Temporary/UI }

<# ----------------------------------------------------------------------------------------------------
    gear_tier_ui
---------------------------------------------------------------------------------------------------- #>
gear_tier_ui := class(ui_component):
    TierOverlay:overlay = overlay{}

    var MaybeTierWidget:?widget = false
    var StarIcons:[]texture_block = array{}

    <# ----------------------------------------------------------------------------------------------------
        Initialization
    ---------------------------------------------------------------------------------------------------- #>
    CreateUI<override>():widget =
        stack_box:
            Orientation := orientation.Horizontal
            Slots := array:
                stack_box_slot:
                    HorizontalAlignment := horizontal_alignment.Fill
                    VerticalAlignment := vertical_alignment.Fill
                    Widget := TierOverlay
                stack_box_slot:
                    HorizontalAlignment := horizontal_alignment.Fill
                    VerticalAlignment := vertical_alignment.Fill
                    Padding := margin{Left := 20.0, Top := -4.0, Right := 0.0, Bottom := 0.0}
                    Widget := stack_box:
                        Orientation := orientation.Horizontal
                        Slots := for (Index := 1..5):
                            StarIcon := texture_block:
                                DefaultImage := UserInterface.Icons.Star_64
                                DefaultTint := ThemeColors.Highlight
                                DefaultDesiredSize := vector2{X := 24.0, Y := 24.0}
                
                            set StarIcons += array{StarIcon}
                            stack_box_slot{Widget := StarIcon}

    <# ----------------------------------------------------------------------------------------------------
        Update Functions
    ---------------------------------------------------------------------------------------------------- #>
    SetTier(Tier:int, Stars:int):void =
        if (TierWidget := MaybeTierWidget?):
            TierOverlay.RemoveWidget of TierWidget

        UpdatedTierWidget := Tier.ToGearTierWidget()
        set MaybeTierWidget = option{UpdatedTierWidget}

        TierOverlay.AddWidget of overlay_slot:
            HorizontalAlignment := horizontal_alignment.Left
            VerticalAlignment := vertical_alignment.Top
            Widget := UpdatedTierWidget

        for (Index -> StarIcon : StarIcons):
            #StarIcon.SetVisibility(Stars >= (Index + 1) and widget_visibility.Visible or widget_visibility.Collapsed)
            StarIcon.SetTint(Stars >= (Index + 1) and ThemeColors.Highlight or ThemeColors.Inactive)