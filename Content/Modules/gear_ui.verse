using { /Fortnite.com/UI }
using { /Verse.org/Assets }
using { /Verse.org/Colors }
using { /UnrealEngine.com/Temporary/SpatialMath }
using { /UnrealEngine.com/Temporary/UI }

<# ----------------------------------------------------------------------------------------------------
    Data Definitions
---------------------------------------------------------------------------------------------------- #>
gear_ui_panel := enum:
    Treasure
    Info
    Summary
    Upgrade
    Tier
    Evolve

ToString(Panel:gear_ui_panel)<transacts>:string =
    case (Panel):
        gear_ui_panel.Treasure => "Treasure"
        gear_ui_panel.Info => "Info"
        gear_ui_panel.Summary => "Summary"
        gear_ui_panel.Upgrade => "Upgrade"
        gear_ui_panel.Tier => "Tier"
        gear_ui_panel.Evolve => "Evolve"

GearRequiredPanels:[]gear_ui_panel = array:
    gear_ui_panel.Info
    gear_ui_panel.Upgrade
    gear_ui_panel.Tier
    gear_ui_panel.Evolve

<# ----------------------------------------------------------------------------------------------------
    gear_ui
---------------------------------------------------------------------------------------------------- #>
gear_ui := class(panel_ui):
    Title<override>:string = "Gear Smith"
    Anchors<override>:anchors = anchors{Minimum := vector2{X := 0.0, Y := 0.0}, Maximum := vector2{X := 1.0, Y := 1.0}}

    var GearEntryUIs:[]gear_entry_ui = array{}
    var GearPanelUIs:[gear_ui_panel]ui_component = map{}
    var CurrencyTexts:[currency_id]text_block = map{}
    var EmptySlotUI:widget = canvas{}
    var ActivePanel:gear_ui_panel = gear_ui_panel.Treasure
    var ActiveIndex:int = 0
    var IsOpening:logic = false

    <# ----------------------------------------------------------------------------------------------------
        Initialization
    ---------------------------------------------------------------------------------------------------- #>
    CreateContent<override>():widget =
        stack_box:
            Orientation := orientation.Horizontal
            Slots := array:
                stack_box_slot:
                    HorizontalAlignment := horizontal_alignment.Fill
                    VerticalAlignment := vertical_alignment.Fill
                    Widget := overlay:
                        Slots := array:
                            overlay_slot:
                                HorizontalAlignment := horizontal_alignment.Fill
                                VerticalAlignment := vertical_alignment.Fill
                                Widget := UserInterface.VerseWidget.WBP_RoundedBox{}
                            overlay_slot:
                                HorizontalAlignment := horizontal_alignment.Fill
                                VerticalAlignment := vertical_alignment.Fill
                                Padding := margin{Left := 40.0, Top := 40.0, Right := 40.0, Bottom := 40.0}
                                Widget := CreateMenuUI()
                                
                stack_box_slot:
                    HorizontalAlignment := horizontal_alignment.Fill
                    VerticalAlignment := vertical_alignment.Fill
                    Distribution := option{1.0}
                    Padding := margin{Left := 20.0, Top := 0.0, Right := 0.0, Bottom := 0.0}
                    Widget := overlay:
                        Slots := array:
                            overlay_slot:
                                HorizontalAlignment := horizontal_alignment.Fill
                                VerticalAlignment := vertical_alignment.Fill
                                Widget := UserInterface.VerseWidget.WBP_RoundedBox{}
                            overlay_slot:
                                HorizontalAlignment := horizontal_alignment.Fill
                                VerticalAlignment := vertical_alignment.Fill
                                Padding := margin{Left := 40.0, Top := 40.0, Right := 40.0, Bottom := 40.0}
                                Widget := CreateEquippedUI()
                stack_box_slot:
                    HorizontalAlignment := horizontal_alignment.Fill
                    VerticalAlignment := vertical_alignment.Fill
                    Distribution := option{1.0}
                    Padding := margin{Left := 20.0, Top := 0.0, Right := 0.0, Bottom := 0.0}
                    Widget := overlay:
                        Slots := array:
                            overlay_slot:
                                HorizontalAlignment := horizontal_alignment.Fill
                                VerticalAlignment := vertical_alignment.Fill
                                Widget := UserInterface.VerseWidget.WBP_RoundedBox{}
                            overlay_slot:
                                HorizontalAlignment := horizontal_alignment.Fill
                                VerticalAlignment := vertical_alignment.Fill
                                Padding := margin{Left := 40.0, Top := 40.0, Right := 40.0, Bottom := 40.0}
                                Widget := CreateInfoUI()
                            overlay_slot:
                                HorizontalAlignment := horizontal_alignment.Fill
                                VerticalAlignment := vertical_alignment.Fill
                                Widget := set EmptySlotUI = CreateEmptySlotUI()

    CreateMenuUI():widget =
        stack_box:
            Orientation := orientation.Vertical
            Slots := array:
                stack_box_slot:
                    HorizontalAlignment := horizontal_alignment.Fill
                    VerticalAlignment := vertical_alignment.Fill
                    Widget := CreateCurrenciesUI()
                stack_box_slot:
                    HorizontalAlignment := horizontal_alignment.Fill
                    VerticalAlignment := vertical_alignment.Bottom
                    Distribution := option{1.0}
                    Widget := CreateMenuButtonsUI()

    CreateCurrenciesUI():widget =
        CurrencyIds := array:
            currency_id.Gold
            currency_id.TreasureChest
            currency_id.GearRecipe

        stack_box:
            Orientation := orientation.Vertical
            Slots := for (Index -> CurrencyId : CurrencyIds):
                stack_box_slot:
                    HorizontalAlignment := horizontal_alignment.Fill
                    VerticalAlignment := vertical_alignment.Fill
                    Padding := margin{Left := 0.0, Top := (Index = 0) and 0.0 or 10.0, Right := 0.0, Bottom := 0.0}
                    Widget := CreateCurrencyUI(CurrencyId)

    CreateCurrencyUI(Id:currency_id):widget =
        CurrencyText := text_block:
            DefaultTextColor := ThemeColors.Success
            DefaultShadowColor := ThemeColors.Success * 0.125
            DefaultShadowOffset := option{vector2{X := 2.0, Y := 2.0}}
            DefaultShadowOpacity := 1.0

        if (set CurrencyTexts[Id] = CurrencyText) {}

        overlay:
            Slots := array:
                overlay_slot:
                    HorizontalAlignment := horizontal_alignment.Fill
                    VerticalAlignment := vertical_alignment.Fill
                    Widget := UserInterface.VerseWidget.WBP_RoundedBox{}
                overlay_slot:
                    HorizontalAlignment := horizontal_alignment.Fill
                    VerticalAlignment := vertical_alignment.Fill
                    Padding := margin{Left := 40.0, Top := 10.0, Right := 40.0, Bottom := 10.0}
                    Widget := stack_box:
                        Orientation := orientation.Horizontal
                        Slots := array:
                            stack_box_slot:
                                HorizontalAlignment := horizontal_alignment.Fill
                                VerticalAlignment := vertical_alignment.Fill
                                Widget := texture_block:
                                    DefaultImage := Id.ToIcon()
                                    DefaultDesiredSize := vector2{X := 40.0, Y := 40.0}
                            stack_box_slot:
                                HorizontalAlignment := horizontal_alignment.Fill
                                VerticalAlignment := vertical_alignment.Fill
                                Padding := margin{Left := 10.0, Top := 0.0, Right := 0.0, Bottom := 0.0}
                                Widget := CurrencyText

    CreateMenuButtonsUI():widget =
        MenuButtons := array:
            (gear_ui_panel.Treasure, "Treasure")
            (gear_ui_panel.Info, "Info")
            (gear_ui_panel.Upgrade, "Upgrade")
            (gear_ui_panel.Tier, "Fuse")
            (gear_ui_panel.Evolve, "Evolve")   
            (gear_ui_panel.Summary, "Summary")
            
        stack_box:
            Orientation := orientation.Vertical
            Slots := for (Index -> MenuButton : MenuButtons):
                Button := button_quiet{DefaultText := MakeMessage("{MenuButton(1)}")}
                spawn{AwaitMenuButtonClicked(Button.OnClick(), MenuButton(0))}

                stack_box_slot:
                    HorizontalAlignment := horizontal_alignment.Fill
                    VerticalAlignment := vertical_alignment.Fill
                    Padding := margin{Left := 0.0, Top := (Index = 0) and 0.0 or 10.0, Right := 0.0, Bottom := 0.0}
                    Widget := Button

    CreateEquippedUI():widget =
        stack_box:
            Orientation := orientation.Horizontal
            Slots := for (ColumnIndex := 1..2):
                stack_box_slot:
                    HorizontalAlignment := horizontal_alignment.Fill
                    VerticalAlignment := vertical_alignment.Fill
                    Distribution := option{1.0}
                    Padding := margin{Left := (ColumnIndex = 1) and 0.0 or 20.0, Top := 0.0, Right := 0.0, Bottom := 0.0}
                    Widget := stack_box:
                        Orientation := orientation.Vertical
                        Slots := for (RowIndex := 1..3):
                            EntryIndex := (ColumnIndex - 1) * 3 + (RowIndex - 1)
                            GearEntryUI := gear_entry_ui{EntryIndex := EntryIndex, Player := Player}
                            GearEntryUI.Initialize()
                            GearEntryUI.SelectedEvent.Subscribe(OnSelected)
                            set GearEntryUIs += array{GearEntryUI}

                            stack_box_slot:
                                HorizontalAlignment := horizontal_alignment.Fill
                                VerticalAlignment := vertical_alignment.Fill
                                Distribution := option{1.0}
                                Padding := margin{Left := 0.0, Top := (RowIndex = 1) and 0.0 or 10.0, Right := 0.0, Bottom := 0.0}
                                Widget := GearEntryUI.GetUI()

    CreateInfoUI():widget =
        GearUIPanels := array:
            gear_ui_panel.Treasure
            gear_ui_panel.Info
            gear_ui_panel.Upgrade
            gear_ui_panel.Tier
            gear_ui_panel.Evolve
            gear_ui_panel.Summary

        overlay:
            Slots := for (Index -> GearUIPanel : GearUIPanels):
                PanelUI := GetPanelUI(GearUIPanel)
                PanelUI.Initialize()

                if (TreasureUI := gear_panel_treasure_ui[PanelUI]):
                    TreasureUI.TreasureChestOpenedEvent.Subscribe(OnTreasureChestOpened)
                    TreasureUI.TreasureChestCompleteEvent.Subscribe(OnTreasureChestComplete)

                else if (UpgradeUI := gear_panel_upgrade_ui[PanelUI]):
                    UpgradeUI.UpgradedEvent.Subscribe(Update)

                else if (TierUI := gear_panel_tier_ui[PanelUI]):
                    TierUI.FusedEvent.Subscribe(Update)

                else if (EvolveUI := gear_panel_evolve_ui[PanelUI]):
                    EvolveUI.EvolvedEvent.Subscribe(Update)

                if (Index > 0):
                    PanelUI.Hide()

                if (set GearPanelUIs[GearUIPanel] = PanelUI) {}

                overlay_slot:
                    HorizontalAlignment := horizontal_alignment.Fill
                    VerticalAlignment := vertical_alignment.Fill
                    Widget := PanelUI.GetUI()

    CreateEmptySlotUI():widget =
        overlay:
            Slots := array:
                overlay_slot:
                    HorizontalAlignment := horizontal_alignment.Fill
                    VerticalAlignment := vertical_alignment.Fill
                    Padding := margin{Left := 40.0, Top := 40.0, Right := 40.0, Bottom := 40.0}
                    Widget := UserInterface.VerseWidget.WBP_RoundedBox_Light{}
                overlay_slot:
                    Widget := text_block:
                        DefaultText := MakeMessage("Empty Slot!")
                        DefaultTextColor := ThemeColors.Failure
                        DefaultShadowColor := ThemeColors.Failure * 0.125
                        DefaultShadowOffset := option{vector2{X := 2.0, Y := 2.0}}
                        DefaultShadowOpacity := 1.0

    <# ----------------------------------------------------------------------------------------------------
        Get Functions
    ---------------------------------------------------------------------------------------------------- #>
    GetPanelUI(Id:gear_ui_panel):ui_component =
        case (Id):
            gear_ui_panel.Treasure => gear_panel_treasure_ui{Player := Player}
            gear_ui_panel.Info => gear_panel_info_ui{Player := Player}
            gear_ui_panel.Upgrade => gear_panel_upgrade_ui{Player := Player}
            gear_ui_panel.Tier => gear_panel_tier_ui{Player := Player}
            gear_ui_panel.Evolve => gear_panel_evolve_ui{Player := Player}
            gear_ui_panel.Summary => gear_panel_summary_ui{Player := Player}

    IsEmptySlot()<transacts>:logic =
        logic{not Player.GetGear[ActiveIndex.ToGearSlot()] and GearRequiredPanels.Find[ActivePanel]}

    <# ----------------------------------------------------------------------------------------------------
        Update Functions
    ---------------------------------------------------------------------------------------------------- #>
    OnShow<override>():void =
        SetActiveGear(0)
        SetPanel(gear_ui_panel.Treasure)
        Update()

    Update():void =
        for (CurrencyId -> CurrencyText : CurrencyTexts):
            CurrencyMessage := MakeMessage("{FormatAsCurrency(Player.GetCurrency(CurrencyId))}")
            CurrencyText.SetText(CurrencyMessage)

        for (SlotIndex -> GearEntryUI : GearEntryUIs):
            GearEntryUI.Update()

    SetPanel(Panel:gear_ui_panel):void =
        if (IsOpening?) then return
        if (GearPanelUI := GearPanelUIs[ActivePanel]):
            GearPanelUI.Hide()

        set ActivePanel = Panel
        UpdatePanels()

    SetActiveGear(SelectedIndex:int):void =
        set ActiveIndex = SelectedIndex
        for (Index -> GearEntryUI : GearEntryUIs):
            GearEntryUI.SetActive(logic{Index = SelectedIndex})

        UpdatePanels()

    UpdatePanels():void =
        if (IsEmptySlot()?):
            if (GearPanelUI := GearPanelUIs[ActivePanel]):
                GearPanelUI.Hide()
            
            EmptySlotUI.SetVisibility(widget_visibility.Visible)
            return
        else:
            if (GearPanelUI := GearPanelUIs[ActivePanel]):
                GearPanelUI.Show()

            EmptySlotUI.SetVisibility(widget_visibility.Hidden)

        for (Id -> GearPanelUI : GearPanelUIs):
            if (InfoUI := gear_panel_info_ui[GearPanelUI]):
                InfoUI.Update(ActiveIndex.ToGearSlot())
            else if (InfoUI := gear_panel_upgrade_ui[GearPanelUI]):
                InfoUI.Update(ActiveIndex.ToGearSlot())
            else if (InfoUI := gear_panel_tier_ui[GearPanelUI]):
                InfoUI.Update(ActiveIndex.ToGearSlot())
            else if (InfoUI := gear_panel_evolve_ui[GearPanelUI]):
                InfoUI.Update(ActiveIndex.ToGearSlot())
            else:
                GearPanelUI.Update()

    <# ----------------------------------------------------------------------------------------------------
        Event Handlers
    ---------------------------------------------------------------------------------------------------- #>
    AwaitMenuButtonClicked(Listenable:listenable(widget_message), Panel:gear_ui_panel)<suspends>:void =
        loop:
            Listenable.Await()
            SetPanel(Panel)

    OnSelected(SelectedIndex:int):void =
        if (IsOpening?) then return
        SetActiveGear(SelectedIndex)

    OnTreasureChestOpened():void =
        set IsOpening = true
        Update()

    OnTreasureChestComplete():void =
        set IsOpening = false
        Update()