using { /Verse.org/Assets }
using { /UnrealEngine.com/Temporary/SpatialMath }
using { /UnrealEngine.com/Temporary/UI }

<# ----------------------------------------------------------------------------------------------------
    Data Definitions
---------------------------------------------------------------------------------------------------- #>
InfoPanelImages:[]texture = array:
    UserInterface.InfoPanel.InfoPanel_Instructions
    UserInterface.InfoPanel.InfoPanel_ExperienceOrbs
    UserInterface.InfoPanel.InfoPanel_GoldenTree

<# ----------------------------------------------------------------------------------------------------
    info_panel_ui
---------------------------------------------------------------------------------------------------- #>
info_panel_ui := class(panel_ui):
    HasTitle<override>:logic = false
    BackgroundOpacity<override>:type {_X:float where 0.000000 <= _X, _X <= 1.000000} = 0.7

    Image:texture_block = texture_block:
        DefaultImage := UserInterface.InfoPanel.InfoPanel_Instructions
        DefaultDesiredSize := vector2{X := 512.0, Y := 512.0}

    var MaybePaginationUI:?pagination_ui = false
    var Subscriptions:[]events_cancelable = array{}

    <# ----------------------------------------------------------------------------------------------------
        Initialization
    ---------------------------------------------------------------------------------------------------- #>
    Uninitialize<override>():void =
        for (Subscription : Subscriptions):
            Subscription.Cancel()

        set Subscriptions = array{}

    CreateContent<override>():widget =
        PaginationUI := pagination_ui:
            Player := Player
            EntriesPerPage := 1
            TotalEntries := InfoPanelImages.Length

        PaginationUI.Initialize()
        set MaybePaginationUI = option{PaginationUI}
        set Subscriptions += array:
            PaginationUI.PageChangedEvent.Subscribe(OnPageChanged)

        stack_box:
            Orientation := orientation.Vertical
            Slots := array:
                stack_box_slot:
                    HorizontalAlignment := horizontal_alignment.Fill
                    VerticalAlignment := vertical_alignment.Fill
                    Padding := margin{Left := 0.0, Top := 20.0, Right := 0.0, Bottom := 0.0}
                    Widget := overlay:
                        Slots := array:
                            overlay_slot:
                                HorizontalAlignment := horizontal_alignment.Fill
                                VerticalAlignment := vertical_alignment.Fill
                                Widget := MakeBox()
                            overlay_slot:
                                HorizontalAlignment := horizontal_alignment.Fill
                                VerticalAlignment := vertical_alignment.Fill
                                Padding := margin{Left := 2.0, Top := 2.0, Right := 2.0, Bottom := 2.0}
                                Widget := Image
                stack_box_slot:
                    HorizontalAlignment := horizontal_alignment.Fill
                    VerticalAlignment := vertical_alignment.Fill
                    Padding := margin{Left := 0.0, Top := 20.0, Right := 0.0, Bottom := 0.0}
                    Widget := PaginationUI.GetUI()

    <# ----------------------------------------------------------------------------------------------------
        Event Handlers
    ---------------------------------------------------------------------------------------------------- #>
    OnPageChanged():void =
        if (PaginationUI := MaybePaginationUI?, InfoPanelImage := InfoPanelImages[PaginationUI.GetCurrentPage() - 1]):
            Image.SetImage(InfoPanelImage)