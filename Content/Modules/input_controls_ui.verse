using { /Fortnite.com/UI }
using { /Verse.org/Assets }
using { /Verse.org/Colors }
using { /Verse.org/Simulation }
using { /UnrealEngine.com/Temporary/SpatialMath }
using { /UnrealEngine.com/Temporary/UI }

<# ----------------------------------------------------------------------------------------------------
    input_controls_ui
---------------------------------------------------------------------------------------------------- #>
input_controls_ui := class(panel_ui):
    Title<override>:string = "Input Controls"
    Anchors<override>:anchors = anchors{Minimum := vector2{X := 0.15, Y := 0.5}, Maximum := vector2{X := 0.85, Y := 0.5}}
    BackgroundOpacity<override>:type {_X:float where 0.000000 <= _X, _X <= 1.000000} = 0.7

    var Subscriptions:[]events_cancelable = array{}
    var InfoWidget:widget = canvas{}
    var ActiveTexts:[input_mode]text_block = map{}

    <# ----------------------------------------------------------------------------------------------------
        Initialization
    ---------------------------------------------------------------------------------------------------- #>
    Uninitialize<override>():void =
        for (Subscription : Subscriptions):
            Subscription.Cancel()

        set Subscriptions = array{}

    CreateContent<override>():widget =
        set Subscriptions += array{ClosedEvent.Subscribe(OnClosed)}

        stack_box:
            Orientation := orientation.Vertical
            Slots := array:
                stack_box_slot:
                    VerticalAlignment := vertical_alignment.Fill
                    Widget := text_block:
                        DefaultTextSize := ThemeTextSize.Small
                        DefaultText := MakeMessage("Select your current input controls.")
                        DefaultTextColor := ThemeColors.PrimaryText
                stack_box_slot:
                    HorizontalAlignment := horizontal_alignment.Fill
                    VerticalAlignment := vertical_alignment.Fill
                    Padding := margin{Left := 0.0, Top := 20.0, Right := 0.0, Bottom := 0.0}
                    Widget := CreateEntriesUI()

    CreateEntriesUI():widget =
        EntryInfos:[]tuple(texture, input_mode, string) = array:
            (UserInterface.Icons.Input_KeyboardMouse_128, input_mode.KeyboardMouse, "Keyboard & Mouse")
            (UserInterface.Icons.Input_Controller_128, input_mode.Controller, "Controller")
            (UserInterface.Icons.Input_Touch_128, input_mode.Touch, "Touch Screen")

        StackBox := stack_box:
            Orientation := orientation.Horizontal
            Slots := for (Index -> EntryInfo : EntryInfos):
                Button := button_quiet{DefaultText := MakeMessage("Select")}
                spawn{AwaitButtonClicked(Button.OnClick(), EntryInfo(1))}

                stack_box_slot:
                    HorizontalAlignment := horizontal_alignment.Fill
                    VerticalAlignment := vertical_alignment.Fill
                    Distribution := option{1.0}
                    Padding := margin{Left := (Index = 0) and 0.0 or 20.0, Top := 0.0, Right := 0.0, Bottom := 0.0}
                    Widget := CreateEntryUI(EntryInfo(0), EntryInfo(1), EntryInfo(2), Button)

        StackBox.AddWidget of stack_box_slot:
            HorizontalAlignment := horizontal_alignment.Fill
            VerticalAlignment := vertical_alignment.Fill
            Distribution := option{1.0}
            Padding := margin{Left := 20.0, Top := 0.0, Right := 0.0, Bottom := 0.0}
            Widget := set InfoWidget = CreateInfoUI()

        StackBox

    CreateEntryUI(Icon:texture, InputControlMode:input_mode, Text:string, Button:text_button_base):widget =
        ActiveText := text_block:
            DefaultTextSize := ThemeTextSize.Small
            DefaultText := MakeMessage("Active")
            DefaultTextColor := ThemeColors.Success

        if (set ActiveTexts[InputControlMode] = ActiveText) {}

        overlay:
            Slots := array:
                overlay_slot:
                    HorizontalAlignment := horizontal_alignment.Fill
                    VerticalAlignment := vertical_alignment.Fill
                    Widget := UserInterface.VerseWidget.WBP_RoundedBox{}
                overlay_slot:
                    HorizontalAlignment := horizontal_alignment.Fill
                    VerticalAlignment := vertical_alignment.Fill
                    Padding := margin{Left := 40.0, Top := 40.0, Right := 40.0, Bottom := 40.0}
                    Widget := stack_box:
                        Orientation := orientation.Vertical
                        Slots := array:
                            stack_box_slot:
                                VerticalAlignment := vertical_alignment.Top
                                Distribution := option{1.0}
                                Widget := texture_block:
                                    DefaultImage := Icon
                                    DefaultDesiredSize := vector2{X := 128.0, Y := 128.0}
                            stack_box_slot:
                                VerticalAlignment := vertical_alignment.Fill
                                Widget := ActiveText
                            stack_box_slot:
                                VerticalAlignment := vertical_alignment.Fill
                                Widget := text_block:
                                    DefaultTextSize := ThemeTextSize.Small
                                    DefaultText := MakeMessage("{Text}")
                                    DefaultTextColor := ThemeColors.PrimaryText
                                    DefaultTextOpacity := 0.5
                            stack_box_slot:
                                HorizontalAlignment := horizontal_alignment.Fill
                                VerticalAlignment := vertical_alignment.Fill
                                Padding := margin{Left := 0.0, Top := 40.0, Right := 0.0, Bottom := 0.0}
                                Widget := Button

    CreateInfoUI():widget =
        overlay:
            Slots := array:
                overlay_slot:
                    HorizontalAlignment := horizontal_alignment.Fill
                    VerticalAlignment := vertical_alignment.Fill
                    Widget := UserInterface.VerseWidget.WBP_RoundedBox{}
                overlay_slot:
                    HorizontalAlignment := horizontal_alignment.Fill
                    VerticalAlignment := vertical_alignment.Fill
                    Padding := margin{Left := 20.0, Top := 20.0, Right := 20.0, Bottom := 40.0}
                    Widget := stack_box:
                        Orientation := orientation.Vertical
                        Slots := array:
                            stack_box_slot:
                                VerticalAlignment := vertical_alignment.Fill
                                Widget := text_block:
                                    DefaultTextSize := ThemeTextSize.Small
                                    DefaultText := MakeMessage("INFO")
                                    DefaultTextColor := ThemeColors.PrimaryText
                            stack_box_slot:
                                VerticalAlignment := vertical_alignment.Top
                                Padding := margin{Left := 0.0, Top := 20.0, Right := 0.0, Bottom := 0.0}
                                Widget := UserInterface.VerseWidget.WBP_InputControlInfo{}
                            stack_box_slot:
                                VerticalAlignment := vertical_alignment.Fill
                                Padding := margin{Left := 0.0, Top := 20.0, Right := 0.0, Bottom := 0.0}
                                Widget := text_block:
                                    DefaultTextSize := ThemeTextSize.VerySmall
                                    DefaultText := MakeMessage("You may change this later by interacting with the above device.".WrapText(512.0))
                                    DefaultTextColor := ThemeColors.PrimaryText
                                    DefaultTextOpacity := 0.5

    <# ----------------------------------------------------------------------------------------------------
        Update Functionss
    ---------------------------------------------------------------------------------------------------- #>
    Update():void =
        PlayerInputMode := Player.GetInputMode()
        for (InputControlMode -> ActiveText : ActiveTexts):
            ActiveText.SetVisibility(InputControlMode = PlayerInputMode and widget_visibility.Visible or widget_visibility.Hidden)
        
    <# ----------------------------------------------------------------------------------------------------
        Event Handlers
    ---------------------------------------------------------------------------------------------------- #>
    OnClosed(InPlayer:player):void =
        if (InputManager := GetInputManager[]):
            InputManager.InputModeChangedEvent.Signal(Player)

    OnShow<override>():void =
        Update()

    AwaitButtonClicked(Listenable:listenable(widget_message), InputControlMode:input_mode)<suspends>:void =
        WaitForPlayerManager(Player)

        if (PlayerRemovedEvent := Player.GetPlayerRemovedEvent[]):
            race:
                PlayerRemovedEvent.Await()
                loop:
                    Listenable.Await()
                    Player.SetInputMode(InputControlMode)
                    Player.PlayAudio(audio_id.Activate)
                    Update()

        