using { /Fortnite.com/Devices }
using { /Verse.org/Simulation }

<# ----------------------------------------------------------------------------------------------------
    Constants
---------------------------------------------------------------------------------------------------- #>
DirectionInputTriggerIds:[]input_id = array{input_id.Forward, input_id.Backward, input_id.Left, input_id.Right}

InputTriggerCollections<public> := module:
    Lobby<public>:[]input_id = array{input_id.Forward, input_id.Backward, input_id.Left, input_id.Right, input_id.Menu, input_id.Dash, input_id.AutoAim}
    Game<public>:[]input_id = array{input_id.Forward, input_id.Backward, input_id.Left, input_id.Right, input_id.Menu, input_id.Dash, input_id.AutoAim, input_id.AbilityStats, input_id.OpenTreasureChest}
    PostGame<public>:[]input_id = array{input_id.Skip}

<# ----------------------------------------------------------------------------------------------------
    Data Definitions
---------------------------------------------------------------------------------------------------- #>
input_id<public> := enum:
    Forward
    Backward
    Left
    Right
    Menu
    Dash
    AbilityStats
    OpenTreasureChest
    AutoAim
    Skip

ToString(Id:input_id)<transacts>:string =
    case (Id):
        input_id.Forward => "Forward"
        input_id.Backward => "Backward"
        input_id.Left => "Left"
        input_id.Right => "Right"
        input_id.Menu => "Menu"
        input_id.Dash => "Dash"
        input_id.AbilityStats => "Ability Stats"
        input_id.OpenTreasureChest => "Open Treasure Chest"
        input_id.AutoAim => "Auto Aim"
        input_id.Skip => "Skip"

input_trigger := class<concrete>:
    @editable
    Id:input_id = input_id.Forward
    @editable
    Device:input_trigger_device = input_trigger_device{}

input_mode<public> := enum<persistable>:
    KeyboardMouse
    Controller
    Touch

<# ----------------------------------------------------------------------------------------------------
    Global Functions
---------------------------------------------------------------------------------------------------- #>
var GlobalInputManager:weak_map(session, input_manager) = map{}

GetInputManager()<decides><transacts>:input_manager =
    GlobalInputManager[GetSession()]

WaitForInputManager()<suspends>:void =
    loop:
        if (GetInputManager[].GetIsInitialized()?) {break}
        Sleep(0.3)

(Agent:agent).IsInputHeld(Id:input_id)<decides><transacts>:void =
    GetInputManager[].GetIsHeld[Id, Agent]

GetInputPressedEvent(Id:input_id)<decides><transacts>:listenable(agent) =
    GetInputManager[].GetInputTrigger[Id].PressedEvent

GetInputReleasedEvent(Id:input_id)<decides><transacts>:listenable(tuple(agent, float)) =
    GetInputManager[].GetInputTrigger[Id].ReleasedEvent

(Agent:agent).RegisterInput(Id:input_id):void =
    if (InputManager := GetInputManager[]):
        InputManager.Register(Agent, Id)

(Agent:agent).UnregisterInput(Id:input_id):void =
    if (InputManager := GetInputManager[]):
        InputManager.Unregister(Agent, Id)

(Agent:agent).SetInputTriggers(Ids:[]input_id):void =
    if (InputManager := GetInputManager[]):
        for (InputTrigger : InputManager.MappedInputTriggers):
            InputTrigger.Unregister(Agent)

        for (Id : Ids):
            InputManager.Register(Agent, Id)

<# ----------------------------------------------------------------------------------------------------
    input_manager
---------------------------------------------------------------------------------------------------- #>
input_manager := class(creative_device):
    @editable
    InputTriggers:[]input_trigger = array{}
    
    @editable
    InputSettingButton:button_device = button_device{}

    InputUnregisteredEvent:events(player) = events(player){}
    InputModeChangedEvent:events(player) = events(player){}

    var IsInitialized:logic = false
    var MappedInputTriggers:[input_id]input_trigger_device = map{}

    OnBegin<override>()<suspends>:void =
        if (set GlobalInputManager[GetSession()] = Self) {}

        for (InputTrigger : InputTriggers, set MappedInputTriggers[InputTrigger.Id] = InputTrigger.Device) {}

        InputSettingButton.InteractedWithEvent.Subscribe(OnInteractedWith)
            
        set IsInitialized = true

    <# ----------------------------------------------------------------------------------------------------
        Get Functions
    ---------------------------------------------------------------------------------------------------- #>
    GetIsInitialized()<transacts>:logic =
        IsInitialized

    GetInputTrigger(Id:input_id)<decides><transacts>:input_trigger_device =
        MappedInputTriggers[Id]

    GetIsHeld(Id:input_id, Agent:agent)<decides><transacts>:void =
        GetInputTrigger[Id].IsHeld[Agent]

    <# ----------------------------------------------------------------------------------------------------
        Update Functions
    ---------------------------------------------------------------------------------------------------- #>
    Register(Agent:agent, Id:input_id):void =
        if (InputTrigger := GetInputTrigger[Id]):
            InputTrigger.Register(Agent)

    Unregister(Agent:agent, Id:input_id):void =
        if (InputTrigger := GetInputTrigger[Id], Player := player[Agent]):
            InputTrigger.Unregister(Agent)
            InputUnregisteredEvent.Signal(Player)

    OnInteractedWith(Agent:agent):void =
        if (Player := player[Agent]):
            Player.ShowPanelUI(panel_ui_id.InputControl)