using { /Fortnite.com/Characters }
using { /Fortnite.com/Devices }
using { /Fortnite.com/Devices/CreativeAnimation }
using { /Verse.org/Assets }
using { /Verse.org/Random }
using { /Verse.org/Simulation }
using { /UnrealEngine.com/Temporary/SpatialMath }

<# ----------------------------------------------------------------------------------------------------
    Constants
---------------------------------------------------------------------------------------------------- #>
ItemPickupSpeed:float = 2000.0
ItemPickupDistanceSquared:float = 10000.0

<# ----------------------------------------------------------------------------------------------------
    Data Definitions
---------------------------------------------------------------------------------------------------- #>
item_drop_id := enum:
    Health
    Magnet
    TreasureChest

item_drop_definition := class<concrete>:
    @editable
    Id:item_drop_id = item_drop_id.Health
    @editable
    PropAsset:creative_prop_asset = DefaultCreativePropAsset
    @editable
    MaybeMapIndicators:?[]map_indicator_device = false

ToString(Id:item_drop_id)<transacts>:string =
    case (Id):
        item_drop_id.TreasureChest => "Treasure Chest"
        item_drop_id.Magnet => "Magnet"
        item_drop_id.Health => "Health"

<# ----------------------------------------------------------------------------------------------------
    item_drop factory
---------------------------------------------------------------------------------------------------- #>
item_drop_factory := class<concrete>:
    GetItemDrop(Manager:item_spawner_manager, Id:item_drop_id)<transacts>:item_drop =
        case (Id):
            item_drop_id.TreasureChest => item_drop_treasure_chest{Manager := Manager}
            item_drop_id.Magnet => item_drop_magnet{Manager := Manager}
            item_drop_id.Health => item_drop_health{Manager := Manager}

<# ----------------------------------------------------------------------------------------------------
    item_drop base class
---------------------------------------------------------------------------------------------------- #>
item_drop := class<abstract><unique>:
    Id:item_drop_id
    Manager:item_spawner_manager
    DisableEvent:event(void) = event(void){}

    var MaybeSpawnedProp:?creative_prop = false
    var IsActive:logic = false
    var SpawnPosition:vector3 = vector3{}
    var IsSpawning:logic = false

    <# ----------------------------------------------------------------------------------------------------
        Update Functions
    ---------------------------------------------------------------------------------------------------- #>
    OnBegin(Transform:transform, ItemDrop:item_drop_definition):void =
        set SpawnPosition = Transform.Translation.Make2D()

        if:
            SpawnedProp := MaybeSpawnedProp?
            SpawnedProp.TeleportTo[Transform]
        else:
            SpawnPropResult := SpawnProp(ItemDrop.PropAsset, SpawnPosition, IdentityRotation().ApplyYaw(DegreesToRadians(90.0)))
            set MaybeSpawnedProp = SpawnPropResult(0)

        OnSpawn()
        spawn{AwaitPickup()}
        set IsActive = true

    <# ----------------------------------------------------------------------------------------------------
        Get Functions
    ---------------------------------------------------------------------------------------------------- #>
    GetPosition()<transacts>:vector3 =
        MaybeSpawnedProp?.GetTransform().Translation or vector3{}

    CanPickup(Player:player)<transacts>:logic =
        true

    <# ----------------------------------------------------------------------------------------------------
        Update Functions
    ---------------------------------------------------------------------------------------------------- #>
    SpawnAnimation()<suspends>:void =
        set IsSpawning = true
        Sleep(FiveTicks)
        if (SpawnedProp := MaybeSpawnedProp?, AnimationController := SpawnedProp.GetAnimationController[]):
            DeltaLocation := GetRandomFloat(75.0, 125.0)
            KeyFrames := array:
                keyframe_delta:
                    DeltaLocation := vector3{Z := DeltaLocation}
                    DeltaRotation := rotation{}
                    DeltaScale := UnitVector * 1.2
                    Time := 0.2
                    Interpolation := InterpolationTypes.EaseOut
                keyframe_delta:
                    DeltaLocation := vector3{Z := -DeltaLocation}
                    DeltaRotation := rotation{}
                    DeltaScale := UnitVector * (1.0 / 1.2)
                    Time := 0.2
                    Interpolation := InterpolationTypes.EaseIn

            SpawnedProp.Show()
            AnimationController.SetAnimation(KeyFrames, ?Mode := animation_mode.OneShot)
            AnimationController.Play()
            AnimationController.MovementCompleteEvent.Await()

        set IsSpawning = false
        OnSpawnComplete()

    Pickup(Player:player)<suspends>:void =
        race:
            DisableEvent.Await()
            if (SpawnedProp := MaybeSpawnedProp?, Character := Player.GetFortCharacter[], AnimationController := SpawnedProp.GetAnimationController[]):
                Direction := (SpawnedProp.GetTransform().Translation - Character.GetTransform().Translation).MakeUnitVector[] or vector3{}

                KeyFrames01 := array:
                    keyframe_delta:
                        DeltaLocation := Direction * 100.0 + vector3{Z := 100.0}
                        DeltaRotation := rotation{}
                        DeltaScale := UnitVector
                        Time := 0.3
                        Interpolation := InterpolationTypes.EaseOut
                    
                AnimationController.SetAnimation(KeyFrames01, ?Mode := animation_mode.OneShot)
                AnimationController.Play()
                
                Listenable:listenable(tuple()) = AnimationController.MovementCompleteEvent
                Listenable.Await()

                PickupDistance := Distance(Character.GetTransform().Translation, SpawnedProp.GetTransform().Translation)
                PickupTime := PickupDistance / ItemPickupSpeed

                KeyFrames02 := array:
                    keyframe_delta:
                        DeltaLocation := Character.GetTransform().Translation - SpawnedProp.GetTransform().Translation
                        DeltaRotation := rotation{}
                        DeltaScale := UnitVector
                        Time := PickupTime
                        Interpolation := InterpolationTypes.EaseIn

                AnimationController.SetAnimation(KeyFrames02, ?Mode := animation_mode.OneShot)
                AnimationController.Play()
                Listenable.Await()
                GrantReward(Player)
                Disable()

    Disable():void =
        DisableEvent.Signal()
        if (SpawnedProp := MaybeSpawnedProp?):
            SpawnedProp.Hide()
        Manager.ReturnToObjectPool(Self)
        OnDisable()

    GrantReward(Player:player):void

    Reposition():void =
        if (SpawnedProp := MaybeSpawnedProp?):
            spawn{SpawnedProp.MoveTo(GetRandomPositionOnCircle(150.0, ?Origin := SpawnedProp.GetTransform().Translation), IdentityRotation(), 0.5)}

    <# ----------------------------------------------------------------------------------------------------
        Event Handlers
    ---------------------------------------------------------------------------------------------------- #>
    OnSpawn():void = {}
    OnDisable():void = {}
    OnSpawnComplete():void = {}

    AwaitPickup()<suspends>:void =
        race:
            DisableEvent.Await()
            block:
                SpawnAnimation()
                var MaybeMinDistanceSquared:?float = false
                loop:                    
                    DistanceFactor := Pow(((Floor[Sqrt(MaybeMinDistanceSquared?) / 1000.0] + 1) * 1.0) or 1.0, 1.5)
                    Sleep(ThreeTicks * DistanceFactor)
                    set MaybeMinDistanceSquared = false
                    for:
                        Player : GetAlivePlayers()
                        Player.IsActive[]
                        CanPickup(Player)?
                        SpawnedProp := MaybeSpawnedProp?, SpawnedProp.IsValid[]
                        Character := Player.GetFortCharacter[]
                        PlayerDistanceSquared := DistanceSquaredXY(Character.GetTransform().Translation, SpawnedProp.GetTransform().Translation)
                    do:
                        if (PlayerDistanceSquared <= ItemPickupDistanceSquared):
                            set IsActive = false
                            Pickup(Player)
                            break
                        else if (not MaybeMinDistanceSquared? or PlayerDistanceSquared < MaybeMinDistanceSquared?):
                            set MaybeMinDistanceSquared = option{PlayerDistanceSquared}

<# ----------------------------------------------------------------------------------------------------
    item_drop_treasure_chest
---------------------------------------------------------------------------------------------------- #>
item_drop_treasure_chest := class(item_drop):
    Id<override>:item_drop_id = item_drop_id.TreasureChest

    var MaybeActiveMapIndicator:?map_indicator_device = false

    GrantReward<override>(Player:player):void =
        AwardAccolade(accolade_id.TreasureChest, GetActivePlayers())
        AddGameCurrency(game_currency_id.TreasureChest, 1.0)

        ShowMessage(message_id.TreasureChestCollect, "+1", ?DisplayTime := 1.8)
        for (ActivePlayer : GetActivePlayers(), Character := Player.GetFortCharacter[]):
            ActivePlayer.ActivateVFXPowerup(vfx_powerup_id.TreasureChestCollect, ?Duration := 1.0)
            ActivePlayer.PlayAudio(audio_id.Purchase)

        if (ActiveMapIndicator := MaybeActiveMapIndicator?):
            Manager.ReturnMapIndicator(Id, ActiveMapIndicator)
            set MaybeActiveMapIndicator = false

    OnSpawn<override>():void =
        if:
            MapIndicator := Manager.GetMapIndicator[Id]
            SpawnedProp := MaybeSpawnedProp?
            MapIndicator.TeleportTo[SpawnedProp.GetTransform().Translation, IdentityRotation()]
        then:
            set MaybeActiveMapIndicator = option{MapIndicator}

    OnDisable<override>():void =
        if (MapIndicator := MaybeActiveMapIndicator?):
            Manager.ReturnMapIndicator(Id, MapIndicator)
            set MaybeActiveMapIndicator = false

<# ----------------------------------------------------------------------------------------------------
    item_drop_magnet
---------------------------------------------------------------------------------------------------- #>
item_drop_magnet := class(item_drop):
    Id<override>:item_drop_id = item_drop_id.Magnet
    
    var MaybeActiveMapIndicator:?map_indicator_device = false

    GrantReward<override>(Player:player):void =
        Player.PickupAllExperience()
        Player.PlayAudio(audio_id.Magnet)

        AwardAccolade(accolade_id.Magnet, GetActivePlayers())

        if (ActiveMapIndicator := MaybeActiveMapIndicator?):
            Manager.ReturnMapIndicator(Id, ActiveMapIndicator)
            set MaybeActiveMapIndicator = false

    OnSpawn<override>():void =
        if:
            MapIndicator := Manager.GetMapIndicator[Id]
            SpawnedProp := MaybeSpawnedProp?
            MapIndicator.TeleportTo[SpawnedProp.GetTransform().Translation, IdentityRotation()]
        then:
            set MaybeActiveMapIndicator = option{MapIndicator}

    OnDisable<override>():void =
        if (MapIndicator := MaybeActiveMapIndicator?):
            Manager.ReturnMapIndicator(Id, MapIndicator)
            set MaybeActiveMapIndicator = false

<# ----------------------------------------------------------------------------------------------------
    item_drop_health
---------------------------------------------------------------------------------------------------- #>
item_drop_health := class(item_drop):
    Id<override>:item_drop_id = item_drop_id.Health

    var MaybeActiveMapIndicator:?map_indicator_device = false
    var HealCount:int = 0

    CanPickup<override>(Player:player)<transacts>:logic =
        logic{Character := Player.GetFortCharacter[], Character.GetHealth() < Character.GetMaxHealth()}

    Pickup<override>(Player:player)<suspends>:void =
        race:
            DisableEvent.Await()
            if (SpawnedProp := MaybeSpawnedProp?, Character := Player.GetFortCharacter[], AnimationController := SpawnedProp.GetAnimationController[]):
                KeyFrames := array:
                    keyframe_delta:
                        DeltaLocation := vector3{Z := 100.0}
                        DeltaRotation := rotation{}
                        DeltaScale := UnitVector * 1.15
                        Time := 0.15
                        Interpolation := InterpolationTypes.EaseOut
                    
                AnimationController.SetAnimation(KeyFrames, ?Mode := animation_mode.OneShot)
                AnimationController.Play()
                Listenable:listenable(tuple()) = AnimationController.MovementCompleteEvent
                Listenable.Await()
                GrantReward(Player)
                Disable()

    GrantReward<override>(Player:player):void =
        HealAmount := type{_X:float where 0.000000 <= _X, _X <= 1.000000}[Player.GetPowerUpValue(power_up_id.HealthPickup)] or 0.5
        Player.Heal(HealAmount)
        AwardAccolade(accolade_id.HealthOrb, GetActivePlayers())

        if (GetActivePlayers().Length > 1):
            Player.ActivateHealingAura()

        if (ActiveMapIndicator := MaybeActiveMapIndicator?):
            Manager.ReturnMapIndicator(Id, ActiveMapIndicator)
            set MaybeActiveMapIndicator = false

    OnSpawn<override>():void =
        set HealCount = 0
        if:
            MapIndicator := Manager.GetMapIndicator[Id]
            SpawnedProp := MaybeSpawnedProp?
            MapIndicator.TeleportTo[SpawnedProp.GetTransform().Translation, IdentityRotation()]
        then:
            set MaybeActiveMapIndicator = option{MapIndicator}

    OnDisable<override>():void =
        if (MapIndicator := MaybeActiveMapIndicator?):
            Manager.ReturnMapIndicator(Id, MapIndicator)
            set MaybeActiveMapIndicator = false