using { /Fortnite.com/Devices }
using { /Verse.org/Simulation }
using { /UnrealEngine.com/Temporary/SpatialMath }
using { /UnrealEngine.com/Temporary/UI }

<# ----------------------------------------------------------------------------------------------------
    Global Functions
---------------------------------------------------------------------------------------------------- #>
var GlobalItemShopManager:weak_map(session, item_shop_manager) = map{}

GetItemShopManager()<decides><transacts>:item_shop_manager =
    GlobalItemShopManager[GetSession()]

ToggleItemShop(IsEnable:logic):void =
    if (ItemShopManager := GetItemShopManager[]):
        if (IsEnable?):
            ItemShopManager.Enable()
        else:
            ItemShopManager.Disable()

EnableItemShop():void =
    if (ItemShopManager := GetItemShopManager[]):
        ItemShopManager.Enable()

(Player:player).ShowItemShopUI():void =
    Player.ShowItemShopUI(?IsResume := false)

(Player:player).ShowItemShopUI(?IsResume:logic = false):void =
    if (ItemShopManager := GetItemShopManager[]):
        ItemShopManager.ShowItemShopUI(Player, ?IsResume := IsResume)

<# ----------------------------------------------------------------------------------------------------
    item_shop_manager
---------------------------------------------------------------------------------------------------- #>
item_shop_manager := class(creative_device):
    @editable
    Character:character_device = character_device{}

    @editable
    Beacon:beacon_device = beacon_device{}

    @editable
    MapIndicator:map_indicator_device = map_indicator_device{}

    var PlayerItemShopUI:[player]item_shop_ui = map{}
    var PlayerItemShopPurchasedEvents:[player]events(tuple()) = map{}

    <# ----------------------------------------------------------------------------------------------------
        Initialization
    ---------------------------------------------------------------------------------------------------- #>
    OnBegin<override>()<suspends>:void =
        if (set GlobalItemShopManager[GetSession()] = Self) {}

        for (Player : GetPlayspace().GetPlayers()):
            Initialize(Player)

        WaitForGameManager()
        if (GameManager := GetGameManager[]):
            GameManager.GameEndedEvent.Subscribe(OnGameEnded)

        if (MapIndicator.TeleportTo[MapIndicatorDisabledPosition, IdentityRotation()]) {}
        Disable()
        GetPlayspace().PlayerAddedEvent().Subscribe(OnPlayerAdded)
        GetPlayspace().PlayerRemovedEvent().Subscribe(OnPlayerRemoved)

    Initialize(Player:player):void =
        if (set PlayerItemShopPurchasedEvents[Player] = events(tuple()){}) {}

    <# ----------------------------------------------------------------------------------------------------
        Update Functions
    ---------------------------------------------------------------------------------------------------- #>
    Enable():void =
        Beacon.Enable()
        if (MapIndicator.TeleportTo[MapIndicatorEnabledPosition, IdentityRotation()]) {}
        Character.Show()

        ShowMessage(message_id.ItemShopNotification)
        for (Player : GetActivePlayers()):
            Player.PlayAudio(audio_id.Warning)
            MapIndicator.ActivateObjectivePulse(Player)

    Disable():void =
        for (Player : GetAllPlayers()):
            MapIndicator.DeactivateObjectivePulse(Player)

        Beacon.Disable()
        if (MapIndicator.TeleportTo[MapIndicatorDisabledPosition, IdentityRotation()]) {}
        Character.Hide()

    ShowItemShopUI(Player:player, ?IsResume:logic = false):void =
        UI := if (ItemShopUI := PlayerItemShopUI[Player]):
            ItemShopUI
        else:
            ItemShopUI := item_shop_ui{Player := Player}
            ItemShopUI.DoneEvent.Subscribe(OnDone)
            ItemShopUI.ItemPurchasedEvent.Subscribe(OnItemPurchased)
            if (set PlayerItemShopUI[Player] = ItemShopUI) {}
            ItemShopUI

        UI.Show(?IsSignal := logic{not IsResume?})

    <# ----------------------------------------------------------------------------------------------------
        Event Handler
    ---------------------------------------------------------------------------------------------------- #>
    OnPlayerAdded(Player:player):void =
        Initialize(Player)

    OnPlayerRemoved(Player:player):void =
        if (ItemShopUI := PlayerItemShopUI[Player]):
            ItemShopUI.Uninitialize()
            
        set PlayerItemShopUI = RemoveKeyFromMap(PlayerItemShopUI, Player)
        set PlayerItemShopPurchasedEvents = RemoveKeyFromMap(PlayerItemShopPurchasedEvents, Player)

    OnGameEnded():void =
        Disable()

    OnDone(Player:player):void =
        IsComplete := logic{GetTotalPauseCount(pause_type.ItemShop) = 0}
        for (ItemShopUI : PlayerItemShopUI):
            if (IsComplete?):
                ItemShopUI.Hide()
            else:
                ItemShopUI.UpdateDoneCount()

    OnItemPurchased(Player:player):void =
        if (ItemShopPurchasedEvent := PlayerItemShopPurchasedEvents[Player]):
            ItemShopPurchasedEvent.Signal()