using { /Fortnite.com/UI }
using { /Verse.org/Colors }
using { /Verse.org/Simulation }
using { /UnrealEngine.com/Temporary/SpatialMath }
using { /UnrealEngine.com/Temporary/UI }

item_shop_entry_ui := class(ui_component):
    BuyEvent:events(tuple(ability_info, int, logic)) = events(tuple(ability_info, int, logic)){}

    BuyButton:button_quiet = button_quiet{}

    Icon:texture_block = texture_block:
        DefaultImage := UserInterface.Icons.WeaponSword_128
        DefaultDesiredSize := vector2{X := 128.0, Y := 128.0}

    StatIcon:texture_block = texture_block:
        DefaultImage := UserInterface.Icons.WeaponSword_64
        DefaultDesiredSize := vector2{X := 64.0, Y := 64.0}

    NameText:text_block = text_block:
        DefaultTextSize := ThemeTextSize.Small
        DefaultTextColor := ThemeColors.PrimaryText
        DefaultShadowOffset := option{vector2{X := 2.0, Y := 2.0}}
        DefaultShadowOpacity := 0.5
        DefaultJustification := text_justification.Center

    TierText:text_block = text_block:
        DefaultTextSize := ThemeTextSize.Small
        DefaultShadowOffset := option{vector2{X := 2.0, Y := 2.0}}
        DefaultShadowOpacity := 1.0

    StatText:text_block = text_block:
        DefaultTextSize := ThemeTextSize.Small
        DefaultTextColor := ThemeColors.PrimaryText
        DefaultShadowOffset := option{vector2{X := 2.0, Y := 2.0}}
        DefaultShadowOpacity := 0.5

    StatDeltaText:text_block = text_block:
        DefaultTextSize := ThemeTextSize.Small
        DefaultTextColor := ThemeColors.Success
        DefaultShadowColor := ThemeColors.Success * 0.125
        DefaultShadowOffset := option{vector2{X := 2.0, Y := 2.0}}
        DefaultShadowOpacity := 1.0

    EvolutionIcon:texture_block = texture_block:
        DefaultImage := UserInterface.Icons.NewsEntryType_NewFeature_64
        DefaultDesiredSize := vector2{X := 30.0, Y := 30.0}

    CostText:text_block = text_block:
        DefaultShadowOffset := option{vector2{X := 2.0, Y := 2.0}}
        DefaultShadowOpacity := 1.0

    var BoxUIs:[]box_ui = array{}
    var ProgressUI:widget = canvas{}
    var InfoWidget:widget = canvas{}
    var MaybeAbilityInfo:?ability_info = false
    var ProgressPulsings:[]widget = array{}

    var Tier:int = 1
    var IsEvolution:logic = false
    var Subscriptions:[]cancelable = array{}
            
    <# ----------------------------------------------------------------------------------------------------
        Create UI
    ---------------------------------------------------------------------------------------------------- #>
    CreateUI<override>():widget =
        set Subscriptions += array{BuyButton.OnClick().Subscribe(OnClickBuy)}
        
        stack_box:
            Orientation := orientation.Vertical
            Slots := array:
                stack_box_slot:
                    HorizontalAlignment := horizontal_alignment.Fill
                    VerticalAlignment := vertical_alignment.Fill
                    Distribution := option{1.0}
                    Widget := set InfoWidget = stack_box:
                        Orientation := orientation.Vertical
                        Slots := array:
                            stack_box_slot:
                                HorizontalAlignment := horizontal_alignment.Fill
                                VerticalAlignment := vertical_alignment.Fill
                                Distribution := option{1.0}
                                Padding := margin{Left := 0.0, Top := 0.0, Right := 0.0, Bottom := 0.0}
                                Widget := overlay:
                                    Slots := array:
                                        overlay_slot:
                                            HorizontalAlignment := horizontal_alignment.Fill
                                            VerticalAlignment := vertical_alignment.Fill
                                            Widget := MakeBox()
                                        overlay_slot:
                                            HorizontalAlignment := horizontal_alignment.Fill
                                            VerticalAlignment := vertical_alignment.Fill
                                            Widget := stack_box:
                                                Orientation := orientation.Horizontal
                                                Slots := array:
                                                    stack_box_slot:
                                                        HorizontalAlignment := horizontal_alignment.Fill
                                                        VerticalAlignment := vertical_alignment.Fill
                                                        Padding := margin{Left := 1.0, Top := 1.0, Right := 1.0, Bottom := 1.0}
                                                        Widget := CreateProgressUI()
                                                    stack_box_slot:
                                                        HorizontalAlignment := horizontal_alignment.Fill
                                                        VerticalAlignment := vertical_alignment.Fill
                                                        Padding := margin{Left := 40.0, Top := 40.0, Right := 40.0, Bottom := 40.0}
                                                        Distribution := option{1.0}
                                                        Widget := CreateInfoUI()
                            stack_box_slot:
                                HorizontalAlignment := horizontal_alignment.Fill
                                VerticalAlignment := vertical_alignment.Fill
                                Padding := margin{Left := 0.0, Top := 10.0, Right := 0.0, Bottom := 0.0}
                                Widget := CreateCostUI()
                stack_box_slot:
                    HorizontalAlignment := horizontal_alignment.Fill
                    VerticalAlignment := vertical_alignment.Fill
                    Padding := margin{Left := 0.0, Top := 10.0, Right := 0.0, Bottom := 0.0}
                    Widget := BuyButton

    CreateCostUI():widget =
        overlay:
            Slots := array:
                overlay_slot:
                    HorizontalAlignment := horizontal_alignment.Fill
                    VerticalAlignment := vertical_alignment.Fill
                    Widget := UserInterface.VerseWidget.WBP_RoundedBox_Light{}
                overlay_slot:
                    Padding := margin{Left := 0.0, Top := 10.0, Right := 0.0, Bottom := 10.0}
                    Widget := stack_box:
                        Orientation := orientation.Horizontal
                        Slots := array:
                            stack_box_slot:
                                HorizontalAlignment := horizontal_alignment.Left
                                Padding := margin{Left := 0.0, Top := 0.0, Right := 10.0, Bottom := 0.0}
                                Widget := texture_block:
                                    DefaultImage := UserInterface.Icons.CurrencyWood_64
                                    DefaultDesiredSize := vector2{X := 40.0, Y := 40.0}
                            stack_box_slot:
                                HorizontalAlignment := horizontal_alignment.Fill
                                VerticalAlignment := vertical_alignment.Fill
                                Widget := CostText

    CreateProgressUI():widget =
        overlay:
            Slots := array:
                overlay_slot:
                    HorizontalAlignment := horizontal_alignment.Fill
                    VerticalAlignment := vertical_alignment.Fill
                    Widget := color_block:
                        DefaultColor := NamedColors.White
                        DefaultOpacity := 0.1
                        DefaultDesiredSize := vector2{X := 0.0, Y := 520.0}
                overlay_slot:
                    HorizontalAlignment := horizontal_alignment.Center
                    VerticalAlignment := vertical_alignment.Top
                    Padding := margin{Left := 10.0, Top := 20.0, Right := 10.0, Bottom := 20.0}
                    Widget := set ProgressUI = stack_box:
                        Orientation := orientation.Vertical
                        Slots := for (Index := 1..MaxWeaponLevel):
                            IsTierUp := logic{Index = 1 or Mod[Index - 1, WeaponEvolutionInterval] = 0}
                            
                            BoxHeight := IsTierUp? and 30.0 or 15.0
                            BoxUI := MakeBoxUI(?DefaultWidth := 30.0, ?DefaultHeight := BoxHeight, ?DefaultColor := ThemeColors.Disabled, ?DefaultOpacity := 1.0, ?DefaultBorderColor := NamedColors.Black, ?DefaultBorderOpacity := 0.5)
                            set BoxUIs += array{BoxUI}
                            
                            ProgressPulsing := UserInterface.VerseWidget.WBP_ProgressPulsing{}
                            set ProgressPulsings += array{ProgressPulsing}

                            Overlay := overlay:
                                Slots := array:
                                    overlay_slot:
                                        HorizontalAlignment := horizontal_alignment.Fill
                                        VerticalAlignment := vertical_alignment.Fill
                                        Widget := BoxUI.GetWidget()
                                    overlay_slot:
                                        HorizontalAlignment := horizontal_alignment.Fill
                                        VerticalAlignment := vertical_alignment.Fill
                                        Padding := margin{Left := 1.0, Top := 1.0, Right := 1.0, Bottom := 1.0}
                                        Widget := ProgressPulsing

                            if (IsTierUp?):
                                Overlay.AddWidget of overlay_slot:
                                    HorizontalAlignment := horizontal_alignment.Center
                                    VerticalAlignment := vertical_alignment.Center
                                    Widget := texture_block:
                                        DefaultImage := UserInterface.Icons.Star_Flat_64
                                        DefaultDesiredSize := vector2{X := 25.0, Y := 25.0}
                                        DefaultTint := NamedColors.Black

                            stack_box_slot:
                                HorizontalAlignment := horizontal_alignment.Fill
                                VerticalAlignment := vertical_alignment.Top
                                Padding := margin{Left := 0.0, Top := IsTierUp? and 8.0 or 0.0, Right := 0.0, Bottom := IsTierUp? and 8.0 or 2.0}
                                Widget := Overlay

    CreateInfoUI():widget =
        stack_box:
            Orientation := orientation.Vertical
            Slots := array:
                stack_box_slot:
                    HorizontalAlignment := horizontal_alignment.Fill
                    Widget := color_block:
                        DefaultOpacity := 0.0
                        DefaultDesiredSize := vector2{X := 450.0, Y := 0.0}
                stack_box_slot:
                    HorizontalAlignment := horizontal_alignment.Center
                    VerticalAlignment := vertical_alignment.Top
                    Padding := margin{Left := 0.0, Top := 0.0, Right := 0.0, Bottom := 20.0}
                    Widget := overlay:
                        Slots := array:
                            overlay_slot:
                                HorizontalAlignment := horizontal_alignment.Left
                                VerticalAlignment := vertical_alignment.Top
                                Widget := Icon
                            overlay_slot:
                                HorizontalAlignment := horizontal_alignment.Right
                                VerticalAlignment := vertical_alignment.Center
                                Padding := margin{Left := 128.0, Top := 0.0, Right := 0.0, Bottom := 0.0}
                                Widget := StatIcon

                stack_box_slot:
                    HorizontalAlignment := horizontal_alignment.Center
                    VerticalAlignment := vertical_alignment.Fill
                    Widget := TierText
                stack_box_slot:
                    HorizontalAlignment := horizontal_alignment.Center
                    VerticalAlignment := vertical_alignment.Fill
                    Widget := NameText
                stack_box_slot:
                    HorizontalAlignment := horizontal_alignment.Center
                    VerticalAlignment := vertical_alignment.Bottom
                    Distribution := option{1.0}
                    Widget := stack_box:
                        Orientation := orientation.Vertical
                        Slots := array:
                            stack_box_slot:
                                HorizontalAlignment := horizontal_alignment.Center
                                VerticalAlignment := vertical_alignment.Fill
                                Widget := StatText
                            stack_box_slot:
                                HorizontalAlignment := horizontal_alignment.Center
                                VerticalAlignment := vertical_alignment.Fill
                                Widget := stack_box:
                                    Orientation := orientation.Horizontal
                                    Slots := array:
                                        stack_box_slot:
                                            HorizontalAlignment := horizontal_alignment.Left
                                            Padding := margin{Left := 0.0, Top := 0.0, Right := 10.0, Bottom := 0.0}
                                            Widget := EvolutionIcon
                                        stack_box_slot:
                                            HorizontalAlignment := horizontal_alignment.Fill
                                            VerticalAlignment := vertical_alignment.Fill
                                            Widget := StatDeltaText

    Uninitialize():void =
        for (Subscription : Subscriptions):
            Subscription.Cancel()

    <# ----------------------------------------------------------------------------------------------------
        Get Functions
    ---------------------------------------------------------------------------------------------------- #>
    GetActiveAbilityInfo()<transacts>:?ability_info =
        MaybeAbilityInfo

    <# ----------------------------------------------------------------------------------------------------
        Update Functions
    ---------------------------------------------------------------------------------------------------- #>
    Update(AbilityInfo:ability_info, InTier:int, IsRoll:logic):void =
        if (IsRoll?):
            set MaybeAbilityInfo = option{AbilityInfo}
            set Tier = InTier
            
            InfoWidget.SetVisibility(widget_visibility.Visible)

            if (WeaponId := AbilityInfo.MaybeWeaponId?):
                WeaponLevel := Player.GetWeaponLevel(WeaponId)
                ToggleProgress(true)                    
            else:
                ToggleProgress(false)
                EvolutionIcon.SetVisibility(widget_visibility.Collapsed)

            Icon.SetImage(AbilityInfo.Icon)
            Icon.SetTint(AbilityInfo.IconColor)

            if (AbilityStatIcon := AbilityInfo.MaybeStatIcon?):
                StatIcon.SetImage(AbilityStatIcon)
                StatIcon.SetVisibility(widget_visibility.Visible)
                StatIcon.SetTint(AbilityInfo.StatIconColor)
            else:
                StatIcon.SetVisibility(widget_visibility.Collapsed)

            TierText.SetText(MakeMessage("{Tier.ToTierString().ToUpper()}"))
            TierText.SetTextColor(Tier.ToTierColor() * 2.0)
            TierText.SetShadowColor(Tier.ToTierColor() * 0.25)
            NameText.SetText(MakeMessage("{AbilityInfo.Name.WrapText(200.0, ?TextSize := NameText.GetTextSize())}"))
            UpdateStats(Tier)

        if (MaybeAbilityInfo?):
            ItemCost := Player.GetItemShopAbilityCost(Tier)
            CostText.SetText(MakeMessage("{FormatAsCurrency(ItemCost)}"))
            if (Player.CanAfford(game_currency_id.Wood, ItemCost)?):
                CostText.SetTextColor(ThemeColors.Success)
                CostText.SetShadowColor(ThemeColors.Success * 0.125)
                BuyButton.SetText(MakeMessage("Buy"))
            else:
                CostText.SetTextColor(ThemeColors.Failure)
                CostText.SetShadowColor(ThemeColors.Failure * 0.125)
                BuyButton.SetText(MakeMessage("Cannot Afford"))

            if (WeaponId := AbilityInfo.MaybeWeaponId?):
                WeaponLevel := Player.GetWeaponLevel(WeaponId)

                SetProgress(WeaponLevel)
                for (Index -> ProgressPulsing : ProgressPulsings):
                    if:
                        ProjectionLevel := MaxWeaponLevel - (Index + 1)
                        ProjectionLevel = WeaponLevel or 
                        (
                            AbilityInfo.Id = ability_id.WeaponLevel
                            TierValue := Floor[AbilityInfo.TierValue[Tier - 1]?]
                            ProjectionLevel > WeaponLevel and ProjectionLevel <= WeaponLevel + TierValue
                        )
                    then:
                        ProgressPulsing.SetVisibility(widget_visibility.Visible)
                    else:
                        ProgressPulsing.SetVisibility(widget_visibility.Hidden)

                LevelUpAmount := if (AbilityInfo.Id = ability_id.WeaponLevel, TierValue := Floor[AbilityInfo.TierValue[Tier - 1]?]):
                    TierValue + 1
                else:
                    1
                
                CurrentWeaponEvolution := CalculateWeaponEvolutionLevel(WeaponLevel)
                UpdatedWeaponEvolution := CalculateWeaponEvolutionLevel(WeaponLevel + LevelUpAmount)
                set IsEvolution = logic{UpdatedWeaponEvolution > CurrentWeaponEvolution}
                EvolutionIcon.SetVisibility(IsEvolution? and widget_visibility.Visible or widget_visibility.Collapsed)
            else:
                set IsEvolution = false
        else:
            BuyButton.SetText(MakeMessage("Purchased"))

    Clear():void =
        set MaybeAbilityInfo = false
        
    UpdateStats(InTier:int):void =
        if (AbilityInfo := MaybeAbilityInfo?):            
            CurrentValue := Player.GetTotalAbilityValue(AbilityInfo)
            TierValue := if (StatId := AbilityInfo.MaybeStatId?):
                (AbilityInfo.TierValue[InTier - 1]? or 0.0) * (1.0 + Player.GetMultiplicativeStat(StatId))    
            else:
                AbilityInfo.TierValue[InTier - 1]? or 0.0

            var StatDeltaString:string = "{TierValue > 0.0 and "+" or ""}{AbilityInfo.FormatValue(TierValue)}"

            if (WeaponId := AbilityInfo.MaybeWeaponId?, AbilityInfo.Id = ability_id.Cooldown, Player.IsWeaponCooldownLimit(WeaponId)?):
                set StatDeltaString += " (+DMG)"

            StatText.SetText(MakeMessage("{AbilityInfo.FormatValue(CurrentValue)}"))
            StatDeltaText.SetText(MakeMessage(StatDeltaString))

    ToggleProgress(IsShow:logic):void =
        ProgressUI.SetVisibility(IsShow? and widget_visibility.Visible or widget_visibility.Hidden)
    
    SetProgress(WeaponLevel:int):void =
        for (Index -> BoxUI : BoxUIs):
            if (MaxWeaponLevel - (Index + 1) = WeaponLevel):
                BoxUI.SetColor(ThemeColors.Success)
            else if (MaxWeaponLevel - (Index + 2) >= WeaponLevel):
                BoxUI.SetColor(ThemeColors.Disabled)
            else:
                BoxUI.SetColor(ThemeColors.Highlight)

    Buy():void =
        Cost := Player.GetItemShopAbilityCost(Tier)
        if (Player.CanAfford(game_currency_id.Wood, Cost)?, AbilityInfo := MaybeAbilityInfo?):
            InfoWidget.SetVisibility(widget_visibility.Hidden)
            Player.AddAbility(AbilityInfo, ?Tier := Tier)
            Player.RemoveGameCurrency(game_currency_id.Wood, Cost)
            Player.PlayAudio(IsEvolution? and audio_id.Boost or audio_id.Purchase)
            set MaybeAbilityInfo = false
            BuyEvent.Signal(AbilityInfo, Tier, IsEvolution)
        else:
            Player.PlayAudio(audio_id.Error)

    <# ----------------------------------------------------------------------------------------------------
        Event handlers
    ---------------------------------------------------------------------------------------------------- #>
    OnClickBuy(Message:widget_message):void =
        Buy()