using { /Fortnite.com/Characters }
using { /Verse.org/Assets }
using { /Verse.org/Simulation }

<# ----------------------------------------------------------------------------------------------------
    Data Definitions
---------------------------------------------------------------------------------------------------- #>
item_shop_purchase_id := enum:
    Reroll
    Heal
    Shield
    Revive
    AddReroll

item_shop_purchase_info := struct<computes>:
    Id:item_shop_purchase_id
    Name:string
    Cost:float
    CanPurchase:type{_(:player)<transacts>:logic}
    OnPurchase:type{_(:player):void}

ItemShopPurchaseInfos:[item_shop_purchase_id]item_shop_purchase_info = map:
    item_shop_purchase_id.Reroll => item_shop_purchase_info:
        Id := item_shop_purchase_id.Reroll
        Name := "Reroll"
        Cost := 50.0
        CanPurchase := ItemShopCanPurchaseReroll
        OnPurchase := ItemShopOnPurchaseReroll
    item_shop_purchase_id.Heal => item_shop_purchase_info:
        Id := item_shop_purchase_id.Heal
        Name := "Heal +50%"
        Cost := 150.0
        CanPurchase := ItemShopCanPurchaseHeal
        OnPurchase := ItemShopOnPurchaseHeal
    item_shop_purchase_id.Shield => item_shop_purchase_info:
        Id := item_shop_purchase_id.Shield
        Name := "Shield +50%"
        Cost := 300.0
        CanPurchase := ItemShopCanPurchaseShield
        OnPurchase := ItemShopOnPurchaseShield
    item_shop_purchase_id.Revive => item_shop_purchase_info:
        Id := item_shop_purchase_id.Revive
        Name := "Revive +1"
        Cost := 1500.0
        CanPurchase := ItemShopCanPurchaseRevive
        OnPurchase := ItemShopOnPurchaseRevive
    item_shop_purchase_id.AddReroll => item_shop_purchase_info:
        Id := item_shop_purchase_id.AddReroll
        Name := "Reroll +1"
        Cost := 200.0
        CanPurchase := ItemShopCanPurchaseAddReroll
        OnPurchase := ItemShopOnPurchaseAddReroll

<# ----------------------------------------------------------------------------------------------------
    Global Functions
---------------------------------------------------------------------------------------------------- #>
GetItemShopPurchaseInfo(Id:item_shop_purchase_id)<decides><transacts>:item_shop_purchase_info =
    ItemShopPurchaseInfos[Id]

ItemShopOnPurchaseReroll(Player:player):void =
    Player.PlayAudio(audio_id.Reroll)

ItemShopCanPurchaseReroll(Player:player)<transacts>:logic =
    true

ItemShopOnPurchaseHeal(Player:player):void =
    Player.HealPercent(0.5)
    Player.PlayAudio(audio_id.Heal)

ItemShopCanPurchaseHeal(Player:player)<transacts>:logic =
    logic{not Player.IsMaxHealth()? and not Player.GetFortCharacter[].IsDownButNotOut[]}

ItemShopOnPurchaseShield(Player:player):void =
    Player.ShieldPercent(0.5)
    Player.PlayAudio(audio_id.Heal)

ItemShopCanPurchaseShield(Player:player)<transacts>:logic =
    logic{not Player.IsMaxShield()? and not Player.GetFortCharacter[].IsDownButNotOut[]}

ItemShopOnPurchaseRevive(Player:player):void =
    Player.AddPlayerResource(player_resource_id.Revives, 1)
    Player.PlayAudio(audio_id.Purchase)

ItemShopCanPurchaseRevive(Player:player)<transacts>:logic =
    logic{not Player.GetFortCharacter[].IsDownButNotOut[]}
    
ItemShopOnPurchaseAddReroll(Player:player):void =
    Player.AddPlayerResource(player_resource_id.Reroll, 1)
    Player.PlayAudio(audio_id.Purchase)

ItemShopCanPurchaseAddReroll(Player:player)<transacts>:logic =
    true