using { /Fortnite.com/UI }
using { /Verse.org/Assets }
using { /Verse.org/Colors }
using { /UnrealEngine.com/Temporary/SpatialMath }
using { /UnrealEngine.com/Temporary/UI }

item_shop_timer_ui := class(ui_component):
    ProgressBarWidth:float = 300.0

    Icon:texture_block = texture_block:
        DefaultImage := UserInterface.Icons.Badge_ItemShop_64
        DefaultDesiredSize := vector2{X := 40.0, Y := 40.0}

    ProgressBar:color_block = color_block:
        DefaultColor := NamedColors.White
        DefaultOpacity := 0.3

    ProgressText:text_block = text_block:
        DefaultTextColor := ThemeColors.PrimaryText
        DefaultShadowOffset := option{vector2{X := 2.0, Y := 2.0}}
        DefaultShadowOpacity := 0.5

    CancelEvent:event(void) = event(void){}

    <# ----------------------------------------------------------------------------------------------------
        Initialization
    ---------------------------------------------------------------------------------------------------- #>
    CreateUI<override>():widget =
        NewUI := stack_box:
            Orientation := orientation.Horizontal
            Slots := array:
                stack_box_slot:
                    HorizontalAlignment := horizontal_alignment.Fill
                    VerticalAlignment := vertical_alignment.Fill
                    Padding := margin{Left := 0.0, Top := 0.0, Right := 0.0, Bottom := 0.0}
                    Widget := overlay:
                        Slots := array:
                            overlay_slot:
                                HorizontalAlignment := horizontal_alignment.Fill
                                VerticalAlignment := vertical_alignment.Fill
                                Padding := margin{Left := 0.0, Top := 0.0, Right := 0.0, Bottom := 0.0}
                                Widget := color_block:
                                    DefaultColor := NamedColors.Black
                                    DefaultOpacity := 0.6
                            overlay_slot:
                                HorizontalAlignment := horizontal_alignment.Center
                                VerticalAlignment := vertical_alignment.Center
                                Padding := margin{Left := 5.0, Top := 5.0, Right := 5.0, Bottom := 5.0}
                                Widget := Icon

                stack_box_slot:
                    HorizontalAlignment := horizontal_alignment.Fill
                    VerticalAlignment := vertical_alignment.Fill
                    Padding := margin{Left := 0.0, Top := 0.0, Right := 0.0, Bottom := 0.0}
                    Widget := overlay:
                        Slots := array:
                            overlay_slot:
                                HorizontalAlignment := horizontal_alignment.Fill
                                VerticalAlignment := vertical_alignment.Fill
                                Padding := margin{Left := 0.0, Top := 0.0, Right := 0.0, Bottom := 0.0}
                                Widget := MakeBox(?BorderRadius := 2.0, ?Opacity := 0.0, ?BorderColor := NamedColors.Black, ?BorderOpacity := 0.6)
                            overlay_slot:
                                HorizontalAlignment := horizontal_alignment.Fill
                                VerticalAlignment := vertical_alignment.Fill
                                Padding := margin{Left := 4.0, Top := 4.0, Right := 4.0, Bottom := 4.0}
                                Widget := color_block:
                                    DefaultColor := NamedColors.Black
                                    DefaultOpacity := 0.3
                                    DefaultDesiredSize := vector2{X := 300.0, Y := 32.0}
                            overlay_slot:
                                HorizontalAlignment := horizontal_alignment.Left
                                VerticalAlignment := vertical_alignment.Fill
                                Padding := margin{Left := 4.0, Top := 4.0, Right := 4.0, Bottom := 4.0}
                                Widget := ProgressBar
                            overlay_slot:
                                Widget := ProgressText

        Player.AddWidgetToHUD of canvas_slot:
            Anchors := anchors{Minimum := vector2{X := 0.5, Y := 1.0}, Maximum := vector2{X := 0.5, Y := 1.0}}
            Offsets := margin{Left := 0.0, Top := -250.0, Right := 0.0, Bottom := 0.0}
            Alignment := vector2{X := 0.5, Y := 1.0}
            Widget := NewUI

        NewUI

    <# ----------------------------------------------------------------------------------------------------
        Update Functions
    ---------------------------------------------------------------------------------------------------- #>
    Update(Time:float, Progress:float):void =
        ProgressBar.SetDesiredSize(vector2{X := (1.0 - Progress) * ProgressBarWidth, Y := 32.0})
        ProgressText.SetText(MakeMessage("{GetRoundedString(Time, 0)}"))
    
    SetIcon(NewIcon:texture):void =
        Icon.SetImage(NewIcon)
        ProgressText.SetTextColor(ThemeColors.PrimaryText)

    SetWarning():void =
        ProgressText.SetTextColor(ThemeColors.Failure)

    <# ----------------------------------------------------------------------------------------------------
        Event Handlers
    ---------------------------------------------------------------------------------------------------- #>
    