using { /Fortnite.com/Characters }
using { /Fortnite.com/UI }
using { /Verse.org/Assets }
using { /Verse.org/Colors }
using { /Verse.org/Simulation }
using { /UnrealEngine.com/Temporary/SpatialMath }
using { /UnrealEngine.com/Temporary/UI }

<# ----------------------------------------------------------------------------------------------------
    Constants
---------------------------------------------------------------------------------------------------- #>
ItemShopTotalEntries:int = 5

<# ----------------------------------------------------------------------------------------------------
    Global Functions
---------------------------------------------------------------------------------------------------- #>
(Player:player).GetItemShopAbilityCost(Tier:int)<transacts>:float =
    Tier * 250.0 * (1.0 - (Player.HasBoon(boon_id.DiscountCoupon)? and 0.2 or 0.0))

GetItemShopLuckMultiplier()<transacts>:float = 
    (GetCurrentStage() - 2) * 0.25

<# ----------------------------------------------------------------------------------------------------
    item_shop_ui
---------------------------------------------------------------------------------------------------- #>
item_shop_ui := class(panel_ui):
    Title<override>:string = "Item Shop"
    Anchors<override>:anchors = anchors{Minimum := vector2{X := 0.0, Y := 0.0}, Maximum := vector2{X := 1.0, Y := 1.0}}
    BackgroundOpacity<override>:type {_X:float where 0.000000 <= _X, _X <= 1.000000} = 0.7
    HasCloseButton<override>:logic = false

    ItemPurchasedEvent:events(player) = events(player){}
    DoneEvent:events(player) = events(player){}
   
    CurrencyText:text_block = text_block:
        DefaultTextColor := ThemeColors.Success
        DefaultShadowOffset := option{vector2{X := 2.0, Y := 2.0}}
        DefaultShadowOpacity := 1.0
        DefaultShadowColor := ThemeColors.Success * 0.125

    HealthText:text_block = text_block:
        DefaultTextColor := ThemeColors.PrimaryText
        DefaultShadowOffset := option{vector2{X := 2.0, Y := 2.0}}
        DefaultShadowOpacity := 0.5

    ShieldText:text_block = text_block:
        DefaultTextColor := ThemeColors.PrimaryText
        DefaultShadowOffset := option{vector2{X := 2.0, Y := 2.0}}
        DefaultShadowOpacity := 0.5

    RerollText:text_block = text_block:
        DefaultTextColor := ThemeColors.PrimaryText
        DefaultShadowOffset := option{vector2{X := 2.0, Y := 2.0}}
        DefaultShadowOpacity := 0.5

    ReviveText:text_block = text_block:
        DefaultTextColor := ThemeColors.PrimaryText
        DefaultShadowOffset := option{vector2{X := 2.0, Y := 2.0}}
        DefaultShadowOpacity := 0.5

    DoneButton:button_quiet = button_quiet{}
    BuyAllButton:button_quiet = button_quiet{}
    
    var ItemShopEntryUIs:[]item_shop_entry_ui = array{}
    var CurrentAbilities:[]tuple(ability_info, int) = array{}
    var RerollCount:int = 0
    var ItemPurchased:int = 0
    var ItemShopPurchaseButtons:[item_shop_purchase_id]button_quiet = map{}
    var InfoUIs:[ability_select_info_type]ui_component = map{}
    var Subscriptions:[]cancelable = array{}
    var EventsSubscriptions:[]events_cancelable = array{}

    <# ----------------------------------------------------------------------------------------------------
        Initialization
    ---------------------------------------------------------------------------------------------------- #>
    CreateContent<override>():widget =
        if (GameManager := GetGameManager[]):
            set EventsSubscriptions = array:
                GameManager.ResumeCountdownStartEvent.Subscribe(OnResumeCountdownStart)
                GameManager.ResumeCountdownCompleteEvent.Subscribe(OnResumeCountdownComplete)
                GameManager.ResumeCountdownEvent.Subscribe(OnResumeCountdown)
                GameManager.GameResumedEvent.Subscribe(OnGameResumed)
                ClosedEvent.Subscribe(OnClosed)

        TierRatesButton := button_quiet{DefaultText := MakeMessage("Tier Rates")}
        StatsInfoButton := button_quiet{DefaultText := MakeMessage("Stats Info")}
        DoneButton.SetText(MakeMessage("Done"))
        
        set Subscriptions = array:
            TierRatesButton.OnClick().Subscribe(OnClickTierRates)
            StatsInfoButton.OnClick().Subscribe(OnClickStatsInfo)
            DoneButton.OnClick().Subscribe(OnClickDone)
            BuyAllButton.OnClick().Subscribe(OnClickBuyAll)

        AbilitiesStackBox := stack_box{Orientation := orientation.Horizontal}

        set ItemShopEntryUIs = for (Index := 1..ItemShopTotalEntries):
            ItemShopEntryUI := item_shop_entry_ui{Player := Player}
            ItemShopEntryUI.Initialize()
            set EventsSubscriptions += array{ItemShopEntryUI.BuyEvent.Subscribe(OnBuy)}

            AbilitiesStackBox.AddWidget of stack_box_slot:
                HorizontalAlignment := horizontal_alignment.Fill
                VerticalAlignment := vertical_alignment.Fill
                Padding := margin{Left := (Index = 1) and 0.0 or 20.0, Top := 0.0, Right := 0.0, Bottom := 0.0}
                Distribution := option{1.0}
                Widget := ItemShopEntryUI.GetUI()

            ItemShopEntryUI

        stack_box:
            Orientation := orientation.Vertical
            Slots := array:
                stack_box_slot:
                    HorizontalAlignment := horizontal_alignment.Fill
                    VerticalAlignment := vertical_alignment.Fill
                    Widget := stack_box:
                        Orientation := orientation.Horizontal
                        Slots := array:
                            stack_box_slot:
                                HorizontalAlignment := horizontal_alignment.Fill
                                VerticalAlignment := vertical_alignment.Fill
                                Distribution := option{1.0}
                                Widget := CreateCurrencyUI(UserInterface.Icons.CurrencyWood_64, CurrencyText)
                            stack_box_slot:
                                HorizontalAlignment := horizontal_alignment.Fill
                                VerticalAlignment := vertical_alignment.Fill
                                Padding := margin{Left := 20.0, Top := 0.0, Right := 0.0, Bottom := 0.0}
                                Distribution := option{1.0}
                                Widget := CreateCurrencyUI(UserInterface.Icons.Heart_64, HealthText)
                            stack_box_slot:
                                HorizontalAlignment := horizontal_alignment.Fill
                                VerticalAlignment := vertical_alignment.Fill
                                Padding := margin{Left := 20.0, Top := 0.0, Right := 0.0, Bottom := 0.0}
                                Distribution := option{1.0}
                                Widget := CreateCurrencyUI(UserInterface.Icons.Shield_64, ShieldText)
                            stack_box_slot:
                                HorizontalAlignment := horizontal_alignment.Fill
                                VerticalAlignment := vertical_alignment.Fill
                                Padding := margin{Left := 20.0, Top := 0.0, Right := 0.0, Bottom := 0.0}
                                Distribution := option{1.0}
                                Widget := CreateCurrencyUI(UserInterface.Icons.Revive_Flat_64, ReviveText)
                            stack_box_slot:
                                HorizontalAlignment := horizontal_alignment.Fill
                                VerticalAlignment := vertical_alignment.Fill
                                Padding := margin{Left := 20.0, Top := 0.0, Right := 0.0, Bottom := 0.0}
                                Distribution := option{1.0}
                                Widget := CreateCurrencyUI(UserInterface.Icons.AbilityReroll_Flat_64, RerollText)

                stack_box_slot:
                    HorizontalAlignment := horizontal_alignment.Fill
                    VerticalAlignment := vertical_alignment.Fill
                    Padding := margin{Left := 0.0, Top := 10.0, Right := 0.0, Bottom := 0.0}
                    Widget := stack_box:
                        Orientation := orientation.Horizontal
                        Slots :=
                            var ButtonIndex:int = 0
                            for (ItemShopPurchaseInfo : ItemShopPurchaseInfos):
                                Button := button_quiet{}
                                spawn{AwaitItemShopPurchaseClicked(Button.OnClick(), ItemShopPurchaseInfo)}
                                if (set ItemShopPurchaseButtons[ItemShopPurchaseInfo.Id] = Button) {}
                                set ButtonIndex += 1

                                stack_box_slot:
                                    HorizontalAlignment := horizontal_alignment.Fill
                                    VerticalAlignment := vertical_alignment.Fill
                                    Distribution := option{1.0}
                                    Padding := margin{Left := (ButtonIndex = 1) and 0.0 or 20.0, Top := 0.0, Right := 0.0, Bottom := 0.0}
                                    Widget := Button
                stack_box_slot:
                    HorizontalAlignment := horizontal_alignment.Fill
                    VerticalAlignment := vertical_alignment.Fill
                    Padding := margin{Left := 0.0, Top := 40.0, Right := 0.0, Bottom := 0.0}
                    Widget := AbilitiesStackBox
                stack_box_slot:
                    HorizontalAlignment := horizontal_alignment.Fill
                    VerticalAlignment := vertical_alignment.Bottom
                    Padding := margin{Left := 0.0, Top := 20.0, Right := 0.0, Bottom := 20.0}
                    Widget := color_block:
                        DefaultColor := NamedColors.White
                        DefaultOpacity := 0.1
                        DefaultDesiredSize := vector2{X := 0.0, Y := 1.0}
                stack_box_slot:
                    HorizontalAlignment := horizontal_alignment.Fill
                    VerticalAlignment := vertical_alignment.Fill
                    Padding := margin{Left := 0.0, Top := 20.0, Right := 0.0, Bottom := 0.0}
                    Widget := stack_box:
                        Orientation := orientation.Horizontal
                        Slots := array:
                            stack_box_slot:
                                HorizontalAlignment := horizontal_alignment.Fill
                                VerticalAlignment := vertical_alignment.Fill
                                Distribution := option{2.05}
                                Widget := DoneButton
                            stack_box_slot:
                                HorizontalAlignment := horizontal_alignment.Fill
                                VerticalAlignment := vertical_alignment.Fill
                                Distribution := option{1.0}
                                Padding := margin{Left := 20.0, Top := 0.0, Right := 0.0, Bottom := 0.0}
                                Widget := BuyAllButton
                            stack_box_slot:
                                HorizontalAlignment := horizontal_alignment.Fill
                                VerticalAlignment := vertical_alignment.Fill
                                Distribution := option{1.0}
                                Padding := margin{Left := 20.0, Top := 0.0, Right := 0.0, Bottom := 0.0}
                                Widget := StatsInfoButton
                            stack_box_slot:
                                HorizontalAlignment := horizontal_alignment.Fill
                                VerticalAlignment := vertical_alignment.Fill
                                Distribution := option{1.0}
                                Padding := margin{Left := 20.0, Top := 0.0, Right := 0.0, Bottom := 0.0}
                                Widget := TierRatesButton

    CreateCurrencyUI(Icon:texture, Text:text_block):widget =
        overlay:
            Slots := array:
                overlay_slot:
                    HorizontalAlignment := horizontal_alignment.Fill
                    VerticalAlignment := vertical_alignment.Fill
                    Widget := UserInterface.VerseWidget.WBP_RoundedBox{}
                overlay_slot:
                    Padding := margin{Left := 40.0, Top := 10.0, Right := 40.0, Bottom := 10.0}
                    Widget := stack_box:
                        Orientation := orientation.Horizontal
                        Slots := array:
                            stack_box_slot:
                                HorizontalAlignment := horizontal_alignment.Left
                                Padding := margin{Left := 0.0, Top := 0.0, Right := 10.0, Bottom := 0.0}
                                Widget := texture_block:
                                    DefaultImage := Icon
                                    DefaultDesiredSize := vector2{X := 40.0, Y := 40.0}
                            stack_box_slot:
                                HorizontalAlignment := horizontal_alignment.Fill
                                VerticalAlignment := vertical_alignment.Fill
                                Widget := Text

    Uninitialize<override>():void =
        for (ItemShopEntryUI : ItemShopEntryUIs):
            ItemShopEntryUI.Uninitialize()
            
        for (EventsSubscription : EventsSubscriptions):
            EventsSubscription.Cancel()

        for (Subscription : Subscriptions):
            Subscription.Cancel()

    <# ----------------------------------------------------------------------------------------------------
        Get Functions
    ---------------------------------------------------------------------------------------------------- #>
    GetCurrentAbilities()<transacts>:[]ability_id =
        for (ItemShopEntryUI : ItemShopEntryUIs, AbilityInfo := ItemShopEntryUI.GetActiveAbilityInfo()?):
            AbilityInfo.Id
    
    GetRerollCost()<transacts>:float =
        50.0 * (RerollCount + 1)

    CalculateCost(ItemShopPurchaseInfo:item_shop_purchase_info)<transacts>:float =
        BaseCost := (ItemShopPurchaseInfo.Id = item_shop_purchase_id.Reroll) and GetRerollCost() or ItemShopPurchaseInfo.Cost
        DiscountMultiplier := Player.HasBoon(boon_id.DiscountCoupon)? and 0.8 or 1.0
        BaseCost * DiscountMultiplier

    IsShopEmpty()<transacts>:logic =
        logic{not GetCurrentAbilities()[0]}

    GetBuyAllCost()<transacts>:float =
        var TotalCost:float = 0.0
        for (Index -> ItemShopEntryUI : ItemShopEntryUIs, ItemShopEntryUI.MaybeAbilityInfo?):
            set TotalCost += Player.GetItemShopAbilityCost(ItemShopEntryUI.Tier)

        TotalCost

    <# ----------------------------------------------------------------------------------------------------
        Update Functions
    ---------------------------------------------------------------------------------------------------- #>
    Update(?IsRoll:logic = true):void =
        for (Index -> ItemShopEntryUI : ItemShopEntryUIs):
            if (AbilityInfo := CurrentAbilities[Index]):
                ItemShopEntryUI.Update(AbilityInfo(0), AbilityInfo(1), IsRoll)
                ItemShopEntryUI.Show()

        BuyAllButton.SetText(MakeMessage(IsShopEmpty()? and "Reroll ({FormatAsCurrency(CalculateCost(GetItemShopPurchaseInfo[item_shop_purchase_id.Reroll]))})" or "Buy All ({FormatAsCurrency(GetBuyAllCost())})"))

        if (Character := Player.GetFortCharacter[]):
            HealthText.SetText(MakeMessage("{GetRoundedString(Character.GetHealth(), 0)}/{GetRoundedString(Character.GetMaxHealth(), 0)}"))
            HealthText.SetTextColor(Player.IsMaxHealth()? and ThemeColors.Success or ThemeColors.Failure)

            ShieldText.SetText(MakeMessage("{GetRoundedString(Character.GetShield(), 0)}/{GetRoundedString(Character.GetMaxShield(), 0)}"))
            ShieldText.SetTextColor(Player.IsMaxShield()? and ThemeColors.Success or ThemeColors.Failure)

            ReviveText.SetText(MakeMessage("{Player.GetPlayerResource(player_resource_id.Revives)}"))

            RerollText.SetText(MakeMessage("{Player.GetPlayerResource(player_resource_id.Reroll)}"))

        for (Id -> ItemShopPurchaseButton : ItemShopPurchaseButtons, Info := GetItemShopPurchaseInfo[Id]):
            if (Info.CanPurchase(Player)?):
                ItemShopPurchaseButton.SetText(MakeMessage("{Info.Name} ({FormatAsCurrency(CalculateCost(Info))})"))
            else:
                ItemShopPurchaseButton.SetText(MakeMessage("{Info.Name} (N/A)"))

        CurrencyText.SetText(MakeMessage("{FormatAsCurrency(Player.GetGameCurrency(game_currency_id.Wood))}"))

    UpdateDoneCount():void =
        PlayerCount := GetActivePlayers().Length
        DoneButton.SetText(MakeMessage("Done ({PlayerCount - GetTotalPauseCount(pause_type.ItemShop)}/{PlayerCount})"))

    Roll(?IsReroll:logic = false):void =
        IgnoreAbilities:[]ability_id = (IsReroll? and GetCurrentAbilities() or array{})
        set CurrentAbilities = GetRandomAbilities(Player, ItemShopTotalEntries, ?Ignore := IgnoreAbilities, ?LuckOverride := option{GetItemShopLuckMultiplier()})
        set RerollCount += 1
        Update()

    ToggleInfo(AbilitySelectInfoType:ability_select_info_type):void =
        for (Type -> InfoUI : InfoUIs, Type <> AbilitySelectInfoType):
            InfoUI.Hide()

        UI := if (InfoUI := InfoUIs[AbilitySelectInfoType]):
            Visibility := InfoUI.GetUI().GetVisibility()
            if (Visibility = widget_visibility.Visible):
                InfoUI.Hide()
            else:
                InfoUI.Show()
                InfoUI.Update()
            InfoUI
        else:
            InfoUI := case (AbilitySelectInfoType):
                ability_select_info_type.StatsInfo => ability_stat_info_ui{Player := Player, BackgroundType := background_type.Opaque, TextSize := ThemeTextSize.Small, IconSize := 0.75}
                ability_select_info_type.TierRates => ability_select_tier_rates_ui{Player := Player}

            InfoUI.Initialize()
            if (set InfoUIs[AbilitySelectInfoType] = InfoUI) {}

            if (PanelUI := canvas[MaybeUI?]):
                PanelUI.AddWidget of canvas_slot:
                    Anchors := case (AbilitySelectInfoType):
                        ability_select_info_type.StatsInfo => anchors{Minimum := vector2{X := 0.0, Y := 0.0}, Maximum := vector2{X := 0.6, Y := 0.0}}
                        ability_select_info_type.TierRates => anchors{Minimum := vector2{X := 0.0, Y := 0.0}, Maximum := vector2{X := 0.0, Y := 0.0}}
                    Alignment := vector2{X := 0.0, Y := 0.0}
                    ZOrder := 1
                    Widget := InfoUI.GetUI()

                InfoUI.Update()

            InfoUI

        if (RatesUI := ability_select_tier_rates_ui[UI]):
            RatesUI.Update(?LuckOverride := option{GetItemShopLuckMultiplier()})

    <# ----------------------------------------------------------------------------------------------------
        Event Handlers
    ---------------------------------------------------------------------------------------------------- #>
    OnShow<override>():void =
        set RerollCount = -1
        set ItemPurchased = 0
        TitleText.SetText(MakeMessage("{Title}"))
        TitleText.SetTextColor(ThemeColors.PrimaryText)
        TitleText.SetShadowOpacity(0.5)
        DoneButton.SetText(MakeMessage("Done (0/{GetActivePlayers().Length})"))
        for (ItemShopEntryUI : ItemShopEntryUIs) {ItemShopEntryUI.Clear}
        Roll(?IsReroll := true)

        for (InfoUI : InfoUIs):
            InfoUI.Hide()

    OnBuy(AbilityInfo:ability_info, Tier:int, IsEvolution:logic):void =        
        if (IsEvolution?, WeaponId := AbilityInfo.MaybeWeaponId?):
            Hide(?IsSignal := false)
            PauseGameFront(Player, pause_weapon_evolution{Type := pause_type.WeaponEvolution, Id := WeaponId, EvolutionLevel := Player.GetWeaponEvolutionLevel(WeaponId), IsItemShop := true})

        set ItemPurchased += 1
        Update(?IsRoll := false)
        ItemPurchasedEvent.Signal(Player)

    AwaitItemShopPurchaseClicked(Listenable:listenable(widget_message), ItemShopPurchaseInfo:item_shop_purchase_info)<suspends>:void =
        if (PlayerRemovedEvent := Player.GetPlayerRemovedEvent[]):
            race:
                PlayerRemovedEvent.Await()
                loop:
                    Listenable.Await()
                    PurchaseCost := CalculateCost(ItemShopPurchaseInfo)

                    if (ItemShopPurchaseInfo.CanPurchase(Player)? and Player.CanAfford(game_currency_id.Wood, PurchaseCost)?):
                        Player.RemoveGameCurrency(game_currency_id.Wood, PurchaseCost)
                        ItemShopPurchaseInfo.OnPurchase(Player)
                        
                        if (ItemShopPurchaseInfo.Id = item_shop_purchase_id.Reroll):
                            Roll(?IsReroll := true)
                        else:
                            Update(?IsRoll := false)
                            ItemPurchasedEvent.Signal(Player)
                    else:
                        Player.PlayAudio(audio_id.Error)

    OnResumeCountdownStart():void =
        if (not IsShowing?) then return
        Player.PlayAudio(audio_id.TimerWarning)
        TitleText.SetText(MakeMessage("{GameResumeCountdownTimeItemShop}s"))
        TitleText.SetTextColor(ThemeColors.Failure)
        TitleText.SetShadowColor(ThemeColors.Failure * 0.125)
        TitleText.SetShadowOffset(option{vector2{X := 2.0, Y := 2.0}})
        TitleText.SetShadowOpacity(1.0)
        
    OnResumeCountdownComplete():void =
        if (not IsShowing?) then return
        Hide()

    OnResumeCountdown(TimeRemaining:int):void =
        if (not IsShowing?) then return
        if (Player.GetIsAFK()?):
            ResumeGame(Player)
            DoneEvent.Signal(Player)

        TitleText.SetText(MakeMessage("{TimeRemaining}s"))

    OnGameResumed():void =
        if (not IsShowing?) then return
        TitleText.SetText(MakeMessage("{Title}"))
        TitleText.SetTextColor(ThemeColors.PrimaryText)
        TitleText.SetShadowOpacity(0.5)

    OnClickTierRates(Message:widget_message):void =
        ToggleInfo(ability_select_info_type.TierRates)

    OnClickStatsInfo(Message:widget_message):void =
        ToggleInfo(ability_select_info_type.StatsInfo)

    OnClickDone(Message:widget_message):void =
        ResumeGame(Player)
        DoneEvent.Signal(Player)

    OnClickBuyAll(Message:widget_message):void =
        if (IsShopEmpty()?):
            if:
                ItemShopPurchaseInfo := GetItemShopPurchaseInfo[item_shop_purchase_id.Reroll]
                PurchaseCost := CalculateCost(ItemShopPurchaseInfo)
                ItemShopPurchaseInfo.CanPurchase(Player)? and Player.CanAfford(game_currency_id.Wood, PurchaseCost)?
            then:
                Player.RemoveGameCurrency(game_currency_id.Wood, PurchaseCost)
                Roll(?IsReroll := true)
            else:
                Player.PlayAudio(audio_id.Error)

        else if:
            PurchaseCost := GetBuyAllCost()
            Player.CanAfford(game_currency_id.Wood, GetBuyAllCost())?
        then:
            var EvolutionWeaponIds:[]weapon_id = array{}
            var PauseInfos:[]pause_info = array{}
            for (Index -> ItemShopEntryUI : ItemShopEntryUIs, AbilityInfo := ItemShopEntryUI.MaybeAbilityInfo?):
                ItemShopEntryUI.InfoWidget.SetVisibility(widget_visibility.Hidden)
                Player.AddAbility(AbilityInfo, ?Tier := ItemShopEntryUI.Tier)
                set ItemShopEntryUI.MaybeAbilityInfo = false
                set ItemPurchased += 1
                ItemPurchasedEvent.Signal(Player)

                if (ItemShopEntryUI.IsEvolution?, WeaponId := AbilityInfo.MaybeWeaponId?, not EvolutionWeaponIds.Find[WeaponId]):
                    PauseInfo := pause_weapon_evolution{Type := pause_type.WeaponEvolution, Id := WeaponId, EvolutionLevel := Player.GetWeaponEvolutionLevel(WeaponId), IsItemShop := true}
                    set PauseInfos += array{PauseInfo}
                    set EvolutionWeaponIds += array{WeaponId}

            if (PauseInfos[0]):
                Hide(?IsSignal := false)
                PauseGameFront(Player, PauseInfos)
                Player.PlayAudio(audio_id.Boost)
            else:
                Player.PlayAudio(audio_id.Purchase)

            Player.RemoveGameCurrency(game_currency_id.Wood, PurchaseCost)
            Update(?IsRoll := false)
        else:
            Player.PlayAudio(audio_id.Error)

    OnClosed(InPlayer:player):void =
        for (Index := 1..ItemPurchased):
            Player.AwardAccolade(accolade_id.ItemShop)

        set ItemPurchased = 0