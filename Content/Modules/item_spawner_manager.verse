using { /Fortnite.com/Characters }
using { /Fortnite.com/Devices }
using { /Verse.org/Random }
using { /Verse.org/SceneGraph }
using { /Verse.org/SceneGraph/KeyframedMovement }
using { /Verse.org }
using { /Verse.org/Simulation }
using { /UnrealEngine.com/Temporary/SpatialMath }
using { Prefabs }
using { VisualEffects }

<# ----------------------------------------------------------------------------------------------------
    Global Functions
---------------------------------------------------------------------------------------------------- #>
var GlobalItemSpawnerManager:weak_map(session, item_spawner_manager) = map{}

GetItemSpawnerManager()<decides><transacts>:item_spawner_manager =
    GlobalItemSpawnerManager[GetSession()]

SpawnItemDrop(Id:item_drop_id, Position:vector3):?item_drop =
    if (ItemSpawnerManager := GetItemSpawnerManager[]):
        return ItemSpawnerManager.Spawn(Id, Position)

    return false

ClearItemDrops():void =
    if (ItemSpawnerManager := GetItemSpawnerManager[]):
        ItemSpawnerManager.DisableAll()

    # [TODO] Refactor
    if (ExperienceOrbManager := GetExperienceOrbManager[]):
        ExperienceOrbManager.DespawnAll()

<# ----------------------------------------------------------------------------------------------------
    item_spawner_manager
---------------------------------------------------------------------------------------------------- #>
item_spawner_manager := class(creative_device):
    @editable
    ItemDropDefinitions:[]item_drop_definition = array{}
    
    ItemDropFactory:item_drop_factory = item_drop_factory{}

    var MappedItemDropDefinitions<private>:[item_drop_id]item_drop_definition = map{}
    var ActiveItemDrops:[item_drop_id][]item_drop = map{}
    var ItemDropObjectPoolByType<private>:[item_drop_id]queue(item_drop) = map{}
    var ItemDropMapIndicatorQueue<private>:[item_drop_id]queue(map_indicator_device) = map{}

    OnBegin<override>()<suspends>:void =
        if (set GlobalItemSpawnerManager[GetSession()] = Self) {}
        
        for (ItemDropDefinition : ItemDropDefinitions, set MappedItemDropDefinitions[ItemDropDefinition.Id] = ItemDropDefinition):
            if:
                ItemDropMapIndicators := ItemDropDefinition.MaybeMapIndicators?
                set ItemDropMapIndicatorQueue[ItemDropDefinition.Id] = queue(map_indicator_device){Elements := ItemDropMapIndicators}
            then:
                for (ItemDropMapIndicator : ItemDropMapIndicators):
                    if (ItemDropMapIndicator.TeleportTo[MapIndicatorDisabledPosition, IdentityRotation()]) {}

    <# ----------------------------------------------------------------------------------------------------
        Get Functions
    ---------------------------------------------------------------------------------------------------- #>
    GetItemDropDefinition(Id:item_drop_id)<decides><transacts>:item_drop_definition =
        MappedItemDropDefinitions[Id]

    GetPooledItemDrop(Id:item_drop_id)<decides><transacts>:item_drop =
        if (not ItemDropObjectPoolByType[Id]):
            set ItemDropObjectPoolByType[Id] = queue(item_drop){}

        ItemDropObjectPool := ItemDropObjectPoolByType[Id]
        if (ItemDropObjectPool.IsEmpty()?):
            ItemDropFactory.GetItemDrop(Self, Id)
        else:
            DequeuedObject := ItemDropObjectPool.Dequeue[]
            set ItemDropObjectPoolByType[Id] = DequeuedObject(0)
            DequeuedObject(1)

    GetMapIndicator(Id:item_drop_id)<decides><transacts>:map_indicator_device =
        DequeuedObject := ItemDropMapIndicatorQueue[Id].Dequeue[]
        set ItemDropMapIndicatorQueue[Id] = DequeuedObject(0)
        DequeuedObject(1)

    ReturnMapIndicator(Id:item_drop_id, MapIndicator:map_indicator_device):void =
        if:
            UpdatedItemDropMapIndicatorQueue := ItemDropMapIndicatorQueue[Id].Enqueue(MapIndicator)
            set ItemDropMapIndicatorQueue[Id] = UpdatedItemDropMapIndicatorQueue
            MapIndicator.TeleportTo[MapIndicatorDisabledPosition, IdentityRotation()]

    <# ----------------------------------------------------------------------------------------------------
        Update Functions
    ---------------------------------------------------------------------------------------------------- #>               
    Spawn(Id:item_drop_id, Position:vector3):?item_drop =
        if (ItemDrop := GetPooledItemDrop[Id], ItemDropDefinition := GetItemDropDefinition[Id]):
            ItemDrop.OnBegin(transform{Translation := Position.Make2D(), Rotation := IdentityRotation().ApplyYaw(DegreesToRadians(90.0))}, ItemDropDefinition)
            if (set ActiveItemDrops[Id] = ActiveItemDrops[Id] + array{ItemDrop} or array{ItemDrop}) {}
            return option{ItemDrop}

        return false

    DisableAll():void =
        for (ActiveItemDrop : ActiveItemDrops):
            for (ItemDrop : ActiveItemDrop):
                ItemDrop.Disable()

    ReturnToObjectPool(ItemDrop:item_drop):void =
        if:
            set ActiveItemDrops[ItemDrop.Id] = for (ActiveItemDrop : ActiveItemDrops[ItemDrop.Id], ActiveItemDrop <> ItemDrop) {ActiveItemDrop}

        if:
            ItemDropObjectPool := ItemDropObjectPoolByType[ItemDrop.Id]
            set ItemDropObjectPoolByType[ItemDrop.Id] = ItemDropObjectPool.Enqueue(ItemDrop)