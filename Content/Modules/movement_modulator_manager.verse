using { /Fortnite.com/Devices }
using { /Verse.org/Simulation }
using { /UnrealEngine.com/Temporary/SpatialMath }

<# ----------------------------------------------------------------------------------------------------
    Data definitions
---------------------------------------------------------------------------------------------------- #>
impulse_id := enum:
    Dash
    Knockback
    DashBoon
    Upward

impulse := class<concrete>:
    @editable
    Id:impulse_id = impulse_id.Dash
    @editable
    Device:movement_modulator_device = movement_modulator_device{}

<# ----------------------------------------------------------------------------------------------------
    Global Functions
---------------------------------------------------------------------------------------------------- #>
var GlobalMovementModulatorManager:weak_map(session, movement_modulator) = map{}

GetMovementModulatorManager()<decides><transacts>:movement_modulator =
    GlobalMovementModulatorManager[GetSession()]

(Agent:agent).ApplyImpulse(Id:impulse_id, ?Rotation:rotation = IdentityRotation()):void =
    if (MovementModulatorManager := GetMovementModulatorManager[]):
        MovementModulatorManager.ApplyImpulse(Agent, Id, Rotation)

<# ----------------------------------------------------------------------------------------------------
    movement_modulator
---------------------------------------------------------------------------------------------------- #>
movement_modulator := class(creative_device):
    @editable
    Impulses:[]impulse = array{}
    
    DefaultMusicId:music_id = music_id.Default
    DefaultAudioPlayer:audio_player_device = audio_player_device{}

    var MappedImpulses:[impulse_id]movement_modulator_device = map{}

    <# ----------------------------------------------------------------------------------------------------
        Initialization
    ---------------------------------------------------------------------------------------------------- #>
    OnBegin<override>()<suspends>:void =
        if (set GlobalMovementModulatorManager[GetSession()] = Self) {}

        for (Impulse : Impulses, set MappedImpulses[Impulse.Id] = Impulse.Device) {}

    <# ----------------------------------------------------------------------------------------------------
        Get Functions
    ---------------------------------------------------------------------------------------------------- #>
    GetImpulse(Id:impulse_id)<decides><transacts>:movement_modulator_device =
        MappedImpulses[Id]

    <# ----------------------------------------------------------------------------------------------------
        Get Functions
    ---------------------------------------------------------------------------------------------------- #>
    ApplyImpulse(Agent:agent, Id:impulse_id, Rotation:rotation):void =
        if (Impulse := GetImpulse[Id], Impulse.TeleportTo[Impulse.GetTransform().Translation, Rotation]):
            Impulse.Activate(Agent)

        if (Player := player[Agent], MovementModulator := GetImpulse[Id]):
            
            if (MovementModulator.TeleportTo[MovementModulator.GetTransform().Translation, Rotation]):
                MovementModulator.Activate(Player)

    <# ----------------------------------------------------------------------------------------------------
        Event Handlers
    ---------------------------------------------------------------------------------------------------- #>
