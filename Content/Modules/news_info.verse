using { /Verse.org/Assets }
using { /Verse.org/Colors }

news_info := struct<concrete><computes>:
    Date:string = ""
    Title:string = ""
    Version:float = 0.0
    Entries:[]news_entry_info = array{}
    
news_entry_info := struct<concrete><computes>:
    Type:news_entry_type = news_entry_type.NewFeatures
    Notes:string = ""

news_entry_type := enum:
    None
    NewFeatures
    BugFixes
    PerformanceOptimizations
    GameplayChanges

(Type:news_entry_type).ToIcon()<transacts>:texture =
    case (Type):
        news_entry_type.None => UserInterface.Icons.NewsEntryType_NewFeature_64
        news_entry_type.NewFeatures => UserInterface.Icons.NewsEntryType_NewFeature_64
        news_entry_type.BugFixes => UserInterface.Icons.NewsEntryType_BugFix_64
        news_entry_type.PerformanceOptimizations => UserInterface.Icons.NewsEntryType_PerformanceOptimization_64
        news_entry_type.GameplayChanges => UserInterface.Icons.NewsEntryType_GameplayChange_64

(Type:news_entry_type).ToColor()<transacts>:color =
    case (Type):
        news_entry_type.None => MakeColorFromHex("FF3C47")
        news_entry_type.NewFeatures => MakeColorFromHex("F7FF19")
        news_entry_type.BugFixes => MakeColorFromHex("FF3C47")
        news_entry_type.PerformanceOptimizations => MakeColorFromHex("A7FF90")
        news_entry_type.GameplayChanges => MakeColorFromHex("3CBFFF")
        
GetNewsInfos()<transacts>:[]news_info =
    NewsInfos

NewsInfos:[]news_info =
    array:
        news_info:
            Date := "August 20, 2025"
            Title := "Update 46 - Post game sequence skip, various bug fixes & optimizations."
            Version := 1.031
            Entries := array:
                news_entry_info{Type := news_entry_type.NewFeatures, Notes := "Added option to skip the post game sequence (press Interact key)."}

                news_entry_info{Type := news_entry_type.PerformanceOptimizations, Notes := "Reworked Wood Gathered, Player Damaged/Heal number VFX into single particle system."}
                news_entry_info{Type := news_entry_type.None, Notes := "Wood Gathered VFX updated particle size to scale with amount gathered."}

                news_entry_info{Type := news_entry_type.BugFixes, Notes := "Fixed - Weapon position getting desync/delayed at low cooldown."}
                news_entry_info{Type := news_entry_type.None, Notes := "Weapons affected: Bow, Kunai, Hammer, Staff, Torrent, Energy Beam, Energy Bomb."}
                news_entry_info{Type := news_entry_type.BugFixes, Notes := "Fixed - Damage numbers flickering (z-fighting) when close together."}

        news_info:
            Date := "August 17, 2025"
            Title := "Update 45 - Post game sqeuence, mobile improvements, performance optimizations & bug fixes."
            Version := 1.030
            Entries := array:
                news_entry_info{Type := news_entry_type.NewFeatures, Notes := "Added a post game sequence - UI showing player stat summary."}
                news_entry_info{Type := news_entry_type.NewFeatures, Notes := "Added input controls prompt that is shown to new players (KBM, Controller, Touch)."}
                news_entry_info{Type := news_entry_type.None, Notes := "Updated input controls terminal to show prompt UI."}
                news_entry_info{Type := news_entry_type.NewFeatures, Notes := "Updated various UI with smaller font size."}
                news_entry_info{Type := news_entry_type.None, Notes := "Stats Info, News and Accolades UI"}
                news_entry_info{Type := news_entry_type.None, Notes := "Item Shop UI - also updated the layout so cost is next to the buy button."}
                
                news_entry_info{Type := news_entry_type.PerformanceOptimizations, Notes := "Prevent lag spike caused by damaging too many NPCs at the same time."}
                news_entry_info{Type := news_entry_type.None, Notes := "Added random NPC respawn time & random spawn invulnerability time (0.1 - 0.5s)."}
                news_entry_info{Type := news_entry_type.PerformanceOptimizations, Notes := "Optimized weapons to ignore NPCs in eliminated state from collision detection."}
                news_entry_info{Type := news_entry_type.PerformanceOptimizations, Notes := "Optimized NPC respawn logic (reduced by 0.1 ms)."}
                
                news_entry_info{Type := news_entry_type.BugFixes, Notes := "Fixed - Mobile input controls - dashing intermittently targets facing direction instead of moving direction."}
                news_entry_info{Type := news_entry_type.BugFixes, Notes := "Fixed - NPC damage was being calculated if NPC is eliminated and damaged multiple times on the same frame (improves performance)."}
                news_entry_info{Type := news_entry_type.BugFixes, Notes := "Fixed - NPCs were sometimes respawning at the position they were eliminated."}
                news_entry_info{Type := news_entry_type.BugFixes, Notes := "Fixed - Green slimes were not increasing in tiers (normal > elite > general) at higher game speeds."}
                news_entry_info{Type := news_entry_type.BugFixes, Notes := "Fixed - Hammer return projectiles applying damage in the incorrect position."}
                news_entry_info{Type := news_entry_type.BugFixes, Notes := "Fixed - Camera rotation not following boss in boss fight cinematic sequence."}

        news_info:
            Date := "August 6, 2025"
            Title := "Update 44 - Improved weapon VFX animations, optimizations & bug fix."
            Version := 1.022
            Entries := array:
                news_entry_info{Type := news_entry_type.NewFeatures, Notes := "Reworked weapon VFXs to improve animations & fix movement jitter."}
                news_entry_info{Type := news_entry_type.None, Notes := "Shuriken."}
                news_entry_info{Type := news_entry_type.None, Notes := "Tesla Coil (Tether evolution)."}
                news_entry_info{Type := news_entry_type.None, Notes := "Boomerang - also added animation easing."}

                news_entry_info{Type := news_entry_type.PerformanceOptimizations, Notes := "Optimized damage numbers."}
                news_entry_info{Type := news_entry_type.PerformanceOptimizations, Notes := "Optimized various HUD UI (elimination counter, timer, XP bar)."}

                news_entry_info{Type := news_entry_type.BugFixes, Notes := "Fixed - Boss health bar initially showing the incorrect value."}

        news_info:
            Date := "August 1, 2025"
            Title := "Update 43 - More game speeds, more performance optimization & bug fixes."
            Version := 1.021
            Entries := array:
                news_entry_info{Type := news_entry_type.PerformanceOptimizations, Notes := "Significantly optimized damaging NPCs (1.8ms > 0.2ms)."}
                
                news_entry_info{Type := news_entry_type.GameplayChanges, Notes := "Added more Game Speeds 4x - 6x."}
                news_entry_info{Type := news_entry_type.None, Notes := "Unlocks at Arena Level 0/9/36 > 0/9/18/36/54/72 (Game Speed 1/2/3/4/5/6x)"}
                news_entry_info{Type := news_entry_type.GameplayChanges, Notes := "Increased Air Dash power up max level from 1 > 2."}
                news_entry_info{Type := news_entry_type.GameplayChanges, Notes := "Increased Dash Count power up max level from 4 > 5."}
                
                news_entry_info{Type := news_entry_type.BugFixes, Notes := "Fixed - Weapon VFX position delayed (Sword, Lightning, Shield Blast, Trail Blast)."}
                news_entry_info{Type := news_entry_type.BugFixes, Notes := "Fixed - Trail VFX jagged lines (still occurs but significantly improved)."}
                news_entry_info{Type := news_entry_type.BugFixes, Notes := "Fixed - Shield weapon was only activating after first cooldown completed."}
                news_entry_info{Type := news_entry_type.BugFixes, Notes := "Fixed - Multiplicative stats (Health Regen/Max Health) was accumulating after each game."}
                news_entry_info{Type := news_entry_type.BugFixes, Notes := "Fixed - Stat info showing locked when weapon is level 0."}
                news_entry_info{Type := news_entry_type.BugFixes, Notes := "Fixed - End of stage black screen was ending too early."}
                news_entry_info{Type := news_entry_type.BugFixes, Notes := "Fixed - Select Buffer Time setting getting stuck above max value."}
                news_entry_info{Type := news_entry_type.BugFixes, Notes := "Fixed - Max dash charges was capped at 1 when first joining the lobby."}

        news_info:
            Date := "July 26, 2025"
            Title := "Update 42 - Scene Graph collision performance optimization & bug fixes."
            Version := 1.020
            Entries := array:
                news_entry_info{Type := news_entry_type.PerformanceOptimizations, Notes := "Reworked all weapons to utilize Scene Graph collision detection."}
                news_entry_info{Type := news_entry_type.None, Notes := "Measuring at least 10x more performant."}
                news_entry_info{Type := news_entry_type.PerformanceOptimizations, Notes := "Revive blast to utilize Scene Graph collision detection"}
                news_entry_info{Type := news_entry_type.PerformanceOptimizations, Notes := "Dash knockback to utilize Scene Graph collision detection"}
                news_entry_info{Type := news_entry_type.PerformanceOptimizations, Notes := "Dancing Dagger projectile collision prediction optimization."}
                news_entry_info{Type := news_entry_type.PerformanceOptimizations, Notes := "Optimized Shuriken, Boomerang & Staff Swirl at large size."}
                news_entry_info{Type := news_entry_type.PerformanceOptimizations, Notes := "Optimized Shuriken Mastery evolution."}

                news_entry_info{Type := news_entry_type.BugFixes, Notes := "Fixed - Memory leak with boomerang weapon."}
                news_entry_info{Type := news_entry_type.BugFixes, Notes := "Fixed - Beach balls passing through arena walls."}
                news_entry_info{Type := news_entry_type.BugFixes, Notes := "Fixed - Shuriken invisible when cooldown reaches 100% uptime threshold."}
                news_entry_info{Type := news_entry_type.BugFixes, Notes := "Fixed - Sentry/Staff Throw VFX sometimes showing at previous position."}
                news_entry_info{Type := news_entry_type.BugFixes, Notes := "Fixed - Staff Swirl cross cast was only hitting NPCs once (instead of per projectile)."}

        news_info:
            Date := "July 22, 2025"
            Title := "Update 41 - Game lobby & matchmaking reworks & NPC invulnerability bug fix."
            Version := 1.015
            Entries := array:
                news_entry_info{Type := news_entry_type.NewFeatures, Notes := "Added a countdown on game start to wait for any players still loading."}
                news_entry_info{Type := news_entry_type.None, Notes := "This countdown only applies on the first game when matchmaking is enabled."}
                news_entry_info{Type := news_entry_type.None, Notes := "Countdown is only skipped in full lobby (4 players)."}
                news_entry_info{Type := news_entry_type.NewFeatures, Notes := "Added matchmaking status indicator panel in the lobby."}
                news_entry_info{Type := news_entry_type.NewFeatures, Notes := "Replaced the player ready zone with switch to toggle ready/unready."}
                news_entry_info{Type := news_entry_type.None, Notes := "The ready sound now plays for all players."}

                news_entry_info{Type := news_entry_type.GameplayChanges, Notes := "Matchmaking is now permanently disabled after the first game."}
                news_entry_info{Type := news_entry_type.None, Notes := "Matchmaking disables when the game start countdown begins."}
                news_entry_info{Type := news_entry_type.None, Notes := "Matchmaking re-enables after 15s if countdown is cancelled (before first game)."}

                news_entry_info{Type := news_entry_type.BugFixes, Notes := "Fixed - NPC becomes invulnerable if target player is downed (again)."}

        news_info:
            Date := "July 17, 2025"
            Title := "Update 40 - Performance optimizations and bug fixes."
            Version := 1.014
            Entries := array:
                news_entry_info{Type := news_entry_type.NewFeatures, Notes := "Updated projectile VFX with spawn/despawn animations (bow/kunai/hammer return)."}

                news_entry_info{Type := news_entry_type.PerformanceOptimizations, Notes := "Implemented memory optimizations for ability & weapon stats."}
                news_entry_info{Type := news_entry_type.None, Notes := "Memory consumption is now a fixed amount throughout an entire run."}
                news_entry_info{Type := news_entry_type.None, Notes := "Should fix item shop freezing in long endless runs."}
                news_entry_info{Type := news_entry_type.PerformanceOptimizations, Notes := "Optimized projectiles with collision prediction (bow/kunai/hammer return)."}
                news_entry_info{Type := news_entry_type.None, Notes := "Collisions is now only calculated once on cast (instead of over time)."}
                news_entry_info{Type := news_entry_type.PerformanceOptimizations, Notes := "Updated experience orbs stop rendering inactive orbs (instead of hiding them)."}
                news_entry_info{Type := news_entry_type.PerformanceOptimizations, Notes := "Added LODs for slimes meshes."}
                news_entry_info{Type := news_entry_type.None, Notes := "Reduced poly count at lower graphics settings & for lower end devices."}
                news_entry_info{Type := news_entry_type.PerformanceOptimizations, Notes := "Reduced texture resolution (minimap, info panel, trees, treasure chest)."}
                news_entry_info{Type := news_entry_type.PerformanceOptimizations, Notes := "Reduced memory of volume activators (capture points)."}

                news_entry_info{Type := news_entry_type.GameplayChanges, Notes := "Increased projectile speed from 1000 > 1500 (bow/kunai/hammer return)."}
                
                news_entry_info{Type := news_entry_type.BugFixes, Notes := "Fixed - NPC becomes invulnerable if target player is downed."}
                news_entry_info{Type := news_entry_type.BugFixes, Notes := "Fixed - NPC spawn invulnerability not applying on the first frame."}
                news_entry_info{Type := news_entry_type.BugFixes, Notes := "Fixed - Damage numbers sometimes not showing for minibosses/boss slimes."}
                news_entry_info{Type := news_entry_type.BugFixes, Notes := "Fixed - Auto aim was targeting NPCs in eliminated animation."}
                news_entry_info{Type := news_entry_type.BugFixes, Notes := "Fixed - Trail weapon VFX invisible (fix not verified)."}
                news_entry_info{Type := news_entry_type.BugFixes, Notes := "Fixed - Hammer return VFX was grounded instead of at character height."}
                news_entry_info{Type := news_entry_type.BugFixes, Notes := "Fixed - Rank icon not updating immediately after ranking up."}
                news_entry_info{Type := news_entry_type.BugFixes, Notes := "Fixed - Slime eliminated animation sometimes finishes mid-air."}
                news_entry_info{Type := news_entry_type.BugFixes, Notes := "Fixed - Golden Tree gather number incorrect for numbers greater than 4 digits."}
                news_entry_info{Type := news_entry_type.BugFixes, Notes := "Fixed - Tree gather granting rewards multiple times when gathered by multiple players."}

        news_info:
            Date := "July 9, 2025"
            Title := "Update 39 - Rewards & progress saving."
            Version := 1.013
            Entries := array:
                news_entry_info{Type := news_entry_type.GameplayChanges, Notes := "Gold/treasure chest rewards are now granted at the end of every stage."}
                news_entry_info{Type := news_entry_type.GameplayChanges, Notes := "Endless mode progress is now saved at the end of every stage."}
                news_entry_info{Type := news_entry_type.None, Notes := "Added HUD message to indicate when saving occurs."}

        news_info:
            Date := "July 8, 2025"
            Title := "Update 38 - Fixed memory leaks, implemented experience orbs in Scene Graph."
            Version := 1.012
            Entries := array:
                news_entry_info{Type := news_entry_type.NewFeatures, Notes := "Experience orb pickup now follows the players position."}

                news_entry_info{Type := news_entry_type.PerformanceOptimizations, Notes := "Implemented experience orbs as scene graph component/entity:"}
                news_entry_info{Type := news_entry_type.None, Notes := "Using scene graph collision detection (>50x more performant)."}
                news_entry_info{Type := news_entry_type.None, Notes := "Experience orbs are now animated client sided in a single particle system."}

                news_entry_info{Type := news_entry_type.GameplayChanges, Notes := "Changed multi-capture point mission to no longer decay progress."}
                news_entry_info{Type := news_entry_type.GameplayChanges, Notes := "Changed arrowed tiles mission error to -1 progress instead of complete reset."}

                news_entry_info{Type := news_entry_type.BugFixes, Notes := "Memory leak with damage numbers when spawning new NPCs."}
                news_entry_info{Type := news_entry_type.BugFixes, Notes := "Memory leak with weapons each time weapon is reactivated."}
                news_entry_info{Type := news_entry_type.BugFixes, Notes := "NPCs were blocking projectiles while in eliminated animation."}

        news_info:
            Date := "June 30, 2025"
            Title := "Update 37 - Implemented weapons as Scene Graph entity/component (performance improvement)"
            Version := 1.010
            Entries := array:
                news_entry_info{Type := news_entry_type.NewFeatures, Notes := "Added NPC eliminated animation/state."}
                news_entry_info{Type := news_entry_type.NewFeatures, Notes := "Reworked the magnet animation as a VFX."}
                news_entry_info{Type := news_entry_type.None, Notes := "Magnet now pulls orbs towards the player's position."}
                news_entry_info{Type := news_entry_type.None, Notes := "Smoother client side animation & more performant on the server side."}
                news_entry_info{Type := news_entry_type.NewFeatures, Notes := "Added firing animation to Sentry VFX (animation pauses if no targets)."}
                news_entry_info{Type := news_entry_type.NewFeatures, Notes := "Added firing animation to Tesla Coil VFX."}
                news_entry_info{Type := news_entry_type.NewFeatures, Notes := "Added pulsing animation to Torrent VFX."}
                news_entry_info{Type := news_entry_type.NewFeatures, Notes := "Updated Shield attack VFX & cracked shield VFX (spiked shields)."}
                news_entry_info{Type := news_entry_type.NewFeatures, Notes := "Added supply drop opened audio."}
                
                news_entry_info{Type := news_entry_type.PerformanceOptimizations, Notes := "Reimplemented all weapons as Scene Graph entities & components."}
                news_entry_info{Type := news_entry_type.PerformanceOptimizations, Notes := "Implemented object pooling for all weapon VFX."}
                news_entry_info{Type := news_entry_type.PerformanceOptimizations, Notes := "Particle system per projectile reduced to a single particle system for various weapons."}
                news_entry_info{Type := news_entry_type.None, Notes := "Sword, Lightning, Shuriken, Energy Beam, Torrent, Staff & Shield."}
                news_entry_info{Type := news_entry_type.PerformanceOptimizations, Notes := "Combined projectile count into a single particle system for various weapons."}
                news_entry_info{Type := news_entry_type.None, Notes := "Bow, Kunai, Hammer & Energy Bomb."}
                news_entry_info{Type := news_entry_type.PerformanceOptimizations, Notes := "Kunai - spiral cast reduced to a single particle system."}
                news_entry_info{Type := news_entry_type.PerformanceOptimizations, Notes := "Trail - trail blast reduced to a single particle system."}
                news_entry_info{Type := news_entry_type.PerformanceOptimizations, Notes := "Sentry - combined impact area with sentry particle system."}
                news_entry_info{Type := news_entry_type.PerformanceOptimizations, Notes := "Implemented NPC damage number as entity/component with object pooling."}
                news_entry_info{Type := news_entry_type.PerformanceOptimizations, Notes := "Optimized shuriken damage calculations when at large size."}
                
                news_entry_info{Type := news_entry_type.GameplayChanges, Notes := "Shuriken weapon - added small delay before each projectile spawns."}
                news_entry_info{Type := news_entry_type.None, Notes := "Prevents shurikens from all apply damage at the same time at large size."}
                news_entry_info{Type := news_entry_type.GameplayChanges, Notes := "Sword weapon - precision strike is now always in player direction."}
                news_entry_info{Type := news_entry_type.GameplayChanges, Notes := "Torrent weapon - increased duration from 5s > 5.5s (4 > 5 hits)."}
                news_entry_info{Type := news_entry_type.GameplayChanges, Notes := "Torrent weapon - changed projectile split to torrent fire rate (increased damage rate)."}
                news_entry_info{Type := news_entry_type.GameplayChanges, Notes := "Sentry weapon - increased duration from 5s > 5.5s (5 > 6 hits)."}
                news_entry_info{Type := news_entry_type.GameplayChanges, Notes := "Removed NPC 0.5s damage invulnerability after spawning."}
                
                news_entry_info{Type := news_entry_type.BugFixes, Notes := "Fixed - Rectangle area attacks did not account for the size of the target."}
                news_entry_info{Type := news_entry_type.None, Notes := "Indirectly buffs to staff & energy beam weapon."}
                news_entry_info{Type := news_entry_type.BugFixes, Notes := "Fixed - Aura weapon - min cooldown damage multiplier was not applying."}
                news_entry_info{Type := news_entry_type.BugFixes, Notes := "Fixed - Staff throw actual damage size was not match VFX size (was 50% smaller)."}
                news_entry_info{Type := news_entry_type.BugFixes, Notes := "Fixed - Sword precision strike size reduction was applying to the base size."}
                news_entry_info{Type := news_entry_type.BugFixes, Notes := "Fixed - NPC attack cooldown not resetting after respawning (affects green slimes)."}
                news_entry_info{Type := news_entry_type.BugFixes, Notes := "Fixed - Player can get stuck under the ship (added barrier during ship entry)."}
                news_entry_info{Type := news_entry_type.BugFixes, Notes := "Fixed - Energy beam blast position was not centered."}
                news_entry_info{Type := news_entry_type.BugFixes, Notes := "Fixed - Supply drop activation circle remains showing if defeat during supply drop."}
                news_entry_info{Type := news_entry_type.BugFixes, Notes := "Fixed - Ninja slime headband was detached from the animation."}
                news_entry_info{Type := news_entry_type.BugFixes, Notes := "Fixed - Yellow wasteland miniboss eliminations not counting in NPC Codex."}

        news_info:
            Date := "June 19, 2025"
            Title := "Hotfix."
            Version := 1.003
            Entries := array:
                news_entry_info{Type := news_entry_type.BugFixes, Notes := "Fixed - Minibosses was failing to spawn (when trying to spawn the same type)."}

        news_info:
            Date := "June 12, 2025"
            Title := "Update 36 - NPC greetings, gear icons and various bug fixes."
            Version := 1.002
            Entries := array:
                news_entry_info{Type := news_entry_type.NewFeatures, Notes := "Added settings menu to toggle showing gear icon (replacing rank icon)."}
                news_entry_info{Type := news_entry_type.None, Notes := "Gear icon shows based on the highest gear set active."}
                news_entry_info{Type := news_entry_type.NewFeatures, Notes := "Added greeting voice & animation to NPCs in the lobby area."}

                news_entry_info{Type := news_entry_type.GameplayChanges, Notes := "Increase Weapon Reroll power up max level from 5 > 10."}

                news_entry_info{Type := news_entry_type.BugFixes, Notes := "Fixed - Daily Quest/Mission HUD overlapping with the minimap on mobile."}
                news_entry_info{Type := news_entry_type.None, Notes := "HUD is now position based on Input Control settings."}
                news_entry_info{Type := news_entry_type.BugFixes, Notes := "Fixed - Input Control icon/interaction text incorrect state."}
                news_entry_info{Type := news_entry_type.BugFixes, Notes := "Fixed - Ship appearing early/before the end of the stage (again)."}
                news_entry_info{Type := news_entry_type.BugFixes, Notes := "Fixed - Weapons could be activated multiple times and remain after game ended."}
                news_entry_info{Type := news_entry_type.BugFixes, Notes := "Fixed - Lobby music won't turn on if player had music off when joining the game."}
                news_entry_info{Type := news_entry_type.BugFixes, Notes := "Fixed - Some typo & grammar errors."}

        news_info:
            Date := "June 10, 2025"
            Title := "Update 35 - Player movement tuning, enabled bubble chat, small changes & bug fixes."
            Version := 1.001
            Entries := array:
                news_entry_info{Type := news_entry_type.NewFeatures, Notes := "Enabled bubble chat (text chat displays above your player)."}
                news_entry_info{Type := news_entry_type.NewFeatures, Notes := "Added mission failed HUD message."}
                news_entry_info{Type := news_entry_type.NewFeatures, Notes := "Added capture success audio for multi-capture mission."}
                news_entry_info{Type := news_entry_type.NewFeatures, Notes := "Added glow effect on hit for golden slime."}
                
                news_entry_info{Type := news_entry_type.PerformanceOptimizations, Notes := "Optimized projectiles to only check for targets in direction of travel."}
                
                news_entry_info{Type := news_entry_type.GameplayChanges, Notes := "Increased player movement acceleration (more responsive when changing directions)."}
                news_entry_info{Type := news_entry_type.GameplayChanges, Notes := "Increased player crawl speed when DBNO from 168cm/s > 300cm/s."}
                news_entry_info{Type := news_entry_type.GameplayChanges, Notes := "Multi-capture mission reduced capture time from 10s > 8s."}
                news_entry_info{Type := news_entry_type.GameplayChanges, Notes := "Multi-capture mission increase capture size by 20 > 26."}
                news_entry_info{Type := news_entry_type.GameplayChanges, Notes := "Supply drop experience orbs now always spawn even when limit is reached."}
                
                news_entry_info{Type := news_entry_type.BugFixes, Notes := "Fixed - Objective pulse not showing for beach ball & capture point mission."}
                news_entry_info{Type := news_entry_type.BugFixes, Notes := "Fixed - Ship appearing early/before the end of the stage."}
                news_entry_info{Type := news_entry_type.BugFixes, Notes := "Fixed - Camera collision causing floor to also be transparent."}
                news_entry_info{Type := news_entry_type.BugFixes, Notes := "Fixed - Supply drop not spawning (added fallback so activation circle should always spawn)."}

        news_info:
            Date := "June 6, 2025"
            Title := "Update 34 - New slime types, slime codex, variety of improvements & bug fixes."
            Version := 1.000
            Entries := array:
                news_entry_info{Type := news_entry_type.NewFeatures, Notes := "Added new Multi-Capture Point mission."}
                news_entry_info{Type := news_entry_type.None, Notes := "Multiple capture zones spawns randomly around the arena."}
                news_entry_info{Type := news_entry_type.NewFeatures, Notes := "Added new slime types (art only)."}
                news_entry_info{Type := news_entry_type.None, Notes := "Cactus Slimes - Scorching Wastelands."}
                news_entry_info{Type := news_entry_type.None, Notes := "Angel Slimes - The Underworld."}
                news_entry_info{Type := news_entry_type.None, Notes := "Atlantean Slimes - Reality Falls."}
                news_entry_info{Type := news_entry_type.None, Notes := "Mecha Slimes - Mega City."}
                news_entry_info{Type := news_entry_type.None, Notes := "Ninja Slimes - The Dojo."}
                news_entry_info{Type := news_entry_type.None, Notes := "Shogun Slimes - Bushido Dojo."}
                news_entry_info{Type := news_entry_type.None, Notes := "Frog Slimes - Flooded Frogs Temple."}
                news_entry_info{Type := news_entry_type.NewFeatures, Notes := "Added Slime Codex UI."}
                news_entry_info{Type := news_entry_type.None, Notes := "Records the slime types that the player has encountered."}
                news_entry_info{Type := news_entry_type.None, Notes := "Records the player's total eliminations for each slime type."}
                news_entry_info{Type := news_entry_type.NewFeatures, Notes := "Reworked all weapon transparency to now scale with weapon size."}
                news_entry_info{Type := news_entry_type.NewFeatures, Notes := "Updated Sword, Boomerang, Kunai, Hammer, Staff & Energy Beam VFX with transparency."}
                news_entry_info{Type := news_entry_type.NewFeatures, Notes := "Updated boss health bar & game event progress bar to show the specific boss icon."}
                news_entry_info{Type := news_entry_type.NewFeatures, Notes := "Updated boss health bar with shake animation when boss is damaged."}
                news_entry_info{Type := news_entry_type.NewFeatures, Notes := "Added boss glow effect on hit."}
                news_entry_info{Type := news_entry_type.NewFeatures, Notes := "Updated Arena Select UI & Arena Info Panel to show preview of boss & environment."}
                news_entry_info{Type := news_entry_type.NewFeatures, Notes := "Updated the lobby area."}
                news_entry_info{Type := news_entry_type.NewFeatures, Notes := "Boss fight cinematic sequence improvements (increased spawn height, camera shake)."}

                news_entry_info{Type := news_entry_type.PerformanceOptimizations, Notes := "Optimized slime materials (reduced draw calls from 2-4 > 1 per slime)."}

                news_entry_info{Type := news_entry_type.GameplayChanges, Notes := "Player weapons, player movement and slimes paused during boss fight cinematic."}
                news_entry_info{Type := news_entry_type.GameplayChanges, Notes := "Pickaxe is now disabled when player is in the arena."}
                news_entry_info{Type := news_entry_type.GameplayChanges, Notes := "Removed the Golden Statues mission."}
                news_entry_info{Type := news_entry_type.GameplayChanges, Notes := "Hammer weapon cast range no longer scales with size."}
                news_entry_info{Type := news_entry_type.GameplayChanges, Notes := "Hammer weapon base range increased from 300 > 500."}
                news_entry_info{Type := news_entry_type.GameplayChanges, Notes := "Dash cooldown reduced to 0.5s while in the lobby area."}

                news_entry_info{Type := news_entry_type.BugFixes, Notes := "Fixed - Boss fight cinematic sequence not synchronized with boss spawn time."}
                news_entry_info{Type := news_entry_type.BugFixes, Notes := "Fixed - Boss stage was not counting stage complete for daily quest."}
                news_entry_info{Type := news_entry_type.BugFixes, Notes := "Fixed - Snow Storm debuff was removed when movement speed is upgraded."}
                news_entry_info{Type := news_entry_type.BugFixes, Notes := "Fixed - Snow Storm mission was disabling dashing on mobile devices."}

        news_info:
            Date := "May 27, 2025"
            Title := "Update 33 - Boss fight intro cinematic, weapon reworks & bug fixes."
            Version := 0.921
            Entries := array:
                news_entry_info{Type := news_entry_type.NewFeatures, Notes := "Added boss fight intro cinematic sequence."}

                news_entry_info{Type := news_entry_type.GameplayChanges, Notes := "Snow Storm mission now applies 70% base movement speed (instead of -60% debuff)."}
                news_entry_info{Type := news_entry_type.GameplayChanges, Notes := "Hammer weapon reworks:"}
                news_entry_info{Type := news_entry_type.None, Notes := "Projectile count now casts in player direction with 30° angle offset."}
                news_entry_info{Type := news_entry_type.None, Notes := "Multicast has small position offset to improve projectile visibility."}
                news_entry_info{Type := news_entry_type.None, Notes := "Hammer Return evolution also increases the base cast range by +100%."}
                news_entry_info{Type := news_entry_type.None, Notes := "Hammer Anvil evolution now also doubles the projectile count."}
                news_entry_info{Type := news_entry_type.None, Notes := "Reduced min cooldown from 0.6s > 0.4s."}
                news_entry_info{Type := news_entry_type.GameplayChanges, Notes := "Staff weapon reworks:"}
                news_entry_info{Type := news_entry_type.None, Notes := "Reduced min cooldown from 0.5s > 0.3s."}
                news_entry_info{Type := news_entry_type.None, Notes := "Staff Swirl size increased from 50% > 75% of base size."}
                news_entry_info{Type := news_entry_type.GameplayChanges, Notes := "Boomerang weapon reworks:"}
                news_entry_info{Type := news_entry_type.None, Notes := "Renamed evolutions Projectile Count > Multicast."}
                news_entry_info{Type := news_entry_type.None, Notes := "Increase Projectile Range I evolution from +50% > +100%."}
                news_entry_info{Type := news_entry_type.None, Notes := "Removed Projectile Range II evolution."}
                news_entry_info{Type := news_entry_type.None, Notes := "Added Projectile Count evolution - fires in player direction with 30° angle offset."}
                news_entry_info{Type := news_entry_type.GameplayChanges, Notes := "Shield cast now pauses when max shield is reached."}
                news_entry_info{Type := news_entry_type.GameplayChanges, Notes := "Increased range & strength of ship's upward push (prevent getting stuck)."}
                
                news_entry_info{Type := news_entry_type.BugFixes, Notes := "Fixed - Boss fight barrage sometimes not stopping when boss is eliminated."}
                news_entry_info{Type := news_entry_type.BugFixes, Notes := "Fixed - Capture zones not reactivating when player auto revives."}
                news_entry_info{Type := news_entry_type.BugFixes, Notes := "Fixed - Aura weapon base supercharged duration restored (was accidentally removed)."}
                news_entry_info{Type := news_entry_type.BugFixes, Notes := "Fixed - Wood gathered number not showing when more than 4 digit numbers."}
                news_entry_info{Type := news_entry_type.BugFixes, Notes := "Fixed - Boomerang Growth evolution was not applying size correctly (only visually)."}
                news_entry_info{Type := news_entry_type.BugFixes, Notes := "Fixed - HUD blocking input on mobile devices."}

        news_info:
            Date := "May 23, 2025"
            Title := "Update 32 - New weapons, mission adjustments, quality of life improvements & bug fixes."
            Version := 0.920
            Entries := array:
                news_entry_info{Type := news_entry_type.NewFeatures, Notes := "Added new weapon - Tesla Coil"}
                news_entry_info{Type := news_entry_type.None, Notes := "Constructs a tesla coil in the player direction, damaging enemies in its range."}
                news_entry_info{Type := news_entry_type.NewFeatures, Notes := "Added new weapon - Hammer"}
                news_entry_info{Type := news_entry_type.None, Notes := "Drops a hammer in the player direction. Damaging enemies in an area on impact."}
                news_entry_info{Type := news_entry_type.NewFeatures, Notes := "Added new weapon - Staff"}
                news_entry_info{Type := news_entry_type.None, Notes := "A vertical slash that damages enemies in the player direction."}
                news_entry_info{Type := news_entry_type.None, Notes := "Evolutions adds different moves to the attack sequence."}
                news_entry_info{Type := news_entry_type.NewFeatures, Notes := "Added health number indicator for the Golden Statues mission."}
                news_entry_info{Type := news_entry_type.NewFeatures, Notes := "Added Health Pickup Size power up - increase healing aura range (multiplayer only)."}
                news_entry_info{Type := news_entry_type.NewFeatures, Notes := "Added HUD to show current number for the Numbered Tiles mission."}
                news_entry_info{Type := news_entry_type.NewFeatures, Notes := "Added Buy All button to the Item Shop UI."}
                news_entry_info{Type := news_entry_type.None, Notes := "Buy All button becomes Reroll when all items have been bought."}

                news_entry_info{Type := news_entry_type.GameplayChanges, Notes := "Changed Item Shop luck multiplier to now scale with stages (+25% luck per stage)."}
                news_entry_info{Type := news_entry_type.None, Notes := "Updated rates info UI to show the current Item Shop luck multiplier."}
                news_entry_info{Type := news_entry_type.GameplayChanges, Notes := "Player max shield now scales with max health."}
                news_entry_info{Type := news_entry_type.None, Notes := "Item Shop shield purchase now grants 50% amount of max health."}
                news_entry_info{Type := news_entry_type.GameplayChanges, Notes := "Numbered Tiles mission tile distance now scales with players (less players = closer)."}
                news_entry_info{Type := news_entry_type.GameplayChanges, Notes := "Increased Arrowed Tiles mission target from 8 > 12."}
                news_entry_info{Type := news_entry_type.GameplayChanges, Notes := "Increased the size of the capture area by 50% for the capture point mission."}
                news_entry_info{Type := news_entry_type.GameplayChanges, Notes := "Reduced the impact area VFX opacity (sentry, lightning, energy beam blast etc...)."}
                news_entry_info{Type := news_entry_type.GameplayChanges, Notes := "Changed Dancing Daggers VFX to be transparent."}
                
                news_entry_info{Type := news_entry_type.BugFixes, Notes := "Fixed - Sword weapon was ignoring damage stats from power ups & gears."}
                news_entry_info{Type := news_entry_type.BugFixes, Notes := "Fixed - Numbered Tiles mission sometimes not spawning all tiles."}
                news_entry_info{Type := news_entry_type.BugFixes, Notes := "Fixed - Golden Statue mission sometimes not spawning enough for the required target."}
                news_entry_info{Type := news_entry_type.BugFixes, Notes := "Fixed - Arrowed Tiles would be re-trigger when dashing while standing on the tile."}
                news_entry_info{Type := news_entry_type.BugFixes, Notes := "Fixed - Golden Slime damage was still scaling with the game speed."}
                news_entry_info{Type := news_entry_type.BugFixes, Notes := "Fixed - Cheering sound from Wendell mission keeps playing after the mission has ended."}
                news_entry_info{Type := news_entry_type.BugFixes, Notes := "Fixed - Experience orbs spawning in the air."}

        news_info:
            Date := "May 09, 2025"
            Title := "Update 31 - New missions & reworks, various quality of life updates & bug fixes."
            Version := 0.910
            Entries := array:
                news_entry_info{Type := news_entry_type.NewFeatures, Notes := "Added mission Arrowed Tiles - stand on arrow tiles in order show in HUD."}
                news_entry_info{Type := news_entry_type.NewFeatures, Notes := "Added mission Golden Statues - use pickaxe to destroy golden statues."}
                news_entry_info{Type := news_entry_type.NewFeatures, Notes := "Added mission Where's Wendell - find the NPC character hiding amongst the crowd."}
                news_entry_info{Type := news_entry_type.NewFeatures, Notes := "Added sound effect for the Snow Storm mission."}
                news_entry_info{Type := news_entry_type.NewFeatures, Notes := "Updated the Weapon HUD with weapon cooldown animations."}
                news_entry_info{Type := news_entry_type.NewFeatures, Notes := "Updated the Dash HUD position and style (to be consistent with weapon HUD)."}
                news_entry_info{Type := news_entry_type.NewFeatures, Notes := "Updated Ability Select UI to show indicator when cooldown limit is reached (+DMG)."}
                news_entry_info{Type := news_entry_type.NewFeatures, Notes := "Added team healing for Health Pickup:"}
                news_entry_info{Type := news_entry_type.None, Notes := "Player now gets 10s healing aura buff that heals nearby players for +10%/s max health."}
                news_entry_info{Type := news_entry_type.None, Notes := "Healing aura buff only applies in multiplayer."}
                news_entry_info{Type := news_entry_type.NewFeatures, Notes := "Added Health Pickup Duration power up - increase duration of healing aura buff."}
                news_entry_info{Type := news_entry_type.NewFeatures, Notes := "Added healed numbers when getting healed from Health Pickup."}
                news_entry_info{Type := news_entry_type.NewFeatures, Notes := "Added Accolade Info UI - list of all available XP accolades (in the lobby area)."}
                news_entry_info{Type := news_entry_type.NewFeatures, Notes := "Renamed XP Accolades to use the same name to prevent filling up the screen."}
                news_entry_info{Type := news_entry_type.NewFeatures, Notes := "Added blast area indicator for the Energy Beam weapon VFX."}
                news_entry_info{Type := news_entry_type.NewFeatures, Notes := "Added arena minimap icon to the Arena Select UI."}

                news_entry_info{Type := news_entry_type.GameplayChanges, Notes := "Reworked missions to no longer scale with game speed."}
                news_entry_info{Type := news_entry_type.None, Notes := "Game will now wait for mission to be completed before progressing to the next stage."}
                news_entry_info{Type := news_entry_type.GameplayChanges, Notes := "Reworked Beach Ball mission to have ball spawn at the center & goal positions randomly."}
                news_entry_info{Type := news_entry_type.GameplayChanges, Notes := "Reworked Numbered Tiles mission to be randomly positioned around the arena."}
                news_entry_info{Type := news_entry_type.None, Notes := "Error no longer causes the progress to be reset."}
                news_entry_info{Type := news_entry_type.GameplayChanges, Notes := "Small changes made to the Arena 5 (Ice Castle) environment."}
                news_entry_info{Type := news_entry_type.GameplayChanges, Notes := "Small changes made to the Arena 6 (Mega City) environment."}
                news_entry_info{Type := news_entry_type.GameplayChanges, Notes := "Health pickup no longer applies AoE healing on pickup."}
                
                news_entry_info{Type := news_entry_type.BugFixes, Notes := "Fixed - Weapon level not updating correctly when opening multiple treasure chests."}
                news_entry_info{Type := news_entry_type.BugFixes, Notes := "Fixed - Weapons & invulnerability not toggling for the last player entering the ship."}

        news_info:
            Date := "May 05, 2025"
            Title := "Hotfix."
            Version := 0.901
            Entries := array:
                news_entry_info{Type := news_entry_type.GameplayChanges, Notes := "If victory is achieved on an arena the player hasn't unlocked yet:"}
                news_entry_info{Type := news_entry_type.None, Notes := "Changed arena progress now increases by 1 instead of no progress."}
                news_entry_info{Type := news_entry_type.None, Notes := "Weapon progress will also be increased to match the player's arena progress."}

                news_entry_info{Type := news_entry_type.BugFixes, Notes := "Fixed - Capture Point mission sometimes not working."}
                news_entry_info{Type := news_entry_type.BugFixes, Notes := "Fixed - Capture Point mission capture time decreasing overtime on higher game speeds."}
                news_entry_info{Type := news_entry_type.BugFixes, Notes := "Fixed - Player Progress UI is always showing the gear of the local player."}
                news_entry_info{Type := news_entry_type.BugFixes, Notes := "Fixed - Stats Info UI showing boon values from previous game."}
                news_entry_info{Type := news_entry_type.BugFixes, Notes := "Fixed - Game Result UI game duration/survived time not showing the hour value."}
                news_entry_info{Type := news_entry_type.BugFixes, Notes := "Fixed - Rank up animation incorrectly showing when player is carried on arena 9."}

        news_info:
            Date := "May 02, 2025"
            Title := "Update 30 - Gear progress system, gold progression rework, various quality of life & bug fixes."
            Version := 0.900
            Entries := array:
                news_entry_info{Type := news_entry_type.NewFeatures, Notes := "Added Gear progression systems."}
                news_entry_info{Type := news_entry_type.None, Notes := "Eliminating minibosses now grant treasure chests after the game (for gears)."}
                news_entry_info{Type := news_entry_type.None, Notes := "Open treasure chests at the gear smith to obtain new gears & recipes."}
                news_entry_info{Type := news_entry_type.None, Notes := "Gears grant a variety of stats similar to power ups."}
                news_entry_info{Type := news_entry_type.None, Notes := "Upgrade, fuse & evolve gears at the gear smith to increase gear stats."}
                news_entry_info{Type := news_entry_type.NewFeatures, Notes := "Added Treasure Chest Multiplier power up."}
                news_entry_info{Type := news_entry_type.None, Notes := "Increase treasure chest earned after game (for gears)"}
                news_entry_info{Type := news_entry_type.NewFeatures, Notes := "Updated Power Up UI to show breakdown of Ninja Mastery stats."}
                news_entry_info{Type := news_entry_type.NewFeatures, Notes := "Updated the Stats Info UI - total stats to include stats from boons."}
                news_entry_info{Type := news_entry_type.NewFeatures, Notes := "Updated the Game Result UI - to include stats from boons."}
                
                news_entry_info{Type := news_entry_type.GameplayChanges, Notes := "Changed Game Speed x2 mode unlock level from 18 > 9."}
                news_entry_info{Type := news_entry_type.GameplayChanges, Notes := "Changed the end of stage transition mechanic:"}
                news_entry_info{Type := news_entry_type.None, Notes := "If at least 1 player is on the ship, the game will progress to the next stage (not defeat)."}
                news_entry_info{Type := news_entry_type.None, Notes := "Players that fail to get on the ship will lose all revives and be downed."}
                news_entry_info{Type := news_entry_type.GameplayChanges, Notes := "Reworked Gold income:"}
                news_entry_info{Type := news_entry_type.None, Notes := "Time Survived gold increased from 1/s > 3/s."}
                news_entry_info{Type := news_entry_type.None, Notes := "Weapon progress gold multiplier increased from 0.5% > 1.0% per weapon level."}
                news_entry_info{Type := news_entry_type.None, Notes := "Daily Quest gold multiplier increased from 30% > 150% per arena level."}
                news_entry_info{Type := news_entry_type.None, Notes := "Gold from wood reduced from 10% > 2.5% of wood collected."}
                
                news_entry_info{Type := news_entry_type.BugFixes, Notes := "Fixed - Boss elimination not registering if eliminated too quickly (added 3s invulnerability)."}
                news_entry_info{Type := news_entry_type.BugFixes, Notes := "Fixed - Game not pausing if player levels up while auto reviving."}
                news_entry_info{Type := news_entry_type.BugFixes, Notes := "Fixed - While the player is downed invulnerability is lost when game resumes."}
                news_entry_info{Type := news_entry_type.BugFixes, Notes := "Fixed - Chain Lightning focused bounce evolution was not working."}

        news_info:
            Date := "April 16, 2025"
            Title := "Update 29 - Performance optimization, minor changes & bug fixes."
            Version := 0.830
            Entries := array:
                news_entry_info{Type := news_entry_type.NewFeatures, Notes := "Updated Kunai weapon VFX to improve size indication."}

                news_entry_info{Type := news_entry_type.PerformanceOptimizations, Notes := "Optimized Weapon damage calculations (improved ~90% when weapon stats are maxed out)."}
                news_entry_info{Type := news_entry_type.PerformanceOptimizations, Notes := "Optimized Boons stat calculations."}
                news_entry_info{Type := news_entry_type.PerformanceOptimizations, Notes := "Optimized Bow/Kunai weapon (object pooling)."}
                news_entry_info{Type := news_entry_type.PerformanceOptimizations, Notes := "Optimized Sentry weapon (object pooling)."}
                news_entry_info{Type := news_entry_type.PerformanceOptimizations, Notes := "Optimized NPC Boss targeting calculations (check only bosses instead of all NPCs)."}
                news_entry_info{Type := news_entry_type.PerformanceOptimizations, Notes := "Optimized weapon/NPC attack hit detection (ignore Z axis checks)."}
                news_entry_info{Type := news_entry_type.PerformanceOptimizations, Notes := "Optimized experience orb merging (merge to a random & recent orb instead of nearest orb)."}

                news_entry_info{Type := news_entry_type.GameplayChanges, Notes := "Reduced endless mode gold reward scaling from ^2 > ^1.1 per boss eliminated."}
                news_entry_info{Type := news_entry_type.GameplayChanges, Notes := "Lightning weapon focus fire changed from prioritizing highest health to prioritizing bosses (optimization)."}
                news_entry_info{Type := news_entry_type.GameplayChanges, Notes := "Sentry weapon heavy duty removed increased duration & added increased fire rate (optimization)."}

                news_entry_info{Type := news_entry_type.BugFixes, Notes := "Fixed - Controllers getting stuck in Arena Select UI when timed out on game mode view."}
                news_entry_info{Type := news_entry_type.BugFixes, Notes := "Fixed - Tree gathered wood particle effects were not showing."}
                news_entry_info{Type := news_entry_type.BugFixes, Notes := "Fixed - NPC hit impact VFX size was too small."}

        news_info:
            Date := "April 11, 2025"
            Title := "Update 28 - Endless mode, more game settings, UI refresh & a new mission."
            Version := 0.820
            Entries := array:
                news_entry_info{Type := news_entry_type.NewFeatures, Notes := "Update Arena Select UI with the option to select additional game modes."}
                news_entry_info{Type := news_entry_type.NewFeatures, Notes := "Added Endless Mode option."}
                news_entry_info{Type := news_entry_type.None, Notes := "Infinite stages, enemy health/damage increases exponentially (^2) each stage."}
                news_entry_info{Type := news_entry_type.None, Notes := "Boss spawns every 3rd stage and grants bonus gold multiplier."}
                news_entry_info{Type := news_entry_type.None, Notes := "Missions reward legendary ability select after 10 boons is reached."}
                news_entry_info{Type := news_entry_type.NewFeatures, Notes := "Added Endless Mode progress to the player progress UI."}
                news_entry_info{Type := news_entry_type.NewFeatures, Notes := "Added Weapon Count option (to limit max weapons 1-4)."}
                news_entry_info{Type := news_entry_type.None, Notes := "Max banish is reduced by -2 for each weapon removed."}
                news_entry_info{Type := news_entry_type.NewFeatures, Notes := "Added Enemy Count option (45 or 90)."}
                news_entry_info{Type := news_entry_type.None, Notes := "Increases the amount of enemies that can spawn."}
                news_entry_info{Type := news_entry_type.NewFeatures, Notes := "Updated Orbit Camera to allow moving in all directions at normal speed with auto aim off."}
                news_entry_info{Type := news_entry_type.NewFeatures, Notes := "Updated the player direction arrow icon (changed shape & added shadow)."}
                news_entry_info{Type := news_entry_type.NewFeatures, Notes := "Added slow motion effect on boss eliminated & on defeat."}
                news_entry_info{Type := news_entry_type.NewFeatures, Notes := "General UI update (new panel style & rounded corners)."}
                news_entry_info{Type := news_entry_type.NewFeatures, Notes := "Updated stats info UI to fit & show all 10 boons."}
                news_entry_info{Type := news_entry_type.NewFeatures, Notes := "Added new Snow Storm mission - accept challenge that applies a movespeed debuff."}

                news_entry_info{Type := news_entry_type.PerformanceOptimizations, Notes := "Reduced texture resolution of minimap & beacon icons."}

                news_entry_info{Type := news_entry_type.GameplayChanges, Notes := "Reduced the gold multiplier from weapon level from +1% > +0.5% per level."}
                news_entry_info{Type := news_entry_type.GameplayChanges, Notes := "Reduced the boss warning & player under leveled message duration (-50%)."}
                news_entry_info{Type := news_entry_type.GameplayChanges, Notes := "Changed Sword with 2 angle evolution & precision strike to target front & back directions."}

                news_entry_info{Type := news_entry_type.BugFixes, Notes := "Fixed - Boons & Weapon Evolutions could be selected multiple times."}
                news_entry_info{Type := news_entry_type.BugFixes, Notes := "Fixed - Elimination effect was not showing for miniboss & bosses."}

        news_info:
            Date := "April 06, 2025"
            Title := "Update 27 - Bug fixes."
            Version := 0.811
            Entries := array:
                news_entry_info{Type := news_entry_type.NewFeatures, Notes := "Added weapon levels to the Weapon Select UI."}

                news_entry_info{Type := news_entry_type.BugFixes, Notes := "Fixed - Supply Drop & Stage Complete daily quests were showing the incorrect HUD."}
                news_entry_info{Type := news_entry_type.BugFixes, Notes := "Fixed - Capture Point mission capture speed was not scaling with game speed."}
                news_entry_info{Type := news_entry_type.BugFixes, Notes := "Fixed - Dancing Dagger weapon not deactivating."}
                news_entry_info{Type := news_entry_type.BugFixes, Notes := "Fixed - Weapons were reactivating too early while on the ship."}
                news_entry_info{Type := news_entry_type.BugFixes, Notes := "Fixed - Typo in daily quest description."}
