using { /Fortnite.com/AI }
using { /Fortnite.com/Characters }
using { /Fortnite.com/Devices }
using { /Fortnite.com/Game }
using { /Verse.org/Colors }
using { /Verse.org/Random }
using { /Verse.org/SceneGraph }
using { /Verse.org/Simulation }
using { /UnrealEngine.com/Temporary/SpatialMath }
using { NPCState }
using { Prefabs }

<# ----------------------------------------------------------------------------------------------------
    Constants
---------------------------------------------------------------------------------------------------- #>
NPCMovementSpeedBuffRate:float = 0.01
NPCSpawnPositionHeightOffset:float = 180.0
NPCDamagedVFXBufferTime:float = 0.1
NPCBaseMaxHealth:float = 30.0
MaxDamageNumber:float = 9999999999999999.0
NPCCollisionBaseSize:float = 50.0

<# ----------------------------------------------------------------------------------------------------
    Data Definitions
---------------------------------------------------------------------------------------------------- #>
npc_attack_properties := class<concrete>:
    @editable
    AttackType:npc_attack_type = npc_attack_type.Melee
    @editable
    AttackDistance:float = 150.0
    @editable
    AttackRange:float = 100.0
    @editable
    AttackDamage:float = 10.0
    @editable
    AttackCooldown:float = 1.0
    @editable
    AttackTime:float = 1.0
    @editable
    AttackProjectileRadius:?float = false
    @editable
    AttackProjectileSpeed:?float = false
    @editable
    AttackCount:?int = false
    @editable
    StunDuration:?float = false

MakeNPCAttackProperties<constructor>(UpdatedNPCAttackProperties:npc_attack_properties)<transacts> := npc_attack_properties:
    AttackType := UpdatedNPCAttackProperties.AttackType
    AttackDistance := UpdatedNPCAttackProperties.AttackDistance
    AttackRange := UpdatedNPCAttackProperties.AttackRange
    AttackDamage := UpdatedNPCAttackProperties.AttackDamage
    AttackCooldown := UpdatedNPCAttackProperties.AttackCooldown
    AttackTime := UpdatedNPCAttackProperties.AttackTime
    AttackProjectileRadius := UpdatedNPCAttackProperties.AttackProjectileRadius
    AttackProjectileSpeed := UpdatedNPCAttackProperties.AttackProjectileSpeed
    AttackCount := UpdatedNPCAttackProperties.AttackCount

ToString(NPCAttackProperties:npc_attack_properties)<transacts>:string =
    "{NPCAttackProperties.AttackType}\n" +
    "AttackDistance: {NPCAttackProperties.AttackDistance}\n" +
    "AttackRange: {NPCAttackProperties.AttackRange}\n" +
    "AttackDamage: {NPCAttackProperties.AttackDamage}\n" +
    "AttackCooldown: {NPCAttackProperties.AttackCooldown}\n" +
    "AttackTime: {NPCAttackProperties.AttackDistance}\n" +
    "AttackProjectileRadius: {"{NPCAttackProperties.AttackProjectileRadius?}" or "false"}\n" +
    "AttackProjectileSpeed: {"{NPCAttackProperties.AttackProjectileSpeed?}" or "false"}\n"
    "AttackCount: {"{NPCAttackProperties.AttackCount?}" or "false"}\n"

npc_behavior_message_module := module:
    SettingsCategory<public><localizes>:message = "Settings"
    BossSettingsCategory<public><localizes>:message = "Boss Settings"
    AttackSettingsCategory<public><localizes>:message = "Attack Settings"

knockback_properties := struct<concrete>:
    Rotation:rotation = IdentityRotation()  
    Duration:float = 1.0

<# ----------------------------------------------------------------------------------------------------
    Global Functions
---------------------------------------------------------------------------------------------------- #>
GetNPCMaxHealth(BaseHealth:float)<transacts>:float =
    CurrentStage := GetCurrentStage()
    ArenaStageCount := GetArenaStageCount()
    EndlessMultiplier := (CurrentStage > ArenaStageCount) and Pow((CurrentStage - ArenaStageCount) * 1.0, 2.0) * NPCMultiplierPerArenaStage or 0.0
    BaseHealth * (1.0 + CurrentStage * NPCMultiplierPerArenaStage + EndlessMultiplier) * (1.0 + GetArenaHealthMultiplier())

<#>
    X: Determines the position of the decimal point.
        1 => 1.23M
        2 => 32.1M
        3 => 321M

    Z:
        First 3 digits determines the VFX number
        4th digits sets the magnitude letter (K, M, B etc).

GetDamageNumberScale(DamageAmount:float)<transacts>:vector3 =
    if (DamageAmount >= 1000.0):
        Digits := Floor[Log(10.0, DamageAmount)] or 0
        vector3{X := ((Mod[Digits, 3] or 0) + 1) * 1.0, Y := 2.0, Z := ((Floor[DamageAmount / (Pow(10.0, Digits * 1.0 - 2.0))] or 0) * 10 + Floor[Digits * 1.0 / 3.0] or 0) * 1.0}
    else:
        vector3{X := 1.0, Y := 1.0, Z := DamageAmount}

<# ----------------------------------------------------------------------------------------------------
    npc_behavior_base
---------------------------------------------------------------------------------------------------- #>
npc_behavior_base<public> := class<unique>(npc_behavior):
    @editable:
        Categories := array{npc_behavior_message_module.SettingsCategory}
    Id:npc_id = npc_id.Melee

    @editable:
        Categories := array{npc_behavior_message_module.SettingsCategory}
    var AttackProperties:npc_attack_properties = npc_attack_properties{}

    @editable:
        Categories := array{npc_behavior_message_module.SettingsCategory}
    ItemDrop:?item_drop_id = false

    @editable:
        Categories := array{npc_behavior_message_module.SettingsCategory}
    ItemDropExperienceAmount:?int = false

    @editable:
        Categories := array{npc_behavior_message_module.SettingsCategory}
    Size:float = 25.0

    @editable:
        Categories := array{npc_behavior_message_module.SettingsCategory}
    IsBoss:logic = false

    @editable:
        Categories := array{npc_behavior_message_module.SettingsCategory}
    AccoladeId:accolade_count_id = accolade_count_id.EliminateSlimeMelee

    InterruptedEvent:event(npc_state) = event(npc_state){}
    EndEvent:event() = event(){}

    var State:npc_state = npc_state_chase{}
    var MaybeTargetPlayer:?player = false
    var MaybeCooldownStartTime:?float = false
    var BaseHealth:float = 100.0
    var SpawnTime:float = 0.0
    var LastHitTime:float = 0.0
    var PreviousDamageNumberAmount:float = 0.0
    var HealthBarDamageFactor:float = 1.0
    var KnockbackProperties:knockback_properties = knockback_properties{}
    var MaybeDamageNumber:?poolable_entity = false
    var MaybeCollision:?poolable_entity = false

    var MaxHealth:float = 100.0
    var CurrentHealth:float = 100.0
    var InvulnerabilityTime:float = 0.2
        
    <# ----------------------------------------------------------------------------------------------------
        Initialization
    ---------------------------------------------------------------------------------------------------- #>
    OnBegin<override>()<suspends>:void =
        set SpawnTime = GetSimulationElapsedTime()
        Sleep(0.0)
        if (Character := GetAgent[].GetFortCharacter[]):
            if (NPCWave := GetCurrentNPCWave()?):
                set BaseHealth = Character.GetMaxHealth()
                set MaxHealth = GetNPCMaxHealth(BaseHealth)
                set CurrentHealth = MaxHealth
                Self.ActivateMovementModulator(GetGameSpeedMovementMultiplier())
            else:
                Despawn()
                return

            if (DamageNumber := GetDamageNumber[]):
                DamageNumber.Activate()
                set MaybeDamageNumber = option{DamageNumber}

            if (Collision := GetCollisionEntity[], NPCCollision := Collision.Entity.GetComponent[npc_collision_component]):
                NPCCollision.Register(Self)
                Collision.Activate()
                set MaybeCollision = option{Collision}
                spawn{AttachCollision()}

            Self.RegisterNPC()
            UpdateTarget()

            if (TargetPlayer := MaybeTargetPlayer?):
                TeleportToTarget(Character, TargetPlayer)
                spawn{Start()}
            else:
                Despawn()

    OnEnd<override>():void =
        if (DamageNumber := MaybeDamageNumber?):
            DamageNumber.Deactivate()
            set MaybeDamageNumber = false

        if (Collision := MaybeCollision?, NPCCollision := Collision.Entity.GetComponent[npc_collision_component]):
            NPCCollision.Unregister()
            Collision.Deactivate()
            set MaybeCollision = false

        Self.UnregisterNPC(MaybeTargetPlayer)
        EndEvent.Signal()

    <# ----------------------------------------------------------------------------------------------------
        Get Functions
    ---------------------------------------------------------------------------------------------------- #>        
    GetSize()<transacts>:float =
        Size

    GetIsBoss()<transacts>:logic =
        IsBoss

    GetIsInterruptible()<transacts>:logic =
        logic{not IsBoss?}

    GetSpawnInvulnerabilityTime()<transacts>:float =
        InvulnerabilityTime

    IsStateInterruptible()<transacts>:logic =
        logic{State.Priority <= 0}

    IsDamageable()<transacts>:logic =
        logic:
            GetSimulationElapsedTime() - SpawnTime > GetSpawnInvulnerabilityTime()
            CurrentHealth > 0.0

    GetDamageAmount(Amount:float):float =
        Amount

    <# ----------------------------------------------------------------------------------------------------
        Update Functions
    ---------------------------------------------------------------------------------------------------- #>
    Start()<suspends>:void =
        if (Character := GetAgent[].GetFortCharacter[], DownButNotOut := GetDownButNotOut[]):
            race:
                EndEvent.Await()

                loop:
                    if (not Character.IsActive[]):
                        break

                    if (IsStateInterruptible()?):
                        RetargetPlayer()
                    
                    race:
                        set State = State.OnBegin(Self)

                        loop:
                            DownedAgent := DownButNotOut.AgentDownedEvent.Await()
                            if (DownedAgent = MaybeTargetPlayer?):
                                if (Player := player[DownedAgent], Player.GetPlayerResource(player_resource_id.Revives) > 0):
                                    Sleep(PlayerReviveDuration + 0.1)
                                    
                                if (IsStateInterruptible()?):
                                    UpdateTarget()
                                    break
                    
                    Sleep(0.0)

    AttachCollision()<suspends>:void =
        if (Character := GetAgent[].GetFortCharacter[], Collision := MaybeCollision?):
            race:
                EndEvent.Await()
                loop:
                    Collision.Entity.SetGlobalTransform of:
                        FromTransform of transform:
                            Translation := Character.GetTransform().Translation
                            Scale := UnitVector * GetSize() / NPCCollisionBaseSize

                    Sleep(TwoTicks)
        
    RetargetPlayer():void =
        if (TargetPlayer := MaybeTargetPlayer?, TargetPlayer.IsActive[]):
            if (TargetCharacter := TargetPlayer.GetFortCharacter[], TargetCharacter.IsDownButNotOut[] or (not IsGamePaused()? and not TargetCharacter.IsVulnerable[])):
                UpdateTarget()
        else:
            UpdateTarget()

    UpdateTarget(?MaybeNewTarget:?player = false):void =
        if (Agent := GetAgent[], NPCManager := GetNPCManager[], Player := MaybeNewTarget? or NPCManager.GetTarget()?):
            Agent.UpdateNPCTarget(MaybeTargetPlayer, Player)
            set MaybeTargetPlayer = option{Player}
            set State = GetNPCState(npc_state_type.Chase)
        else:
            set State = GetNPCState(npc_state_type.Idle)

    Knockback(InKnockbackProperties:knockback_properties):void =
        if (not GetIsInterruptible()?). return
        set KnockbackProperties = InKnockbackProperties
        InterruptState(npc_state_type.Knockback)

    ShowDamageFeedback():void =
        if (not GetIsInterruptible()?). return
        InterruptState(npc_state_type.Hit)

    InterruptState(Type:npc_state_type):void =
        NewState := GetNPCState(Type)
        if (NewState.Priority < State.Priority) then return
        set State = NewState
        InterruptedEvent.Signal of NewState

    UpdateDamageNumber(Position:vector3, DamageAmount:float, IsCriticalHit:logic, IsUpdate:logic)<suspends>:void =
        if:
            DamageNumberEntity := MaybeDamageNumber?.Entity
            ParticleSystem := VisualEffects.NS_DamageNumber[DamageNumberEntity.GetComponent[particle_system_component]]
        then:
            set ParticleSystem.IsIcon = IsCriticalHit
            set ParticleSystem.Color = IsCriticalHit? and ThemeColors.CriticalHit or NamedColors.White
            set ParticleSystem.Number = DamageAmount
            set ParticleSystem.Position = FromVector3 of Position

            set PreviousDamageNumberAmount = DamageAmount

            if (IsUpdate?):
                ParticleSystem.Disable()
                Sleep(0.0)
                ParticleSystem.Enable()

            ParticleSystem.Play()

    Damage(Player:player, AbilityDamageArgs:ability_damage_args, ?IsKnockback:logic = true):void =
        if (not IsDamageable()?):
            return
        
        if (Agent := GetAgent[], Character := Agent.GetFortCharacter[]):
            Position := Character.GetTransform().Translation.Make2D()

            if (GetSimulationElapsedTime() - LastHitTime >= NPCDamagedVFXBufferTime):
                DamageNumberPosition := Position + vector3{Z := (Size / NPCCollisionBaseSize * 80.0)}

                if (GetSimulationElapsedTime() - LastHitTime <= 1.0):
                    DamageAmount := Min(AbilityDamageArgs.Amount + PreviousDamageNumberAmount, MaxDamageNumber)
                    spawn{UpdateDamageNumber(DamageNumberPosition, DamageAmount, AbilityDamageArgs.IsCriticalHit, true)}
                else:
                    spawn{UpdateDamageNumber(DamageNumberPosition, AbilityDamageArgs.Amount, AbilityDamageArgs.IsCriticalHit, false)}

                set LastHitTime = GetSimulationElapsedTime()
            else:
                set PreviousDamageNumberAmount = Min(PreviousDamageNumberAmount + AbilityDamageArgs.Amount, MaxDamageNumber)

            set CurrentHealth -= AbilityDamageArgs.Amount
        
            if (CurrentHealth <= 0.0):
                AddEnemiesEliminated(Player, Id)
                Player.AwardAccolade(AccoladeId)
                Player.AwardAccolade(accolade_count_id.EliminateSlimes)
                HandleEliminated(Player, Position)
            else:
                if (IsKnockback?):
                    ShowDamageFeedback()

                HandleDamaged(Player, AbilityDamageArgs.Amount)

    Despawn():void =
        if (Character := GetAgent[].GetFortCharacter[], Character.IsActive[]):
            Character.Damage(Character.GetHealth())

    Respawn():void =
        if (Character := GetAgent[].GetFortCharacter[]):
            UpdateTarget()

            if (TargetPlayer := MaybeTargetPlayer?):
                TeleportToTarget(Character, TargetPlayer)

            if (Collision := MaybeCollision?, MeshComponent := Collision.Entity.GetComponent[mesh_component]):
                MeshComponent.Enable()

            Self.ActivateMovementModulator(GetGameSpeedMovementMultiplier())
            Character.ReleaseFromStasis()
            set CurrentHealth = MaxHealth
            set PreviousDamageNumberAmount = 0.0
            set SpawnTime = GetSimulationElapsedTime()
            set InvulnerabilityTime = GetRandomFloat(0.1, 0.5)

    TeleportToTarget(Character:fort_character, TargetPlayer:player):void =
        if (TargetCharacter := TargetPlayer.GetFortCharacter[]):
            var TeleportAttempts:int = 0

            loop:
                if:
                    SpawnPosition := GetSpawnPosition(TargetPlayer).Make2D()
                    SpawnRotation := MakeRotation(SpawnPosition, TargetCharacter.GetTransform().Translation.Make2D())
                    Character.TeleportTo[SpawnPosition + vector3{Z := NPCSpawnPositionHeightOffset}, SpawnRotation]
                then:
                    break

                set TeleportAttempts += 1
                if (TeleportAttempts >= 3):
                    if:
                        SpawnPosition := DefaultArenaOrigin
                        SpawnRotation := MakeRotation(SpawnPosition, TargetCharacter.GetTransform().Translation.Make2D())
                        Character.TeleportTo[SpawnPosition + vector3{Z := NPCSpawnPositionHeightOffset}, SpawnRotation]
                    
                    break

    <# ----------------------------------------------------------------------------------------------------
        Event handlers
    ---------------------------------------------------------------------------------------------------- #>
    HandleEliminated(EliminatingPlayer:player, Position:vector3):void =
        if (ExperienceAmount := ItemDropExperienceAmount?):
            SpawnExperience(Position, ExperienceAmount)

        <# [TODO]
        if:
            NPCManager := GetNPCManager[]
            NPCSpawned := NPCManager.NPCSpawned[Id]
            NPCSpawnTarget := NPCManager.NPCSpawnTarget[Id] or 0
        then:
            Print("HandleEliminated() Id = {GetNPCInfo[Id].Name or ""}, {NPCSpawned}/{NPCSpawnTarget}")
        #>

        if (NPCManager := GetNPCManager[], NPCManager.NPCSpawned[Id] > (NPCManager.NPCSpawnTarget[Id] or 0)):
            Despawn()
        else:
            if (Collision := MaybeCollision?, MeshComponent := Collision.Entity.GetComponent[mesh_component]):
                MeshComponent.Disable()

            InterruptState(npc_state_type.Eliminated)

    HandleDamaged(DamagingPlayer:player, DamageAmount:float):void = {}