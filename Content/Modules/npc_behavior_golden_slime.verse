using { /Fortnite.com/Characters }
using { /Verse.org/Random }
using { /Verse.org/Simulation }
using { /UnrealEngine.com/Temporary/SpatialMath }
using { NPCState }
using { Prefabs }

<# ----------------------------------------------------------------------------------------------------
    npc_behavior_golden_slime
---------------------------------------------------------------------------------------------------- #>
npc_behavior_golden_slime := class(npc_behavior_base):
    
    <# ----------------------------------------------------------------------------------------------------
        Initialization
    ---------------------------------------------------------------------------------------------------- #>
    OnBegin<override>()<suspends>:void =
        Sleep(0.0)
        if (Character := GetAgent[].GetFortCharacter[]):
            set BaseHealth = Character.GetMaxHealth()

            if (Character.TeleportTo[Character.GetTransform().Translation, Character.GetTransform().Rotation.ApplyYaw(PiFloat)]) {}

            if (DamageNumber := GetDamageNumber[]):
                DamageNumber.Activate()
                set MaybeDamageNumber = option{DamageNumber}

            if (Collision := GetCollisionEntity[], NPCCollision := Collision.Entity.GetComponent[npc_collision_component]):
                NPCCollision.Register(Self)
                Collision.Activate()
                set MaybeCollision = option{Collision}
                spawn{AttachCollision()}

            # [Hotfix] Health bar UI is not visible if max health exceeds ~45000.0
            set HealthBarDamageFactor = MinibossMaxHealthValue / GetNPCMaxHealth(Character.GetMaxHealth())
            Character.SetMaxHealth(MinibossMaxHealthValue)

            Self.RegisterNPC()
            Start()

    <# ----------------------------------------------------------------------------------------------------
        Update Functions
    ---------------------------------------------------------------------------------------------------- #>
    Start<override>()<suspends>:void =
        Sleep(1.0)
        if (Character := GetAgent[].GetFortCharacter[]):
            race:
                EndEvent.Await()
                loop:
                    if (not Character.IsActive[]):
                        break

                    race:
                        GetNPCState(npc_state_type.Flee).OnBegin(Self)
                        Sleep(GetRandomFloat(5.0, 10.0))

                    Sleep(GetRandomFloat(2.0, 3.0))

    Damage<override>(InstigatorAgent:agent, AbilityDamageArgs:ability_damage_args, ?IsKnockback:logic = true):void =
        if (not IsDamageable()?):
            return

        if (Character := GetAgent[].GetFortCharacter[]):
            DamageAmountOverride := Character.GetMaxHealth() * 0.01 / (GetActivePlayers().Length * 1.0)
            
            if (GetSimulationElapsedTime() - LastHitTime >= NPCDamagedVFXBufferTime):
                Position := Character.GetTransform().Translation + vector3{Z := (Size / 50.0 * 80.0) - CharacterHeightOffset}
                if (GetSimulationElapsedTime() - LastHitTime <= 1.0):
                    DamageAmount := Min(AbilityDamageArgs.Amount + PreviousDamageNumberAmount, MaxDamageNumber)
                    spawn{UpdateDamageNumber(Position, DamageAmount, AbilityDamageArgs.IsCriticalHit, true)}
                else:
                    spawn{UpdateDamageNumber(Position, AbilityDamageArgs.Amount, AbilityDamageArgs.IsCriticalHit, false)}

                set LastHitTime = GetSimulationElapsedTime()
            else:
                set PreviousDamageNumberAmount = Min(PreviousDamageNumberAmount + AbilityDamageArgs.Amount, MaxDamageNumber)

            Character.Damage(DamageAmountOverride)

    <# ----------------------------------------------------------------------------------------------------
        Event Handlers
    ---------------------------------------------------------------------------------------------------- #>
    HandleEliminated<override>(EliminatingPlayer:player, Position:vector3):void =
        Despawn()