using { /Fortnite.com/Characters }
using { /Fortnite.com/Game }
using { /Verse.org/Random }
using { /Verse.org/Simulation }
using { /UnrealEngine.com/Temporary/SpatialMath }
using { NPCState }

<# ----------------------------------------------------------------------------------------------------
    Constants
---------------------------------------------------------------------------------------------------- #>
MinibossMaxHealthValue:float = 20000.0

<# ----------------------------------------------------------------------------------------------------
    Global Functions
---------------------------------------------------------------------------------------------------- #>
GetMinibossEliminationProgress()<transacts>:int =
    (Floor[GetGameStat(game_stat_id.EliminatedMiniboss)] or 0) - (GetCurrentStage() - 1) * GetMinibossEliminationTarget()

GetMinibossEliminationTarget()<transacts>:int =
    MinibossEliminationTarget := GetArenaManager[].GetActiveArena[].MinibossMultiplier * 3 or 1
    (GetArenaManager[].GetActiveArena[].MinibossMultiplier * 3) or 1

IsMinibossEventCompleted()<transacts>:logic =
    logic{GetMinibossEliminationProgress() >= GetMinibossEliminationTarget()}

<# ----------------------------------------------------------------------------------------------------
    npc_behavior_miniboss
---------------------------------------------------------------------------------------------------- #>
npc_behavior_miniboss := class(npc_behavior_base):
    @editable:
        Categories := array{npc_behavior_message_module.BossSettingsCategory}
    BossAttacks:[]boss_attack = array{}

    @editable:
        Categories := array{npc_behavior_message_module.BossSettingsCategory}
    MinimapIconIndex:int = 0

    var PlayerDamageReceived:[player]float = map{}

    <# ----------------------------------------------------------------------------------------------------
        Get Functions
    ---------------------------------------------------------------------------------------------------- #>        
    GetRandomBossAttack()<decides><transacts>:boss_attack =
        AttackWeights := for (BossAttack : BossAttacks) {BossAttack.Weight}
        RandomIndex := GetWeightedRandom(AttackWeights)
        BossAttacks[RandomIndex]

    GetDamageAmount<override>(Amount:float):float =
        Amount * HealthBarDamageFactor

    <# ----------------------------------------------------------------------------------------------------
        Update Functions
    ---------------------------------------------------------------------------------------------------- #>       
    Start<override>()<suspends>:void =
        if (Agent := GetAgent[], Character := Agent.GetFortCharacter[], DownButNotOut := GetDownButNotOut[]):
            # [Hotfix] Health bar UI is not visible if max health exceeds ~45000.0
            set HealthBarDamageFactor = MinibossMaxHealthValue / GetNPCMaxHealth(BaseHealth)
            Character.SetMaxHealth(MinibossMaxHealthValue)

            race:
                EndEvent.Await()

                if (NPCManager := GetNPCManager[], MinibossMapIndicator := NPCManager.GetMinibossMapIndicator[]):
                    defer:
                        NPCManager.ReturnMinibossMapIndicator(MinibossMapIndicator)
                    
                    MinibossMapIndicator.CinematicSequence.Play()
                    MinibossMapIndicator.CinematicSequence.SetPlaybackTime(MinimapIconIndex * 1.0)
                    MinibossMapIndicator.CinematicSequence.Pause()

                    if (MinibossMapIndicator.MapIndicator.TeleportTo[Character.GetTransform()]):

                    loop:
                        if (Character.IsActive[]):
                            MinibossMapIndicator.MapIndicator.MoveTo(Character.GetTransform(), ThreeTicks)
                        else:
                            break

                loop:
                    if (not Character.IsActive[]):
                        break

                    if (IsStateInterruptible()?):
                        RetargetPlayer()

                    race:
                        loop:
                            DownedAgent := DownButNotOut.AgentDownedEvent.Await()
                            if (DownedAgent = MaybeTargetPlayer?):
                                if (Player := player[DownedAgent], Player.GetPlayerResource(player_resource_id.Revives) > 0):
                                    Sleep(PlayerReviveDuration + 0.1)
                                
                                UpdateTarget()
                                break

                        if (Player := MaybeTargetPlayer?, BossAttack := GetRandomBossAttack[]):
                            set AttackProperties = BossAttack.AttackProperties
                            set State = GetNPCState(npc_state_type.Chase)

                            loop:
                                set State = State.OnBegin(Self)
                                if (npc_state_attack[State]):
                                    break

                                Sleep(0.0)

                            for (Index := 1..BossAttack.CastCount):
                                GetNPCState(npc_state_type.Focus).OnBegin(Self)
                                GetNPCState(npc_state_type.Attack).OnBegin(Self)

                            GetNPCState(npc_state_type.Cooldown).OnBegin(Self)

                    Sleep(0.0)
            
    <# ----------------------------------------------------------------------------------------------------
        Event handlers
    ---------------------------------------------------------------------------------------------------- #>
    HandleDamaged<override>(DamagingPlayer:player, DamageAmount:float):void =
        if (Character := GetAgent[].GetFortCharacter[]):
            DamageArgs := damage_args:
                Instigator := option{DamagingPlayer.GetInstigator()}
                Amount := GetDamageAmount(DamageAmount)

            Character.Damage(DamageArgs)

            if (set PlayerDamageReceived[DamagingPlayer] = PlayerDamageReceived[DamagingPlayer] + DamageAmount or DamageAmount) {}
            DamagingPlayer.AddGameStat(player_game_stat_id.DamageDealtMiniboss, DamageAmount)

    HandleEliminated<override>(EliminatingPlayer:player, Position:vector3):void =
        if (NPCManager := GetNPCManager[], NPCManager.AbilityUpgradesDropped < GetCurrentMinibossEvent()):
            for (Player : GetActivePlayers(), GameProgressBarUI := GetGameManager[].PlayerGameProgressBarUI[Player]):
                GameProgressBarUI.UpdateEventState(NPCManager.AbilityUpgradesDropped * 2, game_event_state.Success)

            SpawnItemDrop(item_drop_id.TreasureChest, Position)

            set NPCManager.AbilityUpgradesDropped += 1
        else:
            SpawnExperience(GetRandomPositionInCircle(150.0, ?Origin := Position), 50, ?IsUseReserve := true)

            for (Index := 1..4):
                SpawnExperience(GetRandomPositionInCircle(150.0, ?Origin := Position), GetRandomInt(20, 40), ?IsUseReserve := true)
                        
        AddGameStat(game_stat_id.EliminatedMiniboss, 1.0)

        for (Player : GetActivePlayers()):
            Player.PlayAudio(audio_id.MinibossEliminated)

        for:
            Player -> DamageReceived : PlayerDamageReceived
            Player.HasBoon(boon_id.BossSiphon)?
            DamagePercent := DamageReceived / MaxHealth
            DamagePercent >= 0.1
        do:
            Player.AddGameStat(player_game_stat_id.EliminatedMiniboss, 1.0)
            Player.SetStat(stat_id.MaxHealth, stat_source.Boon2, GetBoonValueBossSiphon(Player))
            Player.UpdateMaxHealth()

        if (GameManager := GetGameManager[]):
            GameManager.MinibossEliminatedEvent.Signal()

        Despawn()