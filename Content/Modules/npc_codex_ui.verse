using { /Fortnite.com/UI }
using { /Verse.org/Assets }
using { /Verse.org/Colors }
using { /Verse.org/Simulation }
using { /UnrealEngine.com/Temporary/SpatialMath }
using { /UnrealEngine.com/Temporary/UI }

<# ----------------------------------------------------------------------------------------------------
    npc_codex_ui
---------------------------------------------------------------------------------------------------- #>
npc_codex_ui := class(panel_ui):
    Title<override>:string = "Slime Codex"
    BackgroundOpacity<override>:type {_X:float where 0.000000 <= _X, _X <= 1.000000} = 0.7
    Anchors<override>:anchors = anchors{Minimum := vector2{X := 0.1, Y := 0.0}, Maximum := vector2{X := 0.9, Y := 1.0}}

    ArenaText:text_block = text_block:
        DefaultTextColor := ThemeColors.PrimaryText
        DefaultShadowOffset := option{vector2{X := 2.0, Y := 2.0}}
        DefaultShadowOpacity := 0.5

    PreviewMaterial:material_block = material_block:
        DefaultImage := Materials.UI.MI_ArenaPreview_UI{}
        DefaultDesiredSize := vector2{X := 512.0, Y := 256.0}

    NameText:text_block = text_block:
        DefaultTextColor := ThemeColors.PrimaryText
        DefaultShadowOffset := option{vector2{X := 2.0, Y := 2.0}}
        DefaultShadowOpacity := 0.5

    AttackTypesText:text_block = text_block:
        DefaultTextColor := ThemeColors.PrimaryText
        DefaultShadowOffset := option{vector2{X := 2.0, Y := 2.0}}
        DefaultShadowOpacity := 0.5

    TotalText:text_block = text_block:
        DefaultTextColor := ThemeColors.Success
        DefaultShadowOffset := option{vector2{X := 2.0, Y := 2.0}}
        DefaultShadowColor := ThemeColors.Success * 0.125

    EliminationsText:text_block = text_block:
        DefaultTextColor := ThemeColors.Success
        DefaultShadowOffset := option{vector2{X := 2.0, Y := 2.0}}
        DefaultShadowColor := ThemeColors.Success * 0.125

    CancelAnimationEvent:event() = event(){}
    CancelSizeAnimationEvent:event() = event(){}

    GridRow:int = 3
    GridColumn:int = 3
    FadeDuration:float = 1.0
    SizeDuration:float = 0.2

    var SelectedIndex:int = 0
    var NPCCodexEntries:[]texture_block = array{}
    var MaybePaginationUI:?pagination_ui = false
    var PreviewMaterialInstance:Materials.UI.MI_ArenaPreview_UI = Materials.UI.MI_ArenaPreview_UI{}

    var Subscriptions:[]events_cancelable = array{}

    <# ----------------------------------------------------------------------------------------------------
        Initialization
    ---------------------------------------------------------------------------------------------------- #>
    Uninitialize<override>():void =
        for (Subscription : Subscriptions):
            Subscription.Cancel()

        set Subscriptions = array{}

    CreateContent<override>():widget =
        stack_box:
            Orientation := orientation.Horizontal
            Slots := array:
                stack_box_slot:
                    HorizontalAlignment := horizontal_alignment.Fill
                    VerticalAlignment := vertical_alignment.Fill
                    Distribution := option{1.0}
                    Widget := CreateMenuUI()
                stack_box_slot:
                    HorizontalAlignment := horizontal_alignment.Fill
                    VerticalAlignment := vertical_alignment.Fill
                    Distribution := option{1.0}
                    Padding := margin{Left := 20.0, Top := 0.0, Right := 0.0, Bottom := 0.0}
                    Widget := CreateInfoUI()

    CreateMenuUI():widget =
        PaginationUI := pagination_ui:
            Player := Player
            EntriesPerPage := 9
            TotalEntries := GetArenaNPCs.Length

        PaginationUI.Initialize()
        set MaybePaginationUI = option{PaginationUI}
        set Subscriptions += array:
            PaginationUI.PageChangedEvent.Subscribe(OnPageChanged)

        overlay:
            Slots := array:
                overlay_slot:
                    HorizontalAlignment := horizontal_alignment.Fill
                    VerticalAlignment := vertical_alignment.Fill
                    Widget := UserInterface.VerseWidget.WBP_RoundedBox{}
                overlay_slot:
                    HorizontalAlignment := horizontal_alignment.Fill
                    VerticalAlignment := vertical_alignment.Fill
                    Padding := margin{Left := 40.0, Top := 40.0, Right := 40.0, Bottom := 40.0}
                    Widget := stack_box:
                        Orientation := orientation.Vertical
                        Slots := array:
                            stack_box_slot:
                                HorizontalAlignment := horizontal_alignment.Fill
                                VerticalAlignment := vertical_alignment.Fill
                                Widget := overlay:
                                    Slots := array:
                                        overlay_slot:
                                            HorizontalAlignment := horizontal_alignment.Fill
                                            VerticalAlignment := vertical_alignment.Fill
                                            Widget := UserInterface.VerseWidget.WBP_RoundedBox{}
                                        overlay_slot:
                                            HorizontalAlignment := horizontal_alignment.Fill
                                            VerticalAlignment := vertical_alignment.Fill
                                            Padding := margin{Left := 40.0, Top := 20.0, Right := 40.0, Bottom := 20.0}
                                            Widget := stack_box:
                                                Orientation := orientation.Horizontal
                                                Slots := array:
                                                    stack_box_slot:
                                                        HorizontalAlignment := horizontal_alignment.Left
                                                        Widget := texture_block:
                                                            DefaultImage := UserInterface.Icons.SlimeCodex_64
                                                            DefaultDesiredSize := vector2{X := 40.0, Y := 40.0}
                                                    stack_box_slot:
                                                        HorizontalAlignment := horizontal_alignment.Fill
                                                        VerticalAlignment := vertical_alignment.Fill
                                                        Padding := margin{Left := 10.0, Top := 0.0, Right := 0.0, Bottom := 0.0}
                                                        Widget := text_block:
                                                            DefaultText := MakeMessage("Slimes Identified")
                                                            DefaultTextColor := ThemeColors.PrimaryText
                                                            DefaultShadowColor := NamedColors.Black
                                                            DefaultShadowOffset := option{vector2{X := 2.0, Y := 2.0}}
                                                            DefaultShadowOpacity := 0.5
                                                    stack_box_slot:
                                                        HorizontalAlignment := horizontal_alignment.Right
                                                        Distribution := option{1.0}
                                                        Widget := TotalText
                            stack_box_slot:
                                HorizontalAlignment := horizontal_alignment.Fill
                                VerticalAlignment := vertical_alignment.Fill
                                Padding := margin{Left := 0.0, Top := 20.0, Right := 0.0, Bottom := 0.0}
                                Distribution := option{1.0}
                                Widget := CreateGridUI()
                            stack_box_slot:
                                HorizontalAlignment := horizontal_alignment.Fill
                                VerticalAlignment := vertical_alignment.Bottom
                                Padding := margin{Left := 0.0, Top := 20.0, Right := 0.0, Bottom := 20.0}
                                Widget := color_block:
                                    DefaultColor := NamedColors.White
                                    DefaultOpacity := 0.1
                                    DefaultDesiredSize := vector2{X := 0.0, Y := 1.0}
                            stack_box_slot:
                                HorizontalAlignment := horizontal_alignment.Fill
                                VerticalAlignment := vertical_alignment.Fill
                                Widget := PaginationUI.GetUI()

    CreateGridUI():widget =
        stack_box:
            Orientation := orientation.Vertical
            Slots := for (RowIndex := 0..GridRow - 1):
                stack_box_slot:
                    HorizontalAlignment := horizontal_alignment.Fill
                    VerticalAlignment := vertical_alignment.Fill
                    Distribution := option{1.0}
                    Padding := margin{Left := 0.0, Top := (RowIndex = 0) and 0.0 or 20.0, Right := 0.0, Bottom := 0.0}
                    Widget := stack_box:
                        Orientation := orientation.Horizontal
                        Slots := for (ColumnIndex := 0..GridColumn - 1):
                            EntryIndex := GridColumn * RowIndex + ColumnIndex
                            
                            EntryIcon:texture_block = texture_block:
                                DefaultImage := UserInterface.Slimes.Slime_Forest_128
                                DefaultDesiredSize := vector2{X := 128.0, Y := 128.0}

                            Button := button_quiet{}

                            spawn{AwaitButtonClicked(Button.OnClick(), EntryIndex)}
                            set NPCCodexEntries += array{EntryIcon}

                            stack_box_slot:
                                HorizontalAlignment := horizontal_alignment.Fill
                                VerticalAlignment := vertical_alignment.Fill
                                Distribution := option{1.0}
                                Padding := margin{Left := (ColumnIndex = 0) and 0.0 or 20.0, Top := 0.0, Right := 0.0, Bottom := 0.0}
                                Widget := overlay:
                                    Slots := array:
                                        overlay_slot:
                                            Padding := margin{Left := 20.0, Top := 20.0, Right := 20.0, Bottom := 20.0}
                                            Widget := EntryIcon
                                        overlay_slot:
                                            HorizontalAlignment := horizontal_alignment.Fill
                                            VerticalAlignment := vertical_alignment.Fill
                                            Widget := Button

    CreateInfoUI():widget =
        set PreviewMaterialInstance = Materials.UI.MI_ArenaPreview_UI:
            BackgroundImage := UserInterface.Minimaps.T_ArenaBackground_01_512
            Icon := UserInterface.Slimes.Slime_Forest_Boss_128

        PreviewMaterial.SetImage(PreviewMaterialInstance)

        overlay:
            Slots := array:
                overlay_slot:
                    HorizontalAlignment := horizontal_alignment.Fill
                    VerticalAlignment := vertical_alignment.Fill
                    Widget := UserInterface.VerseWidget.WBP_RoundedBox{}
                overlay_slot:
                    HorizontalAlignment := horizontal_alignment.Fill
                    VerticalAlignment := vertical_alignment.Fill
                    Padding := margin{Left := 40.0, Top := 40.0, Right := 40.0, Bottom := 40.0}
                    Widget := stack_box:
                        Orientation := orientation.Vertical
                        Slots := array:
                            stack_box_slot:
                                HorizontalAlignment := horizontal_alignment.Center
                                VerticalAlignment := vertical_alignment.Fill
                                Widget := NameText
                            stack_box_slot:
                                HorizontalAlignment := horizontal_alignment.Fill
                                VerticalAlignment := vertical_alignment.Fill
                                Padding := margin{Left := 0.0, Top := 20.0, Right := 0.0, Bottom := 0.0}
                                Widget := PreviewMaterial
                            stack_box_slot:
                                HorizontalAlignment := horizontal_alignment.Fill
                                VerticalAlignment := vertical_alignment.Top
                                Padding := margin{Left := 0.0, Top := 40.0, Right := 0.0, Bottom := 20.0}
                                Widget := color_block:
                                    DefaultColor := NamedColors.White
                                    DefaultOpacity := 0.1
                                    DefaultDesiredSize := vector2{X := 0.0, Y := 1.0}
                            stack_box_slot:
                                HorizontalAlignment := horizontal_alignment.Left
                                VerticalAlignment := vertical_alignment.Top
                                Widget := stack_box:
                                    Orientation := orientation.Vertical
                                    Slots := array:
                                        stack_box_slot:
                                            HorizontalAlignment := horizontal_alignment.Fill
                                            VerticalAlignment := vertical_alignment.Fill
                                            Widget := text_block:
                                                DefaultTextSize := ThemeTextSize.Small
                                                DefaultText := MakeMessage("Arena")
                                                DefaultTextColor := ThemeColors.PrimaryText
                                                DefaultTextOpacity := 0.5
                                        stack_box_slot:
                                            HorizontalAlignment := horizontal_alignment.Fill
                                            VerticalAlignment := vertical_alignment.Fill
                                            Widget := ArenaText
                            stack_box_slot:
                                HorizontalAlignment := horizontal_alignment.Left
                                VerticalAlignment := vertical_alignment.Top
                                Padding := margin{Left := 0.0, Top := 20.0, Right := 0.0, Bottom := 0.0}
                                Widget := stack_box:
                                    Orientation := orientation.Vertical
                                    Slots := array:
                                        stack_box_slot:
                                            HorizontalAlignment := horizontal_alignment.Fill
                                            VerticalAlignment := vertical_alignment.Fill
                                            Widget := text_block:
                                                DefaultTextSize := ThemeTextSize.Small
                                                DefaultText := MakeMessage("Abilities")
                                                DefaultTextColor := ThemeColors.PrimaryText
                                                DefaultTextOpacity := 0.5
                                        stack_box_slot:
                                            HorizontalAlignment := horizontal_alignment.Fill
                                            VerticalAlignment := vertical_alignment.Fill
                                            Widget := AttackTypesText
                            
                            stack_box_slot:
                                HorizontalAlignment := horizontal_alignment.Fill
                                VerticalAlignment := vertical_alignment.Bottom
                                Distribution := option{1.0}
                                Widget := overlay:
                                    Slots := array:
                                        overlay_slot:
                                            HorizontalAlignment := horizontal_alignment.Fill
                                            VerticalAlignment := vertical_alignment.Fill
                                            Widget := UserInterface.VerseWidget.WBP_RoundedBox{}
                                        overlay_slot:
                                            HorizontalAlignment := horizontal_alignment.Fill
                                            VerticalAlignment := vertical_alignment.Fill
                                            Padding := margin{Left := 40.0, Top := 20.0, Right := 40.0, Bottom := 20.0}
                                            Widget := stack_box:
                                                Orientation := orientation.Horizontal
                                                Slots := array:
                                                    stack_box_slot:
                                                        HorizontalAlignment := horizontal_alignment.Left
                                                        Widget := texture_block:
                                                            DefaultImage := UserInterface.Icons.Elimination_64
                                                            DefaultDesiredSize := vector2{X := 40.0, Y := 40.0}
                                                    stack_box_slot:
                                                        HorizontalAlignment := horizontal_alignment.Fill
                                                        VerticalAlignment := vertical_alignment.Fill
                                                        Padding := margin{Left := 10.0, Top := 0.0, Right := 0.0, Bottom := 0.0}
                                                        Widget := text_block:
                                                            DefaultText := MakeMessage("Eliminations")
                                                            DefaultTextColor := ThemeColors.PrimaryText
                                                            DefaultShadowColor := NamedColors.Black
                                                            DefaultShadowOffset := option{vector2{X := 2.0, Y := 2.0}}
                                                            DefaultShadowOpacity := 0.5
                                                    stack_box_slot:
                                                        HorizontalAlignment := horizontal_alignment.Right
                                                        Distribution := option{1.0}
                                                        Widget := EliminationsText

    <# ----------------------------------------------------------------------------------------------------
        Get Functions
    ---------------------------------------------------------------------------------------------------- #>
    GetCurrentPage()<transacts>:int =
        MaybePaginationUI?.GetCurrentPage() or 1

    GetIconSize(Type:npc_type)<transacts>:float =
        case (Type):
            npc_type.Basic => 0.3
            npc_type.Elite => 0.4
            npc_type.General => 0.5
            npc_type.Miniboss => 0.6
            npc_type.Boss => 0.7

    <# ----------------------------------------------------------------------------------------------------
        Update Functions
    ---------------------------------------------------------------------------------------------------- #>
    Update(ArenaLevel:int):void =
        if (Arena := GetArenas()[ArenaLevel - 1], NPCIds := GetNPCsByArena[Arena]):
            ArenaText.SetText(MakeMessage(Arena.Name))

            for (Index -> NPCCodexEntry : NPCCodexEntries, NPCInfo := GetNPCInfo[NPCIds[Index]]):
                NPCCodexEntry.SetImage(NPCInfo.Icon)

                if (Index = SelectedIndex):
                    NPCCodexEntry.SetTint(ThemeColors.Disabled)
                    UpdateInfo(NPCInfo)
                else:
                    NPCCodexEntry.SetTint(NamedColors.White)

                if (Player.GetAchievement(NPCInfo.AchievementId) <= 0.0):
                    NPCCodexEntry.SetTint(NamedColors.Black)

    UpdateInfo(NPCInfo:npc_info):void =
        Eliminations := Player.GetAchievement(NPCInfo.AchievementId)

        NameText.SetText(MakeMessage((Eliminations > 0.0) and NPCInfo.Name or "???"))
        set PreviewMaterialInstance.BackgroundImage = GetArenaBackgroundImage(GetCurrentPage())
        set PreviewMaterialInstance.Icon = NPCInfo.Icon
        set PreviewMaterialInstance.IconTint = (Eliminations > 0.0) and NamedColors.White or NamedColors.Black
        CancelSizeAnimationEvent.Signal()
        spawn{SizeAnimation(GetIconSize(NPCInfo.Type))}
         
        var AttackTypesString:string = ""
        for (Index -> AttackType : NPCInfo.AttackTypes):
            set AttackTypesString += (Index = 0) and "{AttackType}" or ", {AttackType}"

        AttackTypesText.SetText(MakeMessage((Eliminations > 0.0) and AttackTypesString or "???"))
        EliminationsText.SetText(MakeMessage("{FormatAsCurrency(Eliminations)}"))

    FadeAnimation()<suspends>:void =
        DeltaTimer := delta_timer{}
        
        defer:
            PreviewMaterial.SetTint(NamedColors.White)

        race:
            block:
                CancelAnimationEvent.Await()

            loop:
                Progress := DeltaTimer.GetElapsedTime() / FadeDuration
                PreviewMaterial.SetTint(NamedColors.White * Progress)

                if (DeltaTimer.GetElapsedTime() >= FadeDuration):
                    break

                Sleep(0.0)

    SizeAnimation(TargetSize:float)<suspends>:void =
        DeltaTimer := delta_timer{}
        InitialSize := PreviewMaterialInstance.IconSize
        
        defer:
            set PreviewMaterialInstance.IconSize = TargetSize

        race:
            block:
                CancelSizeAnimationEvent.Await()

            loop:
                Progress := DeltaTimer.GetElapsedTime() / SizeDuration
                CurrentSize := Lerp(InitialSize, TargetSize, Progress)
                set PreviewMaterialInstance.IconSize = CurrentSize

                if (DeltaTimer.GetElapsedTime() >= SizeDuration):
                    break

                Sleep(0.0)

    <# ----------------------------------------------------------------------------------------------------
        Event Handlers
    ---------------------------------------------------------------------------------------------------- #>
    OnShow<override>():void =
        set SelectedIndex = 0

        var Total:int = 0
        for (ArenaNPC : GetArenaNPCs, NPCInfo := GetNPCInfo[ArenaNPC], Player.GetAchievement(NPCInfo.AchievementId) > 0.0):
            set Total += 1

        TotalText.SetText(MakeMessage("{Total}/{GetArenaNPCs.Length}"))

        if (PaginationUI := MaybePaginationUI?):
            PaginationUI.SetPage(1)


    AwaitButtonClicked(Listenable:listenable(widget_message), Index:int)<suspends>:void =
        if (PlayerRemovedEvent := Player.GetPlayerRemovedEvent[]):
            race:
                PlayerRemovedEvent.Await()
                loop:
                    Listenable.Await()
                    set SelectedIndex = Index
                    Update(GetCurrentPage())
                    Player.PlayAudio(audio_id.Activate)

    OnPageChanged():void =
        set SelectedIndex = 0
        Update(GetCurrentPage())
        CancelAnimationEvent.Signal()
        spawn{FadeAnimation()}