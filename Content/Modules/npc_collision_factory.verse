using { /Fortnite.com/Characters }
using { /Verse.org/SceneGraph }
using { /Verse.org/Simulation }
using { /UnrealEngine.com/Temporary/SpatialMath }
using { /Verse.org/Concurrency }
using { Prefabs }

<# ----------------------------------------------------------------------------------------------------
    Global Functions
---------------------------------------------------------------------------------------------------- #>
NPCCollisionInteraction(Channel:collision_channel)<computes>:collision_interaction =
    
    <#
    if (CollisionChannels.avatar[Channel]):
        collision_interaction.Overlap
    else:
        collision_interaction.Ignore
    #>

    collision_interaction.Ignore

<# ----------------------------------------------------------------------------------------------------
    npc_collision_factory
---------------------------------------------------------------------------------------------------- #>
npc_collision_factory := class:
    var NPCCollisions:[]poolable_entity = array{}

    Initialize(SimulationEntity:entity, Count:int):void =
        TargetCount := Min(Count + 10, 90) - NPCCollisions.Length
        if (TargetCount = 0) then return

        NPCCollisionEntities := for (Index := 1..TargetCount):
            NPCCollisionEntity := prefab_npc_collision{}
            NPCCollision := poolable_entity{Entity := NPCCollisionEntity}
            set NPCCollisions += array{NPCCollision}
            NPCCollisionEntity

        SimulationEntity.AddEntities(NPCCollisionEntities)

    GetAvailable()<transacts>:?poolable_entity =
        var MaybeNPCCollision:?poolable_entity = false
        for (NPCCollision : NPCCollisions, not MaybeNPCCollision?, NPCCollision.IsAvailable?):
            set MaybeNPCCollision = option{NPCCollision}
        MaybeNPCCollision

<# ----------------------------------------------------------------------------------------------------
    npc_collision_component
---------------------------------------------------------------------------------------------------- #>
npc_collision_component := class<final_super>(component):
    var MaybeNPC:?npc_behavior_base = false
    
    Register(InNPC:npc_behavior_base):void =
        set MaybeNPC = option{InNPC}
        if (MeshComponent := Entity.GetComponent[mesh_component]):
            MeshComponent.Enable()

    Unregister():void =
        set MaybeNPC = false
        if (MeshComponent := Entity.GetComponent[mesh_component]):
            MeshComponent.Disable()