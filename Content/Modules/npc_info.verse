using { /Verse.org/Assets }

<# ----------------------------------------------------------------------------------------------------
    Data Definitions
---------------------------------------------------------------------------------------------------- #>
npc_id<public> := enum:
    Melee
    MeleeElite
    Projectile
    Area
    Charge
    Dash
    MinibossJump
    Boss
    SnowMelee
    SnowMeleeElite
    SnowProjectile
    SnowArea
    SnowCharge
    SnowDash
    SnowMinibossJump
    SnowBoss
    Golden
    MinibossProjectile
    MinibossArea
    MinibossDash
    MinibossCharge
    MeleeGeneral
    SnowMeleeGeneral
    SnowMinibossProjectile
    SnowMinibossArea
    SnowMinibossDash
    SnowMinibossCharge

    AtlantisMelee
    AtlantisMeleeElite
    AtlantisMeleeGeneral
    AtlantisMinibossJump
    AtlantisMinibossProjectile
    AtlantisMinibossArea
    AtlantisMinibossDash
    AtlantisMinibossCharge
    AtlantisBoss

    ElysiumMelee
    ElysiumMeleeElite
    ElysiumMeleeGeneral
    ElysiumMinibossJump
    ElysiumMinibossProjectile
    ElysiumMinibossArea
    ElysiumMinibossDash
    ElysiumMinibossCharge
    ElysiumBoss

    WastelandMelee
    WastelandMeleeElite
    WastelandMeleeGeneral
    WastelandMinibossJump
    WastelandMinibossProjectile
    WastelandMinibossArea
    WastelandMinibossDash
    WastelandMinibossCharge
    WastelandBoss

    CyberMelee
    CyberMeleeElite
    CyberMeleeGeneral
    CyberMinibossJump
    CyberMinibossProjectile
    CyberMinibossArea
    CyberMinibossDash
    CyberMinibossCharge
    CyberBoss
    
    NinjaMelee
    NinjaMeleeElite
    NinjaMeleeGeneral
    NinjaMinibossJump
    NinjaMinibossProjectile
    NinjaMinibossArea
    NinjaMinibossDash
    NinjaMinibossCharge
    NinjaBoss

    ShogunMelee
    ShogunMeleeElite
    ShogunMeleeGeneral
    ShogunMinibossJump
    ShogunMinibossProjectile
    ShogunMinibossArea
    ShogunMinibossDash
    ShogunMinibossCharge
    ShogunBoss

    FrogMelee
    FrogMeleeElite
    FrogMeleeGeneral
    FrogMinibossJump
    FrogMinibossProjectile
    FrogMinibossArea
    FrogMinibossDash
    FrogMinibossCharge
    FrogBoss

npc_type<public> := enum:
    Basic
    Elite
    General
    Miniboss
    Boss

npc_info := struct<computes>:
    Id:npc_id
    Name:string
    Icon:texture
    IconSmall:texture
    Type:npc_type = npc_type.Basic
    AttackTypes:[]npc_attack_type = array{}
    AchievementId:achievement_id = achievement_id.EliminationSlimeForest

NPCInfos:[npc_id]npc_info = map:
    npc_id.Melee => npc_info:
        Id := npc_id.Melee
        Name := "Forest Slime"
        Icon := UserInterface.Slimes.Slime_Forest_128
        IconSmall := UserInterface.Slimes.Slime_Forest_64
        AttackTypes := array{npc_attack_type.Melee}
        AchievementId := achievement_id.EliminationSlimeForest
    npc_id.MeleeElite => npc_info:
        Id := npc_id.MeleeElite
        Name := "Elite Forest Slime"
        Icon := UserInterface.Slimes.Slime_Forest_Elite_128
        IconSmall := UserInterface.Slimes.Slime_Forest_Elite_64
        Type := npc_type.Elite
        AttackTypes := array{npc_attack_type.Melee}
        AchievementId := achievement_id.EliminationSlimeForestElite
    npc_id.MeleeGeneral => npc_info:
        Id := npc_id.MeleeGeneral
        Name := "General Forest Slime"
        Icon := UserInterface.Slimes.Slime_Forest_General_128
        IconSmall := UserInterface.Slimes.Slime_Forest_General_64
        Type := npc_type.General
        AttackTypes := array{npc_attack_type.Melee}
        AchievementId := achievement_id.EliminationSlimeForestGeneral
    npc_id.MinibossJump => npc_info:
        Id := npc_id.MinibossJump
        Name := "Red Forest Slime"
        Icon := UserInterface.Slimes.Slime_Forest_Miniboss_Jump_128
        IconSmall := UserInterface.Slimes.Slime_Forest_Miniboss_Jump_64
        Type := npc_type.Miniboss
        AttackTypes := array{npc_attack_type.Jump}
        AchievementId := achievement_id.EliminationSlimeForestMinibossRed
    npc_id.MinibossProjectile => npc_info:
        Id := npc_id.MinibossProjectile
        Name := "Blue Forest Slime"
        Icon := UserInterface.Slimes.Slime_Forest_Miniboss_Projectile_128
        IconSmall := UserInterface.Slimes.Slime_Forest_Miniboss_Projectile_64
        Type := npc_type.Miniboss
        AttackTypes := array{npc_attack_type.Projectile}
        AchievementId := achievement_id.EliminationSlimeForestMinibossBlue
    npc_id.MinibossArea => npc_info:
        Id := npc_id.MinibossArea
        Name := "Purple Forest Slime"
        Icon := UserInterface.Slimes.Slime_Forest_Miniboss_Area_128
        IconSmall := UserInterface.Slimes.Slime_Forest_Miniboss_Area_64
        Type := npc_type.Miniboss
        AttackTypes := array{npc_attack_type.Barrage}
        AchievementId := achievement_id.EliminationSlimeForestMinibossPurple
    npc_id.MinibossDash => npc_info:
        Id := npc_id.MinibossDash
        Name := "Orange Forest Slime"
        Icon := UserInterface.Slimes.Slime_Forest_Miniboss_Dash_128
        IconSmall := UserInterface.Slimes.Slime_Forest_Miniboss_Dash_64
        Type := npc_type.Miniboss
        AttackTypes := array{npc_attack_type.Dash}
        AchievementId := achievement_id.EliminationSlimeForestMinibossOrange
    npc_id.MinibossCharge => npc_info:
        Id := npc_id.MinibossCharge
        Name := "Yellow Forest Slime"
        Icon := UserInterface.Slimes.Slime_Forest_Miniboss_Charge_128
        IconSmall := UserInterface.Slimes.Slime_Forest_Miniboss_Charge_64
        Type := npc_type.Miniboss
        AttackTypes := array{npc_attack_type.Charge}
        AchievementId := achievement_id.EliminationSlimeForestMinibossYellow
    npc_id.Boss => npc_info:
        Id := npc_id.Boss
        Name := "Boss Forest Slime"
        Icon := UserInterface.Slimes.Slime_Forest_Boss_128
        IconSmall := UserInterface.Slimes.Slime_Forest_Boss_64
        Type := npc_type.Boss
        AttackTypes := array{npc_attack_type.Jump, npc_attack_type.ForwardJump, npc_attack_type.Dash, npc_attack_type.Projectile}
        AchievementId := achievement_id.EliminationSlimeForestBoss

    npc_id.SnowMelee => npc_info:
        Id := npc_id.SnowMelee
        Name := "Snow Slime"
        Icon := UserInterface.Slimes.Slime_Snow_128
        IconSmall := UserInterface.Slimes.Slime_Snow_64
        AttackTypes := array{npc_attack_type.Melee}
        AchievementId := achievement_id.EliminationSlimeSnow
    npc_id.SnowMeleeElite => npc_info:
        Id := npc_id.SnowMeleeElite
        Name := "Elite Snow Slime"
        Icon := UserInterface.Slimes.Slime_Snow_Elite_128
        IconSmall := UserInterface.Slimes.Slime_Snow_Elite_64
        Type := npc_type.Elite
        AttackTypes := array{npc_attack_type.Melee}
        AchievementId := achievement_id.EliminationSlimeSnowElite
    npc_id.SnowMeleeGeneral => npc_info:
        Id := npc_id.SnowMeleeGeneral
        Name := "General Snow Slime"
        Icon := UserInterface.Slimes.Slime_Snow_General_128
        IconSmall := UserInterface.Slimes.Slime_Snow_General_64
        Type := npc_type.General
        AttackTypes := array{npc_attack_type.Melee}
        AchievementId := achievement_id.EliminationSlimeSnowGeneral
    npc_id.SnowMinibossProjectile => npc_info:
        Id := npc_id.SnowMinibossProjectile
        Name := "Red Snow Slime"
        Icon := UserInterface.Slimes.Slime_Snow_Miniboss_Jump_128
        IconSmall := UserInterface.Slimes.Slime_Snow_Miniboss_Jump_64
        Type := npc_type.Miniboss
        AttackTypes := array{npc_attack_type.Jump}
        AchievementId := achievement_id.EliminationSlimeSnowMinibossRed
    npc_id.SnowMinibossArea => npc_info:
        Id := npc_id.SnowMinibossArea
        Name := "Blue Snow Slime"
        Icon := UserInterface.Slimes.Slime_Snow_Miniboss_Projectile_128
        IconSmall := UserInterface.Slimes.Slime_Snow_Miniboss_Projectile_64
        Type := npc_type.Miniboss
        AttackTypes := array{npc_attack_type.Projectile}
        AchievementId := achievement_id.EliminationSlimeSnowMinibossBlue
    npc_id.SnowMinibossDash => npc_info:
        Id := npc_id.SnowMinibossDash
        Name := "Purple Snow Slime"
        Icon := UserInterface.Slimes.Slime_Snow_Miniboss_Area_128
        IconSmall := UserInterface.Slimes.Slime_Snow_Miniboss_Area_64
        Type := npc_type.Miniboss
        AttackTypes := array{npc_attack_type.Barrage}
        AchievementId := achievement_id.EliminationSlimeSnowMinibossPurple
    npc_id.SnowMinibossCharge => npc_info:
        Id := npc_id.SnowMinibossCharge
        Name := "Orange Snow Slime"
        Icon := UserInterface.Slimes.Slime_Snow_Miniboss_Dash_128
        IconSmall := UserInterface.Slimes.Slime_Snow_Miniboss_Dash_64
        Type := npc_type.Miniboss
        AttackTypes := array{npc_attack_type.Dash}
        AchievementId := achievement_id.EliminationSlimeSnowMinibossOrange
    npc_id.SnowMinibossJump => npc_info:
        Id := npc_id.SnowMinibossJump
        Name := "Yellow Snow Slime"
        Icon := UserInterface.Slimes.Slime_Snow_Miniboss_Charge_128
        IconSmall := UserInterface.Slimes.Slime_Snow_Miniboss_Charge_64
        Type := npc_type.Miniboss
        AttackTypes := array{npc_attack_type.Charge}
        AchievementId := achievement_id.EliminationSlimeSnowMinibossYellow
    npc_id.SnowBoss => npc_info:
        Id := npc_id.SnowBoss
        Name := "Boss Snow Slime"
        Icon := UserInterface.Slimes.Slime_Snow_Boss_128
        IconSmall := UserInterface.Slimes.Slime_Snow_Boss_64
        Type := npc_type.Boss
        AttackTypes := array{npc_attack_type.Jump, npc_attack_type.ForwardJump, npc_attack_type.Dash, npc_attack_type.Projectile}
        AchievementId := achievement_id.EliminationSlimeSnowBoss
    
    npc_id.AtlantisMelee => npc_info:
        Id := npc_id.AtlantisMelee
        Name := "Atlantis Slime"
        Icon := UserInterface.Slimes.Slime_Atlantis_128
        IconSmall := UserInterface.Slimes.Slime_Atlantis_64
        AttackTypes := array{npc_attack_type.Melee}
        AchievementId := achievement_id.EliminationSlimeAtlantis
    npc_id.AtlantisMeleeElite => npc_info:
        Id := npc_id.AtlantisMeleeElite
        Name := "Elite Atlantis Slime"
        Icon := UserInterface.Slimes.Slime_Atlantis_Elite_128
        IconSmall := UserInterface.Slimes.Slime_Atlantis_Elite_64
        Type := npc_type.Elite
        AttackTypes := array{npc_attack_type.Melee}
        AchievementId := achievement_id.EliminationSlimeAtlantisElite
    npc_id.AtlantisMeleeGeneral => npc_info:
        Id := npc_id.AtlantisMeleeGeneral
        Name := "General Atlantis Slime"
        Icon := UserInterface.Slimes.Slime_Atlantis_General_128
        IconSmall := UserInterface.Slimes.Slime_Atlantis_General_64
        Type := npc_type.General
        AttackTypes := array{npc_attack_type.Melee}
        AchievementId := achievement_id.EliminationSlimeAtlantisGeneral
    npc_id.AtlantisMinibossJump => npc_info:
        Id := npc_id.AtlantisMinibossJump
        Name := "Red Atlantis Slime"
        Icon := UserInterface.Slimes.Slime_Atlantis_Miniboss_Jump_128
        IconSmall := UserInterface.Slimes.Slime_Atlantis_Miniboss_Jump_64
        Type := npc_type.Miniboss
        AttackTypes := array{npc_attack_type.Jump}
        AchievementId := achievement_id.EliminationSlimeAtlantisMinibossRed
    npc_id.AtlantisMinibossProjectile => npc_info:
        Id := npc_id.AtlantisMinibossProjectile
        Name := "Blue Atlantis Slime"
        Icon := UserInterface.Slimes.Slime_Atlantis_Miniboss_Projectile_128
        IconSmall := UserInterface.Slimes.Slime_Atlantis_Miniboss_Projectile_64
        Type := npc_type.Miniboss
        AttackTypes := array{npc_attack_type.Projectile}
        AchievementId := achievement_id.EliminationSlimeAtlantisMinibossBlue
    npc_id.AtlantisMinibossArea => npc_info:
        Id := npc_id.AtlantisMinibossArea
        Name := "Purple Atlantis Slime"
        Icon := UserInterface.Slimes.Slime_Atlantis_Miniboss_Area_128
        IconSmall := UserInterface.Slimes.Slime_Atlantis_Miniboss_Area_64
        Type := npc_type.Miniboss
        AttackTypes := array{npc_attack_type.Barrage}
        AchievementId := achievement_id.EliminationSlimeAtlantisMinibossPurple
    npc_id.AtlantisMinibossDash => npc_info:
        Id := npc_id.AtlantisMinibossDash
        Name := "Orange Atlantis Slime"
        Icon := UserInterface.Slimes.Slime_Atlantis_Miniboss_Dash_128
        IconSmall := UserInterface.Slimes.Slime_Atlantis_Miniboss_Dash_64
        Type := npc_type.Miniboss
        AttackTypes := array{npc_attack_type.Dash}
        AchievementId := achievement_id.EliminationSlimeAtlantisMinibossOrange
    npc_id.AtlantisMinibossCharge => npc_info:
        Id := npc_id.AtlantisMinibossCharge
        Name := "Yellow Atlantis Slime"
        Icon := UserInterface.Slimes.Slime_Atlantis_Miniboss_Charge_128
        IconSmall := UserInterface.Slimes.Slime_Atlantis_Miniboss_Charge_64
        Type := npc_type.Miniboss
        AttackTypes := array{npc_attack_type.Charge}
        AchievementId := achievement_id.EliminationSlimeAtlantisMinibossYellow
    npc_id.AtlantisBoss => npc_info:
        Id := npc_id.AtlantisBoss
        Name := "Boss Atlantis Slime"
        Icon := UserInterface.Slimes.Slime_Atlantis_Boss_128
        IconSmall := UserInterface.Slimes.Slime_Atlantis_Boss_64
        Type := npc_type.Boss
        AttackTypes := array{npc_attack_type.Jump, npc_attack_type.ForwardJump, npc_attack_type.Dash, npc_attack_type.Projectile}
        AchievementId := achievement_id.EliminationSlimeAtlantisBoss

    npc_id.ElysiumMelee => npc_info:
        Id := npc_id.ElysiumMelee
        Name := "Angel Slime"
        Icon := UserInterface.Slimes.Slime_Elysium_128
        IconSmall := UserInterface.Slimes.Slime_Elysium_64
        AttackTypes := array{npc_attack_type.Melee}
        AchievementId := achievement_id.EliminationSlimeElysium
    npc_id.ElysiumMeleeElite => npc_info:
        Id := npc_id.ElysiumMeleeElite
        Name := "Elite Angel Slime"
        Icon := UserInterface.Slimes.Slime_Elysium_Elite_128
        IconSmall := UserInterface.Slimes.Slime_Elysium_Elite_64
        Type := npc_type.Elite
        AttackTypes := array{npc_attack_type.Melee}
        AchievementId := achievement_id.EliminationSlimeElysiumElite
    npc_id.ElysiumMeleeGeneral => npc_info:
        Id := npc_id.ElysiumMeleeGeneral
        Name := "General Angel Slime"
        Icon := UserInterface.Slimes.Slime_Elysium_General_128
        IconSmall := UserInterface.Slimes.Slime_Elysium_General_64
        Type := npc_type.General
        AttackTypes := array{npc_attack_type.Melee}
        AchievementId := achievement_id.EliminationSlimeElysiumGeneral
    npc_id.ElysiumMinibossJump => npc_info:
        Id := npc_id.ElysiumMinibossJump
        Name := "Red Angel Slime"
        Icon := UserInterface.Slimes.Slime_Elysium_Miniboss_Jump_128
        IconSmall := UserInterface.Slimes.Slime_Elysium_Miniboss_Jump_64
        Type := npc_type.Miniboss
        AttackTypes := array{npc_attack_type.Jump}
        AchievementId := achievement_id.EliminationSlimeElysiumMinibossRed
    npc_id.ElysiumMinibossProjectile => npc_info:
        Id := npc_id.ElysiumMinibossProjectile
        Name := "Blue Angel Slime"
        Icon := UserInterface.Slimes.Slime_Elysium_Miniboss_Projectile_128
        IconSmall := UserInterface.Slimes.Slime_Elysium_Miniboss_Projectile_64
        Type := npc_type.Miniboss
        AttackTypes := array{npc_attack_type.Projectile}
        AchievementId := achievement_id.EliminationSlimeElysiumMinibossBlue
    npc_id.ElysiumMinibossArea => npc_info:
        Id := npc_id.ElysiumMinibossArea
        Name := "Purple Angel Slime"
        Icon := UserInterface.Slimes.Slime_Elysium_Miniboss_Area_128
        IconSmall := UserInterface.Slimes.Slime_Elysium_Miniboss_Area_64
        Type := npc_type.Miniboss
        AttackTypes := array{npc_attack_type.Barrage}
        AchievementId := achievement_id.EliminationSlimeElysiumMinibossPurple
    npc_id.ElysiumMinibossDash => npc_info:
        Id := npc_id.ElysiumMinibossDash
        Name := "Orange Angel Slime"
        Icon := UserInterface.Slimes.Slime_Elysium_Miniboss_Dash_128
        IconSmall := UserInterface.Slimes.Slime_Elysium_Miniboss_Dash_64
        Type := npc_type.Miniboss
        AttackTypes := array{npc_attack_type.Dash}
        AchievementId := achievement_id.EliminationSlimeElysiumMinibossOrange
    npc_id.ElysiumMinibossCharge => npc_info:
        Id := npc_id.ElysiumMinibossCharge
        Name := "Yellow Angel Slime"
        Icon := UserInterface.Slimes.Slime_Elysium_Miniboss_Charge_128
        IconSmall := UserInterface.Slimes.Slime_Elysium_Miniboss_Charge_64
        Type := npc_type.Miniboss
        AttackTypes := array{npc_attack_type.Charge}
        AchievementId := achievement_id.EliminationSlimeElysiumMinibossYellow
    npc_id.ElysiumBoss => npc_info:
        Id := npc_id.ElysiumBoss
        Name := "Boss Angel Slime"
        Icon := UserInterface.Slimes.Slime_Elysium_Boss_128
        IconSmall := UserInterface.Slimes.Slime_Elysium_Boss_64
        Type := npc_type.Boss
        AttackTypes := array{npc_attack_type.Jump, npc_attack_type.ForwardJump, npc_attack_type.Dash, npc_attack_type.Projectile}
        AchievementId := achievement_id.EliminationSlimeElysiumBoss

    npc_id.WastelandMelee => npc_info:
        Id := npc_id.WastelandMelee
        Name := "Cactus Slime"
        Icon := UserInterface.Slimes.Slime_Wasteland_128
        IconSmall := UserInterface.Slimes.Slime_Wasteland_64
        AttackTypes := array{npc_attack_type.Melee}
        AchievementId := achievement_id.EliminationSlimeWasteland
    npc_id.WastelandMeleeElite => npc_info:
        Id := npc_id.WastelandMeleeElite
        Name := "Elite Cactus Slime"
        Icon := UserInterface.Slimes.Slime_Wasteland_Elite_128
        IconSmall := UserInterface.Slimes.Slime_Wasteland_Elite_64
        Type := npc_type.Elite
        AttackTypes := array{npc_attack_type.Melee}
        AchievementId := achievement_id.EliminationSlimeWastelandElite
    npc_id.WastelandMeleeGeneral => npc_info:
        Id := npc_id.WastelandMeleeGeneral
        Name := "General Cactus Slime"
        Icon := UserInterface.Slimes.Slime_Wasteland_General_128
        IconSmall := UserInterface.Slimes.Slime_Wasteland_General_64
        Type := npc_type.General
        AttackTypes := array{npc_attack_type.Melee}
        AchievementId := achievement_id.EliminationSlimeWastelandGeneral
    npc_id.WastelandMinibossJump => npc_info:
        Id := npc_id.WastelandMinibossJump
        Name := "Red Cactus Slime"
        Icon := UserInterface.Slimes.Slime_Wasteland_Miniboss_Jump_128
        IconSmall := UserInterface.Slimes.Slime_Wasteland_Miniboss_Jump_64
        Type := npc_type.Miniboss
        AttackTypes := array{npc_attack_type.Jump}
        AchievementId := achievement_id.EliminationSlimeWastelandMinibossRed
    npc_id.WastelandMinibossProjectile => npc_info:
        Id := npc_id.WastelandMinibossProjectile
        Name := "Blue Cactus Slime"
        Icon := UserInterface.Slimes.Slime_Wasteland_Miniboss_Projectile_128
        IconSmall := UserInterface.Slimes.Slime_Wasteland_Miniboss_Projectile_64
        Type := npc_type.Miniboss
        AttackTypes := array{npc_attack_type.Projectile}
        AchievementId := achievement_id.EliminationSlimeWastelandMinibossBlue
    npc_id.WastelandMinibossArea => npc_info:
        Id := npc_id.WastelandMinibossArea
        Name := "Purple Cactus Slime"
        Icon := UserInterface.Slimes.Slime_Wasteland_Miniboss_Area_128
        IconSmall := UserInterface.Slimes.Slime_Wasteland_Miniboss_Area_64
        Type := npc_type.Miniboss
        AttackTypes := array{npc_attack_type.Barrage}
        AchievementId := achievement_id.EliminationSlimeWastelandMinibossPurple
    npc_id.WastelandMinibossDash => npc_info:
        Id := npc_id.WastelandMinibossDash
        Name := "Orange Cactus Slime"
        Icon := UserInterface.Slimes.Slime_Wasteland_Miniboss_Dash_128
        IconSmall := UserInterface.Slimes.Slime_Wasteland_Miniboss_Dash_64
        Type := npc_type.Miniboss
        AttackTypes := array{npc_attack_type.Dash}
        AchievementId := achievement_id.EliminationSlimeWastelandMinibossOrange
    npc_id.WastelandMinibossCharge => npc_info:
        Id := npc_id.WastelandMinibossCharge
        Name := "Yellow Cactus Slime"
        Icon := UserInterface.Slimes.Slime_Wasteland_Miniboss_Charge_128
        IconSmall := UserInterface.Slimes.Slime_Wasteland_Miniboss_Charge_64
        Type := npc_type.Miniboss
        AttackTypes := array{npc_attack_type.Charge}
        AchievementId := achievement_id.EliminationSlimeWastelandMinibossYellow
    npc_id.WastelandBoss => npc_info:
        Id := npc_id.WastelandBoss
        Name := "Boss Cactus Slime"
        Icon := UserInterface.Slimes.Slime_Wasteland_Boss_128
        IconSmall := UserInterface.Slimes.Slime_Wasteland_Boss_64
        Type := npc_type.Boss
        AttackTypes := array{npc_attack_type.Jump, npc_attack_type.ForwardJump, npc_attack_type.Dash, npc_attack_type.Projectile}
        AchievementId := achievement_id.EliminationSlimeWastelandBoss

    npc_id.CyberMelee => npc_info:
        Id := npc_id.CyberMelee
        Name := "Mecha Slime"
        Icon := UserInterface.Slimes.Slime_Cyber_128
        IconSmall := UserInterface.Slimes.Slime_Cyber_64
        AttackTypes := array{npc_attack_type.Melee}
        AchievementId := achievement_id.EliminationSlimeCyber
    npc_id.CyberMeleeElite => npc_info:
        Id := npc_id.CyberMeleeElite
        Name := "Elite Mecha Slime"
        Icon := UserInterface.Slimes.Slime_Cyber_Elite_128
        IconSmall := UserInterface.Slimes.Slime_Cyber_Elite_64
        Type := npc_type.Elite
        AttackTypes := array{npc_attack_type.Melee}
        AchievementId := achievement_id.EliminationSlimeCyberElite
    npc_id.CyberMeleeGeneral => npc_info:
        Id := npc_id.CyberMeleeGeneral
        Name := "General Mecha Slime"
        Icon := UserInterface.Slimes.Slime_Cyber_General_128
        IconSmall := UserInterface.Slimes.Slime_Cyber_General_64
        Type := npc_type.General
        AttackTypes := array{npc_attack_type.Melee}
        AchievementId := achievement_id.EliminationSlimeCyberGeneral
    npc_id.CyberMinibossJump => npc_info:
        Id := npc_id.CyberMinibossJump
        Name := "Red Mecha Slime"
        Icon := UserInterface.Slimes.Slime_Cyber_Miniboss_Jump_128
        IconSmall := UserInterface.Slimes.Slime_Cyber_Miniboss_Jump_64
        Type := npc_type.Miniboss
        AttackTypes := array{npc_attack_type.Jump}
        AchievementId := achievement_id.EliminationSlimeCyberMinibossRed
    npc_id.CyberMinibossProjectile => npc_info:
        Id := npc_id.CyberMinibossProjectile
        Name := "Blue Mecha Slime"
        Icon := UserInterface.Slimes.Slime_Cyber_Miniboss_Projectile_128
        IconSmall := UserInterface.Slimes.Slime_Cyber_Miniboss_Projectile_64
        Type := npc_type.Miniboss
        AttackTypes := array{npc_attack_type.Projectile}
        AchievementId := achievement_id.EliminationSlimeCyberMinibossBlue
    npc_id.CyberMinibossArea => npc_info:
        Id := npc_id.CyberMinibossArea
        Name := "Purple Mecha Slime"
        Icon := UserInterface.Slimes.Slime_Cyber_Miniboss_Area_128
        IconSmall := UserInterface.Slimes.Slime_Cyber_Miniboss_Area_64
        Type := npc_type.Miniboss
        AttackTypes := array{npc_attack_type.Barrage}
        AchievementId := achievement_id.EliminationSlimeCyberMinibossPurple
    npc_id.CyberMinibossDash => npc_info:
        Id := npc_id.CyberMinibossDash
        Name := "Orange Mecha Slime"
        Icon := UserInterface.Slimes.Slime_Cyber_Miniboss_Dash_128
        IconSmall := UserInterface.Slimes.Slime_Cyber_Miniboss_Dash_64
        Type := npc_type.Miniboss
        AttackTypes := array{npc_attack_type.Dash}
        AchievementId := achievement_id.EliminationSlimeCyberMinibossOrange
    npc_id.CyberMinibossCharge => npc_info:
        Id := npc_id.CyberMinibossCharge
        Name := "Yellow Mecha Slime"
        Icon := UserInterface.Slimes.Slime_Cyber_Miniboss_Charge_128
        IconSmall := UserInterface.Slimes.Slime_Cyber_Miniboss_Charge_64
        Type := npc_type.Miniboss
        AttackTypes := array{npc_attack_type.Charge}
        AchievementId := achievement_id.EliminationSlimeCyberMinibossYellow
    npc_id.CyberBoss => npc_info:
        Id := npc_id.CyberBoss
        Name := "Boss Mecha Slime"
        Icon := UserInterface.Slimes.Slime_Cyber_Boss_128
        IconSmall := UserInterface.Slimes.Slime_Cyber_Boss_64
        Type := npc_type.Boss
        AttackTypes := array{npc_attack_type.Jump, npc_attack_type.ForwardJump, npc_attack_type.Dash, npc_attack_type.Projectile}
        AchievementId := achievement_id.EliminationSlimeCyberBoss

    npc_id.NinjaMelee => npc_info:
        Id := npc_id.NinjaMelee
        Name := "Ninja Slime"
        Icon := UserInterface.Slimes.Slime_Ninja_128
        IconSmall := UserInterface.Slimes.Slime_Ninja_64
        AttackTypes := array{npc_attack_type.Melee}
        AchievementId := achievement_id.EliminationSlimeNinja
    npc_id.NinjaMeleeElite => npc_info:
        Id := npc_id.NinjaMeleeElite
        Name := "Elite Ninja Slime"
        Icon := UserInterface.Slimes.Slime_Ninja_Elite_128
        IconSmall := UserInterface.Slimes.Slime_Ninja_Elite_64
        Type := npc_type.Elite
        AttackTypes := array{npc_attack_type.Melee}
        AchievementId := achievement_id.EliminationSlimeNinjaElite
    npc_id.NinjaMeleeGeneral => npc_info:
        Id := npc_id.NinjaMeleeGeneral
        Name := "General Ninja Slime"
        Icon := UserInterface.Slimes.Slime_Ninja_General_128
        IconSmall := UserInterface.Slimes.Slime_Ninja_General_64
        Type := npc_type.General
        AttackTypes := array{npc_attack_type.Melee}
        AchievementId := achievement_id.EliminationSlimeNinjaGeneral
    npc_id.NinjaMinibossJump => npc_info:
        Id := npc_id.NinjaMinibossJump
        Name := "Red Ninja Slime"
        Icon := UserInterface.Slimes.Slime_Ninja_Miniboss_Jump_128
        IconSmall := UserInterface.Slimes.Slime_Ninja_Miniboss_Jump_64
        Type := npc_type.Miniboss
        AttackTypes := array{npc_attack_type.Jump}
        AchievementId := achievement_id.EliminationSlimeNinjaMinibossRed
    npc_id.NinjaMinibossProjectile => npc_info:
        Id := npc_id.NinjaMinibossProjectile
        Name := "Blue Ninja Slime"
        Icon := UserInterface.Slimes.Slime_Ninja_Miniboss_Projectile_128
        IconSmall := UserInterface.Slimes.Slime_Ninja_Miniboss_Projectile_64
        Type := npc_type.Miniboss
        AttackTypes := array{npc_attack_type.Projectile}
        AchievementId := achievement_id.EliminationSlimeNinjaMinibossBlue
    npc_id.NinjaMinibossArea => npc_info:
        Id := npc_id.NinjaMinibossArea
        Name := "Purple Ninja Slime"
        Icon := UserInterface.Slimes.Slime_Ninja_Miniboss_Area_128
        IconSmall := UserInterface.Slimes.Slime_Ninja_Miniboss_Area_64
        Type := npc_type.Miniboss
        AttackTypes := array{npc_attack_type.Barrage}
        AchievementId := achievement_id.EliminationSlimeNinjaMinibossPurple
    npc_id.NinjaMinibossDash => npc_info:
        Id := npc_id.NinjaMinibossDash
        Name := "Orange Ninja Slime"
        Icon := UserInterface.Slimes.Slime_Ninja_Miniboss_Dash_128
        IconSmall := UserInterface.Slimes.Slime_Ninja_Miniboss_Dash_64
        Type := npc_type.Miniboss
        AttackTypes := array{npc_attack_type.Dash}
        AchievementId := achievement_id.EliminationSlimeNinjaMinibossOrange
    npc_id.NinjaMinibossCharge => npc_info:
        Id := npc_id.NinjaMinibossCharge
        Name := "Yellow Ninja Slime"
        Icon := UserInterface.Slimes.Slime_Ninja_Miniboss_Charge_128
        IconSmall := UserInterface.Slimes.Slime_Ninja_Miniboss_Charge_64
        Type := npc_type.Miniboss
        AttackTypes := array{npc_attack_type.Charge}
        AchievementId := achievement_id.EliminationSlimeNinjaMinibossYellow
    npc_id.NinjaBoss => npc_info:
        Id := npc_id.NinjaBoss
        Name := "Boss Ninja Slime"
        Icon := UserInterface.Slimes.Slime_Ninja_Boss_128
        IconSmall := UserInterface.Slimes.Slime_Ninja_Boss_64
        Type := npc_type.Boss
        AttackTypes := array{npc_attack_type.Jump, npc_attack_type.ForwardJump, npc_attack_type.Dash, npc_attack_type.Projectile}
        AchievementId := achievement_id.EliminationSlimeNinjaBoss

    npc_id.ShogunMelee => npc_info:
        Id := npc_id.ShogunMelee
        Name := "Shogun Slime"
        Icon := UserInterface.Slimes.Slime_Shogun_128
        IconSmall := UserInterface.Slimes.Slime_Shogun_64
        AttackTypes := array{npc_attack_type.Melee}
        AchievementId := achievement_id.EliminationSlimeShogun
    npc_id.ShogunMeleeElite => npc_info:
        Id := npc_id.ShogunMeleeElite
        Name := "Elite Shogun Slime"
        Icon := UserInterface.Slimes.Slime_Shogun_Elite_128
        IconSmall := UserInterface.Slimes.Slime_Shogun_Elite_64
        Type := npc_type.Elite
        AttackTypes := array{npc_attack_type.Melee}
        AchievementId := achievement_id.EliminationSlimeShogunElite
    npc_id.ShogunMeleeGeneral => npc_info:
        Id := npc_id.ShogunMeleeGeneral
        Name := "General Shogun Slime"
        Icon := UserInterface.Slimes.Slime_Shogun_General_128
        IconSmall := UserInterface.Slimes.Slime_Shogun_General_64
        Type := npc_type.General
        AttackTypes := array{npc_attack_type.Melee}
        AchievementId := achievement_id.EliminationSlimeShogunGeneral
    npc_id.ShogunMinibossJump => npc_info:
        Id := npc_id.ShogunMinibossJump
        Name := "Red Shogun Slime"
        Icon := UserInterface.Slimes.Slime_Shogun_Miniboss_Jump_128
        IconSmall := UserInterface.Slimes.Slime_Shogun_Miniboss_Jump_64
        Type := npc_type.Miniboss
        AttackTypes := array{npc_attack_type.Jump}
        AchievementId := achievement_id.EliminationSlimeShogunMinibossRed
    npc_id.ShogunMinibossProjectile => npc_info:
        Id := npc_id.ShogunMinibossProjectile
        Name := "Blue Shogun Slime"
        Icon := UserInterface.Slimes.Slime_Shogun_Miniboss_Projectile_128
        IconSmall := UserInterface.Slimes.Slime_Shogun_Miniboss_Projectile_64
        Type := npc_type.Miniboss
        AttackTypes := array{npc_attack_type.Projectile}
        AchievementId := achievement_id.EliminationSlimeShogunMinibossBlue
    npc_id.ShogunMinibossArea => npc_info:
        Id := npc_id.ShogunMinibossArea
        Name := "Purple Shogun Slime"
        Icon := UserInterface.Slimes.Slime_Shogun_Miniboss_Area_128
        IconSmall := UserInterface.Slimes.Slime_Shogun_Miniboss_Area_64
        Type := npc_type.Miniboss
        AttackTypes := array{npc_attack_type.Barrage}
        AchievementId := achievement_id.EliminationSlimeShogunMinibossPurple
    npc_id.ShogunMinibossDash => npc_info:
        Id := npc_id.ShogunMinibossDash
        Name := "Orange Shogun Slime"
        Icon := UserInterface.Slimes.Slime_Shogun_Miniboss_Dash_128
        IconSmall := UserInterface.Slimes.Slime_Shogun_Miniboss_Dash_64
        Type := npc_type.Miniboss
        AttackTypes := array{npc_attack_type.Dash}
        AchievementId := achievement_id.EliminationSlimeShogunMinibossOrange
    npc_id.ShogunMinibossCharge => npc_info:
        Id := npc_id.ShogunMinibossCharge
        Name := "Yellow Shogun Slime"
        Icon := UserInterface.Slimes.Slime_Shogun_Miniboss_Charge_128
        IconSmall := UserInterface.Slimes.Slime_Shogun_Miniboss_Charge_64
        Type := npc_type.Miniboss
        AttackTypes := array{npc_attack_type.Charge}
        AchievementId := achievement_id.EliminationSlimeShogunMinibossYellow
    npc_id.ShogunBoss => npc_info:
        Id := npc_id.ShogunBoss
        Name := "Boss Shogun Slime"
        Icon := UserInterface.Slimes.Slime_Shogun_Boss_128
        IconSmall := UserInterface.Slimes.Slime_Shogun_Boss_64
        Type := npc_type.Boss
        AttackTypes := array{npc_attack_type.Jump, npc_attack_type.ForwardJump, npc_attack_type.Dash, npc_attack_type.Projectile}
        AchievementId := achievement_id.EliminationSlimeShogunBoss
    
    npc_id.FrogMelee => npc_info:
        Id := npc_id.FrogMelee
        Name := "Frog Slime"
        Icon := UserInterface.Slimes.Slime_Frog_128
        IconSmall := UserInterface.Slimes.Slime_Frog_64
        AttackTypes := array{npc_attack_type.Melee}
        AchievementId := achievement_id.EliminationSlimeFrog
    npc_id.FrogMeleeElite => npc_info:
        Id := npc_id.FrogMeleeElite
        Name := "Elite Frog Slime"
        Icon := UserInterface.Slimes.Slime_Frog_Elite_128
        IconSmall := UserInterface.Slimes.Slime_Frog_Elite_64
        Type := npc_type.Elite
        AttackTypes := array{npc_attack_type.Melee}
        AchievementId := achievement_id.EliminationSlimeFrogElite
    npc_id.FrogMeleeGeneral => npc_info:
        Id := npc_id.FrogMeleeGeneral
        Name := "General Frog Slime"
        Icon := UserInterface.Slimes.Slime_Frog_General_128
        IconSmall := UserInterface.Slimes.Slime_Frog_General_64
        Type := npc_type.General
        AttackTypes := array{npc_attack_type.Melee}
        AchievementId := achievement_id.EliminationSlimeFrogGeneral
    npc_id.FrogMinibossJump => npc_info:
        Id := npc_id.FrogMinibossJump
        Name := "Red Frog Slime"
        Icon := UserInterface.Slimes.Slime_Frog_Miniboss_Jump_128
        IconSmall := UserInterface.Slimes.Slime_Frog_Miniboss_Jump_64
        Type := npc_type.Miniboss
        AttackTypes := array{npc_attack_type.Jump}
        AchievementId := achievement_id.EliminationSlimeFrogMinibossRed
    npc_id.FrogMinibossProjectile => npc_info:
        Id := npc_id.FrogMinibossProjectile
        Name := "Blue Frog Slime"
        Icon := UserInterface.Slimes.Slime_Frog_Miniboss_Projectile_128
        IconSmall := UserInterface.Slimes.Slime_Frog_Miniboss_Projectile_64
        Type := npc_type.Miniboss
        AttackTypes := array{npc_attack_type.Projectile}
        AchievementId := achievement_id.EliminationSlimeFrogMinibossBlue
    npc_id.FrogMinibossArea => npc_info:
        Id := npc_id.FrogMinibossArea
        Name := "Purple Frog Slime"
        Icon := UserInterface.Slimes.Slime_Frog_Miniboss_Area_128
        IconSmall := UserInterface.Slimes.Slime_Frog_Miniboss_Area_64
        Type := npc_type.Miniboss
        AttackTypes := array{npc_attack_type.Barrage}
        AchievementId := achievement_id.EliminationSlimeFrogMinibossPurple
    npc_id.FrogMinibossDash => npc_info:
        Id := npc_id.FrogMinibossDash
        Name := "Orange Frog Slime"
        Icon := UserInterface.Slimes.Slime_Frog_Miniboss_Dash_128
        IconSmall := UserInterface.Slimes.Slime_Frog_Miniboss_Dash_64
        Type := npc_type.Miniboss
        AttackTypes := array{npc_attack_type.Dash}
        AchievementId := achievement_id.EliminationSlimeFrogMinibossOrange
    npc_id.FrogMinibossCharge => npc_info:
        Id := npc_id.FrogMinibossCharge
        Name := "Yellow Frog Slime"
        Icon := UserInterface.Slimes.Slime_Frog_Miniboss_Charge_128
        IconSmall := UserInterface.Slimes.Slime_Frog_Miniboss_Charge_64
        Type := npc_type.Miniboss
        AttackTypes := array{npc_attack_type.Charge}
        AchievementId := achievement_id.EliminationSlimeFrogMinibossYellow
    npc_id.FrogBoss => npc_info:
        Id := npc_id.FrogBoss
        Name := "Boss Frog Slime"
        Icon := UserInterface.Slimes.Slime_Frog_Boss_128
        IconSmall := UserInterface.Slimes.Slime_Frog_Boss_64
        Type := npc_type.Boss
        AttackTypes := array{npc_attack_type.Jump, npc_attack_type.ForwardJump, npc_attack_type.Dash, npc_attack_type.Projectile}
        AchievementId := achievement_id.EliminationSlimeFrogBoss

    npc_id.Golden => npc_info:
        Id := npc_id.Golden
        Name := "Golden Slime"
        Icon := UserInterface.Icons.NPC_Melee_128
        IconSmall := UserInterface.Icons.NPC_Melee_64
        Type := npc_type.Miniboss

    # [Note] Deprecated.
    npc_id.Projectile => npc_info:
        Id := npc_id.Projectile
        Name := "Blue Slime"
        Icon := UserInterface.Slimes.Slime_Forest_Miniboss_Projectile_128
        IconSmall := UserInterface.Slimes.Slime_Forest_Miniboss_Projectile_64
    npc_id.Area => npc_info:
        Id := npc_id.Area
        Name := "Purple Slime"
        Icon := UserInterface.Slimes.Slime_Forest_Miniboss_Area_128
        IconSmall := UserInterface.Slimes.Slime_Forest_Miniboss_Area_64
    npc_id.Charge => npc_info:
        Id := npc_id.Charge
        Name := "Yellow Slime"
        Icon := UserInterface.Slimes.Slime_Forest_Miniboss_Charge_128
        IconSmall := UserInterface.Slimes.Slime_Forest_Miniboss_Charge_64
    npc_id.Dash => npc_info:
        Id := npc_id.Dash
        Name := "Orange Slime"
        Icon := UserInterface.Slimes.Slime_Forest_Miniboss_Dash_128
        IconSmall := UserInterface.Slimes.Slime_Forest_Miniboss_Dash_64

    npc_id.SnowProjectile => npc_info:
        Id := npc_id.SnowProjectile
        Name := "Blue Snow Slime"
        Icon := UserInterface.Slimes.Slime_Snow_Miniboss_Projectile_128
        IconSmall := UserInterface.Slimes.Slime_Snow_Miniboss_Projectile_64
    npc_id.SnowArea => npc_info:
        Id := npc_id.SnowArea
        Name := "Purple Snow Slime"
        Icon := UserInterface.Slimes.Slime_Snow_Miniboss_Area_128
        IconSmall := UserInterface.Slimes.Slime_Snow_Miniboss_Area_64
    npc_id.SnowCharge => npc_info:
        Id := npc_id.SnowCharge
        Name := "Yellow Snow Slime"
        Icon := UserInterface.Slimes.Slime_Snow_Miniboss_Charge_128
        IconSmall := UserInterface.Slimes.Slime_Snow_Miniboss_Charge_64
    npc_id.SnowDash => npc_info:
        Id := npc_id.SnowDash
        Name := "Orange Snow Slime"
        Icon := UserInterface.Slimes.Slime_Snow_Miniboss_Dash_128
        IconSmall := UserInterface.Slimes.Slime_Snow_Miniboss_Dash_64

<# ----------------------------------------------------------------------------------------------------
    Global Functions
---------------------------------------------------------------------------------------------------- #>
GetNPCInfo(Id:npc_id)<decides><transacts>:npc_info = 
    NPCInfos[Id]

GetNPCsByArena(Arena:arena)<decides><transacts>:[]npc_id =
    case (Arena.Level):
        1 => array:
            npc_id.Melee, npc_id.MeleeElite, npc_id.MeleeGeneral,
            npc_id.MinibossJump, npc_id.MinibossProjectile, npc_id.MinibossArea,
            npc_id.MinibossDash, npc_id.MinibossCharge, npc_id.Boss

        2 => array:
            npc_id.WastelandMelee, npc_id.WastelandMeleeElite, npc_id.WastelandMeleeGeneral,
            npc_id.WastelandMinibossJump, npc_id.WastelandMinibossProjectile, npc_id.WastelandMinibossArea,
            npc_id.WastelandMinibossDash, npc_id.WastelandMinibossCharge, npc_id.WastelandBoss

        3 => array:
            npc_id.ElysiumMelee, npc_id.ElysiumMeleeElite, npc_id.ElysiumMeleeGeneral,
            npc_id.ElysiumMinibossJump, npc_id.ElysiumMinibossProjectile, npc_id.ElysiumMinibossArea,
            npc_id.ElysiumMinibossDash, npc_id.ElysiumMinibossCharge, npc_id.ElysiumBoss

        4 => array:
            npc_id.AtlantisMelee, npc_id.AtlantisMeleeElite, npc_id.AtlantisMeleeGeneral,
            npc_id.AtlantisMinibossJump, npc_id.AtlantisMinibossProjectile, npc_id.AtlantisMinibossArea,
            npc_id.AtlantisMinibossDash, npc_id.AtlantisMinibossCharge, npc_id.AtlantisBoss

        5 => array:
            npc_id.SnowMelee, npc_id.SnowMeleeElite, npc_id.SnowMeleeGeneral,
            npc_id.SnowMinibossProjectile, npc_id.SnowMinibossArea, npc_id.SnowMinibossDash,
            npc_id.SnowMinibossCharge, npc_id.SnowMinibossJump, npc_id.SnowBoss

        6 => array:
            npc_id.CyberMelee, npc_id.CyberMeleeElite, npc_id.CyberMeleeGeneral,
            npc_id.CyberMinibossJump, npc_id.CyberMinibossProjectile, npc_id.CyberMinibossArea,
            npc_id.CyberMinibossDash, npc_id.CyberMinibossCharge, npc_id.CyberBoss

        7 => array:
            npc_id.NinjaMelee, npc_id.NinjaMeleeElite, npc_id.NinjaMeleeGeneral,
            npc_id.NinjaMinibossJump, npc_id.NinjaMinibossProjectile, npc_id.NinjaMinibossArea,
            npc_id.NinjaMinibossDash, npc_id.NinjaMinibossCharge, npc_id.NinjaBoss

        8 => array:
            npc_id.ShogunMelee, npc_id.ShogunMeleeElite, npc_id.ShogunMeleeGeneral,
            npc_id.ShogunMinibossJump, npc_id.ShogunMinibossProjectile, npc_id.ShogunMinibossArea,
            npc_id.ShogunMinibossDash, npc_id.ShogunMinibossCharge, npc_id.ShogunBoss

        9 => array:
            npc_id.FrogMelee, npc_id.FrogMeleeElite, npc_id.FrogMeleeGeneral,
            npc_id.FrogMinibossJump, npc_id.FrogMinibossProjectile, npc_id.FrogMinibossArea,
            npc_id.FrogMinibossDash, npc_id.FrogMinibossCharge, npc_id.FrogBoss

GetArenaNPCs:[]npc_id =
    array:
        npc_id.Melee, npc_id.MeleeElite, npc_id.MeleeGeneral,
        npc_id.MinibossJump, npc_id.MinibossProjectile, npc_id.MinibossArea,
        npc_id.MinibossDash, npc_id.MinibossCharge, npc_id.Boss,
        npc_id.WastelandMelee, npc_id.WastelandMeleeElite, npc_id.WastelandMeleeGeneral,
        npc_id.WastelandMinibossJump, npc_id.WastelandMinibossProjectile, npc_id.WastelandMinibossArea,
        npc_id.WastelandMinibossDash, npc_id.WastelandMinibossCharge, npc_id.WastelandBoss,
        npc_id.ElysiumMelee, npc_id.ElysiumMeleeElite, npc_id.ElysiumMeleeGeneral,
        npc_id.ElysiumMinibossJump, npc_id.ElysiumMinibossProjectile, npc_id.ElysiumMinibossArea,
        npc_id.ElysiumMinibossDash, npc_id.ElysiumMinibossCharge, npc_id.ElysiumBoss,
        npc_id.AtlantisMelee, npc_id.AtlantisMeleeElite, npc_id.AtlantisMeleeGeneral,
        npc_id.AtlantisMinibossJump, npc_id.AtlantisMinibossProjectile, npc_id.AtlantisMinibossArea,
        npc_id.AtlantisMinibossDash, npc_id.AtlantisMinibossCharge, npc_id.AtlantisBoss,
        npc_id.SnowMelee, npc_id.SnowMeleeElite, npc_id.SnowMeleeGeneral,
        npc_id.SnowMinibossProjectile, npc_id.SnowMinibossArea, npc_id.SnowMinibossDash,
        npc_id.SnowMinibossCharge, npc_id.SnowMinibossJump, npc_id.SnowBoss,
        npc_id.CyberMelee, npc_id.CyberMeleeElite, npc_id.CyberMeleeGeneral,
        npc_id.CyberMinibossJump, npc_id.CyberMinibossProjectile, npc_id.CyberMinibossArea,
        npc_id.CyberMinibossDash, npc_id.CyberMinibossCharge, npc_id.CyberBoss,
        npc_id.NinjaMelee, npc_id.NinjaMeleeElite, npc_id.NinjaMeleeGeneral,
        npc_id.NinjaMinibossJump, npc_id.NinjaMinibossProjectile, npc_id.NinjaMinibossArea,
        npc_id.NinjaMinibossDash, npc_id.NinjaMinibossCharge, npc_id.NinjaBoss,
        npc_id.ShogunMelee, npc_id.ShogunMeleeElite, npc_id.ShogunMeleeGeneral,
        npc_id.ShogunMinibossJump, npc_id.ShogunMinibossProjectile, npc_id.ShogunMinibossArea,
        npc_id.ShogunMinibossDash, npc_id.ShogunMinibossCharge, npc_id.ShogunBoss,
        npc_id.FrogMelee, npc_id.FrogMeleeElite, npc_id.FrogMeleeGeneral,
        npc_id.FrogMinibossJump, npc_id.FrogMinibossProjectile, npc_id.FrogMinibossArea,
        npc_id.FrogMinibossDash, npc_id.FrogMinibossCharge, npc_id.FrogBoss