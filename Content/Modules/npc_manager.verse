using { /Fortnite.com/AI }
using { /Fortnite.com/Characters }
using { /Fortnite.com/Devices }
using { /Verse.org/Assets }
using { /Verse.org/Random }
using { /Verse.org/SceneGraph }
using { /Verse.org/Simulation }
using { /UnrealEngine.com/Temporary/SpatialMath }
using { NPCState }

<# ----------------------------------------------------------------------------------------------------
    Constants
---------------------------------------------------------------------------------------------------- #>
DefaultNPCSpawnDistance:float = 1350.0
DefaultNPCRepositionDistanceSquared:float = 2250000.0 # 1500.0
DefaultNPCKnockbackDistanceSquared:float = 160000.0 # 400.0
DefaultNPCKnockbackDistanceSquaredRevive:float = 640000.0 # 800.0

<# ----------------------------------------------------------------------------------------------------
    Data Definitions
---------------------------------------------------------------------------------------------------- #>
npc_stage := class<concrete>:
    @editable
    NPCWaves:[]npc_wave = array{}

npc_spawner := class<concrete>:
    @editable
    Id:npc_id = npc_id.Melee
    @editable
    Device:npc_spawner_device = npc_spawner_device{}

npc_spawn := class<concrete>:
    @editable
    Id:npc_id = npc_id.Melee
    @editable
    Interval:float = 3.0
    @editable
    MaybeLimit:?int = false
    @editable
    IsBoss:logic = false

npc_wave := class<concrete>:
    @editable
    Duration:float = 60.0
    @editable
    NPCSpawners:[]npc_spawn = array{}

miniboss_map_indicator := class<concrete>:
    @editable
    MapIndicator:map_indicator_device = map_indicator_device{}
    @editable
    CinematicSequence:cinematic_sequence_device = cinematic_sequence_device{}

<# ----------------------------------------------------------------------------------------------------
    Global Functions
---------------------------------------------------------------------------------------------------- #>
var GlobalNPCManager:weak_map(session, npc_manager) = map{}

GetNPCManager()<decides><transacts>:npc_manager = 
    GlobalNPCManager[GetSession()]

GetAllNPCs()<transacts>:[]npc_behavior_base =
    GetNPCManager[].GetNPCs() or array{}

GetAllNPCBosses()<transacts>:[]npc_behavior_base =
    GetNPCManager[].GetNPCBosses() or array{}

GetNPCSpawners()<transacts>:[]npc_spawner =
    GetNPCManager[].GetNPCSpawners() or array{}

GetSpawnPosition(Agent:agent)<transacts>:vector3 =
    GetNPCManager[].GetSpawnPosition(Agent) or vector3{}

(Agent:agent).GetRepositionLocation(Target:agent)<transacts>:vector3 =
    GetNPCManager[].GetRepositionLocation(Agent, Target) or vector3{}

GetNPCAttack(Type:npc_attack_type)<transacts>:npc_attack =
    GetNPCManager[].GetNPCAttack(Type) or npc_attack_melee{}

GetNPCState(Type:npc_state_type)<transacts>:npc_state =
    GetNPCManager[].GetNPCState(Type) or npc_state_chase{}

GetCurrentNPCWave()<transacts>:?npc_wave =
    GetNPCManager[].GetCurrentNPCWave() or false

GetCurrentNPCWaveIndex()<transacts>:int =
    GetNPCManager[].CurrentWaveIndex or 0

SpawnNPC(Id:npc_id, TargetPlayer:player, ?MaybePosition:?vector3 = false):?npc_spawner_device =
    if (NPCManager := GetNPCManager[]):
        NPCManager.Spawn(Id, TargetPlayer, ?MaybePosition := MaybePosition)
    else:
        false

DespawnNPC():void =
    if (NPCManager := GetNPCManager[]):
        NPCManager.StopSpawners()

KnockbackNPC(Position:vector3, ?KnockbackDistanceSquared:float = DefaultNPCKnockbackDistanceSquared, ?Duration:float = 1.0):void =
    if (NPCManager := GetNPCManager[]):
        NPCManager.KnockbackNPC(Position, ?KnockbackDistanceSquared := KnockbackDistanceSquared, ?Duration := Duration)

KnockbackNPC(Player:player, ?KnockbackDistanceSquared:float = DefaultNPCKnockbackDistanceSquared, ?Duration:float = 1.0):void =
    if (NPCManager := GetNPCManager[]):
        NPCManager.KnockbackNPC(Player, ?KnockbackDistanceSquared := KnockbackDistanceSquared, ?Duration := Duration)

(NPCBehaviorBase:npc_behavior_base).RegisterNPC():void =
    if (NPCManager := GetNPCManager[]):
        NPCManager.RegisterNPC(NPCBehaviorBase)

(Agent:agent).UpdateNPCTarget(MaybyExistingTarget:?player, Target:player):void =
    if (NPCManager := GetNPCManager[]):
        NPCManager.UpdateNPCTarget(MaybyExistingTarget, Target)

(NPCBehaviorBase:npc_behavior_base).UnregisterNPC(MaybeTarget:?player):void =
    if (NPCManager := GetNPCManager[]):
        NPCManager.UnregisterNPC(NPCBehaviorBase, MaybeTarget)

SpawnRandomMiniboss():void =
    if (NPCManager := GetNPCManager[]):
        spawn{NPCManager.SpawnRandomMiniboss()}

ActivateBossHitEffect():void =
    if (NPCManager := GetNPCManager[]):
        NPCManager.BossHitEffectCinematicSequence.Play()

(NPCBehaviorBase:npc_behavior_base).GetDamageNumber()<decides><transacts>:poolable_entity =
    GetNPCManager[].DamageNumberEntityFactory.GetAvailable()?

(NPCBehaviorBase:npc_behavior_base).GetCollisionEntity()<decides><transacts>:poolable_entity =
    GetNPCManager[].CollisionEntityFactory.GetAvailable()?

<# ----------------------------------------------------------------------------------------------------
    npc_manager
---------------------------------------------------------------------------------------------------- #>
npc_manager := class(creative_device):
    @editable
    NPCStages:[]npc_stage = array{}

    @editable
    NPCSpawners:[]npc_spawner = array{}

    @editable
    MinibossMapIndicators:[]miniboss_map_indicator = array{}

    @editable
    BossHitEffectCinematicSequence:cinematic_sequence_device = cinematic_sequence_device{}
    
    var MinibossMapIndicatorQueue<private>:queue(miniboss_map_indicator) = queue(miniboss_map_indicator){}

    StopSpawnerEvent:events(tuple()) = events(tuple()){}

    NPCAttackFactory:npc_attack_factory = npc_attack_factory{}
    NPCStateFactory:npc_state_factory = npc_state_factory{}
    DamageNumberEntityFactory:damage_number_factory = damage_number_factory{}
    CollisionEntityFactory:npc_collision_factory = npc_collision_factory{}

    @editable
    MaxSpawnCount:int = 90

    var NPCs:[]npc_behavior_base = array{}
    var NPCBosses:[]npc_behavior_base = array{}
    var CurrentNPCWave:?npc_wave = false
    var MappedNPCSpawners:[npc_id]npc_spawner_device = map{}
    var PlayerNPCTargetCount:[player]int = map{}
    var NPCSpawned:[npc_id]int = map{}
    var NPCSpawnTarget:[npc_id]int = map{}
    var FillSpawnCount:int = 0
    var CurrentWaveIndex:int = 0
    var AbilityUpgradesDropped:int = 0

    OnBegin<override>()<suspends>:void =
        if (set GlobalNPCManager[GetSession()] = Self) {}

        for (Index -> NPCSpawner : NPCSpawners):
            if (set MappedNPCSpawners[NPCSpawner.Id] = NPCSpawner.Device) {}

        for (MinibossMapIndicator : MinibossMapIndicators):
            if (MinibossMapIndicator.MapIndicator.TeleportTo[MapIndicatorDisabledPosition, IdentityRotation()]) {}
            MinibossMapIndicator.CinematicSequence.Play()
            MinibossMapIndicator.CinematicSequence.Pause()

        set MinibossMapIndicatorQueue = queue(miniboss_map_indicator){Elements := MinibossMapIndicators}
        
        if (SimulationEntity := Self.GetSimulationEntity[]):
            DamageNumberEntityFactory.Initialize(SimulationEntity, GetMaxSpawnCount())
            CollisionEntityFactory.Initialize(SimulationEntity, GetMaxSpawnCount())

        for (Player : GetPlayspace().GetPlayers()):
            Initialize(Player)

        GetPlayspace().PlayerAddedEvent().Subscribe(OnPlayerAdded)
        GetPlayspace().PlayerRemovedEvent().Subscribe(OnPlayerRemoved)

    Initialize(Player:player):void =
        if (set PlayerNPCTargetCount[Player] = 0) {}

    <# ----------------------------------------------------------------------------------------------------
        Get Functions
    ---------------------------------------------------------------------------------------------------- #>
    GetNPCs()<transacts>:[]npc_behavior_base =
        NPCs

    GetNPCBosses()<transacts>:[]npc_behavior_base =
        NPCBosses

    GetTotalDuration(ArenaIndex:int)<transacts>:float =
        var TotalDuration:float = 0.0
        for (NPCWave : NPCStages[ArenaIndex].NPCWaves):
            set TotalDuration += NPCWave.Duration
        TotalDuration

    (npc_manager:)GetSpawnPosition(Agent:agent)<transacts>:vector3 =
        var SpawnPosition:vector3 = DefaultArenaOrigin
        var PlayerIndex:int = 0
        var CurrentTarget:agent = Agent

        loop:
            if (Character := CurrentTarget.GetFortCharacter[]):
                ArenaSize := GetActiveArenaSize()
                DistanceFromArenaBounds := ArenaSize - Distance(Character.GetTransform().Translation.Make2D(), vector3{})
                RandomPosition := if (DistanceFromArenaBounds < DefaultNPCSpawnDistance):
                    RatioOutside := 1.0 - Pow((DefaultNPCSpawnDistance - DistanceFromArenaBounds) / (DefaultNPCSpawnDistance * 2.0), 0.7)
                    DirectionToArenaOrigin := (vector3{} - Character.GetTransform().Translation.Make2D()).MakeUnitVector[] or vector3{}
                    TargetDirection := MakeRotationFromYawPitchRollDegrees(GetRandomFloat(-RatioOutside * 180, RatioOutside * 180), 0.0, 0.0).RotateVector(DirectionToArenaOrigin)
                    Character.GetTransform().Translation.Make2D() + TargetDirection * DefaultNPCSpawnDistance
                else:
                    GetRandomPositionOnCircle(DefaultNPCSpawnDistance, ?Origin := Character.GetTransform().Translation)

                if:
                    DistanceFromOrigin := Distance(DefaultArenaOrigin, RandomPosition) > ArenaSize
                    DistanceOutside := ArenaSize - DistanceFromOrigin
                    Direction := (RandomPosition - DefaultArenaOrigin).MakeUnitVector[]
                then:
                    set SpawnPosition = RandomPosition + Direction * DistanceOutside
                    break
                else:
                    set SpawnPosition = RandomPosition

                    if (PlayerIndex >= GetAllPlayers().Length - 1):
                        break
                    else:
                        set PlayerIndex += 1

                    Positions := for (Player : GetAllPlayers(), PlayerCharacter := Player.GetFortCharacter[]):
                        if (PlayerCharacter.IsDownButNotOut[], RevivePosition := Player.GetRevivePosition[]):
                            RevivePosition
                        else:
                            PlayerCharacter.GetTransform().Translation

                    if (NearestIndex := GetNearestPosition(SpawnPosition, Positions)?, NearestAgent := GetAllPlayers()[NearestIndex], NearestAgent <> CurrentTarget):
                        set CurrentTarget = NearestAgent
                    else:
                        break
            else:
                break

        SpawnPosition

    (npc_manager:)GetRepositionLocation(Agent:agent, TargetAgent:agent)<transacts>:vector3 =
        var SpawnPosition:vector3 = DefaultArenaOrigin
        
        var Index:int = 0
        var CurrentTarget:agent = TargetAgent

        loop:
            if:
                Character := Agent.GetFortCharacter[]
                TargetCharacter := CurrentTarget.GetFortCharacter[]
                TargetPosition := TargetCharacter.IsDownButNotOut[] and CurrentTarget.GetRevivePosition[] or TargetCharacter.GetTransform().Translation
                Direction := (TargetPosition - Character.GetTransform().Translation).MakeUnitVector[]
            then:
                ArenaSize := GetActiveArenaSize()
                DistanceFromArenaBounds := ArenaSize - Distance(TargetCharacter.GetTransform().Translation.Make2D(), vector3{})
                RandomPosition := if (DistanceFromArenaBounds < DefaultNPCSpawnDistance):
                    RatioOutside := 1.0 - Pow((DefaultNPCSpawnDistance - DistanceFromArenaBounds) / (DefaultNPCSpawnDistance * 2.0), 0.7)
                    DirectionToArenaOrigin := (vector3{} - TargetCharacter.GetTransform().Translation.Make2D()).MakeUnitVector[] or vector3{}
                    TargetDirection := MakeRotationFromYawPitchRollDegrees(GetRandomFloat(-RatioOutside * 180, RatioOutside * 180), 0.0, 0.0).RotateVector(DirectionToArenaOrigin)
                    TargetCharacter.GetTransform().Translation.Make2D() + TargetDirection * DefaultNPCSpawnDistance
                else:
                    Radius := DefaultNPCSpawnDistance * (TargetCharacter.IsDownButNotOut[] and 0.75 or 1.0)
                    Origin := TargetCharacter.GetTransform().Translation
                    AngleDegrees := RadiansToDegrees(ArcTan(Direction.Y, Direction.X))

                    AngleDelta := 60.0
                    AngleMax := AngleDegrees + AngleDelta
                    AngleMin := AngleDegrees - AngleDelta

                    Angle := DegreesToRadians(GetRandomFloat(AngleMin, AngleMax))
                    vector3{X := Origin.X + Cos(Angle) * Radius, Y := Origin.Y + Sin(Angle) * Radius, Z := Origin.Z}

                if (DistanceFromOrigin := Distance(DefaultArenaOrigin, RandomPosition) > ArenaSize):
                    DistanceOutside := ArenaSize - DistanceFromOrigin
                    if (DirectionOutside := (RandomPosition - DefaultArenaOrigin).MakeUnitVector[]):
                        set SpawnPosition = RandomPosition + DirectionOutside * DistanceOutside
                        break
                else:
                    set SpawnPosition = RandomPosition

                    if (Index >= GetAllPlayers().Length - 1):
                        break
                    else:
                        set Index += 1

                    Positions := for (Player : GetAllPlayers(), PlayerCharacter := Player.GetFortCharacter[]):
                        if (PlayerCharacter.IsDownButNotOut[], RevivePosition := Player.GetRevivePosition[]):
                            RevivePosition
                        else:
                            PlayerCharacter.GetTransform().Translation

                    if (NearestIndex := GetNearestPosition(SpawnPosition, Positions)?, NearestAgent := GetAllPlayers()[NearestIndex], NearestAgent <> CurrentTarget):
                        set CurrentTarget = NearestAgent
                    else:
                        break
            else:
                break

        SpawnPosition

    (npc_manager:)GetSpawnPosition(Origin:vector3)<transacts>:vector3 =
        var SpawnPosition:vector3 = Origin

        if (RandomPlayer := GetRandomAlivePlayer()?, Character := RandomPlayer.GetFortCharacter[]):
            RandomPosition := GetRandomPositionOnCircle(DefaultNPCSpawnDistance, ?Origin := Character.GetTransform().Translation)
            ArenaSize := GetActiveArenaSize()
            if (DistanceFromOrigin := Distance(Origin, RandomPosition) > ArenaSize):
                DistanceOutside := ArenaSize - DistanceFromOrigin
                if (Direction := (RandomPosition - Origin).MakeUnitVector[]):
                    set SpawnPosition = RandomPosition + Direction * DistanceOutside
            else:
                set SpawnPosition = RandomPosition

        SpawnPosition

    (npc_manager:)GetNPCSpawners()<transacts>:[]npc_spawner =
        NPCSpawners

    (npc_manager:)GetNPCSpawner(Id:npc_id)<decides><transacts>:npc_spawner_device =
        MappedNPCSpawners[Id]
    
    (npc_manager:)GetNPCAttack(Type:npc_attack_type)<transacts>:npc_attack =
        NPCAttackFactory.GetNPCAttack(Type)

    (npc_manager:)GetNPCState(Type:npc_state_type)<transacts>:npc_state =
        NPCStateFactory.GetNPCState(Type)

    (npc_manager:)GetCurrentNPCWave()<transacts>:?npc_wave =
        CurrentNPCWave

    GetMinibossMapIndicator()<decides><transacts>:miniboss_map_indicator =
        DequeuedObject := MinibossMapIndicatorQueue.Dequeue[]
        set MinibossMapIndicatorQueue = DequeuedObject(0)
        DequeuedObject(1)

    ReturnMinibossMapIndicator(MinibossMapIndicator:miniboss_map_indicator):void =
        if (MinibossMapIndicator.MapIndicator.TeleportTo[MapIndicatorDisabledPosition, IdentityRotation()]) {}
        set MinibossMapIndicatorQueue = MinibossMapIndicatorQueue.Enqueue(MinibossMapIndicator)

    GetMaxSpawnCount()<transacts>:int =
        GetGameSettings[].IsDoubleEnemyCount? and Min(80, MaxSpawnCount * 2) or MaxSpawnCount

    <# ----------------------------------------------------------------------------------------------------
        Update Functions
    ---------------------------------------------------------------------------------------------------- #>
    StartWaves(ArenaIndex:int)<suspends>:void =
        if (CurrentNPCWave?):
            return

        set AbilityUpgradesDropped = 0
        ArenaNPCWaves:[]npc_wave = NPCStages[ArenaIndex].NPCWaves or array{}

        race:
            StopSpawnerEvent.Await()
            for (Index -> NPCWave : ArenaNPCWaves):
                set CurrentNPCWave = option{NPCWave}
                set CurrentWaveIndex = Index

                set NPCSpawnTarget = map{}
                var MaybeFillNPCSpawner:?npc_spawn = false
                var TotalSpawnCount:int = 0
                IsFinalWave := logic{Index = ArenaNPCWaves.Length - 1}
                for (NPCSpawner : NPCWave.NPCSpawners):
                    if (not NPCSpawner.IsBoss?):
                        if (Limit := NPCSpawner.MaybeLimit?, set NPCSpawnTarget[NPCSpawner.Id] = Limit * (GetGameSettings[].IsDoubleEnemyCount? and 2 or 1)):
                            set TotalSpawnCount += Limit
                        else:
                            set MaybeFillNPCSpawner = option{NPCSpawner}

                    spawn{StartSpawner(NPCSpawner, NPCWave, IsFinalWave)}

                if:
                    FillNPCSpawner := MaybeFillNPCSpawner?
                    
                    set FillSpawnCount = GetMaxSpawnCount() - TotalSpawnCount
                    FillSpawnCount > 0
                    set NPCSpawnTarget[FillNPCSpawner.Id] = FillSpawnCount

                Sleep(NPCWave.Duration / GetGameSpeed())
                AwardAccolade(accolade_id.SurvivedWave, GetActivePlayers())

    StartSpawner(NPCSpawn:npc_spawn, NPCWave:npc_wave, IsFinalWave:logic)<suspends>:void =
        if (NPCSpawn.IsBoss?):
            race:
                StopSpawnerEvent.Await()
                for (Index := 1..(NPCSpawn.MaybeLimit? or 1)):
                    Sleep(NPCSpawn.Interval)
                    if (NPCSpawner := (npc_manager:)GetNPCSpawner[NPCSpawn.Id]):
                        NPCSpawner.Spawn()

            return
        
        race:
            Sleep((IsFinalWave? and Inf or NPCWave.Duration) / GetGameSpeed())
            StopSpawnerEvent.Await()
            loop:
                if (SpawnedCount := NPCSpawned[NPCSpawn.Id]):
                    if (Limit := NPCSpawn.MaybeLimit?):
                        if (SpawnedCount >= Limit):
                            break
                    else if (SpawnedCount >= FillSpawnCount):
                        break
            
                if (NPCs.Length < GetMaxSpawnCount(), NPCSpawner := (npc_manager:)GetNPCSpawner[NPCSpawn.Id]):
                    NPCSpawner.Spawn()

                Sleep(NPCSpawn.Interval)

    (npc_manager:)SpawnRandomMiniboss()<suspends>:void =
        if (ActiveArena := GetActiveArena[]):   
            for (Index := 1..ActiveArena.MinibossMultiplier):
                RandomIndex := GetRandomInt(0, ActiveArena.Minibosses.Length - 1)
                if (MinibossId := ActiveArena.Minibosses[RandomIndex]):
                    Spawn(MinibossId)
                    Sleep(0.0)

    StopSpawners():void =
        StopSpawnerEvent.Signal()

        for (NPCSpawner : NPCSpawners):
            NPCSpawner.Device.DespawnAll(false)

        set NPCs = array{}
        set NPCBosses = array{}
        set CurrentNPCWave = false

    Spawn(Id:npc_id):?npc_spawner_device =
        if (NPCSpawner := (npc_manager:)GetNPCSpawner[Id]):
            NPCSpawner.Spawn()
            option{NPCSpawner}
        else:
            false

    Spawn(Id:npc_id, TargetPlayer:player, ?MaybePosition:?vector3 = false):?npc_spawner_device =
        if (NPCSpawner := (npc_manager:)GetNPCSpawner[Id]):
            TargetPosition := MaybePosition? or GetSpawnPosition(vector3{})
            if (NPCSpawner.TeleportTo[TargetPosition, rotation{}]) {}
            NPCSpawner.Spawn()
            option{NPCSpawner}
        else:
            false

    (npc_manager:)KnockbackNPC(Player:player, ?KnockbackDistanceSquared:float = DefaultNPCKnockbackDistanceSquared, ?Duration:float = 1.0):void =
        if (Character := Player.GetFortCharacter[]):
            (npc_manager:)KnockbackNPC(Character.GetTransform().Translation, ?KnockbackDistanceSquared := KnockbackDistanceSquared, ?Duration := Duration)

    (npc_manager:)KnockbackNPC(Position:vector3, ?KnockbackDistanceSquared:float = DefaultNPCKnockbackDistanceSquared, ?Duration:float = 1.0):void =
        for (NPCBehaviorBase : GetAllNPCs(), NPCCharacter := NPCBehaviorBase.GetAgent[].GetFortCharacter[]):
            if:
                DistanceSquaredXY(Position, NPCCharacter.GetTransform().Translation) <= KnockbackDistanceSquared
                Direction := (NPCCharacter.GetTransform().Translation - Position).MakeUnitVector[]
                TargetRotation := MakeRotation(vector3{X := 0.0, Y := 0.0, Z := 1.0}, ArcTan(Direction.Y, Direction.X))
            then:
                NPCBehaviorBase.Knockback(knockback_properties{Duration := Duration, Rotation  := TargetRotation})

    RegisterNPC(NPCBehaviorBase:npc_behavior_base):void =
        set NPCs += array{NPCBehaviorBase}

        if (NPCBehaviorBase.GetIsBoss()?):
            set NPCBosses += array{NPCBehaviorBase}

        if (set NPCSpawned[NPCBehaviorBase.Id] = NPCSpawned[NPCBehaviorBase.Id] + 1 or 1) {}

    UnregisterNPC(NPCBehaviorBase:npc_behavior_base, MaybeTarget:?player):void =
        if:
            UpdatedNPCs := NPCs.RemoveFirstElement[NPCBehaviorBase]
            set NPCs = UpdatedNPCs

        if:
            NPCBehaviorBase.GetIsBoss()?
            UpdatedNPCBosses := NPCBosses.RemoveFirstElement[NPCBehaviorBase]
            set NPCBosses = UpdatedNPCBosses
        then:

        if:
            Target := MaybeTarget?
            set PlayerNPCTargetCount[Target] -= 1

        if (set NPCSpawned[NPCBehaviorBase.Id] -= 1) {}

    UpdateNPCTarget(MaybyExistingTarget:?player, Target:player):void =
        if:
            not PlayerNPCTargetCount[Target]
            set PlayerNPCTargetCount[Target] = 1
        else if:
            set PlayerNPCTargetCount[Target] += 1

        if:
            ExistingTarget := MaybyExistingTarget?
            set PlayerNPCTargetCount[ExistingTarget] -= 1

    GetTarget()<transacts>:?player =
        var Target:?player = false
        var CurrentNPCTargetCount:int = 0

        AlivePlayers := GetAlivePlayers()

        for (Player : AlivePlayers, Player.GetFortCharacter[].IsVulnerable[], NPCTargetCount := PlayerNPCTargetCount[Player]):
            if (not Target?):
                set Target = option{Player}
                set CurrentNPCTargetCount = NPCTargetCount
            else if (NPCTargetCount < CurrentNPCTargetCount):
                set Target = option{Player}
                set CurrentNPCTargetCount = NPCTargetCount

        if (not Target?):
            set Target = option{AlivePlayers[GetRandomInt(0, AlivePlayers.Length - 1)]}
           
        Target

    <# ----------------------------------------------------------------------------------------------------
        Event Handlers
    ---------------------------------------------------------------------------------------------------- #>
    OnPlayerAdded(Player:player):void =
        Initialize(Player)
        
    OnPlayerRemoved(Player:player):void =
        set PlayerNPCTargetCount = RemoveKeyFromMap(PlayerNPCTargetCount, Player)