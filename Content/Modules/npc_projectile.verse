using { /Fortnite.com/AI }
using { /Fortnite.com/Characters }
using { /Fortnite.com/Devices }
using { /Verse.org/Simulation }
using { /UnrealEngine.com/Temporary/SpatialMath }
using { WeaponAbilities }

<# ----------------------------------------------------------------------------------------------------
    npc_projectile
---------------------------------------------------------------------------------------------------- #>
npc_projectile<public> := class:
    var AgentsHit:[]agent = array{}
    var AgentsHitCount:int = 0
    var MaybePropjectileVFX:?creative_prop = false

    <# ----------------------------------------------------------------------------------------------------
        Get Functions
    ---------------------------------------------------------------------------------------------------- #>
    HasHit(InAgentHit:agent)<transacts>:logic =
        for (AgentHit : AgentsHit, AgentHit = InAgentHit) {return true}
        false

    <# ----------------------------------------------------------------------------------------------------
        Update Functions
    ---------------------------------------------------------------------------------------------------- #>
    Activate(Properties:projectile_properties):void =
        spawn{Projectile(Properties)}

    Projectile(Properties:projectile_properties)<suspends>:void =
        if (Character := Properties.Owner.GetFortCharacter[], Character.IsActive[]):
            set AgentsHit = array{}
            set AgentsHitCount = 0

            TargetRotation := Properties.Rotation.ApplyYaw(Properties.Angle)
            Direction := TargetRotation.GetLocalForward()
            TargetPosition := Properties.Position + Direction * Properties.Distance
            Lifetime := Properties.Distance / Properties.Speed
            
            set MaybePropjectileVFX = SpawnVFX(Properties.VFX, Properties.Position, ?Rotation := TargetRotation, ?Scale := Properties.Radius * 0.1 * UnitVector, ?Duration := Lifetime)

            TargetPlayers := GetAlivePlayers()
            AttackRange := Properties.Radius + PlayerCharacterSize
            AttackRangeSquared := AttackRange * AttackRange

            var StartTime:float = GetSimulationElapsedTime()
            race:
                block:
                    if (PropjectileVFX := MaybePropjectileVFX?):
                        PropjectileVFX.MoveTo(TargetPosition, TargetRotation, Lifetime)
                    else:
                        Sleep(Lifetime)
                    
                loop:
                    ElapsedTime :=  GetSimulationElapsedTime() - StartTime
                    Progress := ElapsedTime / Lifetime
                    CurrentPosition := Lerp(Properties.Position, TargetPosition, Progress)

                    for:
                        TargetPlayer : TargetPlayers
                        not HasHit(TargetPlayer)?
                        TargetCharacter := TargetPlayer.GetFortCharacter[]
                        DistanceSquaredXY(TargetCharacter.GetTransform().Translation, CurrentPosition) <= AttackRangeSquared
                        NPCBehaviorBase := npc_behavior_base[Properties.Owner.GetNPCBehavior[]]
                    do:                             
                        TargetPlayer.ApplyDamage(Properties.Damage, ?IsBoss := NPCBehaviorBase.GetIsBoss())
                        set AgentsHit += array{TargetPlayer}
                        set AgentsHitCount += 1

                    if (Progress >= 1.0 or (Properties.Piercing > 0 and AgentsHitCount >= Properties.Piercing)):
                        break
                    else:
                        Sleep(ThreeTicks)

            if (PropjectileVFX := MaybePropjectileVFX?):
                PropjectileVFX.Dispose()
                set MaybePropjectileVFX = false