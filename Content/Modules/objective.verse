using { /Fortnite.com/Characters }
using { /Fortnite.com/Devices }
using { /Verse.org/Assets }
using { /Verse.org/Simulation }
using { /UnrealEngine.com/Temporary/SpatialMath }

<# ----------------------------------------------------------------------------------------------------
    Constants
---------------------------------------------------------------------------------------------------- #>
ObjectiveTimerWarning:float = 10.0 # [Note] Requires updating the widget animation.
ObjectiveBaseRewardAmount:float = 100.0

<# ----------------------------------------------------------------------------------------------------
    Data Definitions
---------------------------------------------------------------------------------------------------- #>
tracker_info := class<concrete>:
    @editable
    Title:string = ""
    @editable
    Description:string = ""
    @editable
    Device:tracker_device = tracker_device{}

objective_id := enum:
    Collectible
    GoldenSlime
    BeachBall
    CapturePoint
    NumberedTiles
    Storm
    SnowStorm
    MultiCapturePoint
    ArrowTiles
    FindCharacter

ToString(Id:objective_id)<transacts>:string =
    case (Id):
        objective_id.Collectible => "Collectibles"
        objective_id.GoldenSlime => "Golden Slime"
        objective_id.BeachBall => "Beach Ball"
        objective_id.CapturePoint => "Capture Point"
        objective_id.NumberedTiles => "Numbered Tiles"
        objective_id.Storm => "Survive The Storm"
        objective_id.SnowStorm => "Survive The Snow"
        objective_id.MultiCapturePoint => "Multi Capture Point"
        objective_id.ArrowTiles => "Arrowed Tiles"
        objective_id.FindCharacter => "Where's Wendell"

(Id:objective_id).ToIcon()<transacts>:texture =
    case (Id):
        objective_id.Collectible => UserInterface.Icons.DailyQuests64
        objective_id.GoldenSlime => UserInterface.Icons.Slime_Golden_64
        objective_id.BeachBall => UserInterface.Icons.BeachBall_64
        objective_id.CapturePoint => UserInterface.Icons.CapturePoint_64
        objective_id.NumberedTiles => UserInterface.Icons.NumberedTiles_64
        objective_id.Storm => UserInterface.Icons.StormZone_64
        objective_id.SnowStorm => UserInterface.Icons.SnowStorm_64
        objective_id.MultiCapturePoint => UserInterface.Icons.MultiCapturePoint_64
        objective_id.ArrowTiles => UserInterface.Icons.ArrowedTiles_64
        objective_id.FindCharacter => UserInterface.Icons.FindCharacter_64

GetObjectivesCompleted()<transacts>:int =
    GetGameManager[].ObjectivesCompleted or 0

<# ----------------------------------------------------------------------------------------------------
    objective
---------------------------------------------------------------------------------------------------- #>
objective := class<unique><concrete>:
    Id:objective_id = objective_id.BeachBall

    @editable
    HUDMessage:hud_message_device = hud_message_device{}

    @editable
    HUDMessageMobile:hud_message_device = hud_message_device{}
    
    <#
    [Note] Replaced tracker with HUDMessage due to join in progress bugs.

    @editable
    TrackerInfo:tracker_info = tracker_info{}

    @editable
    TrackerTimer:tracker_device = tracker_device{}
    #>

    @editable
    TimeLimit:float = 60.0

    StartEvent:events(tuple()) = events(tuple()){}
    StopEvent:events(tuple()) = events(tuple()){}
    SuccessEvent:events(tuple()) = events(tuple()){}
    CompleteEvent:events(tuple()) = events(tuple()){}
    ChallengeAccepted:events(tuple()) = events(tuple()){}

    var HUDMessageIntro:hud_message_device = hud_message_device{}
    var HUDMessageTimer:hud_message_device = hud_message_device{}
    var HUDMessageTimerMobile:hud_message_device = hud_message_device{}
    var CinematicSequence:cinematic_sequence_device = cinematic_sequence_device{}

    <# ----------------------------------------------------------------------------------------------------
        Initialization
    ---------------------------------------------------------------------------------------------------- #>
    OnBegin()<suspends>:void = {}

    <# ----------------------------------------------------------------------------------------------------
        Get Functions
    ---------------------------------------------------------------------------------------------------- #>
    GetTimeLimit()<transacts>:float =
        TimeLimit

    <# ----------------------------------------------------------------------------------------------------
        Update Functions
    ---------------------------------------------------------------------------------------------------- #>
    Start()<suspends>:void =
        defer:
            #TrackerInfo.Device.RemoveFromAll()
            #TrackerTimer.RemoveFromAll()
            HUDMessage.Hide()
            HUDMessageMobile.Hide()
            HUDMessageTimer.Hide()
            HUDMessageTimerMobile.Hide()
            OnStop()
            CompleteEvent.Signal()

        #TrackerInfo.Device.AssignToAll()
        #TrackerInfo.Device.SetValue(0)
        #TrackerInfo.Device.SetTitleText(MakeMessage("Mission - {TrackerInfo.Title}"))
        #TrackerInfo.Device.SetDescriptionText(MakeMessage("{TrackerInfo.Description}"))

        CinematicSequence.Play()
        CinematicSequence.Pause()
        Sleep(0.0)
        CinematicSequence.SetPlaybackTime(0.0)

        #TrackerTimer.AssignToAll()
        #TrackerTimer.SetTarget(Floor[GetTimeLimit()] or 0)
        #TrackerTimer.SetValue(Floor[GetTimeLimit()] or 0)

        HUDMessageIntro.Show(MakeMessage("{Id}"), ?DisplayTime := 3.8)
        for (Player : GetActivePlayers()):
            HUDMessageTimerDevice := (Player.GetInputMode() = input_mode.Touch) and HUDMessageTimerMobile or HUDMessageTimer
            HUDMessageTimerDevice.Show(Player, MakeMessage("{Floor[GetTimeLimit()] or 0}"), ?DisplayTime := 0.0)

        for (Index -> Player : GetActivePlayers()):
            Player.PlayAudio(audio_id.Notification)

        StartEvent.Signal()

        race:
            StopEvent.Await()

            block:
                race:
                    ChallengeAccepted.Await()
                    block:
                        Sleep(GetTimeLimit() - ObjectiveTimerWarning)
                        for (Player : GetActivePlayers()):
                            Player.PlayAudio(audio_id.TimerWarning)
                Sleep(Inf)

            block:
                StartTime := GetSimulationElapsedTime()

                race:
                    block:
                        ChallengeAccepted.Await()
                        HUDMessageTimer.Hide()
                        HUDMessageTimerMobile.Hide()
                        #TrackerTimer.RemoveFromAll()
                        
                    loop:
                        Sleep(1.0)
                        ElapsedTime := GetSimulationElapsedTime() - StartTime
                        TimeRemaining := GetTimeLimit() - ElapsedTime
                        
                        for (Player : GetActivePlayers()):
                            HUDMessageTimerDevice := (Player.GetInputMode() = input_mode.Touch) and HUDMessageTimerMobile or HUDMessageTimer
                            HUDMessageTimerDevice.Show(Player, MakeMessage("{Ceil[TimeRemaining] or 0}"), ?DisplayTime := 0.0)

                        #TrackerTimer.SetValue(Floor[TimeRemaining] or 0)

                        if (ElapsedTime >= GetTimeLimit()):
                            ShowMessage(message_id.MissionFailed)

                            for (Player : GetActivePlayers()):
                                Player.PlayAudio(audio_id.ObjectiveFailure)

                                if (GameProgressBarUI := GetGameManager[].PlayerGameProgressBarUI[Player]):
                                    GameProgressBarUI.UpdateEventState(GameEventMissionIndex, game_event_state.Failed)

                            Stop()
                            break

                Sleep(Inf)

            block:
                SuccessEvent.Await()
                spawn{GrantRewards()}

    Stop():void =
        StopEvent.Signal()

    GrantRewards()<suspends>:void =
        AddObjectivesComplete()
        ShowMessage(message_id.MissionComplete, "+{FormatAsCurrency(ObjectiveBaseRewardAmount * (1.0 + GetArenaGoldMultiplier()))}")

        for (Player : GetActivePlayers()):
            Player.PlayAudio(audio_id.ObjectiveSuccess)
            Player.AwardAccolade(accolade_id.CompleteMission)

            if (GameProgressBarUI := GetGameManager[].PlayerGameProgressBarUI[Player]):
                GameProgressBarUI.UpdateEventState(GameEventMissionIndex, game_event_state.Success)

        Sleep(2.0)
        
        if (GetObjectivesCompleted() <= MaxBoonCount):
            PauseGame(pause_info{Type := pause_type.Boon})
        else:
            PauseGame(pause_level_up{Type := pause_type.LevelUp, Level := 0, IsLegendary := true})

    UpdateHUD(Progress:int, Target:int):void =
        for (Player : GetActivePlayers()):
            HUDMessageDevice := (Player.GetInputMode() = input_mode.Touch) and HUDMessageMobile or HUDMessage
            HUDMessageDevice.Show(Player, MakeMessage("{Progress}/{Target}"), ?DisplayTime := 0.0)

    <# ----------------------------------------------------------------------------------------------------
        Event Handlers
    ---------------------------------------------------------------------------------------------------- #>
    OnStop():void = {}