using { /Fortnite.com/Devices }
using { /Verse.org/Random }
using { /Verse.org/Simulation }
using { /UnrealEngine.com/Temporary/SpatialMath }

<# ----------------------------------------------------------------------------------------------------
    Constants
---------------------------------------------------------------------------------------------------- #>
ObjectiveArrowTilesTarget:int = 12

<# ----------------------------------------------------------------------------------------------------
    objective_arrow_tiles
---------------------------------------------------------------------------------------------------- #>
objective_arrow_tiles := class<concrete>(objective):
    Id<override>:objective_id = objective_id.ArrowTiles

    @editable
    MapIndicator:map_indicator_device = map_indicator_device{}

    @editable
    ColorChangingTiles:[]color_changing_tiles_device = array{}

    @editable
    Volumes:[]volume_device = array{}

    ErrorEvent:events(tuple()) = events(tuple()){}
    UpdateEvent:events(int) = events(int){}

    var Sequence:[]int = array{}
    var Progress:int = 0

    <# ----------------------------------------------------------------------------------------------------
        Initialization
    ---------------------------------------------------------------------------------------------------- #>
    OnBegin<override>()<suspends>:void =
        StartEvent.Subscribe(OnStart)
        if (MapIndicator.TeleportTo[MapIndicatorDisabledPosition, IdentityRotation()]) {}
        
        for (Index -> ColorChangingTile : ColorChangingTiles):
            ColorChangingTile.Disable()
            ColorChangingTile.Hide()
            ColorChangingTile.GetTransform()

    <# ----------------------------------------------------------------------------------------------------
        Get Functions
    ---------------------------------------------------------------------------------------------------- #>
    IsCorrect(Index:int)<decides><transacts>:void =
        Index = Sequence[Progress]

    <# ----------------------------------------------------------------------------------------------------
        Event Handlers
    ---------------------------------------------------------------------------------------------------- #>
    OnStart():void =
        set Progress = 0
        
        var PreviousIndex:int = -1
        set Sequence = for (Index := 1..ObjectiveArrowTilesTarget):
            ArrowIndexes := for (ArrowIndex := 0..3, ArrowIndex <> PreviousIndex) {ArrowIndex}
            NextArrowIndex := ArrowIndexes[GetRandomInt(0, 2)] or 0
            set PreviousIndex = NextArrowIndex

        for (Player : GetActivePlayers()):
            Player.ShowObjectiveArrowTilesUI(Self, Sequence)

        if (MapIndicator.TeleportTo[MapIndicatorEnabledPosition, IdentityRotation()]) {}

        for (Index -> Player : GetActivePlayers()):
            MapIndicator.ActivateObjectivePulse(Player)

        UpdateHUD(0, ObjectiveArrowTilesTarget)

        for (Index -> ColorChangingTile : ColorChangingTiles, Volume := Volumes[Index]):
            ColorChangingTile.Enable()
            ColorChangingTile.Reset()
            ColorChangingTile.Show()
            spawn{AwaitActivated(Volume.AgentEntersEvent, ColorChangingTile, Index)}

    OnStop<override>():void =
        if (MapIndicator.TeleportTo[MapIndicatorDisabledPosition, IdentityRotation()]) {}
        
        for (Player : GetAllPlayers()):
            MapIndicator.DeactivateObjectivePulse(Player)
        
        for (ColorChangingTile : ColorChangingTiles):
            ColorChangingTile.Disable()
            ColorChangingTile.Hide()

    AwaitActivated(Listenable:listenable(agent), ColorChangingTile:color_changing_tiles_device, Index:int)<suspends>:void =
        race:
            CompleteEvent.Await()
            loop:
                Agent := Listenable.Await()
                ColorChangingTile.SetTeam(Agent)

                if (IsCorrect[Index]):
                    set Progress += 1
                    UpdateHUD(Progress, ObjectiveArrowTilesTarget)
                    CinematicSequence.SetPlaybackTime((Progress * 1.0) / (ObjectiveArrowTilesTarget * 1.0))

                    if (Progress = ObjectiveArrowTilesTarget):
                        SuccessEvent.Signal()
                    else:
                        for (Player : GetAllPlayers()):
                            Player.PlayAudio(audio_id.Activate)
                            UpdateEvent.Signal(Progress)
                else:                
                    for (Player : GetAllPlayers()):
                        Player.PlayAudio(audio_id.Error)

                    ErrorEvent.Signal()

                    Sleep(1.0)

                    set Progress = Max(Progress - 1, 0)
                    UpdateEvent.Signal(Progress)
                    UpdateHUD(Progress, ObjectiveArrowTilesTarget)
                    CinematicSequence.SetPlaybackTime(Progress * 1.0)

                Sleep(0.5)
                ColorChangingTile.Reset()