using { /Fortnite.com/UI }
using { /Verse.org/Assets }
using { /Verse.org/Colors }
using { /UnrealEngine.com/Temporary/SpatialMath }
using { /UnrealEngine.com/Temporary/UI }

<# ----------------------------------------------------------------------------------------------------
    objective_arrow_tiles_ui
---------------------------------------------------------------------------------------------------- #>
objective_arrow_tiles_ui := class(ui_component):
    ObjectiveArrowTiles:objective_arrow_tiles

    var ArrowIcons:[]texture_block = array{}
    var CurrentProgress:int = 0
    var EventsSubscriptions:[]events_cancelable = array{}

    <# ----------------------------------------------------------------------------------------------------
        Initialization
    ---------------------------------------------------------------------------------------------------- #>
    CreateUI<override>():widget =
        set EventsSubscriptions = array:
            ObjectiveArrowTiles.ErrorEvent.Subscribe(OnError)
            ObjectiveArrowTiles.UpdateEvent.Subscribe(OnUpdate)
            ObjectiveArrowTiles.CompleteEvent.Subscribe(Hide)

        NewUI := overlay:
            Slots := array:
                overlay_slot:
                    HorizontalAlignment := horizontal_alignment.Fill
                    VerticalAlignment := vertical_alignment.Fill
                    Widget := UserInterface.VerseWidget.WBP_Panel_Background_Transparent{}
                overlay_slot:
                    HorizontalAlignment := horizontal_alignment.Fill
                    VerticalAlignment := vertical_alignment.Fill
                    Padding := margin{Left := 20.0, Top := 20.0, Right := 20.0, Bottom := 20.0}
                    Widget := CreateArrowIcons()

        Player.AddWidgetToHUD of canvas_slot:
            Anchors := anchors{Minimum := vector2{X := 0.5, Y := 0.0}, Maximum := vector2{X := 0.5, Y := 0.0}}
            Offsets := margin{Left := 0.0, Top := 180.0, Right := 0.0, Bottom := 0.0}
            Alignment := vector2{X := 0.5, Y := 0.0}
            Widget := NewUI

        NewUI

    CreateArrowIcons():widget =
        stack_box:
            Orientation := orientation.Horizontal
            Slots := for (Index := 1..ObjectiveArrowTilesTarget):
                ArrowIcon := texture_block:
                    DefaultImage := UserInterface.Icons.UpArrow_64
                    DefaultDesiredSize := vector2{X := 40.0, Y := 40.0}

                set ArrowIcons += array{ArrowIcon}
                
                stack_box_slot:
                    HorizontalAlignment := horizontal_alignment.Fill
                    VerticalAlignment := vertical_alignment.Fill
                    Padding := margin{Left := (Index = 1) and 0.0 or 10.0, Top := 0.0, Right := 0.0, Bottom := 0.0}
                    Widget := ArrowIcon
            
    Uninitialize():void =
        for (EventsSubscription : EventsSubscriptions):
            EventsSubscription.Cancel()

        set EventsSubscriptions = array{}

    <# ----------------------------------------------------------------------------------------------------
        Get Functions
    ---------------------------------------------------------------------------------------------------- #>
    (Index:int).ToArrowIcon()<transacts>:texture =
        case (Index):
            0 => UserInterface.Icons.UpArrow_64
            1 => UserInterface.Icons.DownArrow_64
            2 => UserInterface.Icons.LeftArrow_64
            3 => UserInterface.Icons.RightArrow_64
            _ => UserInterface.Icons.UpArrow_64
            
    <# ----------------------------------------------------------------------------------------------------
        Update Functions
    ---------------------------------------------------------------------------------------------------- #>
    Show<override>(ArrowIndexes:[]int):void =
        set CurrentProgress = 0
        for (Index -> ArrowIndex : ArrowIndexes, ArrowIcon := ArrowIcons[Index]):
            ArrowIcon.SetTint(NamedColors.White)
            ArrowIcon.SetImage(ArrowIndex.ToArrowIcon())

        (super:)Show()

    Update(Progress:int):void =
        set CurrentProgress = Progress
        for (Index -> ArrowIcon : ArrowIcons):
            ArrowIcon.SetTint((Progress - 1) >= Index and ThemeColors.Success or NamedColors.White)
        
    <# ----------------------------------------------------------------------------------------------------
        Event Handlers
    ---------------------------------------------------------------------------------------------------- #>
    OnError():void =
        if (ArrowIcon := ArrowIcons[CurrentProgress]):
            ArrowIcon.SetTint(ThemeColors.Failure)

    OnUpdate(Progress:int):void =
        Update(Progress)