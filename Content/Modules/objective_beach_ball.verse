using { /Fortnite.com/Devices }
using { /Verse.org/Random }
using { /Verse.org/Simulation }
using { /UnrealEngine.com/Temporary/SpatialMath }

<# ----------------------------------------------------------------------------------------------------
    objective_beach_ball
---------------------------------------------------------------------------------------------------- #>
objective_beach_ball := class<concrete>(objective):
    Id<override>:objective_id = objective_id.BeachBall

    @editable
    CaptureArea:capture_area_device = capture_area_device{}

    @editable
    MapIndicator:map_indicator_device = map_indicator_device{}

    @editable
    ChannelEnable:channel_device = channel_device{}

    @editable
    ChannelDisable:channel_device = channel_device{}

    @editable
    BallSpawners:[]creative_prop = array{}

    var Score:int = 0

    <# ----------------------------------------------------------------------------------------------------
        Initialization
    ---------------------------------------------------------------------------------------------------- #>
    OnBegin<override>()<suspends>:void =
        StartEvent.Subscribe(OnStart)
        CaptureArea.Disable()
        if (MapIndicator.TeleportTo[MapIndicatorDisabledPosition, IdentityRotation()]) {}
        ChannelDisable.Transmit(false)
        CaptureArea.AreaIsScoredEvent.Subscribe(OnAreaIsScored)
    
    <# ----------------------------------------------------------------------------------------------------
        Get Functions
    ---------------------------------------------------------------------------------------------------- #>
    GetTarget()<transacts>:int =
        Max(1, GetActivePlayers().Length)

    <# ----------------------------------------------------------------------------------------------------
        Update Functions
    ---------------------------------------------------------------------------------------------------- #>
    RepositionCaptureArea()<suspends>:void =
        ArenaSize := GetActiveArenaSize()
        
        var Attempts:int = 1
        loop:
            Position := GetRandomPositionOnCircle(ArenaSize * 0.9)
            RaycastPosition := GetRaycast(Position)
            if (Attempts >= 30):
                break
            
            if:
                RaycastPosition.Z <= 10.0
                CaptureArea.TeleportTo[Position - vector3{Z := 50.0}, IdentityRotation()]
            then:
                CaptureArea.Enable()
                #CaptureArea.ActivateObjectivePulse()

                if (MapIndicator.TeleportTo[Position + vector3{Z := 100.0}, IdentityRotation()]) {}
                for (Player : GetActivePlayers()):
                    MapIndicator.ActivateObjectivePulse(Player)

                break

            set Attempts += 1
        
        ChannelEnable.Transmit(false)

    <# ----------------------------------------------------------------------------------------------------
        Event Handlers
    ---------------------------------------------------------------------------------------------------- #>
    OnStart():void =
        set Score = 0
        spawn{RepositionCaptureArea()}
        #TrackerInfo.Device.SetTarget(GetTarget())
        UpdateHUD(Score, GetTarget())

    OnStop<override>():void =
        CaptureArea.Disable()
        #CaptureArea.DeactivateObjectivePulse()

        if (MapIndicator.TeleportTo[MapIndicatorDisabledPosition, IdentityRotation()]) {}
        for (Player : GetActivePlayers()):
            MapIndicator.DeactivateObjectivePulse(Player)
    
        ChannelDisable.Transmit(false)

    OnAreaIsScored(Agent:agent):void =
        set Score += 1
        #TrackerInfo.Device.SetValue(Score)
        UpdateHUD(Score, GetTarget())
        CinematicSequence.SetPlaybackTime((Score * 1.0) / (GetTarget() * 1.0))

        if (Score >= GetTarget()):
            SuccessEvent.Signal()