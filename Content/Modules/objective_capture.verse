using { /Fortnite.com/Characters }
using { /Fortnite.com/Devices }
using { /Verse.org/Simulation }
using { /UnrealEngine.com/Temporary/SpatialMath }

<# ----------------------------------------------------------------------------------------------------
    objective_capture
---------------------------------------------------------------------------------------------------- #>
objective_capture := class<concrete>(objective):
    Id<override>:objective_id = objective_id.CapturePoint
    
    @editable
    VolumeActivator:volume_activator = volume_activator{}
    
    @editable
    CaptureArea:capture_area_device = capture_area_device{}

    @editable
    MapIndicator:map_indicator_device = map_indicator_device{}

    @editable
    CaptureTime:float = 30.0

    CancelObjectiveEvent:events(tuple()) = events(tuple()){}

    <# ----------------------------------------------------------------------------------------------------
        Initialization
    ---------------------------------------------------------------------------------------------------- #>
    OnBegin<override>()<suspends>:void =
        StartEvent.Subscribe(OnStart)
        CaptureArea.Disable()
        if (MapIndicator.TeleportTo[MapIndicatorDisabledPosition, IdentityRotation()]) {}
        VolumeActivator.OnBegin()

    <# ----------------------------------------------------------------------------------------------------
        Get Functions
    ---------------------------------------------------------------------------------------------------- #>
    GetCaptureTime()<transacts>:float =
        CaptureTime

    <# ----------------------------------------------------------------------------------------------------
        Update Functions
    ---------------------------------------------------------------------------------------------------- #>
    StartObjective()<suspends>:void =
        VolumeActivator.Enable(?OverrideActivationTime := option{GetCaptureTime()})

        #TrackerInfo.Device.SetTarget(Floor[CaptureTime] or 0)
        UpdateHUD(0, Floor[GetCaptureTime()] or 0)

        CaptureArea.Enable()
        #CaptureArea.ActivateObjectivePulse()
        if (MapIndicator.TeleportTo[DefaultArenaOrigin, IdentityRotation()]) {}
        for (Player : GetActivePlayers()):
            MapIndicator.ActivateObjectivePulse(Player)

        race:
            CancelObjectiveEvent.Await()

            block:
                VolumeActivator.ActivatedEvent.Await()
                SuccessEvent.Signal()

            loop:
                UpdatedValue := Round[VolumeActivator.GetProgress() * GetCaptureTime()] or 0
                #TrackerInfo.Device.SetValue(UpdatedValue)
                UpdateHUD(UpdatedValue, Floor[GetCaptureTime()] or 0)
                CinematicSequence.SetPlaybackTime((UpdatedValue * 1.0) / GetCaptureTime())
                Sleep(1.0)

    <# ----------------------------------------------------------------------------------------------------
        Event Handlers
    ---------------------------------------------------------------------------------------------------- #>
    OnStart():void =
        spawn{StartObjective()}
        
    OnStop<override>():void =
        CancelObjectiveEvent.Signal()
        CaptureArea.Disable()
        #CaptureArea.DeactivateObjectivePulse()
        if (MapIndicator.TeleportTo[MapIndicatorDisabledPosition, IdentityRotation()]) {}
        for (Player : GetActivePlayers()):
            MapIndicator.DeactivateObjectivePulse(Player)
        VolumeActivator.Disable()