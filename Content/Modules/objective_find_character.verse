using { /Fortnite.com/Characters }
using { /Fortnite.com/Devices }
using { /Verse.org/Random }
using { /Verse.org/Simulation }
using { /UnrealEngine.com/Temporary/SpatialMath }

<# ----------------------------------------------------------------------------------------------------
    objective_find_character
---------------------------------------------------------------------------------------------------- #>
objective_find_character := class<concrete>(objective):
    Id<override>:objective_id = objective_id.FindCharacter
    
    @editable
    VolumeActivator:volume_activator = volume_activator{}
    
    @editable
    Character:character_device = character_device{}
    
    @editable
    CrowdVolume:crowd_volume_device = crowd_volume_device{}

    var CharacterInitialTransform:transform = transform{}

    <# ----------------------------------------------------------------------------------------------------
        Initialization
    ---------------------------------------------------------------------------------------------------- #>
    OnBegin<override>()<suspends>:void =
        set CharacterInitialTransform = Character.GetTransform()

        StartEvent.Subscribe(OnStart)
        Character.Hide()
        Character.Disable()
        CrowdVolume.Disable()
        VolumeActivator.OnBegin()
        VolumeActivator.ActivatedEvent.Subscribe(OnActivated)

    <# ----------------------------------------------------------------------------------------------------
        Update Functions
    ---------------------------------------------------------------------------------------------------- #>
    RepositionCharacter()<suspends>:void =
        ArenaSize := GetActiveArenaSize()
        
        var Attempts:int = 1
        loop:
            Position := GetRandomPositionOnCircle(GetRandomFloat(0.0, ArenaSize * 0.9))
            RaycastPosition := GetRaycast(Position)
            if (Attempts >= 30):
                break
            
            if:
                RaycastPosition.Z <= 10.0
                Character.TeleportTo[Position, MakeRotationFromYawPitchRollDegrees(GetRandomFloat(0.0, 360.0), 0.0, 0.0)]
            then:
                Character.Enable()
                Character.Show()
                VolumeActivator.Enable(?TargetPosition := option{Position})
                break

            set Attempts += 1

    <# ----------------------------------------------------------------------------------------------------
        Event Handlers
    ---------------------------------------------------------------------------------------------------- #>
    OnStart():void =
        spawn{RepositionCharacter()}
        CrowdVolume.Enable()
        VolumeActivator.Enable()
        UpdateHUD(0, 1)
        
    OnStop<override>():void =
        if (Character.TeleportTo[CharacterInitialTransform]) {}
        Character.Hide()
        Character.Disable()
        CrowdVolume.Disable()
        VolumeActivator.Disable()

    OnActivated(Player:player, InVolumeActivator:volume_activator):void =
        SuccessEvent.Signal()