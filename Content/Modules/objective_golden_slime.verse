using { /Fortnite.com/AI }
using { /Fortnite.com/Characters }
using { /Fortnite.com/Devices }
using { /Verse.org/Simulation }
using { /UnrealEngine.com/Temporary/SpatialMath }

<# ----------------------------------------------------------------------------------------------------
    objective_golden_slime
---------------------------------------------------------------------------------------------------- #>
objective_golden_slime := class<concrete>(objective):
    Id<override>:objective_id = objective_id.GoldenSlime

    @editable
    Beacon:beacon_device = beacon_device{}

    @editable
    MapIndicator:map_indicator_device = map_indicator_device{}

    var MaybeNPCBehaviorBase:?npc_behavior_base = false

    <# ----------------------------------------------------------------------------------------------------
        Initialization
    ---------------------------------------------------------------------------------------------------- #>
    OnBegin<override>()<suspends>:void =
        Beacon.Disable()
        if (MapIndicator.TeleportTo[MapIndicatorDisabledPosition, IdentityRotation()]) {}
        StartEvent.Subscribe(OnStart)

    <# ----------------------------------------------------------------------------------------------------
        Update Functions
    ---------------------------------------------------------------------------------------------------- #>
    Objective()<suspends>:void =
        if (RandomPlayer := GetRandomAlivePlayer()?):
            MaybeNPCSpawner := SpawnNPC(npc_id.Golden, RandomPlayer, ?MaybePosition := option{DefaultArenaOrigin + vector3{Z := 300.0}})

            if (NPCSpawner := MaybeNPCSpawner?):
                NPCAgent := NPCSpawner.SpawnedEvent.Await()

                if (Character := NPCAgent.GetFortCharacter[], NPCBehaviorBase := npc_behavior_base[NPCAgent.GetNPCBehavior[]]):
                    set MaybeNPCBehaviorBase = option{NPCBehaviorBase}

                    if:
                        MapIndicator.TeleportTo[Character.GetTransform()]
                        Beacon.TeleportTo[Character.GetTransform()]

                    race:
                        StopEvent.Await()
                        loop{MapIndicator.MoveTo(Character.GetTransform(), ThreeTicks)}
                        loop{Beacon.MoveTo(Character.GetTransform(), ThreeTicks)}
                        block:
                            NPCBehaviorBase.EndEvent.Await()
                            SuccessEvent.Signal()

    <# ----------------------------------------------------------------------------------------------------
        Event Handlers
    ---------------------------------------------------------------------------------------------------- #>
    OnStart():void =
        #TrackerInfo.Device.SetTarget(1)
        UpdateHUD(0, 1)
        Beacon.Enable()
        spawn{Objective()}

    OnStop<override>():void =
        Beacon.Disable()
        if (MapIndicator.TeleportTo[MapIndicatorDisabledPosition, IdentityRotation()]) {}
     
        if (NPCBehaviorBase := MaybeNPCBehaviorBase?):
            NPCBehaviorBase.Despawn()
            set MaybeNPCBehaviorBase = false