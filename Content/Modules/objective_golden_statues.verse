using { /Fortnite.com/Characters }
using { /Fortnite.com/Devices }
using { /Fortnite.com/Devices/CreativeAnimation }
using { /Verse.org/Random }
using { /Verse.org/Simulation }
using { /UnrealEngine.com/Temporary/SpatialMath }
using { Utility }

# [Note] Deprecated.

<# ----------------------------------------------------------------------------------------------------
    objective_golden_statues
---------------------------------------------------------------------------------------------------- #>
objective_golden_statues := class<concrete>(objective):
    Id<override>:objective_id = objective_id.Collectible
    
    @editable
    PropManipulator:prop_manipulator_device = prop_manipulator_device{}
    
    @editable
    PropStatues:[]creative_prop = array{}

    CancelObjectiveEvent:events(tuple()) = events(tuple()){}

    var Progress:int = 0
    var InitialTransforms:[]transform = array{}
    var InitialScale:vector3 = UnitVector
    var PropHealths:[]int = array{}
    var SpawnedProps:[]creative_prop = array{}

    <# ----------------------------------------------------------------------------------------------------
        Initialization
    ---------------------------------------------------------------------------------------------------- #>
    OnBegin<override>()<suspends>:void =
        StartEvent.Subscribe(OnStart)

        set InitialTransforms = for (PropStatue : PropStatues):
            PropStatue.GetTransform()

        if (PropStatue := PropStatues[0]):
            set InitialScale = PropStatue.GetTransform().Scale
        
        PropManipulator.HarvestingEvent.Subscribe(OnHarvesting)
        PropManipulator.ResourceDepletionEvent.Subscribe(OnResourceDepletion)

    <# ----------------------------------------------------------------------------------------------------
        Get Functions
    ---------------------------------------------------------------------------------------------------- #>
    GetTarget()<transacts>:int =
        PlayerCount := GetActivePlayers().Length
        TargetCount := 3 + (PlayerCount - 1) * 3

    GetNearestIndex(Agent:agent)<decides><transacts>:int =
        Character := Agent.GetFortCharacter[]
        PlayerPosition := Character.GetTransform().Translation
        TargetPositions := for (PropStatue : PropStatues) {PropStatue.GetTransform().Translation}
        NearestIndex := GetNearestPosition(PlayerPosition, TargetPositions)?

    <# ----------------------------------------------------------------------------------------------------
        Update Functions
    ---------------------------------------------------------------------------------------------------- #>
    PositionStatues()<suspends>:void =
        RandomPositions := Shuffle(PoissonDiscSamplingGeneratePoints(1000.0, GetActiveArenaSize() * 0.9, ?MinRadius := 0.0))
   
        var PropIndex:int = 0
        for (RandomPosition : RandomPositions, PropIndex <= PropStatues.Length - 1):
            RaycastPosition := GetRaycast(RandomPosition)
            if (RaycastPosition.Z <= 25.0, PropStatue := PropStatues[PropIndex]):
                PropStatue.Show()
                TargetTransform := transform:
                    Translation := vector3{X := RandomPosition.X, Y := RandomPosition.Y, Z := 0.0}
                    Rotation := MakeRotationFromYawPitchRollDegrees(GetRandomFloat(0.0, 360.0), 0.0, 0.0)
                    Scale := InitialScale

                if (PropStatue.TeleportTo[TargetTransform]) {}

                MaybeSpawnedProp := SpawnVFX(vfx_id.Number, TargetTransform.Translation, ?Scale := vector3{X := 1.0, Y := 1.0, Z := 10.0}, ?Duration := 0.0)
                if (SpawnedProp := MaybeSpawnedProp?):
                    set SpawnedProps += array{SpawnedProp}

                set PropHealths += array{10}
                set PropIndex += 1

    Destroy(Prop:creative_prop)<suspends>:void =
        if (AnimationController := Prop.GetAnimationController[]):
            KeyFrames := array:
                keyframe_delta:
                    DeltaLocation := vector3{}
                    DeltaRotation := IdentityRotation()
                    DeltaScale := UnitVector * 1.5
                    Interpolation := InterpolationTypes.EaseOut
                    Time := 0.2
                keyframe_delta:
                    DeltaLocation := vector3{}
                    DeltaRotation := IdentityRotation()
                    DeltaScale := UnitVector * 0.1
                    Interpolation := InterpolationTypes.EaseOut
                    Time := 0.1

            AnimationController.SetAnimation(KeyFrames, ?Mode := animation_mode.OneShot)
            AnimationController.Play()
            AnimationController.MovementCompleteEvent.Await()

        SpawnVFX(vfx_id.Confetti, Prop.GetTransform().Translation, ?Duration := 3.0)
        
        CurrentTransform := Prop.GetTransform()
        TargetPosition := vector3:
            X := CurrentTransform.Translation.X
            Y := CurrentTransform.Translation.Y
            Z := -3000.0
        if (Prop.TeleportTo[TargetPosition, CurrentTransform.Rotation]) {}
        Prop.Hide()

        set Progress += 1
        UpdateHUD(Progress, GetTarget())
        CinematicSequence.SetPlaybackTime((Progress * 1.0) / (GetTarget() * 1.0))

        if (Progress >= GetTarget()):
            SuccessEvent.Signal()

    <# ----------------------------------------------------------------------------------------------------
        Event Handlers
    ---------------------------------------------------------------------------------------------------- #>
    OnStart():void =
        set Progress = 0
        set PropHealths = array{}
        set SpawnedProps = array{}
        
        PropManipulator.RestockResources()
        UpdateHUD(0, GetTarget())
        spawn{PositionStatues()}
        
    OnStop<override>():void =
        for (Index -> PropStatue : PropStatues, InitialTransform := InitialTransforms[Index], PropStatue.TeleportTo[InitialTransform]):
            PropStatue.Hide()

        for (SpawnedProp : SpawnedProps):
            SpawnedProp.Dispose()

    OnHarvesting(Agent:agent):void =
        if (Index := GetNearestIndex[Agent], SpawnedProp := SpawnedProps[Index]):
            var TargetTransform:transform = SpawnedProp.GetTransform()
            set TargetTransform.Scale = vector3{X := 1.0, Y := 1.0, Z := TargetTransform.Scale.Z - 1.0}
            if (SpawnedProp.TeleportTo[TargetTransform]) {}

    OnResourceDepletion(Agent:agent):void =
        if (Index := GetNearestIndex[Agent], PropStatue := PropStatues[Index], SpawnedProp := SpawnedProps[Index]):
            SpawnedProp.Dispose()
            spawn{Destroy(PropStatue)}