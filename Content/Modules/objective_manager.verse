using { /Fortnite.com/Devices }
using { /Verse.org/Random }
using { /Verse.org/Simulation }

<# ----------------------------------------------------------------------------------------------------
    Global Functions
---------------------------------------------------------------------------------------------------- #>
var GlobalObjectiveManager:weak_map(session, objective_manager) = map{}

GetObjectiveManager()<decides><transacts>:objective_manager = 
    GlobalObjectiveManager[GetSession()]

StartNextObjective():void =
    if (ObjectiveManager := GetObjectiveManager[]):
        ObjectiveManager.StartObjective()
        
StartObjective(Id:objective_id):void =
    if (ObjectiveManager := GetObjectiveManager[]):
        ObjectiveManager.Start(Id)

StopObjective():void =
    if (ObjectiveManager := GetObjectiveManager[]):
        ObjectiveManager.Stop()

GetNextObjective()<decides><transacts>:objective =
    GetObjectiveManager[].GetNextObjective[]

GetCurrentObjective()<transacts>:?objective =
    GetObjectiveManager[].MaybeCurrentObjective or false
        
<# ----------------------------------------------------------------------------------------------------
    objective_manager
---------------------------------------------------------------------------------------------------- #>
objective_manager := class(creative_device):
    @editable
    HUDMessageIntro:hud_message_device = hud_message_device{}
    
    @editable
    HUDMessageTimer:hud_message_device = hud_message_device{}
    
    @editable
    HUDMessageTimerMobile:hud_message_device = hud_message_device{}

    @editable
    CinematicSequence:cinematic_sequence_device = cinematic_sequence_device{}

    @editable
    ObjectiveBeachBall:objective_beach_ball = objective_beach_ball{}

    @editable
    ObjectiveCollect:objective_collect = objective_collect{}

    @editable
    ObjectiveGoldenSlime:objective_golden_slime = objective_golden_slime{}

    @editable
    ObjectiveCapture:objective_capture = objective_capture{}

    @editable
    ObjectiveNumberedTiles:objective_numbered_tiles = objective_numbered_tiles{}

    @editable
    ObjectiveArrowTiles:objective_arrow_tiles = objective_arrow_tiles{}
    
    @editable
    ObjectiveStorm:objective_storm = objective_storm{}
    
    @editable
    ObjectiveSnowStorm:objective_snow_storm = objective_snow_storm{}
        
    @editable
    ObjectiveFindCharacter:objective_find_character = objective_find_character{}
    
    @editable
    ObjectiveMultiCapturePoint:objective_multi_capture_point = objective_multi_capture_point{}

    ObjectiveSuccessEvent:events(tuple()) = events(tuple()){}
    StopEvent:event(void) = event(void){}

    var Objectives:[]objective = array{}
    var GameObjectives:[]objective = array{}
    var CompletedGameObjectives:[]objective = array{}
    var MaybeCurrentObjective:?objective = false

    <# ----------------------------------------------------------------------------------------------------
        Initialization
    ---------------------------------------------------------------------------------------------------- #>
    OnBegin<override>()<suspends>:void =
        if (set GlobalObjectiveManager[GetSession()] = Self) {}

        set Objectives = array:
            ObjectiveBeachBall
            ObjectiveCollect
            ObjectiveGoldenSlime
            ObjectiveCapture
            ObjectiveNumberedTiles
            ObjectiveArrowTiles
            ObjectiveStorm
            ObjectiveSnowStorm
            ObjectiveMultiCapturePoint
            ObjectiveFindCharacter

        for (Objective : Objectives):
            Objective.OnBegin()
            Objective.SuccessEvent.Subscribe(OnSuccess)
            Objective.CompleteEvent.Subscribe(OnComplete)
            set Objective.HUDMessageIntro = HUDMessageIntro
            set Objective.HUDMessageTimer = HUDMessageTimer
            set Objective.HUDMessageTimerMobile = HUDMessageTimerMobile
            set Objective.CinematicSequence = CinematicSequence

        WaitForGameManager()
        if (GameManager := GetGameManager[]):
            GameManager.GameStartedEvent.Subscribe(OnGameStarted)
            GameManager.GameEndedEvent.Subscribe(Stop)
            GameManager.StageEndedEvent.Subscribe(OnStageEnded)
            GameManager.BossFightStartedEvent.Subscribe(Stop)

    <# ----------------------------------------------------------------------------------------------------
        Get Functions
    ---------------------------------------------------------------------------------------------------- #>
    (objective_manager:)GetNextObjective()<decides><transacts>:objective =
        GameObjectives[Mod[GetCurrentStage() - 1, Objectives.Length]] or Objectives[GetRandomInt(0, Objectives.Length - 1)]

    <# ----------------------------------------------------------------------------------------------------
        Update Functions
    ---------------------------------------------------------------------------------------------------- #>
    StartObjective():void =
        if (NextObjective := (objective_manager:)GetNextObjective[]):
            spawn{NextObjective.Start()}
            set MaybeCurrentObjective = option{NextObjective}
            set CompletedGameObjectives += array{NextObjective}

    Start(Id:objective_id):void =
        if (CurrentObjective := MaybeCurrentObjective?):
            Stop()

        Objective := case (Id):
            objective_id.Collectible => ObjectiveCollect
            objective_id.GoldenSlime => ObjectiveGoldenSlime
            objective_id.BeachBall => ObjectiveBeachBall
            objective_id.CapturePoint => ObjectiveCapture
            objective_id.NumberedTiles => ObjectiveNumberedTiles
            objective_id.Storm => ObjectiveStorm
            objective_id.SnowStorm => ObjectiveSnowStorm
            objective_id.ArrowTiles => ObjectiveArrowTiles
            objective_id.FindCharacter => ObjectiveFindCharacter
            objective_id.MultiCapturePoint => ObjectiveMultiCapturePoint

        spawn{Objective.Start()}
        set MaybeCurrentObjective = option{Objective}

    Stop():void =
        StopEvent.Signal()

        if (CurrentObjective := MaybeCurrentObjective?):
            CurrentObjective.Stop()
            set MaybeCurrentObjective = false

    <# ----------------------------------------------------------------------------------------------------
        Event Handlers
    ---------------------------------------------------------------------------------------------------- #>
    OnGameStarted():void =
        set CompletedGameObjectives = array{}
        set GameObjectives = Shuffle(Objectives)
        set MaybeCurrentObjective = false

    OnSuccess():void =
        ObjectiveSuccessEvent.Signal()

    OnComplete():void =
        set MaybeCurrentObjective = false

    OnStageEnded():void =
        Stop()
        
        if (Mod[GetCurrentStage(), Objectives.Length] = 0):
            set GameObjectives = Shuffle(Objectives)