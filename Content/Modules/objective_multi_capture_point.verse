using { /Fortnite.com/Characters }
using { /Fortnite.com/Devices }
using { /Fortnite.com/Devices/CreativeAnimation }
using { /Verse.org/Random }
using { /Verse.org/Simulation }
using { /UnrealEngine.com/Temporary/SpatialMath }
using { Utility }

<# ----------------------------------------------------------------------------------------------------
    objective_multi_capture_point
---------------------------------------------------------------------------------------------------- #>
objective_multi_capture_point := class<concrete>(objective):
    Id<override>:objective_id = objective_id.MultiCapturePoint
    
    @editable
    VolumeActivators:[]volume_activator = array{}

    CancelObjectiveEvent:events(tuple()) = events(tuple()){}

    var Progress:int = 0

    <# ----------------------------------------------------------------------------------------------------
        Initialization
    ---------------------------------------------------------------------------------------------------- #>
    OnBegin<override>()<suspends>:void =
        StartEvent.Subscribe(OnStart)

        for (VolumeActivator : VolumeActivators):
            VolumeActivator.OnBegin()
            VolumeActivator.ActivatedEvent.Subscribe(OnActivated)

    <# ----------------------------------------------------------------------------------------------------
        Get Functions
    ---------------------------------------------------------------------------------------------------- #>
    GetTarget()<transacts>:int =
        PlayerCount := GetActivePlayers().Length
        TargetCount := 3 + (PlayerCount - 1) * 3

    <# ----------------------------------------------------------------------------------------------------
        Update Functions
    ---------------------------------------------------------------------------------------------------- #>
    PositionVolumeActivators()<suspends>:void =
        RandomPositions := Shuffle(PoissonDiscSamplingGeneratePoints(1000.0, GetActiveArenaSize() * 0.9, ?MinRadius := 0.0))
        
        var Index:int = 0
        for (RandomPosition : RandomPositions, Index <= VolumeActivators.Length - 1):
            RaycastPosition := GetRaycast(RandomPosition)
            if (RaycastPosition.Z <= 25.0, VolumeActivator := VolumeActivators[Index]):
                TargetPosition := vector3{X := RandomPosition.X, Y := RandomPosition.Y, Z := 0.0}
                VolumeActivator.Enable(?TargetPosition := option{TargetPosition})
                set Index += 1

    <# ----------------------------------------------------------------------------------------------------
        Event Handlers
    ---------------------------------------------------------------------------------------------------- #>
    OnStart():void =
        set Progress = 0
        UpdateHUD(0, GetTarget())
        spawn{PositionVolumeActivators()}
        
    OnStop<override>():void =
        for (VolumeActivator : VolumeActivators):
            VolumeActivator.Disable()

    OnActivated(Player:player, InVolumeActivator:volume_activator):void =
        set Progress += 1
        UpdateHUD(Progress, GetTarget())
        CinematicSequence.SetPlaybackTime((Progress * 1.0) / (GetTarget() * 1.0))
        InVolumeActivator.Disable()
        Player.PlayAudio(audio_id.CaptureSuccess)

        if (Progress >= GetTarget()):
            SuccessEvent.Signal()