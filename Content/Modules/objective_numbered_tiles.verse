using { /Fortnite.com/Devices }
using { /Verse.org/Random }
using { /Verse.org/Simulation }
using { /UnrealEngine.com/Temporary/SpatialMath }
using {Utility}

<# ----------------------------------------------------------------------------------------------------
    objective_numbered_tiles
---------------------------------------------------------------------------------------------------- #>
objective_numbered_tiles := class<concrete>(objective):
    Id<override>:objective_id = objective_id.NumberedTiles

    @editable
    ColorChangingTileRoots:[]creative_prop = array{}

    @editable
    ColorChangingTiles:[]color_changing_tiles_device = array{}

    var Progress:int = 0
    var TargetCount:int = 0
    var InitialTransforms:[]transform = array{}

    <# ----------------------------------------------------------------------------------------------------
        Initialization
    ---------------------------------------------------------------------------------------------------- #>
    OnBegin<override>()<suspends>:void =
        StartEvent.Subscribe(OnStart)
       
        set InitialTransforms = for (Index -> ColorChangingTile : ColorChangingTiles, ColorChangingTileRoot := ColorChangingTileRoots[Index]):
            ColorChangingTile.Disable()
            ColorChangingTile.Hide()
            spawn{AwaitActivated(ColorChangingTile, Index)}
            ColorChangingTileRoot.GetTransform()

    <# ----------------------------------------------------------------------------------------------------
        Get Functions
    ---------------------------------------------------------------------------------------------------- #>
    GetMaxSpawnDistanceBetween()<transacts>:float =
        600.0 + (GetActivePlayers().Length - 1) * 100.0

    GetMaxSpawnDistance()<transacts>:float =
        GetActiveArenaSize() * (0.4 + (0.167 * (GetActivePlayers().Length - 1)))

    <# ----------------------------------------------------------------------------------------------------
        Update Functions
    ---------------------------------------------------------------------------------------------------- #>
    RepositionsTiles()<suspends>:void =
        RandomPositions := Shuffle(PoissonDiscSamplingGeneratePoints(GetMaxSpawnDistanceBetween(), GetMaxSpawnDistance(), ?MinRadius := 0.0))
    
        var TileIndex:int = 0
        var TeleportSuccess:int = 0
        for (RandomPosition : RandomPositions, TileIndex < ColorChangingTiles.Length):
            RaycastPosition := GetRaycast(RandomPosition)
            if (RaycastPosition.Z <= 25.0, ColorChangingTileRoot := ColorChangingTileRoots[TileIndex], ColorChangingTile := ColorChangingTiles[TileIndex]):
                ColorChangingTile.Enable()
                ColorChangingTile.Reset()
                ColorChangingTile.Show()
                if (ColorChangingTileRoot.TeleportTo[vector3{X := RandomPosition.X, Y := RandomPosition.Y, Z := 0.0}, IdentityRotation()]):
                    set TeleportSuccess += 1

                set TileIndex += 1

        set TargetCount = TeleportSuccess
        UpdateHUD(Progress, TargetCount)
        ShowMessage(message_id.ObjectiveNumberedTilesProgress, "{Progress}", ?DisplayTime := 0.0)

    <# ----------------------------------------------------------------------------------------------------
        Event Handlers
    ---------------------------------------------------------------------------------------------------- #>
    OnStart():void =
        set Progress = 1

        #TrackerInfo.Device.SetTarget(TargetCount)
        UpdateHUD(Progress, ColorChangingTiles.Length)
        ShowMessage(message_id.ObjectiveNumberedTilesProgress, "{Progress}", ?DisplayTime := 0.0)
        spawn{RepositionsTiles()}

    OnStop<override>():void =
        HideMessage(message_id.ObjectiveNumberedTilesProgress)

        for (Index -> ColorChangingTile : ColorChangingTiles, ColorChangingTileRoot := ColorChangingTileRoots[Index], InitialTransform := InitialTransforms[Index]):
            ColorChangingTile.Disable()
            ColorChangingTile.Hide()
            if (ColorChangingTileRoot.TeleportTo[InitialTransform]) {}

    AwaitActivated(InColorChangingTile:color_changing_tiles_device, Index:int)<suspends>:void =
        loop:
            Listenable:listenable(agent) = InColorChangingTile.ActivatedEvent
            Listenable.Await()

            if (Index + 1 = Progress):
                set Progress += 1
                #TrackerInfo.Device.SetValue(Progress)
                UpdateMessage(message_id.ObjectiveNumberedTilesProgress, "{Progress}")
                CinematicSequence.SetPlaybackTime((Progress * 1.0) / (TargetCount * 1.0))

                if (Progress = TargetCount + 1):
                    SuccessEvent.Signal()
                else:
                    UpdateHUD(Progress, TargetCount)
                    for (Player : GetAllPlayers()):
                        Player.PlayAudio(audio_id.Activate)
            else:
                for (Player : GetAllPlayers()):
                    Player.PlayAudio(audio_id.Error)

                Sleep(1.0)
                InColorChangingTile.Reset()