using { /Fortnite.com/AI }
using { /Fortnite.com/Characters }
using { /Fortnite.com/Devices }
using { /Verse.org/Random }
using { /Verse.org/Simulation }
using { /UnrealEngine.com/Temporary/SpatialMath }

<# ----------------------------------------------------------------------------------------------------
    objective_snow_storm
---------------------------------------------------------------------------------------------------- #>
objective_snow_storm := class<concrete>(objective):
    Id<override>:objective_id = objective_id.SnowStorm

    @editable
    Character:character_device = character_device{}

    @editable
    MapIndicator:map_indicator_device = map_indicator_device{}

    @editable
    Beacon:beacon_device = beacon_device{}

    @editable
    VFXSpawner:vfx_spawner_device = vfx_spawner_device{}

    @editable
    MutatorZone:mutator_zone_device = mutator_zone_device{}

    @editable
    VolumeActivator:volume_activator = volume_activator{}

    @editable
    SnowStormDuration:float = 30.0

    var StartTime:float = 0.0
    
    <# ----------------------------------------------------------------------------------------------------
        Initialization
    ---------------------------------------------------------------------------------------------------- #>
    OnBegin<override>()<suspends>:void =
        Character.Hide()
        Beacon.Disable()
        if (MapIndicator.TeleportTo[MapIndicatorDisabledPosition, IdentityRotation()]) {}
        VFXSpawner.Disable()
        MutatorZone.Disable()
        VolumeActivator.OnBegin()
        VolumeActivator.ActivatedEvent.Subscribe(OnActivated)
        StartEvent.Subscribe(OnStart)

    Objective()<suspends>:void =
        race:
            StopEvent.Await()
            block:
                Sleep(GetTarget())
                SuccessEvent.Signal()
            loop:
                ElapsedTime := GetSimulationElapsedTime() - StartTime
                UpdateHUD(Floor[ElapsedTime] or 0, Floor[GetTarget()] or 30)
                CinematicSequence.SetPlaybackTime(ElapsedTime / GetTarget())
                Sleep(1.0)

    <# ----------------------------------------------------------------------------------------------------
        Get Function
    ---------------------------------------------------------------------------------------------------- #>
    GetTarget()<transacts>:float =
        SnowStormDuration

    GetIsActive()<transacts>:logic = 
        logic{GetSimulationElapsedTime() - StartTime <= GetTarget()}

    <# ----------------------------------------------------------------------------------------------------
        Event Handlers
    ---------------------------------------------------------------------------------------------------- #>
    OnStart():void =
        UpdateHUD(0, Floor[GetTarget()] or 30)
        Character.Show()
        VolumeActivator.Enable()
        Beacon.Enable()
        if (MapIndicator.TeleportTo[MapIndicatorEnabledPosition, IdentityRotation()]) {}
        for (Index -> Player : GetActivePlayers()):
            MapIndicator.ActivateObjectivePulse(Player)

    OnStop<override>():void =
        VFXSpawner.Disable()
        MutatorZone.Disable()

        Character.Hide()
        Beacon.Disable()
        VolumeActivator.Disable()
        for (Player : GetActivePlayers()):
            Player.ReactivateMovespeed()
            Player.StopAudio(audio_id.SnowStorm)

        if (MapIndicator.TeleportTo[MapIndicatorDisabledPosition, IdentityRotation()]) {}

        for (Player : GetAllPlayers()):
            MapIndicator.DeactivateObjectivePulse(Player)

    OnActivated(InPlayer:player, InVolumeActivator:volume_activator):void =
        set StartTime = GetSimulationElapsedTime()
        Character.Hide()
        Beacon.Disable()
        VolumeActivator.Disable()
        for (Player : GetActivePlayers()):
            Player.DeactivateMovespeed()
            Player.PlayAudio(audio_id.SnowStorm)

        if (MapIndicator.TeleportTo[MapIndicatorDisabledPosition, IdentityRotation()]) {}

        for (Player : GetAllPlayers()):
            MapIndicator.DeactivateObjectivePulse(Player)

        VFXSpawner.Enable()
        MutatorZone.Enable()
        spawn{Objective()}
        ChallengeAccepted.Signal()