using { /Verse.org/Simulation }

<# ----------------------------------------------------------------------------------------------------
    Constants
---------------------------------------------------------------------------------------------------- #>
WeaponUnlockCostIncrementRate:float = 0.25
WeaponUnlockBaseCost:float = 10000.0

<# ----------------------------------------------------------------------------------------------------
    Global Functions
---------------------------------------------------------------------------------------------------- #>
var PlayerDataMap:weak_map(player, player_data) = map{}

(Player:player).GetData()<decides><transacts>:player_data =
    Player.IsActive[]
    PlayerDataMap[Player]

<# ----------------------------------------------------------------------------------------------------
    Weapon Unlock
---------------------------------------------------------------------------------------------------- #>
(Player:player).IsWeaponUnlocked(Id:weapon_id)<transacts>:logic =
    if (StartingWeapons.Find[Id]):
        return true
    else:
        return logic{Player.GetData[].WeaponsUnlocked[Id.ToInt()]}

(Player:player).GetWeaponUnlockCost()<transacts>:float =
    (Floor[WeaponUnlockBaseCost * Pow((1.0 + WeaponUnlockCostIncrementRate), Player.GetWeaponsUnlocked() * 1.0)] or 0) * 1.0

(Player:player).GetWeaponsUnlocked()<transacts>:int =
    Player.GetData[].WeaponsUnlocked.Length or 0

(Player:player).CanAffordWeaponUnlock()<transacts>:logic =
    logic{Player.GetGoldAvailable(currency_id.Gold) >= Player.GetWeaponUnlockCost()}

(Player:player).UnlockWeapon(Id:weapon_id):void =
    if (not Player.CanAffordWeaponUnlock()?) then return
    Player.RemoveCurrency(currency_id.Gold, Player.GetWeaponUnlockCost())

    if:
        CurrentPlayerData := Player.GetData[]
        
        Player.GetWeaponUnlockCost()
        var UpdatedWeaponsUnlocked:[int]logic = CurrentPlayerData.WeaponsUnlocked
        set UpdatedWeaponsUnlocked[Id.ToInt()] = true

        set PlayerDataMap[Player] = player_data:
            MakePlayerData<constructor>(CurrentPlayerData)
            WeaponsUnlocked := UpdatedWeaponsUnlocked

<# ----------------------------------------------------------------------------------------------------
    Arena Progress
---------------------------------------------------------------------------------------------------- #>
(Player:player).AddArenaProgress():void =
    if:
        CurrentPlayerData := Player.GetData[]
        
        set PlayerDataMap[Player] = player_data:
            MakePlayerData<constructor>(CurrentPlayerData)
            ArenaProgress := Min(CurrentPlayerData.ArenaProgress + 1, GetMaxArenaLevel())

(Player:player).SetArenaProgress(ArenaProgress:int):void =
    if:
        CurrentPlayerData := Player.GetData[]

        set PlayerDataMap[Player] = player_data:
            MakePlayerData<constructor>(CurrentPlayerData)
            ArenaProgress := ArenaProgress

(Player:player).GetArenaProgress()<transacts>:int =
    Player.GetData[].ArenaProgress or 0

(Player:player).SetEndlessModeProgress(ArenaIndex:int, Stage:int):void =
    if:
        CurrentPlayerData := Player.GetData[]
        Delta := ArenaIndex - (CurrentPlayerData.EndlessModeProgress.Length - 1)
        Delta > 0
        UpdatedEndlessModeProgress := CurrentPlayerData.EndlessModeProgress + for (Index := 1..Delta) {0}

        set PlayerDataMap[Player] = player_data:
            MakePlayerData<constructor>(CurrentPlayerData)
            EndlessModeProgress := UpdatedEndlessModeProgress

    if:
        CurrentPlayerData := Player.GetData[]
        var UpdatedEndlessModeProgress:[]int = CurrentPlayerData.EndlessModeProgress
        UpdatedProgress := Stage - 1
        UpdatedProgress > UpdatedEndlessModeProgress[ArenaIndex]
        set UpdatedEndlessModeProgress[ArenaIndex] = UpdatedProgress

        set PlayerDataMap[Player] = player_data:
            MakePlayerData<constructor>(CurrentPlayerData)
            EndlessModeProgress := UpdatedEndlessModeProgress

(Player:player).GetEndlessModeProgress(ArenaIndex:int)<transacts>:int =
    Player.GetData[].EndlessModeProgress[ArenaIndex] or 0

<# ----------------------------------------------------------------------------------------------------
    player_daily_quests
---------------------------------------------------------------------------------------------------- #>
player_daily_quests := class<final><persistable><computes>:
    LastQuestAssignedTime:float = 0.0
    LastRerollTime:float = 0.0 # [Note] Used as reroll count.
    ActiveQuests:[]daily_quest = array{}

MakePlayerDailyQuests<constructor>(UpdatedPlayerDailyQuests:player_daily_quests)<transacts> := player_daily_quests:
    LastQuestAssignedTime := UpdatedPlayerDailyQuests.LastQuestAssignedTime
    LastRerollTime := UpdatedPlayerDailyQuests.LastRerollTime
    ActiveQuests := UpdatedPlayerDailyQuests.ActiveQuests

daily_quest := class<final><persistable><computes>:
    Id:int
    Tier:int = 0
    Progress:float = 0.0
    MaybeWeaponId:?int = false

MakeDailyQuest<constructor>(UpdatedMakeDailyQuest:daily_quest)<transacts> := daily_quest:
    Id := UpdatedMakeDailyQuest.Id
    Tier := UpdatedMakeDailyQuest.Tier
    Progress := UpdatedMakeDailyQuest.Progress
    MaybeWeaponId := UpdatedMakeDailyQuest.MaybeWeaponId

<# ----------------------------------------------------------------------------------------------------
    player_data
---------------------------------------------------------------------------------------------------- #>
player_data := class<final><persistable><computes>:
    Version:int = 0
    News:int = 0
    Settings:player_settings = player_settings{}
    (player_data:)PowerUps:[int]int = map{}
    Currencies:[int]float = map{}
    ArenaProgress:int = 0
    WeaponEquipped:int = 2
    WeaponProgress:[int]int = map{}
    WeaponsUnlocked:[int]logic = map{}
    DailyQuests:player_daily_quests = player_daily_quests{}
    EndlessModeProgress:[]int = array{}
    Gears:[int]player_gear = map{}
    Achievements:[]float = array{}

MakePlayerData<constructor>(UpdatedPlayerData:player_data)<transacts> := player_data:
    Version := UpdatedPlayerData.Version
    News := UpdatedPlayerData.News
    Settings := UpdatedPlayerData.Settings
    PowerUps := UpdatedPlayerData.PowerUps
    Currencies := UpdatedPlayerData.Currencies
    ArenaProgress := UpdatedPlayerData.ArenaProgress
    WeaponEquipped := UpdatedPlayerData.WeaponEquipped
    WeaponProgress := UpdatedPlayerData.WeaponProgress
    WeaponsUnlocked := UpdatedPlayerData.WeaponsUnlocked
    DailyQuests := UpdatedPlayerData.DailyQuests
    EndlessModeProgress := UpdatedPlayerData.EndlessModeProgress
    Gears := UpdatedPlayerData.Gears
    Achievements := UpdatedPlayerData.Achievements