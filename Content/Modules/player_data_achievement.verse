using { /Verse.org/Simulation }

<# ----------------------------------------------------------------------------------------------------
    Data Definitions
---------------------------------------------------------------------------------------------------- #>
achievement_id<public> := enum:
    EliminationSlimeForest
    EliminationSlimeForestElite
    EliminationSlimeForestGeneral
    EliminationSlimeForestMinibossRed
    EliminationSlimeForestMinibossBlue
    EliminationSlimeForestMinibossPurple
    EliminationSlimeForestMinibossOrange
    EliminationSlimeForestMinibossYellow
    EliminationSlimeForestBoss
    
    EliminationSlimeWasteland
    EliminationSlimeWastelandElite
    EliminationSlimeWastelandGeneral
    EliminationSlimeWastelandMinibossRed
    EliminationSlimeWastelandMinibossBlue
    EliminationSlimeWastelandMinibossPurple
    EliminationSlimeWastelandMinibossOrange
    EliminationSlimeWastelandMinibossYellow
    EliminationSlimeWastelandBoss
    
    EliminationSlimeElysium
    EliminationSlimeElysiumElite
    EliminationSlimeElysiumGeneral
    EliminationSlimeElysiumMinibossRed
    EliminationSlimeElysiumMinibossBlue
    EliminationSlimeElysiumMinibossPurple
    EliminationSlimeElysiumMinibossOrange
    EliminationSlimeElysiumMinibossYellow
    EliminationSlimeElysiumBoss

    EliminationSlimeAtlantis
    EliminationSlimeAtlantisElite
    EliminationSlimeAtlantisGeneral
    EliminationSlimeAtlantisMinibossRed
    EliminationSlimeAtlantisMinibossBlue
    EliminationSlimeAtlantisMinibossPurple
    EliminationSlimeAtlantisMinibossOrange
    EliminationSlimeAtlantisMinibossYellow
    EliminationSlimeAtlantisBoss
    
    EliminationSlimeSnow
    EliminationSlimeSnowElite
    EliminationSlimeSnowGeneral
    EliminationSlimeSnowMinibossRed
    EliminationSlimeSnowMinibossBlue
    EliminationSlimeSnowMinibossPurple
    EliminationSlimeSnowMinibossOrange
    EliminationSlimeSnowMinibossYellow
    EliminationSlimeSnowBoss
    
    EliminationSlimeCyber
    EliminationSlimeCyberElite
    EliminationSlimeCyberGeneral
    EliminationSlimeCyberMinibossRed
    EliminationSlimeCyberMinibossBlue
    EliminationSlimeCyberMinibossPurple
    EliminationSlimeCyberMinibossOrange
    EliminationSlimeCyberMinibossYellow
    EliminationSlimeCyberBoss

    EliminationSlimeNinja
    EliminationSlimeNinjaElite
    EliminationSlimeNinjaGeneral
    EliminationSlimeNinjaMinibossRed
    EliminationSlimeNinjaMinibossBlue
    EliminationSlimeNinjaMinibossPurple
    EliminationSlimeNinjaMinibossOrange
    EliminationSlimeNinjaMinibossYellow
    EliminationSlimeNinjaBoss
    
    EliminationSlimeShogun
    EliminationSlimeShogunElite
    EliminationSlimeShogunGeneral
    EliminationSlimeShogunMinibossRed
    EliminationSlimeShogunMinibossBlue
    EliminationSlimeShogunMinibossPurple
    EliminationSlimeShogunMinibossOrange
    EliminationSlimeShogunMinibossYellow
    EliminationSlimeShogunBoss
    
    EliminationSlimeFrog
    EliminationSlimeFrogElite
    EliminationSlimeFrogGeneral
    EliminationSlimeFrogMinibossRed
    EliminationSlimeFrogMinibossBlue
    EliminationSlimeFrogMinibossPurple
    EliminationSlimeFrogMinibossOrange
    EliminationSlimeFrogMinibossYellow
    EliminationSlimeFrogBoss

(Id:achievement_id).ToInt()<transacts>:int =
    case (Id):
        achievement_id.EliminationSlimeForest => 0
        achievement_id.EliminationSlimeForestElite => 1
        achievement_id.EliminationSlimeForestGeneral => 2
        achievement_id.EliminationSlimeForestMinibossRed => 3
        achievement_id.EliminationSlimeForestMinibossBlue => 4
        achievement_id.EliminationSlimeForestMinibossPurple => 5
        achievement_id.EliminationSlimeForestMinibossOrange => 6
        achievement_id.EliminationSlimeForestMinibossYellow => 7
        achievement_id.EliminationSlimeForestBoss => 8

        achievement_id.EliminationSlimeWasteland => 9
        achievement_id.EliminationSlimeWastelandElite => 10
        achievement_id.EliminationSlimeWastelandGeneral => 11
        achievement_id.EliminationSlimeWastelandMinibossRed => 12
        achievement_id.EliminationSlimeWastelandMinibossBlue => 13
        achievement_id.EliminationSlimeWastelandMinibossPurple => 14
        achievement_id.EliminationSlimeWastelandMinibossOrange => 15
        achievement_id.EliminationSlimeWastelandMinibossYellow => 16
        achievement_id.EliminationSlimeWastelandBoss => 17

        achievement_id.EliminationSlimeElysium => 18
        achievement_id.EliminationSlimeElysiumElite => 19
        achievement_id.EliminationSlimeElysiumGeneral => 20
        achievement_id.EliminationSlimeElysiumMinibossRed => 21
        achievement_id.EliminationSlimeElysiumMinibossBlue => 22
        achievement_id.EliminationSlimeElysiumMinibossPurple => 23
        achievement_id.EliminationSlimeElysiumMinibossOrange => 24
        achievement_id.EliminationSlimeElysiumMinibossYellow => 25
        achievement_id.EliminationSlimeElysiumBoss => 26

        achievement_id.EliminationSlimeAtlantis => 27
        achievement_id.EliminationSlimeAtlantisElite => 28
        achievement_id.EliminationSlimeAtlantisGeneral => 29
        achievement_id.EliminationSlimeAtlantisMinibossRed => 30
        achievement_id.EliminationSlimeAtlantisMinibossBlue => 31
        achievement_id.EliminationSlimeAtlantisMinibossPurple => 32
        achievement_id.EliminationSlimeAtlantisMinibossOrange => 33
        achievement_id.EliminationSlimeAtlantisMinibossYellow => 34
        achievement_id.EliminationSlimeAtlantisBoss => 35

        achievement_id.EliminationSlimeSnow => 36
        achievement_id.EliminationSlimeSnowElite => 37
        achievement_id.EliminationSlimeSnowGeneral => 38
        achievement_id.EliminationSlimeSnowMinibossRed => 39
        achievement_id.EliminationSlimeSnowMinibossBlue => 40
        achievement_id.EliminationSlimeSnowMinibossPurple => 41
        achievement_id.EliminationSlimeSnowMinibossOrange => 42
        achievement_id.EliminationSlimeSnowMinibossYellow => 43
        achievement_id.EliminationSlimeSnowBoss => 44

        achievement_id.EliminationSlimeCyber => 45
        achievement_id.EliminationSlimeCyberElite => 46
        achievement_id.EliminationSlimeCyberGeneral => 47
        achievement_id.EliminationSlimeCyberMinibossRed => 48
        achievement_id.EliminationSlimeCyberMinibossBlue => 49
        achievement_id.EliminationSlimeCyberMinibossPurple => 50
        achievement_id.EliminationSlimeCyberMinibossOrange => 51
        achievement_id.EliminationSlimeCyberMinibossYellow => 52
        achievement_id.EliminationSlimeCyberBoss => 53
        
        achievement_id.EliminationSlimeNinja => 54
        achievement_id.EliminationSlimeNinjaElite => 55
        achievement_id.EliminationSlimeNinjaGeneral => 56
        achievement_id.EliminationSlimeNinjaMinibossRed => 57
        achievement_id.EliminationSlimeNinjaMinibossBlue => 58
        achievement_id.EliminationSlimeNinjaMinibossPurple => 59
        achievement_id.EliminationSlimeNinjaMinibossOrange => 60
        achievement_id.EliminationSlimeNinjaMinibossYellow => 61
        achievement_id.EliminationSlimeNinjaBoss => 62

        achievement_id.EliminationSlimeShogun => 63
        achievement_id.EliminationSlimeShogunElite => 64
        achievement_id.EliminationSlimeShogunGeneral => 65
        achievement_id.EliminationSlimeShogunMinibossRed => 66
        achievement_id.EliminationSlimeShogunMinibossBlue => 67
        achievement_id.EliminationSlimeShogunMinibossPurple => 68
        achievement_id.EliminationSlimeShogunMinibossOrange => 69
        achievement_id.EliminationSlimeShogunMinibossYellow => 70
        achievement_id.EliminationSlimeShogunBoss => 71

        achievement_id.EliminationSlimeFrog => 72
        achievement_id.EliminationSlimeFrogElite => 73
        achievement_id.EliminationSlimeFrogGeneral => 74
        achievement_id.EliminationSlimeFrogMinibossRed => 75
        achievement_id.EliminationSlimeFrogMinibossBlue => 76
        achievement_id.EliminationSlimeFrogMinibossPurple => 77
        achievement_id.EliminationSlimeFrogMinibossOrange => 78
        achievement_id.EliminationSlimeFrogMinibossYellow => 79
        achievement_id.EliminationSlimeFrogBoss => 80

(Id:int).ToAchievementId()<transacts>:achievement_id =
    case (Id):
        0 => achievement_id.EliminationSlimeForest
        1 => achievement_id.EliminationSlimeForestElite
        2 => achievement_id.EliminationSlimeForestGeneral
        3 => achievement_id.EliminationSlimeForestMinibossRed
        4 => achievement_id.EliminationSlimeForestMinibossBlue
        5 => achievement_id.EliminationSlimeForestMinibossPurple
        6 => achievement_id.EliminationSlimeForestMinibossOrange
        7 => achievement_id.EliminationSlimeForestMinibossYellow
        8 => achievement_id.EliminationSlimeForestBoss

        9 => achievement_id.EliminationSlimeWasteland
        10 => achievement_id.EliminationSlimeWastelandElite
        11 => achievement_id.EliminationSlimeWastelandGeneral
        12 => achievement_id.EliminationSlimeWastelandMinibossRed
        13 => achievement_id.EliminationSlimeWastelandMinibossBlue
        14 => achievement_id.EliminationSlimeWastelandMinibossPurple
        15 => achievement_id.EliminationSlimeWastelandMinibossOrange
        16 => achievement_id.EliminationSlimeWastelandMinibossYellow
        17 => achievement_id.EliminationSlimeWastelandBoss

        18 => achievement_id.EliminationSlimeElysium
        19 => achievement_id.EliminationSlimeElysiumElite
        20 => achievement_id.EliminationSlimeElysiumGeneral
        21 => achievement_id.EliminationSlimeElysiumMinibossRed
        22 => achievement_id.EliminationSlimeElysiumMinibossBlue
        23 => achievement_id.EliminationSlimeElysiumMinibossPurple
        24 => achievement_id.EliminationSlimeElysiumMinibossOrange
        25 => achievement_id.EliminationSlimeElysiumMinibossYellow
        26 => achievement_id.EliminationSlimeElysiumBoss

        27 => achievement_id.EliminationSlimeAtlantis
        28 => achievement_id.EliminationSlimeAtlantisElite
        29 => achievement_id.EliminationSlimeAtlantisGeneral
        30 => achievement_id.EliminationSlimeAtlantisMinibossRed
        31 => achievement_id.EliminationSlimeAtlantisMinibossBlue
        32 => achievement_id.EliminationSlimeAtlantisMinibossPurple
        33 => achievement_id.EliminationSlimeAtlantisMinibossOrange
        34 => achievement_id.EliminationSlimeAtlantisMinibossYellow
        35 => achievement_id.EliminationSlimeAtlantisBoss

        36 => achievement_id.EliminationSlimeSnow
        37 => achievement_id.EliminationSlimeSnowElite
        38 => achievement_id.EliminationSlimeSnowGeneral
        39 => achievement_id.EliminationSlimeSnowMinibossRed
        40 => achievement_id.EliminationSlimeSnowMinibossBlue
        41 => achievement_id.EliminationSlimeSnowMinibossPurple
        42 => achievement_id.EliminationSlimeSnowMinibossOrange
        43 => achievement_id.EliminationSlimeSnowMinibossYellow
        44 => achievement_id.EliminationSlimeSnowBoss

        45 => achievement_id.EliminationSlimeCyber
        46 => achievement_id.EliminationSlimeCyberElite
        47 => achievement_id.EliminationSlimeCyberGeneral
        48 => achievement_id.EliminationSlimeCyberMinibossRed
        49 => achievement_id.EliminationSlimeCyberMinibossBlue
        50 => achievement_id.EliminationSlimeCyberMinibossPurple
        51 => achievement_id.EliminationSlimeCyberMinibossOrange
        52 => achievement_id.EliminationSlimeCyberMinibossYellow
        53 => achievement_id.EliminationSlimeCyberBoss
        
        54 => achievement_id.EliminationSlimeNinja
        55 => achievement_id.EliminationSlimeNinjaElite
        56 => achievement_id.EliminationSlimeNinjaGeneral
        57 => achievement_id.EliminationSlimeNinjaMinibossRed
        58 => achievement_id.EliminationSlimeNinjaMinibossBlue
        59 => achievement_id.EliminationSlimeNinjaMinibossPurple
        60 => achievement_id.EliminationSlimeNinjaMinibossOrange
        61 => achievement_id.EliminationSlimeNinjaMinibossYellow
        62 => achievement_id.EliminationSlimeNinjaBoss

        63 => achievement_id.EliminationSlimeShogun
        64 => achievement_id.EliminationSlimeShogunElite
        65 => achievement_id.EliminationSlimeShogunGeneral
        66 => achievement_id.EliminationSlimeShogunMinibossRed
        67 => achievement_id.EliminationSlimeShogunMinibossBlue
        68 => achievement_id.EliminationSlimeShogunMinibossPurple
        69 => achievement_id.EliminationSlimeShogunMinibossOrange
        70 => achievement_id.EliminationSlimeShogunMinibossYellow
        71 => achievement_id.EliminationSlimeShogunBoss

        72 => achievement_id.EliminationSlimeFrog
        73 => achievement_id.EliminationSlimeFrogElite
        74 => achievement_id.EliminationSlimeFrogGeneral
        75 => achievement_id.EliminationSlimeFrogMinibossRed
        76 => achievement_id.EliminationSlimeFrogMinibossBlue
        77 => achievement_id.EliminationSlimeFrogMinibossPurple
        78 => achievement_id.EliminationSlimeFrogMinibossOrange
        79 => achievement_id.EliminationSlimeFrogMinibossYellow
        80 => achievement_id.EliminationSlimeFrogBoss

        _ => achievement_id.EliminationSlimeForest

<# ----------------------------------------------------------------------------------------------------
    Global Functions
---------------------------------------------------------------------------------------------------- #>
(Player:player).AddAchievement(Id:achievement_id, Amount:float):void =
    if:
        CurrentPlayerData := Player.GetData[]
        Delta := Id.ToInt() - (CurrentPlayerData.Achievements.Length - 1)
        Delta > 0
        UpdatedAchievements := CurrentPlayerData.Achievements + for (Index := 1..Delta) {0.0}

        set PlayerDataMap[Player] = player_data:
            MakePlayerData<constructor>(CurrentPlayerData)
            Achievements := UpdatedAchievements

    if:
        CurrentPlayerData := Player.GetData[]
        var UpdatedAchievements:[]float = CurrentPlayerData.Achievements
        UpdatedProgress := UpdatedAchievements[Id.ToInt()] + Amount
        set UpdatedAchievements[Id.ToInt()] = UpdatedProgress

        set PlayerDataMap[Player] = player_data:
            MakePlayerData<constructor>(CurrentPlayerData)
            Achievements := UpdatedAchievements

(Player:player).GetAchievement(Id:achievement_id)<transacts>:float =
    Player.GetData[].Achievements[Id.ToInt()] or 0.0