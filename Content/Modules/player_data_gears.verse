using { /Verse.org/Simulation }


<# ----------------------------------------------------------------------------------------------------
    player_data_gears
---------------------------------------------------------------------------------------------------- #>
player_gear := class<final><persistable><computes>:
    Evolution:int = 0
    Level:int = 1
    Tier:int = 1
    Stars:int = 0
    Quantity:int = 0

MakePlayerGear<constructor>(UpdatedPlayerGear:player_gear)<transacts> := player_gear:
    Evolution := UpdatedPlayerGear.Evolution
    Level := UpdatedPlayerGear.Level
    Tier := UpdatedPlayerGear.Tier
    Stars := UpdatedPlayerGear.Stars
    Quantity := UpdatedPlayerGear.Quantity

<# ----------------------------------------------------------------------------------------------------
    Global Functions
---------------------------------------------------------------------------------------------------- #>
(Player:player).GetGear()<transacts>:[int]player_gear =
    Player.GetData[].Gears or map{}

(Player:player).GetGear(Slot:gear_slot)<decides><transacts>:player_gear = 
    Player.GetGear[Slot.ToInt()]

(Player:player).GetGear(SlotIndex:int)<decides><transacts>:player_gear = 
    Player.GetData[].Gears[SlotIndex]

(Player:player).GetGearQuantity(Slot:gear_slot)<transacts>:int = 
    Player.GetGear[Slot.ToInt()].Quantity or 0

(Player:player).RemoveGearQuantity(Slot:gear_slot, Amount:int):void = 
    if:
        CurrentPlayerData := Player.GetData[]
        CurrentGear := CurrentPlayerData.Gears[Slot.ToInt()]
        UpdatedGear := player_gear:
            MakePlayerGear<constructor>(CurrentGear)
            Quantity := CurrentGear.Quantity - Amount

        var UpdatedGears:[int]player_gear = CurrentPlayerData.Gears
        set UpdatedGears[Slot.ToInt()] = UpdatedGear

        set PlayerDataMap[Player] = player_data:
            MakePlayerData<constructor>(CurrentPlayerData)
            Gears := UpdatedGears

(Player:player).GrantGearRewards(Rewards:[int]int):void =
    if:
        CurrentPlayerData := Player.GetData[]
        var UpdatedGears:[int]player_gear = CurrentPlayerData.Gears
        
        for (GearSlotIndex -> Reward : Rewards):
            if:
                not UpdatedGears[GearSlotIndex]
                set UpdatedGears[GearSlotIndex] = player_gear{Quantity := Reward}
            else:
                UpdatedGear := player_gear:
                    MakePlayerGear<constructor>(CurrentPlayerData.Gears[GearSlotIndex])
                    Quantity := CurrentPlayerData.Gears[GearSlotIndex].Quantity + Reward

                set UpdatedGears[GearSlotIndex] = UpdatedGear

        set PlayerDataMap[Player] = player_data:
            MakePlayerData<constructor>(CurrentPlayerData)
            Gears := UpdatedGears

(Player:player).UpgradeGear(Slot:gear_slot, Amount:int):void =
    if:
        CurrentPlayerData := Player.GetData[]
        CurrentGear := CurrentPlayerData.Gears[Slot.ToInt()]
        UpdatedGear := player_gear:
            MakePlayerGear<constructor>(CurrentGear)
            Level := CurrentGear.Level + Amount

        var UpdatedGears:[int]player_gear = CurrentPlayerData.Gears
        set UpdatedGears[Slot.ToInt()] = UpdatedGear

        set PlayerDataMap[Player] = player_data:
            MakePlayerData<constructor>(CurrentPlayerData)
            Gears := UpdatedGears

(Player:player).FuseGear(Slot:gear_slot):void =
    if:
        CurrentPlayerData := Player.GetData[]
        CurrentGear := CurrentPlayerData.Gears[Slot.ToInt()]
        UpdatedGear := player_gear:
            MakePlayerGear<constructor>(CurrentGear)
            Tier := (CurrentGear.Stars < MaxGearStars) and CurrentGear.Tier or CurrentGear.Tier + 1
            Stars := (CurrentGear.Stars < MaxGearStars) and CurrentGear.Stars + 1 or 0

        var UpdatedGears:[int]player_gear = CurrentPlayerData.Gears
        set UpdatedGears[Slot.ToInt()] = UpdatedGear

        set PlayerDataMap[Player] = player_data:
            MakePlayerData<constructor>(CurrentPlayerData)
            Gears := UpdatedGears

(Player:player).EvolveGear(Slot:gear_slot):void =
    if:
        CurrentPlayerData := Player.GetData[]
        CurrentGear := CurrentPlayerData.Gears[Slot.ToInt()]
        UpdatedGear := player_gear:
            Evolution := CurrentGear.Evolution + 1
            Quantity := CurrentGear.Quantity

        var UpdatedGears:[int]player_gear = CurrentPlayerData.Gears
        set UpdatedGears[Slot.ToInt()] = UpdatedGear

        set PlayerDataMap[Player] = player_data:
            MakePlayerData<constructor>(CurrentPlayerData)
            Gears := UpdatedGears