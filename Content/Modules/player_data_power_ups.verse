using { /Verse.org/Simulation }

<# ----------------------------------------------------------------------------------------------------
    player_data_power_ups
---------------------------------------------------------------------------------------------------- #>
(Player:player).GetPowerUpLevel(Id:power_up_id)<transacts>:int =
    Player.GetData[].PowerUps[Id.ToInt()] or 0

(Player:player).GetPowerUpValue(Id:power_up_id)<transacts>:float =
    if (PowerUpInfo := GetPowerUpInfo[Id]):
        PowerUpInfo.BaseValue + Player.GetPowerUpLevel(Id) * PowerUpInfo.IncrementValue
    else:
        0.0

CalculatePowerUpCost(Id:power_up_id, Level:int)<decides><transacts>:float =
    PowerUpInfo := GetPowerUpInfo[Id]
    Floor[Pow(Level * 1.0, PowerUpInfo.CostFactor * 1.0) * PowerUpInfo.BaseCost] * 1.0

(Player:player).GetTotalPowerUpCost()<transacts>:float =
    var TotalPowerUpCost:float = 0.0
    for (PowerUpId : GetPowerUpIds()):
        for (PowerUpLevel := 1..Player.GetPowerUpLevel(PowerUpId), PowerUpCost := CalculatePowerUpCost[PowerUpId, PowerUpLevel]):
            set TotalPowerUpCost += PowerUpCost
    TotalPowerUpCost

(Player:player).AddPowerUp(Id:power_up_id):void =
    if:
        CurrentPlayerData := Player.GetData[]
        var UpdatedPowerUps:[int]int = CurrentPlayerData.PowerUps
        set UpdatedPowerUps[Id.ToInt()] = UpdatedPowerUps[Id.ToInt()] + 1 or 1
        set PlayerDataMap[Player] = player_data:
            MakePlayerData<constructor>(CurrentPlayerData)
            PowerUps := UpdatedPowerUps
    then:
        Player.UpdatePowerUpCosts()

(Player:player).RefundPowerUp(Id:power_up_id):void =
    if:
        CurrentPlayerData := Player.GetData[]
        var UpdatedPowerUps:[int]int = CurrentPlayerData.PowerUps
        set UpdatedPowerUps[Id.ToInt()] = 0

        set PlayerDataMap[Player] = player_data:
            MakePlayerData<constructor>(CurrentPlayerData)
            PowerUps := UpdatedPowerUps
    then:
        Player.UpdatePowerUpCosts()

(Player:player).RefundOnePowerUp(Id:power_up_id)<decides><transacts>:void =
    CurrentPlayerData := Player.GetData[]
    var UpdatedPowerUps:[int]int = CurrentPlayerData.PowerUps
    UpdatedPowerUps[Id.ToInt()] >= 1
    set UpdatedPowerUps[Id.ToInt()] -= 1

    set PlayerDataMap[Player] = player_data:
        MakePlayerData<constructor>(CurrentPlayerData)
        PowerUps := UpdatedPowerUps
    
    Player.UpdatePowerUpCosts()

(Player:player).RefundPowerUps(Type:power_up_type):void =
    if:
        CurrentPlayerData := Player.GetData[]
        var UpdatedPowerUps:[int]int = CurrentPlayerData.PowerUps
        for (PowerUpId : GetPowerUpIds(Type)):
            set UpdatedPowerUps[PowerUpId.ToInt()] = 0

        set PlayerDataMap[Player] = player_data:
            MakePlayerData<constructor>(CurrentPlayerData)
            PowerUps := UpdatedPowerUps 
    then:
        Player.UpdatePowerUpCosts()

(Player:player).GetGoldAvailable(Id:currency_id)<transacts>:float =
    Player.GetData[].Currencies[Id.ToInt()] - Player.GetPowerUpCosts() or 0.0

(Player:player).GetSelectBufferTime()<transacts>:float =
    Clamp(Player.GetData[].Settings.SelectBufferTime, 0.0, MaxSelectBufferTime) or 0.0
    