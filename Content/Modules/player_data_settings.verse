using { /Fortnite.com/Characters }
using { /Verse.org/Simulation }
using { /UnrealEngine.com/Temporary/SpatialMath }

<# ----------------------------------------------------------------------------------------------------
    Constants
---------------------------------------------------------------------------------------------------- #>
MaxSelectBufferTime:float = 2.0

<# ----------------------------------------------------------------------------------------------------
    player_settings
---------------------------------------------------------------------------------------------------- #>
player_settings := class<final><persistable><computes>:
    Camera:int = 0
    CameraDistance:int = 2
    Music:logic = true
    AutoAim:logic = true
    SelectBufferTime:float = 0.0
    InputMode:input_mode = input_mode.KeyboardMouse
    ShowGear:logic = false

MakePlayerSettings<constructor>(UpdatedPlayerSettings:player_settings)<transacts> := player_settings:
    Camera := UpdatedPlayerSettings.Camera
    CameraDistance := UpdatedPlayerSettings.CameraDistance
    Music := UpdatedPlayerSettings.Music
    AutoAim := UpdatedPlayerSettings.AutoAim
    SelectBufferTime := UpdatedPlayerSettings.SelectBufferTime
    InputMode := UpdatedPlayerSettings.InputMode
    ShowGear := UpdatedPlayerSettings.ShowGear

<# ----------------------------------------------------------------------------------------------------
    Global Functions
---------------------------------------------------------------------------------------------------- #>
(Player:player).GetAutoAimRotation()<transacts>:rotation =
    if (Character := Player.GetFortCharacter[]):
        if:
            Player.GetData[].Settings.AutoAim?
            TargetAgents := for(NPCBehavior : GetAllNPCs(), NPCBehavior.IsDamageable()?, NPCAgent := NPCBehavior.GetAgent[]) {NPCAgent}
            NearestAgent := GetNearestAgent(Player, TargetAgents)?
            NearestCharacter := NearestAgent.GetFortCharacter[]
            Direction := (NearestCharacter.GetTransform().Translation.Make2D() - Character.GetTransform().Translation.Make2D()).MakeUnitVector[]
        then:
            MakeRotation(vector3{X := 0.0, Y := 0.0, Z := 1.0}, ArcTan(Direction.Y, Direction.X))
        else if:
            ActiveCamera := Player.GetActiveCamera()
            ActiveCamera = camera_type.Orbit or ActiveCamera = camera_type.OrbitFree
            YawRotation := Character.GetViewRotation().GetYawPitchRollDegrees()[0]
        then:
            MakeRotationFromYawPitchRollDegrees(YawRotation, 0.0, 0.0)
        else:
            Character.GetTransform().Rotation
    else:
        IdentityRotation()

(Player:player).ToggleAutoAim():void =
    if:
        CurrentPlayerData := Player.GetData[]
        UpdatedAutoAim := logic{not CurrentPlayerData.Settings.AutoAim?}

        set PlayerDataMap[Player] = player_data:
            MakePlayerData<constructor>(CurrentPlayerData)
            Settings := player_settings:
                MakePlayerSettings<constructor>(CurrentPlayerData.Settings)
                    AutoAim := UpdatedAutoAim

(Player:player).ToggleSelectBufferTime():void =
    if:
        CurrentPlayerData := Player.GetData[]
        UpdatedSelectBufferTime := (CurrentPlayerData.Settings.SelectBufferTime >= MaxSelectBufferTime) and 0.0 or (CurrentPlayerData.Settings.SelectBufferTime + 0.5)

        set PlayerDataMap[Player] = player_data:
            MakePlayerData<constructor>(CurrentPlayerData)
            Settings := player_settings:
                MakePlayerSettings<constructor>(CurrentPlayerData.Settings)
                    SelectBufferTime := UpdatedSelectBufferTime

(Player:player).GetInputMode()<transacts>:input_mode =
    Player.GetData[].Settings.InputMode or input_mode.KeyboardMouse

(Player:player).SetInputMode(InputMode:input_mode):void =
    if:
        CurrentPlayerData := Player.GetData[]
        set PlayerDataMap[Player] = player_data:
            MakePlayerData<constructor>(CurrentPlayerData)
            Settings := player_settings:
                MakePlayerSettings<constructor>(CurrentPlayerData.Settings)
                    InputMode := InputMode

(Player:player).GetIsShowGear()<transacts>:logic =
    Player.GetData[].Settings.ShowGear or false

(Player:player).ToggleShowGear():void =
    if:
        CurrentPlayerData := Player.GetData[]
        UpdatedShowGear := logic{not CurrentPlayerData.Settings.ShowGear?}

        set PlayerDataMap[Player] = player_data:
            MakePlayerData<constructor>(CurrentPlayerData)
            Settings := player_settings:
                MakePlayerSettings<constructor>(CurrentPlayerData.Settings)
                    ShowGear := UpdatedShowGear