using { /Fortnite.com/Characters }
using { /Fortnite.com/UI }
using { /Verse.org/Simulation }
using { /UnrealEngine.com/Temporary/SpatialMath }
using { /UnrealEngine.com/Temporary/UI }

<# ----------------------------------------------------------------------------------------------------
    Data Definitions
---------------------------------------------------------------------------------------------------- #>
player_menu_option := enum:
    News
    Camera
    CameraDistance
    Music
    AutoAim
    ShowGear
    SelectBufferTime

PlayerMenuOptions:[]player_menu_option = array:
    player_menu_option.News
    player_menu_option.Camera
    player_menu_option.CameraDistance
    player_menu_option.Music
    player_menu_option.AutoAim
    player_menu_option.ShowGear
    player_menu_option.SelectBufferTime

ToString(PlayerMenuOption:player_menu_option)<transacts>:string =
    case (PlayerMenuOption):
        player_menu_option.News => "News"
        player_menu_option.Camera => "Camera"
        player_menu_option.CameraDistance => "Camera Distance"
        player_menu_option.Music => "Music"
        player_menu_option.AutoAim => "Auto Aim"
        player_menu_option.ShowGear => "Show Gear"
        player_menu_option.SelectBufferTime => "Select Buffer Time"

<# ----------------------------------------------------------------------------------------------------
    player_menu_ui
---------------------------------------------------------------------------------------------------- #>
player_menu_ui := class(panel_ui):
    Title<override>:string = "Menu"
    BackgroundOpacity<override>:type {_X:float where 0.000000 <= _X, _X <= 1.000000} = 0.7
    HasCloseButton<override>:logic = false

    ServerTimeText:text_block = text_block:
        DefaultTextColor := ThemeColors.PrimaryText
        DefaultShadowOffset := option{vector2{X := 2.0, Y := 2.0}}
        DefaultShadowOpacity := 0.5

    GiveUpButton:button_regular = button_regular{}

    var AbilityLevelTexts:[ability_id]text_block = map{}
    var MaybeSpawnNPCUI:?test_tool_spawn_npc_ui = false
    var PanelUIs:[string]panel_ui = map{}
    var MaybeDailyQuestsMenuUI:?daily_quests_menu_ui = false
    var OptionButtons:[player_menu_option]text_button_base = map{}
    var AbilityTexts:[ability_id]text_block = map{}
    var IsGiveUpConfirmation:logic = false
    var Subscriptions:[]cancelable = array{}

    <# ----------------------------------------------------------------------------------------------------
        Initialization
    ---------------------------------------------------------------------------------------------------- #>
    CreateContent<override>():widget =
        ClosedEvent.Subscribe(OnClosed)
        CloseButton := button_loud{DefaultText := MakeMessage("Close")}

        set Subscriptions = array:
            GiveUpButton.OnClick().Subscribe(OnClickGiveUp)
            CloseButton.OnClick().Subscribe(OnClickClose)

        AbilitiesStackBox := stack_box:
            Orientation := orientation.Vertical
            Slots := for (AbilityInfo : GetAbilityInfo()):
                AbilityText := text_block:
                    DefaultTextColor := ThemeColors.PrimaryText
    
                if (set AbilityTexts[AbilityInfo.Id] = AbilityText) {}

                stack_box_slot:
                    HorizontalAlignment := horizontal_alignment.Fill
                    Widget := CreateStatUI(AbilityText, "{AbilityInfo.Name}", ?Icon := option{AbilityInfo.IconSmall})

        DailyQuestsMenuUI := daily_quests_menu_ui{Player := Player}
        DailyQuestsMenuUI.Initialize()
        set MaybeDailyQuestsMenuUI = option{DailyQuestsMenuUI}

        
        stack_box:
            Orientation := orientation.Horizontal
            Slots := array:
                stack_box_slot:
                    HorizontalAlignment := horizontal_alignment.Fill
                    VerticalAlignment := vertical_alignment.Fill
                    Distribution := option{1.0}
                    Widget := stack_box:
                        Orientation := orientation.Horizontal
                        Slots := array:
                            stack_box_slot:
                                HorizontalAlignment := horizontal_alignment.Fill
                                VerticalAlignment := vertical_alignment.Fill
                                Widget := stack_box:
                                    Orientation := orientation.Vertical
                                    Slots := array:
                                        stack_box_slot:
                                            HorizontalAlignment := horizontal_alignment.Fill
                                            VerticalAlignment := vertical_alignment.Fill
                                            Widget := CreateServerTimeUI()
                                        stack_box_slot:
                                            HorizontalAlignment := horizontal_alignment.Fill
                                            VerticalAlignment := vertical_alignment.Fill
                                            Padding := margin{Left := 0.0, Top := 20.0, Right := 0.0, Bottom := 0.0}
                                            Widget := CloseButton
                                        stack_box_slot:
                                            HorizontalAlignment := horizontal_alignment.Fill
                                            VerticalAlignment := vertical_alignment.Fill
                                            Padding := margin{Left := 0.0, Top := 20.0, Right := 0.0, Bottom := 0.0}
                                            Widget := CreateMenuButtonsUI()
                                        stack_box_slot:
                                            HorizontalAlignment := horizontal_alignment.Fill
                                            VerticalAlignment := vertical_alignment.Bottom
                                            Distribution := option{1.0}
                                            Padding := margin{Left := 0.0, Top := 80.0, Right := 0.0, Bottom := 0.0}
                                            Widget := GiveUpButton
                            stack_box_slot:
                                HorizontalAlignment := horizontal_alignment.Fill
                                VerticalAlignment := vertical_alignment.Fill
                                Padding := margin{Left := 20.0, Top := 0.0, Right := 0.0, Bottom := 0.0}
                                Widget := DailyQuestsMenuUI.GetUI()

    CreateMenuButtonsUI():widget =
        stack_box:
            Orientation := orientation.Vertical
            Slots := for (Index -> Option : PlayerMenuOptions):
                Button:button_quiet = button_quiet:
                    DefaultText := MakeMessage("{Option}")

                if (set OptionButtons[Option] = Button) {}

                Subscription := case (Option):
                    player_menu_option.News => Button.OnClick().Subscribe(OnClickNews)
                    player_menu_option.Camera => Button.OnClick().Subscribe(OnClickCamera)
                    player_menu_option.CameraDistance => Button.OnClick().Subscribe(OnClickCameraDistance)
                    player_menu_option.Music => Button.OnClick().Subscribe(OnClickMusic)
                    player_menu_option.AutoAim => Button.OnClick().Subscribe(OnClickAutoAim)
                    player_menu_option.SelectBufferTime => Button.OnClick().Subscribe(OnClickSelectBufferTime)
                    player_menu_option.ShowGear => Button.OnClick().Subscribe(OnClickShowGear)

                set Subscriptions += array{Subscription}

                stack_box_slot:
                    HorizontalAlignment := horizontal_alignment.Fill
                    VerticalAlignment := vertical_alignment.Fill
                    Padding := margin{Left := 0.0, Top := (Index = 0) and 0.0 or 10.0, Right := 0.0, Bottom := 0.0}
                    Widget := Button
                    
    CreateServerTimeUI():widget =
        overlay:
            Slots := array:
                overlay_slot:
                    HorizontalAlignment := horizontal_alignment.Fill
                    VerticalAlignment := vertical_alignment.Fill
                    Widget := UserInterface.VerseWidget.WBP_RoundedBox{}
                overlay_slot:
                    HorizontalAlignment := horizontal_alignment.Fill
                    VerticalAlignment := vertical_alignment.Fill
                    Padding := margin{Left := 40.0, Top := 20.0, Right := 40.0, Bottom := 20.0}
                    Widget := stack_box:
                        Orientation := orientation.Horizontal
                        Slots := array:
                            stack_box_slot:
                                HorizontalAlignment := horizontal_alignment.Left
                                Distribution := option{1.0}
                                Padding := margin{Left := 0.0, Top := 0.0, Right := 20.0, Bottom := 0.0}
                                Widget := text_block:
                                    DefaultText := MakeMessage("Server Time:")
                                    DefaultTextColor := ThemeColors.PrimaryText
                                    DefaultTextOpacity := 0.5
                            stack_box_slot:
                                HorizontalAlignment := horizontal_alignment.Right
                                Widget := ServerTimeText

    Uninitialize<override>():void =
        for (Subscription : Subscriptions):
            Subscription.Cancel()

    <# ----------------------------------------------------------------------------------------------------
        Get Functions
    ---------------------------------------------------------------------------------------------------- #>
    CanGiveUp()<transacts>:logic =
        logic:
            IsGameInProgress()? 
            GetGameManager[].GameTime >= 0.1
            not Player.GetFortCharacter[].IsDownButNotOut[]
            not Player.IsInStageVolume()?

    <# ----------------------------------------------------------------------------------------------------
        Update Functions
    ---------------------------------------------------------------------------------------------------- #>
    Update():void =
        ServerTimeText.SetText(MakeMessage("{GetServerTime().FormatAsTime()}"))

        GiveUpButton.SetText(MakeMessage(IsGiveUpConfirmation? and "Confirm" or "Give Up"))

        if (CanGiveUp()?):
            GiveUpButton.SetVisibility(widget_visibility.Visible)
        else:
            GiveUpButton.SetVisibility(widget_visibility.Collapsed)

        if (CameraButton := OptionButtons[player_menu_option.Camera], CameraDistanceButton := OptionButtons[player_menu_option.CameraDistance]):
            CameraButton.SetText(MakeMessage("Camera: {Player.GetActiveCamera()}"))

            if (Player.GetActiveCamera() <> camera_type.Debug, CameraDistance := Player.GetData[].Settings.CameraDistance):
                CameraDistanceButton.SetText(MakeMessage("Camera Distance: {CameraDistance}"))
            else:
                CameraDistanceButton.SetText(MakeMessage(""))

        if (MusicButton := OptionButtons[player_menu_option.Music], IsMusicEnabled := Player.GetData[].Settings.Music):
            MusicButton.SetText(MakeMessage("Music: {IsMusicEnabled? and "ON" or "OFF"}"))

        if (AutoAimButton := OptionButtons[player_menu_option.AutoAim], IsAutoAim := Player.GetData[].Settings.AutoAim):
            AutoAimButton.SetText(MakeMessage("Auto Aim: {IsAutoAim? and "ON" or "OFF"}"))

        if (ShowGearButton := OptionButtons[player_menu_option.ShowGear], IsShowGear := Player.GetData[].Settings.ShowGear):
            ShowGearButton.SetText(MakeMessage("Show Gear: {IsShowGear? and "ON" or "OFF"}"))

        if (SelectBufferTimeButton := OptionButtons[player_menu_option.SelectBufferTime], SelectBufferTime := Player.GetData[].Settings.SelectBufferTime):
            SelectBufferTimeButton.SetText(MakeMessage("Select Buffer Time: {FormatAsTime(SelectBufferTime)}"))

    ShowPanelUI(Panel:string):void =
        UI := if (PanelUI := PanelUIs[Panel]):
            PanelUI
        else:
            PanelUI := case (Panel):
                "News" => news_ui{Player := Player}
                _ => return

            if (set PanelUIs[Panel] = PanelUI) {}
            PanelUI

        UI.Show()

    <# ----------------------------------------------------------------------------------------------------
        Event Handlers
    ---------------------------------------------------------------------------------------------------- #>
    OnShow<override>():void =
        set IsGiveUpConfirmation = false
        Update()
        
        if (DailyQuestsMenuUI := MaybeDailyQuestsMenuUI?):
            DailyQuestsMenuUI.Update()

        Player.PlayAudio(audio_id.Activate)

    OnClosed(InPlayer:player):void =
        if (IsGameInProgress()?, Player.GetCurrentPause[].Type = pause_type.PlayerMenu):
            ResumeGame(Player, ?IsInstant := true)

    OnClickNews(Message:widget_message):void =
        Hide()
        ShowPanelUI("News")

    OnClickCamera(Message:widget_message):void =
        Player.ToggleCamera()
        Update()

    OnClickCameraDistance(Message:widget_message):void =
        Player.ToggleCameraDistance()
        Player.RefreshControls()
        Update()

    OnClickMusic(Message:widget_message):void =
        Player.ToggleMusic()
        Update()

    OnClickGiveUp(Message:widget_message):void =
        if (not CanGiveUp()?):
            Player.PlayAudio(audio_id.Error)
            return

        if (IsGiveUpConfirmation?):
            Player.GiveUp()
            Hide()
        else:
            set IsGiveUpConfirmation = true
            Update()
            Player.PlayAudio(audio_id.Activate)

    OnClickAutoAim(Message:widget_message):void =
        Player.ToggleAutoAim()
        Player.RefreshControls()
        Update()

    OnClickSelectBufferTime(Message:widget_message):void =
        Player.ToggleSelectBufferTime()
        Update()

    OnClickShowGear(Message:widget_message):void =
        Player.ToggleShowGear()
        Player.ActivateVFXPowerup of Player.GetRankVFXPowerUp()
        Update()