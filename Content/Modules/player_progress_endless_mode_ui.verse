using { /Fortnite.com/UI }
using { /Verse.org/Colors }
using { /Verse.org/Simulation }
using { /UnrealEngine.com/Temporary/SpatialMath }
using { /UnrealEngine.com/Temporary/UI }

player_progress_endless_mode_ui := class(ui_component):

    var EndlessModeProgressTexts:[]text_block = array{}
    
    <# ----------------------------------------------------------------------------------------------------
        Initialization
    ---------------------------------------------------------------------------------------------------- #>
    CreateUI<override>():widget =
        StackBox := stack_box:
            Orientation := orientation.Vertical
            Slots := array:
                stack_box_slot:
                    Padding := margin{Left := 0.0, Top := 0.0, Right := 0.0, Bottom := 20.0}
                    Widget := text_block:
                        DefaultText := MakeMessage("Endless Mode")
                        DefaultTextColor := ThemeColors.PrimaryText
                        DefaultShadowOffset := option{vector2{X := 2.0, Y := 2.0}}
                        DefaultShadowOpacity := 0.5
                stack_box_slot:
                    HorizontalAlignment := horizontal_alignment.Fill
                    VerticalAlignment := vertical_alignment.Fill
                    Padding := margin{Left := 0.0, Top := 0.0, Right := 0.0, Bottom := 20.0}
                    Widget := stack_box:
                        Orientation := orientation.Horizontal
                        Slots := array:
                            stack_box_slot:
                                HorizontalAlignment := horizontal_alignment.Left
                                Distribution := option{2.0}
                                Widget := text_block:
                                    DefaultText := MakeMessage("Arena")
                                    DefaultTextColor := ThemeColors.PrimaryText
                                    DefaultTextOpacity := 0.5
                            stack_box_slot:
                                HorizontalAlignment := horizontal_alignment.Left
                                Distribution := option{1.0}
                                Widget := text_block:
                                    DefaultText := MakeMessage("Stage")
                                    DefaultTextColor := ThemeColors.PrimaryText
                                    DefaultTextOpacity := 0.5
        
        for (Index -> Arena : GetArenas()):
            EndlessModeProgressText := text_block:
                DefaultTextColor := ThemeColors.PrimaryText
                DefaultShadowOffset := option{vector2{X := 2.0, Y := 2.0}}
                DefaultShadowOpacity := 0.5

            set EndlessModeProgressTexts += array{EndlessModeProgressText}

            StackBox.AddWidget of stack_box_slot:
                HorizontalAlignment := horizontal_alignment.Fill
                Padding := margin{Left := 0.0, Top := (Index = 0) and 0.0 or 10.0, Right := 0.0, Bottom := 0.0}
                Widget := stack_box:
                    Orientation := orientation.Horizontal
                    Slots := array:
                        stack_box_slot:
                            HorizontalAlignment := horizontal_alignment.Left
                            Distribution := option{2.0}
                            Widget := text_block:
                                DefaultText := MakeMessage("{Arena.Name}")
                                DefaultTextColor := ThemeColors.PrimaryText
                        stack_box_slot:
                            HorizontalAlignment := horizontal_alignment.Left
                            Distribution := option{1.0}
                            Widget := EndlessModeProgressText

        StackBox
            
    <# ----------------------------------------------------------------------------------------------------
        Update Functions
    ---------------------------------------------------------------------------------------------------- #>
    Update(InPlayer:player):void =
        for (Index -> EndlessModeProgressText : EndlessModeProgressTexts):
            Progress := Player.GetEndlessModeProgress(Index)
            EndlessModeProgressText.SetText(MakeMessage("{Progress}"))
            EndlessModeProgressText.SetTextColor((Progress > 0) and ThemeColors.PrimaryText or ThemeColors.Inactive)
            EndlessModeProgressText.SetShadowOpacity((Progress > 0) and 0.5 or 0.0)