using { /Fortnite.com/UI }
using { /Verse.org/Assets }
using { /Verse.org/Colors }
using { /Verse.org/Simulation }
using { /UnrealEngine.com/Temporary/SpatialMath }
using { /UnrealEngine.com/Temporary/UI }

<# ----------------------------------------------------------------------------------------------------
    player_progress_rank_info_ui
---------------------------------------------------------------------------------------------------- #>
player_progress_rank_info_ui := class(ui_component):

    <# ----------------------------------------------------------------------------------------------------
        Initialization
    ---------------------------------------------------------------------------------------------------- #>
    CreateUI<override>():widget =
        overlay:
            Slots := array:
                overlay_slot:
                    HorizontalAlignment := horizontal_alignment.Fill
                    VerticalAlignment := vertical_alignment.Fill
                    Widget := UserInterface.VerseWidget.WBP_RoundedBox{}
                overlay_slot:
                    HorizontalAlignment := horizontal_alignment.Fill
                    VerticalAlignment := vertical_alignment.Fill
                    Padding := margin{Left := 40.0, Top := 40.0, Right := 40.0, Bottom := 40.0}
                    Widget := stack_box:
                        Orientation := orientation.Vertical
                        Slots := array:
                            stack_box_slot:
                                HorizontalAlignment := horizontal_alignment.Fill
                                VerticalAlignment := vertical_alignment.Fill
                                Widget := text_block:
                                    DefaultText := MakeMessage("Rank Info")
                                    DefaultTextColor := ThemeColors.PrimaryText
                                    DefaultShadowColor := NamedColors.Black
                                    DefaultShadowOffset := option{vector2{X := 2.0, Y := 2.0}}
                                    DefaultShadowOpacity := 0.5
                            stack_box_slot:
                                HorizontalAlignment := horizontal_alignment.Fill
                                VerticalAlignment := vertical_alignment.Fill
                                Padding := margin{Left := 0.0, Top := 20.0, Right := 0.0, Bottom := 0.0}
                                Widget := stack_box:
                                    Orientation := orientation.Horizontal
                                    Slots := array:
                                        stack_box_slot:
                                            HorizontalAlignment := horizontal_alignment.Left
                                            VerticalAlignment := vertical_alignment.Fill
                                            Distribution := option{1.0}
                                            Widget := text_block:
                                                DefaultText := MakeMessage("Rank")
                                                DefaultTextColor := ThemeColors.PrimaryText
                                                DefaultTextOpacity := 0.5
                                        stack_box_slot:
                                            HorizontalAlignment := horizontal_alignment.Left
                                            VerticalAlignment := vertical_alignment.Fill
                                            Distribution := option{1.0}
                                            Widget := text_block:
                                                DefaultText := MakeMessage("Arena Level")
                                                DefaultTextColor := ThemeColors.PrimaryText
                                                DefaultTextOpacity := 0.5
                                        stack_box_slot:
                                            HorizontalAlignment := horizontal_alignment.Left
                                            VerticalAlignment := vertical_alignment.Fill
                                            Distribution := option{1.0}
                                            Widget := text_block:
                                                DefaultText := MakeMessage("Daily Quest Rerolls")
                                                DefaultTextColor := ThemeColors.PrimaryText
                                                DefaultTextOpacity := 0.5
                                        stack_box_slot:
                                            HorizontalAlignment := horizontal_alignment.Left
                                            VerticalAlignment := vertical_alignment.Fill
                                            Distribution := option{2.0}
                                            Padding := margin{Left := 40.0, Top := 0.0, Right := 0.0, Bottom := 0.0}
                                            Widget := text_block:
                                                DefaultText := MakeMessage("Treasure Chest Multiplier")
                                                DefaultTextColor := ThemeColors.PrimaryText
                                                DefaultTextOpacity := 0.5
                            stack_box_slot:
                                HorizontalAlignment := horizontal_alignment.Fill
                                VerticalAlignment := vertical_alignment.Fill
                                Padding := margin{Left := 0.0, Top := 20.0, Right := 0.0, Bottom := 0.0}
                                Widget := stack_box:
                                    Orientation := orientation.Vertical
                                    Slots := for (RankIndex -> Rank : Ranks):
                                        stack_box_slot:
                                            HorizontalAlignment := horizontal_alignment.Fill
                                            VerticalAlignment := vertical_alignment.Fill
                                            Padding := margin{Left := 0.0, Top := (RankIndex = 0) and 0.0 or 10.0, Right := 0.0, Bottom := 0.0}
                                            Widget := stack_box:
                                                Orientation := orientation.Horizontal
                                                Slots := array:
                                                    stack_box_slot:
                                                        HorizontalAlignment := horizontal_alignment.Fill
                                                        VerticalAlignment := vertical_alignment.Fill
                                                        Distribution := option{1.0}
                                                        Widget := stack_box:
                                                            Orientation := orientation.Horizontal
                                                            Slots := array:
                                                                stack_box_slot:
                                                                    HorizontalAlignment := horizontal_alignment.Left
                                                                    Widget := texture_block:
                                                                        DefaultImage := RankIndex.ToRankIcon()
                                                                        DefaultDesiredSize := vector2{X := 40.0, Y := 40.0}
                                                                stack_box_slot:
                                                                    HorizontalAlignment := horizontal_alignment.Fill
                                                                    VerticalAlignment := vertical_alignment.Fill
                                                                    Padding := margin{Left := 10.0, Top := 0.0, Right := 0.0, Bottom := 0.0}
                                                                    Widget := text_block:
                                                                        DefaultText := MakeMessage("{RankIndex.ToRankString()}")
                                                                        DefaultTextColor := (RankIndex + 1).ToTierColor() * 2.0
                                                                        DefaultShadowColor := (RankIndex + 1).ToTierColor() * 0.25
                                                                        DefaultShadowOffset := option{vector2{X := 2.0, Y := 2.0}}
                                                                        DefaultShadowOpacity := 1.0
                                                    stack_box_slot:
                                                        HorizontalAlignment := horizontal_alignment.Fill
                                                        VerticalAlignment := vertical_alignment.Fill
                                                        Distribution := option{1.0}
                                                        Widget := stack_box:
                                                            Orientation := orientation.Horizontal
                                                            Slots := array:
                                                                stack_box_slot:
                                                                    HorizontalAlignment := horizontal_alignment.Fill
                                                                    VerticalAlignment := vertical_alignment.Fill
                                                                    Widget := text_block:
                                                                        DefaultText := MakeMessage("Lv.")
                                                                        DefaultTextColor := ThemeColors.Highlight
                                                                        DefaultShadowColor := ThemeColors.Highlight * 0.125
                                                                        DefaultShadowOffset := option{vector2{X := 2.0, Y := 2.0}}
                                                                        DefaultShadowOpacity := 1.0
                                                                stack_box_slot:
                                                                    HorizontalAlignment := horizontal_alignment.Fill
                                                                    VerticalAlignment := vertical_alignment.Fill
                                                                    Padding := margin{Left := 10.0, Top := 0.0, Right := 0.0, Bottom := 0.0}
                                                                    Widget := text_block:
                                                                        DefaultText := MakeMessage("{RankIndex * GetArenaCount()}")
                                                                        DefaultTextColor := ThemeColors.PrimaryText
                                                                        DefaultShadowColor := NamedColors.Black
                                                                        DefaultShadowOffset := option{vector2{X := 2.0, Y := 2.0}}
                                                                        DefaultShadowOpacity := 0.5
                                                    stack_box_slot:
                                                        HorizontalAlignment := horizontal_alignment.Fill
                                                        VerticalAlignment := vertical_alignment.Fill
                                                        Distribution := option{1.0}
                                                        Widget := text_block:
                                                            DefaultText := MakeMessage("{CalculateDailyQuestRerolls(RankIndex)}")
                                                            DefaultTextColor := ThemeColors.PrimaryText
                                                            DefaultShadowColor := NamedColors.Black
                                                            DefaultShadowOffset := option{vector2{X := 2.0, Y := 2.0}}
                                                            DefaultShadowOpacity := 0.5
                                                    stack_box_slot:
                                                        HorizontalAlignment := horizontal_alignment.Fill
                                                        VerticalAlignment := vertical_alignment.Fill
                                                        Distribution := option{2.0}
                                                        Padding := margin{Left := 40.0, Top := 0.0, Right := 0.0, Bottom := 0.0}
                                                        Widget := text_block:
                                                            DefaultText := MakeMessage("{FormatAsPercent(CalculateTreasureChestRankMultiplier(RankIndex))}")
                                                            DefaultTextColor := ThemeColors.PrimaryText
                                                            DefaultShadowColor := NamedColors.Black
                                                            DefaultShadowOffset := option{vector2{X := 2.0, Y := 2.0}}
                                                            DefaultShadowOpacity := 0.5