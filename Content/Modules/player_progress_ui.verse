using { /Fortnite.com/UI }
using { /Verse.org/Colors }
using { /Verse.org/Simulation }
using { /UnrealEngine.com/Temporary/SpatialMath }
using { /UnrealEngine.com/Temporary/UI }

<# ----------------------------------------------------------------------------------------------------
    player_progress_ui
---------------------------------------------------------------------------------------------------- #>
player_progress_ui := class(panel_ui):
    Title<override>:string = "Player Progress"
    Anchors<override>:anchors = anchors{Minimum := vector2{X := 0.0, Y := 0.0}, Maximum := vector2{X := 1.0, Y := 1.0}}
    BackgroundOpacity<override>:type {_X:float where 0.000000 <= _X, _X <= 1.000000} = 0.7

    StatInfoButton:button_quiet = button_quiet{}

    var PlayerButtons:[]button_quiet = array{}
    var MaybePlayerProgressStatUI:?player_progress_stat_ui = false
    var MaybePlayerProgressWeaponUI:?player_progress_weapon_ui = false
    var MaybePlayerProgressEndlessModeUI:?player_progress_endless_mode_ui = false
    var RankInfoUI:widget = canvas{}
    var StatInfoUI:widget = canvas{}
    var IsShowingStatInfo:logic = true

    <# ----------------------------------------------------------------------------------------------------
        Initialization
    ---------------------------------------------------------------------------------------------------- #>
    CreateContent<override>():widget =
        PlayerProgressStatUI := player_progress_stat_ui{Player := Player}
        PlayerProgressStatUI.Initialize()
        set MaybePlayerProgressStatUI = option{PlayerProgressStatUI}

        PlayerProgressWeaponUI := player_progress_weapon_ui{Player := Player}
        PlayerProgressWeaponUI.Initialize()
        set MaybePlayerProgressWeaponUI = option{PlayerProgressWeaponUI}

        PlayerProgressEndlessModeUI := player_progress_endless_mode_ui{Player := Player}
        PlayerProgressEndlessModeUI.Initialize()
        PlayerProgressEndlessModeUI.Hide()
        set MaybePlayerProgressEndlessModeUI = option{PlayerProgressEndlessModeUI}

        PlayerProgressRankInfoUI := player_progress_rank_info_ui{Player := Player}
        PlayerProgressRankInfoUI.Initialize()
        set RankInfoUI = PlayerProgressRankInfoUI.GetUI()

        StatInfoButton.SetText(MakeMessage("Endless Mode"))
        StatInfoButton.OnClick().Subscribe(OnClickStatInfo)

        stack_box:
            Orientation := orientation.Horizontal
            Slots := array:
                stack_box_slot:
                    HorizontalAlignment := horizontal_alignment.Fill
                    VerticalAlignment := vertical_alignment.Fill
                    Widget := overlay:
                        Slots := array:
                            overlay_slot:
                                HorizontalAlignment := horizontal_alignment.Fill
                                VerticalAlignment := vertical_alignment.Fill
                                Widget := UserInterface.VerseWidget.WBP_RoundedBox{}
                            overlay_slot:
                                HorizontalAlignment := horizontal_alignment.Fill
                                VerticalAlignment := vertical_alignment.Fill
                                Padding := margin{Left := 40.0, Top := 40.0, Right := 40.0, Bottom := 40.0}
                                Widget := CreateMenuUI()
                stack_box_slot:
                    HorizontalAlignment := horizontal_alignment.Fill
                    VerticalAlignment := vertical_alignment.Fill
                    Padding := margin{Left := 20.0, Top := 0.0, Right := 0.0, Bottom := 0.0}
                    Distribution := option{1.0}
                    Widget := overlay:
                        Slots := array:
                            overlay_slot:
                                HorizontalAlignment := horizontal_alignment.Fill
                                VerticalAlignment := vertical_alignment.Fill
                                Widget := set StatInfoUI = stack_box:
                                    Orientation := orientation.Horizontal
                                    Slots := array:
                                        stack_box_slot:
                                            HorizontalAlignment := horizontal_alignment.Fill
                                            VerticalAlignment := vertical_alignment.Fill
                                            Distribution := option{1.0}
                                            Widget := overlay:
                                                Slots := array:
                                                    overlay_slot:
                                                        HorizontalAlignment := horizontal_alignment.Fill
                                                        VerticalAlignment := vertical_alignment.Fill
                                                        Widget := UserInterface.VerseWidget.WBP_RoundedBox{}
                                                    overlay_slot:
                                                        HorizontalAlignment := horizontal_alignment.Fill
                                                        VerticalAlignment := vertical_alignment.Fill
                                                        Padding := margin{Left := 40.0, Top := 40.0, Right := 40.0, Bottom := 40.0}
                                                        Widget := stack_box:
                                                            Orientation := orientation.Vertical
                                                            Slots := array:
                                                                stack_box_slot:
                                                                    HorizontalAlignment := horizontal_alignment.Fill
                                                                    VerticalAlignment := vertical_alignment.Fill
                                                                    Widget := overlay:
                                                                        Slots := array:
                                                                            overlay_slot:
                                                                                HorizontalAlignment := horizontal_alignment.Fill
                                                                                VerticalAlignment := vertical_alignment.Fill
                                                                                Widget := PlayerProgressStatUI.GetUI()
                                                                            overlay_slot:
                                                                                HorizontalAlignment := horizontal_alignment.Fill
                                                                                VerticalAlignment := vertical_alignment.Fill
                                                                                Widget := PlayerProgressEndlessModeUI.GetUI()
                                                                stack_box_slot:
                                                                    HorizontalAlignment := horizontal_alignment.Fill
                                                                    VerticalAlignment := vertical_alignment.Bottom
                                                                    Distribution := option{1.0}
                                                                    Padding := margin{Left := 0.0, Top := 20.0, Right := 0.0, Bottom := 0.0}
                                                                    Widget := StatInfoButton
                                        stack_box_slot:
                                            HorizontalAlignment := horizontal_alignment.Fill
                                            VerticalAlignment := vertical_alignment.Fill
                                            Distribution := option{1.0}
                                            Padding := margin{Left := 20.0, Top := 0.0, Right := 0.0, Bottom := 0.0}
                                            Widget := PlayerProgressWeaponUI.GetUI()
                            overlay_slot:
                                HorizontalAlignment := horizontal_alignment.Fill
                                VerticalAlignment := vertical_alignment.Fill
                                Widget := RankInfoUI

    <# ----------------------------------------------------------------------------------------------------
        Menu UI
    ---------------------------------------------------------------------------------------------------- #>
    CreateMenuUI():widget =
        StackBox := stack_box:
            Orientation := orientation.Vertical
            Slots := array:
                stack_box_slot:
                    VerticalAlignment := vertical_alignment.Fill
                    Padding := margin{Left := 0.0, Top := 0.0, Right := 0.0, Bottom := 20.0}
                    Widget := text_block:
                        DefaultText := MakeMessage("Menu")
                        DefaultTextColor := ThemeColors.PrimaryText
                        DefaultTextOpacity := 0.5

        for (Index := 0..MaxPlayers - 1):
            PlayerButton := button_quiet{}
            spawn{AwaitPlayerButtonClicked(PlayerButton.OnClick(), Index)}
            set PlayerButtons += array{PlayerButton}
                
            StackBox.AddWidget of stack_box_slot:
                HorizontalAlignment := horizontal_alignment.Fill
                VerticalAlignment := vertical_alignment.Fill
                Padding := margin{Left := 0.0, Top := (Index = 0) and 0.0 or 10.0, Right := 0.0, Bottom := 0.0}
                Widget := PlayerButton

        RankInfoButton := button_quiet{DefaultText := MakeMessage("Rank Info")}
        RankInfoButton.OnClick().Subscribe(OnClickRankInfo)

        StackBox.AddWidget of stack_box_slot:
            HorizontalAlignment := horizontal_alignment.Fill
            VerticalAlignment := vertical_alignment.Bottom
            Distribution := option{1.0}
            Widget := RankInfoButton

        StackBox
    
    <# ----------------------------------------------------------------------------------------------------
        Update Functions
    ---------------------------------------------------------------------------------------------------- #>
    SetPlayer(InPlayer:player):void =
        set IsShowingStatInfo = true
        StatInfoUI.SetVisibility(widget_visibility.Visible)
        RankInfoUI.SetVisibility(widget_visibility.Hidden)

        if (PlayerProgressStatUI := MaybePlayerProgressStatUI?, PlayerProgressEndlessModeUI := MaybePlayerProgressEndlessModeUI?):
            PlayerProgressStatUI.Update(InPlayer)
            PlayerProgressEndlessModeUI.Update(InPlayer)
            UpdateInfoPanel()

        if (PlayerProgressWeaponUI := MaybePlayerProgressWeaponUI?):
            PlayerProgressWeaponUI.Update(InPlayer)
            PlayerProgressWeaponUI.Show()

    ShowRankInfo():void =
        StatInfoUI.SetVisibility(widget_visibility.Hidden)
        RankInfoUI.SetVisibility(widget_visibility.Visible)

        if (PlayerProgressStatUI := MaybePlayerProgressStatUI?):
            PlayerProgressStatUI.Hide()

        if (PlayerProgressWeaponUI := MaybePlayerProgressWeaponUI?):
            PlayerProgressWeaponUI.Hide()

    UpdateInfoPanel():void =
        StatInfoButton.SetText(MakeMessage(IsShowingStatInfo? and "Endless Mode" or "Stat Info"))
        if (PlayerProgressStatUI := MaybePlayerProgressStatUI?, PlayerProgressEndlessModeUI := MaybePlayerProgressEndlessModeUI?):
            if (IsShowingStatInfo?):
                PlayerProgressStatUI.Show()
                PlayerProgressEndlessModeUI.Hide()
            else:
                PlayerProgressStatUI.Hide()
                PlayerProgressEndlessModeUI.Show()

    <# ----------------------------------------------------------------------------------------------------
        Event Handlers
    ---------------------------------------------------------------------------------------------------- #>
    OnShow<override>():void =
        Players := GetAllPlayers()
        for (Index -> PlayerButton : PlayerButtons):
            if (InPlayer := Players[Index]):
                PlayerButton.SetText(MakeMessage(InPlayer))
                PlayerButton.SetVisibility(widget_visibility.Visible)
            else:
                PlayerButton.SetVisibility(widget_visibility.Hidden)

        if (InPlayer := Players[0]):
            SetPlayer(InPlayer)

    AwaitPlayerButtonClicked(Listenable:listenable(widget_message), Index:int)<suspends>:void =
        if (PlayerRemovedEvent := Player.GetPlayerRemovedEvent[]):
            race:
                PlayerRemovedEvent.Await()
                loop:
                    Listenable.Await()
                    if (InPlayer := GetAllPlayers()[Index]):
                        SetPlayer(InPlayer)

    OnClickRankInfo(Message:widget_message):void =
        ShowRankInfo()

    OnClickStatInfo(Message:widget_message):void =
        set IsShowingStatInfo = logic{not IsShowingStatInfo?}
        UpdateInfoPanel()