using { /Fortnite.com/UI }
using { /Verse.org/Colors }
using { /Verse.org/Simulation }
using { /UnrealEngine.com/Temporary/SpatialMath }
using { /UnrealEngine.com/Temporary/UI }

player_progress_weapon_entry_ui := class(ui_component):
    CancelAnimationEvent:event(void) = event(void){}
    
    WeaponIcon:texture_block = texture_block:
        DefaultImage := UserInterface.Icons.WeaponBow_64
        DefaultDesiredSize := vector2{X := 40.0, Y := 40.0}

    LevelText:text_block = text_block:
        DefaultTextColor := ThemeColors.PrimaryText
        DefaultShadowColor := NamedColors.Black
        DefaultShadowOffset := option{vector2{X := 2.0, Y := 2.0}}
        DefaultShadowOpacity := 0.5

    ProgressBar:color_block = color_block:
        DefaultColor := ThemeColors.Highlight

    ProgressBarSize:vector2 = vector2{X := 450.0, Y := 16.0}
    BorderWidth:float = 1.0

    <# ----------------------------------------------------------------------------------------------------
        Initialization
    ---------------------------------------------------------------------------------------------------- #>
    CreateUI<override>():widget =
        stack_box:
            Orientation := orientation.Horizontal
            Slots := array:
                stack_box_slot:
                    HorizontalAlignment := horizontal_alignment.Left
                    Widget := WeaponIcon
                stack_box_slot:
                    HorizontalAlignment := horizontal_alignment.Fill
                    VerticalAlignment := vertical_alignment.Fill
                    Distribution := option{1.0}
                    Padding := margin{Left := 20.0, Top := 0.0, Right := 0.0, Bottom := 0.0}
                    Widget := stack_box:
                        Orientation := orientation.Horizontal
                        Slots := array:
                            stack_box_slot:
                                HorizontalAlignment := horizontal_alignment.Fill
                                VerticalAlignment := vertical_alignment.Fill
                                Widget := text_block:
                                    DefaultText := MakeMessage("Lv.")
                                    DefaultTextColor := ThemeColors.Highlight
                                    DefaultShadowColor := ThemeColors.Highlight * 0.125
                                    DefaultShadowOffset := option{vector2{X := 2.0, Y := 2.0}}
                                    DefaultShadowOpacity := 1.0
                            stack_box_slot:
                                HorizontalAlignment := horizontal_alignment.Fill
                                VerticalAlignment := vertical_alignment.Fill
                                Padding := margin{Left := 10.0, Top := 0.0, Right := 0.0, Bottom := 0.0}
                                Widget := LevelText
                stack_box_slot:
                    HorizontalAlignment := horizontal_alignment.Fill
                    Padding := margin{Left := 20.0, Top := 0.0, Right := 0.0, Bottom := 0.0}
                    Widget := overlay:
                        Slots := array:
                            overlay_slot:
                                HorizontalAlignment := horizontal_alignment.Fill
                                VerticalAlignment := vertical_alignment.Fill
                                Widget := MakeBox(?Width := ProgressBarSize.X, ?Height := ProgressBarSize.Y, ?Opacity := 0.1, ?BorderColor := NamedColors.Black, ?BorderOpacity := 0.2)
                            overlay_slot:
                                HorizontalAlignment := horizontal_alignment.Left
                                VerticalAlignment := vertical_alignment.Fill
                                Padding := margin{Left := BorderWidth, Top := BorderWidth, Right := BorderWidth, Bottom := BorderWidth}
                                Widget := overlay:
                                    Slots := array:
                                        overlay_slot:
                                            HorizontalAlignment := horizontal_alignment.Fill
                                            VerticalAlignment := vertical_alignment.Fill
                                            Widget := ProgressBar
                                        overlay_slot:
                                            HorizontalAlignment := horizontal_alignment.Right
                                            VerticalAlignment := vertical_alignment.Fill
                                            Widget := color_block:
                                                DefaultColor := NamedColors.White
                                                DefaultOpacity := 0.5
                                                DefaultDesiredSize := vector2{X := 2.0, Y := 0.0}
                            overlay_slot:
                                HorizontalAlignment := horizontal_alignment.Fill
                                VerticalAlignment := vertical_alignment.Top
                                Padding := margin{Left := BorderWidth, Top := BorderWidth, Right := BorderWidth, Bottom := BorderWidth}
                                Widget := color_block:
                                    DefaultColor := NamedColors.White
                                    DefaultOpacity := 0.3
                                    DefaultDesiredSize := vector2{X := 0.0, Y := ProgressBarSize.Y / 2.0} 

    <# ----------------------------------------------------------------------------------------------------
        Update Functions
    ---------------------------------------------------------------------------------------------------- #>
    Update(InPlayer:player, WeaponId:weapon_id, Level:int):void =
        if (WeaponInfo := GetWeaponInfo[WeaponId]):
            WeaponIcon.SetImage(WeaponInfo.IconSmall)
            WeaponIcon.SetTint(InPlayer.IsWeaponUnlocked(WeaponId)? and NamedColors.White or ThemeColors.Disabled)

        LevelText.SetText(MakeMessage("{Level}"))
        Progress := (Level * 1.0) / (GetMaxArenaLevel() * 1.0)
        CancelAnimationEvent.Signal()
        spawn{Animate(Progress)}

    Animate(TargetProgress:float)<suspends>:void =
        DeltaTimer := delta_timer{}
        ProgressBar.SetDesiredSize(vector2{X := 0.0})

        race:
            CancelAnimationEvent.Await()
            loop:
                Progress := DeltaTimer.GetElapsedTime() / 0.3
                ProgressBar.SetDesiredSize(vector2{X := Lerp(0.0, ProgressBarSize.X * TargetProgress, Progress)})
                
                if (DeltaTimer.GetElapsedTime() >= 0.3):
                    ProgressBar.SetDesiredSize(vector2{X := ProgressBarSize.X * TargetProgress})
                    break
                else:
                    Sleep(0.0)