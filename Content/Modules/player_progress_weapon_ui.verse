using { /Fortnite.com/UI }
using { /Verse.org/Colors }
using { /Verse.org/Simulation }
using { /UnrealEngine.com/Temporary/SpatialMath }
using { /UnrealEngine.com/Temporary/UI }

player_progress_weapon_ui := class(ui_component):
    EntriesPerPage:int = 9

    CurrentPageText:text_block = text_block:
        DefaultTextColor := ThemeColors.PrimaryText
        DefaultTextOpacity := 0.5

    GoldMultiplierText:text_block = text_block:
        DefaultTextColor := ThemeColors.Success
        DefaultShadowColor := ThemeColors.Success * 0.125
        DefaultShadowOffset := option{vector2{X := 2.0, Y := 2.0}}
        DefaultShadowOpacity := 1.0

    var PlayerProgressWeaponEntryUIs:[]player_progress_weapon_entry_ui = array{}
    var CurrentPage:int = 1
    var MaybeActivePlayer:?player = false

    <# ----------------------------------------------------------------------------------------------------
        Initialization
    ---------------------------------------------------------------------------------------------------- #>
    CreateUI<override>():widget =
        PreviousButton := button_quiet{DefaultText := MakeMessage("◀")}
        PreviousButton.OnClick().Subscribe(OnClickPrevious)

        NextButton := button_quiet{DefaultText := MakeMessage("▶")}
        NextButton.OnClick().Subscribe(OnClickNext)

        overlay:
            Slots := array:
                overlay_slot:
                    HorizontalAlignment := horizontal_alignment.Fill
                    VerticalAlignment := vertical_alignment.Fill
                    Widget := UserInterface.VerseWidget.WBP_RoundedBox{}
                overlay_slot:
                    HorizontalAlignment := horizontal_alignment.Fill
                    VerticalAlignment := vertical_alignment.Fill
                    Padding := margin{Left := 40.0, Top := 40.0, Right := 40.0, Bottom := 40.0}
                    Widget := stack_box:
                        Orientation := orientation.Vertical
                        Slots := array:
                            stack_box_slot:
                                VerticalAlignment := vertical_alignment.Fill
                                Widget := text_block:
                                    DefaultText := MakeMessage("Weapon Progress")
                                    DefaultTextColor := ThemeColors.PrimaryText
                                    DefaultShadowColor := NamedColors.Black
                                    DefaultShadowOffset := option{vector2{X := 2.0, Y := 2.0}}
                                    DefaultShadowOpacity := 0.5
                            stack_box_slot:
                                HorizontalAlignment := horizontal_alignment.Fill
                                VerticalAlignment := vertical_alignment.Top
                                Padding := margin{Left := 0.0, Top := 20.0, Right := 0.0, Bottom := 0.0}
                                Widget := overlay:
                                    Slots := array:
                                        overlay_slot:
                                            HorizontalAlignment := horizontal_alignment.Fill
                                            VerticalAlignment := vertical_alignment.Fill
                                            Widget := UserInterface.VerseWidget.WBP_RoundedBox{}
                                        overlay_slot:
                                            Padding := margin{Left := 40.0, Top := 10.0, Right := 40.0, Bottom := 10.0}
                                            Widget := stack_box:
                                                Orientation := orientation.Horizontal
                                                Slots := array:
                                                    stack_box_slot:
                                                        HorizontalAlignment := horizontal_alignment.Left
                                                        Padding := margin{Left := 0.0, Top := 0.0, Right := 10.0, Bottom := 0.0}
                                                        Widget := texture_block:
                                                            DefaultImage := UserInterface.Icons.Gold_64
                                                            DefaultDesiredSize := vector2{X := 40.0, Y := 40.0}
                                                    stack_box_slot:
                                                        HorizontalAlignment := horizontal_alignment.Fill
                                                        VerticalAlignment := vertical_alignment.Fill
                                                        Widget := GoldMultiplierText
                            stack_box_slot:
                                HorizontalAlignment := horizontal_alignment.Fill
                                Padding := margin{Left := 0.0, Top := 20.0, Right := 0.0, Bottom := 0.0}
                                Widget := stack_box:
                                    Orientation := orientation.Vertical
                                    Slots := for (Index := 1..EntriesPerPage):
                                        PlayerProgressWeaponEntryUI := player_progress_weapon_entry_ui{Player := Player}
                                        PlayerProgressWeaponEntryUI.Initialize()
                                        set PlayerProgressWeaponEntryUIs += array{PlayerProgressWeaponEntryUI}

                                        stack_box_slot:
                                            HorizontalAlignment := horizontal_alignment.Fill
                                            Padding := margin{Left := 0.0, Top := (Index = 0) and 0.0 or 5.0, Right := 0.0, Bottom := 0.0}
                                            Widget := PlayerProgressWeaponEntryUI.GetUI()
                            
                            stack_box_slot:
                                HorizontalAlignment := horizontal_alignment.Fill
                                VerticalAlignment := vertical_alignment.Bottom
                                Padding := margin{Left := 0.0, Top := 20.0, Right := 0.0, Bottom := 0.0}
                                Distribution := option{1.0}
                                Widget := stack_box:
                                    Orientation := orientation.Horizontal
                                    Slots := array:
                                        stack_box_slot:
                                            HorizontalAlignment := horizontal_alignment.Fill
                                            VerticalAlignment := vertical_alignment.Fill
                                            Distribution := option{1.0}
                                            Widget := PreviousButton
                                        stack_box_slot:
                                            HorizontalAlignment := horizontal_alignment.Fill
                                            Padding := margin{Left := 20.0, Top := 0.0, Right := 20.0, Bottom := 0.0}
                                            Widget := CurrentPageText
                                        stack_box_slot:
                                            HorizontalAlignment := horizontal_alignment.Fill
                                            VerticalAlignment := vertical_alignment.Fill
                                            Distribution := option{1.0}
                                            Widget := NextButton

    <# ----------------------------------------------------------------------------------------------------
        Get Functions
    ---------------------------------------------------------------------------------------------------- #>
    GetMaxPage()<transacts>:int =
        Ceil[(GetTotalWeapons() * 1.0) / (EntriesPerPage * 1.0)] or 1

    <# ----------------------------------------------------------------------------------------------------
        Update Functions
    ---------------------------------------------------------------------------------------------------- #>
    Update(InPlayer:player):void =
        set MaybeActivePlayer = option{InPlayer}

        GoldMultiplierText.SetText(MakeMessage("+{FormatAsPercent(InPlayer.WeaponProgressGoldMultiplier())}"))
        CurrentPageText.SetText(MakeMessage("{CurrentPage}/{GetMaxPage()}"))
        
        StartIndex := (CurrentPage - 1) * EntriesPerPage
        WeaponIds := GetWeaponIds()
        for (Index -> PlayerProgressWeaponEntryUI : PlayerProgressWeaponEntryUIs):
            if (WeaponId := WeaponIds[StartIndex + Index]):
                Progress := InPlayer.GetWeaponProgress(WeaponId)
                PlayerProgressWeaponEntryUI.Update(InPlayer, WeaponId, Progress)
                PlayerProgressWeaponEntryUI.Show()
            else:
                PlayerProgressWeaponEntryUI.Hide()

    UpdatePage():void =
        if (ActivePlayer := MaybeActivePlayer?):
            Update(ActivePlayer)
            
    <# ----------------------------------------------------------------------------------------------------
        Event Handlers
    ---------------------------------------------------------------------------------------------------- #>
    OnClickPrevious(Message:widget_message):void =
        set CurrentPage = (CurrentPage = 1) and GetMaxPage() or CurrentPage - 1
        UpdatePage()
    
    OnClickNext(Message:widget_message):void =
        set CurrentPage = (CurrentPage = GetMaxPage()) and 1 or CurrentPage + 1
        UpdatePage()