using { /Fortnite.com/Characters }
using { /Verse.org/Assets }
using { /Verse.org/Random }
using { /Verse.org/Simulation }

<# ----------------------------------------------------------------------------------------------------
    Constants
---------------------------------------------------------------------------------------------------- #>
BaseRerolls:int = 3
BaseWeaponRerolls:int = 1
BaseBoonRerolls:int = 1
DefaultVulnerabilityBufferTime:float = 1.5
PlayerDamagedBufferTime:float = 1.0

<# ----------------------------------------------------------------------------------------------------
    Data Definitions
---------------------------------------------------------------------------------------------------- #>        
player_game_stat_id := enum:
    EliminatedMiniboss
    DamageDealt
    DamageDealtMiniboss
    DamageDealtBoss
    WoodCollected
    PlayersRevived
    AutoRevived
    TimesDowned

player_resource_id := enum:
    Reroll
    WeaponReroll
    BoonReroll
    Boost
    Revives

stat_id<public> := enum:
    Damage
    Cooldown
    Size
    CriticalRate
    CriticalDamage
    MoveSpeed
    DashCooldown
    GatherRate
    ItemPickupDistance
    BossDamage
    MaxHealth
    HealthRegeneration
    ExperienceMultiplier
    GoldMultiplier
    Luck

ToString(Id:stat_id)<transacts>:string =
    case (Id):
        stat_id.Damage => "Damage"
        stat_id.Cooldown => "Cooldown"
        stat_id.Size => "Size"
        stat_id.CriticalRate => "Critical Rate"
        stat_id.CriticalDamage => "Critical Damage"
        stat_id.MoveSpeed => "Move Speed"
        stat_id.DashCooldown => "Dash Cooldown"
        stat_id.GatherRate => "Gather Rate"
        stat_id.ItemPickupDistance => "Item Pickup Distance"
        stat_id.BossDamage => "Boss Damage"
        stat_id.MaxHealth => "Max Health"
        stat_id.HealthRegeneration => "Health Regeneration"
        stat_id.ExperienceMultiplier => "Experience Multiplier"
        stat_id.GoldMultiplier => "Gold Multiplier"
        stat_id.Luck => "Luck"

(Id:stat_id).ToIcon()<transacts>:texture =
    case (Id):
        stat_id.Damage => UserInterface.Icons.ProjectileDamage_64
        stat_id.Cooldown => UserInterface.Icons.AbilityCooldown_64
        stat_id.Size => UserInterface.Icons.AbilitySize_64
        stat_id.CriticalRate => UserInterface.Icons.CriticalRate_64
        stat_id.CriticalDamage => UserInterface.Icons.CriticalDamage_64
        stat_id.MoveSpeed => UserInterface.Icons.MoveSpeed_64
        stat_id.DashCooldown => UserInterface.Icons.DashCooldown_64
        stat_id.GatherRate => UserInterface.Icons.GatherRate_64
        stat_id.ItemPickupDistance => UserInterface.Icons.ItemPickupDistance_64
        stat_id.BossDamage => UserInterface.Icons.Boon_BossHunter_64
        stat_id.MaxHealth => UserInterface.Icons.MaxHealth_64
        stat_id.HealthRegeneration => UserInterface.Icons.HealthRegeneration_64
        stat_id.ExperienceMultiplier => UserInterface.Icons.ExperienceMultiplier_64
        stat_id.GoldMultiplier => UserInterface.Icons.Gold_64
        stat_id.Luck => UserInterface.Icons.AbilityLuck_64

(Value:float).Format(Id:stat_id)<transacts>:string =
    Format := case (Id):
        stat_id.MaxHealth => FormatAsFloat(Value, ?Decimals := 1)
        stat_id.HealthRegeneration => FormatAsRate(Value)
        _ => FormatAsPercent(Value, ?Decimals := 1)

stat_source := enum:
    Base
    Ability    
    Boon
    Boon2

<# ----------------------------------------------------------------------------------------------------
    Global Functions
---------------------------------------------------------------------------------------------------- #>
(Player:player).GetStats()<decides><transacts>:player_stat =
    GetPlayerManager[].GetPlayerStat[Player]

(Player:player).GetAbilityValue(AbilityInfo:ability_info)<transacts>:float =
    if (WeaponId := AbilityInfo.MaybeWeaponId?, StatId := AbilityInfo.MaybeStatId?):
        Player.GetWeaponStats(WeaponId, StatId)
    else:
        Player.GetAbilityValue(AbilityInfo.Id)

(Player:player).GetTotalAbilityValue(AbilityInfo:ability_info)<transacts>:float =
    if (WeaponId := AbilityInfo.MaybeWeaponId?, StatId := AbilityInfo.MaybeStatId?):
        Player.GetWeaponStats(WeaponId, StatId)
    else if (StatId := AbilityInfo.MaybeStatId?):
        Player.GetStat(StatId)
    else:
        Player.GetAbilityValue(AbilityInfo)

(Player:player).GetAbilityValue(Id:ability_id)<transacts>:float =
    Player.GetStats[].GetAbilityValue(Id) or 0.0

(Player:player).GetWeaponStats(WeaponId:weapon_id, StatId:stat_id)<transacts>:float =
    Player.GetStats[].GetWeaponStats(WeaponId, StatId) or 0.0

(Player:player).AddAbility(Id:ability_info, ?Tier:int = 1):void =
    if (PlayerStat := Player.GetStats[]):
        PlayerStat.AddAbility(Id, Tier)

(Player:player).AddWeaponEvolution(AbilityInfo:ability_info, EvolutionIndex:int):void =
    if (PlayerStat := Player.GetStats[]):
        PlayerStat.AddWeaponEvolution(AbilityInfo, EvolutionIndex)

(Player:player).GetBanishedAbilities()<transacts>:[]ability_info_id =
    Player.GetStats[].GetBanishedAbilities() or array{}

(Player:player).BanishAbility(AbilityInfo:ability_info):void =
    if (PlayerStat := Player.GetStats[]):
        PlayerStat.BanishAbility(AbilityInfo)

(Player:player).AddGameStat(Id:player_game_stat_id, Amount:float):void =
    if (PlayerStat := Player.GetStats[]):
        PlayerStat.AddGameStat(Id, Amount)
        PlayerStat.GameStatChangedEvent.Signal(Id, Amount)

(Player:player).GetGameStat(Id:player_game_stat_id)<transacts>:float =
    Player.GetStats[].GetGameStat(Id) or 0.0

(Player:player).ResetStats():void =
    if (PlayerStat := Player.GetStats[]):
        PlayerStat.ResetStats()

(Player:player).GetBanish()<transacts>:int =
    Player.GetStats[].GetBanish() or 0

(Player:player).GetPlayerResource(Id:player_resource_id)<transacts>:int =
    Player.GetStats[].GetPlayerResource[Id] or 0

(Player:player).AddPlayerResource(Id:player_resource_id, Amount:int):void =
    if (PlayerStats := Player.GetStats[]):
        PlayerStats.AddPlayerResource(Id, Amount)
    
(Player:player).RemovePlayerResource(Id:player_resource_id, Amount:int):void =
    if (PlayerStats := Player.GetStats[]):
        PlayerStats.RemovePlayerResource(Id, Amount)

(Player:player).PutInStasis():void =
    Player.PutInStasis(?IsAllowFalling := true)

(Player:player).PutInStasis(?IsAllowFalling:logic = true):void =
    if (PlayerStats := Player.GetStats[]):
        PlayerStats.PutInStasis(?IsAllowFalling := IsAllowFalling)

(Player:player).ReleaseFromStasis():void =
    if (PlayerStats := Player.GetStats[]):
        PlayerStats.ReleaseFromStasis()

(Player:player).IsInStasis()<transacts>:logic =
    Player.GetStats[].GetIsInStasis() or false

(Player:player).EquipWeapon(Id:weapon_id):void =
    if:
        CurrentPlayerData := Player.GetData[]
        set PlayerDataMap[Player] = player_data:
            MakePlayerData<constructor>(CurrentPlayerData)
            WeaponEquipped := Id.ToInt()

    if (WeaponEquippedEvent := Player.GetStats[].WeaponEquippedEvent):
        WeaponEquippedEvent.Signal()

    if (ManagerWeaponEquippedEvent := GetPlayerManager[].WeaponEquippedEvent):
        ManagerWeaponEquippedEvent.Signal(Player)

(Player:player).EquipNewWeapon(Id:weapon_id):void =
    if (PlayerStats := Player.GetStats[]):
        PlayerStats.EquipNewWeapon(Id)

(Player:player).GetEquippedWeapon()<transacts>:?weapon_id =
    option{Player.GetData[].WeaponEquipped.ToWeaponId() or DefaultWeapon}

(Player:player).GetWeaponLevels()<transacts>:[weapon_id]int =
    Player.GetStats[].GetWeaponLevels() or map{}

(Player:player).GetWeaponLevel(Id:weapon_id)<transacts>:int =
    Player.GetStats[].GetWeaponLevel(Id) or 0

(Player:player).HasWeapon(Id:weapon_id)<transacts>:logic =
    Player.GetStats[].HasWeapon(Id) or false

(Player:player).GetWeaponEvolutionLevel(Id:weapon_id)<transacts>:int =
    Player.GetStats[].GetWeaponEvolutionLevel(Id) or 0

(Player:player).GetWeaponEvolutions()<transacts>:[weapon_id][]ability_info =
    Player.GetStats[].GetWeaponEvolutions() or map{}

(Player:player).GetWeaponEvolution(Id:weapon_id)<transacts>:[]ability_info =
    Player.GetStats[].GetWeaponEvolution[Id] or array{}

(Player:player).GetWeaponEvolutionValue(WeaponId:weapon_id, Id:ability_id)<transacts>:float =
    var BaseValue:float = 0.0
    var IsFound:logic = false
    
    if (WeaponInfo := GetWeaponInfo[WeaponId]):
        for (EvolutionInfos : WeaponInfo.Evolution, not IsFound?):
            for (EvolutionInfo : EvolutionInfos, not IsFound?, EvolutionInfo.Id = Id):
                set BaseValue = EvolutionInfo.BaseValue
                set IsFound = true

    TotalTierValue := SumFloat of for (Index -> AbilityInfo : Player.GetWeaponEvolution(WeaponId), AbilityInfo.Id = Id, TierValue := AbilityInfo.TierValue[0]?) {TierValue}

    BaseValue + TotalTierValue

(Player:player).IsWeaponEvolutionAbilityActive(WeaponId:weapon_id, Id:ability_id)<transacts>:logic =
    logic{for (AbilityInfo : Player.GetWeaponEvolution(WeaponId), AbilityInfo.Id = Id) {AbilityInfo}[0]}

(Player:player).GetIsAFK()<transacts>:logic =
    Player.GetStats[].GetIsAFK() or false

(Player:player).SetAFK(IsAFK:logic):void =
    if (PlayerStats := Player.GetStats[]):
        PlayerStats.SetAFK(IsAFK)

(Player:player).Stun(Duration:float):void =
    if (Character := Player.GetFortCharacter[].IsVulnerable[], PlayerStats := Player.GetStats[]):
        spawn{PlayerStats.Stun(Duration)}

(Player:player).SetVulnerability(IsVulnerable:logic, ?BufferTime:float = 0.0):void =
    if (PlayerStats := Player.GetStats[]):
        PlayerStats.SetVulnerability(IsVulnerable, ?BufferTime := BufferTime)

CalculateWeaponEvolutionLevel(WeaponLevel:int)<transacts>:int =
    if (WeaponLevel >= 18). 3
    else if (WeaponLevel >= 12). 2
    else if (WeaponLevel >= 6). 1
    else. 0

(Player:player).ApplyDamage(Amount:float, ?IsBoss:logic = false):void =
    if (PlayerStats := Player.GetStats[], Character := Player.GetFortCharacter[]):
        LastDamagedTime := IsBoss? and PlayerStats.LastDamagedTimeBoss or PlayerStats.LastDamagedTime
        ElapsedTime := GetSimulationElapsedTime() - LastDamagedTime
        if (ElapsedTime >= PlayerDamagedBufferTime):
            if (Player.HasBoon(boon_id.NinjaRun)?, DodgeChance := GetBoonValueNinjaRun(Player), GetRandomFloat(0.0, 1.0) <= DodgeChance):
                SpawnVFX(vfx_id.Dodged, Character.GetTransform().Translation.Make2D(), ?Duration := 1.0)
            else:
                Character.Damage(Amount)

            if (IsBoss?):
                set PlayerStats.LastDamagedTimeBoss = GetSimulationElapsedTime()
            else:
                set PlayerStats.LastDamagedTime = GetSimulationElapsedTime()

(Player:player).GetBoons()<transacts>:[boon_id]int =
    Player.GetStats[].GetBoons() or map{}

(Player:player).HasBoon(Id:boon_id)<transacts>:logic =
    logic{Player.GetBoons()[Id]}

(Player:player).AddBoon(BoonInfo:boon_info):void =
    if (PlayerStats := Player.GetStats[]):
        PlayerStats.AddBoon(BoonInfo)

(Player:player).GetStat(Id:stat_id)<transacts>:float =
    Player.GetStats[].GetStat(Id) or 0.0

(Player:player).SetStat(Id:stat_id, Source:stat_source, Amount:float):void =
    if (PlayerStats := Player.GetStats[]):
        PlayerStats.SetStat(Id, Source, Amount)

(Player:player).IsAutoReviving()<transacts>:logic =
    Player.GetStats[].IsAutoReviving or false

(Player:player).GetMultiplicativeStat(Id:stat_id)<transacts>:float =
    Player.GetStats[].MultiplicativeStats[Id] or 0.0

<# ----------------------------------------------------------------------------------------------------
    Constructor
---------------------------------------------------------------------------------------------------- #>
MakePlayerStats<constructor>(UpdatedPlayerStats:player_stat)<transacts> := player_stat:
    Player := UpdatedPlayerStats.Player
    AbilityValues := UpdatedPlayerStats.AbilityValues
    MultiplicativeStats := UpdatedPlayerStats.MultiplicativeStats
    BanishedAbilities := UpdatedPlayerStats.BanishedAbilities
    WeaponLevels := UpdatedPlayerStats.WeaponLevels
    WeaponStats := UpdatedPlayerStats.WeaponStats
    WeaponEvolutions := UpdatedPlayerStats.WeaponEvolutions
    Boons := UpdatedPlayerStats.Boons
    BoonValues := UpdatedPlayerStats.BoonValues
    Stats := UpdatedPlayerStats.Stats
    TotalStats := UpdatedPlayerStats.TotalStats
    GameStats := UpdatedPlayerStats.GameStats
    Resources := UpdatedPlayerStats.Resources

<# ----------------------------------------------------------------------------------------------------
    player_stat
---------------------------------------------------------------------------------------------------- #>
player_stat := class:
    Player:player

    AbilityChangedEvent<public>:events(ability_id) = events(ability_id){}
    StatsResetEvent<public>:events(tuple()) = events(tuple()){}
    WeaponEquippedEvent<public>:events(tuple()) = events(tuple()){}
    NewWeaponEquippedEvent<public>:events(weapon_id) = events(weapon_id){}
    WeaponEvolvedEvent<public>:events(tuple()) = events(tuple()){}
    CancelUpdateVulnerabilityEvent:events(tuple()) = events(tuple()){}
    PlayerStatsChangedEvent<public>:events(tuple(player_resource_id, int)) = events(tuple(player_resource_id, int)){}
    GameStatChangedEvent:events(tuple(player_game_stat_id, float)) = events(tuple(player_game_stat_id, float)){}

    # Temporary stats
    # [Note] Constructor (above) needs to be updated if new stats are added.    
    var AbilityValues:[ability_id]float = map{}
    var MultiplicativeStats:[stat_id]float = map{}
    var BanishedAbilities:[]ability_info_id = array{}
    var WeaponLevels:[weapon_id]int = map{}
    var WeaponStats:[weapon_id][stat_id]float = map{}
    var WeaponEvolutions:[weapon_id][]int = map{}
    var Boons:[boon_id]int = map{}
    var BoonValues:[boon_id]float = map{}
    var Stats:[stat_id][stat_source]float = map{}
    var TotalStats:[stat_id]float = map{}
    var GameStats:[player_game_stat_id]float = map{}
    var Resources:[player_resource_id]int = map{}

    # Statuses
    var IsInStasis<private>:logic = false
    var IsStunned<private>:logic = false
    var IsAutoReviving:logic = false
    var IsAFK<private>:logic = false
    var LastDamagedTime:float = 0.0
    var LastDamagedTimeBoss:float = 0.0

    <# ----------------------------------------------------------------------------------------------------
        Get Functions
    ---------------------------------------------------------------------------------------------------- #>
    GetStat(Id:stat_id)<transacts>:float =
        TotalStats[Id] or 0.0

    SetStat(Id:stat_id, Source:stat_source, Amount:float):void =
        if:
            not Stats[Id]
            set Stats[Id] = map{}

        if (set Stats[Id][Source] = Amount) {}
        
        TotalStat := SumFloat of for (Stat : Stats[Id]) {Stat}
        if (set TotalStats[Id] = TotalStat) {}

    GetPlayerResource(Id:player_resource_id)<decides><transacts>:int =
        Resources[Id]

    GetBanish()<transacts>:int =
        MaxWeapons := GetGameSettings[].Weapons or 4
        MaxBanish := 10 - (4 - MaxWeapons) * 2
        PlayerBanish := Floor[Player.GetPowerUpValue(power_up_id.AbilityBanish)] or 0
        Min(MaxBanish, PlayerBanish) - BanishedAbilities.Length

    GetAbilityValue(Id:ability_id)<transacts>:float =
        AbilityValues[Id] or 0.0

    GetWeaponStats(WeaponId:weapon_id)<decides><transacts>:[stat_id]float =
        WeaponStats[WeaponId]

    GetWeaponStats(WeaponId:weapon_id, StatId:stat_id)<transacts>:float =
        WeaponStats[WeaponId][StatId] or 0.0

    GetAbilityEvolutionValue(WeaponId:weapon_id, AbilityId:ability_id)<transacts>:float =
        SumFloat of for (AbilityInfo : Player.GetWeaponEvolution(WeaponId), AbilityInfo.Id = AbilityId, TierValue := AbilityInfo.TierValue[0]?) {TierValue}

    GetGameStat(Id:player_game_stat_id)<transacts>:float =
        GameStats[Id] or 0.0

    GetIsInStasis()<transacts>:logic =
        (player_stat:)IsInStasis

    GetBanishedAbilities()<transacts>:[]ability_info_id =
        BanishedAbilities

    GetWeaponLevels()<transacts>:[weapon_id]int =
        WeaponLevels

    GetWeaponLevel(Id:weapon_id)<transacts>:int =
        WeaponLevels[Id] or 0

    GetWeaponEvolutionLevel(Id:weapon_id)<transacts>:int =
        CalculateWeaponEvolutionLevel(GetWeaponLevel(Id))

    # [TODO] Review for performance
    GetWeaponEvolutions()<transacts>:[weapon_id][]ability_info =
        var WeaponEvolutionInfos:[weapon_id][]ability_info = map{}
        
        for:
            Id -> WeaponEvolution : WeaponEvolutions
            WeaponEvolutionInfo := GetWeaponEvolution[Id]
            set WeaponEvolutionInfos[Id] = WeaponEvolutionInfo
        do {}

        WeaponEvolutionInfos

    GetWeaponEvolution(Id:weapon_id)<decides><transacts>:[]ability_info =
        WeaponInfo := GetWeaponInfo[Id]

        for (EvolutionLevel -> EvolutionIndex : WeaponEvolutions[Id]):
            WeaponInfo.Evolution[EvolutionLevel][EvolutionIndex]
        
    HasWeapon(Id:weapon_id)<transacts>:logic =
        logic{WeaponLevels[Id]}

    GetIsAFK()<transacts>:logic =
        IsAFK

    GetBoons()<transacts>:[boon_id]int =
        Boons

    <# ----------------------------------------------------------------------------------------------------
        Update Functions
    ---------------------------------------------------------------------------------------------------- #>
    AddAbility(AbilityInfo:ability_info, Tier:int):void =
        if (WeaponId := AbilityInfo.MaybeWeaponId?):
            LevelUpAmount := if:
                AbilityInfo.Id = ability_id.WeaponLevel
                TierValue := Floor[AbilityInfo.TierValue[Tier - 1]?]
            then:
                1 + TierValue
            else:
                if:
                    not WeaponStats[WeaponId]
                    set WeaponStats[WeaponId] = map{}

                if:
                    StatId := AbilityInfo.MaybeStatId?
                    TierValue := AbilityInfo.TierValue[Tier - 1]?
                    set WeaponStats[WeaponId][StatId] = WeaponStats[WeaponId][StatId] + TierValue or TierValue

                1

            if (set WeaponLevels[WeaponId] = WeaponLevels[WeaponId] + LevelUpAmount or LevelUpAmount) {}
            HandleAbilityChanged(WeaponId, AbilityInfo.Id)

        else if:
            TierValue := AbilityInfo.TierValue[Tier - 1]?
            TotalValue := TierValue * (1.0 + (MultiplicativeStats[AbilityInfo.MaybeStatId?] or 0.0))
            set AbilityValues[AbilityInfo.Id] = AbilityValues[AbilityInfo.Id] + TotalValue or TotalValue
        then:
            HandleAbilityChanged(AbilityInfo)

    HandleAbilityChanged(AbilityInfo:ability_info):void =
        if (StatId := AbilityInfo.MaybeStatId?, AbilityValue := AbilityValues[AbilityInfo.Id]):
            SetStat(StatId, stat_source.Ability, AbilityValue)

        AbilityChangedEvent.Signal(AbilityInfo.Id)

    HandleAbilityChanged(WeaponId:weapon_id, AbilityId:ability_id):void =
        AbilityChangedEvent.Signal(AbilityId)

        if (Player.HasBoon(boon_id.MultiTool)?): # [TODO] Review order before/after signal
            BoonOnGrantMultiTool(Player)

    AddWeaponEvolution(AbilityInfo:ability_info, EvolutionIndex:int):void =
        if (WeaponId := AbilityInfo.MaybeWeaponId?):
            if:
                not WeaponEvolutions[WeaponId]
                set WeaponEvolutions[WeaponId] = array{}

            if:
                set WeaponEvolutions[WeaponId] += array{EvolutionIndex}
            then:
                Player.AwardAccolade(accolade_id.WeaponEvolved)
                Player.PlayAudio(audio_id.WeaponEvolution)
                WeaponEvolvedEvent.Signal()

    AddGameStat(Id:player_game_stat_id, Amount:float):void =
        if (not GameStats[Id], set GameStats[Id] = Amount):
            return

        if (set GameStats[Id] += Amount) {}
    
    # Called on game start/end.
    (player_stat:)ResetStats():void =
        var BaseStats:[stat_id]float = map{}

        set AbilityValues = map{}
        set Stats = map{}
        set MultiplicativeStats = map{}

        for:
            AbilityInfo : GetAbilityInfo()
            StatId := AbilityInfo.MaybeStatId?
            set BaseStats[StatId] = AbilityInfo.BaseValue
        do {}

        for (PowerUpInfo : GetPowerUpInfo(), StatId := PowerUpInfo.MaybeStatId?):
            case (PowerUpInfo.StatType):
                stat_type.Add => if:
                    set BaseStats[StatId] = BaseStats[StatId] + Player.GetPowerUpValue(PowerUpInfo.Id) or Player.GetPowerUpValue(PowerUpInfo.Id)

                stat_type.Multiply => if:
                    Value := Player.GetPowerUpValue(PowerUpInfo.Id)
                    set MultiplicativeStats[StatId] = MultiplicativeStats[StatId] + Value or Value
                    set BaseStats[StatId] *= (1.0 + Value)

        if:
            NinjaMasteryValue := CalculateNinjaMasteryValue()
            set BaseStats[stat_id.Damage] += NinjaMasteryValue
            set BaseStats[stat_id.MaxHealth] += NinjaMasteryValue * 100.0

        for:
            StatSummary : Player.GetGearStatSummary()
            set BaseStats[StatSummary(0)] += StatSummary(1)
        do {}

        for (Id -> BaseStat : BaseStats):
            SetStat(Id, stat_source.Base, BaseStat)

        set WeaponStats = map{}
        set WeaponEvolutions = map{}
        set GameStats = map{}
        set Boons = map{}
        set BoonValues = map{}
        set Resources = map:
            player_resource_id.Reroll => BaseRerolls + Floor[Player.GetPowerUpValue(power_up_id.AbilityReroll)] or 0
            player_resource_id.WeaponReroll => BaseWeaponRerolls + Floor[Player.GetPowerUpValue(power_up_id.WeaponReroll)] or 0
            player_resource_id.Boost => Floor[Player.GetPowerUpValue(power_up_id.AbilityBoost)] or 0
            player_resource_id.Revives => Floor[Player.GetPowerUpValue(power_up_id.Revive)] or 0
            player_resource_id.BoonReroll => BaseBoonRerolls + Floor[Player.GetPowerUpValue(power_up_id.BoonReroll)] or 0

        set BanishedAbilities = array{}
        
        if (EquippedWeapon := Player.GetEquippedWeapon()?):
            set WeaponLevels = map{EquippedWeapon => 0}

        PlayerStatsChangedEvent.Signal(player_resource_id.Revives, Floor[Player.GetPowerUpValue(power_up_id.Revive)] or 0)
        StatsResetEvent.Signal()

    AddPlayerResource(Id:player_resource_id, Amount:int):void =
        if:
            UpdatedPlayerStats := Resources[Id] + Amount
            set Resources[Id] += Amount
        then:
            PlayerStatsChangedEvent.Signal(Id, UpdatedPlayerStats)

    RemovePlayerResource(Id:player_resource_id, Amount:int):void =
        if:
            UpdatedPlayerStats := Max(0, Resources[Id] - Amount)
            set Resources[Id] = UpdatedPlayerStats
        then:
            PlayerStatsChangedEvent.Signal(Id, UpdatedPlayerStats)

    PutInStasis(?IsAllowFalling:logic = true):void =
        if (Character := Player.GetFortCharacter[]):
            Character.PutInStasis(stasis_args{AllowTurning := false, AllowFalling := IsAllowFalling, AllowEmotes := true})
            set (player_stat:)IsInStasis = true

    ReleaseFromStasis():void =
        if (Character := Player.GetFortCharacter[]):
            Character.ReleaseFromStasis()
            set (player_stat:)IsInStasis = false

    EquipNewWeapon(Id:weapon_id):void =
        if (set WeaponLevels[Id] = 0) {}
        if (set WeaponStats[Id] = map{}) {}
        Player.AwardAccolade(accolade_id.EquipWeapon)
        NewWeaponEquippedEvent.Signal(Id)

    BanishAbility(AbilityInfo:ability_info):void =
        set BanishedAbilities += array{AbilityInfo.GetId()}

    SetAFK(AFK:logic):void =
        if (AFK = IsAFK) then return
        set IsAFK = AFK
        if (IsAFK?):
            Player.ActivateVFXPowerup(vfx_powerup_id.AFK)
        else:
            Player.DeactivateVFXPowerup(vfx_powerup_id.AFK)
        
    Stun(Duration:float)<suspends>:void =
        if (IsStunned?) then return
        set IsStunned = true

        if (Character := Player.GetFortCharacter[]):
            SpawnVFX(vfx_id.Stun, Character.GetTransform().Translation, ?Duration := Duration)
            Character.PutInStasis(stasis_args{AllowTurning := false, AllowFalling := true, AllowEmotes := false})
            Sleep(Duration)
            Character.ReleaseFromStasis()

        set IsStunned = false

    SetVulnerability(IsVulnerable:logic, ?BufferTime:float = 0.0):void =
        if (Player.IsInStageVolume()? and IsVulnerable?) then return
        if (IsVulnerable?, Player.GetFortCharacter[].IsDownButNotOut[]) then return

        CancelUpdateVulnerabilityEvent.Signal()
        if (BufferTime > 0.0):
            spawn{UpdateVulnerability(IsVulnerable, ?BufferTime := BufferTime)}
        else if (Character := Player.GetFortCharacter[]):
            Character.SetVulnerability(IsVulnerable)
            if (IsVulnerable?):
                Player.DeactivateVFXPowerup(vfx_powerup_id.Invulnerability)
            else:
                Player.ActivateVFXPowerup(vfx_powerup_id.Invulnerability)

    UpdateVulnerability(IsVulnerable:logic, ?BufferTime:float = 0.0)<suspends>:void =
        race:
            CancelUpdateVulnerabilityEvent.Await()

            if (Character := Player.GetFortCharacter[]):
                Sleep(BufferTime)
                Character.SetVulnerability(IsVulnerable)
                if (IsVulnerable?):
                    Player.DeactivateVFXPowerup(vfx_powerup_id.Invulnerability)
                else:
                    Player.ActivateVFXPowerup(vfx_powerup_id.Invulnerability)

    AddBoon(BoonInfo:boon_info):void =
        if (set Boons[BoonInfo.Id] = 1) {}
        if (OnGrant := BoonInfo.MaybeOnGrant?):
            OnGrant(Player)

    CalculateNinjaMasteryValue()<transacts>:float =
        SumFloat of for (Id : NinjaMasteryPowerUpIds):
            Player.GetPowerUpValue(Id)
