using { /Verse.org/Simulation }

player_xp_bar := class:  
    PlayerManager:player_manager
    
    <# ----------------------------------------------------------------------------------------------------
        Initialization
    ---------------------------------------------------------------------------------------------------- #>        
    Initialize():void =
        UpdateLevel()
        UpdateXP()

        PlayerManager.CinematicSequenceExperienceBar.Play()
        PlayerManager.CinematicSequenceExperienceBar.Pause()
        PlayerManager.CinematicSequenceExperienceBar.SetPlaybackTime(0.0)

        PlayerManager.LeveledUpEvent.Subscribe(OnLeveledUp)
        PlayerManager.XPChangedEvent.Subscribe(OnXPChanged)
        PlayerManager.StatsResetEvent.Subscribe(OnStatsReset)

    <# ----------------------------------------------------------------------------------------------------
        Update Functions
    ---------------------------------------------------------------------------------------------------- #>
    UpdateLevel():void =
        UpdateMessage(message_id.ExperienceBarLevel, "{PlayerManager.GetLevel()}")

    UpdateXP():void =
        XPPercent := (PlayerManager.GetXP() * 1.0) / (PlayerManager.GetXPRequired[] * 1.0) or 0.0
        UpdateMessage(message_id.ExperienceBar, "{PlayerManager.GetXP()}/{PlayerManager.GetXPRequired[] or 0} ({FormatAsPercent(XPPercent)})")

        if (Progress := (PlayerManager.GetXP() * 1.0)/(PlayerManager.GetXPRequired[] * 1.0)):
            PlayerManager.CinematicSequenceExperienceBar.SetPlaybackTime(Progress)
        else:
            PlayerManager.CinematicSequenceExperienceBar.SetPlaybackTime(1.0)

    <# ----------------------------------------------------------------------------------------------------
        Event Handlers
    ---------------------------------------------------------------------------------------------------- #>
    OnLeveledUp():void =
        UpdateLevel()

    OnXPChanged(XP:int):void =
        UpdateXP()

    OnStatsReset():void =
        UpdateLevel()
        UpdateXP()