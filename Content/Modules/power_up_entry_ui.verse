using { /Fortnite.com/UI }
using { /Verse.org/Colors }
using { /Verse.org/Simulation }
using { /UnrealEngine.com/Temporary/SpatialMath }
using { /UnrealEngine.com/Temporary/UI }

power_up_entry_ui := class(ui_component):
    ButtonClickedEvent:events(tuple()) = events(tuple()){}

    Icon:texture_block = texture_block:
        DefaultImage := UserInterface.Icons.ProjectileDamage_128
        DefaultDesiredSize := vector2{X := 100.0, Y := 100.0}

    Button:button_quiet = button_quiet{}

    ProgressHeight:float = 28.0

    var ColorBlocks:[]color_block = array{}
    var ProgressPulsings:[]widget = array{}
    var MaybeActiveInfo:?power_up_info = false
    var Subscriptions:[]cancelable = array{}

    <# ----------------------------------------------------------------------------------------------------
        Initialization
    ---------------------------------------------------------------------------------------------------- #>
    CreateUI<override>():widget =
        set Subscriptions = array{Button.OnClick().Subscribe(OnClickButton)}

        stack_box:
            Orientation := orientation.Vertical
            Slots := array:
                stack_box_slot:
                    HorizontalAlignment := horizontal_alignment.Fill
                    VerticalAlignment := vertical_alignment.Fill
                    Widget := overlay:
                        Slots := array:
                            overlay_slot:
                                Padding := margin{Left := 20.0, Top := 20.0, Right := 20.0, Bottom := 20.0}
                                Widget := Icon
                            overlay_slot:
                                HorizontalAlignment := horizontal_alignment.Fill
                                VerticalAlignment := vertical_alignment.Fill
                                Widget := Button
                stack_box_slot:
                    HorizontalAlignment := horizontal_alignment.Fill
                    VerticalAlignment := vertical_alignment.Fill
                    Padding := margin{Left := 10.0, Top := 10.0, Right := 10.0, Bottom := 0.0}
                    Widget := stack_box:
                        Orientation := orientation.Horizontal
                        Slots := for (Index := 1..10):
                            ColorBlock := color_block:
                                DefaultDesiredSize := vector2{X := 1.0, Y := ProgressHeight}
            
                            ProgressPulsing := UserInterface.VerseWidget.WBP_ProgressPulsing{}

                            set ColorBlocks += array{ColorBlock}
                            set ProgressPulsings += array{ProgressPulsing}
            
                            stack_box_slot:
                                HorizontalAlignment := horizontal_alignment.Fill
                                VerticalAlignment := vertical_alignment.Fill
                                Distribution := option{1.0}
                                Padding := margin{Left := (Index = 0) and 0.0 or 2.0, Top := 0.0, Right := 0.0, Bottom := 0.0}
                                Widget := overlay:
                                    Slots := array:
                                        overlay_slot:
                                            HorizontalAlignment := horizontal_alignment.Fill
                                            VerticalAlignment := vertical_alignment.Fill
                                            Widget := color_block:
                                                DefaultColor := NamedColors.Black
                                                DefaultOpacity := 0.5
                                                DefaultDesiredSize := vector2{X := 1.0, Y := ProgressHeight}
                                        overlay_slot:
                                            HorizontalAlignment := horizontal_alignment.Fill
                                            VerticalAlignment := vertical_alignment.Fill
                                            Padding := margin{Left := 1.0, Top := 1.0, Right := 1.0, Bottom := 1.0}
                                            Widget := ColorBlock
                                        overlay_slot:
                                            HorizontalAlignment := horizontal_alignment.Fill
                                            VerticalAlignment := vertical_alignment.Fill
                                            Padding := margin{Left := 1.0, Top := 1.0, Right := 1.0, Bottom := 1.0}
                                            Widget := ProgressPulsing
                                        overlay_slot:
                                            HorizontalAlignment := horizontal_alignment.Fill
                                            VerticalAlignment := vertical_alignment.Top
                                            Padding := margin{Left := 1.0, Top := 1.0, Right := 1.0, Bottom := 1.0}
                                            Widget := color_block:
                                                DefaultColor := NamedColors.White
                                                DefaultOpacity := 0.1
                                                DefaultDesiredSize := vector2{X := 1.0, Y := ProgressHeight * 0.5}

    Uninitialize():void =
        for (Subscription : Subscriptions):
            Subscription.Cancel()
            
    <# ----------------------------------------------------------------------------------------------------
        Get Functions
    ---------------------------------------------------------------------------------------------------- #>

    <# ----------------------------------------------------------------------------------------------------
        Update Functions
    ---------------------------------------------------------------------------------------------------- #>
    Update(Info:power_up_info):void =
        set MaybeActiveInfo = option{Info}
        Icon.SetImage(Info.Icon)

        PowerUpLevel := Player.GetPowerUpLevel(Info.Id)
        for (Index -> ColorBlock : ColorBlocks, ParentWidget := ColorBlock.GetParentWidget[]):
            if (Index < Info.MaxLevel):
                ParentWidget.SetVisibility(widget_visibility.Visible)
                ColorBlock.SetColor((Index < PowerUpLevel) and ThemeColors.Highlight or ThemeColors.Inactive)
            else:
                ParentWidget.SetVisibility(widget_visibility.Collapsed)

        for (ProgressPulsing : ProgressPulsings):
            ProgressPulsing.SetVisibility(widget_visibility.Hidden)

        if:
            Info.IsUnlocked(Player)?
            PowerUpCost := CalculatePowerUpCost[Info.Id, PowerUpLevel + 1]
            Player.GetGoldAvailable(currency_id.Gold) >= PowerUpCost
            Player.GetPowerUpLevel(Info.Id) < Info.MaxLevel
            ProgressPulsing := ProgressPulsings[PowerUpLevel]
        then:
            ProgressPulsing.SetVisibility(widget_visibility.Visible)

        if (Info.IsUnlocked(Player)?):
            Button.SetText(MakeMessage(""))
        else:
            Button.SetText(MakeMessage("🔒"))
            
    SetActive():void =
        Icon.SetTint(ThemeColors.Highlight)
        
        if:
            ActiveInfo := MaybeActiveInfo?
            PowerUpCost := CalculatePowerUpCost[ActiveInfo.Id, Player.GetPowerUpLevel(ActiveInfo.Id) + 1]
            Player.GetGoldAvailable(currency_id.Gold) >= PowerUpCost
            Player.GetPowerUpLevel(ActiveInfo.Id) < ActiveInfo.MaxLevel
        then:
            Button.SetText(MakeMessage("{ActiveInfo.IsUnlocked(Player)? and "Buy" or "🔒"}"))
        else:
            Button.SetText(MakeMessage(""))

    ClearActive():void =
        Icon.SetTint(NamedColors.White)
        if (ActiveInfo := MaybeActiveInfo?, ActiveInfo.IsUnlocked(Player)?):
            Button.SetText(MakeMessage(""))
        else:
            Button.SetText(MakeMessage("🔒"))

    Hide<override>():void =
        UI.SetVisibility(widget_visibility.Hidden)

    <# ----------------------------------------------------------------------------------------------------
        Event Handlers
    ---------------------------------------------------------------------------------------------------- #>
    OnClickButton(Message:widget_message):void =
        ButtonClickedEvent.Signal()
