using { /Verse.org/Assets }
using { /Verse.org/Simulation }

<# ----------------------------------------------------------------------------------------------------
    Data Definitions
---------------------------------------------------------------------------------------------------- #>
power_up_id := enum:
    Damage
    Cooldown
    Size
    MaxHealth
    HealthRegeneration
    ItemPickupDistance
    CriticalRate
    CriticalDamage
    ExperienceMultiplier
    AbilityReroll
    AbilityBanish
    AbilityBoost
    DashCount
    Luck
    GoldMultiplier
    WeaponReroll
    TreasureChestLuck
    Revive
    GatherRate
    TreasureChestFocus
    ReviveRegeneration
    AirDash
    BoonReroll
    TreasureChestTier
    ReviveBlast
    MoveSpeed
    NinjaMastery01
    NinjaMastery02
    NinjaMastery03
    NinjaMastery04
    NinjaMastery05
    NinjaMastery06
    NinjaMastery07
    NinjaMastery08
    NinjaMastery09
    HealthPickup
    GoldenTreeRate
    TreasureChestMultiplier
    HealthPickupDuration
    HealthPickupSize

stat_type := enum:
    Add
    Multiply

ToString(Type:stat_type)<transacts>:string =
    case (Type):
        stat_type.Add => "Additive"
        stat_type.Multiply => "Multiplicative"

(Id:power_up_id).ToInt()<transacts>:int =
    case (Id):
        power_up_id.Damage => 0
        power_up_id.Cooldown => 1
        power_up_id.Size => 2
        power_up_id.MaxHealth => 3
        power_up_id.HealthRegeneration => 4
        power_up_id.ItemPickupDistance => 5
        power_up_id.CriticalRate => 6
        power_up_id.CriticalDamage => 7
        power_up_id.ExperienceMultiplier => 8
        power_up_id.AbilityReroll => 9
        power_up_id.AbilityBanish => 10
        power_up_id.AbilityBoost => 11
        power_up_id.DashCount => 12
        power_up_id.Luck => 13
        power_up_id.GoldMultiplier => 14
        power_up_id.WeaponReroll => 15
        power_up_id.TreasureChestLuck => 16
        power_up_id.Revive => 17
        power_up_id.GatherRate => 18
        power_up_id.TreasureChestFocus => 19
        power_up_id.ReviveRegeneration => 20
        power_up_id.AirDash => 21
        power_up_id.BoonReroll => 22
        power_up_id.TreasureChestTier => 23
        power_up_id.ReviveBlast => 24
        power_up_id.MoveSpeed => 25
        power_up_id.NinjaMastery01 => 26
        power_up_id.NinjaMastery02 => 27
        power_up_id.NinjaMastery03 => 28
        power_up_id.NinjaMastery04 => 29
        power_up_id.NinjaMastery05 => 30
        power_up_id.NinjaMastery06 => 31
        power_up_id.NinjaMastery07 => 32
        power_up_id.NinjaMastery08 => 33
        power_up_id.NinjaMastery09 => 34
        power_up_id.HealthPickup => 35
        power_up_id.GoldenTreeRate => 36
        power_up_id.TreasureChestMultiplier => 37
        power_up_id.HealthPickupDuration => 38
        power_up_id.HealthPickupSize => 39

power_up_info := struct<computes>:
    Id:power_up_id
    Name:string
    Description:string
    Type:power_up_type
    Icon:texture
    IconSmall:texture
    BaseCost:float
    CostFactor:float = 2.0
    MaxLevel:int
    BaseValue:float = 0.0
    IncrementValue:float
    MaybeStatId:?stat_id = false
    FormatValue:type{_(:float)<transacts>:string}
    IsDeltaValue:logic = true
    IsPermanent:logic = false
    StatType:stat_type = stat_type.Add
    IsUnlocked:type{_(:player)<transacts>:logic}
    
IsUnlockedPowerUpDefault(Player:player)<transacts>:logic =
    true

power_up_type := enum:
    AbilityBuff
    AbilitySelect
    Income
    Survival
    NinjaMastery
    
PowerUpTypes:[]power_up_type = array:
    power_up_type.AbilityBuff
    power_up_type.AbilitySelect
    power_up_type.Income
    power_up_type.Survival
    power_up_type.NinjaMastery

ToString(PowerUpType:power_up_type)<transacts>:string =
    case (PowerUpType):
        power_up_type.AbilityBuff => "Ability Buff"
        power_up_type.AbilitySelect => "Ability Select"
        power_up_type.Income => "Income"
        power_up_type.Survival => "Survival"
        power_up_type.NinjaMastery => "Ninja Mastery"

<# ----------------------------------------------------------------------------------------------------
    PowerUps Info
---------------------------------------------------------------------------------------------------- #>
PowerUps:[power_up_id]power_up_info = map:
    power_up_id.Damage => power_up_info:
        Id := power_up_id.Damage
        Name := "Damage"
        Description := "Increase base damage of all weapons."
        Type := power_up_type.AbilityBuff
        Icon := UserInterface.Icons.ProjectileDamage_128
        IconSmall := UserInterface.Icons.ProjectileDamage_64
        BaseCost := 100.0
        MaxLevel := 10
        IncrementValue := 0.1
        FormatValue := FormatAsPercent
        MaybeStatId := option{stat_id.Damage}
        IsUnlocked := IsUnlockedPowerUpDefault
    power_up_id.Cooldown => power_up_info:
        Id := power_up_id.Cooldown
        Name := "Cooldown"
        Description := "Reduce the cooldown time of all weapons."
        Type := power_up_type.AbilityBuff
        Icon := UserInterface.Icons.AbilityCooldown_128
        IconSmall := UserInterface.Icons.AbilityCooldown_64
        BaseCost := 100.0
        MaxLevel := 10
        IncrementValue := 0.05
        FormatValue := FormatAsPercent
        MaybeStatId := option{stat_id.Cooldown}
        IsUnlocked := IsUnlockedPowerUpDefault
    power_up_id.Size => power_up_info:
        Id := power_up_id.Size
        Name := "Size"
        Description := "Increase the size of all weapons."
        Type := power_up_type.AbilityBuff
        Icon := UserInterface.Icons.AbilitySize_128
        IconSmall := UserInterface.Icons.AbilitySize_64
        BaseCost := 100.0
        MaxLevel := 10
        IncrementValue := 0.05
        FormatValue := FormatAsPercent
        MaybeStatId := option{stat_id.Size}
        IsUnlocked := IsUnlockedPowerUpDefault
    power_up_id.MaxHealth => power_up_info:
        Id := power_up_id.MaxHealth
        Name := "Max Health"
        Description := "Increase base max health."
        Type := power_up_type.AbilityBuff
        Icon := UserInterface.Icons.MaxHealth_128
        IconSmall := UserInterface.Icons.MaxHealth_64
        BaseCost := 100.0
        MaxLevel := 10
        IncrementValue := 0.1
        FormatValue := FormatAsPercent
        MaybeStatId := option{stat_id.MaxHealth}
        StatType := stat_type.Multiply
        IsUnlocked := IsUnlockedPowerUpDefault
    power_up_id.HealthRegeneration => power_up_info:
        Id := power_up_id.HealthRegeneration
        Name := "Health Regeneration"
        Description := "Increase base health regeneration rate."
        Type := power_up_type.AbilityBuff
        Icon := UserInterface.Icons.HealthRegeneration_128
        IconSmall := UserInterface.Icons.HealthRegeneration_64
        BaseCost := 100.0
        MaxLevel := 10
        IncrementValue := 0.1
        FormatValue := FormatAsPercent
        MaybeStatId := option{stat_id.HealthRegeneration}
        StatType := stat_type.Multiply
        IsUnlocked := IsUnlockedPowerUpDefault
    power_up_id.ItemPickupDistance => power_up_info:
        Id := power_up_id.ItemPickupDistance
        Name := "Pickup Range"
        Description := "Increase distance player can pickup EXP orbs."
        Type := power_up_type.AbilityBuff
        Icon := UserInterface.Icons.ItemPickupDistance_128
        IconSmall := UserInterface.Icons.ItemPickupDistance_64
        BaseCost := 100.0
        MaxLevel := 10
        IncrementValue := 0.1
        FormatValue := FormatAsPercent
        MaybeStatId := option{stat_id.ItemPickupDistance}
        IsUnlocked := IsUnlockedPowerUpDefault
    power_up_id.CriticalRate => power_up_info:
        Id := power_up_id.CriticalRate
        Name := "Critical Rate"
        Description := "Increase chance of critical hit."
        Type := power_up_type.AbilityBuff
        Icon := UserInterface.Icons.CriticalRate_128
        IconSmall := UserInterface.Icons.CriticalRate_64
        BaseCost := 100.0
        MaxLevel := 10
        IncrementValue := 0.05
        FormatValue := FormatAsPercent
        MaybeStatId := option{stat_id.CriticalRate}
        IsUnlocked := IsUnlockedPowerUpDefault
    power_up_id.CriticalDamage => power_up_info:
        Id := power_up_id.CriticalDamage
        Name := "Critical Damage"
        Description := "Increase critical damage multiplier."
        Type := power_up_type.AbilityBuff
        Icon := UserInterface.Icons.CriticalDamage_128
        IconSmall := UserInterface.Icons.CriticalDamage_64
        BaseCost := 100.0
        MaxLevel := 10
        IncrementValue := 0.05
        FormatValue := FormatAsPercent
        MaybeStatId := option{stat_id.CriticalDamage}
        IsUnlocked := IsUnlockedPowerUpDefault
    power_up_id.ExperienceMultiplier => power_up_info:
        Id := power_up_id.ExperienceMultiplier
        Name := "Experience Multiplier"
        Description := "Increase experience earned from orbs."
        Type := power_up_type.AbilityBuff
        Icon := UserInterface.Icons.ExperienceMultiplier_128
        IconSmall := UserInterface.Icons.ExperienceMultiplier_64
        BaseCost := 100.0
        MaxLevel := 10
        IncrementValue := 0.1
        FormatValue := FormatAsPercent
        MaybeStatId := option{stat_id.ExperienceMultiplier}
        IsUnlocked := IsUnlockedPowerUpDefault

    power_up_id.AbilityReroll => power_up_info:
        Id := power_up_id.AbilityReroll
        Name := "Ability Reroll"
        Description := "Increase amount of rerolls per run. Reroll replaces the offered abilities with different ones."
        Type := power_up_type.AbilitySelect
        Icon := UserInterface.Icons.AbilityReroll_128
        IconSmall := UserInterface.Icons.AbilityReroll_64
        BaseCost := 500.0
        CostFactor := 2.0
        MaxLevel := 10
        IncrementValue := 1.0
        FormatValue := FormatAsInt
        IsUnlocked := IsUnlockedPowerUpDefault
    power_up_id.AbilityBanish => power_up_info:
        Id := power_up_id.AbilityBanish
        Name := "Ability Banish"
        Description := "Increase amount of banishes per run. Banish stops the ability from being offered for the rest of the run."
        Type := power_up_type.AbilitySelect
        Icon := UserInterface.Icons.AbilityBanish_128
        IconSmall := UserInterface.Icons.AbilityBanish_64
        BaseCost := 750.0
        CostFactor := 2.0
        MaxLevel := 10
        IncrementValue := 1.0
        FormatValue := FormatAsInt
        IsUnlocked := IsUnlockedPowerUpDefault
    power_up_id.AbilityBoost => power_up_info:
        Id := power_up_id.AbilityBoost
        Name := "Ability Boost"
        Description := "Increase amount of boosts per run. Boosts increases the tier of the selected ability (e.g. Epic > Legendary)."
        Type := power_up_type.AbilitySelect
        Icon := UserInterface.Icons.AbilityBoost_128
        IconSmall := UserInterface.Icons.AbilityBoost_64
        BaseCost := 1000.0
        CostFactor := 3.0
        MaxLevel := 5
        IncrementValue := 1.0
        FormatValue := FormatAsInt
        IsUnlocked := IsUnlockedPowerUpDefault
    power_up_id.WeaponReroll => power_up_info:
        Id := power_up_id.WeaponReroll
        Name := "Weapon Reroll"
        Description := "Increase amount of weapon rerolls per run. Reroll replaces the offered weapons with different ones."
        Type := power_up_type.AbilitySelect
        Icon := UserInterface.Icons.RerollWeapon_128
        IconSmall := UserInterface.Icons.RerollWeapon_64
        BaseCost := 1500.0
        CostFactor := 3.0
        MaxLevel := 10
        IncrementValue := 1.0
        FormatValue := FormatAsInt
        IsUnlocked := IsUnlockedPowerUpDefault
    power_up_id.BoonReroll => power_up_info:
        Id := power_up_id.BoonReroll
        Name := "Boon Reroll"
        Description := "Increase amount of boon rerolls per run. Reroll replaces the offered boons with different ones."
        Type := power_up_type.AbilitySelect
        Icon := UserInterface.Icons.BoonReroll_128
        IconSmall := UserInterface.Icons.BoonReroll_64
        BaseCost := 2000.0
        CostFactor := 3.0
        MaxLevel := 5
        IncrementValue := 1.0
        FormatValue := FormatAsInt
        IsUnlocked := IsUnlockedPowerUpDefault
    power_up_id.Luck => power_up_info:
        Id := power_up_id.Luck
        Name := "Luck"
        Description := "Increase chance of receiving higher tier abilities (chance of legendary is doubled at 50% luck). Doesn't apply to item shop."
        Type := power_up_type.AbilitySelect
        Icon := UserInterface.Icons.AbilityLuck_128
        IconSmall := UserInterface.Icons.AbilityLuck_64
        BaseCost := 750.0
        CostFactor := 2.0
        MaxLevel := 10
        IncrementValue := 0.1
        FormatValue := FormatAsPercent
        MaybeStatId := option{stat_id.Luck}
        IsUnlocked := IsUnlockedPowerUpDefault
    power_up_id.TreasureChestLuck => power_up_info:
        Id := power_up_id.TreasureChestLuck
        Name := "Treasure Luck"
        Description := "Increase chance of receiving higher tiers from Treasure Chests (stacks with Luck stats)."
        Type := power_up_type.AbilitySelect
        Icon := UserInterface.Icons.TreasureChestLuck_128
        IconSmall := UserInterface.Icons.TreasureChestLuck_64
        BaseCost := 500.0
        CostFactor := 2.0
        MaxLevel := 10
        IncrementValue := 0.2
        FormatValue := FormatAsPercent
        IsUnlocked := IsUnlockedPowerUpDefault
    power_up_id.TreasureChestFocus => power_up_info:
        Id := power_up_id.TreasureChestFocus
        Name := "Treasure Focus"
        Description := "Treasure chests will only grants abilities that the player has selected during the game (from level up/item shop)."
        Type := power_up_type.AbilitySelect
        Icon := UserInterface.Icons.TreasureChestFocus_128
        IconSmall := UserInterface.Icons.TreasureChestFocus_64
        BaseCost := 25000.0
        CostFactor := 0.0
        MaxLevel := 1
        IncrementValue := 1.0
        FormatValue := FormatAsInt
        IsUnlocked := IsUnlockedPowerUpDefault
    power_up_id.TreasureChestTier => power_up_info:
        Id := power_up_id.TreasureChestTier
        Name := "Treasure Tier"
        Description := "Increases the lowest tier that Treasure Chests grant (e.g. Treasure Chest will no longer grant Common tier at level 1)."
        Type := power_up_type.AbilitySelect
        Icon := UserInterface.Icons.TreasureChestTier_128
        IconSmall := UserInterface.Icons.TreasureChestTier_64
        BaseCost := 75000.0
        CostFactor := 3.0
        MaxLevel := 2
        IncrementValue := 1.0
        FormatValue := FormatAsTier
        IsDeltaValue := false
        IsUnlocked := IsUnlockedPowerUpDefault

    power_up_id.DashCount => power_up_info:
        Id := power_up_id.DashCount
        Name := "Dash Count"
        Description := "Increase the max charge of dash ability."
        Type := power_up_type.Survival
        Icon := UserInterface.Icons.DashCooldown_128
        IconSmall := UserInterface.Icons.DashCooldown_64
        BaseCost := 1000.0
        CostFactor := 3.0
        MaxLevel := 5
        IncrementValue := 1.0
        FormatValue := FormatAsInt
        IsUnlocked := IsUnlockedPowerUpDefault
    power_up_id.AirDash => power_up_info:
        Id := power_up_id.AirDash
        Name := "Air Dash"
        Description := "Enables the ability to activate an extra dash while airborne (can be used to jump over fences)."
        Type := power_up_type.Survival
        Icon := UserInterface.Icons.AirDash_128
        IconSmall := UserInterface.Icons.AirDash_64
        BaseCost := 5000.0
        CostFactor := 1.0
        MaxLevel := 2
        IncrementValue := 1.0
        FormatValue := FormatAsInt
        IsUnlocked := IsUnlockedPowerUpDefault
    power_up_id.MoveSpeed => power_up_info:
        Id := power_up_id.MoveSpeed
        Name := "Movement Speed"
        Description := "Increase base movement speed."
        Type := power_up_type.Survival
        Icon := UserInterface.Icons.MoveSpeed_128
        IconSmall := UserInterface.Icons.MoveSpeed_64
        BaseCost := 800.0
        CostFactor := 2.0
        MaxLevel := 5
        IncrementValue := 0.1
        FormatValue := FormatAsPercent
        MaybeStatId := option{stat_id.MoveSpeed}
        IsUnlocked := IsUnlockedPowerUpDefault
    power_up_id.HealthPickup => power_up_info:
        Id := power_up_id.HealthPickup
        Name := "Health Pickup"
        Description := "Increases the amount healed from health pickups (% of max health)."
        Type := power_up_type.Survival
        Icon := UserInterface.Icons.HealthPickup_128
        IconSmall := UserInterface.Icons.HealthPickup_64
        BaseCost := 500.0
        CostFactor := 2.0
        MaxLevel := 10
        BaseValue := 0.5
        IncrementValue := 0.05
        FormatValue := FormatAsPercent
        IsUnlocked := IsUnlockedPowerUpDefault
    power_up_id.HealthPickupDuration => power_up_info:
        Id := power_up_id.HealthPickupDuration
        Name := "Health Pickup Duration"
        Description := "Increases the duration of health pickup team healing effect (multiplayer only)."
        Type := power_up_type.Survival
        Icon := UserInterface.Icons.HealthPickupDuration_128
        IconSmall := UserInterface.Icons.HealthPickupDuration_64
        BaseCost := 1000.0
        CostFactor := 2.0
        MaxLevel := 5
        BaseValue := 10.0
        IncrementValue := 1.0
        FormatValue := FormatAsTime
        IsUnlocked := IsUnlockedPowerUpDefault
    power_up_id.HealthPickupSize => power_up_info:
        Id := power_up_id.HealthPickupSize
        Name := "Health Pickup Size"
        Description := "Increases the size of health pickup team healing effect (multiplayer only)."
        Type := power_up_type.Survival
        Icon := UserInterface.Icons.HealthPickupSize_128
        IconSmall := UserInterface.Icons.HealthPickupSize_64
        BaseCost := 2000.0
        CostFactor := 2.0
        MaxLevel := 10
        IncrementValue := 0.1
        FormatValue := FormatAsPercent
        IsUnlocked := IsUnlockedPowerUpDefault
    power_up_id.Revive => power_up_info:
        Id := power_up_id.Revive
        Name := "Revive"
        Description := "Automatically revive when you are downed."
        Type := power_up_type.Survival
        Icon := UserInterface.Icons.Revive_128
        IconSmall := UserInterface.Icons.Revive_64
        BaseCost := 50000.0
        CostFactor := 2.0
        MaxLevel := 2
        IncrementValue := 1.0
        FormatValue := FormatAsInt
        IsUnlocked := IsUnlockedPowerUpDefault
    power_up_id.ReviveRegeneration => power_up_info:
        Id := power_up_id.ReviveRegeneration
        Name := "Revive Regeneration"
        Description := "Revives can regenerate after a cooldown (requires revive power up)."
        Type := power_up_type.Survival
        Icon := UserInterface.Icons.ReviveRegeneration_128
        IconSmall := UserInterface.Icons.ReviveRegeneration_64
        BaseCost := 1000.0
        CostFactor := 2.0
        MaxLevel := 10
        BaseValue := 180.0
        IncrementValue := -6.0
        FormatValue := FormatAsTime
        IsUnlocked := IsUnlockedPowerUpDefault
    power_up_id.ReviveBlast => power_up_info:
        Id := power_up_id.ReviveBlast
        Name := "Revive Blast"
        Description := "Deals damage to enemies around you on auto revive. Damage and area increase with level."
        Type := power_up_type.Survival
        Icon := UserInterface.Icons.ReviveBlast_128
        IconSmall := UserInterface.Icons.ReviveBlast_64
        BaseCost := 2000.0
        CostFactor := 2.0
        MaxLevel := 10
        IncrementValue := 0.1
        FormatValue := FormatAsPercent
        IsUnlocked := IsUnlockedPowerUpDefault

    power_up_id.GoldMultiplier => power_up_info:
        Id := power_up_id.GoldMultiplier
        Name := "Gold Multiplier"
        Description := "Increase amount of gold earned from completing a game."
        Type := power_up_type.Income
        Icon := UserInterface.Icons.Gold_128
        IconSmall := UserInterface.Icons.Gold_64
        BaseCost := 5000.0
        CostFactor := 2.0
        MaxLevel := 10
        IncrementValue := 0.2
        FormatValue := FormatAsPercent
        MaybeStatId := option{stat_id.GoldMultiplier}
        IsUnlocked := IsUnlockedPowerUpDefault
    power_up_id.GatherRate => power_up_info:
        Id := power_up_id.GatherRate
        Name := "Gather Rate"
        Description := "Increase wood gathering speed & amount (x2 effectiveness for gathering speed)."
        Type := power_up_type.Income
        Icon := UserInterface.Icons.GatherRate_128
        IconSmall := UserInterface.Icons.GatherRate_64
        BaseCost := 2000.0
        CostFactor := 2.0
        MaxLevel := 5
        IncrementValue := 0.05
        FormatValue := FormatAsPercent
        MaybeStatId := option{stat_id.GatherRate}
        IsUnlocked := IsUnlockedPowerUpDefault
    power_up_id.GoldenTreeRate => power_up_info:
        Id := power_up_id.GoldenTreeRate
        Name := "Golden Tree Rate"
        Description := "Increase chance of Golden Tree spawning on stages 2+ (average of all players)."
        Type := power_up_type.Income
        Icon := UserInterface.Icons.GoldenTree_128
        IconSmall := UserInterface.Icons.GoldenTree_64
        BaseCost := 5000.0
        CostFactor := 2.0
        MaxLevel := 10
        BaseValue := 0.05
        IncrementValue := 0.01
        FormatValue := FormatAsPercent
        IsUnlocked := IsUnlockedPowerUpDefault
    power_up_id.TreasureChestMultiplier => power_up_info:
        Id := power_up_id.TreasureChestMultiplier
        Name := "Treasure Chest Multiplier"
        Description := "Increase amount of treasure chests (for gear) earned from completing a game."
        Type := power_up_type.Income
        Icon := UserInterface.Icons.TreasureChest_128
        IconSmall := UserInterface.Icons.TreasureChest_64
        BaseCost := 5000.0
        CostFactor := 2.0
        MaxLevel := 10
        IncrementValue := 0.2
        FormatValue := FormatAsPercent
        IsUnlocked := IsUnlockedPowerUpDefault

    power_up_id.NinjaMastery01 => power_up_info:
        Id := power_up_id.NinjaMastery01
        Name := "Genin I"
        Description := "Increases both base damage and max health a small amount."
        Type := power_up_type.NinjaMastery
        Icon := UserInterface.Icons.RankIcon_01_128
        IconSmall := UserInterface.Icons.RankIcon_01_64
        BaseCost := 2500.0
        CostFactor := 1.2
        MaxLevel := 10
        IncrementValue := 0.02
        FormatValue := FormatAsPercent
        IsUnlocked := IsUnlockedPowerUpNinjaMastery1
    power_up_id.NinjaMastery02 => power_up_info:
        Id := power_up_id.NinjaMastery02
        Name := "Genin II"
        Description := "Increases both base damage and max health a small amount."
        Type := power_up_type.NinjaMastery
        Icon := UserInterface.Icons.RankIcon_02_128
        IconSmall := UserInterface.Icons.RankIcon_02_64
        BaseCost := 5000.0
        CostFactor := 1.2
        MaxLevel := 10
        IncrementValue := 0.04
        FormatValue := FormatAsPercent
        IsUnlocked := IsUnlockedPowerUpNinjaMastery2
    power_up_id.NinjaMastery03 => power_up_info:
        Id := power_up_id.NinjaMastery03
        Name := "Genin III"
        Description := "Increases both base damage and max health a small amount."
        Type := power_up_type.NinjaMastery
        Icon := UserInterface.Icons.RankIcon_03_128
        IconSmall := UserInterface.Icons.RankIcon_03_64
        BaseCost := 10000.0
        CostFactor := 1.2
        MaxLevel := 10
        IncrementValue := 0.06
        FormatValue := FormatAsPercent
        IsUnlocked := IsUnlockedPowerUpNinjaMastery3
    power_up_id.NinjaMastery04 => power_up_info:
        Id := power_up_id.NinjaMastery04
        Name := "Chunin I"
        Description := "Increases both base damage and max health a moderate amount."
        Type := power_up_type.NinjaMastery
        Icon := UserInterface.Icons.RankIcon_04_128
        IconSmall := UserInterface.Icons.RankIcon_04_64
        BaseCost := 20000.0
        CostFactor := 1.2
        MaxLevel := 10
        IncrementValue := 0.08
        FormatValue := FormatAsPercent
        IsUnlocked := IsUnlockedPowerUpNinjaMastery4
    power_up_id.NinjaMastery05 => power_up_info:
        Id := power_up_id.NinjaMastery05
        Name := "Chunin II"
        Description := "Increases both base damage and max health a moderate amount."
        Type := power_up_type.NinjaMastery
        Icon := UserInterface.Icons.RankIcon_05_128
        IconSmall := UserInterface.Icons.RankIcon_05_64
        BaseCost := 50000.0
        CostFactor := 1.2
        MaxLevel := 10
        IncrementValue := 0.10
        FormatValue := FormatAsPercent
        IsUnlocked := IsUnlockedPowerUpNinjaMastery5
    power_up_id.NinjaMastery06 => power_up_info:
        Id := power_up_id.NinjaMastery06
        Name := "Chunin III"
        Description := "Increases both base damage and max health a moderate amount."
        Type := power_up_type.NinjaMastery
        Icon := UserInterface.Icons.RankIcon_06_128
        IconSmall := UserInterface.Icons.RankIcon_06_64
        BaseCost := 75000.0
        CostFactor := 1.2
        MaxLevel := 10
        IncrementValue := 0.12
        FormatValue := FormatAsPercent
        IsUnlocked := IsUnlockedPowerUpNinjaMastery6
    power_up_id.NinjaMastery07 => power_up_info:
        Id := power_up_id.NinjaMastery07
        Name := "Jonin I"
        Description := "Increases both base damage and max health a large amount."
        Type := power_up_type.NinjaMastery
        Icon := UserInterface.Icons.RankIcon_07_128
        IconSmall := UserInterface.Icons.RankIcon_07_64
        BaseCost := 100000.0
        CostFactor := 1.2
        MaxLevel := 10
        IncrementValue := 0.14
        FormatValue := FormatAsPercent
        IsUnlocked := IsUnlockedPowerUpNinjaMastery7
    power_up_id.NinjaMastery08 => power_up_info:
        Id := power_up_id.NinjaMastery08
        Name := "Jonin II"
        Description := "Increases both base damage and max health a large amount."
        Type := power_up_type.NinjaMastery
        Icon := UserInterface.Icons.RankIcon_08_128
        IconSmall := UserInterface.Icons.RankIcon_08_64
        BaseCost := 150000.0
        CostFactor := 1.2
        MaxLevel := 10
        IncrementValue := 0.16
        FormatValue := FormatAsPercent
        IsUnlocked := IsUnlockedPowerUpNinjaMastery8
    power_up_id.NinjaMastery09 => power_up_info:
        Id := power_up_id.NinjaMastery09
        Name := "Jonin III"
        Description := "Increases both base damage and max health a large amount."
        Type := power_up_type.NinjaMastery
        Icon := UserInterface.Icons.RankIcon_09_128
        IconSmall := UserInterface.Icons.RankIcon_09_64
        BaseCost := 200000.0
        CostFactor := 1.2
        MaxLevel := 10
        IncrementValue := 0.20
        FormatValue := FormatAsPercent
        IsUnlocked := IsUnlockedPowerUpNinjaMastery9

NinjaMasteryPowerUpIds:[]power_up_id = array:
    power_up_id.NinjaMastery01
    power_up_id.NinjaMastery02
    power_up_id.NinjaMastery03
    power_up_id.NinjaMastery04
    power_up_id.NinjaMastery05
    power_up_id.NinjaMastery06
    power_up_id.NinjaMastery07
    power_up_id.NinjaMastery08
    power_up_id.NinjaMastery09

<# ----------------------------------------------------------------------------------------------------
    Global Functions
---------------------------------------------------------------------------------------------------- #>
GetPowerUpInfo()<transacts>:[power_up_id]power_up_info =
    PowerUps

GetPowerUpInfo(Id:power_up_id)<decides><transacts>:power_up_info =
    GetPowerUpInfo()[Id]

GetPowerUpIds()<transacts>:[]power_up_id =
    for (Id -> PowerUpInfo : GetPowerUpInfo()) {Id}

GetPowerUpIds(Type:power_up_type)<transacts>:[]power_up_id =
    for (Id -> PowerUpInfo : GetPowerUpInfo(), PowerUpInfo.Type = Type) {Id}

(Player:player).CanAffordPowerUp()<transacts>:logic = 
    var IsAffordPowerUp:logic = false
    for (PowerUpId -> PowerUpInfo : GetPowerUpInfo(), not IsAffordPowerUp?):
        if:
            Player.GetPowerUpLevel(PowerUpId) < PowerUpInfo.MaxLevel
            PowerUpCost := CalculatePowerUpCost[PowerUpId, Player.GetPowerUpLevel(PowerUpId) + 1]
            Player.GetGoldAvailable(currency_id.Gold) >= PowerUpCost
        then:
            set IsAffordPowerUp = true

    IsAffordPowerUp

GetTotalPowerUpCost()<transacts>:float =
    var TotalPowerUpCost:float = 0.0
    for (PowerUpInfo : GetPowerUpInfo()):       
        for (PowerUpLevel := 1..PowerUpInfo.MaxLevel):
            set TotalPowerUpCost += (Floor[Pow(PowerUpLevel * 1.0, PowerUpInfo.CostFactor * 1.0) * PowerUpInfo.BaseCost] or 0) * 1.0
        
    TotalPowerUpCost

(Id:power_up_id).GetMaxLevel()<transacts>:int =
    GetPowerUpInfo[Id].MaxLevel or 1

<# ----------------------------------------------------------------------------------------------------
    IsUnlocked Functions
---------------------------------------------------------------------------------------------------- #>
IsUnlockedPowerUpNinjaMastery1(Player:player)<transacts>:logic = logic{Player.GetRank() >= 0}
IsUnlockedPowerUpNinjaMastery2(Player:player)<transacts>:logic = logic{Player.GetRank() >= 1}
IsUnlockedPowerUpNinjaMastery3(Player:player)<transacts>:logic = logic{Player.GetRank() >= 2}
IsUnlockedPowerUpNinjaMastery4(Player:player)<transacts>:logic = logic{Player.GetRank() >= 3}
IsUnlockedPowerUpNinjaMastery5(Player:player)<transacts>:logic = logic{Player.GetRank() >= 4}
IsUnlockedPowerUpNinjaMastery6(Player:player)<transacts>:logic = logic{Player.GetRank() >= 5}
IsUnlockedPowerUpNinjaMastery7(Player:player)<transacts>:logic = logic{Player.GetRank() >= 6}
IsUnlockedPowerUpNinjaMastery8(Player:player)<transacts>:logic = logic{Player.GetRank() >= 7}
IsUnlockedPowerUpNinjaMastery9(Player:player)<transacts>:logic = logic{Player.GetRank() >= 8}