using { /Fortnite.com/UI }
using { /Verse.org/Colors }
using { /UnrealEngine.com/Temporary/SpatialMath }
using { /UnrealEngine.com/Temporary/UI }

<# ----------------------------------------------------------------------------------------------------
    power_up_stat_entry_ui
---------------------------------------------------------------------------------------------------- #>
power_up_stat_entry_ui := class(ui_component):
    ValueText:text_block = text_block:
        DefaultTextColor := ThemeColors.PrimaryText
        DefaultShadowOffset := option{vector2{X := 2.0, Y := 2.0}}
        DefaultShadowOpacity := 0.5

    DeltaText:text_block = text_block:
        DefaultTextColor := ThemeColors.Success
        DefaultShadowColor := ThemeColors.Success * 0.125
        DefaultShadowOffset := option{vector2{X := 2.0, Y := 2.0}}
        DefaultShadowOpacity := 1.0

    CreateUI<override>():widget =
        stack_box:
            Orientation := orientation.Horizontal
            Slots := array:
                stack_box_slot:
                    HorizontalAlignment := horizontal_alignment.Fill
                    VerticalAlignment := vertical_alignment.Fill
                    Widget := ValueText
                stack_box_slot:
                    HorizontalAlignment := horizontal_alignment.Fill
                    VerticalAlignment := vertical_alignment.Fill
                    Padding := margin{Left := 10.0, Top := 0.0, Right := 0.0, Bottom := 0.0}
                    Widget := DeltaText

    Update(Value:float, Delta:float, FormatValue:type{_(:float)<transacts>:string}, ?Prefix:string = ""):void =
        ValueText.SetText(MakeMessage("{Prefix}" + "{FormatValue(Value)}" + (Delta = 0.0 and "" or " ▶")))
        DeltaText.SetText(MakeMessage("{FormatValue(Value + Delta)}" + " ({Delta > 0.0 and "+" or ""}{FormatValue(Delta)})"))
        DeltaText.SetVisibility(Delta = 0.0 and widget_visibility.Collapsed or widget_visibility.Visible)        