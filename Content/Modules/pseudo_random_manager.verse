using { /Fortnite.com/Devices }
using { /UnrealEngine.com/Temporary/Diagnostics }
using { /UnrealEngine.com/Temporary/SpatialMath }
using { /Verse.org/Random }
using { /Verse.org/Simulation }
using { /Verse.org/Simulation/Tags }

<# ----------------------------------------------------------------------------------------------------
    Data definitions
---------------------------------------------------------------------------------------------------- #>
pseudo_random_id := enum{HealthDrop}

pseudo_random := struct<concrete>:
    Constant:float = 0.0
    Count:int = 0
    
<# ----------------------------------------------------------------------------------------------------
    Global functions
---------------------------------------------------------------------------------------------------- #>
var GlobalPseudoRandomManager:weak_map(session, pseudo_random_manager) = map{}

GetPseudoRandomManager()<decides><transacts>:pseudo_random_manager =
    GlobalPseudoRandomManager[GetSession()]

GetPsuedoRandomResult(Id:pseudo_random_id):logic =
    if (PseudoRandomManager := GetPseudoRandomManager[]):
        PseudoRandomManager.GetPsuedoRandomResult(Id)
    else:
        false

<# ----------------------------------------------------------------------------------------------------
    pseudo_random_manager
---------------------------------------------------------------------------------------------------- #>
pseudo_random_manager := class(creative_device):

    var PseudoRandoms:[pseudo_random_id]pseudo_random = map{}

    <# ----------------------------------------------------------------------------------------------------
        Initialization
    ---------------------------------------------------------------------------------------------------- #>
    OnBegin<override>()<suspends>:void =
        if (set GlobalPseudoRandomManager[GetSession()] = Self) {}

        WaitForGameManager()
        if (GameManager := GetGameManager[]):
            GameManager.GameStartedEvent.Subscribe(OnGameStarted)

    <# ----------------------------------------------------------------------------------------------------
        Get Functions
    ---------------------------------------------------------------------------------------------------- #>
    (pseudo_random_manager:)GetPsuedoRandomResult(Id:pseudo_random_id):logic =
        var PsuedoConstant:float = PseudoRandoms[Id].Constant or 0.0

        if (PseudoRandom := PseudoRandoms[Id]):
            set PsuedoConstant *= (PseudoRandom.Count + 1) * 1.0

        IsSuccess := logic{GetRandomFloat(0.0, 1.0) <= PsuedoConstant}

        if:
            IsSuccess?
            UpdatedPseudoRandom := pseudo_random{Constant := PseudoRandoms[Id].Constant, Count := 0}
            set PseudoRandoms[Id] = UpdatedPseudoRandom
        else:
            if:
                UpdatedPseudoRandom := pseudo_random{Constant := PseudoRandoms[Id].Constant, Count := PseudoRandoms[Id].Count + 1}
                set PseudoRandoms[Id] = UpdatedPseudoRandom

        IsSuccess

    GetConstantFromPercent(Percent:float):float =
        var CUpper:float = Percent
        var CLower:float = 0.0
        var CMid:float = 0.0
        var P1:float = 0.0
        var P2:float = 1.0

        loop:
            set CMid = (CUpper + CLower) / 2.0;
            set P1 = GetPercentFromConstant(CMid)

            if (Abs(P1 - P2) <= 0.0):
                break

            if (P1 > Percent):
                set CUpper = CMid
            else:
                set CLower = CMid

            set P2 = P1

        CMid

    GetPercentFromConstant(Constant:float):float =
        var PProcOnN:float = 0.0
        var PProcByN:float = 0.0
        var SumNpProcOnN:float = 0.0

        MaxFails:int = Ceil[1.0 / Constant] or 0
        for (N := 1..MaxFails):
            set PProcOnN = Min(1.0, N * Constant) * (1.0 - PProcByN)
            set PProcByN += PProcOnN
            set SumNpProcOnN += N * PProcOnN

        1.0 / SumNpProcOnN

    <# ----------------------------------------------------------------------------------------------------
        Update Functions
    ---------------------------------------------------------------------------------------------------- #>
    Initialize(Id:pseudo_random_id, Constant:float):void =
        if (set PseudoRandoms[Id] = pseudo_random{Constant := Constant, Count := 0}):

    <# ----------------------------------------------------------------------------------------------------
        Event Handlers
    ---------------------------------------------------------------------------------------------------- #>
    OnGameStarted():void =
        Percent := DefaultGatherableItemDropRate / Max(1.0, GetActivePlayers().Length * 1.0)
        Constant := GetConstantFromPercent(DefaultGatherableItemDropRate / Max(1.0, GetActivePlayers().Length * 1.0))
        Initialize(pseudo_random_id.HealthDrop, GetConstantFromPercent(DefaultGatherableItemDropRate / Max(1.0, GetActivePlayers().Length * 1.0)))