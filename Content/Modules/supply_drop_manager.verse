using { /Fortnite.com/Characters }
using { /Fortnite.com/Devices }
using { /Verse.org/Random }
using { /Verse.org/Simulation }
using { /UnrealEngine.com/Temporary/SpatialMath }
using { Utility }

<# ----------------------------------------------------------------------------------------------------
    Constants
---------------------------------------------------------------------------------------------------- #>
SupplyDropDefaultActivationTime:float = 30.0

<# ----------------------------------------------------------------------------------------------------
    Global Functions
---------------------------------------------------------------------------------------------------- #>
var GlobalSupplyManager:weak_map(session, supply_drop_manager) = map{}

GetSupplyDropManager()<decides><transacts>:supply_drop_manager = 
    GlobalSupplyManager[GetSession()]

SpawnSupplyDrop():void =
    if (SupplyManager := GetSupplyDropManager[]):
        spawn{SupplyManager.Start()}

DestroySupplyDrop():void =
    if (SupplyManager := GetSupplyDropManager[]):
        SupplyManager.Stop()

<# ----------------------------------------------------------------------------------------------------
    supply_drop_manager
---------------------------------------------------------------------------------------------------- #>
supply_drop_manager := class(creative_device):
    @editable
    SupplyDropSpawner:supply_drop_spawner_device = supply_drop_spawner_device{}

    @editable
    VolumeActivator:volume_activator = volume_activator{}

    SupplyDropActivatedEvent<public>:events(tuple()) = events(tuple()){}
    StoppedEvent:event() = event(){}

    var SpawnPosition:vector3 = vector3{}

    <# ----------------------------------------------------------------------------------------------------
        Initialization
    ---------------------------------------------------------------------------------------------------- #>
    OnBegin<override>()<suspends>:void =
        if (set GlobalSupplyManager[GetSession()] = Self) {}

        VolumeActivator.ActivatedEvent.Subscribe(OnActivated)
        VolumeActivator.OnBegin()

        WaitForGameManager()
        if (GameManager := GetGameManager[]):
            GameManager.GameEndedEvent.Subscribe(Stop)
            GameManager.BossFightStartedEvent.Subscribe(Stop)
            GameManager.StageEndedEvent.Subscribe(Stop)

    <# ----------------------------------------------------------------------------------------------------
        Update Functions
    ---------------------------------------------------------------------------------------------------- #>
    Start()<suspends>:void =
        RandomPositions := Shuffle(PoissonDiscSamplingGeneratePoints(1500.0, GetActiveArenaSize() - VolumeActivator.Size * 10.0, ?MinRadius := 600.0))
        
        var Attempts:int = 1
        loop:
            RandomPosition := GetRandomPositionInDonut(600.0, 3500.0, ?Origin := DefaultArenaOrigin)
            RaycastPosition := GetRaycast(RandomPosition)

            if (Attempts >= 30):
                break
            
            if (RaycastPosition.Z <= 10.0, SupplyDropSpawner.TeleportTo[RaycastPosition.Make2D(), IdentityRotation()]):
                SupplyDropSpawner.Spawn()
                VolumeActivator.UpdateMapIndicator(RaycastPosition.Make2D())
                break

            set Attempts += 1

        ShowMessage(message_id.SupplyDropNotification, "")
        for (ActivePlayer : GetActivePlayers()):
            ActivePlayer.PlayAudio(audio_id.SupplyDrop)

        race:
            StoppedEvent.Await()
            block:
                race:
                    SupplyDropSpawner.LandingEvent.Await()
                    Sleep(18.0)

                Enable()

    Stop():void =
        StoppedEvent.Signal()
        VolumeActivator.Disable()
        if (Player := GetAllPlayers()[0]):
            SupplyDropSpawner.Open(Player)

    Enable():void =
        SupplyDropPosition := SupplyDropSpawner.GetTransform().Translation.Make2D()
        SupplyDropDamage := GetNPCMaxHealth(NPCBaseMaxHealth)

        if (RandomPlayer := GetRandomAlivePlayer()?):
            for:
                NPCBehaviorBase : GetAllNPCs()
                NPCCharacter := NPCBehaviorBase.GetAgent[].GetFortCharacter[]
                DamageRange := VolumeActivator.Size * 10.0 + NPCBehaviorBase.GetSize()
                DistanceSquaredXY(NPCCharacter.GetTransform().Translation, SupplyDropPosition) <= DamageRange * DamageRange
                NPCBehaviorBase.IsDamageable()?
            do:
                AbilityDamageArgs := ability_damage_args{Amount := SupplyDropDamage, IsCriticalHit := false}
                NPCBehaviorBase.Damage(RandomPlayer, AbilityDamageArgs)

        KnockbackRange := VolumeActivator.Size * 10.0
        KnockbackNPC(SupplyDropPosition, ?KnockbackDistanceSquared := KnockbackRange * KnockbackRange, ?Duration := 1.0)
        ActivateCameraShake(SupplyDropPosition)
        VolumeActivator.Enable(?TargetPosition := option{SupplyDropPosition - vector3{Z := 300.0}}, ?OverrideActivationTime := option{SupplyDropDefaultActivationTime / GetGameSpeed()})
        SupplyDropSpawner.DestroyBalloon()

    <# ----------------------------------------------------------------------------------------------------
        Event Handlers
    ---------------------------------------------------------------------------------------------------- #>
    OnActivated(Player:player, InVolumeActivator:volume_activator):void =
        VolumeActivator.Disable()
        SupplyDropSpawner.Open(Player)

        for (ActivePlayer : GetActivePlayers(), GameProgressBarUI := GetGameManager[].PlayerGameProgressBarUI[ActivePlayer]):
            GameProgressBarUI.UpdateEventState(GameEventSupplyDropIndex, game_event_state.Success)
            ActivePlayer.PlayAudio(audio_id.SupplyDropOpen)
    
        for (Index := 1..GetRandomInt(10, 20)):
            SpawnExperience(GetRandomPositionInCircle(VolumeActivator.Size * 10.0, ?Origin := SupplyDropSpawner.GetTransform().Translation.Make2D()), GetRandomInt(10, 60), ?IsUseReserve := true)

        SpawnItemDrop(item_drop_id.Magnet, SupplyDropSpawner.GetTransform().Translation.Make2D())

        AwardAccolade(accolade_id.SupplyDropSecured, GetActivePlayers())
        SupplyDropActivatedEvent.Signal()