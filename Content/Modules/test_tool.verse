using { /Fortnite.com/AI }
using { /Fortnite.com/Characters }
using { /Fortnite.com/Devices }
using { /Fortnite.com/FortPlayerUtilities }
using { /Fortnite.com/Game }
using { /Verse.org/Colors }
using { /Verse.org/Concurrency }
using { /Verse.org/Random }
using { /Verse.org/SceneGraph }
using { /Verse.org/SceneGraph/KeyframedMovement }
using { /Verse.org/Simulation }
using { /Verse.org/SpatialMath }
using { /UnrealEngine.com/Temporary/Diagnostics }
using { /UnrealEngine.com/Temporary/UI }
using { RealTime }
using { WeaponAbilities }
using { VisualEffects }
using { Prefabs }

<# ----------------------------------------------------------------------------------------------------
    Global Functions
---------------------------------------------------------------------------------------------------- #>
var GlobalTestTool:weak_map(session, test_tool) = map{}

GetTestTool()<decides><transacts>:test_tool =
    GlobalTestTool[GetSession()]

(Player:player).GetTestToolDebugValue()<transacts>:float =
    GetTestTool[].TestToolUIs[Player].DebugValue or 0.0

DebugDraw()<transacts>:debug_draw =
    GetTestTool[].GetDebugDraw() or debug_draw{Channel := debug_draw_global_channel}

debug_draw_global_channel := class(debug_draw_channel) {}

test_trigger := class<concrete>:
    @editable
    Trigger:trigger_device = trigger_device{}
    @editable
    Tracker:tracker_device = tracker_device{}

<# ----------------------------------------------------------------------------------------------------
    test_tool
---------------------------------------------------------------------------------------------------- #>
test_tool := class(creative_device):
    @editable
    ItemGranterSignalRemote:item_granter_device = item_granter_device{}

    @editable
    SignalRemoteManager:signal_remote_manager_device = signal_remote_manager_device{}
    
    @editable
    HUDController:hud_controller_device = hud_controller_device{}

    @editable
    ConditionalButton:conditional_button_device = conditional_button_device{}
    
    @editable
    WeaponFireMutator:mutator_zone_device = mutator_zone_device{}

    @editable
    PlayerMovementSettings:player_movement_settings_device = player_movement_settings_device{}

    var (test_tool:)DebugDraw:debug_draw = debug_draw{Channel := debug_draw_global_channel}
    var TestToolUIs:[player]test_tool_ui = map{}

    <# ----------------------------------------------------------------------------------------------------
        Initialization
    ---------------------------------------------------------------------------------------------------- #>
    OnBegin<override>()<suspends>:void =
        if (set GlobalTestTool[GetSession()] = Self) {}

        for (Player : GetPlayspace().GetPlayers()):
            spawn{Initialize(Player)}

        HUDController.Enable()
        WeaponFireMutator.Disable()
        SignalRemoteManager.PrimarySignalEvent.Subscribe(OnPrimarySignal)
        SignalRemoteManager.SecondarySignalEvent.Subscribe(OnSecondarySignal)

        GetPlayspace().PlayerAddedEvent().Subscribe(OnPlayerAdded)
        GetPlayspace().PlayerRemovedEvent().Subscribe(OnPlayerRemoved)
        
    Initialize(Player:player)<suspends>:void =
        ItemGranterSignalRemote.GrantItem(Player)

        WaitForInputManager()
        spawn{TrackIsHoldingSignalRemote(Player)}
        
    <# ----------------------------------------------------------------------------------------------------
        Get Functions
    ---------------------------------------------------------------------------------------------------- #>
    GetDebugDraw()<transacts>:debug_draw =
        (test_tool:)DebugDraw

    <# ----------------------------------------------------------------------------------------------------
        Update Functions
    ---------------------------------------------------------------------------------------------------- #>
    ShowTestTool(Agent:agent):void =
        if (Player := player[Agent]):
            TestToolUI := if (UI := TestToolUIs[Player]):
                UI
            else:
                UI := test_tool_ui{Player := Player}
                if (set TestToolUIs[Player] = UI) {}
                UI

            TestToolUI.Update()
            TestToolUI.Show()
            Agent.PlayAudio(audio_id.Activate)

    AsyncTest(Agent:agent)<suspends>:void =
        if (Player := player[Agent], Character := Agent.GetFortCharacter[]) {}

    TrackIsHoldingSignalRemote(Player:player)<suspends>:void =
        loop:
            if (IsGameInProgress()?):
                if (ConditionalButton.IsHoldingItem[Player]):
                    Player.UnregisterInput(input_id.AbilityStats)
                else:
                    Player.RegisterInput(input_id.AbilityStats)

            Sleep(ThreeTicks)

    <# ----------------------------------------------------------------------------------------------------
        Event handlers
    ---------------------------------------------------------------------------------------------------- #>
    OnPlayerAdded(Player:player):void =
        spawn{Initialize(Player)}

    OnPlayerRemoved(Player:player):void =
        set TestToolUIs = RemoveKeyFromMap(TestToolUIs, Player)

    OnPrimarySignal(Agent:agent):void =
        ShowTestTool(Agent)

    OnSecondarySignal(Agent:agent):void =
        if (Player := player[Agent], FortCharacter := Agent.GetFortCharacter[]):
            Player.PlayAudio(audio_id.Activate)
            spawn{AsyncTest(Agent)}

            if (IsGameInProgress()?):
                PauseGame(pause_info{Type := pause_type.Boon})
            else:
                if (GameLobby := GetGameLobby[]):
                    GameLobby.StartGame(?IsSkipIntro := true)
            

<# ----------------------------------------------------------------------------------------------------
    Test Functions
---------------------------------------------------------------------------------------------------- #>
<#
# Activate objectives in sequential order.
var IsObjectiveStarted:logic = false
var ObjectiveIndex:int = 0

ObjectiveIds := array:
    objective_id.Collectible
    objective_id.GoldenSlime
    objective_id.BeachBall
    objective_id.CapturePoint
    objective_id.NumberedTiles
    objective_id.Storm
    objective_id.SnowStorm
    objective_id.MultiCapturePoint
    objective_id.ArrowTiles
    objective_id.FindCharacter

if (not IsObjectiveStarted?):
    if (ObjectiveId := ObjectiveIds[ObjectiveIndex]):
        StartObjective(ObjectiveId)
    set ObjectiveIndex += 1
    if (ObjectiveIndex > ObjectiveIds.Length - 1):
        set ObjectiveIndex = 0
else:
    StopObjective()

set IsObjectiveStarted = logic{not IsObjectiveStarted?}


# Activate objective.
if (GetCurrentObjective()?):
    StopObjective()
else:
    StartObjective(objective_id.GoldenSlime)


# Start game instantly.
if (GameManager := GetGameManager[]):
    GameManager.ForceStart(0)


# Enables Item Shop
EnableItemShop()
AddGameCurrency(game_currency_id.Wood, 3000.0)
AddGameCurrency(game_currency_id.TreasureChest, 10.0)


# Grant random daily quests.
if:
    CurrentPlayerData := Player.GetData[]
    UpdatedLastQuestAssignedTime := GetDailyQuestsManager[].RealTimeDevice.GetDateAndTime().GetTotalSeconds()
    UpdatedActiveQuests := GetDailyQuestsManager[].GetNewDailyQuests[Player, 3]
    set PlayerDataMap[Player] = player_data:
        MakePlayerData<constructor>(CurrentPlayerData)
        DailyQuests := player_daily_quests:
            MakePlayerDailyQuests<constructor>(CurrentPlayerData.DailyQuests)
            LastQuestAssignedTime := UpdatedLastQuestAssignedTime
            ActiveQuests := UpdatedActiveQuests


# Spawn Item Drop in front of the player character.
Direction := Character.GetTransform().Rotation.GetLocalForward()
TargetPosition := Character.GetTransform().Translation + Direction * 300.0
SpawnItemDrop(item_drop_id.AbilityUpgrade, TargetPosition)


# Spawns Supply Drop.
var IsSpawned:logic = false
if (not IsSpawned?):
    SpawnSupplyDrop()
else:
    DestroySupplyDrop()
set IsSpawned = logic{not IsSpawned?}


# Show player rank up HUD Message.
var Rank:int = 1
case (Rank):
    1 => Player.ShowMessage(message_id.RankUp01)
    2 => Player.ShowMessage(message_id.RankUp02)
    3 => Player.ShowMessage(message_id.RankUp03)
    4 => Player.ShowMessage(message_id.RankUp04)
    5 => Player.ShowMessage(message_id.RankUp05)
    6 => Player.ShowMessage(message_id.RankUp06)
    7 => Player.ShowMessage(message_id.RankUp07)
    8 => Player.ShowMessage(message_id.RankUp08)
    9 => Player.ShowMessage(message_id.RankUp09)
    _ => Player.ShowMessage(message_id.RankUp01)

set Rank = (Rank = 9) and 1 or (Rank + 1)


# Set player stats & weapon evolution
var IsInitialized:logic = false

if:
    not IsInitialized?
    PlayerStats := Player.GetStats[]
    UpdatedWeaponEvolutions := map:
        weapon_id.Shuriken => array:
            GetWeaponInfo[weapon_id.Shuriken].Evolution[0][0]
            GetWeaponInfo[weapon_id.Shuriken].Evolution[1][0]
            GetWeaponInfo[weapon_id.Shuriken].Evolution[2][0]

    set PlayerStats.WeaponEvolutions = UpdatedWeaponEvolutions
    set PlayerStats.AbilityLevels[ability_id.MaxHealth] = for (Index := 1..100) {5}
then:
    set IsInitialized = true

# Show pause menu.
PauseGame(Player, pause_level_up{Type := pause_type.LevelUp, Level := GetLevel() + 1})
PauseGame(Player, pause_weapon_evolution{Type := pause_type.WeaponEvolution, Id := weapon_id.Bow, EvolutionLevel := 1, IsItemShop := false})
PauseGame(Player, pause_info{Type := pause_type.ItemShop})
PauseGame(Player, pause_info{Type := pause_type.TreasureChest})
PauseGame(pause_info{Type := pause_type.Boon})

# Set the day offset for the RealTimeDevice.
if (RealTimeDevice := GetDailyQuestsManager[].RealTimeDevice):
    RealTimeDevice.AddDebugTimeOffsetDays(1)


# Initialize player stat/weapons/boons.
if (PlayerStats := Player.GetStats[], AbilityManager := GetAbilityManager[]):
    set PlayerStats.Boons = map:
        boon_id.BossHunter => 1
        boon_id.WoodenClub => 1
        boon_id.AllForOne => 1
        boon_id.MultiTool => 1
        boon_id.WoodenArrow => 1
        boon_id.CriticalOverdrive => 1
        boon_id.DiscountCoupon => 1
        boon_id.BossSiphon => 1
        boon_id.SpareDice => 1
        boon_id.LuckyDice => 1

    Boons := for (BoonId -> Boon : PlayerStats.Boons) {BoonId}
    for (Id -> BoonInfo : BoonInfos, Boons.Find[Id]):
        Player.AddBoon(BoonInfo)

    Player.ActivateNewAbility(weapon_id.Sentry)
    Player.EquipNewWeapon(weapon_id.Sentry)
    Player.ActivateNewAbility(weapon_id.Shield)
    Player.EquipNewWeapon(weapon_id.Shield)
    Player.ActivateNewAbility(weapon_id.Shuriken)
    Player.EquipNewWeapon(weapon_id.Shuriken)

    WeaponsIds := array:
        weapon_id.Aura
        weapon_id.Sentry
        weapon_id.Shield
        weapon_id.Shuriken

    UpdatedWeaponEvolutions := map:
        weapon_id.Aura => array:
            GetWeaponInfo[weapon_id.Aura].Evolution[0][1]
            GetWeaponInfo[weapon_id.Aura].Evolution[1][0]
            GetWeaponInfo[weapon_id.Aura].Evolution[2][1]
        weapon_id.Sentry => array:
            GetWeaponInfo[weapon_id.Sentry].Evolution[0][1]
            GetWeaponInfo[weapon_id.Sentry].Evolution[1][0]
            GetWeaponInfo[weapon_id.Sentry].Evolution[2][1]
        weapon_id.Shield => array:
            GetWeaponInfo[weapon_id.Shield].Evolution[0][0]
            GetWeaponInfo[weapon_id.Shield].Evolution[1][1]
            GetWeaponInfo[weapon_id.Shield].Evolution[2][0]
        weapon_id.Shuriken => array:
            GetWeaponInfo[weapon_id.Shuriken].Evolution[0][0]
            GetWeaponInfo[weapon_id.Shuriken].Evolution[1][0]
            GetWeaponInfo[weapon_id.Shuriken].Evolution[2][1]

    set PlayerStats.WeaponEvolutions = UpdatedWeaponEvolutions

    for (AbilityInfo : GetAbilityInfo()):
        for (Index := 1..25):
            PlayerStats.AddAbility(AbilityInfo, 3)

    for (WeaponAbilityInfo : AbilityManager.GetWeaponAbilityInfo(Player)):
        Count := (WeaponAbilityInfo.Id = ability_id.Size) and 3 or 25
        for (Index := 1..Count):
            PlayerStats.AddAbility(WeaponAbilityInfo, 3)

    if (GameManager := GetGameManager[]):
        set GameManager.CurrentStage = 35

    SetLevel(100)
#>