using { /Fortnite.com/Characters }
using { /Fortnite.com/UI }
using { /UnrealEngine.com/Temporary/SpatialMath }
using { /UnrealEngine.com/Temporary/UI }

<# ----------------------------------------------------------------------------------------------------
    test_tool_ability_ui
---------------------------------------------------------------------------------------------------- #>
test_tool_ability_ui := class(ui_component):
    GameSettingArenaText:text_block = text_block:
        DefaultTextColor := ThemeColors.PrimaryText

    GameSettingDifficultyText:text_block = text_block:
        DefaultTextColor := ThemeColors.PrimaryText

    GameSettingSpeedText:text_block = text_block:
        DefaultTextColor := ThemeColors.PrimaryText
    
    var AbilityValues:[ability_id]text_block = map{}

    <# ----------------------------------------------------------------------------------------------------
        Initialization
    ---------------------------------------------------------------------------------------------------- #>
    CreateUI<override>():widget =
        ButtonAddXP100 := button_quiet{DefaultText := MakeMessage("+100")}
        ButtonAddXP1000 := button_quiet{DefaultText := MakeMessage("+1000")}
        ButtonAddXP10000 := button_quiet{DefaultText := MakeMessage("+10000")}
        spawn{OnAddXPClicked(ButtonAddXP100.OnClick(), 100)}
        spawn{OnAddXPClicked(ButtonAddXP1000.OnClick(), 1000)}
        spawn{OnAddXPClicked(ButtonAddXP10000.OnClick(), 10000)}

        stack_box:
            Orientation := orientation.Vertical
            Slots := array:
                stack_box_slot:
                    HorizontalAlignment := horizontal_alignment.Fill
                    VerticalAlignment := vertical_alignment.Fill
                    Widget := stack_box:
                        Orientation := orientation.Horizontal
                        Slots := array:
                            stack_box_slot:
                                HorizontalAlignment := horizontal_alignment.Fill
                                VerticalAlignment := vertical_alignment.Fill
                                Distribution := option{1.0}
                                Widget := ButtonAddXP100
                            stack_box_slot:
                                HorizontalAlignment := horizontal_alignment.Fill
                                VerticalAlignment := vertical_alignment.Fill
                                Distribution := option{1.0}
                                Padding := margin{Left := 10.0, Top := 0.0, Right := 0.0, Bottom := 0.0}
                                Widget := ButtonAddXP1000
                            stack_box_slot:
                                HorizontalAlignment := horizontal_alignment.Fill
                                VerticalAlignment := vertical_alignment.Fill
                                Distribution := option{1.0}
                                Padding := margin{Left := 10.0, Top := 0.0, Right := 0.0, Bottom := 0.0}
                                Widget := ButtonAddXP10000
                stack_box_slot:
                    HorizontalAlignment := horizontal_alignment.Fill
                    VerticalAlignment := vertical_alignment.Fill
                    Padding := margin{Left := 0.0, Top := 20.0, Right := 0.0, Bottom := 0.0}
                    Distribution := option{1.0}
                    Widget := stack_box:
                        Orientation := orientation.Vertical
                        Slots := for (AbilityInfo : GetAbilityInfo()):
                            ButtonReset := button_quiet{DefaultText := MakeMessage("Reset")}
                            ButtonAdd1 := button_quiet{DefaultText := MakeMessage("+1")}
                            ButtonAdd5 := button_quiet{DefaultText := MakeMessage("+5")}
                            ButtonAdd25 := button_quiet{DefaultText := MakeMessage("+25")}
                            ButtonAdd250 := button_quiet{DefaultText := MakeMessage("+250")}
                            spawn{OnAbilityResetClicked(ButtonReset.OnClick(), AbilityInfo.Id)}
                            spawn{OnAbilityAddClicked(ButtonAdd1.OnClick(), AbilityInfo.Id, 1)}
                            spawn{OnAbilityAddClicked(ButtonAdd5.OnClick(), AbilityInfo.Id, 5)}              
                            spawn{OnAbilityAddClicked(ButtonAdd25.OnClick(), AbilityInfo.Id, 25)}
                            spawn{OnAbilityAddClicked(ButtonAdd250.OnClick(), AbilityInfo.Id, 250)}

                            AbilityValue := text_block:
                                DefaultTextSize := ThemeTextSize.Small
                                DefaultTextColor := ThemeColors.Highlight

                            if (set AbilityValues[AbilityInfo.Id] = AbilityValue) {}

                            stack_box_slot:
                                HorizontalAlignment := horizontal_alignment.Fill
                                VerticalAlignment := vertical_alignment.Fill
                                Distribution := option{1.0}
                                Padding := margin{Left := 0.0, Top := 5.0, Right := 0.0, Bottom := 0.0}
                                Widget := stack_box:
                                    Orientation := orientation.Horizontal
                                    Slots := array:
                                        stack_box_slot:
                                            Widget := texture_block:
                                                DefaultImage := AbilityInfo.IconSmall
                                                DefaultDesiredSize := vector2{X := 30.0, Y := 30.0}
                                        stack_box_slot:
                                            HorizontalAlignment := horizontal_alignment.Left
                                            Padding := margin{Left := 10.0, Top := 0.0, Right := 0.0, Bottom := 0.0}
                                            Distribution := option{2.0}
                                            Widget := text_block:
                                                DefaultTextSize := ThemeTextSize.Small
                                                DefaultText := MakeMessage("{AbilityInfo.ShortName}")
                                                DefaultTextColor := ThemeColors.PrimaryText
                                        stack_box_slot:
                                            HorizontalAlignment := horizontal_alignment.Left
                                            Distribution := option{1.0}
                                            Padding := margin{Left := 20.0, Top := 0.0, Right := 0.0, Bottom := 0.0}
                                            Widget := AbilityValue
                                        stack_box_slot:
                                            Padding := margin{Left := 10.0, Top := 0.0, Right := 0.0, Bottom := 0.0}
                                            Widget := ButtonReset 
                                        stack_box_slot:
                                            Padding := margin{Left := 10.0, Top := 0.0, Right := 0.0, Bottom := 0.0}
                                            Widget := ButtonAdd1
                                        stack_box_slot:
                                            Padding := margin{Left := 10.0, Top := 0.0, Right := 0.0, Bottom := 0.0}
                                            Widget := ButtonAdd5
                                        stack_box_slot:
                                            Padding := margin{Left := 10.0, Top := 0.0, Right := 0.0, Bottom := 0.0}
                                            Widget := ButtonAdd25
                                        stack_box_slot:
                                            Padding := margin{Left := 10.0, Top := 0.0, Right := 0.0, Bottom := 0.0}
                                            Widget := ButtonAdd250

    <# ----------------------------------------------------------------------------------------------------
        Update Functions
    ---------------------------------------------------------------------------------------------------- #>
    Update<override>():void =
        for (Id -> AbilityValue : AbilityValues):
            UpdateAbility(Id)

    UpdateAbility(Id:ability_id):void =
        if (AbilityValue := AbilityValues[Id], AbilityInfo := GetAbilityInfo[Id], StatId := AbilityInfo.MaybeStatId?):
            AbilityValue.SetText(MakeMessage("{AbilityInfo.FormatValue(Player.GetStat(StatId))}"))

    Hide<override>():void =
        UI.SetVisibility(widget_visibility.Hidden)

    <# ----------------------------------------------------------------------------------------------------
        Event Handlers
    ---------------------------------------------------------------------------------------------------- #>
    OnAbilityResetClicked(Listenable:listenable(widget_message), Id:ability_id)<suspends>:void =
        loop:
            Listenable.Await()
            if:
                AbilityInfo := GetAbilityInfo[Id]
                PlayerStats := Player.GetStats[]
                set PlayerStats.AbilityValues[Id] = 0.0
            then:
                PlayerStats.HandleAbilityChanged(AbilityInfo)
                UpdateAbility(Id)

    OnAbilityAddClicked(Listenable:listenable(widget_message), Id:ability_id, Tier:int)<suspends>:void =
        loop:
            Listenable.Await()
            if (AbilityInfo := GetAbilityInfo[Id]):
                var TierAdded:int = 0
                loop:
                    if (TierAdded >= Tier):
                        break
                    
                    TierToAdd := Min(5, Tier)
                    Player.AddAbility(AbilityInfo, ?Tier := TierToAdd)
                    set TierAdded += TierToAdd

                UpdateAbility(Id)

    OnAddXPClicked(Listenable:listenable(widget_message), Amount:int)<suspends>:void =
        loop:
            Listenable.Await()
            AddXP(Amount)