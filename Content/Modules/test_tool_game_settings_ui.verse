using { /Fortnite.com/Characters }
using { /Fortnite.com/UI }
using { /Verse.org/Simulation }
using { /UnrealEngine.com/Temporary/SpatialMath }
using { /UnrealEngine.com/Temporary/UI }

<# ----------------------------------------------------------------------------------------------------
    test_tool_game_settings_ui
---------------------------------------------------------------------------------------------------- #>
test_tool_game_settings_ui := class(ui_component):
    GameStartEvent:events(tuple()) = events(tuple()){}

    GameSettingArenaText:text_block = text_block:
        DefaultTextColor := ThemeColors.PrimaryText

    GameSettingDifficultyText:text_block = text_block:
        DefaultTextColor := ThemeColors.PrimaryText

    GameSettingSpeedText:text_block = text_block:
        DefaultTextColor := ThemeColors.PrimaryText

    GameSettingWeaponCountText:text_block = text_block:
        DefaultTextColor := ThemeColors.PrimaryText

    GameSettingNPCCountText:text_block = text_block:
        DefaultTextColor := ThemeColors.PrimaryText

    NPCCountButton:button_quiet = button_quiet{}
    EndlessModeButton:button_quiet = button_quiet{}
    GameStartButton:button_quiet = button_quiet{}

    LoadCancelledEvent:event() = event(){}
    
    <# ----------------------------------------------------------------------------------------------------
        Initialization
    ---------------------------------------------------------------------------------------------------- #>
    CreateUI<override>():widget =
        ArenaDecrementButton := button_quiet{DefaultText := MakeMessage("-")}
        ArenaDecrementButton.OnClick().Subscribe(OnClickArenaDecrement)
        ArenaIncrementButton := button_quiet{DefaultText := MakeMessage("+")}
        ArenaIncrementButton.OnClick().Subscribe(OnClickArenaIncrement)
        
        DifficultyDecrementButton := button_quiet{DefaultText := MakeMessage("-")}
        DifficultyDecrementButton.OnClick().Subscribe(OnClickDifficultyDecrement)
        DifficultyIncrementButton := button_quiet{DefaultText := MakeMessage("+")}
        DifficultyIncrementButton.OnClick().Subscribe(OnClickDifficultyIncrement)

        WeaponCountDecrementButton := button_quiet{DefaultText := MakeMessage("-")}
        WeaponCountDecrementButton.OnClick().Subscribe(OnClickWeaponCountDecrement)
        WeaponCountIncrementButton := button_quiet{DefaultText := MakeMessage("+")}
        WeaponCountIncrementButton.OnClick().Subscribe(OnClickWeaponCountIncrement)
        
        GameSpeedDecrementButton := button_quiet{DefaultText := MakeMessage("-")}
        GameSpeedDecrementButton.OnClick().Subscribe(OnClickGameSpeedDecrement)
        GameSpeedDecrementButton10 := button_quiet{DefaultText := MakeMessage("-10")}
        GameSpeedDecrementButton10.OnClick().Subscribe(OnClickGameSpeedDecrement10)
        GameSpeedIncrementButton := button_quiet{DefaultText := MakeMessage("+")}
        GameSpeedIncrementButton.OnClick().Subscribe(OnClickGameSpeedIncrement)
        GameSpeedIncrementButton10 := button_quiet{DefaultText := MakeMessage("+10")}
        GameSpeedIncrementButton10.OnClick().Subscribe(OnClickGameSpeedIncrement10)

        NPCCountButton.OnClick().Subscribe(OnClickNPCCount)
        EndlessModeButton.OnClick().Subscribe(OnClickEndlessMode)
        
        GameStartButton.SetText(MakeMessage("Start Game"))
        GameStartButton.OnClick().Subscribe(OnClickGameStart)

        stack_box:
            Orientation := orientation.Vertical
            Slots := array:
                stack_box_slot:
                    HorizontalAlignment := horizontal_alignment.Fill
                    VerticalAlignment := vertical_alignment.Fill
                    Widget := text_block:
                        DefaultText := MakeMessage("Game Settings")
                        DefaultTextColor := ThemeColors.PrimaryText
                        DefaultTextOpacity := 0.5
                stack_box_slot:
                    HorizontalAlignment := horizontal_alignment.Fill
                    VerticalAlignment := vertical_alignment.Fill
                    Padding := margin{Left := 0.0, Top := 20.0, Right := 0.0, Bottom := 0.0}
                    Widget := stack_box:
                        Orientation := orientation.Horizontal
                        Slots := array:
                            stack_box_slot:
                                HorizontalAlignment := horizontal_alignment.Fill
                                VerticalAlignment := vertical_alignment.Fill
                                Distribution := option{1.0}
                                Widget := ArenaDecrementButton
                            stack_box_slot:
                                HorizontalAlignment := horizontal_alignment.Fill
                                VerticalAlignment := vertical_alignment.Fill
                                Distribution := option{2.0}
                                Padding := margin{Left := 20.0, Top := 0.0, Right := 20.0, Bottom := 0.0}
                                Widget := GameSettingArenaText
                            stack_box_slot:
                                HorizontalAlignment := horizontal_alignment.Fill
                                VerticalAlignment := vertical_alignment.Fill
                                Distribution := option{1.0}
                                Widget := ArenaIncrementButton
                stack_box_slot:
                    HorizontalAlignment := horizontal_alignment.Fill
                    VerticalAlignment := vertical_alignment.Fill
                    Padding := margin{Left := 0.0, Top := 10.0, Right := 0.0, Bottom := 0.0}
                    Widget := stack_box:
                        Orientation := orientation.Horizontal
                        Slots := array:
                            stack_box_slot:
                                HorizontalAlignment := horizontal_alignment.Fill
                                VerticalAlignment := vertical_alignment.Fill
                                Distribution := option{1.0}
                                Widget := DifficultyDecrementButton
                            stack_box_slot:
                                HorizontalAlignment := horizontal_alignment.Fill
                                VerticalAlignment := vertical_alignment.Fill
                                Distribution := option{2.0}
                                Padding := margin{Left := 20.0, Top := 0.0, Right := 20.0, Bottom := 0.0}
                                Widget := GameSettingDifficultyText
                            stack_box_slot:
                                HorizontalAlignment := horizontal_alignment.Fill
                                VerticalAlignment := vertical_alignment.Fill
                                Distribution := option{1.0}
                                Widget := DifficultyIncrementButton
                stack_box_slot:
                    HorizontalAlignment := horizontal_alignment.Fill
                    VerticalAlignment := vertical_alignment.Fill
                    Padding := margin{Left := 0.0, Top := 10.0, Right := 0.0, Bottom := 0.0}
                    Widget := stack_box:
                        Orientation := orientation.Horizontal
                        Slots := array:
                            stack_box_slot:
                                HorizontalAlignment := horizontal_alignment.Fill
                                VerticalAlignment := vertical_alignment.Fill
                                Distribution := option{1.0}
                                Widget := WeaponCountDecrementButton
                            stack_box_slot:
                                HorizontalAlignment := horizontal_alignment.Fill
                                VerticalAlignment := vertical_alignment.Fill
                                Distribution := option{2.0}
                                Padding := margin{Left := 20.0, Top := 0.0, Right := 20.0, Bottom := 0.0}
                                Widget := GameSettingWeaponCountText
                            stack_box_slot:
                                HorizontalAlignment := horizontal_alignment.Fill
                                VerticalAlignment := vertical_alignment.Fill
                                Distribution := option{1.0}
                                Widget := WeaponCountIncrementButton
                stack_box_slot:
                    HorizontalAlignment := horizontal_alignment.Fill
                    VerticalAlignment := vertical_alignment.Fill
                    Padding := margin{Left := 0.0, Top := 10.0, Right := 0.0, Bottom := 0.0}
                    Widget := stack_box:
                        Orientation := orientation.Horizontal
                        Slots := array:
                            stack_box_slot:
                                HorizontalAlignment := horizontal_alignment.Fill
                                VerticalAlignment := vertical_alignment.Fill
                                Widget := GameSpeedDecrementButton
                            stack_box_slot:
                                HorizontalAlignment := horizontal_alignment.Fill
                                VerticalAlignment := vertical_alignment.Fill
                                Padding := margin{Left := 20.0, Top := 0.0, Right := 0.0, Bottom := 0.0}
                                Widget := GameSpeedDecrementButton10
                            stack_box_slot:
                                HorizontalAlignment := horizontal_alignment.Fill
                                VerticalAlignment := vertical_alignment.Fill
                                Distribution := option{1.0}
                                Padding := margin{Left := 20.0, Top := 0.0, Right := 20.0, Bottom := 0.0}
                                Widget := GameSettingSpeedText
                            stack_box_slot:
                                HorizontalAlignment := horizontal_alignment.Fill
                                VerticalAlignment := vertical_alignment.Fill
                                Widget := GameSpeedIncrementButton
                            stack_box_slot:
                                HorizontalAlignment := horizontal_alignment.Fill
                                VerticalAlignment := vertical_alignment.Fill
                                Padding := margin{Left := 20.0, Top := 0.0, Right := 0.0, Bottom := 0.0}
                                Widget := GameSpeedIncrementButton10
                stack_box_slot:
                    HorizontalAlignment := horizontal_alignment.Fill
                    VerticalAlignment := vertical_alignment.Fill
                    Padding := margin{Left := 0.0, Top := 10.0, Right := 0.0, Bottom := 0.0}
                    Widget := EndlessModeButton
                stack_box_slot:
                    HorizontalAlignment := horizontal_alignment.Fill
                    VerticalAlignment := vertical_alignment.Fill
                    Padding := margin{Left := 0.0, Top := 10.0, Right := 0.0, Bottom := 0.0}
                    Widget := NPCCountButton
                stack_box_slot:
                    HorizontalAlignment := horizontal_alignment.Fill
                    VerticalAlignment := vertical_alignment.Fill
                    Padding := margin{Left := 0.0, Top := 20.0, Right := 0.0, Bottom := 0.0}
                    Widget := GameStartButton

    <# ----------------------------------------------------------------------------------------------------
        Update Functions
    ---------------------------------------------------------------------------------------------------- #>
    Update<override>():void =
        if (GameSettings := GetGameSettings[]):
            GameSettingArenaText.SetText(MakeMessage("Arena: {GameSettings.Arena.Level}"))
            GameSettingDifficultyText.SetText(MakeMessage("Difficulty: {GameSettings.Difficulty}"))
            GameSettingWeaponCountText.SetText(MakeMessage("Weapon Count: {GameSettings.Weapons}"))
            GameSettingSpeedText.SetText(MakeMessage("Game Speed: x{GameSettings.GameSpeed}"))
            NPCCountButton.SetText(MakeMessage("NPC Count: {GameSettings.IsDoubleEnemyCount? and "90" or "45"}"))
            EndlessModeButton.SetText(MakeMessage("Endless Mode: {GameSettings.IsEndless? and "ON" or "OFF"}"))

    Hide<override>():void =
        UI.SetVisibility(widget_visibility.Hidden)

    Load(ArenaIndex:int)<suspends>:void =
        if (ArenaManager := GetArenaManager[]):
            race:
                block:
                    LoadCancelledEvent.Await()
                block:
                    GameStartButton.SetText(MakeMessage("Loading ..."))
                    loop:
                        if (not ArenaManager.IsLoading?):
                            break

                        Sleep(0.1)

                    LoadTask := spawn{ArenaManager.LoadArena(ArenaIndex)}
                    LoadTask.Await()
                    GameStartButton.SetText(MakeMessage("Start Game"))

    <# ----------------------------------------------------------------------------------------------------
        Event Handlers
    ---------------------------------------------------------------------------------------------------- #>
    OnClickArenaDecrement(Message:widget_message):void =
        if (GameSettings := GetGameSettings[], ArenaIndex := Clamp(GameSettings.Arena.Level - 2, 0, GetArenas().Length - 1), Arena := GetArenas()[ArenaIndex]):
            UpdatedGameSettings := game_settings:
                MakeGameSettings<constructor>(GameSettings)
                Arena := Arena

            SetGameSettings(UpdatedGameSettings)
            LoadCancelledEvent.Signal()
            spawn{Load(ArenaIndex)}
            Update()
        
    OnClickArenaIncrement(Message:widget_message):void =
        if (GameSettings := GetGameSettings[], ArenaIndex := Clamp(GameSettings.Arena.Level, 0, GetArenas().Length - 1), Arena := GetArenas()[ArenaIndex]):
            UpdatedGameSettings := game_settings:
                MakeGameSettings<constructor>(GameSettings)
                Arena := Arena

            SetGameSettings(UpdatedGameSettings)
            LoadCancelledEvent.Signal()
            spawn{Load(ArenaIndex)}
            Update()

    OnClickDifficultyDecrement(Message:widget_message):void =
        if (GameSettings := GetGameSettings[]):
            UpdatedGameSettings := game_settings:
                MakeGameSettings<constructor>(GameSettings)
                Difficulty := Clamp(GameSettings.Difficulty - 1, 1, MaxDifficultyLevel)

            SetGameSettings(UpdatedGameSettings)
            Update()

    OnClickDifficultyIncrement(Message:widget_message):void =
        if (GameSettings := GetGameSettings[]):
            UpdatedGameSettings := game_settings:
                MakeGameSettings<constructor>(GameSettings)
                Difficulty := Clamp(GameSettings.Difficulty + 1, 1, MaxDifficultyLevel)

            SetGameSettings(UpdatedGameSettings)
            Update()

    OnClickWeaponCountDecrement(Message:widget_message):void =
        if (GameSettings := GetGameSettings[]):
            UpdatedGameSettings := game_settings:
                MakeGameSettings<constructor>(GameSettings)
                Weapons := Clamp(GameSettings.Weapons - 1, 1, MaxWeaponCount)

            SetGameSettings(UpdatedGameSettings)
            Update()

    OnClickWeaponCountIncrement(Message:widget_message):void =
        if (GameSettings := GetGameSettings[]):
            UpdatedGameSettings := game_settings:
                MakeGameSettings<constructor>(GameSettings)
                Weapons := Clamp(GameSettings.Weapons + 1, 1, MaxWeaponCount)

            SetGameSettings(UpdatedGameSettings)
            Update()

    OnClickGameSpeedDecrement(Message:widget_message):void =
        if (GameSettings := GetGameSettings[]):
            UpdatedGameSettings := game_settings:
                MakeGameSettings<constructor>(GameSettings)
                GameSpeed := Max(0, GameSettings.GameSpeed - 1)

            SetGameSettings(UpdatedGameSettings)
            Update()

    OnClickGameSpeedIncrement(Message:widget_message):void =
        if (GameSettings := GetGameSettings[]):
            UpdatedGameSettings := game_settings:
                MakeGameSettings<constructor>(GameSettings)
                GameSpeed := Max(0, GameSettings.GameSpeed + 1)

            SetGameSettings(UpdatedGameSettings)
            Update()

    OnClickGameSpeedDecrement10(Message:widget_message):void =
        if (GameSettings := GetGameSettings[]):
            UpdatedGameSettings := game_settings:
                MakeGameSettings<constructor>(GameSettings)
                GameSpeed := Max(0, GameSettings.GameSpeed - 10)

            SetGameSettings(UpdatedGameSettings)
            Update()

    OnClickGameSpeedIncrement10(Message:widget_message):void =
        if (GameSettings := GetGameSettings[]):
            UpdatedGameSettings := game_settings:
                MakeGameSettings<constructor>(GameSettings)
                GameSpeed := Max(0, GameSettings.GameSpeed + 10)

            SetGameSettings(UpdatedGameSettings)
            Update()

    OnClickGameStart(Message:widget_message):void =
        if (GameLobby := GetGameLobby[]):
            GameLobby.StartGame(?IsSkipIntro := true)
            GameStartEvent.Signal()

    OnClickEndlessMode(Message:widget_message):void =
        if (GameSettings := GetGameSettings[]):
            UpdatedGameSettings := game_settings:
                MakeGameSettings<constructor>(GameSettings)
                IsEndless := logic{not GameSettings.IsEndless?}

            SetGameSettings(UpdatedGameSettings)
            Update()

    OnClickNPCCount(Message:widget_message):void =
        if (GameSettings := GetGameSettings[]):
            UpdatedGameSettings := game_settings:
                MakeGameSettings<constructor>(GameSettings)
                IsDoubleEnemyCount := logic{not GameSettings.IsDoubleEnemyCount?}

            SetGameSettings(UpdatedGameSettings)
            Update()
