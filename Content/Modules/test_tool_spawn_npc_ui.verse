using { /Fortnite.com/Characters }
using { /Fortnite.com/UI }
using { /UnrealEngine.com/Temporary/SpatialMath }
using { /UnrealEngine.com/Temporary/UI }

test_tool_spawn_npc_ui := class(ui_component):
    EntriesPerPage:int = 8

    Slider:slider_regular = slider_regular:
        DefaultValue := 1.0
        DefaultMinValue := 1.0
        DefaultMaxValue := 90.0
        DefaultStepSize := 1.0

    var SpawnCount:int = 1
    var CurrentPage:int = 1
    var SpawnNPCEntryUIs:[]test_tool_spawn_npc_entry_ui = array{}

    <# ----------------------------------------------------------------------------------------------------
        Initialization
    ---------------------------------------------------------------------------------------------------- #>
    CreateUI<override>():widget =
        Slider.OnValueChanged().Subscribe(OnValueChanged)

        DespawnButton := button_quiet{DefaultText := MakeMessage("Despawn")}
        DespawnButton.OnClick().Subscribe(OnClickDespawn)

        StackBox := stack_box:
            Orientation := orientation.Vertical
            Slots := array:
                stack_box_slot:
                    Padding := margin{Left := 0.0, Top := 0.0, Right := 0.0, Bottom := 10.0}
                    Widget := stack_box:
                        Orientation := orientation.Horizontal
                        Slots := array:
                            stack_box_slot:
                                HorizontalAlignment := horizontal_alignment.Left
                                Distribution := option{1.0}
                                Widget := text_block:
                                    DefaultText := MakeMessage("Spawn Count:")
                                    DefaultTextColor := ThemeColors.PrimaryText
                            stack_box_slot:
                                HorizontalAlignment := horizontal_alignment.Fill
                                Padding := margin{Left := 10.0, Top := 0.0, Right := 0.0, Bottom := 0.0}
                                Widget := Slider
                stack_box_slot:
                    HorizontalAlignment := horizontal_alignment.Fill
                    Padding := margin{Left := 0.0, Top := 0.0, Right := 0.0, Bottom := 10.0}
                    Widget := DespawnButton

        NPCSpawners := GetNPCSpawners()
        for (Index := 0..(EntriesPerPage - 1), NPCSpawner := NPCSpawners[Index]):
            SpawnNPCEntryUI := test_tool_spawn_npc_entry_ui{Id := NPCSpawner.Id}
            SpawnNPCEntryUI.Initialize()
            SpawnNPCEntryUI.ClickedSpawnEvent.Subscribe(OnClickedSpawn)
            SpawnNPCEntryUI.ClickedSpawnFrontEvent.Subscribe(OnClickedSpawnFront)
            set SpawnNPCEntryUIs += array{SpawnNPCEntryUI}

            Slot := stack_box_slot:
                HorizontalAlignment := horizontal_alignment.Fill
                VerticalAlignment := vertical_alignment.Fill
                Padding := margin{Left := 0.0, Top := (Index = 0) and 0.0 or 5.0, Right := 0.0, Bottom := 0.0}
                Widget := SpawnNPCEntryUI.GetUI()

            StackBox.AddWidget(Slot)
        
        PreviousButton := button_quiet{DefaultText := MakeMessage("◀")}
        PreviousButton.OnClick().Subscribe(OnClickPrevious)

        NextButton := button_quiet{DefaultText := MakeMessage("▶")}
        NextButton.OnClick().Subscribe(OnClickNext)

        PaginationSlot := stack_box_slot:
            HorizontalAlignment := horizontal_alignment.Fill
            VerticalAlignment := vertical_alignment.Fill
            Padding := margin{Left := 0.0, Top := 10.0, Right := 0.0, Bottom := 0.0}
            Widget := stack_box:
                Orientation := orientation.Horizontal
                Slots := array:
                    stack_box_slot:
                        HorizontalAlignment := horizontal_alignment.Fill
                        Distribution := option{1.0}
                        Padding := margin{Left := 0.0, Top := 0.0, Right := 20.0, Bottom := 0.0}
                        Widget := PreviousButton
                    stack_box_slot:
                        HorizontalAlignment := horizontal_alignment.Fill
                        Distribution := option{1.0}
                        Widget := NextButton

        StackBox.AddWidget(PaginationSlot)

        StackBox

    <# ----------------------------------------------------------------------------------------------------
        Get Functions
    ---------------------------------------------------------------------------------------------------- #>
    GetMaxPage()<transacts>:int =
        Ceil[(GetNPCSpawners().Length * 1.0) / (EntriesPerPage * 1.0)] or 1

    <# ----------------------------------------------------------------------------------------------------
        Update Functions
    ---------------------------------------------------------------------------------------------------- #>
    UpdatePage():void =
        NPCSpawners := GetNPCSpawners()
        for (Index -> SpawnNPCEntryUI : SpawnNPCEntryUIs):
            PageIndex := Index + (CurrentPage - 1) * EntriesPerPage
            if (NPCSpawner := NPCSpawners[PageIndex]):
                SpawnNPCEntryUI.SetId(NPCSpawner.Id)
            else:
                SpawnNPCEntryUI.Hide()

    Hide<override>():void =
        UI.SetVisibility(widget_visibility.Hidden)

    <# ----------------------------------------------------------------------------------------------------
        Event Handlers
    ---------------------------------------------------------------------------------------------------- #>
    OnClickedSpawn(Id:npc_id):void =
        Player.PlayAudio(audio_id.Activate)
        for (Index := 1..SpawnCount):
            SpawnNPC(Id, Player)

    OnClickedSpawnFront(Id:npc_id):void =
        Player.PlayAudio(audio_id.Activate)

        if (Character := Player.GetFortCharacter[]):
            for (Index := 1..SpawnCount):
                SpawnNPC(Id, Player, ?MaybePosition := option{Character.GetTransform().Translation + Character.GetTransform().Rotation.GetLocalForward() * 250.0})

    OnValueChanged(Message:widget_message):void =
        UpdatedSpawnCount := Slider.GetValue()
        set SpawnCount = Floor[UpdatedSpawnCount] or 1
        
    OnClickDespawn(Message:widget_message):void =
        DespawnNPC()

    OnClickPrevious(Message:widget_message):void =
        set CurrentPage = (CurrentPage = 1) and GetMaxPage() or CurrentPage - 1
        UpdatePage()

    OnClickNext(Message:widget_message):void =
        set CurrentPage = (CurrentPage = GetMaxPage()) and 1 or CurrentPage + 1
        UpdatePage()