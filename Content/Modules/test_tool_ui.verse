using { /Fortnite.com/Characters }
using { /Fortnite.com/UI }
using { /Verse.org/Simulation }
using { /UnrealEngine.com/Temporary/SpatialMath }
using { /UnrealEngine.com/Temporary/UI }

<# ----------------------------------------------------------------------------------------------------
    Data Definitions
---------------------------------------------------------------------------------------------------- #>
test_tool_panel_id := enum:
    Ability
    GameSettings
    NPCSpawner
    WeaponEvolution

ToString(Id:test_tool_panel_id)<transacts>:string =
    case (Id):
        test_tool_panel_id.Ability => "ABI"
        test_tool_panel_id.GameSettings => "GAM"
        test_tool_panel_id.NPCSpawner => "NPC"
        test_tool_panel_id.WeaponEvolution => "EVO"

<# ----------------------------------------------------------------------------------------------------
    test_tool_ui
---------------------------------------------------------------------------------------------------- #>
test_tool_ui := class(panel_ui):
    Title<override>:string = "Test Tools"
    BackgroundOpacity<override>:type {_X:float where 0.000000 <= _X, _X <= 1.000000} = 0.7
    Anchors<override>:anchors = anchors{Minimum := vector2{X := 0.5, Y := 0.0}, Maximum := vector2{X := 0.5, Y := 1.0}}
    
    PanelIds:[]test_tool_panel_id = array:
        test_tool_panel_id.Ability
        test_tool_panel_id.GameSettings
        test_tool_panel_id.NPCSpawner
        test_tool_panel_id.WeaponEvolution

    Options:[]string = array:
        "Toggle Ability"
        "Toggle Level Up"
        "Reset Stats"
        "Game In Progress"
        "Stop Mission"
        "Start Bossfight"
        "Victory"
        "Teleport To Hub"
        "Teleport To Arena"

    DebugValueText:text_block = text_block:
        DefaultTextColor := ThemeColors.PrimaryText

    DebugValueIncrementAmount:float = 10.0

    var DebugValue:float = 10.0
    var CurrentPanel:test_tool_panel_id = test_tool_panel_id.Ability
    var Panels:[test_tool_panel_id]ui_component = map{}
    var Buttons:[string]text_button_base = map{}
    var IsToggleAbility:logic = true

    <# ----------------------------------------------------------------------------------------------------
        Initialization
    ---------------------------------------------------------------------------------------------------- #>
    CreateContent<override>():widget =
        ClosedEvent.Subscribe(OnClosed)

        ButtonIncrement := button_quiet{DefaultText := MakeMessage("+")}
        ButtonIncrement.OnClick().Subscribe(OnClickIncrement)

        ButtonDecrement := button_quiet{DefaultText := MakeMessage("-")}
        ButtonDecrement.OnClick().Subscribe(OnClickDecrement)

        stack_box:
            Orientation := orientation.Horizontal
            Slots := array:
                stack_box_slot:
                    HorizontalAlignment := horizontal_alignment.Fill
                    VerticalAlignment := vertical_alignment.Fill
                    Padding := margin{Left := 0.0, Top := 0.0, Right := 0.0, Bottom := 0.0}
                    Widget := stack_box:
                        Orientation := orientation.Vertical
                        Slots := array:
                            stack_box_slot:
                                HorizontalAlignment := horizontal_alignment.Fill
                                VerticalAlignment := vertical_alignment.Fill
                                Widget := stack_box:
                                    Orientation := orientation.Horizontal
                                    Slots := for (Index -> PanelId : PanelIds):
                                        Button := button_quiet{DefaultText := MakeMessage("{PanelId}")}
                                        spawn{AwaitPanelClicked(Button.OnClick(), PanelId)}

                                        stack_box_slot:
                                            HorizontalAlignment := horizontal_alignment.Fill
                                            VerticalAlignment := vertical_alignment.Fill
                                            Distribution := option{1.0}
                                            Padding := margin{Left := (Index = 0) and 0.0 or 10.0, Top := 0.0, Right := 0.0, Bottom := 0.0}
                                            Widget := Button
                            stack_box_slot:
                                HorizontalAlignment := horizontal_alignment.Fill
                                VerticalAlignment := vertical_alignment.Fill
                                Padding := margin{Left := 0.0, Top := 10.0, Right := 0.0, Bottom := 0.0}
                                Widget := stack_box:
                                    Orientation := orientation.Vertical
                                    Slots := for (Index -> Option : Options):
                                        Button := button_quiet:
                                            DefaultText := MakeMessage("{Option}")

                                        if (set Buttons[Option] = Button) {}
                            
                                        case (Option):
                                            "Toggle Ability" => Button.OnClick().Subscribe(OnClickToggleAbility)
                                            "Toggle Level Up" => Button.OnClick().Subscribe(OnClickToggleLevelUp)
                                            "Reset Stats" => Button.OnClick().Subscribe(OnClickResetLevel)
                                            "Game In Progress" => Button.OnClick().Subscribe(OnClickGameInProgress)
                                            "Stop Mission" => Button.OnClick().Subscribe(OnClickStopMission)
                                            "Start Bossfight" => Button.OnClick().Subscribe(OnClickStartBossFight)
                                            "Victory" => Button.OnClick().Subscribe(OnClickVictory)
                                            "Teleport To Hub" => Button.OnClick().Subscribe(OnClickTeleportToHub)
                                            "Teleport To Arena" => Button.OnClick().Subscribe(OnClickTeleportToArena)
                                            _ => {}
                        
                                        stack_box_slot:
                                            HorizontalAlignment := horizontal_alignment.Fill
                                            VerticalAlignment := vertical_alignment.Fill
                                            Padding := margin{Left := 0.0, Top := Index = 0 and 0.0 or 10.0, Right := 0.0, Bottom := 0.0}
                                            Widget := Button
                            stack_box_slot:
                                HorizontalAlignment := horizontal_alignment.Fill
                                VerticalAlignment := vertical_alignment.Top
                                Padding := margin{Left := 0.0, Top := 10.0, Right := 0.0, Bottom := 0.0}
                                Distribution := option{1.0}
                                Widget := stack_box:
                                    Orientation := orientation.Horizontal
                                    Slots := array:
                                        stack_box_slot:
                                            HorizontalAlignment := horizontal_alignment.Fill
                                            VerticalAlignment := vertical_alignment.Fill
                                            Distribution := option{1.0}
                                            Widget := ButtonDecrement
                                        stack_box_slot:
                                            HorizontalAlignment := horizontal_alignment.Fill
                                            Padding := margin{Left := 10.0, Top := 0.0, Right := 10.0, Bottom := 0.0}
                                            Widget := DebugValueText
                                        stack_box_slot:
                                            HorizontalAlignment := horizontal_alignment.Fill
                                            VerticalAlignment := vertical_alignment.Fill
                                            Distribution := option{1.0}
                                            Widget := ButtonIncrement
                stack_box_slot:
                    HorizontalAlignment := horizontal_alignment.Fill
                    VerticalAlignment := vertical_alignment.Fill
                    Padding := margin{Left := 20.0, Top := 0.0, Right := 0.0, Bottom := 0.0}
                    Widget := overlay:
                        Slots := array:
                            overlay_slot:
                                HorizontalAlignment := horizontal_alignment.Fill
                                VerticalAlignment := vertical_alignment.Fill
                                Padding := margin{Left := 0.0, Top := 0.0, Right := 0.0, Bottom := 0.0}
                                Widget := UserInterface.VerseWidget.WBP_RoundedBox{}
                            overlay_slot:
                                HorizontalAlignment := horizontal_alignment.Fill
                                VerticalAlignment := vertical_alignment.Fill
                                Padding := margin{Left := 40.0, Top := 40.0, Right := 40.0, Bottom := 40.0}
                                Widget := overlay:
                                    Slots :=
                                        AbilityUI := test_tool_ability_ui{Player := Player}
                                        AbilityUI.Initialize()

                                        GameSettingsUI := test_tool_game_settings_ui{Player := Player}
                                        GameSettingsUI.Initialize()
                                        GameSettingsUI.GameStartEvent.Subscribe(Hide)

                                        SpawnNPCUI := test_tool_spawn_npc_ui{Player := Player}
                                        SpawnNPCUI.Initialize()

                                        WeaponEvolutionUI := test_tool_weapon_evolution_ui{Player := Player}
                                        WeaponEvolutionUI.Initialize()

                                        if:
                                            set Panels[test_tool_panel_id.Ability] = AbilityUI
                                            set Panels[test_tool_panel_id.GameSettings] = GameSettingsUI
                                            set Panels[test_tool_panel_id.NPCSpawner] = SpawnNPCUI
                                            set Panels[test_tool_panel_id.WeaponEvolution] = WeaponEvolutionUI

                                        array:
                                            overlay_slot:
                                                HorizontalAlignment := horizontal_alignment.Fill
                                                VerticalAlignment := vertical_alignment.Fill
                                                Widget := stack_box:
                                                    Orientation := orientation.Vertical
                                                    Slots := array:
                                                        stack_box_slot:
                                                            HorizontalAlignment := horizontal_alignment.Fill
                                                            VerticalAlignment := vertical_alignment.Fill
                                                            Distribution := option{4.0}
                                                            Widget := AbilityUI.GetUI()
                                                        stack_box_slot:
                                                            Distribution := option{1.0}
                                                            Widget := canvas{}
                                            overlay_slot:
                                                HorizontalAlignment := horizontal_alignment.Fill
                                                VerticalAlignment := vertical_alignment.Fill
                                                Widget := GameSettingsUI.GetUI()
                                            overlay_slot:
                                                HorizontalAlignment := horizontal_alignment.Fill
                                                VerticalAlignment := vertical_alignment.Fill
                                                Widget := SpawnNPCUI.GetUI()
                                            overlay_slot:
                                                HorizontalAlignment := horizontal_alignment.Fill
                                                VerticalAlignment := vertical_alignment.Fill
                                                Widget := WeaponEvolutionUI.GetUI()

    <# ----------------------------------------------------------------------------------------------------
        Ability UI
    ---------------------------------------------------------------------------------------------------- #>
    

    <# ----------------------------------------------------------------------------------------------------
        Update Functions
    ---------------------------------------------------------------------------------------------------- #>
    Update():void =
        DebugValueText.SetText(MakeMessage("{GetRoundedString(DebugValue, 2)}"))
        
        for (Id -> Panel : Panels):
            if (Id = CurrentPanel):
                Panel.Show()
            else:
                Panel.Hide()

        if (Panel := Panels[CurrentPanel]):
            Panel.Update()

        if (Button := Buttons["Toggle Level Up"]):
            Button.SetText(MakeMessage("Level Up {GetLevel() >= MaxPlayerLevel and "OFF" or "ON"}"))

        if (Button := Buttons["Toggle Ability"]):
            Button.SetText(MakeMessage("Toggle Ability:  {IsToggleAbility? and "ON" or "OFF"}"))

    SetPanel(Id:test_tool_panel_id):void =
        set CurrentPanel = Id
        Update()

    <# ----------------------------------------------------------------------------------------------------
        Event Handlers
    ---------------------------------------------------------------------------------------------------- #>
    OnShow<override>():void =
        Update()
        if (IsGameInProgress()? and GetAllPlayers().Length = 1):
            Player.DeactivateAbility()
            PauseGame()

    OnClosed(InPlayer:player):void =
        Player.DeactivateAbility()
        
        if (IsToggleAbility?):
            Player.ReactivateAbility()

        if (IsGameInProgress()? and GetAllPlayers().Length = 1):
            ResumeGame(?IsInstant := true)

    OnClickResetLevel(Message:widget_message):void =
        Hide()
        Player.ResetStats()

    OnClickGameInProgress(Message:widget_message):void =
        if (GameManager := GetGameManager[]):
            if (IsGameInProgress()?):
                set GameManager.ActivePlayers = array{}
            else:
                set GameManager.ActivePlayers = GetAllPlayers()

        if (Button := text_button_base[Message.Source]):
            Button.SetText(MakeMessage("GameInProgress: {IsGameInProgress()? and "true" or "false"}"))

    OnClickTeleportToArena(Message:widget_message):void =
        if:
            set GetGameManager[].ActivePlayers += array{Player}
            Character := Player.GetFortCharacter[]
            Character.TeleportTo[vector3{Z := 100.0}, rotation{}]

            if (NPCManager := GetNPCManager[], not NPCManager.CurrentNPCWave?):
                set NPCManager.CurrentNPCWave = option{npc_wave{}}

    OnClickTeleportToHub(Message:widget_message):void =
        if:
            GameManager := GetGameManager[]
            Locator := GameManager.GameEndPositionLocator
            Character := Player.GetFortCharacter[]
            Character.TeleportTo[Locator.GetTransform().Translation + vector3{Z := CharacterHeightOffset}, Locator.GetTransform().Rotation]
        then:
            set GameManager.ActivePlayers = array{}
            Hide()
            Player.PlayAudio(audio_id.Activate)

    OnClickToggleAbility(Message:widget_message):void =
        set IsToggleAbility = logic{not IsToggleAbility?}
        Update()

    OnClickToggleLevelUp(Message:widget_message):void =
        if (GetLevel() >= MaxPlayerLevel):
            SetLevel(1)
        else:
            SetLevel(MaxPlayerLevel)

        Update()

    OnClickStopMission(Message:widget_message):void =
        StopObjective()
        Hide()

    OnClickStartBossFight(Message:widget_message):void =
        if (NPCManager := GetNPCManager[], GameManager := GetGameManager[]):
            GameManager.StopTimerEvent.Signal()
            spawn{GameManager.StartBossFight()}
            Hide()

    OnClickVictory(Message:widget_message):void =
        if (GameManager := GetGameManager[]):
            GameManager.Stop(true)
            Hide()

    OnClickIncrement(Message:widget_message):void =
        set DebugValue += DebugValueIncrementAmount
        Player.SetArenaProgress(Floor[DebugValue] or 1)
        Update()

    OnClickDecrement(Message:widget_message):void =
        set DebugValue -= DebugValueIncrementAmount
        Player.SetArenaProgress(Floor[DebugValue] or 1)
        Update()

    AwaitPanelClicked(Listenable:listenable(widget_message), Id:test_tool_panel_id)<suspends>:void =
        loop:
            Listenable.Await()
            SetPanel(Id)
        