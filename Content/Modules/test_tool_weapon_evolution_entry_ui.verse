using { /Fortnite.com/UI }
using { /Verse.org/Simulation }
using { /UnrealEngine.com/Temporary/SpatialMath }
using { /UnrealEngine.com/Temporary/UI }

test_tool_weapon_evolution_entry_ui := class:
    Player:player
    WeaponIndex:int

    Icon:texture_block = texture_block:
        DefaultImage := UserInterface.Icons.WeaponBow_64
        DefaultDesiredSize := vector2{X := 40.0, Y := 40.0}

    NameText:text_block = text_block:
        DefaultTextSize := 20.0
        DefaultTextColor := ThemeColors.PrimaryText

    var Widget:widget = canvas{}
    var EvolutionWidgetsByIndex:[int][int]tuple(texture_block, text_block) = map{}
    var MaybeWeaponId:?weapon_id = false

    Initialize():void =
        set Widget = CreateUI()
        Update()

    CreateUI():widget =
        stack_box:
            Orientation := orientation.Vertical
            Slots := array:
                stack_box_slot:
                    HorizontalAlignment := horizontal_alignment.Left
                    VerticalAlignment := vertical_alignment.Top
                    Widget := CreateInfoUI()
                stack_box_slot:
                    HorizontalAlignment := horizontal_alignment.Fill
                    VerticalAlignment := vertical_alignment.Fill
                    Padding := margin{Left := 0.0, Top := 20.0, Right := 0.0, Bottom := 0.0}
                    Widget := stack_box:
                        Orientation := orientation.Horizontal
                        Slots := for (Index := 0..2):
                            stack_box_slot:
                                HorizontalAlignment := horizontal_alignment.Fill
                                VerticalAlignment := vertical_alignment.Fill
                                Padding := margin{Left := (Index = 0) and 0.0 or 20.0, Top := 0.0, Right := 0.0, Bottom := 0.0}
                                Widget := CreateEvolutionUI(Index)

    CreateInfoUI():widget =
        stack_box:
            Orientation := orientation.Horizontal
            Slots := array:
                stack_box_slot:
                    HorizontalAlignment := horizontal_alignment.Left
                    Widget := Icon
                stack_box_slot:
                    HorizontalAlignment := horizontal_alignment.Left
                    Padding := margin{Left := 10.0, Top := 0.0, Right := 0.0, Bottom := 0.0}
                    Widget := NameText

    CreateEvolutionUI(EvolutionIndex:int):widget =
        stack_box:
            Orientation := orientation.Vertical
            Slots := array:
                stack_box_slot:
                    HorizontalAlignment := horizontal_alignment.Fill
                    VerticalAlignment := vertical_alignment.Fill
                    Widget := stack_box:
                        Orientation := orientation.Vertical
                        Slots := for (EvolutionOption := 0..1):
                            stack_box_slot:
                                HorizontalAlignment := horizontal_alignment.Fill
                                VerticalAlignment := vertical_alignment.Fill
                                Padding := margin{Left := 0.0, Top := (EvolutionOption = 0) and 0.0 or 5.0, Right := 0.0, Bottom := 0.0}
                                Widget := CreateEvolutionOptionUI(EvolutionIndex, EvolutionOption)

    CreateEvolutionOptionUI(EvolutionIndex:int, EvolutionOption:int):widget =
        EvolutionIcon := texture_block:
            DefaultImage := UserInterface.Icons.WeaponBow_64
            DefaultDesiredSize := vector2{X := 32.0, Y := 32.0}
        
        EvolutionButton := button_quiet{}
        spawn{AwaitEvolutionButtonClick(EvolutionButton.OnClick(), EvolutionIndex, EvolutionOption)}

        EvolutionText := text_block:
            DefaultTextSize := 20.0
            DefaultTextColor := ThemeColors.PrimaryText

        EvolutionWidget:tuple(texture_block, text_block) = (EvolutionIcon, EvolutionText)
        
        if:
            not EvolutionWidgetsByIndex[EvolutionIndex]
            set EvolutionWidgetsByIndex[EvolutionIndex] = map{}
        
        if:
            set EvolutionWidgetsByIndex[EvolutionIndex][EvolutionOption] = EvolutionWidget

        stack_box:
            Orientation := orientation.Horizontal
            Slots := array:
                stack_box_slot:
                    HorizontalAlignment := horizontal_alignment.Left
                    Widget := overlay:
                        Slots := array:
                            overlay_slot:
                                Widget := EvolutionIcon
                            overlay_slot:
                                HorizontalAlignment := horizontal_alignment.Fill
                                VerticalAlignment := vertical_alignment.Fill
                                Widget := EvolutionButton
                stack_box_slot:
                    HorizontalAlignment := horizontal_alignment.Left
                    Padding := margin{Left := 10.0, Top := 0.0, Right := 0.0, Bottom := 0.0}
                    Widget := EvolutionText

    <# ----------------------------------------------------------------------------------------------------
        Get Functions
    ---------------------------------------------------------------------------------------------------- #>
    GetUI():widget =
        Widget

    <# ----------------------------------------------------------------------------------------------------
        Update Functions
    ---------------------------------------------------------------------------------------------------- #>
    Show(Id:weapon_id):void =
        set MaybeWeaponId = option{Id}
        Update()
        Widget.SetVisibility(widget_visibility.Visible)

    Hide():void =
        Widget.SetVisibility(widget_visibility.Collapsed)

    Update():void =
        if (WeaponId := MaybeWeaponId?, WeaponInfo := GetWeaponInfo[WeaponId]):
            WeaponEvolutionInfos := WeaponInfo.Evolution
            
            Icon.SetImage(WeaponInfo.IconSmall)
            NameText.SetText(MakeMessage(WeaponInfo.Name + " (Lv. {Player.GetWeaponLevel(WeaponId)})"))

            for (EvolutionIndex -> EvolutionWidgets : EvolutionWidgetsByIndex, EvolutionOption -> EvolutionWidget : EvolutionWidgets):
                if (WeaponEvolutionInfo := WeaponEvolutionInfos[EvolutionIndex][EvolutionOption]):
                    EvolutionWidget(0).SetImage(WeaponEvolutionInfo.IconSmall)
                    EvolutionWidget(1).SetText(MakeMessage(WeaponEvolutionInfo.Name))

                    if (PlayerStat := Player.GetStats[], PlayerStat.WeaponEvolutions[WeaponId][EvolutionIndex] = EvolutionOption):
                        EvolutionWidget(1).SetTextColor(ThemeColors.Success)
                    else:
                        EvolutionWidget(1).SetTextColor(ThemeColors.PrimaryText)

    <# ----------------------------------------------------------------------------------------------------
        Event Handlers
    ---------------------------------------------------------------------------------------------------- #>
    AwaitEvolutionButtonClick(Listenable:listenable(widget_message), EvolutionIndex:int, EvolutionOption:int)<suspends>:void =
        loop:
            Listenable.Await()

            if:
                WeaponId := MaybeWeaponId?
                PlayerStat := Player.GetStats[]
            then:
                if:
                    not PlayerStat.WeaponEvolutions[WeaponId][EvolutionIndex]
                    set PlayerStat.WeaponEvolutions[WeaponId] = for (Index := 0..EvolutionIndex):
                        PlayerStat.WeaponEvolutions[WeaponId][Index] or 0

                if:
                    set PlayerStat.WeaponEvolutions[WeaponId][EvolutionIndex] = EvolutionOption
                then:
                    EvolutionLevel := (EvolutionIndex + 1) * 6
                    if:
                        Player.GetWeaponLevel(WeaponId) < EvolutionLevel
                        set PlayerStat.WeaponLevels[WeaponId] = EvolutionLevel

                Update()