using { /Fortnite.com/Characters }
using { /Fortnite.com/UI }
using { /UnrealEngine.com/Temporary/SpatialMath }
using { /UnrealEngine.com/Temporary/UI }

test_tool_weapon_evolution_ui := class(ui_component):
    
    var EntryUIs:[]test_tool_weapon_evolution_entry_ui = array{}

    <# ----------------------------------------------------------------------------------------------------
        Initialization
    ---------------------------------------------------------------------------------------------------- #>
    CreateUI<override>():widget =
        stack_box:
            Orientation := orientation.Vertical
            Slots := for (Index := 0..3):
                EntryUI := test_tool_weapon_evolution_entry_ui{Player := Player, WeaponIndex := Index}
                EntryUI.Initialize()
                set EntryUIs += array{EntryUI}

                stack_box_slot:
                    HorizontalAlignment := horizontal_alignment.Fill
                    VerticalAlignment := vertical_alignment.Fill
                    Padding := margin{Left := 0.0, Top := (Index = 0) and 0.0 or 20.0, Right := 0.0, Bottom := 0.0}
                    Widget := EntryUI.GetUI()

    <# ----------------------------------------------------------------------------------------------------
        Get Functions
    ---------------------------------------------------------------------------------------------------- #>

    <# ----------------------------------------------------------------------------------------------------
        Update Functions
    ---------------------------------------------------------------------------------------------------- #>
    Update<override>():void =
        if (PlayerStats := Player.GetStats[]):
            WeaponLevels := for (WeaponId -> WeaponLevel : PlayerStats.WeaponLevels) {WeaponId}
            for (Index -> EntryUI : EntryUIs):
                if (WeaponId := WeaponLevels[Index]):
                    EntryUI.Show(WeaponId)
                else:
                    EntryUI.Hide()            

    Hide<override>():void =
        UI.SetVisibility(widget_visibility.Hidden)
