<#>
    Snippet to show the server tick rate in UI by MattRix
    https://dev.epicgames.com/community/snippets/D98/fortnite-tick-rate-graph-device

using { /Fortnite.com/Devices }
using { /Fortnite.com/Game }

using { /Verse.org/Simulation }
using { /Verse.org/Colors }
using { /Verse.org/Random }

using { /UnrealEngine.com/Temporary/Diagnostics }
using { /UnrealEngine.com/Temporary/UI }
using { /UnrealEngine.com/Temporary/SpatialMath }
using { /Fortnite.com/UI }

tick_rate_device := class(creative_device):      
    var AverageTickRate : float = 0.0

    Canvas<private> : canvas = canvas{}
            
    OnBegin<override>()<suspends>:void =
        var PreviousTime : float = GetSimulationElapsedTime()

        loop:
            Sleep(0.0)
            CurrentTime := GetSimulationElapsedTime()
            DeltaTime := CurrentTime - PreviousTime
            set PreviousTime = CurrentTime

            TickRate := 1.0 / DeltaTime

            set AverageTickRate += (TickRate - AverageTickRate) / 3.0 # smooth it out so the text doesn't flicker distractingly as much
            ShowMessage(message_id.TickRate, GetRoundedString(AverageTickRate, 2))