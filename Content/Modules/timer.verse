using { /Fortnite.com/Characters }
using { /Fortnite.com/UI }
using { /Verse.org/Simulation }
using { /UnrealEngine.com/Temporary/SpatialMath }
using { /UnrealEngine.com/Temporary/UI }

<# ----------------------------------------------------------------------------------------------------
    timer
---------------------------------------------------------------------------------------------------- #>
timer := class:
    CompletedEvent<public>:events(tuple()) = events(tuple()){}
    TickedEvent<public>:events(float) = events(float){}
    PausedEvent:events(tuple()) = events(tuple()){}

    var MaxDuration:float = 0.0
    var ElapsedTime:float = 0.0

    <# ----------------------------------------------------------------------------------------------------
        Get Functions
    ---------------------------------------------------------------------------------------------------- #>
    GetMaxDuration<public>()<transacts>:float =
        MaxDuration

    GetElapsedTime<public>()<transacts>:float =
        ElapsedTime

    GetRemainingTime<public>()<transacts>:float =
        Max(0.0, MaxDuration - ElapsedTime)

    GetProgress<public>()<transacts>:float =
        1.0 - GetRemainingTime() / MaxDuration

    <# ----------------------------------------------------------------------------------------------------
        Update Functions
    ---------------------------------------------------------------------------------------------------- #>
    Start<public>():void =
        set ElapsedTime = 0.0
        spawn{Timer()}

    Timer()<suspends>:void =
        race:
            PausedEvent.Await()
            loop:
                Sleep(1.0)
                set ElapsedTime += 1.0
                TickedEvent.Signal(ElapsedTime)

                if (ElapsedTime >= MaxDuration):
                    CompletedEvent.Signal()
                    break

    Pause<public>():void =
        PausedEvent.Signal()

    Resume<public>():void =
        spawn{Timer()}

    SetMaxDuration<public>(Time:float):void =
        set MaxDuration = Time