using { /Fortnite.com/Devices }
using { /Verse.org/Simulation }
using { /UnrealEngine.com/Temporary/UI }
using { /UnrealEngine.com/Temporary/SpatialMath }

<# ----------------------------------------------------------------------------------------------------
    Data Definitions
---------------------------------------------------------------------------------------------------- #>
panel_ui_id := enum:
    WeaponSelect
    DailyQuest
    PlayerProgress
    PlayerPowerUps
    BoonCodex
    Gears
    Accolades
    NPCCodex
    InputControl

<# ----------------------------------------------------------------------------------------------------
    Global Functions
---------------------------------------------------------------------------------------------------- #>
var GlobalUIManager:weak_map(session, ui_manager) = map{}

GetUIManager()<decides><transacts>:ui_manager =
    GlobalUIManager[GetSession()]

WaitForUIManager(Player:player)<suspends>:void =
    loop:
        if (GetUIManager[].GetIsInitialized[Player]) {break}
        Sleep(0.3)

(Player:player).AddWidgetToHUD(CanvasSlot:canvas_slot):void =
    if (UIManager := GetUIManager[]):
        UIManager.AddWidgetToHUD(Player, CanvasSlot)

(Player:player).ToggleHUD(IsShow:logic):void =
    if (UIManager := GetUIManager[]):
        UIManager.ToggleHUD(Player, IsShow)

(Player:player).TogglePlayerHUD(IsShow:logic):void =
    if (UIManager := GetUIManager[]):
        UIManager.TogglePlayerHUD(Player, IsShow)

(Player:player).ShowGamePausedUI():void =
    if (UIManager := GetUIManager[]):
        UIManager.ShowGamePausedUI(Player)

(Player:player).ShowObjectiveArrowTilesUI(Objective:objective_arrow_tiles, Sequence:[]int):void =
    if (UIManager := GetUIManager[]):
        UIManager.ShowObjectiveArrowTilesUI(Player, Objective, Sequence)

(Player:player).GetActivePanel()<decides><transacts>:panel_ui =
    GetUIManager[].GetActivePanel[Player]

(Player:player).UpdateActivePanel(PanelUI:panel_ui):void =
    if (UIManager := GetUIManager[]):
        UIManager.UpdateActivePanel(Player, PanelUI)

(Player:player).ClearActivePanel():void =
    if (UIManager := GetUIManager[]):
        UIManager.ClearActivePanel(Player)

(Player:player).CloseActivePanel(?IsSignal:logic = false):void =
    if (UIManager := GetUIManager[]):
        UIManager.CloseActivePanel(Player, ?IsSignal := IsSignal)

(Player:player).CloseActivePanel():void =
    Player.CloseActivePanel(?IsSignal := false)

(Player:player).UpdatePowerUpIndicator():void =
    if (UIManager := GetUIManager[]):
        UIManager.UpdatePowerUpIndicator(Player)

(Player:player).ShowPanelUI(Id:panel_ui_id):void =
    if (UIManager := GetUIManager[]):
        UIManager.ShowPanelUI(Player, Id)

(Player:player).ShowPlayerMenuUI():void =
    if (UIManager := GetUIManager[], PlayerMenuUI := UIManager.PlayerMenuUIs[Player]):
        PlayerMenuUI.Show()

<# ----------------------------------------------------------------------------------------------------
    ui_manager
---------------------------------------------------------------------------------------------------- #>
ui_manager := class(creative_device):
    @editable
    HUDControllerDefault:hud_controller_device = hud_controller_device{}
    
    @editable
    HUDControllerDebug:hud_controller_device = hud_controller_device{}

    @editable
    HUDControllerHideHUD:hud_controller_device = hud_controller_device{}

    @editable
    ButtonMenu:button_device = button_device{}

    @editable
    ButtonInfoPanel:button_device = button_device{}
    
    @editable
    PowerUpBeacon:beacon_device = beacon_device{}
    
    @editable
    PowerUpBeaconNew:beacon_device = beacon_device{}

    @editable
    GameResultButton:button_device = button_device{}

    var PlayerHUDs:[player]canvas = map{}
    var PlayerHUDIsShowing:[player]logic = map{}
    var PlayerIsInitialized<private>:[player]logic = map{}
    var PlayerGamePausedUI<private>:[player]game_paused_ui = map{}
    var PlayerMenuUIs:[player]player_menu_ui = map{}
    var PlayerInfoPanelUIs:[player]info_panel_ui = map{}
    var PlayerAbilityStatInfoUI:[player]ability_stat_info_ui = map{}
    var PlayerActivePanel:[player]panel_ui = map{}
    var PlayerWeaponEquippedUI:[player]weapon_equipped_ui = map{}
    var PlayerPanelUIs:[player][panel_ui_id]panel_ui = map{}
    var PlayerObjectiveArrowTilesUIs:[player]objective_arrow_tiles_ui = map{}

    <# ----------------------------------------------------------------------------------------------------
        Initialization
    ---------------------------------------------------------------------------------------------------- #>
    OnBegin<override>()<suspends>:void =
        if (set GlobalUIManager[GetSession()] = Self) {}

        for (Player : GetPlayspace().GetPlayers()):
            spawn{Initialize(Player)}

        GetPlayspace().PlayerAddedEvent().Subscribe(OnPlayerAdded)
        GetPlayspace().PlayerRemovedEvent().Subscribe(OnPlayerRemoved)
        ButtonMenu.InteractedWithEvent.Subscribe(OnInteractedWithMenu)
        ButtonInfoPanel.InteractedWithEvent.Subscribe(OnInteractedWithInfoPanel)
        GameResultButton.InteractedWithEvent.Subscribe(OnInteractedWithGameResult)

        WaitForInputManager()
        if (InputManager := GetInputManager[]):
            InputManager.InputUnregisteredEvent.Subscribe(OnInputUnregistered)
            
        if (MenuInputPressedEvent := GetInputPressedEvent[input_id.Menu]):
            MenuInputPressedEvent.Subscribe(OnMenuInputPressedEvent)

        if (AbilityStatsPressedEvent := GetInputPressedEvent[input_id.AbilityStats], AbilityStatsReleasedEvent := GetInputReleasedEvent[input_id.AbilityStats]):
            AbilityStatsPressedEvent.Subscribe(OnAbilityStatsInputPressedEvent)
            AbilityStatsReleasedEvent.Subscribe(OnAbilityStatsInputReleasedEvent)

    Initialize(Player:player)<suspends>:void =
        if (PlayerIsInitialized[Player]?) then return

        PlayerHUD := canvas{}
        if (PlayerUI := GetPlayerUI[Player], set PlayerHUDs[Player] = PlayerHUD):
            PlayerUI.AddWidget(PlayerHUD, player_ui_slot{InputMode := ui_input_mode.None})

        AbilityStatInfoUI := ability_stat_info_ui{Player := Player, BackgroundType := background_type.Transparent, TextSize := ThemeTextSize.Small, IconSize := 0.75}
        AbilityStatInfoUI.Initialize()
        AbilityStatInfoUI.Hide()

        Player.AddWidgetToHUD of canvas_slot:
            Anchors := anchors{Minimum := vector2{X := 0.20, Y := 0.5}, Maximum := vector2{X := 0.8, Y := 0.5}}
            Offsets := margin{Left := 0.0, Top := 0.0, Right := 0.0, Bottom := 0.0}
            Alignment := vector2{X := 0.5, Y := 0.5}
            Widget := AbilityStatInfoUI.GetUI()

        WeaponEquippedUI := weapon_equipped_ui{Player := Player}
        WeaponEquippedUI.Initialize()
        Player.AddWidgetToHUD of canvas_slot:
            Anchors := anchors{Minimum := vector2{X := 1.0, Y := 0.0}, Maximum := vector2{X := 1.0, Y := 0.0}}
            Offsets := margin{Left := -210.0, Top := 404.0, Right := 0.0, Bottom := 0.0}
            Alignment := vector2{X := 1.0, Y := 0.0}
            Widget := WeaponEquippedUI.GetUI()

        if:
            set PlayerMenuUIs[Player] = player_menu_ui{Player := Player}
            set PlayerAbilityStatInfoUI[Player] = AbilityStatInfoUI
            set PlayerWeaponEquippedUI[Player] = WeaponEquippedUI
            set PlayerIsInitialized[Player] = true

        WaitForInputManager()
        Player.SetInputTriggers(InputTriggerCollections.Lobby)
        Player.UpdatePowerUpIndicator()

    <# ----------------------------------------------------------------------------------------------------
        Get Functions
    ---------------------------------------------------------------------------------------------------- #>
    GetIsInitialized(Player:player)<decides><transacts>:void =
        PlayerIsInitialized[Player]

    GetActivePanel(Player:player)<decides><transacts>:panel_ui =
        PlayerActivePanel[Player]

    <# ----------------------------------------------------------------------------------------------------
        Update Functions
    ---------------------------------------------------------------------------------------------------- #>
    AddWidgetToHUD(Player:player, CanvasSlot:canvas_slot):void =
        if (HUD := PlayerHUDs[Player]):
            HUD.AddWidget(CanvasSlot)

    ToggleHUD(Player:player, IsShow:logic):void =
        TogglePlayerHUD(Player, IsShow)
        if (IsShow?):
            HUDControllerHideHUD.Disable()
            HUDControllerDefault.Enable()
            HUDControllerDebug.Enable()
            ShowMessage(message_id.Revives)
            ShowMessage(message_id.ExperienceBar)
            ShowMessage(message_id.ExperienceBarLevel)
        else:
            HUDControllerHideHUD.Enable()
            HUDControllerDefault.Disable()
            HUDControllerDebug.Disable()
            Player.HideMessage(message_id.Revives)
            Player.HideMessage(message_id.ExperienceBar)
            Player.HideMessage(message_id.ExperienceBarLevel)

    TogglePlayerHUD(Player:player, IsShow:logic):void =
        if (PlayerUI := GetPlayerUI[Player], PlayerHUD := PlayerHUDs[Player]):
            if (IsShow?):        
                PlayerUI.AddWidget(PlayerHUD, player_ui_slot{InputMode := ui_input_mode.None})
            else:
                PlayerUI.RemoveWidget(PlayerHUD)

    ShowGamePausedUI(Player:player):void =
        UI := if (GamePausedUI := PlayerGamePausedUI[Player]):
            GamePausedUI
        else:
            GamePausedUI := game_paused_ui{Player := Player}
            if (set PlayerGamePausedUI[Player] = GamePausedUI) {}
            GamePausedUI

        UI.Show()

    ShowObjectiveArrowTilesUI(Player:player, Objective:objective_arrow_tiles, Sequence:[]int):void =
        UI := if (ObjectiveArrowTilesUI := PlayerObjectiveArrowTilesUIs[Player]):
            ObjectiveArrowTilesUI
        else:
            ObjectiveArrowTilesUI := objective_arrow_tiles_ui{ObjectiveArrowTiles := Objective, Player := Player}
            ObjectiveArrowTilesUI.Initialize()
            if (set PlayerObjectiveArrowTilesUIs[Player] = ObjectiveArrowTilesUI) {}
            ObjectiveArrowTilesUI
        
        UI.Show(Sequence)

    ShowPanelUI(Player:player, Id:panel_ui_id):void =
        if:
            not PlayerPanelUIs[Player]
            set PlayerPanelUIs[Player] = map{}
        
        UI := if (PanelUI := PlayerPanelUIs[Player][Id]):
            PanelUI
        else:
            PanelUI := case (Id):
                panel_ui_id.WeaponSelect => weapon_select_ui{Player := Player}
                panel_ui_id.PlayerProgress => player_progress_ui{Player := Player}
                panel_ui_id.PlayerPowerUps => power_up_ui{Player := Player}
                panel_ui_id.DailyQuest => daily_quests_ui{Player := Player}
                panel_ui_id.BoonCodex => boon_codex_ui{Player := Player}
                panel_ui_id.Gears => gear_ui{Player := Player}
                panel_ui_id.Accolades => accolades_ui{Player := Player}
                panel_ui_id.NPCCodex => npc_codex_ui{Player := Player}
                panel_ui_id.InputControl => input_controls_ui{Player := Player}
                
            if (set PlayerPanelUIs[Player][Id] = PanelUI) {}
            PanelUI

        UI.Show()

    UpdateActivePanel(Player:player, PanelUI:panel_ui):void =
        if (set PlayerActivePanel[Player] = PanelUI) {}

    ClearActivePanel(Player:player):void =
        set PlayerActivePanel = RemoveKeyFromMap(PlayerActivePanel, Player)

    CloseActivePanel(Player:player, ?IsSignal:logic = false):void =
        if (ActivePanel := PlayerActivePanel[Player]):
            ActivePanel.Hide(?IsSignal := IsSignal)
            ClearActivePanel(Player)

    UpdatePowerUpIndicator(Player:player):void =
        if (Player.CanAffordPowerUp()?):
            PowerUpBeacon.RemoveFromShowList(Player)
            PowerUpBeaconNew.AddToShowList(Player)
        else:
            PowerUpBeacon.AddToShowList(Player)
            PowerUpBeaconNew.RemoveFromShowList(Player)

    <# ----------------------------------------------------------------------------------------------------
        Event Handlers
    ---------------------------------------------------------------------------------------------------- #>
    OnPlayerAdded(Player:player):void =
        spawn{Initialize(Player)}

    OnPlayerRemoved(Player:player):void =
        for (PlayerPanelUI : PlayerPanelUIs[Player]):
            PlayerPanelUI.Uninitialize()

        if (WeaponEquippedUI := PlayerWeaponEquippedUI[Player]):
            WeaponEquippedUI.Uninitialize()

        if (PlayerMenuUI := PlayerMenuUIs[Player]):
            PlayerMenuUI.Uninitialize()

        if (GamePausedUI := PlayerGamePausedUI[Player]):
            GamePausedUI.Uninitialize()

        if (ObjectiveArrowTilesUI := PlayerObjectiveArrowTilesUIs[Player]):
            ObjectiveArrowTilesUI.Uninitialize()

        if (InfoPanelUI := PlayerInfoPanelUIs[Player]):
            InfoPanelUI.Uninitialize()

        set PlayerHUDs = RemoveKeyFromMap(PlayerHUDs, Player)
        set PlayerIsInitialized = RemoveKeyFromMap(PlayerIsInitialized, Player)
        set PlayerHUDIsShowing = RemoveKeyFromMap(PlayerHUDIsShowing, Player)
        set PlayerGamePausedUI = RemoveKeyFromMap(PlayerGamePausedUI, Player)
        set PlayerMenuUIs = RemoveKeyFromMap(PlayerMenuUIs, Player)
        set PlayerActivePanel = RemoveKeyFromMap(PlayerActivePanel, Player)
        set PlayerAbilityStatInfoUI = RemoveKeyFromMap(PlayerAbilityStatInfoUI, Player)
        set PlayerWeaponEquippedUI = RemoveKeyFromMap(PlayerWeaponEquippedUI, Player)
        set PlayerPanelUIs = RemoveKeyFromMap(PlayerPanelUIs, Player)
        set PlayerInfoPanelUIs = RemoveKeyFromMap(PlayerInfoPanelUIs, Player)
        set PlayerObjectiveArrowTilesUIs = RemoveKeyFromMap(PlayerObjectiveArrowTilesUIs, Player)

    OnMenuInputPressedEvent(Agent:agent):void =
        if (Player := player[Agent], PlayerMenuUI := PlayerMenuUIs[Player]):
            # [Hotfix]
            if (Player.IsPlayerPaused()?, GetSecondsSinceEpoch() - Player.GetPauseTime() >= 1.0, CurrentPause := Player.GetCurrentPause[]):
                Player.ShowPauseUI(CurrentPause)
                Player.PlayAudio(audio_id.Activate)
            else:
                if (IsGameInProgress()? and GetAllPlayers().Length = 1):
                    PauseGame(Player, pause_info{Type := pause_type.PlayerMenu})
                else:
                    PlayerMenuUI.Show()

    OnAbilityStatsInputPressedEvent(Agent:agent):void =
        if (Player := player[Agent], AbilityStatInfoUI := PlayerAbilityStatInfoUI[Player]):
            AbilityStatInfoUI.Show()
            AbilityStatInfoUI.Update()
        
    OnAbilityStatsInputReleasedEvent(Agent:agent, DurationHeld:float):void =
        if (Player := player[Agent], AbilityStatInfoUI := PlayerAbilityStatInfoUI[Player]):
            AbilityStatInfoUI.Hide()

    OnInputUnregistered(Player:player):void =
        if (AbilityStatInfoUI := PlayerAbilityStatInfoUI[Player]):
            AbilityStatInfoUI.Hide()

    OnInteractedWithMenu(Agent:agent):void =
        if (Player := player[Agent], MenuUI := PlayerMenuUIs[Player]):
            MenuUI.Show()

    OnInteractedWithInfoPanel(Agent:agent):void =
        if (Player := player[Agent]):
            UI := if (InfoPanelUI := PlayerInfoPanelUIs[Player]):
                InfoPanelUI
            else:
                InfoPanelUI := info_panel_ui{Player := Player}
                if (set PlayerInfoPanelUIs[Player] = InfoPanelUI) {}
                InfoPanelUI
    
            UI.Show()
            Player.PlayAudio(audio_id.Activate)

    OnInteractedWithGameResult(Agent:agent):void =
        if (Player := player[Agent]):
            if (GetGameManager[].MaybeGameResult?.PlayerGameResultUI[Player]):
                Player.ShowGameResultsUI()
                Player.PlayAudio(audio_id.Activate)
            else:
                Player.PlayAudio(audio_id.Error)