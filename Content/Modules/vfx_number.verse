using { /Fortnite.com/Devices }
using { /Verse.org/SceneGraph }
using { /Verse.org/Simulation }
using { /UnrealEngine.com/Temporary/SpatialMath }
using { VisualEffects }

<# ----------------------------------------------------------------------------------------------------
    Data Definitions
---------------------------------------------------------------------------------------------------- #>
number_particle_system<public> := class<abstract>:
    ParticleSystem:particle_system_component = VisualEffects.NS_Gather{Entity := entity{}}
    SetPosition(Position:vector3):void
    SetNumber(Number:float):void
    Initialize():void = ParticleSystem.Entity.Initialize()
    Play():void = ParticleSystem.Play()

number_particle_system_gather := class(number_particle_system):
    ParticleSystem<override>:VisualEffects.NS_Gather = VisualEffects.NS_Gather{Entity := entity{}}
    SetPosition<override>(Position:vector3):void = set ParticleSystem.Position = FromVector3 of Position
    SetNumber<override>(Number:float):void = set ParticleSystem.Number = Number

MakeNumberParticleSystemGather():number_particle_system_gather =
    Entity := entity{}
    ParticleSystem := VisualEffects.NS_Gather{Entity := Entity}
    Entity.AddComponents(array{ParticleSystem})
    number_particle_system_gather{ParticleSystem := ParticleSystem}

number_particle_system_player_damage := class(number_particle_system):
    ParticleSystem<override>:VisualEffects.NS_DamageNumberPlayer = VisualEffects.NS_DamageNumberPlayer{Entity := entity{}}
    SetPosition<override>(Position:vector3):void = set ParticleSystem.Position = FromVector3 of Position
    SetNumber<override>(Number:float):void = set ParticleSystem.Number = Number

MakeNumberParticleSystemPlayerDamage():number_particle_system_player_damage =
    Entity := entity{}
    ParticleSystem := VisualEffects.NS_DamageNumberPlayer{Entity := Entity}
    Entity.AddComponents(array{ParticleSystem})
    number_particle_system_player_damage{ParticleSystem := ParticleSystem}

number_particle_system_player_heal := class(number_particle_system):
    ParticleSystem<override>:VisualEffects.NS_DamageNumberHeal = VisualEffects.NS_DamageNumberHeal{Entity := entity{}}
    SetPosition<override>(Position:vector3):void = set ParticleSystem.Position = FromVector3 of Position
    SetNumber<override>(Number:float):void = set ParticleSystem.Number = Number

MakeNumberParticleSystemPlayerHeal():number_particle_system_player_heal =
    Entity := entity{}
    ParticleSystem := VisualEffects.NS_DamageNumberHeal{Entity := Entity}
    Entity.AddComponents(array{ParticleSystem})
    number_particle_system_player_heal{ParticleSystem := ParticleSystem}

<# ----------------------------------------------------------------------------------------------------
    Global Functions
---------------------------------------------------------------------------------------------------- #>
SpawnNumberVFX(NumberParticleSystem:number_particle_system, Position:vector3, DamageAmount:float):void =
    NumberParticleSystem.SetNumber(DamageAmount)
    NumberParticleSystem.SetPosition(Position)

    if (NumberParticleSystem.ParticleSystem.IsSimulating[]):
        NumberParticleSystem.Play()
    else:
        NumberParticleSystem.Initialize()