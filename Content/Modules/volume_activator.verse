using { /Fortnite.com/Characters }
using { /Fortnite.com/Devices }
using { /Fortnite.com/Devices/CreativeAnimation }
using { /Verse.org/Simulation }
using { /UnrealEngine.com/Temporary/SpatialMath }

<# ----------------------------------------------------------------------------------------------------
    Data Definitions
---------------------------------------------------------------------------------------------------- #>
CaptureAudioIds:[]audio_id =
    array:
        audio_id.Capture
        audio_id.Capture2
        audio_id.Capture3
        audio_id.Capture4

<# ----------------------------------------------------------------------------------------------------
    volume_activator
---------------------------------------------------------------------------------------------------- #>
volume_activator := class<concrete>:
    @editable
    VolumeDevice:volume_device = volume_device{}

    @editable
    MapIndicatorDevice:?map_indicator_device = false

    @editable
    VfxId:vfx_id = vfx_id.RingProgress

    @editable
    IsProgressAlwaysVisible:logic = false

    @editable
    IsAudio:logic = false

    @editable
    var ActivationTime:float = 1.0
    
    @editable
    DecayRate:float = 0.0
    
    @editable
    Size:float = 20.0

    @editable
    PlayerMultiplier:type{_X:float where _X >= 0.0} = 0.0

    ActivatedEvent:events(tuple(player, volume_activator)) = events(tuple(player, volume_activator)){}
    ActivateCancelledEvent:events(tuple()) = events(tuple()){}

    var InitialPosition:vector3 = vector3{}
    var MaybeSpawnedVFX:?creative_prop = false
    var IsActivateTask:logic = false

    <# ----------------------------------------------------------------------------------------------------
        Initialization
    ---------------------------------------------------------------------------------------------------- #>
    OnBegin()<suspends>:void =
        set InitialPosition = VolumeDevice.GetTransform().Translation
        
        if (MapIndicator := MapIndicatorDevice?):
            if (MapIndicator.TeleportTo[MapIndicatorDisabledPosition, IdentityRotation()]) {}

        Disable()
        VolumeDevice.AgentEntersEvent.Subscribe(OnAgentEnters)
        VolumeDevice.AgentExitsEvent.Subscribe(OnAgentExits)

    <# ----------------------------------------------------------------------------------------------------
        Get Functions
    ---------------------------------------------------------------------------------------------------- #>
    GetActivationTime()<transacts>:float =
        ActivationTime * (1.0 - GetProgress()) / GetPlayerMultiplier()
        
    GetDecayTime()<transacts>:float =
        (Floor[ActivationTime / DecayRate * GetProgress()] or 0) * 1.0

    GetProgress()<transacts>:float =
        MaybeSpawnedVFX?.GetTransform().Scale.Z or 0.0

    GetIsActivating()<transacts>:logic =
        for (Player : VolumeDevice.GetPlayersInVolume(), not Player.GetFortCharacter[].IsDownButNotOut[]):
            return true

        return false

    GetPlayersActivatingCount()<transacts>:int =
        for (Player : VolumeDevice.GetPlayersInVolume(), not Player.GetFortCharacter[].IsDownButNotOut[]) {Player}.Length

    GetPlayerMultiplier()<transacts>:float =
        Max(1.0, 1.0 + (GetPlayersActivatingCount() - 1) * PlayerMultiplier)

    GetAudio()<transacts>:audio_id =
        if (PlayerMultiplier > 0.0):
            case (VolumeDevice.GetPlayersInVolume().Length - 1):
                0 => audio_id.Capture
                1 => audio_id.Capture2
                2 => audio_id.Capture3
                3 => audio_id.Capture4
                _ => audio_id.Capture
        else:
            audio_id.Capture

    <# ----------------------------------------------------------------------------------------------------
        Update Functions
    ---------------------------------------------------------------------------------------------------- #>
    Enable(?TargetPosition:?vector3 = false, ?OverrideActivationTime:?float = false):void =
        if (Position := TargetPosition?):
            set InitialPosition = Position.Make2D()

        if (OverrideTime := OverrideActivationTime?):
            set ActivationTime = OverrideTime

        if (IsProgressAlwaysVisible?):
            set MaybeSpawnedVFX = SpawnVFX(VfxId, InitialPosition, ?Scale := vector3{X := Size, Y := Size, Z := MinScale}, ?Duration := 0.0)

        if (VolumeDevice.TeleportTo[InitialPosition, IdentityRotation()]) {}

        UpdateMapIndicator(InitialPosition)

    UpdateMapIndicator(Position:vector3):void =
        if (MapIndicator := MapIndicatorDevice?):
            if (MapIndicator.TeleportTo[Position, IdentityRotation()]) {}

    Disable():void =
        ActivateCancelledEvent.Signal()

        if (VolumeDevice.TeleportTo[vector3{Z := -1000.0}, IdentityRotation()]) {}
        
        if (SpawnedVFX := MaybeSpawnedVFX?):
            SpawnedVFX.Dispose()
            set MaybeSpawnedVFX = false

        if (MapIndicator := MapIndicatorDevice?):
            if (MapIndicator.TeleportTo[MapIndicatorDisabledPosition, IdentityRotation()]) {}

    StartActivate(Player:player)<suspends>:void =
        if (not MaybeSpawnedVFX?):
            set MaybeSpawnedVFX = SpawnVFX(VfxId, InitialPosition, ?Scale := vector3{X := Size + 0.01, Y := Size, Z := MinScale}, ?Duration := ActivationTime)

        defer:
            set IsActivateTask = false

            if (SpawnedVFX := MaybeSpawnedVFX?):
                if (IsProgressAlwaysVisible?):
                    TargetTransform := transform:
                        Translation := SpawnedVFX.GetTransform().Translation
                        Rotation := SpawnedVFX.GetTransform().Rotation
                        Scale := vector3{X := Size, Y := Size, Z := MinScale}

                    if (SpawnedVFX.TeleportTo[TargetTransform]) {}
                else:
                    SpawnedVFX.Dispose()
                    set MaybeSpawnedVFX = false

        race:
            ActivateCancelledEvent.Await()

            if (DecayRate = 0.0):
                loop:
                    if (PlayerManager := GetPlayerManager[]):
                        PlayerManager.DownButNotOut.AgentDownedEvent.Await()
                        if (not GetIsActivating()?):
                            break

                    Sleep(0.0)
            else:
                Sleep(Inf)

            if (SpawnedVFX := MaybeSpawnedVFX?):
                if (DecayRate = 0.0):
                    if (SpawnedVFX.TeleportTo[transform{Translation := InitialPosition, Rotation := IdentityRotation(), Scale := vector3{X := Size + 0.01, Y := Size, Z := MinScale}}]) {}                   
                    SpawnedVFX.MoveTo(transform{Translation := InitialPosition, Rotation := IdentityRotation(), Scale := vector3{X := Size + 0.01, Y := Size, Z := 1.0}}, ActivationTime)
                    ActivatedEvent.Signal(Player, Self)
                else:
                    loop:
                        if (not Player.IsActive[]):
                            break

                        IsActivating := GetIsActivating()

                        TargetTransform := transform:
                            Translation := InitialPosition
                            Rotation := IdentityRotation()
                            Scale := vector3{X := Size + (IsActivating? and 0.01 or 0.0), Y := Size, Z := IsActivating? and 1.0 or MinScale}

                        if:
                            SpawnedVFX.TeleportTo[transform{Translation := TargetTransform.Translation, Rotation := TargetTransform.Rotation, Scale := vector3{X := Size + (IsActivating? and 0.01 or 0.0), Y := Size, Z := SpawnedVFX.GetTransform().Scale.Z}}]
                            
                        OverTime := IsActivating? and GetActivationTime() or GetDecayTime()
                        
                        MoveToResult := race:
                            SpawnedVFX.MoveTo(TargetTransform, Max(MinScale, OverTime))

                            if (AgentDownedEvent:listenable(agent) = GetGameManager[].DownButNotOut.AgentDownedEvent):
                                Agent := AgentDownedEvent.Await()
                                StopAudio(Agent)
                                move_to_result.WillNotReachDestination

                            block:
                                race:
                                    VolumeDevice.AgentEntersEvent.Await()
                                    VolumeDevice.AgentExitsEvent.Await()

                                move_to_result.WillNotReachDestination
                            
                        if (MoveToResult = move_to_result.DestinationReached):
                            if (IsActivating?):
                                ActivatedEvent.Signal(Player, Self)
                                break
                            else:
                                VolumeDevice.AgentEntersEvent.Await()
                        else:

                        Sleep(0.0)

    UpdateAudio():void =
        UpdatedAudio := GetAudio()
        for (Player : VolumeDevice.GetPlayersInVolume()):
            StopAudio(Player)
            Player.PlayAudio(UpdatedAudio)
        
    StopAudio(Agent:agent):void =
        if (PlayerMultiplier > 0.0):
            for (Id : CaptureAudioIds):
                Agent.StopAudio(Id)
        else:
            Agent.StopAudio(audio_id.Capture)

    PlayCompleteAnimation()<suspends>:void =
         if (SpawnedVFX := MaybeSpawnedVFX?, AnimationController := SpawnedVFX.GetAnimationController[]):
            KeyFrames := array:
                keyframe_delta:
                    DeltaLocation := vector3{}
                    DeltaRotation := IdentityRotation()
                    DeltaScale := vector3{X := 1.2, Y := 1.2, Z := 1.0}
                    Interpolation := InterpolationTypes.EaseOut
                    Time := 0.1
                keyframe_delta:
                    DeltaLocation := vector3{}
                    DeltaRotation := IdentityRotation()
                    DeltaScale := vector3{X := 0.5, Y := 0.5, Z := 1.0}
                    Interpolation := InterpolationTypes.EaseOut
                    Time := 0.2

            AnimationController.SetAnimation(KeyFrames, ?Mode := animation_mode.OneShot)
            AnimationController.Play()
            AnimationController.MovementCompleteEvent.Await()

    <# ----------------------------------------------------------------------------------------------------
        Event Handlers
    ---------------------------------------------------------------------------------------------------- #>
    OnAgentEnters(Agent:agent):void =
        if (Player := player[Agent], Character := Player.GetFortCharacter[]):
            if (Player.GetIsAFK()?) then return
            if (Character.IsDownButNotOut[]) then return
            if (Player.IsPlayerPaused()?) then return

            if (not IsActivateTask?):
                spawn{StartActivate(Player)}
                set IsActivateTask = true

            if (IsAudio?):
                if (PlayerMultiplier > 0.0):
                    UpdateAudio()
                else:
                    Player.PlayAudio(audio_id.Capture)

    OnAgentExits(Agent:agent):void =
        if (Player := player[Agent]):
            if (DecayRate = 0.0 and VolumeDevice.GetPlayersInVolume().Length = 0):
                ActivateCancelledEvent.Signal()

            if (IsAudio?):
                StopAudio(Player)