using { /Fortnite.com/Devices }
using { /Verse.org/Colors }
using { /Verse.org/Simulation }
using { /UnrealEngine.com/Temporary/SpatialMath }
using { /UnrealEngine.com/Temporary }
using { /UnrealEngine.com/Temporary/UI }

<# ----------------------------------------------------------------------------------------------------
    Initialization
---------------------------------------------------------------------------------------------------- #>
WeaponEquippedUIEntrySpacing:float = 54.0

<# ----------------------------------------------------------------------------------------------------
    weapon_equipped_ui
---------------------------------------------------------------------------------------------------- #>
weapon_equipped_ui := class(ui_component):

    var Icons:[]texture_block = array{}
    var Subscriptions:[]cancelable = array{}
    var EventsSubscriptions:[]events_cancelable = array{}

    <# ----------------------------------------------------------------------------------------------------
        Initialization
    ---------------------------------------------------------------------------------------------------- #>
    CreateUI<override>():widget =
        spawn{Setup()}

        StackBox := stack_box:
            Orientation := orientation.Vertical
            Slots := for (Index := 1..MaxPlayers):
                Icon := texture_block:
                    DefaultImage := UserInterface.Icons.WeaponBow_64
                    DefaultDesiredSize := vector2{X := 32.0, Y := 32.0}

                set Icons += array{Icon}

                stack_box_slot:
                    HorizontalAlignment := horizontal_alignment.Fill
                    VerticalAlignment := vertical_alignment.Fill
                    Padding := margin{Left := 0.0, Top := (Index = 1) and 0.0 or WeaponEquippedUIEntrySpacing, Right := 0.0, Bottom := 0.0}
                    Widget := Icon

    Setup()<suspends>:void =
        WaitForPlayerManager(Player)
        WaitForGameManager()

        if (PlayerManager := GetPlayerManager[], GameManager := GetGameManager[]):
            set EventsSubscriptions = array:
                PlayerManager.WeaponEquippedEvent.Subscribe(OnWeaponEquipped)
                GameManager.GameStartedEvent.Subscribe(OnGameStarted)
                GameManager.GameEndedEvent.Subscribe(OnGameEnded)

            set Subscriptions = array:
                PlayerManager.GetPlayspace().PlayerAddedEvent().Subscribe(OnPlayerAdded)
                PlayerManager.GetPlayspace().PlayerRemovedEvent().Subscribe(OnPlayerRemoved)

        Sleep(0.0)
        Update()

    Uninitialize():void =
        for (EventsSubscription : EventsSubscriptions):
            EventsSubscription.Cancel()

        for (Subscription : Subscriptions):
            Subscription.Cancel()

    <# ----------------------------------------------------------------------------------------------------
        Get Functions
    ---------------------------------------------------------------------------------------------------- #>
    GetPlayerColor(Index:int)<transacts>:color =
        case (Index):
            0 => ThemeColors.PlayerColor1
            1 => ThemeColors.PlayerColor2
            2 => ThemeColors.PlayerColor3
            3 => ThemeColors.PlayerColor4
            _ => ThemeColors.PlayerColor1

    <# ----------------------------------------------------------------------------------------------------
        Update Functions
    ---------------------------------------------------------------------------------------------------- #>
    Update<override>():void =
        PlayersByJoinedTime := for (AllPlayer : GetAllPlayers(), AllPlayer <> Player) {(AllPlayer, AllPlayer.GetJoinedTime())}
        SortedPlayersByJoinedTime := SortBy(PlayersByJoinedTime, LessPlayerFloat)

        for (Index -> Icon : Icons):
            if:
                SortedPlayer := (Index = 0) and Player or SortedPlayersByJoinedTime[Index - 1](0)
                WeaponId := SortedPlayer.GetEquippedWeapon()?
                WeaponInfo := GetWeaponInfo[WeaponId]
            then:
                Icon.SetImage(WeaponInfo.IconSmall)
                Icon.SetVisibility(widget_visibility.Visible)
                <#
                if (IsGameInProgress()?):
                    Icon.SetTint(GetPlayerColor(GetPlayerIndex(SortedPlayer)))
                #>
            else:
                Icon.SetVisibility(widget_visibility.Hidden)

    <# ----------------------------------------------------------------------------------------------------
        Event Handlers
    ---------------------------------------------------------------------------------------------------- #>
    OnGameStarted():void =
        #Update()
        Hide()

    OnGameEnded():void =
        Show()
        Update()
        <#
        for (Index -> Icon : Icons):
            Icon.SetTint(NamedColors.White)
        #>

    OnPlayerAdded(InPlayer:player):void =
        Update()

    OnPlayerRemoved(InPlayer:player):void =
        Update()

    OnWeaponEquipped(InPlayer:player):void =
        Update()