using { /Fortnite.com/UI }
using { /Verse.org/Colors }
using { /UnrealEngine.com/Temporary/SpatialMath }
using { /UnrealEngine.com/Temporary/UI }

weapon_select_entry_ui := class(ui_component):
    Icon:texture_block = texture_block:
        DefaultImage := UserInterface.Icons.WeaponSword_128
        DefaultDesiredSize := vector2{X := 100.0, Y := 100.0}

    LevelText:text_block = text_block:
        DefaultTextSize := ThemeTextSize.Small
        DefaultTextColor := ThemeColors.PrimaryText
        DefaultShadowOffset := option{vector2{X := 2.0, Y := 2.0}}
        DefaultShadowOpacity := 0.5

    EquipButton:button_quiet = button_quiet{}

    var MaybeWeaponInfo:?weapon_info = false
    var Subscriptions:[]cancelable = array{}

    <# ----------------------------------------------------------------------------------------------------
        Initialization
    ---------------------------------------------------------------------------------------------------- #>
    CreateUI<override>():widget =
        EquipButton.SetText(MakeMessage("Equip"))

        set Subscriptions = array:
            EquipButton.OnClick().Subscribe(OnClickEquip)

        stack_box:
            Orientation := orientation.Vertical
            Slots := array:
                stack_box_slot:
                    HorizontalAlignment := horizontal_alignment.Fill
                    VerticalAlignment := vertical_alignment.Top
                    Padding := margin{Left := 20.0, Top := 0.0, Right := 20.0, Bottom := 0.0}
                    Widget := overlay:
                        Slots := array:
                            overlay_slot:
                                HorizontalAlignment := horizontal_alignment.Left
                                VerticalAlignment := vertical_alignment.Center
                                Widget := Icon
                            overlay_slot:
                                HorizontalAlignment := horizontal_alignment.Right
                                VerticalAlignment := vertical_alignment.Bottom
                                Widget := stack_box:
                                    Orientation := orientation.Horizontal
                                    Slots := array:
                                        stack_box_slot:
                                            HorizontalAlignment := horizontal_alignment.Fill
                                            VerticalAlignment := vertical_alignment.Fill
                                            Widget := text_block:
                                                DefaultTextSize := ThemeTextSize.Small
                                                DefaultText := MakeMessage("Lv.")
                                                DefaultTextColor := ThemeColors.Highlight
                                                DefaultShadowColor := ThemeColors.Highlight * 0.125
                                                DefaultShadowOffset := option{vector2{X := 2.0, Y := 2.0}}
                                                DefaultShadowOpacity := 1.0
                                        stack_box_slot:
                                            HorizontalAlignment := horizontal_alignment.Fill
                                            VerticalAlignment := vertical_alignment.Fill
                                            Padding := margin{Left := 7.5, Top := 0.0, Right := 0.0, Bottom := 0.0}
                                            Widget := LevelText
                stack_box_slot:
                    HorizontalAlignment := horizontal_alignment.Fill
                    VerticalAlignment := vertical_alignment.Bottom
                    Padding := margin{Left := 0.0, Top := 20.0, Right := 0.0, Bottom := 0.0}
                    Distribution := option{1.0}
                    Widget := EquipButton

    Uninitialize():void =
        for (Subscription : Subscriptions):
            Subscription.Cancel()

    <# ----------------------------------------------------------------------------------------------------
        Update Functions
    ---------------------------------------------------------------------------------------------------- #>
    Hide<override>():void =
        UI.SetVisibility(widget_visibility.Hidden)

    Set(WeaponInfo:weapon_info):void =
        set MaybeWeaponInfo = option{WeaponInfo}
        Icon.SetImage(WeaponInfo.Icon)
        UI.SetVisibility(widget_visibility.Visible)

    Update<override>():void =
        if (WeaponInfo := MaybeWeaponInfo?):
            LevelText.SetText(MakeMessage("{Player.GetWeaponProgress(WeaponInfo.Id)}"))

            if (EquippedWeapon := Player.GetEquippedWeapon()?, WeaponInfo.Id = EquippedWeapon):
                Icon.SetTint(ThemeColors.Highlight * 2.0)
                EquipButton.SetText(MakeMessage("Equipped"))
            else:
                if (Player.IsWeaponUnlocked(WeaponInfo.Id)?):
                    Icon.SetTint(NamedColors.White)
                    EquipButton.SetText(MakeMessage("Equip"))
                else:
                    Icon.SetTint(ThemeColors.Disabled)
                    EquipButton.SetText(MakeMessage("Locked"))

    <# ----------------------------------------------------------------------------------------------------
        Event Handlers
    ---------------------------------------------------------------------------------------------------- #>
    OnClickEquip(Message:widget_message):void =
        if (WeaponInfo := MaybeWeaponInfo?):
            if (not Player.IsWeaponUnlocked(WeaponInfo.Id)?):
                return

            if (EquippedWeapon := Player.GetEquippedWeapon()?):
                if (WeaponInfo.Id <> EquippedWeapon):
                    Player.EquipWeapon(WeaponInfo.Id)
            else:
                Player.EquipWeapon(WeaponInfo.Id)