using { /Fortnite.com/UI }
using { /Verse.org/Assets }
using { /Verse.org/Colors }
using { /UnrealEngine.com/Temporary/SpatialMath }
using { /UnrealEngine.com/Temporary/UI }

weapon_select_ui := class(panel_ui):
    Title<override>:string = "Weapon Select"
    BackgroundOpacity<override>:type {_X:float where 0.000000 <= _X, _X <= 1.000000} = 0.7
    Anchors<override>:anchors = anchors{Minimum := vector2{X := 0.05, Y := 0.0}, Maximum := vector2{X := 0.95, Y := 1.0}}

    NameText:text_block = text_block:
        DefaultTextColor := ThemeColors.PrimaryText
        DefaultShadowOffset := option{vector2{X := 2.0, Y := 2.0}}
        DefaultShadowOpacity := 0.5

    Icon:texture_block = texture_block:
        DefaultImage := UserInterface.Icons.WeaponBow_128
        DefaultDesiredSize := vector2{X := 128.0, Y := 128.0}

    DescriptionText:text_block = text_block:
        DefaultTextColor := ThemeColors.PrimaryText
        DefaultTextOpacity := 0.5
    
    CostText:text_block = text_block:
        DefaultTextColor := ThemeColors.Success
        DefaultShadowColor := ThemeColors.Success * 0.125
        DefaultShadowOffset := option{vector2{X := 2.0, Y := 2.0}}
        DefaultShadowOpacity := 1.0

    DamageText:text_block = text_block:
        DefaultTextColor := ThemeColors.Success
        DefaultShadowColor := ThemeColors.Success * 0.125
        DefaultShadowOffset := option{vector2{X := 2.0, Y := 2.0}}
        DefaultShadowOpacity := 1.0

    CooldownText:text_block = text_block:
        DefaultTextColor := ThemeColors.Success
        DefaultShadowColor := ThemeColors.Success * 0.125
        DefaultShadowOffset := option{vector2{X := 2.0, Y := 2.0}}
        DefaultShadowOpacity := 1.0

    SizeText:text_block = text_block:
        DefaultTextColor := ThemeColors.Success
        DefaultShadowColor := ThemeColors.Success * 0.125
        DefaultShadowOffset := option{vector2{X := 2.0, Y := 2.0}}
        DefaultShadowOpacity := 1.0

    CurrencyText:text_block = text_block:
        DefaultTextColor := ThemeColors.Success
        DefaultShadowColor := ThemeColors.Success * 0.125
        DefaultShadowOffset := option{vector2{X := 2.0, Y := 2.0}}
        DefaultShadowOpacity := 1.0

    PageNumberText:text_block = text_block:
        DefaultTextColor := ThemeColors.PrimaryText
        DefaultTextOpacity := 0.5

    UnlockButton:button_quiet = button_quiet{}

    GridRow:int = 3
    GridColumn:int = 3
    EntriesPerPage:int = 9

    var WeaponSelectEntryUIs:[]weapon_select_entry_ui = array{}
    var CurrencyUI:widget = canvas{}
    var WeaponStatsUI:widget = canvas{}
    var WeaponCostUI:widget = canvas{}
    var CurrentPage:int = 1
    var MaybeWeaponInfo:?weapon_info = false
    var EventsSubscriptions:[]events_cancelable = array{}
    var Subscriptions:[]cancelable = array{}

    <# ----------------------------------------------------------------------------------------------------
        Initialization
    ---------------------------------------------------------------------------------------------------- #>
    CreateContent<override>():widget =
        if (WeaponEquippedEvent := Player.GetStats[].WeaponEquippedEvent):
            set EventsSubscriptions = array{WeaponEquippedEvent.Subscribe(OnWeaponEquipped)}

        stack_box:
            Orientation := orientation.Horizontal
            Slots := array:
                stack_box_slot:
                    HorizontalAlignment := horizontal_alignment.Fill
                    VerticalAlignment := vertical_alignment.Fill
                    Distribution := option{1.0}
                    Widget := CreateGrid()
                stack_box_slot:
                    HorizontalAlignment := horizontal_alignment.Fill
                    VerticalAlignment := vertical_alignment.Fill
                    Distribution := option{1.0}
                    Padding := margin{Left := 20.0, Top := 0.0, Right := 0.0, Bottom := 0.0}
                    Widget := CreateInfoUI()

    CreateGrid():widget =
        PreviousButton := button_quiet{DefaultText := MakeMessage("◀")}
        NextButton := button_quiet{DefaultText := MakeMessage("▶")}

        set Subscriptions = array:
            PreviousButton.OnClick().Subscribe(OnClickPrevious)
            NextButton.OnClick().Subscribe(OnClickNext)

        overlay:
            Slots := array:
                overlay_slot:
                    HorizontalAlignment := horizontal_alignment.Fill
                    VerticalAlignment := vertical_alignment.Fill
                    Widget := UserInterface.VerseWidget.WBP_RoundedBox{}
                overlay_slot:
                    HorizontalAlignment := horizontal_alignment.Fill
                    VerticalAlignment := vertical_alignment.Fill
                    Padding := margin{Left := 40.0, Top := 40.0, Right := 40.0, Bottom := 40.0}
                    Widget := stack_box:
                        Orientation := orientation.Vertical
                        Slots := array:
                            stack_box_slot:
                                HorizontalAlignment := horizontal_alignment.Fill
                                VerticalAlignment := vertical_alignment.Fill
                                Widget := stack_box:
                                    Orientation := orientation.Vertical
                                    Slots := for (RowIndex := 0..GridRow - 1):
                                        stack_box_slot:
                                            HorizontalAlignment := horizontal_alignment.Fill
                                            VerticalAlignment := vertical_alignment.Fill
                                            Distribution := option{1.0}
                                            Padding := margin{Left := 0.0, Top := (RowIndex = 0) and 0.0 or 50.0, Right := 0.0, Bottom := 0.0}
                                            Widget := stack_box:
                                                Orientation := orientation.Horizontal
                                                Slots := for (ColumnIndex := 0..GridColumn - 1):
                                                    EntryIndex := GridRow * RowIndex + ColumnIndex
                                                    WeaponSelectEntryUI := weapon_select_entry_ui{Player := Player}
                                                    WeaponSelectEntryUI.Initialize()

                                                    if (WeaponId := StartingWeapons[EntryIndex], WeaponInfo := GetWeaponInfo[WeaponId]):
                                                        WeaponSelectEntryUI.Set(WeaponInfo)
                                                        spawn{AwaitButtonClicked(WeaponSelectEntryUI.EquipButton.OnClick(), EntryIndex)}

                                                    set WeaponSelectEntryUIs += array{WeaponSelectEntryUI}

                                                    stack_box_slot:
                                                        HorizontalAlignment := horizontal_alignment.Fill
                                                        VerticalAlignment := vertical_alignment.Fill
                                                        Distribution := option{1.0}
                                                        Padding := margin{Left := (ColumnIndex = 0) and 0.0 or 20.0, Top := 0.0, Right := 0.0, Bottom := 0.0}
                                                        Widget := WeaponSelectEntryUI.GetUI()
                            stack_box_slot:
                                HorizontalAlignment := horizontal_alignment.Fill
                                VerticalAlignment := vertical_alignment.Bottom
                                Padding := margin{Left := 0.0, Top := 20.0, Right := 0.0, Bottom := 20.0}
                                Distribution := option{1.0}
                                Widget := color_block:
                                    DefaultColor := NamedColors.White
                                    DefaultOpacity := 0.1
                                    DefaultDesiredSize := vector2{X := 0.0, Y := 1.0}
                            stack_box_slot:
                                HorizontalAlignment := horizontal_alignment.Fill
                                VerticalAlignment := vertical_alignment.Fill
                                Widget := stack_box:
                                    Orientation := orientation.Horizontal
                                    Slots := array:
                                        stack_box_slot:
                                            HorizontalAlignment := horizontal_alignment.Fill
                                            Distribution := option{1.0}
                                            Widget := PreviousButton
                                        stack_box_slot:
                                            Distribution := option{1.0}
                                            Padding := margin{Left := 20.0, Top := 0.0, Right := 20.0, Bottom := 0.0}
                                            Widget := PageNumberText
                                        stack_box_slot:
                                            HorizontalAlignment := horizontal_alignment.Fill
                                            Distribution := option{1.0}
                                            Widget := NextButton

    CreateInfoUI():widget =
        set Subscriptions += array{UnlockButton.OnClick().Subscribe(OnClickBuy)}

        stack_box:
            Orientation := orientation.Vertical
            Slots := array:
                stack_box_slot:
                    HorizontalAlignment := horizontal_alignment.Fill
                    VerticalAlignment := vertical_alignment.Fill
                    Padding := margin{Left := 0.0, Top := 0.0, Right := 0.0, Bottom := 20.0}
                    Widget := set CurrencyUI = CreateCurrencyUI()
                stack_box_slot:
                    HorizontalAlignment := horizontal_alignment.Fill
                    VerticalAlignment := vertical_alignment.Fill
                    Distribution := option{1.0}
                    Widget := overlay:
                        Slots := array:
                            overlay_slot:
                                HorizontalAlignment := horizontal_alignment.Fill
                                VerticalAlignment := vertical_alignment.Fill
                                Widget := UserInterface.VerseWidget.WBP_RoundedBox{}
                            overlay_slot:
                                HorizontalAlignment := horizontal_alignment.Fill
                                VerticalAlignment := vertical_alignment.Fill
                                Padding := margin{Left := 40.0, Top := 40.0, Right := 40.0, Bottom := 40.0}
                                Widget := stack_box:
                                    Orientation := orientation.Vertical
                                    Slots := array:
                                        stack_box_slot:
                                            HorizontalAlignment := horizontal_alignment.Center
                                            VerticalAlignment := vertical_alignment.Top
                                            Padding := margin{Left := 0.0, Top := 0.0, Right := 0.0, Bottom := 20.0}
                                            Widget := NameText
                                        stack_box_slot:
                                            VerticalAlignment := vertical_alignment.Top
                                            Widget := Icon
                                        stack_box_slot:
                                            HorizontalAlignment := horizontal_alignment.Fill
                                            VerticalAlignment := vertical_alignment.Top
                                            Padding := margin{Left := 0.0, Top := 20.0, Right := 0.0, Bottom := 40.0}
                                            Widget := color_block:
                                                DefaultColor := NamedColors.White
                                                DefaultOpacity := 0.1
                                                DefaultDesiredSize := vector2{X := 0.0, Y := 1.0}
                                        stack_box_slot:
                                            HorizontalAlignment := horizontal_alignment.Left
                                            VerticalAlignment := vertical_alignment.Top
                                            Widget := DescriptionText
                                        stack_box_slot:
                                            HorizontalAlignment := horizontal_alignment.Fill
                                            VerticalAlignment := vertical_alignment.Bottom
                                            Padding := margin{Left := 0.0, Top := 20.0, Right := 0.0, Bottom := 40.0}
                                            Distribution := option{1.0}
                                            Widget := color_block:
                                                DefaultColor := NamedColors.White
                                                DefaultOpacity := 0.1
                                                DefaultDesiredSize := vector2{X := 0.0, Y := 1.0}
                                        stack_box_slot:
                                            HorizontalAlignment := horizontal_alignment.Fill
                                            VerticalAlignment := vertical_alignment.Fill
                                            Widget := set WeaponStatsUI = CreateWeaponStatsUI()
                                        stack_box_slot:
                                            HorizontalAlignment := horizontal_alignment.Fill
                                            VerticalAlignment := vertical_alignment.Fill
                                            Widget := set WeaponCostUI = CreateWeaponCostUI()
                                        stack_box_slot:
                                            HorizontalAlignment := horizontal_alignment.Fill
                                            VerticalAlignment := vertical_alignment.Fill
                                            Padding := margin{Left := 0.0, Top := 40.0, Right := 0.0, Bottom := 0.0}
                                            Widget := UnlockButton

    CreateCurrencyUI():widget =
        overlay:
            Slots := array:
                overlay_slot:
                    HorizontalAlignment := horizontal_alignment.Fill
                    VerticalAlignment := vertical_alignment.Fill
                    Widget := UserInterface.VerseWidget.WBP_RoundedBox{}
                overlay_slot:
                    VerticalAlignment := vertical_alignment.Fill
                    Padding := margin{Left := 40.0, Top := 20.0, Right := 40.0, Bottom := 20.0}
                    Widget := stack_box:
                        Orientation := orientation.Horizontal
                        Slots := array:
                            stack_box_slot:
                                HorizontalAlignment := horizontal_alignment.Left
                                Padding := margin{Left := 0.0, Top := 0.0, Right := 10.0, Bottom := 0.0}
                                Widget := texture_block:
                                    DefaultImage := currency_id.Gold.ToIcon()
                                    DefaultDesiredSize := vector2{X := 40.0, Y := 40.0}
                            stack_box_slot:
                                HorizontalAlignment := horizontal_alignment.Fill
                                VerticalAlignment := vertical_alignment.Fill
                                Widget := CurrencyText

    CreateWeaponCostUI():widget =
        CreateStatEntry(UserInterface.Icons.Gold_64, NamedColors.White, "Cost", CostText)

    CreateWeaponStatsUI():widget =
        stack_box:
            Orientation := orientation.Vertical
            Slots := array:
                stack_box_slot:
                    HorizontalAlignment := horizontal_alignment.Center
                    VerticalAlignment := vertical_alignment.Top
                    Padding := margin{Left := 0.0, Top := 0.0, Right := 0.0, Bottom := 20.0}
                    Widget := text_block:
                        DefaultText := MakeMessage("Base Stats")
                        DefaultTextColor := ThemeColors.PrimaryText
                        DefaultShadowOffset := option{vector2{X := 2.0, Y := 2.0}}
                        DefaultShadowOpacity := 0.5
                stack_box_slot:
                    HorizontalAlignment := horizontal_alignment.Fill
                    Widget := CreateStatEntry(UserInterface.Icons.ProjectileDamage_64, ThemeColors.WeaponDamage, "Damage", DamageText)
                stack_box_slot:
                    HorizontalAlignment := horizontal_alignment.Fill
                    Padding := margin{Left := 0.0, Top := 5.0, Right := 0.0, Bottom := 0.0}
                    Widget := CreateStatEntry(UserInterface.Icons.AbilityCooldown_64, ThemeColors.WeaponCooldown, "Cooldown", CooldownText)
                stack_box_slot:
                    HorizontalAlignment := horizontal_alignment.Fill
                    Padding := margin{Left := 0.0, Top := 5.0, Right := 0.0, Bottom := 0.0}
                    Widget := CreateStatEntry(UserInterface.Icons.AbilitySize_64, ThemeColors.WeaponSize, "Size", SizeText)

    CreateStatEntry(StatIcon:texture, StatIconColor:color, Label:string, Text:text_block):widget =
        stack_box:
            Orientation := orientation.Horizontal
            Slots := array:
                stack_box_slot:
                    HorizontalAlignment := horizontal_alignment.Left
                    Widget := texture_block:
                        DefaultImage := StatIcon
                        DefaultTint := StatIconColor
                        DefaultDesiredSize := vector2{X := 40.0, Y := 40.0}
                stack_box_slot:
                    HorizontalAlignment := horizontal_alignment.Fill
                    VerticalAlignment := vertical_alignment.Fill
                    Padding := margin{Left := 20.0, Top := 0.0, Right := 0.0, Bottom := 0.0}
                    Widget := text_block:
                        DefaultText := MakeMessage("{Label}")
                        DefaultTextColor := ThemeColors.PrimaryText
                        DefaultTextOpacity := 0.5
                stack_box_slot:
                    HorizontalAlignment := horizontal_alignment.Right
                    VerticalAlignment := vertical_alignment.Fill
                    Distribution := option{1.0}
                    Widget := Text

    Uninitialize<override>():void =
        for (WeaponSelectEntryUI : WeaponSelectEntryUIs):
            WeaponSelectEntryUI.Uninitialize()

        for (EventsSubscription : EventsSubscriptions):
            EventsSubscription.Cancel()

        for (Subscription : Subscriptions):
            Subscription.Cancel()

    <# ----------------------------------------------------------------------------------------------------
        Get Functions
    ---------------------------------------------------------------------------------------------------- #>
    GetMaxPages()<transacts>:int =
        Ceil[(GetTotalWeapons() * 1.0) / (EntriesPerPage * 1.0)] or 0

    <# ----------------------------------------------------------------------------------------------------
        Update Functions
    ---------------------------------------------------------------------------------------------------- #>
    Update():void =
        for (WeaponSelectEntryUI : WeaponSelectEntryUIs):
            WeaponSelectEntryUI.Update()

    UpdateInfo(WeaponInfo:weapon_info):void =
        NameText.SetText(MakeMessage("{WeaponInfo.Name}"))
        Icon.SetImage(WeaponInfo.Icon)
        DescriptionText.SetText(MakeMessage("{WeaponInfo.Description.WrapText(650.0)}"))

        if (Player.IsWeaponUnlocked(WeaponInfo.Id)?):
            Icon.SetTint(NamedColors.White)
            DamageText.SetText(MakeMessage("{GetRoundedString(WeaponInfo.Damage, 0)}"))
            CooldownText.SetText(MakeMessage("{GetRoundedString(WeaponInfo.Cooldown, 0)}s"))
            SizeText.SetText(MakeMessage("{GetRoundedString(WeaponInfo.Size, 0)}"))
            UnlockButton.SetText(MakeMessage("Evolution Info"))
            WeaponStatsUI.SetVisibility(widget_visibility.Visible)
            WeaponCostUI.SetVisibility(widget_visibility.Collapsed)
            CurrencyUI.SetVisibility(widget_visibility.Collapsed)
        else:
            Icon.SetTint(ThemeColors.Disabled)
            CurrencyText.SetText(MakeMessage("{FormatAsCurrency(Player.GetGoldAvailable(currency_id.Gold))}"))
            CostText.SetText(MakeMessage("{FormatAsCurrency(Player.GetWeaponUnlockCost())}"))
            CostText.SetColor(Player.CanAffordWeaponUnlock()? and ThemeColors.Success or ThemeColors.Failure)
            UnlockButton.SetText(MakeMessage("Unlock"))
            WeaponStatsUI.SetVisibility(widget_visibility.Collapsed)
            WeaponCostUI.SetVisibility(widget_visibility.Visible)
            CurrencyUI.SetVisibility(widget_visibility.Visible)

    UpdatePage():void =
        WeaponIds := GetWeaponIds()
        PageNumberText.SetText(MakeMessage("{CurrentPage}/{GetMaxPages()}"))
        
        for (Index -> WeaponSelectEntryUI : WeaponSelectEntryUIs):
            EntyIndex := Index + EntriesPerPage * (CurrentPage - 1)
            if (WeaponId := WeaponIds[EntyIndex], WeaponInfo := GetWeaponInfo[WeaponId]):
                WeaponSelectEntryUI.Set(WeaponInfo)
                WeaponSelectEntryUI.Update()
            else:
                WeaponSelectEntryUI.Hide()

    <# ----------------------------------------------------------------------------------------------------
        Event Handlers
    ---------------------------------------------------------------------------------------------------- #>
    OnWeaponEquipped():void =
        Update()
        Player.PlayAudio(audio_id.Activate)

    OnShow<override>():void =
        set CurrentPage = 1
        Update()
        UpdatePage()
        if (EquippedWeapon := Player.GetEquippedWeapon()?, WeaponInfo := GetWeaponInfo[EquippedWeapon]):
            set MaybeWeaponInfo = option{WeaponInfo}
            UpdateInfo(WeaponInfo)

    AwaitButtonClicked(Listenable:listenable(widget_message), Index:int)<suspends>:void =
        if (PlayerRemovedEvent := Player.GetPlayerRemovedEvent[]):
            race:
                PlayerRemovedEvent.Await()
                loop:
                    Listenable.Await()
                    EntyIndex := Index + EntriesPerPage * (CurrentPage - 1)
                    if (WeaponId := GetWeaponIds()[EntyIndex], WeaponInfo := GetWeaponInfo[WeaponId]):
                        set MaybeWeaponInfo = option{WeaponInfo}
                        UpdateInfo(WeaponInfo)

    OnClickPrevious(Message:widget_message):void =
        set CurrentPage = (CurrentPage = 1) and GetMaxPages() or CurrentPage - 1
        UpdatePage()

    OnClickNext(Message:widget_message):void =
        set CurrentPage = (CurrentPage = GetMaxPages()) and 1 or CurrentPage + 1
        UpdatePage()

    OnClickBuy(Message:widget_message):void =
        if (WeaponInfo := MaybeWeaponInfo?):
            if (Player.IsWeaponUnlocked(WeaponInfo.Id)?):
                Hide()
                
                Player.ShowWeaponEvolutionUI(WeaponInfo.Id, Player.GetWeaponEvolutionLevel(WeaponInfo.Id), ?IsPreview := true)
                return

            if (Player.CanAffordWeaponUnlock()?):
                Player.UnlockWeapon(WeaponInfo.Id)
                Player.PlayAudio(audio_id.PowerUp)
                UpdateInfo(WeaponInfo)
                UpdatePage()
            else:
                Player.PlayAudio(audio_id.Error)