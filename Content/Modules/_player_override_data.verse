IsOverridePlayerDataOnce:logic = false
GetPlayerDataOverrides()<transacts>:?[]player_data = false
<#
    option:
        array:
            player_data:
                Version := 0
                News := Floor[GetNewsInfos()[0].Version * 1000] or 0
                Settings := player_settings:
                    Camera := 0
                    CameraDistance := 3
                    AutoAim := true
                    SelectBufferTime := 0.0
                ArenaProgress := 36
                EndlessModeProgress := array:
                    0, 3, 0, 5, 0, 0
                PowerUps := map:
                    power_up_id.Damage.ToInt() => power_up_id.Damage.GetMaxLevel()
                    power_up_id.Cooldown.ToInt() => power_up_id.Cooldown.GetMaxLevel()
                    power_up_id.Size.ToInt() => power_up_id.Size.GetMaxLevel()
                    power_up_id.MaxHealth.ToInt() => power_up_id.MaxHealth.GetMaxLevel()
                    power_up_id.HealthRegeneration.ToInt() => power_up_id.HealthRegeneration.GetMaxLevel()
                    power_up_id.ItemPickupDistance.ToInt() => power_up_id.ItemPickupDistance.GetMaxLevel()
                    power_up_id.CriticalRate.ToInt() => power_up_id.CriticalRate.GetMaxLevel()
                    power_up_id.CriticalDamage.ToInt() => power_up_id.CriticalDamage.GetMaxLevel()
                    power_up_id.ExperienceMultiplier.ToInt() => power_up_id.ExperienceMultiplier.GetMaxLevel()
                    power_up_id.AbilityReroll.ToInt() => 100
                    power_up_id.AbilityBanish.ToInt() => power_up_id.AbilityBanish.GetMaxLevel()
                    power_up_id.AbilityBoost.ToInt() => power_up_id.AbilityBoost.GetMaxLevel()
                    power_up_id.DashCount.ToInt() => power_up_id.DashCount.GetMaxLevel()
                    power_up_id.Luck.ToInt() => power_up_id.Luck.GetMaxLevel()
                    power_up_id.GoldMultiplier.ToInt() => power_up_id.GoldMultiplier.GetMaxLevel()
                    power_up_id.HealthPickup.ToInt() => power_up_id.HealthPickup.GetMaxLevel()
                    power_up_id.GoldenTreeRate.ToInt() => power_up_id.GoldenTreeRate.GetMaxLevel()
                    power_up_id.WeaponReroll.ToInt() => power_up_id.WeaponReroll.GetMaxLevel()
                    power_up_id.TreasureChestLuck.ToInt() => power_up_id.TreasureChestLuck.GetMaxLevel()
                    power_up_id.GatherRate.ToInt() => power_up_id.GatherRate.GetMaxLevel()
                    power_up_id.TreasureChestFocus.ToInt() => power_up_id.TreasureChestFocus.GetMaxLevel()
                    power_up_id.Revive.ToInt() => power_up_id.Revive.GetMaxLevel()
                    power_up_id.ReviveRegeneration.ToInt() => power_up_id.ReviveRegeneration.GetMaxLevel()
                    power_up_id.AirDash.ToInt() => power_up_id.AirDash.GetMaxLevel()
                    power_up_id.BoonReroll.ToInt() => power_up_id.BoonReroll.GetMaxLevel()
                    power_up_id.TreasureChestTier.ToInt() => power_up_id.TreasureChestTier.GetMaxLevel()
                    power_up_id.ReviveBlast.ToInt() => power_up_id.ReviveBlast.GetMaxLevel()
                    power_up_id.MoveSpeed.ToInt() => power_up_id.MoveSpeed.GetMaxLevel()
                    power_up_id.NinjaMastery01.ToInt() => power_up_id.NinjaMastery01.GetMaxLevel()
                    power_up_id.NinjaMastery02.ToInt() => power_up_id.NinjaMastery02.GetMaxLevel()
                    power_up_id.NinjaMastery03.ToInt() => power_up_id.NinjaMastery03.GetMaxLevel()
                    power_up_id.NinjaMastery04.ToInt() => power_up_id.NinjaMastery04.GetMaxLevel()
                    power_up_id.NinjaMastery05.ToInt() => power_up_id.NinjaMastery05.GetMaxLevel()
                    power_up_id.NinjaMastery06.ToInt() => power_up_id.NinjaMastery06.GetMaxLevel()
                    power_up_id.NinjaMastery07.ToInt() => power_up_id.NinjaMastery07.GetMaxLevel()
                    power_up_id.NinjaMastery08.ToInt() => power_up_id.NinjaMastery08.GetMaxLevel()
                    power_up_id.NinjaMastery09.ToInt() => power_up_id.NinjaMastery09.GetMaxLevel()
                Currencies := map:
                    currency_id.Gold.ToInt() => 1000000000.0
                    currency_id.TreasureChest.ToInt() => 150000.0
                    currency_id.GearRecipe.ToInt() => 0.0
                WeaponEquipped := weapon_id.Kunai.ToInt()
                WeaponsUnlocked := map:
                    weapon_id.ChainLightning.ToInt() => true
                    weapon_id.EnergyBomb.ToInt() => true
                    weapon_id.Kunai.ToInt() => true
                    weapon_id.Boomerang.ToInt() => true
                    weapon_id.EnergyBeam.ToInt() => true
                    weapon_id.DancingDagger.ToInt() => true
                    weapon_id.TeslaCoil.ToInt() => true
                    weapon_id.Hammer.ToInt() => true
                    weapon_id.Staff.ToInt() => true
                WeaponProgress := map:
                    weapon_id.Sword.ToInt() => 1
                    weapon_id.Bow.ToInt() => 2
                    weapon_id.Shuriken.ToInt() => 3
                    weapon_id.Sentry.ToInt() => 4
                    weapon_id.Aura.ToInt() => 5
                    weapon_id.Torrent.ToInt() => 6
                    weapon_id.Lightning.ToInt() => 7
                    weapon_id.Trail.ToInt() => 7
                    weapon_id.Shield.ToInt() => 7
                    weapon_id.ChainLightning.ToInt() => 4
                    weapon_id.EnergyBomb.ToInt() => 3
                    weapon_id.Boomerang.ToInt() => 6
                    weapon_id.EnergyBeam.ToInt() => 7
                    weapon_id.Kunai.ToInt() => 6

                DailyQuests := player_daily_quests:
                    LastQuestAssignedTime := GetSecondsSinceEpoch() + 62135596800.0 # - (86400.0 * 3.0) # Current date daily minus offset
                    ActiveQuests := array:
                        daily_quest:
                            Id := daily_quest_id.CompleteStage.ToInt()
                            Tier := 1
                            Progress := 0.0
                        daily_quest:
                            Id := daily_quest_id.WeaponEvolution.ToInt()
                            Tier := 3
                            Progress := 1.0
                        daily_quest:
                            Id := daily_quest_id.ItemShop.ToInt()
                            Tier := 3
                            Progress := 3.0

                Gears := map:
                    gear_slot.Head.ToInt() => player_gear:
                        Evolution := 3
                        Level := 35
                        Tier := 5
                        Stars := 4
                        Quantity := 8
                    gear_slot.Chest.ToInt() => player_gear:
                        Evolution := 2
                        Level := 32
                        Tier := 5
                        Stars := 3
                        Quantity := 5
                    gear_slot.Pants.ToInt() => player_gear:
                        Evolution := 2
                        Level := 33
                        Tier := 4
                        Stars := 4
                        Quantity := 15
                    gear_slot.Shoulder.ToInt() => player_gear:
                        Evolution := 2
                        Level := 32
                        Tier := 4
                        Stars := 3
                        Quantity := 4
                    gear_slot.Gloves.ToInt() => player_gear:
                        Evolution := 2
                        Level := 32
                        Tier := 3
                        Stars := 2
                        Quantity := 6
                    gear_slot.Boots.ToInt() => player_gear:
                        Evolution := 2
                        Level := 30
                        Tier := 3
                        Stars := 2
                        Quantity := 5

                Achievements := array:
                    1500.0, 500.0, 300.0, 10.0, 15.0, 5.0, 3.0, 7.0, 2.0,
                    1500.0, 500.0, 300.0, 10.0, 15.0, 5.0, 3.0, 7.0, 2.0,
                    1500.0, 500.0, 300.0, 10.0, 15.0, 5.0, 3.0, 7.0, 2.0,
                    1500.0, 500.0, 300.0, 0.0, 0.0, 0.0, 0.0, 0.0, 0.0,
                    0.0, 0.0, 0.0, 0.0, 0.0, 0.0, 0.0, 0.0, 0.0,
                    0.0, 0.0, 0.0, 0.0, 0.0, 0.0, 0.0, 0.0, 0.0,
                    0.0, 0.0, 0.0, 0.0, 0.0, 0.0, 0.0, 0.0, 0.0,
                    0.0, 0.0, 0.0, 0.0, 0.0, 0.0, 0.0, 0.0, 0.0

            player_data:
                Version := 0
                News := GetNewsInfos().Length
                Settings := player_settings:
                    InputMode := input_mode.Touch
                ArenaProgress := 5
                WeaponEquipped := weapon_id.DancingDagger.ToInt()
                WeaponsUnlocked := map:
                    weapon_id.DancingDagger.ToInt() => true
                DailyQuests := player_daily_quests:
                    LastQuestAssignedTime := GetSecondsSinceEpoch() + 62135596800.0 - 90000.0
                    ActiveQuests := array:
                        daily_quest:
                            Id := daily_quest_id.CompleteStage.ToInt()
                            Tier := 1
                            Progress := 0.0
                        daily_quest:
                            Id := daily_quest_id.GatherTree.ToInt()
                            Tier := 0
                            Progress := 0.0
                        daily_quest:
                            Id := daily_quest_id.WeaponEvolution.ToInt()
                            Tier := 0
                            Progress := 0.0
            player_data:
                Version := 0
                News := GetNewsInfos().Length
#>