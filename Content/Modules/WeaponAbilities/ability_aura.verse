using { /Fortnite.com/Characters }
using { /Fortnite.com/Devices }
using { /Verse.org/SceneGraph }
using { /Verse.org/Simulation }
using { /UnrealEngine.com/Temporary/SpatialMath }

<# ----------------------------------------------------------------------------------------------------
    Constants
---------------------------------------------------------------------------------------------------- #>
AbilityAuraBonusRadius:float = 25.0
AbilityAuraUpdateSizeTime:float = 0.5

<# ----------------------------------------------------------------------------------------------------
    ability_aura
---------------------------------------------------------------------------------------------------- #>
ability_aura<public> := class(ability):
    Id<override>:weapon_id = weapon_id.Aura

    var AttackRadiusMultiplier:float = 1.0
    var EvolutionDuration:float = 0.0
    var EvolutionPower:float = 0.0

    var CinematicSequence:cinematic_sequence_device = cinematic_sequence_device{}
    var IsBoost:logic = false
    var PreviousSize:float = 0.0
    var PauseTime:float = 0.0
    var CurrentVFXId:vfx_powerup_sequence_id = vfx_powerup_sequence_id.AbilityAuraEvolution0

    <# ----------------------------------------------------------------------------------------------------
        Initialization
    ---------------------------------------------------------------------------------------------------- #>
    Initialize<override>():void =
        (super:)Initialize()
        VFXPowerSequence := Player.ActivateVFXPowerup(vfx_powerup_sequence_id.AbilityAuraEvolution0)
        set CinematicSequence = VFXPowerSequence.CinematicSequence

    Deactivate<override>(?IsPause:logic = false):void =
        DeactivatedEvent.Signal()
        Player.DeactivateVFXPowerup(CurrentVFXId)
        CinematicSequence.Stop()
        
        set IsActive = false
        if (Remainder := FMod[GetSimulationElapsedTime() - StartTime, 10.0]):
            set PauseTime = Remainder * 1.0

    <# ----------------------------------------------------------------------------------------------------
        Get Functions
    ---------------------------------------------------------------------------------------------------- #>
    GetRadiusMultiplier()<transacts>:float =
        AttackRadiusMultiplier * (IsBoost? and 1.5 or 1.0)

    GetVFXId(Index:int):vfx_powerup_sequence_id =
        case (Index):
            0 => vfx_powerup_sequence_id.AbilityAuraEvolution0
            1 => vfx_powerup_sequence_id.AbilityAuraEvolution1
            2 => vfx_powerup_sequence_id.AbilityAuraEvolution2
            3 => vfx_powerup_sequence_id.AbilityAuraEvolution3
            _ => vfx_powerup_sequence_id.AbilityAuraEvolution0

    <# ----------------------------------------------------------------------------------------------------
        Update Functions
    ---------------------------------------------------------------------------------------------------- #>
    UpdateSize()<suspends>:void =
        TargetSize := GetRadiusMultiplier()
        DeltaSize := TargetSize - PreviousSize
        
        CinematicSequence.SetPlaybackTime(PreviousSize)
        CinematicSequence.SetPlayRate(Abs(DeltaSize) * (1.0 / AbilityAuraUpdateSizeTime))

        if (Sgn(DeltaSize) = 1.0):
            CinematicSequence.Play()
        else if (Sgn(DeltaSize) = -1.0):
            CinematicSequence.PlayReverse()
            
        Sleep(AbilityAuraUpdateSizeTime)
        CinematicSequence.SetPlaybackTime(TargetSize)
        CinematicSequence.Pause()
        set PreviousSize = TargetSize

    Ability<override>()<suspends>:void =
        set CurrentVFXId = GetVFXId(Player.GetWeaponEvolutionLevel(Id))
        Player.ActivateVFXPowerup(CurrentVFXId)

        race:
            DeactivatedEvent.Await()

            if (EvolutionDuration = 10.0):
                set IsBoost = true
                UpdateSize()
                Sleep(Inf)
            else:
                set IsBoost = false
                if (PauseTime < 10.0 - EvolutionDuration):
                    UpdateSize()

                Sleep(10.0 - EvolutionDuration - AbilityAuraUpdateSizeTime - PauseTime)
                set IsBoost = logic{Player.GetWeaponEvolutionLevel(Id) > 0}

                UpdateSize()
                Sleep(EvolutionDuration - (PauseTime > 10.0 - EvolutionDuration and PauseTime - (10.0 - EvolutionDuration) or 0.0) - AbilityAuraUpdateSizeTime)

                loop:
                    if (IsBoost?):
                        set IsBoost = false
                        UpdateSize()

                    Sleep(10.0 - EvolutionDuration - AbilityAuraUpdateSizeTime)

                    set IsBoost = logic{Player.GetWeaponEvolutionLevel(Id) > 0}
                    if (IsBoost?):
                        UpdateSize()

                    Sleep(EvolutionDuration - AbilityAuraUpdateSizeTime)

            if (Character := Player.GetFortCharacter[], SimulationEntity := GetSimulationEntity[]):
                loop:
                    Sleep(AttackCooldownTime * (IsBoost? and 0.75 or 1.0))
                    set CastTime = GetSimulationElapsedTime()

                    for:
                        OverlapHit : SimulationEntity.FindOverlapHits(FromTransform of Character.GetTransform(), collision_sphere{Radius := AttackRadius * GetRadiusMultiplier()})
                        TargetComponent := OverlapHit.TargetComponent
                        TargetEntity := TargetComponent.Entity
                        NPCCollisionComponent := TargetEntity.GetComponent[npc_collision_component]
                        NPCHit := NPCCollisionComponent.MaybeNPC?
                        NPCHit.IsDamageable()?
                    do:
                        if:
                            IsUniqueEvolution? and IsBoost?
                            NPCPosition := NPCHit.GetAgent[].GetFortCharacter[].GetTransform().Translation.Make2D()
                            Direction := (NPCPosition - Character.GetTransform().Translation.Make2D()).MakeUnitVector[]
                        then:
                            DirectionRotation := MakeRotation(vector3{X := 0.0, Y := 0.0, Z := 1.0}, ArcTan(Direction.Y, Direction.X))
                            NPCHit.Knockback(knockback_properties{Duration := 1.0, Rotation := DirectionRotation})

                        Damage(NPCHit, Player.GetDamageArgs(NPCHit, Id, AttackDamage * (IsBoost? and 2.0 + EvolutionPower or 1.0), ?Multiplier := AttackDamageMultiplier), ?IsKnockback := false)
        
    UpdateStats<override>():void =
        (super:)UpdateStats()
        if (WeaponInfo := GetWeaponInfo[Id]):
            set AttackRadiusMultiplier = 1.0 + Player.GetWeaponStats(Id, stat_id.Size) + Player.GetStat(stat_id.Size)
            set AttackRadius = WeaponInfo.Size

            set EvolutionDuration = Player.GetWeaponEvolutionValue(Id, ability_id.AuraDuration)
            set EvolutionPower = Player.GetWeaponEvolutionValue(Id, ability_id.AuraPower)