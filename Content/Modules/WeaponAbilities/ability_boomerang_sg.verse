using { /Fortnite.com/Characters }
using { /Fortnite.com/Devices }
using { /Verse.org/SceneGraph }
using { /Verse.org/SceneGraph/KeyframedMovement }
using { /Verse.org/Simulation }
using { /UnrealEngine.com/Temporary/SpatialMath }
using { Prefabs }
using { VisualEffects }

<# ----------------------------------------------------------------------------------------------------
    Constants
---------------------------------------------------------------------------------------------------- #>
DefaultAbilityBoomerangRange:float = 600.0
DefaultAbilityBoomerangLaunchDuration:float = 1.0
DefaultAbilityBoomerangPauseDuration:float = 0.5
DefaultAbilityBoomerangReturnDuration:float = 3.5
DefaultAbilityBoomerangTotalDuration:float = DefaultAbilityBoomerangLaunchDuration + DefaultAbilityBoomerangPauseDuration + DefaultAbilityBoomerangReturnDuration
DefaultAbilityBoomerangSpeed:float = 1000.0

<# ----------------------------------------------------------------------------------------------------
    ability_boomerang_sg
---------------------------------------------------------------------------------------------------- #>
ability_boomerang_sg<public> := class(ability):
    Id<override>:weapon_id = weapon_id.Boomerang

    var MulticastCount:int = 1
    var ProjectileRange:float = 0.0
    var WeaponEntities:[]entity = array{}

    <# ----------------------------------------------------------------------------------------------------
        Initialization
    ---------------------------------------------------------------------------------------------------- #>
    Deactivate<override>():void =
        DeactivatedEvent.Signal()
        set CooldownTimeRemaining = Max(0.0, AttackCooldownTime - (GetSimulationElapsedTime() - CastTime))
        set IsActive = false

    Deinitialize<override>():void =
        (super:)Deinitialize()
        
        for (WeaponEntity : WeaponEntities):
            WeaponEntity.RemoveFromParent()
            set WeaponEntities = array{}

    <# ----------------------------------------------------------------------------------------------------
        Get Functions
    ---------------------------------------------------------------------------------------------------- #>
    GetAvailableWeaponEntity()<suspends>:entity =
        for (WeaponEntity : WeaponEntities, not WeaponEntity.GetComponent[particle_system_component].IsEnabled[]):
            return WeaponEntity

        NewWeaponEntity := prefab_weapon_boomerang{}
        NewWeaponEntity.InitializeParticle()
        set WeaponEntities += array{NewWeaponEntity}
        NewWeaponEntity   

    <# ----------------------------------------------------------------------------------------------------
        Update Functions
    ---------------------------------------------------------------------------------------------------- #>
    Ability<override>()<suspends>:void =
        race:
            DeactivatedEvent.Await()
            block:
                Sleep(CooldownTimeRemaining)
                loop:
                    set CastTime = GetSimulationElapsedTime()
                    spawn{Cast()}
                    Sleep(AttackCooldownTime)

    Cast()<suspends>:void =
        for (ProjectileIndex := 0..ProjectileCount - 1):
            spawn{Projectile(ProjectileIndex)}

        for (CastIndex := 1..MulticastCount):
            Player.PlayAudio(audio_id.Projectile)
            Sleep(0.1)

    Projectile(ProjectileIndex:int)<suspends>:void =
        WeaponEntity := GetAvailableWeaponEntity()
        BoomerangStartTime := GetSimulationElapsedTime()
        Collision:collision_sphere = collision_sphere{Radius := AttackRadius}

        var NPCsHit:[npc_behavior_base]logic = map{}
        
        if:
            Character := Player.GetFortCharacter[]
            ParticleSystem := NS_AbilityBoomerang_v2[WeaponEntity.GetComponent[particle_system_component]]
        then:
            CastPosition := Character.GetTransform().Translation
            Angle := DegreesToRadians(30.0) * ProjectileIndex - (Floor[ProjectileCount * 1.0 / 2.0] or 0) * 1.0 * DegreesToRadians(30.0)
            TargetRotation := Player.GetAutoAimRotation().ApplyYaw(Angle)
            TargetPosition := CastPosition + TargetRotation.GetLocalForward() * DefaultAbilityBoomerangRange * (1.0 + ProjectileRange)
            TargetScaleMultiplier := IsUniqueEvolution? and (1.5 + ProjectileRange) or 1.0

            var CurrentPosition:vector3 = CastPosition
            var CurrentScaleMultiplier:float = 1.0
            var FinalDamage:float = AttackDamage

            set ParticleSystem.Color = GetWeaponColor(Player.GetWeaponEvolutionLevel(Id))
            set ParticleSystem.ProjectileMulticast = MulticastCount
            set ParticleSystem.Radius = AttackRadius * 0.1
            set ParticleSystem.IsReturned = false
            set ParticleSystem.ScaleMultiplier = TargetScaleMultiplier
                
            ParticleSystem.Enable()

            race:
                block:
                    block:
                        loop:
                            Progress := (GetSimulationElapsedTime() - BoomerangStartTime) / DefaultAbilityBoomerangLaunchDuration
                            EasedProgress := 1.0 - Pow(1.0 - Progress, 5.0)
                            set CurrentPosition = Lerp(CastPosition, TargetPosition, EasedProgress)
                            set CurrentScaleMultiplier = Lerp(1.0, TargetScaleMultiplier, EasedProgress)
                            set ParticleSystem.Position = FromVector3 of CurrentPosition

                            if (EasedProgress >= 1.0):
                                set CurrentPosition = TargetPosition
                                set CurrentScaleMultiplier = TargetScaleMultiplier
                                set ParticleSystem.Position = FromVector3 of CurrentPosition
                                break

                            Sleep(TwoTicks)
                   
                    Sleep(DefaultAbilityBoomerangPauseDuration)
                    set NPCsHit = map{}
                    if (IsUniqueEvolution?):
                        set FinalDamage *= 1.5 + ProjectileRange

                    loop:
                        if (GetSimulationElapsedTime() - BoomerangStartTime >= DefaultAbilityBoomerangTotalDuration):
                            break
                        else:
                            DirectionVector := Character.GetTransform().Translation - CurrentPosition
                            
                            if (Direction := DirectionVector.MakeUnitVector[]):
                                set CurrentPosition = CurrentPosition + Direction * DefaultAbilityBoomerangSpeed * TwoTicks

                            set ParticleSystem.Position = FromVector3 of CurrentPosition

                            if (DirectionVector.LengthSquared() <= 2500.0):
                                break

                            Sleep(TwoTicks)

                    set ParticleSystem.IsReturned = true
                    Sleep(0.2)
                
                loop:
                    set Collision.Radius = AttackRadius * CurrentScaleMultiplier

                    for:
                        OverlapHit : WeaponEntity.FindOverlapHits(FromTransform of transform{Translation := CurrentPosition}, Collision)
                        TargetComponent := OverlapHit.TargetComponent
                        TargetEntity := TargetComponent.Entity
                        NPCCollisionComponent := TargetEntity.GetComponent[npc_collision_component]
                        HitNPC := NPCCollisionComponent.MaybeNPC?
                        HitNPC.IsDamageable()?
                        not NPCsHit[HitNPC]
                        set NPCsHit[HitNPC] = true
                    do:
                        Damage(HitNPC, Player.GetDamageArgs(HitNPC, Id, FinalDamage, ?Multiplier := AttackDamageMultiplier))

                    Sleep(ThreeTicks)

            ParticleSystem.Disable()

    UpdateStats<override>():void =
        (super:)UpdateStats()
        if (WeaponInfo := GetWeaponInfo[Id]):
            set MulticastCount = 1 + Floor[Player.GetWeaponEvolutionValue(Id, ability_id.Multicast)] or 0
            set ProjectileCount = 1 + Floor[Player.GetWeaponEvolutionValue(Id, ability_id.ProjectileCount)] or 0
            set ProjectileRange = Player.GetWeaponEvolutionValue(Id, ability_id.ProjectileRange)
            set AttackDamage *= (MulticastCount * 1.0)