using { /Fortnite.com/Characters }
using { /Verse.org/SceneGraph }
using { /Verse.org/Simulation }
using { /UnrealEngine.com/Temporary/SpatialMath }
using { /UnrealEngine.com/Temporary }
using { Prefabs }
using { VisualEffects }

<# ----------------------------------------------------------------------------------------------------
    Constants
---------------------------------------------------------------------------------------------------- #>
DefaultAbilityProjectileDamage:float = 10.0
DefaultAbilityProjectileDistance:float = 1000.0
DefaultAbilityProjectileSpeed:float = 1500.0
DefaultAbilityBowProjectileCount:int = 3
DefaultAbilityBowAngle:float = 15.0

<# ----------------------------------------------------------------------------------------------------
    Data Definitions
---------------------------------------------------------------------------------------------------- #>
projectile_properties<public> := struct:
    Owner<public>:agent
    Position<public>:vector3
    Rotation<public>:rotation
    Angle<public>:float = 0.0
    Speed<public>:float
    (projectile_properties:)Distance<public>:float
    Damage<public>:float
    DamageMultiplier<public>:float = 1.0
    Radius<public>:float
    VFX<public>:vfx_id
    Piercing<public>:int = 0
    Multicast<public>:int = 1
    Evolution<public>:int = 0

projectile_properties_sg<public> := struct:
    Owner<public>:agent
    Index<public>:int
    Position<public>:vector3
    Rotation<public>:rotation
    Angle<public>:float = 0.0
    (projectile_properties_sg:)Distance<public>:float
    Duration<public>:float
    Speed<public>:float
    Damage<public>:float
    DamageMultiplier<public>:float = 1.0
    Radius<public>:float
    Piercing<public>:int = 0
    MaybeParticleSystem<public>:?particle_system_component = false
        
<# ----------------------------------------------------------------------------------------------------
    ability_bow_sg
---------------------------------------------------------------------------------------------------- #>
ability_bow_sg<public> := class(ability):
    Id<override>:weapon_id = weapon_id.Bow

    var ProjectilePiercing:int = 0
    var MulticastCount:int = 1
    var ObjectPool:[]projectile_sg = array{}
    var WeaponEntities:[]entity = array{}

    <# ----------------------------------------------------------------------------------------------------
        Initialization
    ---------------------------------------------------------------------------------------------------- #>
    Deactivate<override>():void =
        DeactivatedEvent.Signal()
        set CooldownTimeRemaining = Max(0.0, AttackCooldownTime - (GetSimulationElapsedTime() - CastTime))
        set IsActive = false

    Deinitialize<override>():void =
        (super:)Deinitialize()
        
        for (Object : ObjectPool):
            Object.Deactivate(?IsDispose := true)

        for (WeaponEntity : WeaponEntities):
            WeaponEntity.RemoveFromParent()

        set ObjectPool = array{}
        set WeaponEntities = array{}
        
    <# ----------------------------------------------------------------------------------------------------
        Get Functions
    ---------------------------------------------------------------------------------------------------- #>
    GetAvailablePoolObject<private>()<transacts>:projectile_sg =
        for (Object : ObjectPool, Object.IsAvailable?):
            return Object

        NewObject := projectile_sg{Ability := Self}
        set ObjectPool += array{NewObject}
        NewObject

    GetAvailableWeaponEntity()<suspends>:entity =
        for (WeaponEntity : WeaponEntities, not WeaponEntity.GetComponent[particle_system_component].IsEnabled[]):
            return WeaponEntity

        NewWeaponEntity := prefab_weapon_bow_v2{}
        NewWeaponEntity.InitializeParticle()
        set WeaponEntities += array{NewWeaponEntity}
        NewWeaponEntity

    <# ----------------------------------------------------------------------------------------------------
        Update Functions
    ---------------------------------------------------------------------------------------------------- #>
    Ability<override>()<suspends>:void =
        race:
            DeactivatedEvent.Await()
            block:
                Sleep(CooldownTimeRemaining)
                loop:
                    set CastTime = GetSimulationElapsedTime()
                    spawn{Cast()}
                    Sleep(AttackCooldownTime)

    Cast()<suspends>:void =
        WeaponEntity := GetAvailableWeaponEntity()

        if:
            Character := Player.GetFortCharacter[]
            ParticleSystem := NS_AbilityBow_v2[WeaponEntity.GetComponent[particle_system_component]]
        then:
            Rotation := Player.GetAutoAimRotation()
            Duration := DefaultAbilityProjectileDistance / DefaultAbilityProjectileSpeed

            set ParticleSystem.Color = GetWeaponColor(Player.GetWeaponEvolutionLevel(Id))
            set ParticleSystem.Multicast = MulticastCount
            set ParticleSystem.ProjectileCount = ProjectileCount
            set ParticleSystem.ProjectileSpeed = DefaultAbilityProjectileSpeed
            set ParticleSystem.ProjectileStates = for (Index := 1..ProjectileCount) {0}
            set ParticleSystem.Angle = DefaultAbilityBowAngle
            set ParticleSystem.Radius = AttackRadius * 0.1
            set ParticleSystem.DeathStateDuration = 0.2
            set ParticleSystem.Position = FromVector3 of Character.GetTransform().Translation
            set ParticleSystem.Rotation = Rotation.GetYawPitchRollDegrees()[0] or 0.0

            ParticleSystem.Enable()

            for (Index := 0..ProjectileCount - 1):
                Angle := DegreesToRadians(DefaultAbilityBowAngle) * (Index - (Floor[ProjectileCount * 1.0 / 2.0] or 0))
            
                ProjectileProperties := projectile_properties_sg:
                    Owner := Player
                    Index := Index
                    Position := Character.GetTransform().Translation
                    Rotation := Rotation
                    Angle := Angle
                    Distance := DefaultAbilityProjectileDistance
                    Duration := Duration
                    Speed := DefaultAbilityProjectileSpeed
                    Damage := AttackDamage
                    DamageMultiplier := AttackDamageMultiplier
                    Radius := AttackRadius
                    Piercing := ProjectilePiercing
                    MaybeParticleSystem := option{ParticleSystem}

                Projectile := GetAvailablePoolObject()
                spawn{Projectile.Activate(ProjectileProperties)}

            branch:
                for (CastIndex := 1..MulticastCount):
                    Player.PlayAudio(IsUniqueEvolution? and audio_id.ProjectileEvolved or audio_id.Projectile)
                    Sleep(0.1)

            Sleep(Duration + 0.2)
            ParticleSystem.Disable()

    UpdateStats<override>():void =
        (super:)UpdateStats()
        if (WeaponInfo := GetWeaponInfo[Id]):
            TargetProjectileCount := DefaultAbilityBowProjectileCount + Floor[Player.GetWeaponEvolutionValue(Id, ability_id.ProjectileCount)] or 0
            set ProjectileCount = IsUniqueEvolution? and 1 or TargetProjectileCount
            set AttackRadius *= (IsUniqueEvolution? and (1.0 + TargetProjectileCount * 0.5) or 1.0)
            set MulticastCount = 1 + Floor[Player.GetWeaponEvolutionValue(Id, ability_id.Multicast)] or 0
            set ProjectilePiercing = IsUniqueEvolution? and -1 or (1 + (Floor[Player.GetWeaponEvolutionValue(Id, ability_id.ProjectilePiercing)] or 1)) * MulticastCount
           
            set AttackDamage *= (IsUniqueEvolution? and (1.0 + TargetProjectileCount * 0.2) or 1.0) * MulticastCount

<# ----------------------------------------------------------------------------------------------------
    projectile_sg
---------------------------------------------------------------------------------------------------- #>
projectile_sg<public> := class(poolable):
    Ability<public>:ability
    DeactivatedEvent<public>:event(void) = event(void){}

    var NPCsHitCount:int = 0
    
    <# ----------------------------------------------------------------------------------------------------
        Update Functions
    ---------------------------------------------------------------------------------------------------- #>
    Activate<public>(Properties:projectile_properties_sg)<suspends>:void =
        (poolable:)Activate()

        if:
            Player := player[Properties.Owner]
            SimulationEntity := GetSimulationEntity[]
        then:
            set NPCsHitCount = 0

            TargetRotation := Properties.Rotation.ApplyYaw(Properties.Angle)
            DeltaTranslation := TargetRotation.GetLocalForward() * Properties.Distance
            TargetPosition := Properties.Position + DeltaTranslation
            Collision := collision_sphere{Radius := Properties.Radius}

            race:
                DeactivatedEvent.Await()
                block:
                    var TimeTraveled:float = 0.0

                    for:
                        OverlapHit : SimulationEntity.FindSweepHits(FromVector3(DeltaTranslation), FromTransform of transform{Translation := Properties.Position}, Collision)
                        NPCsHitCount < Properties.Piercing or Properties.Piercing < 0
                        TargetComponent := OverlapHit.TargetComponent
                        TargetEntity := TargetComponent.Entity
                        NPCCollisionComponent := TargetEntity.GetComponent[npc_collision_component]
                        NPCHit := NPCCollisionComponent.MaybeNPC?
                        NPCHit.IsDamageable()?
                    do:
                        TravelTime := OverlapHit.SourceHitDistance / Properties.Speed - TimeTraveled
                        Sleep(TravelTime)
                        Ability.Damage(NPCHit, Player.GetDamageArgs(NPCHit, Ability.Id, Properties.Damage, ?Multiplier := Properties.DamageMultiplier))
                        set TimeTraveled += TravelTime
                        set NPCsHitCount += 1

                    if (NPCsHitCount < Properties.Piercing or Properties.Piercing < 0):
                        Sleep(Properties.Duration - TimeTraveled)

            ScheduleDeactivate(?MaybeProperties := option{Properties})

    ScheduleDeactivate<override>(?MaybeProperties:?projectile_properties_sg = false)<suspends>:void =
        if:
            Properties := MaybeProperties?
            ParticleSystem := NS_AbilityBow_v2[Properties.MaybeParticleSystem?]
            set ParticleSystem.ProjectileStates[Properties.Index] = 1
        then:
            Sleep(0.2)

        Deactivate()

    Deactivate<override>(?IsDispose:logic = false):void =
        DeactivatedEvent.Signal()
        (super:)Deactivate()