using { /Fortnite.com/Characters }
using { /Fortnite.com/Devices }
using { /Verse.org/Random }
using { /Verse.org/SceneGraph }
using { /Verse.org/SceneGraph/KeyframedMovement }
using { /Verse.org/Simulation }
using { /UnrealEngine.com/Temporary/SpatialMath }
using { Prefabs }
using { VisualEffects }

<# ----------------------------------------------------------------------------------------------------
    Constants
---------------------------------------------------------------------------------------------------- #>
DefaultAbilityChainLightningBounceDuration:float = 0.15
DefaultAbilityChainLightningRange:float = 600.0
DefaultAbilityChainLightningRangeSquared:float = DefaultAbilityChainLightningRange * DefaultAbilityChainLightningRange

<# ----------------------------------------------------------------------------------------------------
    ability_chain_lightning_sg
---------------------------------------------------------------------------------------------------- #>
ability_chain_lightning_sg<public> := class(ability):
    Id<override>:weapon_id = weapon_id.ChainLightning

    var AttackDamageBounceFalloff:float = 0.3
    var TotalBounceCount:int = 3
    var WeaponEntities:[]entity = array{}

    <# ----------------------------------------------------------------------------------------------------
        Initialization
    ---------------------------------------------------------------------------------------------------- #>
    Deactivate<override>():void =
        DeactivatedEvent.Signal()
        set CooldownTimeRemaining = Max(0.0, AttackCooldownTime - (GetSimulationElapsedTime() - CastTime))
        set IsActive = false

    Deinitialize<override>():void =
        (super:)Deinitialize()
        
        for (WeaponEntity : WeaponEntities):
            WeaponEntity.RemoveFromParent()

        set WeaponEntities = array{}

    <# ----------------------------------------------------------------------------------------------------
        Get Functions
    ---------------------------------------------------------------------------------------------------- #>
    GetAvailableWeaponEntity()<suspends>:entity =
        for (WeaponEntity : WeaponEntities, not WeaponEntity.GetComponent[particle_system_component].IsEnabled[]):
            return WeaponEntity

        NewWeaponEntity := prefab_weapon_chain_lightning{}
        NewWeaponEntity.InitializeParticle()
        set WeaponEntities += array{NewWeaponEntity}
        NewWeaponEntity

    <# ----------------------------------------------------------------------------------------------------
        Update Functions
    ---------------------------------------------------------------------------------------------------- #>
    Ability<override>()<suspends>:void =
        race:
            DeactivatedEvent.Await()
            if (WeaponInfo := GetWeaponInfo[Id]):
                Sleep(CooldownTimeRemaining)
                loop:
                    set CastTime = GetSimulationElapsedTime()
                    spawn{Cast()}
                    Sleep(AttackCooldownTime)
                
    Cast()<suspends>:void =
        WeaponEntity := GetAvailableWeaponEntity()
        if:
            PlayerCharacter := Player.GetFortCharacter[]
            ParticleSystem := NS_AbilityChainLightning_v2[WeaponEntity.GetComponent[particle_system_component]]
            KeyframedMovementComponent := WeaponEntity.GetComponent[keyframed_movement_component]
            SimulationEntity := GetSimulationEntity[]
        then:
            CollisionRange:collision_sphere = collision_sphere{Radius := DefaultAbilityChainLightningRange}
            CollisionAttack:collision_sphere = collision_sphere{Radius := AttackRadius}

            var BounceCount:int = 0
            var MaybePreviousNPC:?npc_behavior_base = false
            var CurrentPosition:vector3 = PlayerCharacter.GetTransform().Translation

            set ParticleSystem.Color = GetWeaponColor(Player.GetWeaponEvolutionLevel(Id))
            set ParticleSystem.Radius = AttackRadius * 0.1
            set ParticleSystem.BounceCount = TotalBounceCount
            set ParticleSystem.BounceDuration = DefaultAbilityChainLightningBounceDuration
            
            WeaponEntity.SetGlobalTransform of FromTransform of transform:
                Translation := CurrentPosition.Make2D()

            Sleep(0.0)
            ParticleSystem.Enable()

            loop:
                if (BounceCount >= TotalBounceCount):
                    break

                var MaybeNearestNPC:?npc_behavior_base = false
                var NearestDistanceSquared:float = 0.0

                for:
                    OverlapHit : SimulationEntity.FindOverlapHits(FromTransform of transform{Translation := CurrentPosition}, CollisionRange)
                    TargetComponent := OverlapHit.TargetComponent
                    TargetEntity := TargetComponent.Entity
                    NPCCollisionComponent := TargetEntity.GetComponent[npc_collision_component]
                    NPCHit := NPCCollisionComponent.MaybeNPC?
                    NPCHit.IsDamageable()?
                    IsUniqueEvolution? or (not MaybePreviousNPC? or NPCHit <> MaybePreviousNPC?)
                    DistanceSquaredFromAbility := DistanceSquaredXY(NPCHit.GetAgent[].GetFortCharacter[].GetTransform().Translation, CurrentPosition)
                    DistanceSquaredFromAbility <= DefaultAbilityChainLightningRangeSquared
                do:
                    set MaybeNearestNPC = option{NPCHit}
                    set NearestDistanceSquared = DistanceSquaredFromAbility

                if (NearestNPCCharacter := MaybeNearestNPC?.GetAgent[].GetFortCharacter[]):
                    if (BounceCount = 0):
                        Player.PlayAudio(audio_id.Projectile)

                    TargetPosition := NearestNPCCharacter.GetTransform().Translation
                    RelativeKeyframes := array:
                        keyframed_movement_delta:
                            Transform := FromTransform of transform:
                                Translation := TargetPosition.Make2D() - FromVector3(WeaponEntity.GetGlobalTransform().Translation)
                                Scale := vector3{}
                            Duration := DefaultAbilityChainLightningBounceDuration
                            Easing := linear_easing_function{}

                    KeyframedMovementComponent.SetKeyframes(RelativeKeyframes, oneshot_keyframed_movement_playback_mode{})
                    KeyframedMovementComponent.Play()
                    KeyframedMovementComponent.FinishedEvent.Await()
                    set CurrentPosition = TargetPosition

                    for:
                        OverlapHit : SimulationEntity.FindOverlapHits(FromTransform of transform{Translation := CurrentPosition}, CollisionAttack)
                        TargetComponent := OverlapHit.TargetComponent
                        TargetEntity := TargetComponent.Entity
                        NPCCollisionComponent := TargetEntity.GetComponent[npc_collision_component]
                        NPCHit := NPCCollisionComponent.MaybeNPC?
                        NPCHit.IsDamageable()?
                    do:
                        Damage(NPCHit, Player.GetDamageArgs(NPCHit, Id, AttackDamage * Pow(AttackDamageBounceFalloff, BounceCount * 1.0), ?Multiplier := AttackDamageMultiplier))

                    set BounceCount += 1
                    set MaybePreviousNPC = MaybeNearestNPC
                else:
                    break

            if (BounceCount > 0):
                Sleep(1.0)

            ParticleSystem.Disable()

    UpdateStats<override>():void =
        (super:)UpdateStats()
        if (WeaponInfo := GetWeaponInfo[Id]):
            set TotalBounceCount = Floor[Player.GetWeaponEvolutionValue(Id, ability_id.ChainLightningBounce)] or 0
            set AttackDamageBounceFalloff = Player.GetWeaponEvolutionValue(Id, ability_id.ChainLightningResilience)