using { /Fortnite.com/Characters }
using { /Verse.org/Random }
using { /Verse.org/SceneGraph }
using { /Verse.org/SceneGraph/KeyframedMovement }
using { /Verse.org/Simulation }
using { /UnrealEngine.com/Temporary }
using { /UnrealEngine.com/Temporary/SpatialMath }
using { Prefabs }
using { VisualEffects }

<# ----------------------------------------------------------------------------------------------------
    Constants
---------------------------------------------------------------------------------------------------- #>
DefaultAbilityDancingDaggerProjectileCount:int = 2
DefaultAbilityDancingDaggerDuration:float = 5.0
DefaultAbilityDancingDaggerSpeed:float = 1000.0
DefaultAbilityDancingDaggerRange:float = 1000.0

<# ----------------------------------------------------------------------------------------------------
    ability_dancing_dagger_sg
---------------------------------------------------------------------------------------------------- #>
ability_dancing_dagger_sg<public> := class(ability):
    Id<override>:weapon_id = weapon_id.DancingDagger

    var MulticastCount:int = 1
    var ProjectileSpeed:float = DefaultAbilityDancingDaggerSpeed
    var IsInfinite:logic = false
    var WeaponEntities:[]entity = array{}

    <# ----------------------------------------------------------------------------------------------------
        Initialization
    ---------------------------------------------------------------------------------------------------- #>
    Deactivate<override>():void =
        DeactivatedEvent.Signal()
        set CooldownTimeRemaining = Max(0.0, AttackCooldownTime - (GetSimulationElapsedTime() - CastTime))
        set IsActive = false

    Deinitialize<override>():void =
        (super:)Deinitialize()

        for (WeaponEntity : WeaponEntities):
            WeaponEntity.RemoveFromParent()

        set WeaponEntities = array{}

    <# ----------------------------------------------------------------------------------------------------
        Get Functions
    ---------------------------------------------------------------------------------------------------- #>
    GetAvailableWeaponEntity()<suspends>:entity =
        for (WeaponEntity : WeaponEntities, not WeaponEntity.GetComponent[particle_system_component].IsEnabled[]):
            return WeaponEntity

        NewWeaponEntity := prefab_weapon_dancing_dagger{}
        NewWeaponEntity.InitializeParticle()
        set WeaponEntities += array{NewWeaponEntity}
        NewWeaponEntity

    <# ----------------------------------------------------------------------------------------------------
        Update Functions
    ---------------------------------------------------------------------------------------------------- #>
    Ability<override>()<suspends>:void =
        race:
            DeactivatedEvent.Await()
            if (WeaponInfo := GetWeaponInfo[Id]):
                if (IsInfinite?):
                    Cast()
                    Sleep(Inf)
                else:
                    Sleep(CooldownTimeRemaining)
                    loop:
                        set CastTime = GetSimulationElapsedTime()
                        spawn{Cast()}
                        Sleep(AttackCooldownTime)

    Cast()<suspends>:void =
        for (Index := 1..ProjectileCount):
            spawn{Projectile(Index - 1)}
            Player.PlayAudio(audio_id.Projectile)
            Sleep(0.1)

    Projectile(Index:int)<suspends>:void =
        var BounceCount:int = 0
        WeaponEntity := GetAvailableWeaponEntity()

        Collision:collision_sphere = collision_sphere:
            Radius := AttackRadius

        if:
            Character := Player.GetFortCharacter[]
            ParticleSystem := NS_AbilityDancingDagger_v2[WeaponEntity.GetComponent[particle_system_component]]
            KeyframedMovementComponent := WeaponEntity.GetComponent[keyframed_movement_component]
            SimulationEntity := GetSimulationEntity[]
        then:
            var CurrentPosition:vector3 = Character.GetTransform().Translation
            var TargetPosition:vector3 = GetRandomPositionInCircle(DefaultAbilityDancingDaggerRange, ?Origin := CurrentPosition)
            var TargetRotation:rotation = MakeRotation(CurrentPosition, TargetPosition)

            set ParticleSystem.Color = GetWeaponColor(Player.GetWeaponEvolutionLevel(Id))
            set ParticleSystem.ProjectileMulticast = MulticastCount
            set ParticleSystem.Radius = AttackRadius * 0.1

            WeaponEntity.SetGlobalTransform of FromTransform of transform:
                Translation := CurrentPosition
                Rotation := TargetRotation

            ParticleSystem.Enable()

            race:
                if (IsInfinite?):
                    DeactivatedEvent.Await()
                else:
                    Sleep(DefaultAbilityDancingDaggerDuration)

                loop:
                    TargetDistance := Distance(CurrentPosition, TargetPosition)
                    sync:
                        block:
                            OverTime := TargetDistance / ProjectileSpeed
                            RelativeKeyframes := array:
                                keyframed_movement_delta:
                                    Transform := FromTransform of transform:
                                        Rotation := MakeShortestRotationBetween(FromRotation of WeaponEntity.GetGlobalTransform().Rotation, TargetRotation)
                                        Scale := vector3{}
                                    Duration := 0.1
                                    Easing := linear_easing_function{}
                                keyframed_movement_delta:
                                    Transform := FromTransform of transform:
                                        Translation :=  TargetPosition - FromVector3(WeaponEntity.GetGlobalTransform().Translation)
                                        Scale := vector3{}
                                    Duration := OverTime
                                    Easing := linear_easing_function{}

                            KeyframedMovementComponent.SetKeyframes(RelativeKeyframes, oneshot_keyframed_movement_playback_mode{})
                            KeyframedMovementComponent.Play()
                            FinishedEvent:listenable(tuple()) = KeyframedMovementComponent.FinishedEvent
                            FinishedEvent.Await()

                        block:
                            Displacement := FromVector3(TargetPosition - CurrentPosition)
                            
                            var TimeTraveled:float = 0.0

                            for:
                                OverlapHit : SimulationEntity.FindSweepHits(Displacement, WeaponEntity.GetGlobalTransform(), Collision)
                                TargetComponent := OverlapHit.TargetComponent
                                TargetEntity := TargetComponent.Entity
                                NPCCollisionComponent := TargetEntity.GetComponent[npc_collision_component]
                                NPCHit := NPCCollisionComponent.MaybeNPC?
                                NPCHit.IsDamageable()?
                            do:
                                TravelTime := OverlapHit.SourceHitDistance / ProjectileSpeed - TimeTraveled
                                Sleep(TravelTime)
                                set TimeTraveled += TravelTime
                                Damage(NPCHit, Player.GetDamageArgs(NPCHit, Id, AttackDamage, ?Multiplier := AttackDamageMultiplier))

                    set CurrentPosition = FromVector3(WeaponEntity.GetGlobalTransform().Translation)
                    
                    if:
                        IsUniqueEvolution? and Mod[BounceCount, 2] = 0
                        var IsBossInRange:logic = false
                        for:
                            NPCBehaviorBase : GetAllNPCBosses()
                            not IsBossInRange?
                            NPCCharacter := NPCBehaviorBase.GetAgent[].GetFortCharacter[]
                            AttackRange := DefaultAbilityDancingDaggerRange + NPCBehaviorBase.GetSize()
                            DistanceSquaredXY(NPCCharacter.GetTransform().Translation, Character.GetTransform().Translation) <= AttackRange * AttackRange
                        do:
                            set TargetPosition = NPCCharacter.GetTransform().Translation
                            set IsBossInRange = true

                        IsBossInRange?
                    else:
                        set TargetPosition = GetRandomPositionInCircle(DefaultAbilityDancingDaggerRange, ?Origin := Character.GetTransform().Translation)

                    set TargetRotation = MakeRotation(CurrentPosition, TargetPosition)
                    set BounceCount += 1

            ParticleSystem.Disable()

    LessNPCMaxHealth(NPCMaxHealthA:tuple(npc_behavior_base, float), NPCMaxHealthB:tuple(npc_behavior_base, float))<computes><decides>:void =
        NPCMaxHealthA(1) > NPCMaxHealthB(1)

    UpdateStats<override>():void =
        (super:)UpdateStats()
        if (WeaponInfo := GetWeaponInfo[Id]):
            set ProjectileCount = DefaultAbilityDancingDaggerProjectileCount + Floor[Player.GetWeaponEvolutionValue(Id, ability_id.ProjectileCount)] or 0
            set MulticastCount = 1 + Floor[Player.GetWeaponEvolutionValue(Id, ability_id.Multicast)] or 0
            set ProjectileSpeed = DefaultAbilityDancingDaggerSpeed * (1.0 + Player.GetWeaponEvolutionValue(Id, ability_id.ProjectileSpeed))

            set IsInfinite = logic{AttackCooldownTime <= WeaponInfo.MinCooldown}
            set AttackDamage *= (MulticastCount * 1.0)
            set AttackRadius *= (1.0 + (MulticastCount - 1) * 0.25)