using { /Fortnite.com/Characters }
using { /Verse.org }
using { /Verse.org/SceneGraph }
using { /Verse.org/Simulation }
using { /UnrealEngine.com/Temporary/SpatialMath }
using { Prefabs }
using { VisualEffects }

<# ----------------------------------------------------------------------------------------------------
    Constants
---------------------------------------------------------------------------------------------------- #>
DefaultAbilityEnergyBeamRange:float = 600.0
DefaultAbilityEnergyBeamAngle:float = 30.0

<# ----------------------------------------------------------------------------------------------------
    ability_energy_beam_sg
---------------------------------------------------------------------------------------------------- #>
ability_energy_beam_sg<public> := class(ability):
    Id<override>:weapon_id = weapon_id.EnergyBeam

    var MulticastCount:int = 1
    var ProjectileRange:float = DefaultAbilityEnergyBeamRange
    var WeaponPrefabs:[]prefab_weapon_energy_beam = array{}
    var WeaponEntities:[]entity = array{}

    <# ----------------------------------------------------------------------------------------------------
        Initialization
    ---------------------------------------------------------------------------------------------------- #>
    Deactivate<override>():void =
        DeactivatedEvent.Signal()
        set CooldownTimeRemaining = Max(0.0, AttackCooldownTime - (GetSimulationElapsedTime() - CastTime))
        set IsActive = false

    Deinitialize<override>():void =
        (super:)Deinitialize()
        
        for (WeaponEntity : WeaponEntities):
            WeaponEntity.RemoveFromParent()

        set WeaponEntities = array{}

    <# ----------------------------------------------------------------------------------------------------
        Get Functions
    ---------------------------------------------------------------------------------------------------- #>
    GetAvailableWeaponEntity()<suspends>:entity =
        for (WeaponEntity : WeaponEntities, not WeaponEntity.GetComponent[particle_system_component].IsEnabled[]):
            return WeaponEntity

        NewWeaponEntity := prefab_weapon_energy_beam{}
        NewWeaponEntity.InitializeParticle()
        set WeaponEntities += array{NewWeaponEntity}
        NewWeaponEntity

    <# ----------------------------------------------------------------------------------------------------
        Get Functions
    ---------------------------------------------------------------------------------------------------- #>
    IsAtPeak(Position:vector3, PeakPositions:[]vector3)<transacts>:logic =
        var IsPeak:logic = false
        AttackRange := AttackRadius * 1.5
        for (PeakPosition : PeakPositions, not IsPeak?, DistanceSquaredXY(Position, PeakPosition) <= AttackRange * AttackRange):
            set IsPeak = true
        IsPeak

    <# ----------------------------------------------------------------------------------------------------
        Update Functions
    ---------------------------------------------------------------------------------------------------- #>
    Ability<override>()<suspends>:void =
        race:
            DeactivatedEvent.Await()
            if (WeaponInfo := GetWeaponInfo[Id]):
                Sleep(CooldownTimeRemaining)
                loop:
                    set CastTime = GetSimulationElapsedTime()
                    spawn{Cast(WeaponInfo)}
                    Sleep(AttackCooldownTime)

    Cast(WeaponInfo:weapon_info)<suspends>:void =
        spawn{Projectile(WeaponInfo)}

        for (CastIndex := 1..MulticastCount):
            Player.PlayAudio(audio_id.Projectile)
            Sleep(0.1)
        
    Projectile(WeaponInfo:weapon_info)<suspends>:void =
        WeaponEntity := GetAvailableWeaponEntity()
        if:
            Character := Player.GetFortCharacter[]
            ParticleSystem := NS_AbilityEnergyBeam_v2[WeaponEntity.GetComponent[particle_system_component]]
            SimulationEntity := GetSimulationEntity[]
        then:
            CastPosition := Character.GetTransform().Translation
            CastRotation := Player.GetAutoAimRotation()
            AngleOffset := (ProjectileCount - 1) * DefaultAbilityEnergyBeamAngle / 2.0
            Collision := collision_box:
                Extents := SpatialMath.vector3{Left := -AttackRadius / 2.0, Up := 100.0, Forward := ProjectileRange / 2.0}

            set ParticleSystem.Color = GetWeaponColor(Player.GetWeaponEvolutionLevel(Id))
            set ParticleSystem.IsBurst = IsUniqueEvolution
            set ParticleSystem.Angle = DefaultAbilityEnergyBeamAngle
            set ParticleSystem.Radius = AttackRadius * 0.1
            set ParticleSystem.Range = ProjectileRange * 0.1
            set ParticleSystem.ProjectileCount = ProjectileCount
            set ParticleSystem.ProjectileMulticast = MulticastCount
            set ParticleSystem.Position = FromVector3 of CastPosition.Make2D()
            set ParticleSystem.Rotation = CastRotation.GetYawPitchRollDegrees()[0] or 0.0

            ParticleSystem.Enable()

            Sleep(0.5)
            Player.PlayAudio(audio_id.EnergyBeam)

            PeakPositions := for (ProjectileIndex := 1..ProjectileCount):
                TargetRotation := CastRotation.ApplyYaw(DegreesToRadians((ProjectileIndex - 1) * DefaultAbilityEnergyBeamAngle - AngleOffset))
                Direction := TargetRotation.GetLocalForward()
                TargetPosition := CastPosition + Direction * ProjectileRange / 2.0

                for:
                    OverlapHit : SimulationEntity.FindOverlapHits(FromTransform of transform{Translation := TargetPosition, Rotation := TargetRotation}, Collision)
                    TargetComponent := OverlapHit.TargetComponent
                    TargetEntity := TargetComponent.Entity
                    NPCCollisionComponent := TargetEntity.GetComponent[npc_collision_component]
                    NPCHit := NPCCollisionComponent.MaybeNPC?
                    NPCHit.IsDamageable()?
                do:
                    Damage(NPCHit, Player.GetDamageArgs(NPCHit, Id, AttackDamage, ?Multiplier := AttackDamageMultiplier))

                PeakPosition := CastPosition + Direction * ProjectileRange

            if (IsUniqueEvolution?):
                CollisionBlast := collision_sphere{Radius := AttackRadius * 1.5}

                for (BlastIndex := 1..2):
                    Sleep(0.15)

                    for (PeakPosition : PeakPositions):
                        for:
                            OverlapHit : SimulationEntity.FindOverlapHits(FromTransform of transform{Translation := PeakPosition}, CollisionBlast)
                            TargetComponent := OverlapHit.TargetComponent
                            TargetEntity := TargetComponent.Entity
                            NPCCollisionComponent := TargetEntity.GetComponent[npc_collision_component]
                            NPCHit := NPCCollisionComponent.MaybeNPC?
                            NPCHit.IsDamageable()?
                        do:
                            Damage(NPCHit, Player.GetDamageArgs(NPCHit, Id, AttackDamage, ?Multiplier := AttackDamageMultiplier))

            Sleep(1.0)
            ParticleSystem.Disable()
                
    UpdateStats<override>():void =
        (super:)UpdateStats()
        if (WeaponInfo := GetWeaponInfo[Id]):
            set ProjectileCount = 1 + Floor[Player.GetWeaponEvolutionValue(Id, ability_id.ProjectileCount)] or 0
            set ProjectileRange = DefaultAbilityEnergyBeamRange * (1.0 + Player.GetWeaponEvolutionValue(Id, ability_id.ProjectileRange))
            set MulticastCount = 1 + Floor[Player.GetWeaponEvolutionValue(Id, ability_id.Multicast)] or 0
            
            set AttackDamage *= (IsUniqueEvolution? and 2.0 or 1.0) * MulticastCount
            set AttackRadius *= (1.0 + (MulticastCount - 1) * 0.25)