using { /Fortnite.com/Characters }
using { /Fortnite.com/Devices }
using { /Verse.org/Random }
using { /Verse.org/SceneGraph }
using { /Verse.org/Simulation }
using { /UnrealEngine.com/Temporary/SpatialMath }
using { Prefabs }
using { VisualEffects }

<# ----------------------------------------------------------------------------------------------------
    Constants
---------------------------------------------------------------------------------------------------- #>
DefaultAbilityEnergyBombRange:float = 516.0
DefaultAbilityEnergyBombGravityDuration:float = 0.2
DefaultAbilityEnergyBombAngle:float = 90.0

<# ----------------------------------------------------------------------------------------------------
    ability_energy_bomb_sg
---------------------------------------------------------------------------------------------------- #>
ability_energy_bomb_sg<public> := class(ability):
    Id<override>:weapon_id = weapon_id.EnergyBomb

    var MulticastCount:int = 1
    var WeaponEntities:[]entity = array{}

    <# ----------------------------------------------------------------------------------------------------
        Initialization
    ---------------------------------------------------------------------------------------------------- #>
    Deactivate<override>():void =
        DeactivatedEvent.Signal()
        set CooldownTimeRemaining = Max(0.0, AttackCooldownTime - (GetSimulationElapsedTime() - CastTime))
        set IsActive = false

    Deinitialize<override>():void =
        (super:)Deinitialize()
        
        for (WeaponEntity : WeaponEntities):
            WeaponEntity.RemoveFromParent()

        set WeaponEntities = array{}

    <# ----------------------------------------------------------------------------------------------------
        Get Functions
    ---------------------------------------------------------------------------------------------------- #>
    GetAvailableWeaponEntity()<suspends>:entity =
        for (WeaponEntity : WeaponEntities, not WeaponEntity.GetComponent[particle_system_component].IsEnabled[]):
            return WeaponEntity

        NewWeaponEntity := prefab_weapon_energy_bomb{}
        NewWeaponEntity.InitializeParticle()
        set WeaponEntities += array{NewWeaponEntity}
        NewWeaponEntity
    
    <# ----------------------------------------------------------------------------------------------------
        Update Functions
    ---------------------------------------------------------------------------------------------------- #>
    Ability<override>()<suspends>:void =
        race:
            DeactivatedEvent.Await()
            if (WeaponInfo := GetWeaponInfo[Id]):
                Sleep(CooldownTimeRemaining)
                loop:
                    set CastTime = GetSimulationElapsedTime()
                    spawn{Cast()}
                    Sleep(AttackCooldownTime)

    Cast()<suspends>:void =
        WeaponEntity := GetAvailableWeaponEntity()
        if:
            Character := Player.GetFortCharacter[]
            ParticleSystem := NS_AbilityEnergyBomb_v2[WeaponEntity.GetComponent[particle_system_component]]
        then:
            CastPosition := Character.GetTransform().Translation.Make2D()
            CastRotation := Player.GetAutoAimRotation()

            set ParticleSystem.Color = GetWeaponColor(Player.GetWeaponEvolutionLevel(Id))
            set ParticleSystem.IsUnique = IsUniqueEvolution
            set ParticleSystem.Angle = DefaultAbilityEnergyBombAngle
            set ParticleSystem.Radius = AttackRadius * 0.1
            set ParticleSystem.ProjectileCount = ProjectileCount
            set ParticleSystem.ProjectileMulticast = MulticastCount
            set ParticleSystem.Position = FromVector3 of CastPosition.Make2D()
            set ParticleSystem.Rotation = CastRotation.GetYawPitchRollDegrees()[0] or 0.0

            ParticleSystem.Enable()

            for (CastIndex := 1..MulticastCount):
                for (ProjectileIndex := 1..ProjectileCount):
                    spawn{Projectile(CastPosition, CastRotation, CastIndex, ProjectileIndex)}

                Player.PlayAudio(audio_id.Projectile)
                Sleep(0.3)

            Sleep(3.0)
            ParticleSystem.Disable()

    Projectile(CastPosition:vector3, CastRotation:rotation, CastIndex:int, ProjectileIndex:int)<suspends>:void =
        TargetRotation := if (ProjectileCount > 1):
            Angle := DegreesToRadians(DefaultAbilityEnergyBombAngle) * ((ProjectileIndex - 1) * 1.0 - (Floor[ProjectileCount * 1.0 / 2.0] or 0) * (Mod[ProjectileCount, 2] = 0 and 0.5 or 1.0))
            CastRotation.ApplyYaw(Angle)
        else:
            CastRotation

        TargetPosition := CastPosition + TargetRotation.GetLocalForward() * DefaultAbilityEnergyBombRange
        TargetTransform := FromTransform of transform{Translation := TargetPosition}

        if (IsUniqueEvolution? and CastIndex = 1):
            Sleep(1.5)

            for:
                OverlapHit : GetSimulationEntity[].FindOverlapHits(TargetTransform, collision_sphere{Radius := AttackRadius * 1.5})
                TargetComponent := OverlapHit.TargetComponent
                TargetEntity := TargetComponent.Entity
                NPCCollisionComponent := TargetEntity.GetComponent[npc_collision_component]
                NPCHit := NPCCollisionComponent.MaybeNPC?
                NPCHit.IsDamageable()?
                NPCPosition := NPCHit.GetAgent[].GetFortCharacter[].GetTransform().Translation.Make2D()
                Direction := (TargetPosition - NPCPosition).MakeUnitVector[]
            do:
                DirectionRotation := MakeRotation(vector3{X := 0.0, Y := 0.0, Z := 1.0}, ArcTan(Direction.Y, Direction.X))
                NPCHit.Knockback(knockback_properties{Duration := 1.0, Rotation := DirectionRotation})

            Sleep(0.5)
        else:
            Sleep(2.0)

        for:
            OverlapHit : GetSimulationEntity[].FindOverlapHits(TargetTransform, collision_sphere{Radius := AttackRadius})
            TargetComponent := OverlapHit.TargetComponent
            TargetEntity := TargetComponent.Entity
            NPCCollisionComponent := TargetEntity.GetComponent[npc_collision_component]
            NPCHit := NPCCollisionComponent.MaybeNPC?
            NPCHit.IsDamageable()?
        do:
            Damage(NPCHit, Player.GetDamageArgs(NPCHit, Id, AttackDamage, ?Multiplier := AttackDamageMultiplier))

    UpdateStats<override>():void =
        (super:)UpdateStats()
        if (WeaponInfo := GetWeaponInfo[Id]):
            set ProjectileCount = 1 + Floor[Player.GetWeaponEvolutionValue(Id, ability_id.ProjectileCount)] or 0
            set MulticastCount = 1 + Floor[Player.GetWeaponEvolutionValue(Id, ability_id.Multicast)] or 0