using { /Fortnite.com/Characters }
using { /Verse.org/Random }
using { /Verse.org/SceneGraph }
using { /Verse.org/Simulation }
using { /UnrealEngine.com/Temporary/SpatialMath }
using { Prefabs }
using { VisualEffects }

<# ----------------------------------------------------------------------------------------------------
    Constants
---------------------------------------------------------------------------------------------------- #>
DefaultAbilityHammerProjectileSpeed:float = 1000.0
DefaultAbilityHammerCastRange:float = 500.0
DefaultAbilityHammerStunRange:float = 900.0
DefaultAbilityHammerAngle:float = 30.0

<# ----------------------------------------------------------------------------------------------------
    ability_hammer_sg
---------------------------------------------------------------------------------------------------- #>
ability_hammer_sg<public> := class(ability):
    Id<override>:weapon_id = weapon_id.Hammer

    var MulticastCount:int = 1
    var CastCount:int = 0
    var CastRange:float = DefaultAbilityHammerCastRange
    var IsMaxLevel:logic = false
    var ObjectPool:[]projectile_hammer_sg = array{}
    var WeaponEntities:[]entity = array{}
    var ProjectileEntities:[]entity = array{}

    <# ----------------------------------------------------------------------------------------------------
        Initialization
    ---------------------------------------------------------------------------------------------------- #>
    Deactivate<override>():void =
        DeactivatedEvent.Signal()
        set CooldownTimeRemaining = Max(0.0, AttackCooldownTime - (GetSimulationElapsedTime() - CastTime))
        set IsActive = false

    Deinitialize<override>():void =
        (super:)Deinitialize()
        
        for (Object : ObjectPool):
            Object.Deactivate(?IsDispose := true)

        for (WeaponEntity : WeaponEntities):
            WeaponEntity.RemoveFromParent()

        for (Projectile : ProjectileEntities):
            Projectile.RemoveFromParent()

        set WeaponEntities = array{}
        set ProjectileEntities = array{}

    <# ----------------------------------------------------------------------------------------------------
        Get Functions
    ---------------------------------------------------------------------------------------------------- #>
    GetAvailablePoolObject<private>()<transacts>:projectile_hammer_sg =
        for (Object : ObjectPool, Object.IsAvailable?):
            return Object

        NewObject := projectile_hammer_sg{Ability := Self}
        set ObjectPool += array{NewObject}
        NewObject

    GetAvailableWeaponEntity()<suspends>:entity =
        for (WeaponEntity : WeaponEntities, not WeaponEntity.GetComponent[particle_system_component].IsEnabled[]):
            return WeaponEntity

        NewWeaponEntity := prefab_weapon_hammer{}
        NewWeaponEntity.InitializeParticle()
        set WeaponEntities += array{NewWeaponEntity}
        NewWeaponEntity

    GetAvailableProjectileEntity()<suspends>:entity =
        for (ProjectileEntity : ProjectileEntities, not ProjectileEntity.GetComponent[particle_system_component].IsEnabled[]):
            return ProjectileEntity

        NewProjectileEntity := prefab_weapon_hammer_projectile{}
        NewProjectileEntity.InitializeParticle()
        set ProjectileEntities += array{NewProjectileEntity}
        NewProjectileEntity

    <# ----------------------------------------------------------------------------------------------------
        Update Functions
    ---------------------------------------------------------------------------------------------------- #>
    Ability<override>()<suspends>:void =
        race:
            DeactivatedEvent.Await()
            block:
                Sleep(CooldownTimeRemaining)
                loop:
                    set CastTime = GetSimulationElapsedTime()
                    spawn{Cast()}
                    set CastCount += 1
                    Sleep(AttackCooldownTime)

    Cast()<suspends>:void =
        WeaponEntity := GetAvailableWeaponEntity()
        if:
            Character := Player.GetFortCharacter[]
            ParticleSystem := NS_AbilityHammerDrop_v2[WeaponEntity.GetComponent[particle_system_component]]
        then:
            Position := Character.GetTransform().Translation
            Rotation := Player.GetAutoAimRotation()
            IsSupercharged := logic{(IsMaxLevel? and not IsUniqueEvolution?) and Mod[CastCount, 10] = 0}
            FinalProjectileCount := ProjectileCount * (IsSupercharged? and 2 or 1)

            set ParticleSystem.Color = GetWeaponColor(Player.GetWeaponEvolutionLevel(Id))
            set ParticleSystem.Angle = DefaultAbilityHammerAngle
            set ParticleSystem.Radius = AttackRadius * 0.1
            set ParticleSystem.Range = CastRange
            set ParticleSystem.ProjectileCount = FinalProjectileCount
            set ParticleSystem.ProjectileMulticast = MulticastCount
            set ParticleSystem.Position = FromVector3 of Position.Make2D()
            set ParticleSystem.Rotation = Rotation.GetYawPitchRollDegrees()[0] or 0.0

            ParticleSystem.Enable()

            sync:
                block:
                    for (CastIndex := 1..MulticastCount):
                        for (ProjectileIndex := 0..FinalProjectileCount - 1):
                            spawn{AbilityProjectile(ProjectileIndex, CastIndex, Position, Rotation, IsSupercharged)}

                        Player.PlayAudio(audio_id.ProjectileEvolved)
                        Sleep(0.1)

                    Sleep(2.0)
                    ParticleSystem.Disable()

                if (IsUniqueEvolution?):
                    ProjectileEntity := GetAvailableProjectileEntity()
                    if:
                        ProjectileParticleSystem := NS_AbilityHammer_v2[ProjectileEntity.GetComponent[particle_system_component]]
                    then:
                        set ProjectileParticleSystem.Color = GetWeaponColor(Player.GetWeaponEvolutionLevel(Id))
                        set ProjectileParticleSystem.Angle = DefaultAbilityHammerAngle
                        set ProjectileParticleSystem.Radius = AttackRadius / 2.0 * 0.1
                        set ProjectileParticleSystem.Range = CastRange
                        set ProjectileParticleSystem.ProjectileCount = FinalProjectileCount
                        set ProjectileParticleSystem.ProjectileMulticast = MulticastCount
                        set ProjectileParticleSystem.ProjectileSpeed = DefaultAbilityHammerProjectileSpeed
                        set ProjectileParticleSystem.Position = FromVector3 of Position.Make2D()
                        set ProjectileParticleSystem.Rotation = Rotation.GetYawPitchRollDegrees()[0] or 0.0

                        Sleep(0.5)
                        ProjectileParticleSystem.Enable()
                        Sleep(CastRange / DefaultAbilityHammerProjectileSpeed + 0.2)
                        ProjectileParticleSystem.Disable()

    AbilityProjectile(Index:int, CastIndex:int, Position:vector3, Rotation:rotation, IsSupercharged:logic)<suspends>:void =
        CastRotation := Rotation.ApplyYaw(DegreesToRadians(DefaultAbilityHammerAngle) * (Index - (Floor[ProjectileCount * 1.0 / 2.0] or 0)))
        CastPosition := Position + CastRotation.GetLocalForward() * (CastRange + (CastIndex - 1) * 50.0)
        Collision := collision_sphere{Radius := AttackRadius}
        Sleep(0.5)

        for:
            OverlapHit : GetSimulationEntity[].FindOverlapHits(FromTransform of transform{Translation := CastPosition}, Collision)
            TargetComponent := OverlapHit.TargetComponent
            TargetEntity := TargetComponent.Entity
            NPCCollisionComponent := TargetEntity.GetComponent[npc_collision_component]
            NPCHit := NPCCollisionComponent.MaybeNPC?
            NPCHit.IsDamageable()?
        do:
            Damage(NPCHit, Player.GetDamageArgs(NPCHit, Id, AttackDamage * (IsSupercharged? and 10.0 or 1.0), ?Multiplier := AttackDamageMultiplier))
    
        if (IsUniqueEvolution?):
            ProjectileProperties := projectile_properties_sg:
                Owner := Player
                Index := Index
                Position := CastPosition
                Rotation := CastRotation.ApplyYaw(DegreesToRadians(180.0))
                Angle := 0.0
                Duration := CastRange / DefaultAbilityHammerProjectileSpeed
                Distance := CastRange
                Speed := DefaultAbilityHammerProjectileSpeed
                Damage := AttackDamage * MulticastCount
                DamageMultiplier := AttackDamageMultiplier
                Radius := AttackRadius / 2.0
                Piercing := -1

            Projectile := GetAvailablePoolObject()
            spawn{Projectile.Activate(ProjectileProperties)}

        else if (IsSupercharged?):
            Sleep(0.5)
            DirectionRotation := MakeRotationFromYawPitchRollDegrees(0.0, 90.0, 0.0)
            set Collision.Radius = AttackRadius * 2.0

            for:
                OverlapHit : GetSimulationEntity[].FindOverlapHits(FromTransform of transform{Translation := CastPosition}, Collision)
                TargetComponent := OverlapHit.TargetComponent
                TargetEntity := TargetComponent.Entity
                NPCCollisionComponent := TargetEntity.GetComponent[npc_collision_component]
                NPCHit := NPCCollisionComponent.MaybeNPC?
                NPCHit.IsDamageable()?
            do:
                NPCHit.Knockback(knockback_properties{Duration := 3.0, Rotation := DirectionRotation})

    UpdateStats<override>():void =
        (super:)UpdateStats()
        if (WeaponInfo := GetWeaponInfo[Id]):
            set ProjectileCount = 1 + Floor[Player.GetWeaponEvolutionValue(Id, ability_id.ProjectileCount)] or 0
            set MulticastCount = 1 + Floor[Player.GetWeaponEvolutionValue(Id, ability_id.Multicast)] or 0
            set IsMaxLevel = logic{Player.GetWeaponLevel(Id) >= MaxWeaponLevel}
            set CastRange = DefaultAbilityHammerCastRange * (IsUniqueEvolution? and 2.0 or 1.0)

<# ----------------------------------------------------------------------------------------------------
    projectile_hammer_sg
---------------------------------------------------------------------------------------------------- #>
projectile_hammer_sg := class(projectile_sg):
    ScheduleDeactivate<override>(?MaybeProperties:?projectile_properties_sg = false)<suspends>:void =
        if:
            Properties := MaybeProperties?
            ParticleSystem := NS_AbilityHammer_v2[Properties.MaybeParticleSystem?]
            set ParticleSystem.ProjectileStates[Properties.Index] = 1
        then:
            Sleep(0.2)

        Deactivate()

    Deactivate<override>(?MaybeProperties:?projectile_properties_sg = false, ?IsDispose:logic = false):void =
        DeactivatedEvent.Signal()
        (super:)Deactivate()