using { /Fortnite.com/Characters }
using { /Verse.org/SceneGraph }
using { /Verse.org/Simulation }
using { /UnrealEngine.com/Temporary/SpatialMath }
using { Prefabs }
using { VisualEffects }

<# ----------------------------------------------------------------------------------------------------
    Constants
---------------------------------------------------------------------------------------------------- #>
DefaultAbilityKunaiProjectileCount:int = 4

<# ----------------------------------------------------------------------------------------------------
    ability_kunai_sg
---------------------------------------------------------------------------------------------------- #>
ability_kunai_sg<public> := class(ability):
    Id<override>:weapon_id = weapon_id.Kunai

    var Angle:float = 0.0
    var ProjectilePiercing:int = 0
    var ProjectileAngle:float = 45.0
    var MulticastCount:int = 1

    var ObjectPool:[]projectile_kunai_sg = array{}
    var MaybeWeaponPrefabSpiral:?prefab_weapon_kunai_spiral = false
    var WeaponEntities:[]entity = array{}

    <# ----------------------------------------------------------------------------------------------------
        Initialization
    ---------------------------------------------------------------------------------------------------- #>
    Deactivate<override>():void =
        DeactivatedEvent.Signal()
        set CooldownTimeRemaining = Max(0.0, AttackCooldownTime - (GetSimulationElapsedTime() - CastTime))
        set IsActive = false
        
    Deinitialize<override>():void =
        (super:)Deinitialize()

        for (Object : ObjectPool):
            Object.Deactivate(?IsDispose := true)

        for (WeaponEntity : WeaponEntities):
            WeaponEntity.RemoveFromParent()

        set WeaponEntities = array{}

        if (WeaponPrefabSpiral := MaybeWeaponPrefabSpiral?):
            WeaponPrefabSpiral.RemoveFromParent()
            set MaybeWeaponPrefabSpiral = false

    <# ----------------------------------------------------------------------------------------------------
        Get Functions
    ---------------------------------------------------------------------------------------------------- #>
    GetAvailablePoolObject<private>()<transacts>:projectile_kunai_sg =
        for (Object : ObjectPool, Object.IsAvailable?):
            return Object

        NewObject := projectile_kunai_sg{Ability := Self}
        set ObjectPool += array{NewObject}
        NewObject

    GetAvailableWeaponEntity()<suspends>:entity =
        for (WeaponEntity : WeaponEntities, not WeaponEntity.GetComponent[particle_system_component].IsEnabled[]):
            return WeaponEntity

        NewWeaponEntity := prefab_weapon_kunai{}
        NewWeaponEntity.InitializeParticle()
        set WeaponEntities += array{NewWeaponEntity}
        NewWeaponEntity

    <# ----------------------------------------------------------------------------------------------------
        Update Functions
    ---------------------------------------------------------------------------------------------------- #>
    Ability<override>()<suspends>:void =
        if (IsUniqueEvolution?, not MaybeWeaponPrefabSpiral?):
            SpiralWeaponPrefab := prefab_weapon_kunai_spiral{}
            SpiralWeaponPrefab.InitializeParticle()
            set MaybeWeaponPrefabSpiral = option{SpiralWeaponPrefab}

            for (WeaponEntity : WeaponEntities):
                WeaponEntity.RemoveFromParent()

            set WeaponEntities = array{}

        race:
            DeactivatedEvent.Await()
            if (IsUniqueEvolution?):
                defer:
                    if (ParticleSystem := MaybeWeaponPrefabSpiral?.GetComponent[particle_system_component]):
                        ParticleSystem.Stop()

                CastEvolution()
            else:
                Sleep(CooldownTimeRemaining)
                loop:
                    set CastTime = GetSimulationElapsedTime()
                    spawn{Cast()}
                    Sleep(AttackCooldownTime)

    Cast()<suspends>:void =
        WeaponEntity := GetAvailableWeaponEntity()
        if:
            Character := Player.GetFortCharacter[]
            ParticleSystem := NS_AbilityKunai_v2[WeaponEntity.GetComponent[particle_system_component]]
        then:
            Rotation := MakeRotationFromYawPitchRollDegrees(ProjectileAngle / 2.0, 0.0, 0.0)
            Duration := DefaultAbilityProjectileDistance / DefaultAbilityProjectileSpeed

            set ParticleSystem.Color = GetWeaponColor(Player.GetWeaponEvolutionLevel(Id))
            set ParticleSystem.ProjectileMulticast = MulticastCount
            set ParticleSystem.ProjectileCount = ProjectileCount
            set ParticleSystem.ProjectileStates = for (Index := 1..ProjectileCount) {0}
            set ParticleSystem.ProjectileSpeed = DefaultAbilityProjectileSpeed
            set ParticleSystem.Angle = ProjectileAngle
            set ParticleSystem.Radius = AttackRadius * 0.1
            set ParticleSystem.Position = FromVector3 of Character.GetTransform().Translation

            ParticleSystem.Enable()

            for (Index := 0..ProjectileCount - 1):
                ProjectileProperties := projectile_properties_sg:
                    Owner := Player
                    Index := Index
                    Position := Character.GetTransform().Translation
                    Rotation := Rotation
                    Angle := DegreesToRadians(ProjectileAngle * Index)
                    Duration := Duration
                    Distance := DefaultAbilityProjectileDistance
                    Speed := DefaultAbilityProjectileSpeed
                    Damage := AttackDamage
                    DamageMultiplier := AttackDamageMultiplier
                    Radius := AttackRadius
                    Piercing := ProjectilePiercing
                    MaybeParticleSystem := option{ParticleSystem}

                Projectile := GetAvailablePoolObject()
                spawn{Projectile.Activate(ProjectileProperties)}

            branch:
                for (CastIndex := 1..MulticastCount):
                    Player.PlayAudio(audio_id.Projectile)
                    Sleep(0.1)

            Sleep(Duration + 0.2)
            ParticleSystem.Disable()

    CastEvolution()<suspends>:void =
        if:
            Character := Player.GetFortCharacter[]
            WeaponPrefabSpiral := MaybeWeaponPrefabSpiral?
            ParticleSystem := NS_AbilityKunaiSpiral[WeaponPrefabSpiral.GetComponent[particle_system_component]]
        then:
            CooldownTime := 0.2 - ((ProjectileCount - 4) * 1.0) / 2.0 * 0.05
            Duration := DefaultAbilityProjectileDistance / DefaultAbilityProjectileSpeed
            TotalProjectileCount := Floor[Duration / CooldownTime] or 1

            set ParticleSystem.ProjectileMulticast = MulticastCount
            set ParticleSystem.ProjectileRate = 1.0 / CooldownTime
            set ParticleSystem.ProjectileLifetimes = for (Index := 1..TotalProjectileCount) {1.0}
            set ParticleSystem.Radius = AttackRadius * 0.1
            set ParticleSystem.Position = FromVector3 of Character.GetTransform().Translation

            ParticleSystem.Enable()
            ParticleSystem.Play()

            var ProjectileIndex:int = 0
            
            race:
                loop:
                    set ParticleSystem.Position = FromVector3 of Character.GetTransform().Translation
                    set ParticleSystem.Angle = Angle
                    Sleep(CooldownTime)

                loop:
                    Player.PlayAudio(audio_id.Projectile)

                    ProjectileProperties := projectile_properties_sg:
                        Owner := Player
                        Index := ProjectileIndex
                        Position := Character.GetTransform().Translation
                        Rotation := IdentityRotation()
                        Angle := DegreesToRadians(Angle)
                        Duration := DefaultAbilityProjectileDistance / DefaultAbilityProjectileSpeed
                        Distance := DefaultAbilityProjectileDistance
                        Speed := DefaultAbilityProjectileSpeed
                        Damage := AttackDamage
                        DamageMultiplier := AttackDamageMultiplier
                        Radius := AttackRadius
                        Piercing := ProjectilePiercing
                        MaybeParticleSystem := option{ParticleSystem}

                    set Angle = Mod[Floor[Angle + 5.0] or 0, 360] * 1.0 or Angle
                    Projectile := GetAvailablePoolObject()
                    spawn{Projectile.Activate(ProjectileProperties)}
                    set ProjectileIndex = Mod[ProjectileIndex + 1, TotalProjectileCount] or 0
                    Sleep(CooldownTime)
                
    UpdateStats<override>():void =
        (super:)UpdateStats()
        if (WeaponInfo := GetWeaponInfo[Id]):
            set ProjectileCount = DefaultAbilityKunaiProjectileCount + Floor[Player.GetWeaponEvolutionValue(Id, ability_id.ProjectileCount)] or 0
            set MulticastCount = 1 + Floor[Player.GetWeaponEvolutionValue(Id, ability_id.Multicast)] or 0

            set ProjectilePiercing = (4 + Floor[Player.GetWeaponEvolutionValue(Id, ability_id.ProjectilePiercing)] or 0) * MulticastCount
            set ProjectileAngle = 360.0 / (ProjectileCount * 1.0)

            set AttackDamage *= MulticastCount * (IsUniqueEvolution? and 1.5 or 1.0)

<# ----------------------------------------------------------------------------------------------------
    projectile_kunai_sg
---------------------------------------------------------------------------------------------------- #>
projectile_kunai_sg := class(projectile_sg):
    ScheduleDeactivate<override>(?MaybeProperties:?projectile_properties_sg = false)<suspends>:void =
        if:
            Properties := MaybeProperties?
            ParticleSystem := NS_AbilityKunai_v2[Properties.MaybeParticleSystem?]
            set ParticleSystem.ProjectileStates[Properties.Index] = 1
        then:
            Sleep(0.2)

        Deactivate()

    Deactivate<override>(?IsDispose:logic = false):void =
        DeactivatedEvent.Signal()
        (super:)Deactivate()