using { /Fortnite.com/Characters }
using { /Verse.org/Random }
using { /Verse.org/SceneGraph }
using { /Verse.org/Simulation }
using { /UnrealEngine.com/Temporary/SpatialMath }
using { Prefabs }

<# ----------------------------------------------------------------------------------------------------
    Constants
---------------------------------------------------------------------------------------------------- #>
BaseAbilityLightningRange:float = 900.0

<# ----------------------------------------------------------------------------------------------------
    ability_lightning_sg
---------------------------------------------------------------------------------------------------- #>
ability_lightning_sg<public> := class(ability):
    Id<override>:weapon_id = weapon_id.Lightning

    var EvolutionSize:float = 1.0
    var MaybeWeaponPrefab:?prefab_weapon_lightning = false

    <# ----------------------------------------------------------------------------------------------------
        Initialization
    ---------------------------------------------------------------------------------------------------- #>
    Deactivate<override>():void =
        DeactivatedEvent.Signal()
        set CooldownTimeRemaining = Max(0.0, AttackCooldownTime - (GetSimulationElapsedTime() - CastTime))
        set IsActive = false

    Deinitialize<override>():void =
        (super:)Deinitialize()
        
        if (WeaponPrefab := MaybeWeaponPrefab?):
            WeaponPrefab.RemoveFromParent()

        set MaybeWeaponPrefab = false

    <# ----------------------------------------------------------------------------------------------------
        Update Functions
    ---------------------------------------------------------------------------------------------------- #>
    Ability<override>()<suspends>:void =
        if (not MaybeWeaponPrefab?):
            WeaponPrefab := prefab_weapon_lightning{}
            WeaponPrefab.InitializeParticle()
            set MaybeWeaponPrefab = option{WeaponPrefab}

        if (AbilityLightning := VisualEffects.NS_AbilityLightning[MaybeWeaponPrefab?.GetComponent[particle_system_component]]):
            set AbilityLightning.Color = GetWeaponColor(Player.GetWeaponEvolutionLevel(Id))
            set AbilityLightning.Radius = AttackRadius * 0.1

        race:
            DeactivatedEvent.Await()
            if (WeaponInfo := GetWeaponInfo[Id]):
                Sleep(CooldownTimeRemaining)

                loop:
                    set CastTime = GetSimulationElapsedTime()
                    spawn{Cast(WeaponInfo)}
                    Sleep(AttackCooldownTime)

    Cast(WeaponInfo:weapon_info)<suspends>:void =
        if (Character := Player.GetFortCharacter[], SimulationEntity := GetSimulationEntity[]):
            Collision:collision_sphere = collision_sphere{Radius := BaseAbilityLightningRange}

            var IsTargetBoss:logic = false

            TargetNPCs := if:
                IsUniqueEvolution?

                BossesInRange := for:
                    NPCBehaviorBase : GetAllNPCBosses()
                    NPCBehaviorBase.IsDamageable()?
                    NPCCharacter := NPCBehaviorBase.GetAgent[].GetFortCharacter[]
                    AttackRange := BaseAbilityLightningRange + NPCBehaviorBase.GetSize()
                    DistanceSquaredXY(NPCCharacter.GetTransform().Translation, Character.GetTransform().Translation) <= AttackRange * AttackRange
                do:
                    NPCBehaviorBase

                BossesInRange[0]
            then:
                set IsTargetBoss = true
                BossesInRange
            else:
                for:
                    OverlapHit : SimulationEntity.FindOverlapHits(FromTransform of Character.GetTransform(), Collision)
                    TargetComponent := OverlapHit.TargetComponent
                    TargetEntity := TargetComponent.Entity
                    NPCCollisionComponent := TargetEntity.GetComponent[npc_collision_component]
                    NPCHit := NPCCollisionComponent.MaybeNPC?
                    NPCHit.IsDamageable()?
                do:
                    NPCHit

            TargetNPCCount := TargetNPCs.Length
            if (TargetNPCCount = 0) then return

            var TargetIndex:int = 0
            var MaybeTargetPosition:?vector3 = false

            for (Index := 1..ProjectileCount):
                if (IsTargetBoss?):
                    loop:
                        if:
                            TargetNPC := TargetNPCs[TargetIndex]
                            TargetNPC.IsDamageable()?
                            TargetNPCCharacter := TargetNPC.GetAgent[].GetFortCharacter[]
                        then:
                            set MaybeTargetPosition = option{TargetNPCCharacter.GetTransform().Translation}
                            break

                        set TargetIndex += 1

                        if (TargetIndex > TargetNPCCount - 1):
                            break
                else if:
                    TargetNPC := TargetNPCs[GetRandomInt(0, TargetNPCCount - 1)]
                    TargetNPCCharacter := TargetNPC.GetAgent[].GetFortCharacter[]
                then:
                    set MaybeTargetPosition = option{TargetNPCCharacter.GetTransform().Translation}

                if (TargetPosition := MaybeTargetPosition?):
                    if (AbilityLightning := VisualEffects.NS_AbilityLightning[MaybeWeaponPrefab?.GetComponent[particle_system_component]]):
                        set AbilityLightning.Position = FromVector3 of TargetPosition.Make2D()
                        AbilityLightning.Play()
                        
                    set Collision.Radius = AttackRadius

                    for:
                        OverlapHit : SimulationEntity.FindOverlapHits(FromTransform of transform{Translation := TargetPosition}, Collision)
                        TargetComponent := OverlapHit.TargetComponent
                        TargetEntity := TargetComponent.Entity
                        NPCCollisionComponent := TargetEntity.GetComponent[npc_collision_component]
                        NPCHit := NPCCollisionComponent.MaybeNPC?
                        NPCHit.IsDamageable()?
                    do:
                        Damage(NPCHit, Player.GetDamageArgs(NPCHit, Id, AttackDamage, ?Multiplier := AttackDamageMultiplier))

                Sleep(0.1)

    UpdateStats<override>():void =
        (super:)UpdateStats()
        if (WeaponInfo := GetWeaponInfo[Id]):
            set ProjectileCount = 1 + Floor[Player.GetWeaponEvolutionValue(Id, ability_id.ProjectileCount)] or 0
            set EvolutionSize = 1.0 + Player.GetWeaponEvolutionValue(Id, ability_id.Size)
            set AttackRadius *= EvolutionSize