using { /Fortnite.com/Characters }
using { /Fortnite.com/Devices }
using { /Verse.org/Random }
using { /Verse.org/SceneGraph }
using { /Verse.org/SceneGraph/KeyframedMovement }
using { /Verse.org/Simulation }
using { /UnrealEngine.com/Temporary/SpatialMath }
using { Prefabs }
using { VisualEffects }

<# ----------------------------------------------------------------------------------------------------
    Constants
---------------------------------------------------------------------------------------------------- #>
DefaultAbilitySentryRange:float = 300.0
DefaultAbilitySentryDuration:float = 5.5
DefaultAbilitySentryAttackInterval:float = 1.0
DefaultAbilitySentryRotationDuration:float = 0.15

<# ----------------------------------------------------------------------------------------------------
    Data Definitions
---------------------------------------------------------------------------------------------------- #>
projectile_sentry_properties<public> := struct:
    Player<public>:player
    Damage<public>:float
    DamageMultiplier<public>:float = 1.0
    Radius<public>:float
    FireRate<public>:float
    ProjectileCount<public>:int = 1
    IsUniqueEvolution<public>:logic

<# ----------------------------------------------------------------------------------------------------
    ability_sentry_sg
---------------------------------------------------------------------------------------------------- #>
ability_sentry_sg<public> := class(ability):
    Id<override>:weapon_id = weapon_id.Sentry

    var FireRate:float = 1.0
    var ObjectPool:[]projectile_sentry_sg = array{}

    <# ----------------------------------------------------------------------------------------------------
        Initialization
    ---------------------------------------------------------------------------------------------------- #>
    Deactivate<override>():void =
        DeactivatedEvent.Signal()
        set CooldownTimeRemaining = Max(0.0, AttackCooldownTime - (GetSimulationElapsedTime() - CastTime))
        set IsActive = false

    Deinitialize<override>():void =
        (super:)Deinitialize()

        for (Object : ObjectPool):
            Object.Deactivate(?IsDispose := true)

    <# ----------------------------------------------------------------------------------------------------
        Get Functions
    ---------------------------------------------------------------------------------------------------- #>
    GetAvailablePoolObject<private>()<transacts>:projectile_sentry_sg =
        for (Object : ObjectPool, Object.IsAvailable?):
            return Object

        NewObject := projectile_sentry_sg{Ability := Self}
        set ObjectPool += array{NewObject}
        NewObject
        
    <# ----------------------------------------------------------------------------------------------------
        Update Functions
    ---------------------------------------------------------------------------------------------------- #>
    Ability<override>()<suspends>:void =
        ProjectileProperties := projectile_sentry_properties:
            Player := Player
            Damage := AttackDamage
            DamageMultiplier := AttackDamageMultiplier
            Radius := AttackRadius
            FireRate := FireRate
            ProjectileCount := ProjectileCount
            IsUniqueEvolution := IsUniqueEvolution
        
        race:
            DeactivatedEvent.Await()
            if (WeaponInfo := GetWeaponInfo[Id]):
                Sleep(CooldownTimeRemaining)
                loop:
                    set CastTime = GetSimulationElapsedTime()
                    spawn{GetAvailablePoolObject().Activate(ProjectileProperties)}
                    Sleep(AttackCooldownTime)

    UpdateStats<override>():void =
        (super:)UpdateStats()
        if (WeaponInfo := GetWeaponInfo[Id]):
            set ProjectileCount = 1 + Floor[Player.GetWeaponEvolutionValue(Id, ability_id.ProjectileCount)] or 0
            set FireRate = DefaultAbilitySentryAttackInterval / ((1.0 + Player.GetWeaponEvolutionValue(Id, ability_id.FireRate)) * (IsUniqueEvolution? and 2.0 or 1.0))
            set AttackDamage *= (IsUniqueEvolution? and 2.0 or 1.0) * ProjectileCount

<# ----------------------------------------------------------------------------------------------------
    projectile_sentry_sg
---------------------------------------------------------------------------------------------------- #>
projectile_sentry_sg<public> := class(poolable):
    Ability<public>:ability

    DeactivatedEvent:event(void) = event(void){}

    var MaybeWeaponPrefab:?prefab_weapon_sentry = false

    Activate(Properties:projectile_sentry_properties)<suspends>:void =
        (poolable:)Activate()
        
        if (not MaybeWeaponPrefab?):
            WeaponPrefab := prefab_weapon_sentry{}
            WeaponPrefab.InitializeParticle()
            set MaybeWeaponPrefab = option{WeaponPrefab}

        if:
            PlayerCharacter := Properties.Player.GetFortCharacter[]
            WeaponPrefab := MaybeWeaponPrefab?
            ParticleSystem := NS_AbilitySentry_v2[WeaponPrefab.GetComponent[particle_system_component]]
            KeyframedMovementComponent := WeaponPrefab.GetComponent[keyframed_movement_component]
            SimulationEntity := GetSimulationEntity[]
        then:
            SentryPosition := GetRandomPositionOnCircle(GetRandomFloat(150.0, 300.0), ?Origin := PlayerCharacter.GetTransform().Translation)
            SentryRange := DefaultAbilitySentryRange * (Properties.IsUniqueEvolution? and 2.0 or 1.0)
            CollisionRange := collision_sphere{Radius := SentryRange}
            CollisionAttack := collision_sphere{Radius := Properties.Radius}

            set ParticleSystem.Color = GetWeaponColor(Properties.Player.GetWeaponEvolutionLevel(Ability.Id))
            set ParticleSystem.Range = SentryRange * 0.1
            set ParticleSystem.Radius = Properties.Radius * 0.1
            set ParticleSystem.MeshIndex = Properties.ProjectileCount - 1
            set ParticleSystem.FireRate = Properties.FireRate
            set ParticleSystem.Duration = DefaultAbilitySentryDuration

            Properties.Player.PlayAudio(audio_id.Projectile)

            var IsFirst:logic = true
            race:
                DeactivatedEvent.Await()
                Sleep(DefaultAbilitySentryDuration)
                loop:
                    var MaybeNearestNPC:?npc_behavior_base = false
                    var NearestDistanceSquared:float = 0.0

                    for:
                        OverlapHit : SimulationEntity.FindOverlapHits(FromTransform of transform{Translation := SentryPosition}, CollisionRange)
                        TargetComponent := OverlapHit.TargetComponent
                        TargetEntity := TargetComponent.Entity
                        NPCCollisionComponent := TargetEntity.GetComponent[npc_collision_component]
                        NPCHit := NPCCollisionComponent.MaybeNPC?
                        NPCHit.IsDamageable()?
                        DistanceSquaredFromSentry := DistanceSquaredXY(NPCHit.GetAgent[].GetFortCharacter[].GetTransform().Translation, SentryPosition)
                        not MaybeNearestNPC? or DistanceSquaredFromSentry < NearestDistanceSquared
                    do:
                        set MaybeNearestNPC = option{NPCHit}
                        set NearestDistanceSquared = DistanceSquaredFromSentry

                    if (NearestNPC := MaybeNearestNPC?, NearestNPCCharacter := NearestNPC.GetAgent[].GetFortCharacter[]):
                        TargetPosition := NearestNPCCharacter.GetTransform().Translation
                        set ParticleSystem.AttackPosition = FromVector3(TargetPosition.Make2D())
                        set ParticleSystem.IsAttack = true

                        if (Direction := (TargetPosition.Make2D() - SentryPosition.Make2D()).MakeUnitVector[]):
                            TargetRotation := MakeRotation(vector3{X := 0.0, Y := 0.0, Z := 1.0}, ArcTan(Direction.Y, Direction.X))
                            if (IsFirst?):
                                WeaponPrefab.SetGlobalTransform of FromTransform of transform:
                                    Translation := SentryPosition.Make2D()
                                    Rotation := TargetRotation

                                ParticleSystem.Enable()
                            else:
                                RelativeKeyframes := array:
                                    keyframed_movement_delta:
                                        Transform := FromTransform of transform:                                          
                                            Rotation := MakeShortestRotationBetween(FromRotation of WeaponPrefab.GetGlobalTransform().Rotation, TargetRotation)
                                            Scale := vector3{}
                                        Duration := DefaultAbilitySentryRotationDuration
                                        Easing := linear_easing_function{}

                                KeyframedMovementComponent.SetKeyframes(RelativeKeyframes, oneshot_keyframed_movement_playback_mode{})
                                KeyframedMovementComponent.Play()
                                KeyframedMovementComponent.FinishedEvent.Await()

                        for:
                            OverlapHit : SimulationEntity.FindOverlapHits(FromTransform of transform{Translation := TargetPosition}, CollisionAttack)
                            TargetComponent := OverlapHit.TargetComponent
                            TargetEntity := TargetComponent.Entity
                            NPCCollisionComponent := TargetEntity.GetComponent[npc_collision_component]
                            NPCHit := NPCCollisionComponent.MaybeNPC?
                            NPCHit.IsDamageable()?
                        do:
                            Ability.Damage(NPCHit, Properties.Player.GetDamageArgs(NPCHit, Ability.Id, Properties.Damage, ?Multiplier := Properties.DamageMultiplier))

                        Sleep(Properties.FireRate - DefaultAbilitySentryRotationDuration - OneTick)
                    else:
                        set ParticleSystem.IsAttack = false

                        if (IsFirst?):
                            WeaponPrefab.SetGlobalTransform of FromTransform of transform:
                                Translation := SentryPosition.Make2D()
                                Rotation := PlayerCharacter.GetTransform().Rotation

                            ParticleSystem.Enable()
                        else:
                            Sleep(Properties.FireRate)

                    set IsFirst = false

            Deactivate()

    Deactivate<override>(?IsDispose:logic = false):void =
        if:
            WeaponPrefab := MaybeWeaponPrefab?
            ParticleComponent := WeaponPrefab.GetComponent[particle_system_component]
        then:
            ParticleComponent.Disable()

            if (IsDispose?):
                WeaponPrefab.RemoveFromParent()
                set MaybeWeaponPrefab = false

        DeactivatedEvent.Signal()
        (super:)Deactivate()