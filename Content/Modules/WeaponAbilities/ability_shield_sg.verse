using { /Fortnite.com/Characters }
using { /Fortnite.com/Game }
using { /Verse.org/SceneGraph }
using { /Verse.org/Simulation }
using { /UnrealEngine.com/Temporary/SpatialMath }
using { Prefabs }
using { VisualEffects }

<# ----------------------------------------------------------------------------------------------------
    Constants
---------------------------------------------------------------------------------------------------- #>
DefaultAbilityShieldAttackTime:float = 0.3 # [Note] Needs to synchronize with Shield Blash VFX timing.

<# ----------------------------------------------------------------------------------------------------
    ability_shield_sg
---------------------------------------------------------------------------------------------------- #>
ability_shield_sg<public> := class(ability):
    Id<override>:weapon_id = weapon_id.Shield
    
    var ShieldMaxHealth:float = 0.5
    var ShieldRegenerationRate:float = 0.125
    var ShieldIsActive:logic = false
    var MaybeShieldDamagedSubscription:?cancelable = false
    var MaybeWeaponPrefab:?prefab_weapon_shield_blast = false
    var MaybePreviousVFXId:?vfx_powerup_id = false

    <# ----------------------------------------------------------------------------------------------------
        Initialization
    ---------------------------------------------------------------------------------------------------- #>
    Deactivate<override>():void =
        DeactivatedEvent.Signal()
        Player.DeactivateVFXPowerup(GetVFXId())

        set CooldownTimeRemaining = Max(0.0, AttackCooldownTime - (GetSimulationElapsedTime() - CastTime))
        set IsActive = false

    Deinitialize<override>():void =
        (super:)Deinitialize()

        if (ShieldDamagedSubscription := MaybeShieldDamagedSubscription?):
            ShieldDamagedSubscription.Cancel()
            set MaybeShieldDamagedSubscription = false
        
        if (WeaponPrefab := MaybeWeaponPrefab?):
            WeaponPrefab.RemoveFromParent()

        set MaybeWeaponPrefab = false
        set MaybePreviousVFXId = false

    <# ----------------------------------------------------------------------------------------------------
        Get Functions
    ---------------------------------------------------------------------------------------------------- #>
    GetVFXId()<transacts>:vfx_powerup_id =
        EvolutionLevel := Player.GetWeaponEvolutionLevel(Id)
        case (EvolutionLevel):
            1 => vfx_powerup_id.ShieldEvolved1
            2 => vfx_powerup_id.ShieldEvolved2
            3 => vfx_powerup_id.ShieldEvolved3
            _ => vfx_powerup_id.Shield

    <# ----------------------------------------------------------------------------------------------------
        Update Functions
    ---------------------------------------------------------------------------------------------------- #>
    Activate<override>():void =
        (super:)Activate()

        if (Character := Player.GetFortCharacter[]):
            if (not MaybeShieldDamagedSubscription?):
                ShieldDamagedSubscription := Character.DamagedShieldEvent().Subscribe(OnDamagedShield)
                set MaybeShieldDamagedSubscription = option{ShieldDamagedSubscription}
            
            if (Character.GetShield() > 0.0):
                Player.ActivateVFXPowerup(GetVFXId())

    Ability<override>()<suspends>:void =
        if (not MaybeWeaponPrefab?):
            WeaponPrefab := prefab_weapon_shield_blast{}
            WeaponPrefab.InitializeParticle()
            set MaybeWeaponPrefab = option{WeaponPrefab}

        if (PreviousVFXId := MaybePreviousVFXId?, CurrentVFXId := GetVFXId(), CurrentVFXId <> PreviousVFXId):
            Sleep(0.0)
            Player.DeactivateVFXPowerup(PreviousVFXId)
            Sleep(0.0)

        set MaybePreviousVFXId = option{GetVFXId()}

        race:
            DeactivatedEvent.Await()
            if (Character := Player.GetFortCharacter[]):
                Sleep(CooldownTimeRemaining)
                loop:
                    MaxShield := Character.GetMaxHealth() * ShieldMaxHealth
                    ShieldAmount := MaxShield * ShieldRegenerationRate

                    set ShieldIsActive = true
                    
                    if (Character.GetShield() < MaxShield):
                        set CastTime = GetSimulationElapsedTime()
                        Character.SetShield(Min(MaxShield, Character.GetShield() + ShieldAmount))
                        Player.PlayAudio(audio_id.Shield)
                        Player.ActivateVFXPowerup(GetVFXId())
                        
                    Sleep(AttackCooldownTime)

    Damage()<suspends>:void =
        if:
            Character := Player.GetFortCharacter[]
            WeaponPrefab := MaybeWeaponPrefab?
            ParticleSystem := NS_AbilityShieldBlast_v2[WeaponPrefab.GetComponent[particle_system_component]]
            SimulationEntity := GetSimulationEntity[]
        then:
            IsCracked := logic{Character.GetShield() = 0.0}
            AbilityPosition := Character.GetTransform().Translation.Make2D()
            FinalAttackRadius := AttackRadius * ((IsUniqueEvolution? and IsCracked?) and 1.5 or 1.0)
            Collision:collision_sphere = collision_sphere{Radius := FinalAttackRadius}

            set ParticleSystem.Color = GetWeaponColor(Player.GetWeaponEvolutionLevel(Id))
            set ParticleSystem.Radius = FinalAttackRadius * 0.1
            set ParticleSystem.Position = FromVector3 of AbilityPosition
        
            ParticleSystem.Play()
            Sleep(DefaultAbilityShieldAttackTime)

            for:
                OverlapHit : SimulationEntity.FindOverlapHits(FromTransform of Character.GetTransform(), Collision)
                TargetComponent := OverlapHit.TargetComponent
                TargetEntity := TargetComponent.Entity
                NPCCollisionComponent := TargetEntity.GetComponent[npc_collision_component]
                NPCHit := NPCCollisionComponent.MaybeNPC?
                NPCHit.IsDamageable()?
                NPCPosition := NPCHit.GetAgent[].GetFortCharacter[].GetTransform().Translation.Make2D()
            do:
                if (IsUniqueEvolution? and IsCracked?, Direction := (NPCPosition - AbilityPosition).MakeUnitVector[]):
                    DirectionRotation := MakeRotation(vector3{X := 0.0, Y := 0.0, Z := 1.0}, ArcTan(Direction.Y, Direction.X))
                    NPCHit.Knockback(knockback_properties{Duration := 1.0, Rotation := DirectionRotation})

                Damage(NPCHit, Player.GetDamageArgs(NPCHit, Id, AttackDamage * ((IsUniqueEvolution? and IsCracked?) and 1.5 or 1.0), ?Multiplier := AttackDamageMultiplier))

            if (IsCracked?):
                Player.DeactivateVFXPowerup(GetVFXId())
                set ShieldIsActive = false

    UpdateStats<override>():void =
        (super:)UpdateStats()
        if (WeaponInfo := GetWeaponInfo[Id]):
            set ShieldMaxHealth = Player.GetWeaponEvolutionValue(Id, ability_id.ShieldMaxHealth)
            set ShieldRegenerationRate = Player.GetWeaponEvolutionValue(Id, ability_id.ShieldRegenerationRate) * AttackCooldownTimeDamageMultiplier

    <# ----------------------------------------------------------------------------------------------------
        Event Handlers
    ---------------------------------------------------------------------------------------------------- #>
    OnDamagedShield(DamageResult:damage_result):void =
        if (ShieldIsActive?):
            spawn{Damage()}