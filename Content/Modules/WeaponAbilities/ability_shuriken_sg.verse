using { /Fortnite.com/Characters }
using { /Verse.org/SceneGraph }
using { /Verse.org/Simulation }
using { /UnrealEngine.com/Temporary/SpatialMath }
using { Prefabs }

<# ----------------------------------------------------------------------------------------------------
    Constants
---------------------------------------------------------------------------------------------------- #>
DefaultAbilityRotatorSpeed:float = PiFloat
DefaultAbilityRotatorSize:float = 400.0
DefaultAbilityRotatorRevolutionTime:float = PiFloat * 2 / DefaultAbilityRotatorSpeed
DefaultAbilityRotatorDuration:float = DefaultAbilityRotatorRevolutionTime * 3.0

<# ----------------------------------------------------------------------------------------------------
    ability_shuriken_sg
---------------------------------------------------------------------------------------------------- #>
ability_shuriken_sg<public> := class(ability):
    Id<override>:weapon_id = weapon_id.Shuriken

    CancelEvent:event() = event(){}

    var MulticastCount:int = 1
    var ProjectileSpeed:float = 1.0
    var IsInfinite:logic = false
    var MaybeWeaponPrefab:?prefab_weapon_shuriken_v2 = false

    <# ----------------------------------------------------------------------------------------------------
        Initialization
    ---------------------------------------------------------------------------------------------------- #>
    Deactivate<override>():void =
        DeactivatedEvent.Signal()
        Player.DeactivateVFXPowerup(vfx_powerup_id.ShurikenBase)
        ElapsedTime := GetSimulationElapsedTime() - CastTime
        set CooldownTimeRemaining = IsInfinite? and 0.0 or Max(0.0, AttackCooldownTime - ElapsedTime)
        set IsActive = false
        
        if (IsInfinite?):
            CancelEvent.Signal()

    Deinitialize<override>():void =
        (super:)Deinitialize()
        
        if (WeaponPrefab := MaybeWeaponPrefab?):
            WeaponPrefab.RemoveFromParent()
            set MaybeWeaponPrefab = false

    <# ----------------------------------------------------------------------------------------------------
        Get Functions
    ---------------------------------------------------------------------------------------------------- #>
    GetElapsedTime(Time:float)<transacts>:float =
        GetSimulationElapsedTime() - Time        
    
    <# ----------------------------------------------------------------------------------------------------
        Update Functions
    ---------------------------------------------------------------------------------------------------- #>
    Activate<override>():void =
        (super:)Activate()
        Player.ActivateVFXPowerup(vfx_powerup_id.ShurikenBase)

    Ability<override>()<suspends>:void =
        if (not MaybeWeaponPrefab?):
            WeaponPrefab := prefab_weapon_shuriken_v2{}
            WeaponPrefab.InitializeParticle()
            set MaybeWeaponPrefab = option{WeaponPrefab}

        if:
            AbilityShuriken := VisualEffects.NS_AbilityShuriken_v2[MaybeWeaponPrefab?.GetComponent[particle_system_component]]
        then:
            set AbilityShuriken.Color = GetWeaponColor(Player.GetWeaponEvolutionLevel(Id))
            set AbilityShuriken.ProjectileCount = ProjectileCount
            set AbilityShuriken.Radius = AttackRadius * 0.1
            #set AbilityShuriken.RotationRate = ProjectileSpeed / (2 * PiFloat)
            

        race:
            DeactivatedEvent.Await()

            block:
                Sleep(CooldownTimeRemaining)

                if (IsInfinite?):
                    set CastTime = GetSimulationElapsedTime()
                    Player.PlayAudio(audio_id.Projectile)

                    for (Index := 1..ProjectileCount):
                        spawn{Shuriken(Index)}
                    
                    spawn{ActivateVFX()}
                    Sleep(Inf)

                else:
                    loop:
                        set CastTime = GetSimulationElapsedTime()
                        Player.PlayAudio(audio_id.Projectile)

                        for (Index := 1..ProjectileCount):
                            spawn{Shuriken(Index)}
                            
                        spawn{ActivateVFX()}
                        Sleep(AttackCooldownTime)

    ActivateVFX()<suspends>:void =
        if:
            Character := Player.GetFortCharacter[]
            AbilityShuriken := VisualEffects.NS_AbilityShuriken_v2[MaybeWeaponPrefab?.GetComponent[particle_system_component]]
        then:
            set AbilityShuriken.RotationPhase = 0.0
            set AbilityShuriken.Position = FromVector3 of Character.GetTransform().Translation.Make2D()

            AbilityShuriken.Enable()

            race:
                CancelEvent.Await()
                Sleep(IsInfinite? and Inf or DefaultAbilityRotatorDuration)
                loop:
                    Sleep(TwoTicks)
                    set AbilityShuriken.RotationPhase = (GetSimulationElapsedTime() - CastTime) * ProjectileSpeed / (2 * PiFloat)
                    set AbilityShuriken.Position = FromVector3 of Character.GetTransform().Translation.Make2D()

            AbilityShuriken.Disable()

    Shuriken(Index:int)<suspends>:void =
        if (Character := Player.GetFortCharacter[], SimulationEntity := GetSimulationEntity[]):
            DelayTime := ((Index - 1) * 1.0) / (ProjectileCount * 1.0) * DefaultAbilityRotatorRevolutionTime
            Collision := collision_sphere{Radius := AttackRadius}
            AngleOffset := PiFloat * 2.0 / (ProjectileCount * 1.0) * (Index - 1)

            race:
                CancelEvent.Await()
                Sleep(IsInfinite? and Inf or DefaultAbilityRotatorDuration)
                block:
                    Sleep(DelayTime)
                    if (IsUniqueEvolution?):
                        ShurikenMastery(SimulationEntity, Collision, Character, AngleOffset)
                    else:
                        Shuriken(SimulationEntity, Collision, Character, AngleOffset)

    ShurikenMastery(SimulationEntity:entity, Collision:collision_sphere, Character:fort_character, AngleOffset:float)<suspends>:void =
        loop:
            Angle := GetElapsedTime(CastTime) * ProjectileSpeed + AngleOffset
            Position := Character.GetTransform().Translation + vector3{X := Cos(Angle) * DefaultAbilityRotatorSize, Y := Sin(Angle) * DefaultAbilityRotatorSize}

            for:
                OverlapHit : SimulationEntity.FindOverlapHits(FromTransform of transform{Translation := Position}, Collision)
                TargetComponent := OverlapHit.TargetComponent
                TargetEntity := TargetComponent.Entity
                NPCCollisionComponent := TargetEntity.GetComponent[npc_collision_component]
                NPCHit := NPCCollisionComponent.MaybeNPC?
                NPCHit.IsDamageable()?
            do:
                Damage(NPCHit, Player.GetDamageArgs(NPCHit, Id, AttackDamage, ?Multiplier := AttackDamageMultiplier))

            Sleep(0.5)

    Shuriken(SimulationEntity:entity, Collision:collision_sphere, Character:fort_character, AngleOffset:float)<suspends>:void =
        var NPCsHit:[npc_behavior_base]logic = map{}

        race:
            loop:
                Sleep(DefaultAbilityRotatorRevolutionTime)
                set NPCsHit = map{}

            loop:
                Angle := GetElapsedTime(CastTime) * ProjectileSpeed + AngleOffset
                Position := Character.GetTransform().Translation + vector3{X := Cos(Angle) * DefaultAbilityRotatorSize, Y := Sin(Angle) * DefaultAbilityRotatorSize}

                for:
                    OverlapHit : SimulationEntity.FindOverlapHits(FromTransform of transform{Translation := Position}, Collision)
                    TargetComponent := OverlapHit.TargetComponent
                    TargetEntity := TargetComponent.Entity
                    NPCCollisionComponent := TargetEntity.GetComponent[npc_collision_component]
                    NPCHit := NPCCollisionComponent.MaybeNPC?
                    NPCHit.IsDamageable()?
                    not NPCsHit[NPCHit]
                    set NPCsHit[NPCHit] = true
                do:
                    Damage(NPCHit, Player.GetDamageArgs(NPCHit, Id, AttackDamage, ?Multiplier := AttackDamageMultiplier))

                Sleep(ThreeTicks)

    UpdateStats<override>():void =
        (super:)UpdateStats()
        if (WeaponInfo := GetWeaponInfo[Id]):
            set ProjectileCount = 1 + Floor[Player.GetWeaponEvolutionValue(Id, ability_id.ProjectileCount)] or 0
            set ProjectileSpeed = DefaultAbilityRotatorSpeed * (1.0 - Player.GetWeaponEvolutionValue(Id, ability_id.ProjectileSpeed) / 3.0) * (IsUniqueEvolution? and 0.5 or 1.0)

            set IsInfinite = logic{AttackCooldownTime <= WeaponInfo.MinCooldown}

            set AttackDamage = WeaponInfo.Damage * (IsUniqueEvolution? and 2.0 or 1.0) * (1.0 + Player.GetWeaponEvolutionValue(Id, ability_id.ProjectileSpeed)) * AttackCooldownTimeDamageMultiplier
            set AttackRadius = WeaponInfo.Size * (IsUniqueEvolution? and 2.0 or 1.0) * (1.0 + Player.GetWeaponStats(Id, stat_id.Size) + Player.GetStat(stat_id.Size))