using { /Fortnite.com/Characters }
using { /Fortnite.com/Devices }
using { /Verse.org }
using { /Verse.org/SceneGraph }
using { /Verse.org/Simulation }
using { /UnrealEngine.com/Temporary/SpatialMath }
using { Prefabs }
using { VisualEffects }

<# ----------------------------------------------------------------------------------------------------
    Constants
---------------------------------------------------------------------------------------------------- #>
DefaultAbilityStaffLength:float = 500.0
DefaultAbilityStaffAttackTime:float = 0.2 # [Note] Requires updating the VFX.
DefaultAbilityStaffThrowRange:float = 800.0
DefaultAbilityStaffThrowSpeed:float = 200.0
DefaultAbilityStaffSwirlSize:float = 400.0
DefaultAbilityStaffSwirlDuration:float = 2.0 # [Note] Requires updating the VFX.
DefaultAbilityStaffSwirlSpeed:float = PiFloat / DefaultAbilityStaffSwirlDuration

<# ----------------------------------------------------------------------------------------------------
    Data Definitions
---------------------------------------------------------------------------------------------------- #>
ability_staff_type<public> := enum:
    Smash
    Thrust
    Pillar
    Throw
    Swirl

StaffAbilityIds:[]ability_id = array:
    ability_id.StaffThrust
    ability_id.StaffThrow
    ability_id.StaffSwirl
    ability_id.StaffPillar

(Id:ability_id).ToAbilityStaffType()<transacts>:ability_staff_type =
    case (Id):
        ability_id.StaffThrow => ability_staff_type.Throw
        ability_id.StaffThrust => ability_staff_type.Thrust
        ability_id.StaffSwirl => ability_staff_type.Swirl
        ability_id.StaffPillar => ability_staff_type.Pillar
        _ => ability_staff_type.Smash

(Type:ability_staff_type).MakeWeaponPrefab():entity =
    case (Type):
        ability_staff_type.Thrust => prefab_weapon_staff_thrust{}
        ability_staff_type.Pillar => prefab_weapon_staff_pillar{}
        ability_staff_type.Throw => prefab_weapon_staff_throw{}
        ability_staff_type.Swirl => prefab_weapon_staff_swirl{}
        _ => prefab_weapon_staff_smash{}

<# ----------------------------------------------------------------------------------------------------
    ability_staff_sg
---------------------------------------------------------------------------------------------------- #>
ability_staff_sg<public> := class(ability):
    Id<override>:weapon_id = weapon_id.Staff

    var MulticastCount:int = 1
    var CastCount:int = 0
    
    var IntervalCount:int = 1
    var UniqueInterval:int = 4
    var IsUniqueCast:logic = false
    
    var AbilityStaffTypes:[]ability_staff_type = array{}
    var WeaponPrefabs:[ability_staff_type]entity = map{}

    <# ----------------------------------------------------------------------------------------------------
        Initialization
    ---------------------------------------------------------------------------------------------------- #>
    Deactivate<override>():void =
        DeactivatedEvent.Signal()
        set CooldownTimeRemaining = Max(0.0, AttackCooldownTime - (GetSimulationElapsedTime() - CastTime))
        set IsActive = false

    Deinitialize<override>():void =
        (super:)Deinitialize()
        
        for (WeaponPrefab : WeaponPrefabs):
            WeaponPrefab.RemoveFromParent()

        set WeaponPrefabs = map{}

    <# ----------------------------------------------------------------------------------------------------
        Get Functions
    ---------------------------------------------------------------------------------------------------- #>
    GetAbility(Type:ability_staff_type)<transacts>:type{_(:vector3, :rotation, :logic)<suspends>:void} =
        case (Type):
            ability_staff_type.Smash => Smash
            ability_staff_type.Thrust => Thrust
            ability_staff_type.Pillar => Pillar
            ability_staff_type.Throw => Throw
            ability_staff_type.Swirl => Swirl

    <# ----------------------------------------------------------------------------------------------------
        Update Functions
    ---------------------------------------------------------------------------------------------------- #>
    Ability<override>()<suspends>:void =
        if (not WeaponPrefabs[ability_staff_type.Smash]):
            WeaponPrefab := prefab_weapon_staff_smash{}
            WeaponPrefab.InitializeParticle()
            if (set WeaponPrefabs[ability_staff_type.Smash] = WeaponPrefab) {}

        for:
            AbilityStaffType : AbilityStaffTypes
            not WeaponPrefabs[AbilityStaffType]
        do:
            WeaponPrefab := AbilityStaffType.MakeWeaponPrefab()
            WeaponPrefab.InitializeParticle()
            if (set WeaponPrefabs[AbilityStaffType] = WeaponPrefab) {}

        race:
            DeactivatedEvent.Await()
            if (WeaponInfo := GetWeaponInfo[Id]):
                Sleep(CooldownTimeRemaining)
                loop:
                    set CastTime = GetSimulationElapsedTime()
                    spawn{Ability(WeaponInfo)}
                    Sleep(AttackCooldownTime)

    
    Ability(WeaponInfo:weapon_info)<suspends>:void =
        if (Character := Player.GetFortCharacter[]):
            TargetPosition := Character.GetTransform().Translation
            TargetRotation := Player.GetAutoAimRotation()

            set IsUniqueCast = logic{IsUniqueEvolution? and Mod[IntervalCount, UniqueInterval] = 0}

            if (AbilityStaffType := AbilityStaffTypes[CastCount]):
                spawn{GetAbility(AbilityStaffType)(TargetPosition, TargetRotation, IsUniqueCast)}
                Player.PlayAudio(audio_id.Projectile)
       
        set CastCount = if (CastCount >= AbilityStaffTypes.Length - 1):
            if (IsUniqueCast?):
                set IntervalCount = 1
            else:
                set IntervalCount += 1

            0
        else:
            CastCount + 1

    <# ----------------------------------------------------------------------------------------------------
        Staff Types
    ---------------------------------------------------------------------------------------------------- #>
    Smash(Position:vector3, Rotation:rotation, IsUnique:logic)<suspends>:void =
        AttackLength := DefaultAbilityStaffLength + AttackRadius
        AttackWidth := 35.0 + (AttackRadius * 0.5)
        Collision := collision_box:
            Extents := SpatialMath.vector3{Left := -AttackWidth / 2.0, Up := 100.0, Forward := AttackLength / 2.0}
        
        if:
            WeaponPrefab := WeaponPrefabs[ability_staff_type.Smash]
            ParticleSystem := NS_AbilityStaff_v2[WeaponPrefab.GetComponent[particle_system_component]]
        then:
            set ParticleSystem.Color = GetWeaponColor(Player.GetWeaponEvolutionLevel(Id))
            set ParticleSystem.Length = AttackLength * 0.1
            set ParticleSystem.Width = AttackWidth * 0.1
            set ParticleSystem.ProjectileCount = IsUnique? and 4 or 1
            set ParticleSystem.ProjectileMulticast = MulticastCount
            set ParticleSystem.Position = FromVector3 of Position.Make2D()
            set ParticleSystem.Rotation = Rotation.GetYawPitchRollDegrees()[0] or 0.0

            ParticleSystem.Play()

        Sleep(DefaultAbilityStaffAttackTime)

        for (Index := 1..(IsUnique? and 4 or 1)):
            TargetRotation := Rotation.ApplyYaw(DegreesToRadians((Index - 1) * 90.0))
            TargetPosition := Position + TargetRotation.GetLocalForward() * AttackLength / 2.0

            for:
                OverlapHit : GetSimulationEntity[].FindOverlapHits(FromTransform of transform{Translation := TargetPosition, Rotation := TargetRotation}, Collision)
                TargetComponent := OverlapHit.TargetComponent
                TargetEntity := TargetComponent.Entity
                NPCCollisionComponent := TargetEntity.GetComponent[npc_collision_component]
                NPCHit := NPCCollisionComponent.MaybeNPC?
                NPCHit.IsDamageable()?
            do:
                Damage(NPCHit, Player.GetDamageArgs(NPCHit, Id, AttackDamage, ?Multiplier := AttackDamageMultiplier))

    Thrust(Position:vector3, Rotation:rotation, IsUnique:logic)<suspends>:void =
        AttackLength := DefaultAbilityStaffLength + AttackRadius
        AttackWidth := 35.0 + (AttackRadius * 0.5)
        Collision := collision_box:
            Extents := SpatialMath.vector3{Left := -AttackWidth / 2.0, Up := 100.0, Forward := AttackLength / 2.0}

        if:
            WeaponPrefab := WeaponPrefabs[ability_staff_type.Thrust]
            ParticleSystem := NS_AbilityStaff_Thrust_v2[WeaponPrefab.GetComponent[particle_system_component]]
        then:
            set ParticleSystem.Color = GetWeaponColor(Player.GetWeaponEvolutionLevel(Id))
            set ParticleSystem.Length = AttackLength * 0.1
            set ParticleSystem.Width = AttackWidth * 0.1
            set ParticleSystem.ProjectileCount = IsUnique? and 4 or 1
            set ParticleSystem.ProjectileMulticast = MulticastCount
            set ParticleSystem.Position = FromVector3 of Position.Make2D()
            set ParticleSystem.Rotation = Rotation.GetYawPitchRollDegrees()[0] or 0.0

            ParticleSystem.Play()

        Sleep(DefaultAbilityStaffAttackTime)

        for (Index := 1..(IsUnique? and 4 or 1)):
            TargetRotation := Rotation.ApplyYaw(DegreesToRadians((Index - 1) * 90.0))
            TargetPosition := Position + TargetRotation.GetLocalForward() * AttackLength / 2.0

            for:
                OverlapHit : GetSimulationEntity[].FindOverlapHits(FromTransform of transform{Translation := TargetPosition, Rotation := TargetRotation}, Collision)
                TargetComponent := OverlapHit.TargetComponent
                TargetEntity := TargetComponent.Entity
                NPCCollisionComponent := TargetEntity.GetComponent[npc_collision_component]
                NPCHit := NPCCollisionComponent.MaybeNPC?
                NPCHit.IsDamageable()?
            do:
                Damage(NPCHit, Player.GetDamageArgs(NPCHit, Id, AttackDamage * 2.0, ?Multiplier := AttackDamageMultiplier))
            
        Sleep(0.1)

        for (Index := 1..(IsUnique? and 4 or 1)):
            TargetRotation := Rotation.ApplyYaw(DegreesToRadians((Index - 1) * 90.0))
            TargetPosition := Position + TargetRotation.GetLocalForward() * AttackLength / 2.0

            for:
                OverlapHit : GetSimulationEntity[].FindOverlapHits(FromTransform of transform{Translation := TargetPosition, Rotation := TargetRotation}, Collision)
                TargetComponent := OverlapHit.TargetComponent
                TargetEntity := TargetComponent.Entity
                NPCCollisionComponent := TargetEntity.GetComponent[npc_collision_component]
                NPCHit := NPCCollisionComponent.MaybeNPC?
                NPCHit.IsDamageable()?
            do:
                NPCHit.Knockback(knockback_properties{Duration := 3.0, Rotation := TargetRotation})

    Pillar(Position:vector3, Rotation:rotation, IsUnique:logic)<suspends>:void =
        Range := 150.0 + AttackRadius

        if:
            WeaponPrefab := WeaponPrefabs[ability_staff_type.Pillar]
            ParticleSystem := NS_AbilityStaff_Pillar_v2[WeaponPrefab.GetComponent[particle_system_component]]
        then:
            set ParticleSystem.Color = GetWeaponColor(Player.GetWeaponEvolutionLevel(Id))
            set ParticleSystem.Radius = AttackRadius * 0.1
            set ParticleSystem.Range = Range
            set ParticleSystem.ProjectileCount = IsUnique? and 4 or 1
            set ParticleSystem.ProjectileMulticast = MulticastCount
            set ParticleSystem.Position = FromVector3 of Position.Make2D()
            set ParticleSystem.Rotation = Rotation.GetYawPitchRollDegrees()[0] or 0.0
            
            ParticleSystem.Play()

        Sleep(DefaultAbilityStaffAttackTime)

        for (Index := 1..(IsUnique? and 4 or 1)):
            ProjectileRotation := Rotation.ApplyYaw(DegreesToRadians((Index - 1) * 90.0))
            Direction := ProjectileRotation.GetLocalForward()
            TargetPosition := Position + Direction * Range

            for:
                OverlapHit : GetSimulationEntity[].FindOverlapHits(FromTransform of transform{Translation := TargetPosition}, collision_sphere{Radius := AttackRadius})
                TargetComponent := OverlapHit.TargetComponent
                TargetEntity := TargetComponent.Entity
                NPCCollisionComponent := TargetEntity.GetComponent[npc_collision_component]
                NPCHit := NPCCollisionComponent.MaybeNPC?
                NPCHit.IsDamageable()?
            do:
                Damage(NPCHit, Player.GetDamageArgs(NPCHit, Id, AttackDamage * 3.0, ?Multiplier := AttackDamageMultiplier))

    Throw(Position:vector3, Rotation:rotation, IsUnique:logic)<suspends>:void =
        Lifetime := DefaultAbilityStaffThrowRange / DefaultAbilityStaffThrowSpeed
        ProjectileStartTime := GetSimulationElapsedTime()

        if:
            WeaponPrefab := WeaponPrefabs[ability_staff_type.Throw]
            ParticleSystem := NS_AbilityStaff_Throw_v2[WeaponPrefab.GetComponent[particle_system_component]]
        then:
            set ParticleSystem.Color = GetWeaponColor(Player.GetWeaponEvolutionLevel(Id))
            set ParticleSystem.Radius = AttackRadius * 0.1
            set ParticleSystem.ProjectileDistance = DefaultAbilityStaffThrowRange
            set ParticleSystem.ProjectileSpeed = DefaultAbilityStaffThrowSpeed
            set ParticleSystem.ProjectileCount = IsUnique? and 4 or 1
            set ParticleSystem.ProjectileMulticast = MulticastCount
            set ParticleSystem.Position = FromVector3 of Position.Make2D()
            set ParticleSystem.Rotation = Rotation.GetYawPitchRollDegrees()[0] or 0.0

            ParticleSystem.Play()
        
        loop:
            for (Index := 1..(IsUnique? and 4 or 1)):
                ProjectileRotation := Rotation.ApplyYaw(DegreesToRadians((Index - 1) * 90.0))
                TargetPosition := Position + ProjectileRotation.GetLocalForward() * DefaultAbilityStaffThrowRange
                ElapsedTime := GetSimulationElapsedTime() - ProjectileStartTime
                Progress := ElapsedTime / Lifetime
                CurrentPosition := Lerp(Position, TargetPosition, Progress)

                for:
                    OverlapHit : GetSimulationEntity[].FindOverlapHits(FromTransform of transform{Translation := CurrentPosition}, collision_sphere{Radius := AttackRadius})
                    TargetComponent := OverlapHit.TargetComponent
                    TargetEntity := TargetComponent.Entity
                    NPCCollisionComponent := TargetEntity.GetComponent[npc_collision_component]
                    NPCHit := NPCCollisionComponent.MaybeNPC?
                    NPCHit.IsDamageable()?
                do:
                    Damage(NPCHit, Player.GetDamageArgs(NPCHit, Id, AttackDamage, ?Multiplier := AttackDamageMultiplier))

                if (ElapsedTime >= Lifetime):
                    break

            Sleep(0.5)

    Swirl(Position:vector3, Rotation:rotation, IsUnique:logic)<suspends>:void =
        DeltaTimer := delta_timer{}
        FinalAttackRadius := AttackRadius * 0.75

        if:
            WeaponPrefab := WeaponPrefabs[ability_staff_type.Swirl]
            ParticleSystem := NS_AbilityStaff_Swirl_v2[WeaponPrefab.GetComponent[particle_system_component]]
        then:
            set ParticleSystem.Color = GetWeaponColor(Player.GetWeaponEvolutionLevel(Id))
            set ParticleSystem.Radius = FinalAttackRadius * 0.1
            set ParticleSystem.Range = DefaultAbilityStaffSwirlSize
            set ParticleSystem.Duration = DefaultAbilityStaffSwirlDuration
            set ParticleSystem.ProjectileCount = IsUnique? and 4 or 1
            set ParticleSystem.ProjectileMulticast = MulticastCount
            set ParticleSystem.Position = FromVector3 of Position.Make2D()
            set ParticleSystem.Rotation = Rotation.GetYawPitchRollDegrees()[0] or 0.0

            ParticleSystem.Play()

        var NPCsHit:[int][npc_behavior_base]logic = map{}
        for (Index := 1..(IsUnique? and 4 or 1), set NPCsHit[Index] = map{}) {}

        race:
            Sleep(DefaultAbilityStaffSwirlDuration)
            loop:
                for (Index := 1..(IsUnique? and 4 or 1)):
                    Angle := DeltaTimer.GetElapsedTime() * (2.0 * DefaultAbilityStaffSwirlSpeed) + Rotation.GetAngle() + DegreesToRadians((Index - 1) * 90.0)
                    TargetPosition := Position + vector3{X := Cos(Angle) * DefaultAbilityStaffSwirlSize, Y := Sin(Angle) * DefaultAbilityStaffSwirlSize}

                    for:
                        OverlapHit : GetSimulationEntity[].FindOverlapHits(FromTransform of transform{Translation := TargetPosition}, collision_sphere{Radius := FinalAttackRadius})
                        TargetComponent := OverlapHit.TargetComponent
                        TargetEntity := TargetComponent.Entity
                        NPCCollisionComponent := TargetEntity.GetComponent[npc_collision_component]
                        NPCHit := NPCCollisionComponent.MaybeNPC?
                        NPCHit.IsDamageable()?
                        not NPCsHit[Index][NPCHit]
                        set NPCsHit[Index][NPCHit] = true
                    do:
                        Damage(NPCHit, Player.GetDamageArgs(NPCHit, Id, AttackDamage * 2.0, ?Multiplier := AttackDamageMultiplier))

                Sleep(ThreeTicks)

    <# ----------------------------------------------------------------------------------------------------
        Update Stats
    ---------------------------------------------------------------------------------------------------- #>
    UpdateStats<override>():void =
        (super:)UpdateStats()
        if (WeaponInfo := GetWeaponInfo[Id]):
            set ProjectileCount = 1 + Floor[Player.GetWeaponEvolutionValue(Id, ability_id.ProjectileCount)] or 0
            set MulticastCount = 1 + Floor[Player.GetWeaponEvolutionValue(Id, ability_id.Multicast)] or 0
            set AbilityStaffTypes = array{ability_staff_type.Smash}

            for (StaffAbilityId : StaffAbilityIds, Player.IsWeaponEvolutionAbilityActive(Id, StaffAbilityId)?):
                set AbilityStaffTypes += array{StaffAbilityId.ToAbilityStaffType()}

            set AttackDamage *= MulticastCount * 1.0