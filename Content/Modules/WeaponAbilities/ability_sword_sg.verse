using { /Fortnite.com/Characters }
using { /Verse.org/SceneGraph }
using { /Verse.org/Simulation }
using { /UnrealEngine.com/Temporary/SpatialMath }
using { Prefabs }

<# ----------------------------------------------------------------------------------------------------
    Constants
---------------------------------------------------------------------------------------------------- #>
BaseAbilityConeTime:float = 0.1
BaseAbilityConeAngle<public>:float = 120.0 # Requires updating angle in VFX
BaseAbilityConeAngleBonus:float = 45.0

<# ----------------------------------------------------------------------------------------------------
    ability_sword_sg
---------------------------------------------------------------------------------------------------- #>
ability_sword_sg<public> := class(ability):
    Id<override>:weapon_id = weapon_id.Sword

    var AttackAngle:float = BaseAbilityConeAngle
    var MaybeWeaponPrefabs:?prefab_weapon_sword = false
   
    <# ----------------------------------------------------------------------------------------------------
        Initialization
    ---------------------------------------------------------------------------------------------------- #>
    Deactivate<override>():void =
        DeactivatedEvent.Signal()
        set CooldownTimeRemaining = Max(0.0, AttackCooldownTime - (GetSimulationElapsedTime() - CastTime))
        set IsActive = false

    Deinitialize<override>():void =
        (super:)Deinitialize()

        if (WeaponPrefabs := MaybeWeaponPrefabs?):
            WeaponPrefabs.RemoveFromParent()
        
        set MaybeWeaponPrefabs = false

    <# ----------------------------------------------------------------------------------------------------
        Update Functions
    ---------------------------------------------------------------------------------------------------- #>
    Ability<override>()<suspends>:void =
        if (not MaybeWeaponPrefabs?):
            WeaponPrefab := prefab_weapon_sword{}
            WeaponPrefab.InitializeParticle()
            set MaybeWeaponPrefabs = option{WeaponPrefab}

        if:
            WeaponPrefab := MaybeWeaponPrefabs?
            AbilitySword := VisualEffects.NS_AbilitySword_v2[WeaponPrefab.GetComponent[particle_system_component]]
        then:
            set AbilitySword.Color = GetWeaponColor(Player.GetWeaponEvolutionLevel(Id))
            set AbilitySword.Radius = AttackRadius * 0.1
            set AbilitySword.Angle = BaseAbilityConeAngle * (IsUniqueEvolution? and 0.5 or 1.0)
            set AbilitySword.ProjectileCount = if (AttackAngle <= BaseAbilityConeAngle):
                1
            else if (AttackAngle <= BaseAbilityConeAngle * 2):
                2
            else:
                3

            if (not AbilitySword.IsEnabled[]):
                AbilitySword.Enable()
    
        race:
            DeactivatedEvent.Await()
            block:
                Sleep(CooldownTimeRemaining)
                loop:
                    set CastTime = GetSimulationElapsedTime()
                    spawn{Cast()}
                    Sleep(AttackCooldownTime)

    Cast()<suspends>:void =
        if:
            Character := Player.GetFortCharacter[]
            WeaponPrefab := MaybeWeaponPrefabs?
            AbilitySword := VisualEffects.NS_AbilitySword_v2[WeaponPrefab.GetComponent[particle_system_component]]
            SimulationEntity := GetSimulationEntity[]
        then:
            Collision := collision_sphere{Radius := AttackRadius}

            for (CastIndex := 1..ProjectileCount):
                TargetPosition := Character.GetTransform().Translation
                TargetRotation := Player.GetAutoAimRotation()

                set AbilitySword.Position = FromVector3 of TargetPosition.Make2D()
                set AbilitySword.Rotation = TargetRotation.GetYawPitchRollDegrees()[0] or 0.0
                AbilitySword.Play()
                Player.PlayAudio(audio_id.Projectile)
                Sleep(BaseAbilityConeTime)

                for:
                    OverlapHit : SimulationEntity.FindOverlapHits(FromTransform of transform{Translation := TargetPosition}, Collision)
                    TargetComponent := OverlapHit.TargetComponent
                    TargetEntity := TargetComponent.Entity
                    NPCCollisionComponent := TargetEntity.GetComponent[npc_collision_component]
                    NPCHit := NPCCollisionComponent.MaybeNPC?
                    NPCHit.IsDamageable()?

                    AttackAngle >= 360.0 or (
                        NPCDirection := NPCHit.GetAgent[].GetFortCharacter[].GetTransform().Translation.Make2D() - TargetPosition.Make2D()
                        Angle := RadiansToDegrees(MakeShortestRotationBetween(TargetRotation.GetLocalForward(), NPCDirection).GetAngle())
                        Angle < (AttackAngle + BaseAbilityConeAngleBonus) / 2.0
                    )
                do:
                    Damage(NPCHit, Player.GetDamageArgs(NPCHit, Id, AttackDamage, ?Multiplier := AttackDamageMultiplier))
            
    UpdateStats<override>():void =
        (super:)UpdateStats()
        if (WeaponInfo := GetWeaponInfo[Id]):
            set ProjectileCount = 1 + Floor[Player.GetWeaponEvolutionValue(Id, ability_id.ProjectileCount)] or 0
            set AttackDamage *= (IsUniqueEvolution? and 2.0 or 1.0)
            set AttackRadius = WeaponInfo.Size * (1.0 + (Player.GetWeaponStats(Id, stat_id.Size) + Player.GetStat(stat_id.Size)) * (IsUniqueEvolution? and 0.5 or 1.0))
            set AttackAngle = Player.GetWeaponEvolutionValue(Id, ability_id.SwordAngle)