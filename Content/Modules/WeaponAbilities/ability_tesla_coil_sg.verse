using { /Fortnite.com/Characters }
using { /Verse.org/Random }
using { /Verse.org/SceneGraph }
using { /Verse.org/Simulation }
using { /UnrealEngine.com/Temporary/SpatialMath }
using { Prefabs }
using { VisualEffects }

<# ----------------------------------------------------------------------------------------------------
    Constants
---------------------------------------------------------------------------------------------------- #>
DefaultAbilityTeslaCoilDuration:float = 5.0
DefaultAbilityTeslaCoilAttackInterval:float = 1.0

<# ----------------------------------------------------------------------------------------------------
    Data Definitions
---------------------------------------------------------------------------------------------------- #>
projectile_tesla_coil_properties<public> := struct:
    Player<public>:player
    Damage<public>:float
    DamageMultiplier<public>:float = 1.0
    Radius<public>:float
    FireRate<public>:float
    MeshIndex<public>:int = 1
    IsUniqueEvolution<public>:logic
    MaybeTetherEntity<public>:?entity = false

<# ----------------------------------------------------------------------------------------------------
    ability_tesla_coil_sg
---------------------------------------------------------------------------------------------------- #>
ability_tesla_coil_sg<public> := class(ability):
    Id<override>:weapon_id = weapon_id.TeslaCoil

    var FireRate:float = 1.0
    var MeshIndex:int = 1
    var ObjectPool:[]projectile_tesla_coil_sg = array{}

    <# ----------------------------------------------------------------------------------------------------
        Initialization
    ---------------------------------------------------------------------------------------------------- #>
    Deactivate<override>():void =
        DeactivatedEvent.Signal()
        set CooldownTimeRemaining = Max(0.0, AttackCooldownTime - (GetSimulationElapsedTime() - CastTime))
        set IsActive = false
        
    Deinitialize<override>():void =
        (super:)Deinitialize()
        
        for (Object : ObjectPool):
            Object.Deactivate(?IsDispose := true)

    <# ----------------------------------------------------------------------------------------------------
        Get Functions
    ---------------------------------------------------------------------------------------------------- #>
    GetAvailablePoolObject<private>()<transacts>:projectile_tesla_coil_sg =
        for (Object : ObjectPool, Object.IsAvailable?):
            return Object

        NewObject := projectile_tesla_coil_sg{Ability := Self}
        set ObjectPool += array{NewObject}
        NewObject

    <# ----------------------------------------------------------------------------------------------------
        Update Functions
    ---------------------------------------------------------------------------------------------------- #>
    Ability<override>()<suspends>:void =
        ProjectileProperties := projectile_tesla_coil_properties:
            Player := Player
            Damage := AttackDamage
            DamageMultiplier := AttackDamageMultiplier
            Radius := AttackRadius
            FireRate := FireRate
            MeshIndex := MeshIndex
            IsUniqueEvolution := IsUniqueEvolution
        
        race:
            DeactivatedEvent.Await()
            if (WeaponInfo := GetWeaponInfo[Id]):
                Sleep(CooldownTimeRemaining)
                loop:
                    set CastTime = GetSimulationElapsedTime()
                    spawn{GetAvailablePoolObject().Activate(ProjectileProperties)}
                    Sleep(AttackCooldownTime)

    UpdateStats<override>():void =
        (super:)UpdateStats()
        if (WeaponInfo := GetWeaponInfo[Id]):
            set ProjectileCount = 1 + Floor[Player.GetWeaponEvolutionValue(Id, ability_id.ProjectileCount)] or 0
            set FireRate = DefaultAbilityTeslaCoilAttackInterval / (1.0 + Player.GetWeaponEvolutionValue(Id, ability_id.FireRate))
            set AttackDamage *= (ProjectileCount * 1.0)
            set MeshIndex = (IsUniqueEvolution? and 3 or ProjectileCount - 1)

<# ----------------------------------------------------------------------------------------------------
    projectile_tesla_coil_sg
---------------------------------------------------------------------------------------------------- #>
projectile_tesla_coil_sg<public> := class(poolable):
    Ability<public>:ability

    DeactivatedEvent:event(void) = event(void){}

    var MaybeWeaponPrefab:?prefab_weapon_tesla_coil = false

    Activate(Properties:projectile_tesla_coil_properties)<suspends>:void =
        (poolable:)Activate()

        if (not MaybeWeaponPrefab?):
            WeaponPrefab := prefab_weapon_tesla_coil{}
            WeaponPrefab.InitializeParticle()
            set MaybeWeaponPrefab = option{WeaponPrefab}

        if:
            PlayerCharacter := Properties.Player.GetFortCharacter[]
            WeaponPrefab := MaybeWeaponPrefab?
            ParticleSystem := NS_AbilityTeslaCoil_v2[WeaponPrefab.GetComponent[particle_system_component]]
        then:
            Rotation := Properties.Player.GetAutoAimRotation()
            Direction := Rotation.GetLocalForward()
            PositionOffset := Direction * GetRandomFloat(400.0, 600.0)

            var CurrentPosition:vector3 = PlayerCharacter.GetTransform().Translation.Make2D() + PositionOffset

            set ParticleSystem.Color = GetWeaponColor(Properties.Player.GetWeaponEvolutionLevel(Ability.Id))
            set ParticleSystem.Duration = DefaultAbilitySentryDuration
            set ParticleSystem.Radius = Properties.Radius * 0.1
            set ParticleSystem.MeshIndex = Properties.MeshIndex
            set ParticleSystem.FireRate = Properties.FireRate
            set ParticleSystem.Position = FromVector3 of CurrentPosition

            ParticleSystem.Enable()
            Properties.Player.PlayAudio(audio_id.Projectile)
            
            race:
                DeactivatedEvent.Await()
                Sleep(DefaultAbilityTeslaCoilDuration)
                
                if (Properties.IsUniqueEvolution?):
                    loop:
                        Sleep(ThreeTicks)
                        set CurrentPosition = PlayerCharacter.GetTransform().Translation.Make2D() + PositionOffset
                        set ParticleSystem.Position = FromVector3 of CurrentPosition

                else:
                    Sleep(Inf)

                loop:
                    for:
                        OverlapHit : WeaponPrefab.FindOverlapHits(FromTransform of transform{Translation := CurrentPosition}, collision_sphere{Radius := Properties.Radius})
                        TargetComponent := OverlapHit.TargetComponent
                        TargetEntity := TargetComponent.Entity
                        NPCCollisionComponent := TargetEntity.GetComponent[npc_collision_component]
                        NPCHit := NPCCollisionComponent.MaybeNPC?
                        NPCHit.IsDamageable()?
                    do:
                        Ability.Damage(NPCHit, Properties.Player.GetDamageArgs(NPCHit, Ability.Id, Properties.Damage, ?Multiplier := Properties.DamageMultiplier))
                        
                    Sleep(Properties.FireRate)

            Deactivate()

    Deactivate<override>(?IsDispose:logic = false):void =
        if:
            WeaponPrefab := MaybeWeaponPrefab?
            ParticleComponent := WeaponPrefab.GetComponent[particle_system_component]
        then:
            ParticleComponent.Disable()

            if (IsDispose?):
                WeaponPrefab.RemoveFromParent()
                set MaybeWeaponPrefab = false

        DeactivatedEvent.Signal()
        (super:)Deactivate()