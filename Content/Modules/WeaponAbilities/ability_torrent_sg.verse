using { /Fortnite.com/Characters }
using { /Verse.org/SceneGraph }
using { /Verse.org/Simulation }
using { /UnrealEngine.com/Temporary/SpatialMath }
using { Prefabs }

<# ----------------------------------------------------------------------------------------------------
    Constants
---------------------------------------------------------------------------------------------------- #>
BaseAbilityTorrentRange:float = 800.0
BaseAbilityTorrentDuration:float = 5.5
BaseAbilityTorrentDamageInterval:float = 1.0

<# ----------------------------------------------------------------------------------------------------
    ability_torrent_sg
---------------------------------------------------------------------------------------------------- #>
ability_torrent_sg<public> := class(ability):
    Id<override>:weapon_id = weapon_id.Torrent

    var MaybeWeaponPrefab:?prefab_weapon_torrent = false
    var DamageInterval:float = BaseAbilityTorrentDamageInterval

    <# ----------------------------------------------------------------------------------------------------
        Initialization
    ---------------------------------------------------------------------------------------------------- #>
    Deactivate<override>():void =
        DeactivatedEvent.Signal()
        set CooldownTimeRemaining = Max(0.0, AttackCooldownTime - (GetSimulationElapsedTime() - CastTime))
        set IsActive = false

    Deinitialize<override>():void =
        (super:)Deinitialize()
        
        if (WeaponPrefab := MaybeWeaponPrefab?):
            WeaponPrefab.RemoveFromParent()

        set MaybeWeaponPrefab = false

    <# ----------------------------------------------------------------------------------------------------
        Update Functions
    ---------------------------------------------------------------------------------------------------- #>
    Ability<override>()<suspends>:void =
        if (not MaybeWeaponPrefab?):
            WeaponPrefab := prefab_weapon_torrent{}
            WeaponPrefab.InitializeParticle()
            set MaybeWeaponPrefab = option{WeaponPrefab}

        race:
            DeactivatedEvent.Await()
            sync:
                block:
                    Sleep(CooldownTimeRemaining)
                    loop:
                        set CastTime = GetSimulationElapsedTime()
                        spawn{Cast()}
                        Sleep(AttackCooldownTime)

                if:
                    IsUniqueEvolution?
                    WeaponPrefab := MaybeWeaponPrefab?
                    AbilityTorrent := VisualEffects.NS_AbilityTorrent_v2[WeaponPrefab.GetComponent[particle_system_component]]
                    Character := Player.GetFortCharacter[]
                then:
                    loop:
                        set AbilityTorrent.TargetPosition = FromVector3(Character.GetTransform().Translation.Make2D())
                        Sleep(ThreeTicks)

    Cast()<suspends>:void =
        if (Character := Player.GetFortCharacter[]):
            for (Index := 1..ProjectileCount):
                Player.PlayAudio(audio_id.Projectile)
                TargetPosition := GetRandomPositionInCircle(BaseAbilityTorrentRange, ?Origin := Character.GetTransform().Translation.Make2D())
                spawn{Area(TargetPosition)}
                Sleep(0.1)

    Area(Position:vector3)<suspends>:void =
        if:
            Character := Player.GetFortCharacter[]
            WeaponPrefab := MaybeWeaponPrefab?
            AbilityTorrent := VisualEffects.NS_AbilityTorrent_v2[WeaponPrefab.GetComponent[particle_system_component]]
            SimulationEntity := GetSimulationEntity[]
        then:
            Collision := collision_sphere{Radius := AttackRadius}

            var CurrentPosition:vector3 = Position

            set AbilityTorrent.Color = GetWeaponColor(Player.GetWeaponEvolutionLevel(Id))
            set AbilityTorrent.Radius = AttackRadius * 0.1
            set AbilityTorrent.Duration = BaseAbilityTorrentDuration
            set AbilityTorrent.Speed = IsUniqueEvolution? and 150.0 or 0.0
            set AbilityTorrent.FireRate = DamageInterval
            set AbilityTorrent.Position = FromVector3 of Position

            AbilityTorrent.Play()
            
            race:
                Sleep(BaseAbilityTorrentDuration)

                sync:
                    loop:
                        Sleep(DamageInterval)
                        for:
                            OverlapHit : SimulationEntity.FindOverlapHits(FromTransform of transform{Translation := CurrentPosition}, Collision)
                            TargetComponent := OverlapHit.TargetComponent
                            TargetEntity := TargetComponent.Entity
                            NPCCollisionComponent := TargetEntity.GetComponent[npc_collision_component]
                            NPCHit := NPCCollisionComponent.MaybeNPC?
                            NPCHit.IsDamageable()?
                        do:
                            Damage(NPCHit, Player.GetDamageArgs(NPCHit, Id, AttackDamage, ?Multiplier := AttackDamageMultiplier), ?IsKnockback := false)

                    if (IsUniqueEvolution?):
                        loop:
                            Sleep(ThreeTicks)
                            if (Direction := (Character.GetTransform().Translation.Make2D() - CurrentPosition).MakeUnitVector[]):
                                set CurrentPosition = CurrentPosition + Direction * 150.0 * ThreeTicks

    UpdateStats<override>():void =
        (super:)UpdateStats()
        if (WeaponInfo := GetWeaponInfo[Id]):
            set ProjectileCount = 1 + Floor[Player.GetWeaponEvolutionValue(Id, ability_id.ProjectileCount)] or 0
            set DamageInterval = BaseAbilityTorrentDamageInterval / (1.0 + Player.GetWeaponEvolutionValue(Id, ability_id.TorrentFireRate))