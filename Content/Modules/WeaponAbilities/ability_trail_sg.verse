using { /Fortnite.com/Characters }
using { /Verse.org/SceneGraph }
using { /Verse.org/SceneGraph/KeyframedMovement }
using { /Verse.org/Simulation }
using { /UnrealEngine.com/Temporary/SpatialMath }
using { Prefabs }
using { VisualEffects }

<# ----------------------------------------------------------------------------------------------------
    Constants
---------------------------------------------------------------------------------------------------- #>
BaseAbilityTrailDamageInterval:float = 1.0
BaseAbilityTrailTravelDistanceSquared:float = 10000.0

<# ----------------------------------------------------------------------------------------------------
    ability_trail_sg
---------------------------------------------------------------------------------------------------- #>
ability_trail_sg<public> := class(ability):
    Id<override>:weapon_id = weapon_id.Trail
    
    CancelledEvent:event() = event(){}

    var AttackDuration:float = 1.0

    var TrailChargedChance:float = 0.0
    var TrailChargedPower:float = 0.0
    var SuperChargeInterval:int = 0
    var IsEvolved:logic = false

    var PreviousPosition:vector3 = vector3{}
    var DistanceTraveledSquared:float = 0.0
    var CastCount:int = 0

    var MaybeWeaponPrefab:?prefab_weapon_trail = false
    var MaybeBlastPrefab:?prefab_weapon_trail_blast = false

    <# ----------------------------------------------------------------------------------------------------
        Initialization
    ---------------------------------------------------------------------------------------------------- #>
    Deactivate<override>():void =
        DeactivatedEvent.Signal()
        set IsActive = false

    Deinitialize<override>():void =
        (super:)Deinitialize()
        
        CancelledEvent.Signal()

        if (WeaponPrefab := MaybeWeaponPrefab?):
            WeaponPrefab.RemoveFromParent()
            set MaybeWeaponPrefab = false

        if (BlastPrefab := MaybeBlastPrefab?):
            BlastPrefab.RemoveFromParent()
            set MaybeBlastPrefab = false

    <# ----------------------------------------------------------------------------------------------------
        Update Functions
    ---------------------------------------------------------------------------------------------------- #>
    Ability<override>()<suspends>:void =
        set CastCount = 0

        if (not MaybeWeaponPrefab?):
            WeaponPrefab := prefab_weapon_trail{}
            WeaponPrefab.InitializeParticle()
            set MaybeWeaponPrefab = option{WeaponPrefab}

        if (IsUniqueEvolution?):
            if (not MaybeBlastPrefab?):
                BlastPrefab := prefab_weapon_trail_blast{}
                BlastPrefab.InitializeParticle()
                set MaybeBlastPrefab = option{BlastPrefab}

            if (ParticleSystem := NS_AbilityTrailBlast_v2[MaybeBlastPrefab?.GetComponent[particle_system_component]]):
                set ParticleSystem.Color = GetWeaponColor(Player.GetWeaponEvolutionLevel(Id))
                set ParticleSystem.Radius = (AttackRadius * 1.5 * 1.5) * 0.1

        if:
            WeaponPrefab := MaybeWeaponPrefab?
            KeyframedMovementComponent := WeaponPrefab.GetComponent[keyframed_movement_component]
            ParticleSystem := NS_AbilityTrail[WeaponPrefab.GetComponent[particle_system_component]]
        then:
            set ParticleSystem.BlastInterval = 250.0
            set ParticleSystem.IsUniqueEvolution = IsUniqueEvolution
            set ParticleSystem.Color = GetWeaponColor(Player.GetWeaponEvolutionLevel(Id))
            ParticleSystem.Play()

            race:
                DeactivatedEvent.Await()
                loop:
                    Sleep(0.1)
                    IsSuperCharged := logic{IsEvolved? and Mod[CastCount, SuperChargeInterval] = 0}
                    FinalAttackRadius := AttackRadius * (IsSuperCharged? and 1.5 or 1.0)

                    if (Character := Player.GetFortCharacter[], VFXSize := (Round[FinalAttackRadius * 0.1] * 1.0) + (IsSuperCharged? and 0.001 or 0.0)):
                        set DistanceTraveledSquared += DistanceSquaredXY(Character.GetTransform().Translation, PreviousPosition)
                        set PreviousPosition = Character.GetTransform().Translation
                        
                        RelativeKeyframes := array:
                            keyframed_movement_delta:
                                Transform := FromTransform of transform:
                                    Translation :=  Character.GetTransform().Translation.Make2D() - FromVector3(WeaponPrefab.GetGlobalTransform().Translation)
                                    Scale := vector3{X := VFXSize, Y := VFXSize, Z := AttackDuration} - FromScalarVector3(WeaponPrefab.GetGlobalTransform().Scale)
                                Duration := 0.1
                                Easing := linear_easing_function{}

                        KeyframedMovementComponent.SetKeyframes(RelativeKeyframes, oneshot_keyframed_movement_playback_mode{})
                        KeyframedMovementComponent.Play()

                    if (DistanceTraveledSquared >= BaseAbilityTrailTravelDistanceSquared):
                        spawn{Cast(IsSuperCharged, FinalAttackRadius)}
                        set DistanceTraveledSquared = 0.0
                        set CastCount += 1

            ParticleSystem.Stop()

    Cast(IsSuperCharged:logic, FinalAttackRadius:float)<suspends>:void =
        if (Character := Player.GetFortCharacter[], SimulationEntity := GetSimulationEntity[]):
            FinalAttackDamage := AttackDamage * (IsSuperCharged? and (1.0 + TrailChargedPower) or 1.0)
            AbilityPosition := Character.GetTransform().Translation
            Collision := collision_sphere{Radius := FinalAttackRadius}

            race:
                CancelledEvent.Await()
                Sleep(AttackDuration)

                loop:
                    for:
                        OverlapHit : SimulationEntity.FindOverlapHits(FromTransform of transform{Translation := AbilityPosition}, Collision)
                        TargetComponent := OverlapHit.TargetComponent
                        TargetEntity := TargetComponent.Entity
                        NPCCollisionComponent := TargetEntity.GetComponent[npc_collision_component]
                        NPCHit := NPCCollisionComponent.MaybeNPC?
                        NPCHit.IsDamageable()?
                    do:
                        Damage(NPCHit, Player.GetDamageArgs(NPCHit, Id, FinalAttackDamage, ?Multiplier := AttackDamageMultiplier), ?IsKnockback := false)

                    Sleep(BaseAbilityTrailDamageInterval)
            
            if:
                IsUniqueEvolution? and IsSuperCharged?
                BlastPrefab := MaybeBlastPrefab?
                ParticleSystem := NS_AbilityTrailBlast_v2[BlastPrefab.GetComponent[particle_system_component]]
            then:
                BlastAttackDamage := FinalAttackDamage * 1.5
                BlastAttackRadius := FinalAttackRadius * 1.5

                set ParticleSystem.Position = FromVector3 of AbilityPosition.Make2D()
                set Collision.Radius = BlastAttackRadius

                ParticleSystem.Play()

                for:
                    OverlapHit : GetSimulationEntity[].FindOverlapHits(FromTransform of transform{Translation := AbilityPosition}, Collision)
                    TargetComponent := OverlapHit.TargetComponent
                    TargetEntity := TargetComponent.Entity
                    NPCCollisionComponent := TargetEntity.GetComponent[npc_collision_component]
                    NPCHit := NPCCollisionComponent.MaybeNPC?
                    NPCHit.IsDamageable()?
                    NPCPosition := NPCHit.GetAgent[].GetFortCharacter[].GetTransform().Translation
                    Direction := (NPCPosition.Make2D() - AbilityPosition.Make2D()).MakeUnitVector[]
                do:
                    DirectionRotation := MakeRotation(vector3{X := 0.0, Y := 0.0, Z := 1.0}, ArcTan(Direction.Y, Direction.X))
                    NPCHit.Knockback(knockback_properties{Duration := 1.0, Rotation := DirectionRotation})

                    Damage(NPCHit, Player.GetDamageArgs(NPCHit, Id, BlastAttackDamage, ?Multiplier := AttackDamageMultiplier), ?IsKnockback := false)

    UpdateStats<override>():void =
        (super:)UpdateStats()
        if (WeaponInfo := GetWeaponInfo[Id]):
            set IsEvolved = logic{Player.GetWeaponEvolutionLevel(Id) > 0}
            set TrailChargedChance = Player.GetWeaponEvolutionValue(Id, ability_id.TrailChargedChance)
            set TrailChargedPower = Player.GetWeaponEvolutionValue(Id, ability_id.TrailChargedPower)
            set SuperChargeInterval = Ceil[10.0 / (TrailChargedChance * 10.0)] or 0

            ActualAttackDuration := WeaponInfo.Cooldown + (1.0 + Player.GetWeaponStats(Id, stat_id.Cooldown) + Player.GetStat(stat_id.Cooldown)) / WeaponInfo.Cooldown * 2
            set AttackDuration = Min(WeaponInfo.MinCooldown, ActualAttackDuration)
            DurationDamageMultiplier := ActualAttackDuration / AttackDuration

            set AttackDamage = WeaponInfo.Damage * DurationDamageMultiplier