# This file defines a date_and_time, a persistable representation of a date and time in the Common Era, i.e. after 1/1/1 00:00:00.
# It contains functions for subtracting two date_and_time to get a time_interval.
# It also contains functions for adding or subtracting a time_interval from a date_and_time.

using { /Verse.org/Simulation }
using { Utility }

# A persistable date and time.
date_and_time<public> := class<final><persistable>:
    @editable
    Years<public>:int = -1
    @editable
    Months<public>:int = -1
    @editable
    Days<public>:int = -1

    @editable
    Hours<public>:int = -1
    @editable
    Minutes<public>:int = -1
    @editable
    Seconds<public>:float = -1.0

    # Succeeds if the date and time is valid.
    IsValid<public>()<decides><transacts>:void=
        Years > 0
        Months > 0
        Months <= 12
        Days > 0
        Days <= GetDaysFromMonth(Months)
        Hours >= 0
        Hours < 24
        Minutes >= 0
        Minutes < 60
        Seconds >= 0.0
        Seconds < 60.0

    # Succeeds if the year is a leap year.
    IsLeapYear<public>()<decides><transacts>:void=
        Mod[Years, 4] = 0 and (Mod[Years, 100] <> 0 or Mod[Years, 400] = 0)
    
    GetDaysFromMonth<private>(Month:int)<transacts>:int=
        if:
            Month = 2
        then:
            return if (IsLeapYear[]) then 29 else 28

        if:
            Month = 1 or Month = 3 or Month = 5 or Month = 7 or Month = 8 or Month = 10 or Month = 12
        then:
            return 31

        return 30

    GetSecondsFromMonth<private>(Month:int)<transacts>:int=
        GetDaysFromMonth(Month) * 86400

    GetSecondsFromMonths<private>()<transacts>:int=
        var Result:int = 0
        for(CurrentMonth := 1..Months-1):
            set Result += GetSecondsFromMonth(CurrentMonth)

        return Result

    GetSecondsFromYears<private>()<transacts>:int=
        var Result:int = (Years - 1) * 365 * 86400
        if:
            YearsDiv4 := Floor((Years - 1) / 4)
            YearsDiv100 := Floor((Years - 1) / 100)
            YearsDiv400 := Floor((Years - 1) / 400)
            LeapYears := YearsDiv4 - YearsDiv100 + YearsDiv400
            set Result += LeapYears * 86400

        return Result

    # Returns the amount of seconds that have passed since the beginning of the Common Era (1/1/1 00:00:00).
    GetTotalSeconds<public>()<transacts>:float=
        return Seconds + Minutes * 60.0 + Hours * 3600.0 + (Days - 1) * 86400.0 + 1.0 * GetSecondsFromMonths() + 1.0 * GetSecondsFromYears()

    # Returns the time interval obtained by subtracting DAT from this date and time.
    Subtract<public>(DAT:date_and_time)<transacts>:time_interval=
        if:
            IsValid[]
            DAT.IsValid[]
        then:
            TotalSecondsThis := GetTotalSeconds()
            TotalSecondsThat := DAT.GetTotalSeconds()

            return MakeTimeInterval(TotalSecondsThis - TotalSecondsThat)

        return time_interval{}

    # Returns the date and time obtained by adding the time interval to this date and time.
    Add<public>(TI:time_interval)<transacts>:date_and_time=
        return Add(TI.GetTotalSeconds())

    # Returns the date and time obtained by adding SecondsToAdd seconds to this date and time.
    Add<public>(SecondsToAdd:float)<transacts>:date_and_time=
        if:
            IsValid[]
        then:
            return MakeDateAndTime(GetTotalSeconds() + SecondsToAdd)

        return date_and_time{}

    # Returns the date and time obtained by subtracting the time interval from this date and time.
    Subtract<public>(TI:time_interval)<transacts>:date_and_time=
        return Subtract(TI.GetTotalSeconds())

    # Returns the date and time obtained by subtracting SecondsToSubtract seconds from this date and time.
    Subtract<public>(SecondsToSubtract:float)<transacts>:date_and_time=
        if:
            IsValid[]
        then:
            return MakeDateAndTime(GetTotalSeconds() - SecondsToSubtract)

        return date_and_time{}

# Returns the given date and time as a nicely formatted string. 
ToString<public>(DAT:date_and_time):string=
    if:
        WholeSeconds := Floor[DAT.Seconds]
    then:
        return "{DAT.Years}-{PadNumber(DAT.Months, 2, '0')}-{PadNumber(DAT.Days, 2, '0')} {PadNumber(DAT.Hours, 2, '0')}:{PadNumber(DAT.Minutes, 2, '0')}:{PadNumber(WholeSeconds, 2, '0')}"
    else:
        return ""

# Create a date_and_time given years, months, days, hours, minutes and seconds. 
MakeDateAndTime<public>(Years:int, Months:int, Days:int, Hours:int, Minutes:int, Seconds:float)<transacts>:date_and_time=
    return date_and_time:
        Years := Years
        Months := Months
        Days := Days
        Hours := Hours
        Minutes := Minutes
        Seconds := Seconds

# Create a date_and_time given the amount of seconds that have passed since the beginning of the Common Era (1/1/1 00:00:00).       
MakeDateAndTime<public>(TotalSeconds:float)<transacts>:date_and_time=
    if:
        TotalWholeSeconds := Floor[TotalSeconds]
        FractionSeconds := TotalSeconds - 1.0 * TotalWholeSeconds 
        Seconds := 1.0 * Mod[TotalWholeSeconds, 60] + FractionSeconds
        TotalMinutes := Floor(TotalWholeSeconds / 60)
        Minutes := Mod[TotalMinutes, 60]
        TotalHours := Floor(TotalMinutes / 60)
        Hours := Mod[TotalHours, 24]
        YearsMonthsAndDays := GetYearsMonthsAndDaysFromTotalSeconds(TotalWholeSeconds)
    then:
        return MakeDateAndTime(YearsMonthsAndDays(0), YearsMonthsAndDays(1), YearsMonthsAndDays(2), Hours, Minutes, Seconds)

    return date_and_time{}

IsLeapYear(Year:int)<decides><transacts>:void=
    Mod[Year, 4] = 0 and (Mod[Year, 100] <> 0 or Mod[Year, 400] = 0)

IsDateRolledOver<public>(DateTimeA:date_and_time, DateTimeB:date_and_time)<decides><transacts>:void =
    (DateTimeA.Years <> DateTimeB.Years) or (DateTimeA.Months <> DateTimeB.Months) or (DateTimeA.Days <> DateTimeB.Days)

GetSecondsFromYear(Year:int)<transacts>:int=
    return if (IsLeapYear[Year]) then 366 * 86400 else 365 * 86400

GetSecondsFromMonth(Month:int, Year:int)<transacts>:int=
    if:
        Month = 2
    then:
        return if (IsLeapYear[Year]) then 29 * 86400 else 28 * 86400

    if:
        Month = 1 or Month = 3 or Month = 5 or Month = 7 or Month = 8 or Month = 10 or Month = 12
    then:
        return 31 * 86400

    return 30 * 86400

GetYearsMonthsAndDaysFromTotalSeconds(TotalSeconds:int)<transacts>:tuple(int,int,int)=
    var RemainingDays:int = 0
    
    if:
        Days := Floor(TotalSeconds / 86400)
        set RemainingDays = Days
        Years400 := Floor(RemainingDays / 146097)
        set RemainingDays -= Years400 * 146097
        Years100 := Floor(RemainingDays / 36524)
        set RemainingDays -= Years100 * 36524
        Years4 := Floor(RemainingDays / 1461)
        set RemainingDays -= Years4 * 1461
        Years1 := Floor(RemainingDays / 365)
        set RemainingDays -= Years1 * 365
    then:
        if:
            # Special case for the last day in a 400 year cycle.
            Years100 = 4
            Year := Years400 * 400 + Years100 * 100
        then:
            return (Year, 12, 31)
        else if:
            # Special case for the last day in a 4 year cycle.
            Years1 = 4
            Year := Years400 * 400 + Years100 * 100 + Years4 * 4 + Years1
        then:
            return (Year, 12, 31)
        else:
            Year := Years400 * 400 + Years100 * 100 + Years4 * 4 + Years1 + 1
            var Month:int = 1
            loop:
                if:
                    DaysForCurrentMonth := Floor(GetSecondsFromMonth(Month, Year) / 86400)
                    DaysForCurrentMonth <= RemainingDays
                then:
                    set Month += 1
                    set RemainingDays -= DaysForCurrentMonth
                else:
                    break
            return (Year, Month, RemainingDays + 1)

    return (0, 0, 0)