<# ----------------------------------------------------------------------------------------------------
    Data Definitions
---------------------------------------------------------------------------------------------------- #>
npc_state_type<public> := enum:
    Chase
    Attack
    Cooldown
    Focus
    Flee
    Reposition
    Dance
    Hit
    Knockback
    Idle
    Pause
    Eliminated

ToString<public>(NPCStateType:npc_state_type)<transacts>:string =
    case (NPCStateType):
        npc_state_type.Chase => "Chase"
        npc_state_type.Attack => "Attack"
        npc_state_type.Cooldown => "Cooldown"
        npc_state_type.Focus => "Focus"
        npc_state_type.Flee => "Flee"
        npc_state_type.Reposition => "Reposition"
        npc_state_type.Dance => "Dance"
        npc_state_type.Hit => "Hit"
        npc_state_type.Knockback => "Knockback"
        npc_state_type.Idle => "Idle"
        npc_state_type.Pause => "Pause"
        npc_state_type.Eliminated => "Eliminated"
        
ToString<public>(State:npc_state)<transacts>:string =
    if (npc_state_chase[State]) then "npc_state_chase"
    else if (npc_state_attack[State]) then "npc_state_attack"
    else if (npc_state_cooldown[State]) then "npc_state_cooldown"
    else if (npc_state_focus[State]) then "npc_state_focus"
    else if (npc_state_flee[State]) then "npc_state_flee"
    else if (npc_state_reposition[State]) then "npc_state_reposition"
    else if (npc_state_dance[State]) then "npc_state_dance"
    else if (npc_state_hit[State]) then "npc_state_hit"
    else if (npc_state_knockback[State]) then "npc_state_knockback"
    else if (npc_state_idle[State]) then "npc_state_idle"
    else if (npc_state_pause[State]) then "npc_state_pause"
    else if (npc_state_eliminated[State]) then "npc_state_eliminated"
    else ""

<# ----------------------------------------------------------------------------------------------------
    Data Factory
---------------------------------------------------------------------------------------------------- #>
npc_state_factory<public> := class<concrete>:
    ChaseState:npc_state_chase = npc_state_chase{}
    AttackState:npc_state_attack = npc_state_attack{}
    CooldownState:npc_state_cooldown = npc_state_cooldown{}
    FocusState:npc_state_focus = npc_state_focus{}
    FleeState:npc_state_flee = npc_state_flee{}
    RepositionState:npc_state_reposition = npc_state_reposition{}
    DanceState:npc_state_dance = npc_state_dance{}
    HitState:npc_state_hit = npc_state_hit{}
    KnockbackState:npc_state_knockback = npc_state_knockback{}
    IdleState:npc_state_idle = npc_state_idle{}
    PauseState:npc_state_pause = npc_state_pause{}
    EliminatedState:npc_state_eliminated = npc_state_eliminated{}

    (npc_state_factory:)GetNPCState<public>(Type:npc_state_type)<transacts>:npc_state =
        case (Type):
            npc_state_type.Chase => ChaseState
            npc_state_type.Attack => AttackState
            npc_state_type.Cooldown => CooldownState
            npc_state_type.Focus => FocusState
            npc_state_type.Flee => FleeState
            npc_state_type.Reposition => RepositionState
            npc_state_type.Dance => DanceState
            npc_state_type.Hit => HitState
            npc_state_type.Knockback => KnockbackState
            npc_state_type.Idle => IdleState
            npc_state_type.Pause => PauseState
            npc_state_type.Eliminated => EliminatedState

npc_state<public> := class<abstract>:
    Priority<public>:int = 0
    #IsInterruptible<public>:logic = true

    OnBegin<public>(NPCBehaviorBase:npc_behavior_base)<suspends>:npc_state
