using { /Fortnite.com/AI }
using { /Fortnite.com/Characters }

npc_state_attack<public> := class(npc_state):    
    OnBegin<override>(NPCBehaviorBase:npc_behavior_base)<suspends>:npc_state =
        race:
            return NPCBehaviorBase.InterruptedEvent.Await()
            if:
                Agent := NPCBehaviorBase.GetAgent[]
                Character := Agent.GetFortCharacter[]
                Navigatable := Character.GetNavigatable[]

                TargetPlayer := NPCBehaviorBase.MaybeTargetPlayer?
                TargetPlayerCharacter := TargetPlayer.GetFortCharacter[]
                TargetPlayerCharacter.IsActive[]
            then:
                Navigatable.SetMovementSpeedMultiplier(1.0)
                NPCAttack := GetNPCAttack(NPCBehaviorBase.AttackProperties.AttackType)
                NPCAttack.Execute(NPCBehaviorBase, NPCBehaviorBase.AttackProperties, TargetPlayer)

        return GetNPCState(npc_state_type.Cooldown)