using { /Fortnite.com/AI }
using { /Fortnite.com/Characters }
using { /UnrealEngine.com/Temporary/SpatialMath }
using { /Verse.org/Simulation }

npc_state_cooldown<public> := class(npc_state):
    var StartTime:float = 0.0

    OnBegin<override>(NPCBehaviorBase:npc_behavior_base)<suspends>:npc_state =
        if (CooldownStartTime := NPCBehaviorBase.MaybeCooldownStartTime?):
            set StartTime = CooldownStartTime
        else:
            set StartTime = GetSimulationElapsedTime()
            set NPCBehaviorBase.MaybeCooldownStartTime = option{GetSimulationElapsedTime()}

        if:
            Character := NPCBehaviorBase.GetAgent[].GetFortCharacter[]
            Navigatable := Character.GetNavigatable[]
            FocusInterface := Character.GetFocusInterface[]

            TargetPlayer := NPCBehaviorBase.MaybeTargetPlayer?
            TargetPlayer.IsActive[]
            PlayerCharacter := TargetPlayer.GetFortCharacter[]
            PlayerCharacter.IsActive[]
        then:
            NavigationTarget := MakeNavigationTarget(TargetPlayer)
            race:
                block:
                    InterruptState := NPCBehaviorBase.InterruptedEvent.Await()
                    set NPCBehaviorBase.MaybeCooldownStartTime = false
                    return InterruptState
                
                FocusInterface.MaintainFocus(TargetPlayer)

                block:
                    CooldownTimeRemaining := NPCBehaviorBase.AttackProperties.AttackCooldown - (GetSimulationElapsedTime() - StartTime)
                    if (CooldownTimeRemaining > 0.0):
                        Sleep(CooldownTimeRemaining)

                loop:
                    Navigatable.NavigateTo(NavigationTarget, ?MovementType := movement_types.Running, ?ReachRadius := NPCBehaviorBase.AttackProperties.AttackDistance)
                    Navigatable.Wait(?Duration := FiveTicks)

        set NPCBehaviorBase.MaybeCooldownStartTime = false
        return GetNPCState(npc_state_type.Chase)