using { /Fortnite.com/AI }
using { /Fortnite.com/Animation/PlayAnimation }
using { /Fortnite.com/Characters }
using { /Verse.org/Random }
using { /Verse.org/Simulation }
using { /UnrealEngine.com/Temporary/SpatialMath }

npc_state_eliminated<public> := class(npc_state):
    Priority<override>:int = 3

    OnBegin<override>(NPCBehaviorBase:npc_behavior_base)<suspends>:npc_state =
        if (Character := NPCBehaviorBase.GetAgent[].GetFortCharacter[], PlayAnimationController := Character.GetPlayAnimationController[], Navigatable := Character.GetNavigatable[]):
            Navigatable.StopNavigation()
            Character.PutInStasis of stasis_args:
                AllowTurning := false
                AllowFalling := false

            AnimationSequence := if (Character.TeleportTo[Character.GetTransform().Translation - vector3{Z := 300.0}, Character.GetTransform().Rotation]):
                NPC.AnimationSequences.AS_Slime_Eliminated
            else:
                NPC.AnimationSequences.AS_Slime_Eliminated_Grounded

            PlayAnimationController.PlayAndAwait(AnimationSequence, ?PlayRate := 0.5)
            Sleep(GetRandomFloat(0.1, 0.5))
            NPCBehaviorBase.Respawn()
        
        return GetNPCState(npc_state_type.Chase)