using { /Fortnite.com/AI }
using { /Fortnite.com/Characters }

npc_state_flee<public> := class(npc_state):
    OnBegin<override>(NPCBehaviorBase:npc_behavior_base)<suspends>:npc_state =
        race:
            return NPCBehaviorBase.InterruptedEvent.Await()
            if:
                Agent := NPCBehaviorBase.GetAgent[]
                Character := Agent.GetFortCharacter[]
                Navigatable := Character.GetNavigatable[]
            then:
                NavigationTarget := MakeNavigationTarget(GetRandomPositionInArena())
                Navigatable.SetMovementSpeedMultiplier(1.0 + GetArenaMovespeedMultiplier())
                Navigatable.NavigateTo(NavigationTarget, ?MovementType := movement_types.Running)

                return Self

        return Self