using { /Fortnite.com/Animation/PlayAnimation }
using { /Fortnite.com/Characters }

npc_state_hit<public> := class(npc_state):
    OnBegin<override>(NPCBehaviorBase:npc_behavior_base)<suspends>:npc_state =
        if:
            NPCCharacter := NPCBehaviorBase.GetAgent[].GetFortCharacter[]
            NPCCharacter.IsActive[]
            PlayAnimationController := NPCCharacter.GetPlayAnimationController[]
        then:
            race:
                return NPCBehaviorBase.InterruptedEvent.Await()
                PlayAnimationController.PlayAndAwait(NPC.AnimationSequences.AS_Slime_Hit)

        return GetNPCState(npc_state_type.Chase)