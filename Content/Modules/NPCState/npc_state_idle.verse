using { /Fortnite.com/AI }
using { /Fortnite.com/Characters }
using { /UnrealEngine.com/Temporary/SpatialMath }
using { /Verse.org/Simulation }

npc_state_idle<public> := class(npc_state):
    OnBegin<override>(NPCBehaviorBase:npc_behavior_base)<suspends>:npc_state =
        if (Agent := NPCBehaviorBase.GetAgent[], Character := Agent.GetFortCharacter[], Navigatable := Character.GetNavigatable[]):
            race:
                return NPCBehaviorBase.InterruptedEvent.Await()
                block:
                    Navigatable.StopNavigation()
                    Navigatable.Wait(?Duration := 3.0)
                    return GetNPCState(npc_state_type.Chase)
        
        return GetNPCState(npc_state_type.Chase)