using { /Fortnite.com/AI }
using { /Fortnite.com/Characters }
using { /UnrealEngine.com/Temporary/SpatialMath }
using { /Verse.org/Simulation }

npc_state_knockback<public> := class(npc_state):
    Priority<override>:int = 1
    
    OnBegin<override>(NPCBehaviorBase:npc_behavior_base)<suspends>:npc_state =
        if (Agent := NPCBehaviorBase.GetAgent[], Character := Agent.GetFortCharacter[], Navigatable := Character.GetNavigatable[]):
            race:
                return NPCBehaviorBase.InterruptedEvent.Await()
                block:
                    Navigatable.StopNavigation()
                    Agent.ApplyImpulse(impulse_id.Knockback, ?Rotation := NPCBehaviorBase.KnockbackProperties.Rotation)
                    Navigatable.Wait(?Duration := NPCBehaviorBase.KnockbackProperties.Duration)
                    return GetNPCState(npc_state_type.Chase)
        
        return GetNPCState(npc_state_type.Chase)