using { /Fortnite.com/AI }
using { /Fortnite.com/Characters }
using { /UnrealEngine.com/Temporary/SpatialMath }
using { /Verse.org/Simulation }

npc_state_pause<public> := class(npc_state):
    Priority<override>:int = 2
    
    OnBegin<override>(NPCBehaviorBase:npc_behavior_base)<suspends>:npc_state =
        if (Agent := NPCBehaviorBase.GetAgent[], Character := Agent.GetFortCharacter[], Navigatable := Character.GetNavigatable[]):
            race:
                return NPCBehaviorBase.InterruptedEvent.Await()
                block:
                    Navigatable.StopNavigation()
                    Navigatable.Wait(?Duration := Inf)
                    return GetNPCState(npc_state_type.Chase)
        
        return GetNPCState(npc_state_type.Chase)